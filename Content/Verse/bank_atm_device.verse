using { /Verse.org/Simulation }
using { /Verse.org/Colors }
using { /Verse.org/Concurrency }
using { /Verse.org/Verse }
using { /Verse.org/Random }
using { /Fortnite.com/Devices }
using { /Fortnite.com/UI }
using { /Fortnite.com/Characters }
using { /UnrealEngine.com/Temporary/Diagnostics }
using { /UnrealEngine.com/Temporary/SpatialMath }
using { /UnrealEngine.com/Temporary/UI}

bank_atm_device := class(creative_device) {
    @editable
    OpenBankButtonList: []button_device = array {}

    @editable
    MinMoneyRewardAmountScamLevelList: []int = array {}

    @editable
    MaxMoneyRewardAmountScamLevelList: []int = array {}

    @editable
    MinFansRewardAmountScamLevelList: []int = array {}

    @editable
    MaxFansRewardAmountScamLevelList: []int = array {}

    @editable
    XpRewardScamLevelList: []int = array {}

    @editable
    HudOnSuccess: hud_message_device = hud_message_device {}

    @editable
    HudHadNoItem: hud_message_device = hud_message_device {}

    # Constants

    # Runtime
    
    OnBegin<override>()<suspends>: void = {
        # Listen to events
        for(ItemIndex -> Button: OpenBankButtonList) {
            Button.InteractedWithEvent.Subscribe(OnOpenAtmButton)
            <#
            EventHandler := event_handler_agent(int) {
                A := ItemIndex
                CallbackFunc := OnOpenBankButton
            }

            EventSub := Button.InteractedWithEvent.Subscribe(EventHandler.EventSubscription)
            #>
        }
    }

    OnOpenAtmButton(Agent: agent): void = {
        SaveManager := GetSaveManager()
        if {
            PlayerData := Agent.GetPlayerData[]
            PlayerData.CurCreditCardCoins > 0
        } then {
            PlayerData.RemoveCreditCardCoin(1)
            # Reward the player based on his scam rank
            if {
                # Rewards at index 0 are for TotalScamLevel = 1
                RewardScamLevel := Max(PlayerData.TotalScamLevel - 1, 0)
                MinMoneyRewardAmount := MinMoneyRewardAmountScamLevelList[RewardScamLevel]
                MaxMoneyRewardAmount := MaxMoneyRewardAmountScamLevelList[RewardScamLevel]
                MinFansRewardAmount := MinFansRewardAmountScamLevelList[RewardScamLevel]
                MaxFansRewardAmount := MaxFansRewardAmountScamLevelList[RewardScamLevel]
                XpReward := XpRewardScamLevelList[RewardScamLevel]
            } then {
                PlayerData.AddXp(XpReward, true)

                RandRewardAmountMoney := GetRandomInt(MinMoneyRewardAmount, MaxMoneyRewardAmount)
                PlayerData.AddCurrency(_Tycoon.tycoon_currency.Money, RandRewardAmountMoney, true)

                RandRewardAmountFans := GetRandomInt(MinFansRewardAmount, MaxFansRewardAmount)
                PlayerData.AddCurrency(_Tycoon.tycoon_currency.Fans, RandRewardAmountFans, true)

                HudOnSuccess.Show(Agent)

                # Check for achievement
                # Earn $ 25K from ATMs
                SaveManager.IncrementAchievement(Agent, _Global.achievement_type.Money, 1, RandRewardAmountMoney)
            } else {
                Print("ERR: OnOpenAtmButton: Found no config for TotalScamLevel: {PlayerData.TotalScamLevel}")
            }
        } else {
            HudHadNoItem.Show(Agent)
        }
    }
}
