using { /Verse.org/Simulation }
using { /Verse.org/Colors }
using { /Verse.org/Concurrency }
using { /Verse.org/Verse }
using { /Fortnite.com/Devices }
using { /Fortnite.com/UI }
using { /Fortnite.com/Characters }
using { /UnrealEngine.com/Temporary/Diagnostics }
using { /UnrealEngine.com/Temporary/SpatialMath }
using { /UnrealEngine.com/Temporary/UI}

code_input_device := class(creative_device) {
    @editable
    OpenCodeInputButton: switch_device = switch_device {}

    @editable
    AudioCorrectCode: audio_player_device = audio_player_device {}
    
    @editable
    AudioWrongCode: audio_player_device = audio_player_device {}

    @editable
    AudioButtonClick: audio_player_device = audio_player_device {}

    @editable
    Password: string = "1234"

    @editable
    CodeOneTimeReward: string = "99999999999999999"

    # Constants
    # Runtime
    var AgentUiMap: [agent]_UI.ui_code_input = map {}
    
    OnBegin<override>()<suspends>: void = {
        OpenCodeInputButton.TurnedOffEvent.Subscribe(OnOpenCodeInputButton)

        GetPlayspace().PlayerAddedEvent().Subscribe(OnPlayerAdded)
        GetPlayspace().PlayerRemovedEvent().Subscribe(OnPlayerRemoved)

        # Call logic for initial players
        for(Player : GetPlayspace().GetPlayers()) {
            OnPlayerAdded(Player)
        }
    }

    OnPlayerAdded(Agent: agent): void = {
        Ui := _UI.ui_code_input {}
        Ui.InitWidgetCodeInput(Agent, Self)
        if(set AgentUiMap[Agent] = Ui) {}
    }

    OnPlayerRemoved(Agent: agent): void = {
        # Delete the element from AgentUiMap to avoid memory leak
        if(Ui := AgentUiMap[Agent]) {
            Ui.RemoveUiFromPlayer(Agent)
        }
        set AgentUiMap = AgentUiMap.RemoveKey(Agent)
    }

    OnOpenCodeInputButton(Agent: agent): void = {
        OnShowUiButton(Agent)
    }

    OnShowUiButton(Agent: agent): void = {
        Print("OnShowUiButton")
        SaveManager := GetSaveManager()
        if(Ui := AgentUiMap[Agent]) {
            Ui.AddUiToPlayer(Agent, false)
        }
        # SaveManager.GrantAnalytic(Agent, SaveManager.ANALYTIC_CODE_INPUT_OPENED)
    }
}
