using { /Fortnite.com/Game }
using { /Fortnite.com/Devices }
using { /Fortnite.com/Devices/CreativeAnimation }
using { /Fortnite.com/Characters }
using { /Fortnite.com/Vehicles }
using { /Verse.org/Simulation }
using { /Verse.org/Simulation/Tags }
using { /Verse.org/Random }
using { /Verse.org/Concurrency }
using { /UnrealEngine.com/Temporary/Diagnostics }
using { /UnrealEngine.com/Temporary/SpatialMath }

custom_car_device := class(creative_device) {
    @editable
    CarSpawner<public>: vehicle_spawner_sedan_device = vehicle_spawner_sedan_device {}

    @editable
    CarSwitchList: []switch_device = array {}

    @editable
    CarVfxList: []visual_effect_powerup_device = array {}

    @editable
    CarVfxRemoverList: []visual_effect_powerup_device = array {}

    @editable
    CustomCarBlueprintList: []creative_prop_asset = array {}

    @editable
    CarScamMutatorZone: mutator_zone_device = mutator_zone_device {}

    @editable
    CarKillMutatorZone: mutator_zone_device = mutator_zone_device {}

    # Constants
    IsScamCar: []logic = array {
        false,
        false,
        false,
        false,
        true,
        true
    }

    # Runtime
    var MaybeOwnerAgent: ?agent = false
    var MaybeCarVehicle: ?fort_vehicle = false
    var MaybeSpawnedCarPlaceholderProp: ?creative_prop = false
    var RoutineDrivingCarTick: ?routine = false
    var RoutineScamCarTick: ?routine = false
    var RoutinePlaceholderPropTick: ?routine = false

    var ActiveSkinIndex: int = -1

    # Runs when the device is started in a running game
    OnBegin<override>()<suspends>:void = {
        # Listen to events
        # CarSwitchList
        for(ItemIndex -> CarSwitch: CarSwitchList) {
            EventHandler := event_handler_agent(int) {
                A := ItemIndex
                CallbackFunc := OnCarSwitch
            }

            EventSub := CarSwitch.TurnedOffEvent.Subscribe(EventHandler.EventSubscription)
        }

        # Car Spawner
        CarSpawner.SpawnedEvent.Subscribe(OnCarSpawned)
        CarSpawner.DestroyedEvent.Subscribe(OnCarDestroyed)
        CarSpawner.AgentEntersVehicleEvent.Subscribe(OnCarEntered)
        CarSpawner.AgentExitsVehicleEvent.Subscribe(OnCarExited)

        # Mutator Zone
        CarKillMutatorZone.AgentEntersEvent.Subscribe(OnCarKillMutatorZone)
        CarScamMutatorZone.AgentEntersEvent.Subscribe(OnCarScamMutatorZone)
    }

    AssignOwner(Agent: agent): void = {
        Print("Custom car: AssignOwner")
        set MaybeOwnerAgent = option{Agent}
    }

    OnTycoonReset(): void = {
        Print("Custom car: OnTycoonReset")
        CarSpawner.DestroyVehicle()
        # Don't call Disable, since we don't Enable it currently
        # CarSpawner.Disable()
        DestroyCarPlaceholderProp()
        set ActiveSkinIndex = -1
        set MaybeOwnerAgent = false
    }

    OnCarSwitch(Agent: agent, ItemIndex: int): void = {
        Print("OnCarSwitch")
        if {
            OwnerAgent := MaybeOwnerAgent?
        } then {
            Print("OnCarSwitch 222")
            # Destroy car placeholder prop (if there is one)
            DestroyCarPlaceholderProp()
            # Reenable previous car switch
            Task := spawn{ReenablePreviousCarSwitch(Agent)}
            # Set new active skin
            set ActiveSkinIndex = ItemIndex
            # CarSpawner is disabled by default, since it spawns car OnBegin
            CarSpawner.Enable()
            # Make the player drive the car
            CarSpawner.RespawnVehicle()
            Task2 := spawn{AssignDriverSuspends(option{Agent})}
        }
    }

    AssignDriverSuspends(MaybeAgent: ?agent)<suspends>: void = {
        Sleep(1.0)
        if {
            Agent := MaybeAgent?
        } then {
            CarSpawner.AssignDriver(Agent)
        }
    }

    OnCarSpawned(CarVehicle: fort_vehicle): void = {
        set MaybeCarVehicle = option{CarVehicle}
    }

    OnCarDestroyed(): void = {
        set MaybeCarVehicle = false
        # We don't need to RemoveCurrentCarVfx here, since Destroy also triggeres OnCarExited
    }

    OnCarEntered(Agent: agent): void = {
        if {
            OwnerAgent := MaybeOwnerAgent?
            Agent = OwnerAgent
            CarVfx := CarVfxList[ActiveSkinIndex]
        } then {
            # Destroy car placeholder prop (if there is one)
            DestroyCarPlaceholderProp()
            # Apply the current car skin
            CarVfx.Pickup(Agent)
            # Start driving tick
            DriveRoutine := SpawnRoutine(DrivingCarTick)
            set RoutineDrivingCarTick = option{DriveRoutine}
            # Start scam tick
            if {
                IsScamCar[ActiveSkinIndex] = true
            } then {
                ScamRoutine := SpawnRoutine(ScamCarTick)
                set RoutineScamCarTick = option{ScamRoutine}
            }
        }
    }

    OnCarExited<public>(Agent: agent): void = {
        Utility := GetUtility()
        if {
            OwnerAgent := MaybeOwnerAgent?
            Agent = OwnerAgent
        } then {
            # Remove the car vfx on the player
            RemoveCurrentCarVfx(Agent)
            # Spawn a car placeholder prop on the car position
            SpawnCarPlaceholderProp()
            # Stop routines
            if(Routine := RoutineDrivingCarTick?) {
                Routine.Cancel()
            }
            if(Routine := RoutineScamCarTick?) {
                Routine.Cancel()
            }
            
            # Hide kill mutator zone
            if(CarKillMutatorZone.TeleportTo[Utility.HiddenTrans]) {}
            # Hide scam mutator zone
            if(CarScamMutatorZone.TeleportTo[Utility.HiddenTrans]) {}
        }
    }

    RemoveCurrentCarVfx(Agent: agent): void = {
        if {
            ActiveSkinIndex >= 0
            PrevCarVfxRemover := CarVfxRemoverList[ActiveSkinIndex]
        } then {
            PrevCarVfxRemover.Pickup(Agent)
        }
    }

    SpawnCarPlaceholderProp(): void = {
        if {
            CarVehicle := MaybeCarVehicle?
            CarBlueprint := CustomCarBlueprintList[ActiveSkinIndex]
        } then {
            SpawnTuple := SpawnProp(CarBlueprint, CarVehicle.GetTransform().Translation, CarVehicle.GetTransform().Rotation)
            if(SpawnedProp := SpawnTuple(0)?) {
                set MaybeSpawnedCarPlaceholderProp = option{SpawnedProp}
            } else {
                Print("ERR: SpawnCarPlaceholderProp: Couldn't spawn a prop")
            }
            # Constantly move the prop to the car pos, in case the car is still rolling on exit
            MyRoutine := SpawnRoutine(PlaceholderPropTick)
            set RoutinePlaceholderPropTick = option{MyRoutine}
        }
    }

    DestroyCarPlaceholderProp(): void = {
        # Clean up routines
        if(Routine := RoutineDrivingCarTick?) {
            Routine.Cancel()
        }
        if(Routine := RoutineScamCarTick?) {
            Routine.Cancel()
        }
        
        if(Routine := RoutinePlaceholderPropTick?) {
            Routine.Cancel()
        }
        # Destroy the spawned prop
        if {
            SpawnedProp := MaybeSpawnedCarPlaceholderProp?
            SpawnedProp.IsValid[]
        } then {
            SpawnedProp.Dispose()
        }
        set MaybeSpawnedCarPlaceholderProp = false
    }

    ReenablePreviousCarSwitch(Agent: agent)<suspends>: void = {
        if {
            ActiveSkinIndex >= 0
            PrevCarSwitch := CarSwitchList[ActiveSkinIndex]
        } then {
            PrevCarSwitch.Enable()
            Sleep(1.0)
            PrevCarSwitch.TurnOn(Agent)
        }
    }

    OnCarKillMutatorZone(NpcAgent: agent): void = {
        if {
            # Only damage NPCs
            NpcData := NpcAgent.GetNpcData[]
            NpcFortChar := NpcAgent.GetFortCharacter[]
            NpcFortChar.IsActive[]
            # Get owner agent
            OwnerAgent := MaybeOwnerAgent?
            OwnerFortChar := OwnerAgent.GetFortCharacter[]
            OwnerFortChar.IsActive[]
        } then {
            NpcFortChar.Damage(damage_args{Instigator := option{OwnerAgent.GetInstigator()}, Amount := 100000000.0, Source := option{OwnerFortChar}})
        }
    }

    OnCarScamMutatorZone(NpcAgent: agent): void = {
        GameConfig := GetGameConfig()
        if {
            # Only scam NPCs
            NpcData := NpcAgent.GetNpcData[]
            NpcFortChar := NpcAgent.GetFortCharacter[]
            NpcFortChar.IsActive[]
            # Get owner agent
            OwnerAgent := MaybeOwnerAgent?
            ScamSpeedFactor := GameConfig.CarScamSpeedFactorList[ActiveSkinIndex]
        } then {
            NpcData.TryScammingNpc(OwnerAgent, ScamSpeedFactor)
        }
    }

    DrivingCarTick()<suspends>: void = {
        GameConfig := GetGameConfig()
        Utility := GetUtility()
        loop {
            if {
                CarVehicle := MaybeCarVehicle?
            } then {
                # Alternative method would be: Check distance to every NPC
                # GameConfig.CarNpcKillDistance
                if(CarKillMutatorZone.TeleportTo[CarVehicle.GetTransform().Translation + Utility.PlayerOffset, CarVehicle.GetTransform().Rotation]) {}
            }
            
            Sleep(0.0)
        }
    }

    ScamCarTick()<suspends>: void = {
        GameConfig := GetGameConfig()
        Utility := GetUtility()
        loop {
            if {
                CarVehicle := MaybeCarVehicle?
            } then {
                if(CarScamMutatorZone.TeleportTo[CarVehicle.GetTransform().Translation + Utility.PlayerOffset, CarVehicle.GetTransform().Rotation]) {}
            }
            
            Sleep(0.0)
        }
    }

    PlaceholderPropTick()<suspends>: void = {
        Utility := GetUtility()
        race {
            loop {
                Sleep(0.0)
                if {
                    SpawnedProp := MaybeSpawnedCarPlaceholderProp?
                    SpawnedProp.IsValid[]
                    CarVehicle := MaybeCarVehicle?
                } then {
                    if(SpawnedProp.TeleportTo[CarVehicle.GetTransform().Translation, CarVehicle.GetTransform().Rotation]) {}
                }
            }
            # Cancel the task after 10 seconds (then car should have stopped moving)
            Sleep(10.0)
        }
    }
}