
using { /Fortnite.com/Devices }
using {/Fortnite.com/Characters }
using { /Verse.org/Simulation }
using { /Verse.org/Simulation/Tags }
using { /UnrealEngine.com/Temporary/Diagnostics }
using {/Verse.org/Native}
using {/UnrealEngine.com/Temporary/SpatialMath}

# A Verse-authored creative device that can be placed in a level
custom_teleporter_device := class(creative_device) {
    @editable
    MutatorZone: mutator_zone_device = mutator_zone_device {}

    @editable
    Teleporter: teleporter_device = teleporter_device {}

    @editable
    LevelRequirement: int = 0

    @editable
    TooLowLevelHudMessage: hud_message_device = hud_message_device {}

    # Runs when the device is started in a running game
    OnBegin<override>()<suspends>: void = {
        MutatorZone.AgentEntersEvent.Subscribe(OnMutatorZoneEnter)
    }

    OnMutatorZoneEnter(Agent : agent): void = {
        if {
            PlayerData := Agent.GetPlayerData[]
        } then {
            if(PlayerData.CurLevel + 1 < LevelRequirement) {
                Print("Player level is too low for teleporter!")
                TooLowLevelHudMessage.Show(Agent)
            } else {
                Teleporter.Teleport(Agent)
            }
        }
    }
}