using { /Fortnite.com/Game }
using { /Fortnite.com/Devices }
# using { /Fortnite.com/Devices/CreativeAnimation }
using { /Fortnite.com/Characters }
using { /Verse.org/Simulation }
# using { /Verse.org/Simulation/Tags }
# using { /Verse.org/Random }
# using { /Verse.org/Concurrency }
# using { /UnrealEngine.com/Temporary/Diagnostics }
using { /UnrealEngine.com/Temporary/SpatialMath }

dragon_mount_anim_state := enum {
    Idle
    Moving
}

dragon_mount_boost_state := enum {
    Available
    Active
    Cooldown
}

dragon_mount_device := class(creative_device) {
    @editable
    SignalRemoteManager: signal_remote_manager_device = signal_remote_manager_device {}

    @editable
    ChairDeviceList: []chair_device = array {}

    @editable
    UfoScamMutatorZoneList: []mutator_zone_device = array {}

    @editable
    MovementSpeed: float = 2000.0

    @editable
    MovementBoostRate: float = 2.0

    # Cooldown starts after the boost effect is over (not included)
    @editable
    BoostCooldown: float = 9.0

    @editable
    BoostDuration: float = 3.0

    @editable
    AudioWhileFlying: audio_player_device = audio_player_device {}

    # Constants
    TickRate: float = 0.1

    DragonSpawnOffset: vector3 = vector3{X:=0.0, Y:=0.0, Z:=400.0}
    
    # Runtime variables
    var CurChairIndex: int = 0
    
    # Runs when the device is started in a running game
    OnBegin<override>()<suspends>:void = {
        Utility := GetUtility()
        SignalRemoteManager.PrimarySignalEvent.Subscribe(StartFlying)
        SignalRemoteManager.SecondarySignalEvent.Subscribe(StartFlying)

        for(ChairIndex -> ChairDevice: ChairDeviceList) {
            # ChairDevice.SeatedEvent.Subscribe(StartFlying)

            ExitEventHandler := event_handler_agent(int) {
                A := ChairIndex
                CallbackFunc := OnChairExited
            }

            ChairDevice.ExitedEvent.Subscribe(ExitEventHandler.EventSubscription)
        }

        for(ItemIndex -> MutatorZone: UfoScamMutatorZoneList) {
            EventHandler := event_handler_agent(int) {
                A := ItemIndex
                CallbackFunc := OnNpcEntersChairMutatorZone
            }

            MutatorZone.AgentEntersEvent.Subscribe(EventHandler.EventSubscription)
        }
    }

    StartFlying(Agent: agent): void = {
        Print("Start Flying")
        Utility := GetUtility()
        if {
            FortChar := Agent.GetFortCharacter[]
            PlayerData := Agent.GetPlayerData[]
        } then {
            ChairIndex := GetFreeChairIndex()
            if {
                ChairDevice := ChairDeviceList[ChairIndex]
            } then {
                ChairDevice.Eject()
                if(ChairDevice.TeleportTo[FortChar.GetTransform().Translation + DragonSpawnOffset, FortChar.GetTransform().Rotation]) {}
                ChairDevice.Seat(Agent)
                if(ChairDevice.IsSeated[Agent]) {
                    FortChar.Hide()
                    PlayerData.StartDragonMount(Self, ChairIndex)
                    # TODO: PlayerData.UnequipAllCosmetics(false)
                    AudioWhileFlying.Play(Agent)
                } else {
                    Print("WARN: DragonMount: StartFlying: Couldn't seat the player on chair")
                    if(ChairDevice.TeleportTo[Utility.HiddenPos, rotation{}]) {}
                }
            } else {
                Print("ERR: Dragon: StartFlying: Found no chair device for ChairIndex: {ChairIndex}")
            }
        }
    }

    OnChairExited(Agent: agent, ChairIndex: int): void = {
        Print("OnChairExited: ChairIndex: {ChairIndex}")
        GameManager := GetGameManager()
        if {
            ChairDevice := ChairDeviceList[ChairIndex]
            PlayerData := Agent.GetPlayerData[]
        } then {
            PlayerData.StopDragonMount()
            # TODO: PlayerData.ReapplyAllCosmetics()
            AudioWhileFlying.Stop(Agent)
            ChairDevice.Eject()

            # This must be delayed or dragon might still be visible
            Task := spawn{OnChairExitedSuspends(ChairDevice)}
        }
        # OnChairExited might also be called when a player leaves and has no FortChar
        if {
            FortChar := Agent.GetFortCharacter[]
        } then {
            FortChar.Show()
        }
    }

    OnChairExitedSuspends(ChairDevice: chair_device)<suspends>: void = {
        Utility := GetUtility()
        Sleep(TickRate)
        if(ChairDevice.TeleportTo[Utility.HiddenPos, rotation{}]) {}
    }

    GetFreeChairIndex(): int = {
        set CurChairIndex = CurChairIndex + 1
        if(CurChairIndex >= ChairDeviceList.Length) {
            set CurChairIndex = 0
        }
        return CurChairIndex
    }

    OnNpcEntersChairMutatorZone(NpcAgent: agent, MutatorZoneIndex: int): void = {
        # Print("OnNpcEntersChairMutatorZone")
        GameConfig := GetGameConfig()
        if {
            Chair := ChairDeviceList[MutatorZoneIndex]
        } then {
            MaybePlayerOnChair := Chair.GetSeatedAgent()
            if {
                PlayerOnChair := MaybePlayerOnChair?
                PlayerOnChairFortChar := PlayerOnChair.GetFortCharacter[]
                NpcData := NpcAgent.GetNpcData[]
                NpcFortChar := NpcAgent.GetFortCharacter[]
                NpcFortChar.IsActive[]
            } then {
                NpcData.TryScammingNpc(PlayerOnChair, GameConfig.UfoScamSpeedFactor)
            }
            <#
            if {
                PlayerOnChair := MaybePlayerOnChair?
                PlayerOnChairFortChar := PlayerOnChair.GetFortCharacter[]
                NpcFortChar := NpcAgent.GetFortCharacter[]
            } then {
                Print("OnNpcEntersChairMutatorZone 222")
                NpcFortChar.Damage(damage_args{Instigator := option{PlayerOnChair.GetInstigator()}, Amount := 100000000.0, Source := option{PlayerOnChairFortChar}})
            }
            #>
        }
    }
}
