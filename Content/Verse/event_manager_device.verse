using { /Fortnite.com/Game }
using { /Fortnite.com/Devices }
using { /Fortnite.com/Devices/CreativeAnimation }
using { /Fortnite.com/Characters }
using { /Verse.org/Simulation }
using { /Verse.org/Simulation/Tags }
using { /Verse.org/Random }
using { /Verse.org/Colors }
using { /Verse.org/Concurrency }
using { /UnrealEngine.com/Temporary/Diagnostics }
using { /UnrealEngine.com/Temporary/SpatialMath }
using { /Fortnite.com/Devices/CreativeAnimation/InterpolationTypes }

event_manager_device := class(creative_device) {
    @editable
    EntryPortalButtonList: []button_device = array {}

    @editable
    ExitPortalButton: button_device = button_device {}

    @editable
    EventStartPos: teleporter_device = teleporter_device {}

    @editable
    EventStartDelay: float = 30.0

    @editable
    EventDuration: float = 60.0

    @editable
    EventCooldown: float = 300.0

    # Constants
    
    # Runtime variables
    
    OnBegin<override>()<suspends>:void = {
        # Listen to events
        for(Button: EntryPortalButtonList) {
            Button.InteractedWithEvent.Subscribe(OnEntryPortalButton)
        }

        ExitPortalButton.InteractedWithEvent.Subscribe(OnExitPortalButton)

        Task := spawn{StartEventCooldown()}
    }

    OnEntryPortalButton(Agent: agent): void = {
        EventStartPos.Teleport(Agent)
    }

    OnExitPortalButton(Agent: agent): void = {
        TeleportPlayerBackToHisTycoon(Agent)
    }

    StartEventCooldown()<suspends>:void = {
        Sleep(EventCooldown)

        OpenPortal()
    }


    OpenPortal()<suspends>: void = {
        # Enable entry buttons
        for(Button: EntryPortalButtonList) {
            Button.Enable()
        }

        Sleep(EventStartDelay)

        StartEvent()
    }

    StartEvent()<suspends>: void = {
        # Disable entry buttons
        for(Button: EntryPortalButtonList) {
            Button.Disable()
        }

        Sleep(EventDuration)

        EndEvent()
    }

    EndEvent()<suspends>: void = {
        <# Do we need this, or should players leave manually?
        # Teleport players back to their individual base
        # TeleportPlayerBackToHisTycoon --> Only call on players that are inside the event!!
        #>

        # Start next event cooldown
        StartEventCooldown()
    }

    TeleportPlayerBackToHisTycoon(Agent: agent): void = {
        if {
            PlayerData := Agent.GetPlayerData[]
            TycoonManager := PlayerData.MaybeTycoonManager?
        } then {
            # Teleport player to his tycoon
            TycoonManager.PlayerStartPos.Teleport(Agent)
        }
    }
}