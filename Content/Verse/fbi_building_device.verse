using { /Fortnite.com/Game }
using { /Fortnite.com/Devices }
using { /Fortnite.com/Devices/CreativeAnimation }
using { /Fortnite.com/Characters }
using { /Verse.org/Simulation }
using { /Verse.org/Simulation/Tags }
using { /Verse.org/Random }
using { /Verse.org/Concurrency }
using { /UnrealEngine.com/Temporary/Diagnostics }
using { /UnrealEngine.com/Temporary/SpatialMath }

fbi_building_device := class(creative_device) {
    @editable
    ButtonToOpenDoor: button_device = button_device {}

    @editable
    CinematicOpenDoor: cinematic_sequence_device = cinematic_sequence_device {}

    @editable
    TimerStartOnOpenDoorList: []timer_device = array {}

    @editable
    KeyLevelRequired: int = 0

    @editable
    RoomCooldownAfterDoorOpen: float = 120.0

    @editable
    HudOnCooldown: hud_message_device = hud_message_device {}

    @editable
    HudKeyLevelTooLow: hud_message_device = hud_message_device {}

    @editable
    SwitchCollectItemList: []switch_device = array {}

    @editable
    SwitchRewardCurrencyList: []_Tycoon.tycoon_currency = array {}

    @editable
    SwitchMoneyRewardList: []int = array {}

    @editable
    SwitchFansRewardList: []float = array {}

    @editable
    SwitchDiamondsRewardList: []float = array {}

    # Constants
    
    # Runtime variables
    var TimestampLastDoorOpened: float = 0.0

    # Runs when the device is started in a running game
    OnBegin<override>()<suspends>:void = {
        # Listen to events
        ButtonToOpenDoor.InteractedWithEvent.Subscribe(OnOpenDoorButton)

        for(SwitchIndex -> Switch: SwitchCollectItemList) {
            EventHandler := event_handler_agent(int) {
                A := SwitchIndex
                CallbackFunc := OnSwitchTurnedOff
            }

            EventSub := Switch.TurnedOffEvent.Subscribe(EventHandler.EventSubscription)
        }
    }

    OnOpenDoorButton(Agent: agent): void = {
        if {
            PlayerData := Agent.GetPlayerData[]
        } then {
            # Check if the player has the required key level
            if(PlayerData.HighestFbiKeyUnlocked >= KeyLevelRequired) {
                # Check if this room is on cooldown
                Timestamp := GetSimulationElapsedTime()
                TimeSinceDoorOpened := Timestamp - TimestampLastDoorOpened
                if(TimeSinceDoorOpened >= RoomCooldownAfterDoorOpen) {
                    # Remember timestamp for cooldown
                    set TimestampLastDoorOpened = GetSimulationElapsedTime()
                    # Open the door
                    CinematicOpenDoor.Play(Agent)
                    # Start timers for Max
                    for(Timer: TimerStartOnOpenDoorList) {
                        Timer.Start(Agent)
                    }
                } else {
                    HudOnCooldown.Show(Agent)
                }
            } else {
                HudKeyLevelTooLow.Show(Agent)
            }
        }
    }

    OnSwitchTurnedOff(Agent: agent, ItemIndex: int): void = {
        SaveManager := GetSaveManager()
        if {
            PlayerData := Agent.GetPlayerData[]
            RewardCurrency := SwitchRewardCurrencyList[ItemIndex]
            MoneyReward := SwitchMoneyRewardList[ItemIndex]
            FansReward := SwitchFansRewardList[ItemIndex]
            DiamondsReward := SwitchDiamondsRewardList[ItemIndex]
        } then {
            PlayerData.AddMoney(MoneyReward, true, false)
            PlayerData.AddFans(FansReward, true)
            PlayerData.AddDiamonds(DiamondsReward, true)
            # Increment achievements
            # Earn $ 500K from FBI Containers
            SaveManager.IncrementAchievement(Agent, _Global.achievement_type.Money, 6, MoneyReward)
        }
    }
}
