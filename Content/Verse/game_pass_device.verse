using { /Verse.org/Simulation }
using { /Verse.org/Colors }
using { /Verse.org/Colors/NamedColors }
using { /Verse.org/Concurrency }
using { /Verse.org/Verse }
using { /Verse.org/Assets }
using { /Fortnite.com/Devices }
using { /Fortnite.com/UI }
using { /Fortnite.com/Characters }
using { /UnrealEngine.com/Temporary/Diagnostics }
using { /UnrealEngine.com/Temporary/SpatialMath }
using { /UnrealEngine.com/Temporary/UI}

game_pass_device := class(creative_device) {
    @editable
    OpenUiButtonList: []button_device = array {}

    <#
    @editable
    HudRewardReady: hud_message_device = hud_message_device {}
    #>

    @editable
    AudioOnRewardClaim: audio_player_device = audio_player_device {}

    @editable
    AudioNotEnoughMoney: audio_player_device = audio_player_device {}

    @editable
    ItemPriceList: []int = array {
        25,
        50,
        150,
        300,
        600,
        1200,
        1600,
        3500
    }

    # Constants

    # Runtime variables
    var AgentUiMap: [agent]_UI.ui_game_pass = map {}

    OnBegin<override>()<suspends>: void = {
        (super:)OnBegin()

        for(OpenUiButton: OpenUiButtonList) {
            OpenUiButton.InteractedWithEvent.Subscribe(OnOpenUiButton)
        }

        # GetPlayspace().PlayerAddedEvent().Subscribe(OnPlayerAdded)
        GetPlayspace().PlayerRemovedEvent().Subscribe(OnPlayerRemoved)

        <# Call logic for initial players
        for(Player : GetPlayspace().GetPlayers()) {
            OnPlayerAdded(Player)
        }
        #>
    }

    OnPlayerRemoved(Agent: agent): void = {
        # Delete the element from AgentUiMap to avoid memory leak
        if(Ui := AgentUiMap[Agent]) {
            Ui.RemoveUiFromPlayer(Agent)
        }
        set AgentUiMap = AgentUiMap.RemoveKey(Agent)
    }

    OnOpenUiButton(Agent: agent): void = {
        if(Ui := AgentUiMap[Agent]) {
            Ui.AddUiToPlayer(Agent, false)
        } else {
            Ui := _UI.ui_game_pass {}
            Ui.InitWidgetGamePass(Agent, Self)
            if(set AgentUiMap[Agent] = Ui) {}
            Ui.AddUiToPlayer(Agent, false)
        }
    }

    # 0 have to wait, 1 can buy, 2 already received
    GetItemState(Agent: agent, ItemIndex: int): int = {
        GameManager := GetGameManager()
        var ButtonText: string = ""
        if {
            PlayerData := Agent.GetPlayerData[]
            ItemState := PlayerData.GamePassStateList[ItemIndex]
        } then {
            if(ItemState = 0) {
                # Check if we have to update the item state since enough time passed
            }
            return ItemState
        } else {
            Print("ERR: PlaytimeRewardShop: GetItemState: Didn't find ItemState for Index: {ItemIndex}")
            return 0
        }
    }

    GetButtonText(Agent: agent, ItemIndex: int): string = {
        Utility := GetUtility()
        var ButtonText: string = ""
        ItemState := GetItemState(Agent, ItemIndex)
        case(ItemState) {
            0 =>
                # Use the default shop button text
                set ButtonText = "🔒"
            1 =>
                set ButtonText = "CLAIM"
            2 =>
                set ButtonText = "OK!"
            3 =>
                set ButtonText = "EQUIP"
            _ =>
                set ButtonText = ""
            }
        return ButtonText
    }

    HandleItemClicked(Agent: agent, Ui: _UI.ui_game_pass, ItemIndex: int): void = {
        GameManager := GetGameManager()
        ItemState := GetItemState(Agent, ItemIndex)
        if {
            ItemPrice := ItemPriceList[ItemIndex]
            PlayerData := Agent.GetPlayerData[]
        } then {
            if {
                ItemState = 1
            } then {
                if {
                    PlayerData.CurGamePassCoins >= ItemPrice
                } then {
                    # Subtract the item price
                    PlayerData.AddGamePassCoins(ItemPrice * -1)
                    # Handle item purchase
                    GivePlaytimeReward(Agent, Ui, ItemIndex)
                } else {
                    AudioNotEnoughMoney.Play(Agent)
                }
            }
        }
    }

    GivePlaytimeReward(Agent: agent, Ui: _UI.ui_game_pass, ItemIndex: int): void = {
        if {
            PlayerData := Agent.GetPlayerData[]
            # TextureCheckmark := _UI.TextureCheckmarkList[ItemIndex]
        } then {
            PlayerData.SetGamePassItemState(ItemIndex, 2)
            # TextureCheckmark.SetVisibility(widget_visibility.Visible)
            AudioOnRewardClaim.Play(Agent)
            # Refresh the UI
            Ui.RefreshUi()
        }
    }
}
