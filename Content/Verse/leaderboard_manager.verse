using { /Fortnite.com/Game }
using { /Fortnite.com/Devices }
using { /Fortnite.com/Devices/CreativeAnimation }
using { /Fortnite.com/Characters }
using { /Verse.org/Simulation }
using { /Verse.org/Simulation/Tags }
using { /Verse.org/Random }
using { /Verse.org/Assets }
using { /Verse.org/Concurrency }
using { /UnrealEngine.com/Temporary/Diagnostics }
using { /UnrealEngine.com/Temporary/SpatialMath }

leaderboard_manager_device := class(creative_device) {
    @editable
    LeaderboardPlayerRefList: []player_reference_device = array {}

    @editable
    LeaderboardBillboardList: []billboard_device = array {}

    # Constants
    LeaderboardTickRate: float = 30.0

    OnBegin<override>()<suspends>:void = {
        Task := spawn{ServerTickLeaderboard()}
    }

    ServerTickLeaderboard()<suspends>: void = {
        <#
        loop {
            Sleep(LeaderboardTickRate)
            UpdateLeaderboardEntries()
        }
        #>
        
        Sleep(10.0)
        UpdateLeaderboardEntries()
    }

    UpdateLeaderboardEntries(): void = {
        Utility := GetUtility()
        if {
            PlayerRef1 := LeaderboardPlayerRefList[0]
            PlayerRef2 := LeaderboardPlayerRefList[1]
            PlayerRef3 := LeaderboardPlayerRefList[2]
            Billboard1 := LeaderboardBillboardList[0]
            Billboard2 := LeaderboardBillboardList[1]
            Billboard3 := LeaderboardBillboardList[2]
        } then {
            var Players : []player = GetPlayspace().GetPlayers()
            var PlayerRankOne : ?agent = false
            var PlayerRankTwo : ?agent = false
            var PlayerRankThree : ?agent = false

            # Find the player with the most damage
            var HighestDamage : int = 0
            var PlayerIndexHighestDamage : ?int = false

            for(PlayerIndex -> Player : Players) {
                if {
                    PlayerData := Player.GetPlayerData[]
                    Damage := PlayerData.TotalXp
                } then {
                    if(Damage > HighestDamage) {
                        set HighestDamage = Damage
                        set PlayerRankOne = option{Player}
                        set PlayerIndexHighestDamage = option{PlayerIndex}
                    }
                }
            }

            if(PlayerIndex := PlayerIndexHighestDamage?) {
                if(Temp := Players.RemoveElement[PlayerIndex]) {
                    set Players = Temp
                }
            }

            # Find the player with the second most damage
            set HighestDamage = 0
            set PlayerIndexHighestDamage = false
            
            for(PlayerIndex -> Player : Players) {
                if {
                    PlayerData := Player.GetPlayerData[]
                    Damage := PlayerData.TotalXp
                } then {
                    if(Damage > HighestDamage) {
                        set HighestDamage = Damage
                        set PlayerRankTwo = option{Player}
                        set PlayerIndexHighestDamage = option{PlayerIndex}
                    }
                }
            }

            if(PlayerIndex := PlayerIndexHighestDamage?) {
                if(Temp := Players.RemoveElement[PlayerIndex]) {
                    set Players = Temp
                }
            }

            # Find the player with the third most damage
            set HighestDamage = 0
            
            for(PlayerIndex -> Player : Players) {
                if {
                    PlayerData := Player.GetPlayerData[]
                    Damage := PlayerData.TotalXp
                } then {
                    if(Damage > HighestDamage) {
                        set HighestDamage = Damage
                        set PlayerRankThree = option{Player}
                    }
                }
            }

            PlayerRef1.Clear()
            PlayerRef2.Clear()
            PlayerRef3.Clear()

            if(Player := PlayerRankOne?) {
                if(PlayerData := Player.GetPlayerData[]) {
                    PlayerRef1.Register(Player)
                    Billboard1.SetText(Utility.GetBillboardTextIntPrefix(PlayerData.CurLevel, 2, 0, "Level "))
                }
            }
            if(Player := PlayerRankTwo?) {
                if(PlayerData := Player.GetPlayerData[]) {
                    PlayerRef2.Register(Player)
                    Billboard2.SetText(Utility.GetBillboardTextIntPrefix(PlayerData.CurLevel, 2, 0, "Level "))
                }
            }
            if(Player := PlayerRankThree?) {
                if(PlayerData := Player.GetPlayerData[]) {
                    PlayerRef3.Register(Player)
                    Billboard3.SetText(Utility.GetBillboardTextIntPrefix(PlayerData.CurLevel, 2, 0, "Level "))
                }
            }
        }
    }
}