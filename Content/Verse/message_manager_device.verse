using { /Fortnite.com/Game }
using { /Fortnite.com/Devices }
using { /Fortnite.com/Devices/CreativeAnimation }
using { /Fortnite.com/Characters }
using { /Verse.org/Simulation }
using { /Verse.org/Simulation/Tags }
using { /Verse.org/Random }
using { /Verse.org/Colors }
using { /Verse.org/Concurrency }
using { /UnrealEngine.com/Temporary/Diagnostics }
using { /UnrealEngine.com/Temporary/SpatialMath }
using { /Fortnite.com/Devices/CreativeAnimation/InterpolationTypes }

verse_message_manager_device := class(tag) {}

message_manager_device := class(creative_device) {
    @editable
    HudMessageList: []hud_message_device = array {}

    # Constants
    PlayerNameMessage<localizes>(Agent: agent, MyString: string): message = "{Agent} {MyString}"
    
    # Runtime variables
    var ChatMessageList: []string = array {}
    # var ChatMessageColorList: []color = array {}
    
    OnBegin<override>()<suspends>:void = {
        # Init variables
        for(ItemIndex -> HudMessage: HudMessageList) {
            set ChatMessageList = ChatMessageList + array{""}
        }
    }

    AddMessage(AgentThatCausedMessage: agent, ParmMessage: string): void = {
        Print("message_manager_device - AddMessage")
        Utility := GetUtility()

        # Rotate existing chat messages up, to make space for the new chat message
        for(ItemIndex -> HudMessage: HudMessageList) {
            if {
                ChatMessage := ChatMessageList[ItemIndex]
                ChatMessage.Length > 0
            } then {
                if(set ChatMessageList[ItemIndex - 1] = ChatMessage) {}
            }
        }

        # HudMessageHigherIndex := HudMessageList[ItemIndex - 1]
        # HudMessageHigherIndex.Show(Agent, PlayerNameMessage(AgentThatCausedMessage, ChatMessage))

        # Insert the new text message at the lowest index
        if(set ChatMessageList[HudMessageList.Length - 1] = ParmMessage) {}

        <#
        if(HudMessageLastIndex := HudMessageList[HudMessageList.Length - 1]) {
            HudMessageLastIndex.Show(Agent, PlayerNameMessage(AgentThatCausedMessage, ParmMessage))
        }
        #>

        # Show every hud message for every player in the game
        for(Player : GetPlayspace().GetPlayers()) {
            for(ItemIndex -> HudMessage: HudMessageList) {
                if {
                    ChatMessage := ChatMessageList[ItemIndex]
                    ChatMessage.Length > 0
                } then {
                    HudMessage.Show(Player, PlayerNameMessage(AgentThatCausedMessage, ChatMessage))
                }
            }
        }
    }
}