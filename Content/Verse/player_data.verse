using { /Verse.org/Simulation }
using { /Verse.org/Colors }
using { /Verse.org/Colors/NamedColors }
using { /Verse.org/Concurrency }
using { /Verse.org/Verse }
using { /Fortnite.com/Game }
using { /Fortnite.com/Devices }
using { /Fortnite.com/UI }
using { /Fortnite.com/Characters }
using { /UnrealEngine.com/Temporary/Diagnostics }
using { /UnrealEngine.com/Temporary/SpatialMath }
using { /UnrealEngine.com/Temporary/UI}

player_data<public> := class<unique>() {
    PlayerLocalUi: _UI.ui_main = _UI.ui_main {}
    UiChangelog: _UI.ui_changelog = _UI.ui_changelog {}
    # UiAcceptQuest: _UI.ui_accept_quest = _UI.ui_accept_quest {}
    # UiAcceptQuestButtons: _UI.ui_accept_quest_buttons = _UI.ui_accept_quest_buttons {}
    UiScamNpc: _UI.ui_scam_npc = _UI.ui_scam_npc {}
    UiElevator: _UI.ui_elevator = _UI.ui_elevator {}

    var MaybePlayer: ?player = false
    var MaybeTycoonManager: ?_Tycoon.tycoon_manager_device = false

    var TimestampPlayerJoined: float = 0.0
    var TimestampPlayerUnlockedPlaytimeRewards: float = 0.0
    var TimestampLastParkourCompleted: float = 0.0
    var TimestampJailed: float = 0.0
    var TimestampLastXpGained: float = 0.0

    var MaybeXpHud: ?hud_message_device = false
    var MaybeLevelUpHud: ?hud_message_device = false
    var MaybeQuestProgressHud: ?hud_message_device = false
    var MaybeWildlifeMountManager: ?wildlife_mount_device = false
    var MaybeWildlifeSpawner: ?wildlife_spawner_device = false
    var PlayerListenerList: []cancelable = array {}

    var MaybeActiveQuest: ?_Tycoon.tycoon_button = false
    var ButtonListEnableOnDeath: []button_device = array {}
    var MaybeCurCinematic: ?cinematic_sequence_device = false
    var MaybeSkyboxCinematic: ?cinematic_sequence_device = false
    var QuestZombieKillListenerList: []cancelable = array {}
    
    var XpBoostLevel: int = 0
    var HasUnlockedCityOwner: logic = false
    
    var TeamIndex : int = -1
    
    # Constants
    GamePassXpTickRate: float = 5.0
    QuestProgressTickRate: float = 1.0
    StartMaxHp: float = 100.0
    MaxHpIncreasePerLevel: float = 1.0
    StartShield: float = 100.0
    ShieldIncreasePerArmorLevel: float = 25.0
    ParkourMultiplierDuration: float = 60.0 * 5.0

    XpBoostFactorList: []float = array {
        # Default
        1.0,
        # XP Boost Code from Twitter
        2.0,
        # Rebirth Shop
        3.0,
        # Rebirth Shop
        4.0,
        # Rebirth Shop
        5.0,
        # Rebirth Shop
        6.0
    }

    # Runtime variables
    var ActiveUiList: [int]_UI.ui_base = map {}
    var CurrentUiIndex: int = 0
    # Prevent 2 UIs from having the same Z Order, by counting up a variable
    # var CurrentUiZOrder: int = 0

    var RoutineServerTick: ?routine = false
    var RoutineGamePassXpTick: ?routine = false
    var RoutineQuestProgressTick: ?routine = false
    var RoutineParkourTimeoutTick: ?routine = false
    var MaybeRoutineServerTickFloatingUi: ?routine = false
    var MaybeRoutineAchievementTimerKill: ?routine = false
    var MaybeRoutineAchievementTimerScam: ?routine = false

    # Index is the cosmetic socket id. Set if we have a cosmetic equipped in that socket
    var CosmeticManagerList: [_Tycoon.cosmetic_socket]_Tycoon.tycoon_button_cosmetic = map {}

    var QuestListenerList: []cancelable = array {}

    var RoutineXpHudAmount: ?routine = false
    var RoutineDeactivateShrineBuffItemGranter: ?routine = false

    var IsActiveShrineBuffItemGranter: logic = false
    var IsInNoDragonMountArea: logic = false
    var IsInJail: logic = false
    var IsScammingNpc: logic = false

    var TotalXp: int = 0
    var CurLevel: int = 0
    var CurWeaponDamage<private>: float = 1.0
    var NumOfRebirths: int = 0
    var CurMoney: int = 0
    var CurFans: float = 0.0
    var CurDiamonds: float = 0.0
    var CurGamePassCoins: int = 0
    var CurCreditCardCoins: int = 0
    var CurPetsList: []_Npc.pet = array {}
    var GotVeryLargeAccolade: logic = false
    var TimestampLargeAccoladeEpoch: float = 0.0

    var CurMoneyMultiplier: float = 1.0
    var CurFansMultiplier: float = 1.0
    var CurDiamondsMultiplier: float = 1.0
    var CurXpMultiplier: float = 1.0
    var GamePassXpIncomePerSec: int = 0
    var PetMoneyMultiplier: float = 0.0
    var PetFansMultiplier: float = 0.0
    var PetDiamondsMultiplier: float = 0.0
    var CosmeticMoneyMultiplier: float = 0.0
    var CosmeticFansMultiplier: float = 0.0
    var CosmeticDiamondsMultiplier: float = 0.0
    var ParkourMoneyMultiplier: float = 0.0
    var PcBuffMoneyMultiplier: float = 0.0
    var TimestampLastKill: float = 0.0
    var KillStreakMultiplier: int = 0
    var TimestampOfficerDeath: float = 0.0

    var CurMoneyIncome: int = 10
    var NetWorth: int = 0

    # We need this as additional variable, since it stays (and is saved) after a quest is completed
    # Repeatable quests start with ID 100
    var CurrentQuestId: int = -1
    var DamagePowerupLevel: int = -1
    var ArmorPowerupLevel: int = -1
    var CurCosmeticId: int = -1
    var HighestElevatorFloorIdUnlocked: int = -1
    var HighestFbiKeyUnlocked: int = -1
    var TotalScamLevel: int = 0
    var ScamLevelHead: int = 0
    var ScamLevelBack: int = 0
    var FoundItemForBankAtm: logic = false
    var HasUnlockedGarage: logic = false
    var HasUnlockedTycoonAutoSave: logic = false
    var HackAchievementTimer: float = 0.0

    var RebirthAutomaticIncomeLevel: int = -1
    var AutomaticIncomeMoney: []int = array {
        0
        <#
        200,
        400,
        600,
        1000
        #>
    }

    var AutomaticIncomeDiamonds: []float = array {
        0.0
        <#
        100.0,
        200.0,
        300.0,
        500.0
        #>
    }

    var DragonMountIndex: int = -1
    var MaybeDragonMountManager: ?dragon_mount_device = false
    var RoutineDragonMountTick: ?routine = false
    var DragonMountBoostState: dragon_mount_boost_state = dragon_mount_boost_state.Available

    # 0 have to wait, 1 can buy, 2 already received
    var PlaytimeRewardStateList: []int = array {
        0,
        0,
        0,
        0,
        0,
        0,
        0,
        0,
        0,
        0,
        0,
        0
    }

    # 0 Not unlocked, 1 can buy, 2 already received, 3 equip (only VFX)
    var GamePassStateList: []int = array {
        1,
        0,
        0,
        0,
        0,
        0,
        0,
        0,
        0,
        0,
        0,
        0
    }

    # 0 not owned / can buy, 1 owned
    # 0: wolf mount, 1: dragon mount
    var MountShopStateList: []int = array {
        0,
        0
    }

    # 0 not owned / can buy, 1 owned, 2 equipped
    var PetShopStateList: []int = array {
        0,
        0,
        0,
        0,
        0
    }

    # 0 not owned / can buy, 1 owned, 2 equipped
    var CosmeticShopStateList: []int = array {
        0,
        0,
        0,
        0,
        0
    }

    # Floating XP Hud text stacks
    var CurXpHudAmount: int = 0
    var XpHudDuration: float = 2.0

    var ColorShopStateList: [][]int = array {
        array {
            0,
            0,
            0,
            0,
            0,
            0,
            0,
            0,
            0,
            0,
            0,
            0
        },
        array {
            0,
            0,
            0,
            0,
            0,
            0,
            0,
            0,
            0,
            0,
            0,
            0
        },
        array {
            0,
            0,
            0,
            0,
            0,
            0,
            0,
            0,
            0,
            0,
            0,
            0
        }
    }

    # 0 not owned / can buy, 1 owned, 2 equipped
    var PcCaseShopStateList: []int = array {
        0,
        0,
        0,
        0,
        0,
        0,
        0,
        0,
        0,
        0,
        0,
        0
    }

    # 0 not owned / can buy, 1 owned, 2 equipped
    var PcDisplayShopStateList: []int = array {
        0,
        0,
        0,
        0,
        0,
        0,
        0,
        0,
        0,
        0,
        0,
        0
    }

    # 0 not owned / can buy, 1 owned, 2 equipped
    var PcDeskShopStateList: []int = array {
        0,
        0,
        0,
        0,
        0,
        0,
        0,
        0,
        0,
        0,
        0,
        0
    }

    # 0 not owned, 1 owned
    var CityOwnerStateList: []int = array {
        0,
        0,
        0
    }

    InitPlayerData(Player: player, ParmTeamIndex: int): void = {
        Print("PlayerData - InitPlayerData")
        GameManager := GetGameManager()
        SaveManager := GetSaveManager()

        # Init variables
        set TimestampPlayerJoined = GetSimulationElapsedTime()
        set MaybePlayer = option{Player}
        set TeamIndex = ParmTeamIndex
        
        PlayerLocalUi.InitWidget(Player)
        # ShowPlayerLocalUi(Player)
        UiChangelog.InitWidget(Player)
        # UiAcceptQuest.InitWidget(Player)
        # UiAcceptQuestButtons.InitWidget(Player)
        UiScamNpc.InitWidget(Player)
        UiElevator.InitWidget(Player)
        # Show the HUD on player join
        UpdateCurCreditCardCoin(0)

        # Listen for death event
        if(FortChar := Player.GetFortCharacter[]) {
            DeathListener := FortChar.EliminatedEvent().Subscribe(OnPlayerEliminated)
            set PlayerListenerList = PlayerListenerList + array{DeathListener}
        }
        
        <# Custom health bar
        if(FortChar := Player.GetFortCharacter[]) {
            # Listen for health damaged event
            HealthDamagedListener := FortChar.DamagedEvent().Subscribe(OnPlayerHealthDamaged)
            set PlayerListenerList = PlayerListenerList + array{HealthDamagedListener}
            # Listen for shield damaged event
            ShieldDamagedListener := FortChar.DamagedShieldEvent().Subscribe(OnPlayerShieldDamaged)
            set PlayerListenerList = PlayerListenerList + array{ShieldDamagedListener}
            # Listen for health healed event
            HealthHealedListener := FortChar.HealedEvent().Subscribe(OnPlayerHealthHealed)
            set PlayerListenerList = PlayerListenerList + array{HealthHealedListener}
            # Listen for shield healed event
            ShieldHealedListener := FortChar.HealedShieldEvent().Subscribe(OnPlayerShieldHealed)
            set PlayerListenerList = PlayerListenerList + array{ShieldHealedListener}
        }
        #>

        RoutineServerTickSecond := SpawnRoutine(ServerTickSecond)
        set RoutineServerTick = option{RoutineServerTickSecond}

        RoutineServerTickFloatingUi := SpawnRoutine(ServerTickFloatingUi)
        set MaybeRoutineServerTickFloatingUi = option{RoutineServerTickFloatingUi}

        <# Disabled for now, since not needed
        MyRoutineGamePass := SpawnRoutine(GamePassXpTick)
        set RoutineGamePassXpTick = option{MyRoutineGamePass}

        MyRoutineGamePass := SpawnRoutine(GamePassXpTick)
        set RoutineGamePassXpTick = option{MyRoutineGamePass}
        #>
    }

    HandlePlayerSpawned(): void = {
        SaveManager := GetSaveManager()
        GameManager := GetGameManager()
        if {
            Player := MaybePlayer?
        } then {
            # Restore player local UI (other UIs are not reapplied)
            ShowPlayerLocalUi(Player)
            # Restore cosmetics after respawn
            ReapplyAllCosmetics()
            # Restore health values - Must be delayed
            Task := spawn{CalcPlayerHealthAndShieldSuspends(option{Player}, 1.0)}
            # SetHealth(Player)
            # SetShield(Player)
            # Restore instant reload class
            if {
                Save := SaveManager.ProxySaves[Player]
                Save.UnlockedInstantReload = true
            } then {
                GameManager.ClassInstantReload.ChangeClass(Player)
            }
        }
    }

    CalcPlayerHealthAndShieldSuspends(MaybeAgent: ?agent, DelayTime: float)<suspends>: void = {
        Sleep(DelayTime)
        if {
            Agent := MaybeAgent?
        } then {
            SetHealth(Agent)
            SetShield(Agent)
        }
    }

    AddUiToPlayer(UiScript: _UI.ui_base, Canvas: canvas, UiSlot: player_ui_slot): int = {
        if {
            Player := MaybePlayer?
            PlayerUI := GetPlayerUI[Player]
        } then {
            set CurrentUiIndex = CurrentUiIndex + 1
            Print("PlayerData: AddUiToPlayer: CurrentUiIndex: {CurrentUiIndex}")
            if(set ActiveUiList[CurrentUiIndex] = UiScript) {}
            # This must be last, since it might crash Verse. Like this we can still remove it on player respawn
            PlayerUI.AddWidget(Canvas, UiSlot)
        }
        return CurrentUiIndex
    }

    RemoveUiFromPlayer(UiScript: _UI.ui_base, UiIndexToRemove: int): void = {
        if {
            Player := MaybePlayer?
            PlayerUI := GetPlayerUI[Player]
        } then {
            Print("PlayerData: RemoveUiFromPlayer: UiIndexToRemove: {UiIndexToRemove}")
            UiScript.CleanUpUiOnRemove()
            if {
                Canvas := UiScript.MaybeCanvas?
            } then {
                Print("PlayerData: RemoveUiFromPlayer: Found Canvas")
                PlayerUI.RemoveWidget(Canvas)
                # Also remove from map
                set ActiveUiList = ActiveUiList.RemoveKey(UiIndexToRemove)
            } else {
                Print("ERR: PlayerData: RemoveUiFromPlayer: Didn't find Canvas for index: {UiIndexToRemove}")
            }
        }
    }

    HideUis(Agent: agent): void = {
        GameManager := GetGameManager()
        Utility := GetUtility()
        PlayerLocalUi.RemoveUiFromPlayer(Agent)
        # Hide Huds for Player Frame
        GameManager.HudPlayerFramePlayerName.Hide(Agent)
        GameManager.HudPlayerFrameCurrentLevel.Hide(Agent)
        GameManager.HudPlayerFrameCurrentHealth.Hide(Agent)
        GameManager.HudPlayerFrameCurrentShield.Hide(Agent)
        # Hide Huds for Level Bar
        GameManager.HudCurrentLevel.Hide(Agent)
        GameManager.HudNextLevel.Hide(Agent)
        GameManager.HudCurrentXp.Hide(Agent)
    }

    ShowPlayerLocalUi(Agent: agent): void = {
        Print("ShowPlayerLocalUi")
        PlayerLocalUi.AddUiToPlayer(Agent, false)
        # Set initial UI values
        UpdateLevelBar(true)
        # PlayerLocalUi.UpdateHealthBar(Agent)
        UpdateMoneyMultiplier()
        UpdateFansMultiplier()
        UpdateDiamondsMultiplier()
        UpdateXpMultiplier()
        UpdateScamLevel()
    }

    DestroyPlayerData(Agent: agent): void = {
        Print("DestroyPlayerData")
        GameManager := GetGameManager()
        set MaybePlayer = false

        # Delete map data
        set CosmeticManagerList = map {}
        set ActiveUiList = map {}

        # Clean up routines
        if(Routine := RoutineServerTick?) {
            Routine.Cancel()
        }
        if(Routine := RoutineGamePassXpTick?) {
            Routine.Cancel()
        }
        if(Routine := RoutineQuestProgressTick?) {
            Routine.Cancel()
        }
        if(Routine := RoutineParkourTimeoutTick?) {
            Routine.Cancel()
        }
        if(Routine := MaybeRoutineServerTickFloatingUi?) {
            Routine.Cancel()
        }

        # Return devices to pool
        # Wildlife Mount Spawner
        if {
            WildlifeMountManager := MaybeWildlifeMountManager?
            WildlifeSpawner := MaybeWildlifeSpawner?
        } then {
            WildlifeMountManager.ReturnWildlifeSpawner(WildlifeSpawner)
        }
        set MaybeWildlifeMountManager = false
        set MaybeWildlifeSpawner = false
    }

    
    AddCurrency(Currency: _Tycoon.tycoon_currency, Amount: int, ShowOnHud: logic): void = {
        if {
            Amount > 0
        } then {
            case(Currency) {
                _Tycoon.tycoon_currency.Money =>
                    AddMoney(Amount, ShowOnHud, false)
                _Tycoon.tycoon_currency.Fans =>
                    AddFans(Amount * 1.0, ShowOnHud)
                _Tycoon.tycoon_currency.Diamonds =>
                    AddDiamonds(Amount * 1.0, ShowOnHud)
                _ =>
                    Print("ERR: AddCurrency: Currency not implemented")
            }
        }
    }

    AddMoney<public>(ParmAmount: int, ShowOnHud: logic, IgnoreMoneyMultiplier: logic): void = {
        GameManager := GetGameManager()
        SaveManager := GetSaveManager()
        Utility := GetUtility()
        if {
            Player := MaybePlayer?
            ParmAmount > 0
        } then {
            var Amount: int = ParmAmount
            if {
                IgnoreMoneyMultiplier = false
                AmountFloat := ParmAmount * CurMoneyMultiplier
                AmountRounded := Round[AmountFloat]
                set Amount = AmountRounded
            }
            Print("AddMoney: Amount: {Amount}")
            UpdateCurMoney(CurMoney + Amount)
            if(ShowOnHud = true) {
                # GameManager.PlayerMoneyGainedHud.Show(Player, Utility.FormatMessage("+{ToString(Amount)}"))
                PlayerLocalUi.AddFloatingMessage(Amount, 1)
            }
            # Check for achievement
            # Have $ 1M
            if {
                CurMoney > 1000000
            } then {
                SaveManager.IncrementAchievement(Player, _Global.achievement_type.Money, 0, 1)
            }
            # Have $ 1B
            if {
                CurMoney > 1000000000
            } then {
                SaveManager.IncrementAchievement(Player, _Global.achievement_type.Money, 9, 1)
            }
        }
    }

    RemoveMoney(Amount: int): void = {
        GameManager := GetGameManager()
        SaveManager := GetSaveManager()
        if {
            Player := MaybePlayer?
        } then {
            UpdateCurMoney(CurMoney - Amount)
            # Check for achievement
            # Spend $ 10M
            SaveManager.IncrementAchievement(Player, _Global.achievement_type.Money, 8, Amount)
        }
    }

    AddFans(Amount: float, ShowOnHud: logic): void = {
        GameManager := GetGameManager()
        Utility := GetUtility()
        if {
            Player := MaybePlayer?
            Amount > 0.0
        } then {
            var NpcHelperMultiplier: float = 1.0
            <# Check if we have NPC helper multipliers for fans income
            if {
                TycoonManager := MaybeTycoonManager?
                NpcHelper := TycoonManager.TycoonButtonNpcHelper?
            } then {
                set NpcHelperMultiplier = NpcHelper.GetCurrentRewardFactor()
            }
            #>

            # Print("NpcHelperMultiplier: {NpcHelperMultiplier}")
            AmountMultiplied := Amount * NpcHelperMultiplier

            if {
                AmountRounded := Round[AmountMultiplied]
            } then {
                UpdateCurFans(CurFans + AmountMultiplied)
                UpdateResourceUi()
                if {
                    ShowOnHud = true
                    AmountRounded >= 1
                } then {
                    # GameManager.PlayerFansGainedHud.Show(Player, Utility.FormatMessage("+{ToString(AmountRounded)}"))
                    PlayerLocalUi.AddFloatingMessage(AmountRounded, 2)
                }
            }
        }
    }

    RemoveFans(Amount: int): void = {
        UpdateCurFans(CurFans - (Amount * 1.0))
        UpdateResourceUi()
    }

    AddDiamonds(Amount: float, ShowOnHud: logic): void = {
        GameManager := GetGameManager()
        Utility := GetUtility()
        if {
            Player := MaybePlayer?
            Amount > 0.0
            AmountRounded := Round[Amount]
        } then {
            UpdateCurDiamonds(CurDiamonds + Amount)
            UpdateResourceUi()
            if {
                ShowOnHud = true
                AmountRounded >= 1
            } then {
                # GameManager.PlayerDiamondsGainedHud.Show(Player, Utility.FormatMessage("+{ToString(AmountRounded)}"))
                PlayerLocalUi.AddFloatingMessage(AmountRounded, 3)
            }
        }
    }

    RemoveDiamonds(Amount: int): void = {
        UpdateCurDiamonds(CurDiamonds - (Amount * 1.0))
        UpdateResourceUi()
    }

    AddCreditCardCoin(Amount: int): void = {
        UpdateCurCreditCardCoin(CurCreditCardCoins + Amount)
    }

    RemoveCreditCardCoin(Amount: int): void = {
        UpdateCurCreditCardCoin(CurCreditCardCoins - Amount)
    }

    SetHealth(Agent: agent): void = {
        SaveManager := GetSaveManager()
        GameConfig := GetGameConfig()
        if {
            FortChar := Agent.GetFortCharacter[]
            Save := SaveManager.ProxySaves[Agent]
            TycoonManager := MaybeTycoonManager?
            MaxHealthIncreasePerUpgrade := GameConfig.PermanentUpgradeRewardPerLevelList[0]
        } then {
            # Calculate health - 500 max
            # NewHealthValue := Min(StartMaxHp + (CurLevel * MaxHpIncreasePerLevel) + (Save.UpgradeHealthIncreaseLevel * MaxHealthIncreasePerUpgrade), 500.0)
            NewHealthValue := Min(StartMaxHp + (Save.UpgradeHealthIncreaseLevel * MaxHealthIncreasePerUpgrade), 500.0)
            FortChar.SetMaxHealth(NewHealthValue)
            FortChar.SetHealth(FortChar.GetMaxHealth())
            # PlayerLocalUi.UpdateHealthBar(Agent)
        }
    }

    SetShield(Agent: agent): void = {
        SaveManager := GetSaveManager()
        if {
            FortChar := Agent.GetFortCharacter[]
            Save := SaveManager.ProxySaves[Agent]
        } then {
            # Shield currently not in use, copied from UltraWorld
            NewShieldValue := StartShield + ((ArmorPowerupLevel + 1) * ShieldIncreasePerArmorLevel)
            FortChar.SetMaxShield(NewShieldValue)
            FortChar.SetShield(FortChar.GetMaxShield())
            # PlayerLocalUi.UpdateHealthBar(Agent)
        }
    }

    
    SetHealthAndShieldSuspends(Agent: agent)<suspends>: void = {
        Sleep(5.0)
        SetHealth(Agent)
        SetShield(Agent)
    }

    CheckPowerupAchievement(Agent: agent): void = {
        Print("CheckPowerupAchievement")
        GameManager := GetGameManager()
        SaveManager := GetSaveManager()
        if(DamagePowerupLevel >= 2) {
            if(ArmorPowerupLevel >= 2) {
                # SaveManager.GrantAchievement(Agent, SaveManager.ACHIEVEMENT_POWER_UPS_ALL)
            }
        }
    }

    # Can also be used with negative values
    AddGamePassCoins(AmountToAdd: int): void = {
        SaveManager := GetSaveManager()
        GameManager := GetGameManager()
        Utility := GetUtility()
        
        if {
            Player := MaybePlayer?
            Save := SaveManager.ProxySaves[Player]
        } then {
            set CurGamePassCoins = CurGamePassCoins + AmountToAdd
            set Save.GamePassCoins = CurGamePassCoins
            # GameManager.PlayerGamePassCoinGainedHud.Show(Player, Utility.FormatMessage("+{ToString(AmountToAdd)}"))
            # GamePassCoin doesn't exist for this game and is not shown
            # PlayerLocalUi.AddFloatingMessage(AmountToAdd, 4)
        }
    }

    UpdateCurMoney(NewAmount: int): void = {
        SaveManager := GetSaveManager()
        if {
            Player := MaybePlayer?
            Save := SaveManager.ProxySaves[Player]
        } then {
            set CurMoney = NewAmount
            set Save.Money = NewAmount
            UpdateResourceUi()
        }
    }

    UpdateCurFans(NewAmount: float): void = {
        SaveManager := GetSaveManager()
        if {
            Player := MaybePlayer?
            Save := SaveManager.ProxySaves[Player]
        } then {
            set CurFans = NewAmount
            set Save.Fans = NewAmount
        }
    }

    UpdateCurDiamonds(NewAmount: float): void = {
        SaveManager := GetSaveManager()
        if {
            Player := MaybePlayer?
            Save := SaveManager.ProxySaves[Player]
        } then {
            set CurDiamonds = NewAmount
            set Save.Diamonds = NewAmount
        }
    }

    UpdateCurCreditCardCoin(NewAmount: int): void = {
        GameManager := GetGameManager()
        SaveManager := GetSaveManager()
        Utility := GetUtility()
        if {
            Player := MaybePlayer?
            Save := SaveManager.ProxySaves[Player]
        } then {
            set CurCreditCardCoins = NewAmount
            set Save.CurCreditCardCoins = NewAmount
            GameManager.CurCreditCardCoinHud.Show(Player, Utility.FormatMessage(ToString(CurCreditCardCoins)))
        }
    }

    UpdateCurRebirths(NewAmount: int): void = {
        SaveManager := GetSaveManager()
        if {
            Player := MaybePlayer?
            Save := SaveManager.ProxySaves[Player]
        } then {
            set NumOfRebirths = NewAmount
            set Save.Rebirths = NewAmount
            UpdateRebirthsUi()
        }
    }

    PersistButtonPurchase(ButtonId: int): void = {
        # set TycoonButtonsOwned = TycoonButtonsOwned + array{ButtonId}
        # Tycoon Buttons are only persisted in manual save for this game
        SaveManager := GetSaveManager()
        if {
            Player := MaybePlayer?
            Save := SaveManager.ProxySaves[Player]
        } then {
            set Save.TycoonButtonsOwned = Save.TycoonButtonsOwned + array{ButtonId}
        }
    }

    # RewardPlayerIncomeFactor is for minigames
    # Maybe this should also use the current multiplier for the reward currency?
    RewardPlayerIncomeFactor(Currency: _Tycoon.tycoon_currency, IncomeFactor: float): void = {
        Print("RewardPlayer: IncomeFactor: {IncomeFactor}")
        if(TycoonManager := MaybeTycoonManager?) {
            case(Currency) {
                _Tycoon.tycoon_currency.Money =>
                    if(Payout := Floor[CurMoneyIncome * IncomeFactor]) {
                        AddMoney(Payout, true, false)
                    }
                _Tycoon.tycoon_currency.Fans =>
                    AddFans((TycoonManager.FansIncomePerSec + TycoonManager.StreamerFansIncomePerSec) * IncomeFactor, true)
                _Tycoon.tycoon_currency.Diamonds =>
                    AddDiamonds(TycoonManager.DiamondsIncomePerSec * IncomeFactor, true)
                _ =>
                    Print("ERR: RewardPlayerIncomeFactor: Currency not implemented")
            }
        }
    }

    UpdateResourceUi(): void = {
        Utility := GetUtility()
        if {
            # Only update UI if we have a tycoon assigned
            TycoonManager := MaybeTycoonManager?
            FansFloored := Floor[CurFans]
            DiamondsFloored := Floor[CurDiamonds]
        } then {
            # FormattedMoney := Utility.GetBillboardTextFloat(CurMoney * 1.0, 2)
            PlayerLocalUi.UpdateResourceUi(
                CurMoney,
                FansFloored,
                DiamondsFloored
            )
        }
    }

    UpdateRebirthsUi(): void = {
        Utility := GetUtility()
        if {
            # Only update UI if we have a tycoon assigned
            TycoonManager := MaybeTycoonManager?
        } then {
            PlayerLocalUi.UpdateRebirthsUi(
                NumOfRebirths
            )
        }
    }

    SetXpBoostLevel(ParmXpBoostLevel: int): void = {
        Print("SetXpBoostLevel: ParmXpBoostLevel {ParmXpBoostLevel}")
        SaveManager := GetSaveManager()
        if {
            ParmXpBoostLevel > XpBoostLevel
            Player := MaybePlayer?
            Save := SaveManager.ProxySaves[Player]
        } then {
            set XpBoostLevel = ParmXpBoostLevel
            set Save.XpBoostLevel = ParmXpBoostLevel
            PlayerLocalUi.RefreshPlayerLocalUi(Player)
        }
    }

    UpdateIncomeMoney(NewMoneyIncome: int): void = {
        Utility := GetUtility()
        if {
            TycoonManager := MaybeTycoonManager?
        } then {
            set CurMoneyIncome = NewMoneyIncome
            # Update income UI
            PlayerLocalUi.UpdateIncomeMoneyUi(CurMoneyIncome * 1.0)
        }
    }

    AssignTycoon(TycoonManager: _Tycoon.tycoon_manager_device): void = {
        set MaybeTycoonManager = option {TycoonManager}
        UpdateResourceUi()
        UpdateRebirthsUi()
        spawn{RestoreTycoonButtonsOwned()}
        <#
        RestoreTycoonWallColor()
        RestorePcUpgradesCase()
        RestorePcUpgradesDisplay()
        RestorePcUpgradesDesk()
        #>
    }

    UnassignTycoon(): void = {
        Print("PlayerData - UnassignTycoon")
        UnequipAllCosmetics(true)
        if {
            TycoonManager := MaybeTycoonManager?
            Agent := MaybePlayer?
        } then {
            # Remove home beacon
            TycoonManager.YourHomeBeacon.RemoveFromShowList(Agent)
            # Reset the tycoon
            TycoonManager.ResetTycoon()
            set MaybeTycoonManager = false
            # Pets are destroyed on rebirth, so we can set the multiplier to 0
            set PetMoneyMultiplier = 0.0
            set PetFansMultiplier = 0.0
            set PetDiamondsMultiplier = 0.0
            set CosmeticMoneyMultiplier = 0.0
            set CosmeticFansMultiplier = 0.0
            set CosmeticDiamondsMultiplier = 0.0
            set ParkourMoneyMultiplier = 0.0
            set HasUnlockedCityOwner = false
        }
    }

    SetCarShopItemState(ItemIndex: int, NewState: int): void = {
        # Save the purchase
        SaveManager := GetSaveManager()
        if {
            NewState >= 1
            Agent := MaybePlayer?
            PlayerData := Agent.GetPlayerData[]
            TycoonManager := PlayerData.MaybeTycoonManager?
            Save := SaveManager.ProxySaves[Agent]
            # Only trigger this code if the car was not previously owned
            Save.CarShopStateList[ItemIndex] = 0
        } then {
            if(set Save.CarShopStateList[ItemIndex] = 1) {
                # Check to grant achievement
                # Unlock the Scam UFO in Rebirth 5 or higher
                if {
                    ItemIndex = 6
                    PlayerData.NumOfRebirths >= 5
                } then {
                    SaveManager.IncrementAchievement(Agent, _Global.achievement_type.Rebirth, 4, 1)
                }
                # Directly start driving the new car
                TycoonManager.ShopGarage.HandleCarSummon(Agent, ItemIndex)
            } else {
                Print("ERR: SetCarShopItemState: Unknown ItemIndex: {ItemIndex}")
            }
        }
    }

    <#
    UnequipCurCosmetic(): void = {
        Print("PlayerData - UnequipCosmetic")
        GameManager := GetGameManager()
        Persistence := GetSaveManager()
        if(Agent := MaybePlayer?) {
            # Check if we have to remove the old cosmetic
            if(CurCosmeticId >= 0) {
                if(CosmeticVfxRemover := GameManager.CosmeticVfxRemoverList[CurCosmeticId]) {
                    # We despawn the old cosmetic by calling a removal pickup
                    CosmeticVfxRemover.Pickup(Agent)
                    SetCosmeticShopItemState(CurCosmeticId, 1)
                    set CurCosmeticId = -1
                    # Remove the save. Trackers start at 0
                    if(Save := Persistence.ProxySaves[Agent]) {
                        set Save.CurrentlyEquippedCosmetic = 0
                    }
                    <#
                    GameManager.CurrentlyEquippedCosmeticTracker.SetValue(Agent, 0)
                    # GameManager.CurrentlyEquippedCosmeticTracker.Save(Agent)
                    #>
                } else {
                    Print("ERR: UnequipCurCosmetic - Found no CosmeticVfxRemover for CurCosmeticId: {CurCosmeticId}")
                }
            }
        }
    }

    EquipCosmetic(NewCosmeticId: int): void = {
        Print("PlayerData - EquipCosmetic: NewCosmeticId: {NewCosmeticId}")
        GameManager := GetGameManager()
        Persistence := GetSaveManager()
        if(Agent := MaybePlayer?) {
            # Equip the new cosmetic
            if(NewCosmeticVfx := GameManager.CosmeticVfxList[NewCosmeticId]) {
                NewCosmeticVfx.Pickup(Agent)
            } else {
                Print("ERR: EquipCosmetic: Found no CosmeticVfx for ID: {NewCosmeticId}")
            }
            set CurCosmeticId = NewCosmeticId
            SetCosmeticShopItemState(NewCosmeticId, 2)
            # Save the currently equipped cosmetic
            # Trackers start at 0, so we have to add 1 to IDs
            if(Save := Persistence.ProxySaves[Agent]) {
                set Save.CurrentlyEquippedCosmetic = NewCosmeticId + 1
            }
            <#
            GameManager.CurrentlyEquippedCosmeticTracker.SetValue(Agent, NewCosmeticId + 1)
            # GameManager.CurrentlyEquippedCosmeticTracker.Save(Agent)
            #>
        }
    }
    #>

    SetAutomaticIncomeLevel(ParmIncomeLevel: int): void = {
        Print("SetAutomaticIncomeLevel: ParmIncomeLevel {ParmIncomeLevel}")
        if {
            ParmIncomeLevel > RebirthAutomaticIncomeLevel
        } then {
            set RebirthAutomaticIncomeLevel = ParmIncomeLevel
        }
    }

    LoadXp(OldXp: int): void = {
        # Print("LoadXp")
        GameManager := GetGameManager()
        Utility := GetUtility()
        if {
            Player := MaybePlayer?
        } then {
            set TotalXp = OldXp
            UpdateLevelBar(true)
            # Set health again after the player has spawned
            SetHealth(Player)
            SetShield(Player)
            Task := spawn{SetHealthAndShieldSuspends(Player)}
        }
    }

    LoadRebirths(OldRebirths: int): void = {
        GameManager := GetGameManager()
        Utility := GetUtility()
        if {
            Player := MaybePlayer?
        } then {
            set NumOfRebirths = OldRebirths
            UpdateRebirthsUi()
        }
    }

    LoadGamePass(ParmGamePassCoins: int, OldGamePassLevel: int): void = {
        set CurGamePassCoins = ParmGamePassCoins
        # Load game pass level
        var CurrentLevel: int = -1
        loop {
            if(OldGamePassLevel > CurrentLevel) {
                set CurrentLevel = CurrentLevel + 1
                # This also reclaims the award
                SetGamePassItemState(CurrentLevel, 2)
            } else {
                break
            }
        }
    }

    AddXp<public>(ParmAmount: int, ShowOnHud: logic): void = {
        GameManager := GetGameManager()
        SaveManager := GetSaveManager()
        Utility := GetUtility()
        if {
            ParmAmount > 0
            Player := MaybePlayer?
            Save := SaveManager.ProxySaves[Player]
        } then {
            Print("AddXp: ParmAmount: {ParmAmount}")
            var AmountToAdd: int = ParmAmount
            var XpMultiplier: float = CurXpMultiplier
            if {
                XpBoostLevel >= 0
                XpBoostFactor := XpBoostFactorList[XpBoostLevel]
            } then {
                set XpMultiplier = XpMultiplier + XpBoostFactor
            }
            if(Temp := Round[AmountToAdd * XpMultiplier]) {
                set AmountToAdd = Temp
            }
            set TotalXp = TotalXp + AmountToAdd
            set Save.TotalXp = TotalXp
            UpdateLevelBar(false)
            if(ShowOnHud = true) {
                # GameManager.PlayerXpGainedHud.Show(Player, Utility.FormatMessage("+{ToString(AmountToAdd)}"))
                PlayerLocalUi.AddFloatingMessage(AmountToAdd, 0)
            }
        }
    }

    UpdateMoneyMultiplier(): void = {
        GameConfig := GetGameConfig()
        set CurMoneyMultiplier = 1.0 + (GameConfig.RebirthIncomeMultiplier * NumOfRebirths) # + PetMoneyMultiplier + CosmeticMoneyMultiplier + ParkourMoneyMultiplier + PcBuffMoneyMultiplier
        PlayerLocalUi.UpdateMoneyMultiplier(CurMoneyMultiplier)
    }

    UpdateFansMultiplier(): void = {
        GameConfig := GetGameConfig()
        set CurFansMultiplier = 1.0 + (GameConfig.RebirthIncomeMultiplier * NumOfRebirths) # + PetFansMultiplier + CosmeticFansMultiplier
    }

    UpdateDiamondsMultiplier(): void = {
        GameConfig := GetGameConfig()
        set CurDiamondsMultiplier = 1.0 + (GameConfig.RebirthIncomeMultiplier * NumOfRebirths) # + PetDiamondsMultiplier + CosmeticDiamondsMultiplier
    }

    UpdateXpMultiplier(): void = {
        GameConfig := GetGameConfig()
        set CurXpMultiplier = 1.0 + (GameConfig.RebirthXpMultiplier * NumOfRebirths)
    }

    IncreaseKillStreak(Agent: agent, AmountToAdd: int): void = {
        GameManager := GetGameManager()
        SaveManager := GetSaveManager()
        GameConfig := GetGameConfig()
        set TimestampLastKill = GetSimulationElapsedTime()
        set KillStreakMultiplier = KillStreakMultiplier + AmountToAdd
        # -2, since the first sound with index 0 plays at 2 kills
        KillStreakSoundId := Min(KillStreakMultiplier - 2, GameManager.KillStreakSoundList.Length - 1)
        if {
            KillStreakSound := GameManager.KillStreakSoundList[KillStreakSoundId]
        } then {
            KillStreakSound.Play(Agent)
        }
    }

    UpdateXpHud(Agent: agent, AmountToAdd: int): void = {
        Utility := GetUtility()
        if {
            XpHud := MaybeXpHud?
        } then {
            set CurXpHudAmount = CurXpHudAmount + AmountToAdd
            XpHud.Show(Agent, Utility.FormatMessage("XP: {ToString(CurXpHudAmount)}"))
            set TimestampLastXpGained = GetSimulationElapsedTime()
            # Spawn routine to rest the XP hud amount, if it's not already running
            if(Routine := RoutineXpHudAmount?) {
                # Do nothing, routine is already running
            } else {
                MyRoutine := SpawnRoutine(XpHudTick)
                set RoutineXpHudAmount = option{MyRoutine}
            }
        }
    }

    XpHudTick()<suspends>: void = {
        loop {
            Timestamp := GetSimulationElapsedTime()
            TimePassedSinceEvent := Timestamp - TimestampLastXpGained
            TimeRemainingForEvent := XpHudDuration - TimePassedSinceEvent
            if(TimeRemainingForEvent <= 0.0) {
                break
            } else {
                Sleep(TimeRemainingForEvent)
            }
        }

        # Set back to 0
        set CurXpHudAmount = 0

        # Clean up routine
        if(Routine := RoutineXpHudAmount?) {
            set RoutineXpHudAmount = false
            Routine.Cancel()
        }
    }

    UpdatePetIncomeMultiplier(): void = {
        if {
            TycoonManager := MaybeTycoonManager?
        } then {
            var TotalMoneyMultiplier: float = 0.0
            var TotalFansMultiplier: float = 0.0
            var TotalDiamondsMultiplier: float = 0.0
            var TotalRecruitChanceMultiplier: float = 0.0
            for(Pet: TycoonManager.CurPetsList) {
                set TotalMoneyMultiplier = TotalMoneyMultiplier + Pet.PetMultiplierMoney
                set TotalFansMultiplier = TotalFansMultiplier + Pet.PetMultiplierFans
                set TotalDiamondsMultiplier = TotalDiamondsMultiplier + Pet.PetMultiplierDiamonds
                set TotalRecruitChanceMultiplier = TotalRecruitChanceMultiplier + Pet.PetMultiplierRecruitChance
            }
            set PetMoneyMultiplier = TotalMoneyMultiplier
            set PetFansMultiplier = TotalFansMultiplier
            set PetDiamondsMultiplier = TotalDiamondsMultiplier

            UpdateMoneyMultiplier()
            UpdateFansMultiplier()
            UpdateDiamondsMultiplier()
        }
    }

    UpdateCosmeticIncomeMultiplier(CosmeticButton: _Tycoon.tycoon_button_cosmetic): void = {
        Print("UpdateCosmeticIncomeMultiplier")
        if {
            TycoonManager := MaybeTycoonManager?
        } then {
            # Reset cosmetic multipliers
            set CosmeticMoneyMultiplier = 0.0
            set CosmeticFansMultiplier = 0.0
            set CosmeticDiamondsMultiplier = 0.0
            # Apply new cosmetic multiplier
            set CosmeticMoneyMultiplier = CosmeticButton.CosmeticMoneyMultiplier
            set CosmeticFansMultiplier = CosmeticButton.CosmeticFansMultiplier
            set CosmeticDiamondsMultiplier = CosmeticButton.CosmeticDiamondsMultiplier
            # Recalculate multipliers
            UpdateMoneyMultiplier()
            UpdateFansMultiplier()
            UpdateDiamondsMultiplier()
        }
    }

    UpdateLevelBar(IsLoadingSavedXp: logic): void = {
        Print("PlayerData: UpdateLevelBar")
        GameManager := GetGameManager()
        if {
            Agent := MaybePlayer?
        } then {
            OldLevel := CurLevel
            NewLevel := PlayerLocalUi.UpdateXpUi(TotalXp)

            # Update CurLevel in PlayerData - This must be first
            set CurLevel = NewLevel

            if(IsLoadingSavedXp = false) {
                # Check if we levelled up
                if(OldLevel <> NewLevel) {
                    HandleLevelUp(Agent, OldLevel, NewLevel)
                }
            }
        }
    }

    HandleLevelUp(Agent: agent, OldLevel: int, NewLevel: int): void = {
        SaveManager := GetSaveManager()
        MessageManager := GetMessageManager()
        AccoladeManager := GetAccoladeManager()
        PlayerLocalUi.HandleLevelUpUiMain()
        MessageManager.AddMessage(Agent, "has reached Level {NewLevel + 1}")

        # Check Level Up Achievements
        <# Levels start with 0 in Verse, but with 1 in game
        if(OldLevel < 9 and NewLevel >= 9) {
            SaveManager.GrantAchievement(Agent, SaveManager.ACHIEVEMENT_LEVEL_10)
            SaveManager.GrantAnalytic(Agent, SaveManager.ANALYTIC_LEVEL_10)
        } else if(OldLevel < 19 and NewLevel >= 19) {
            SaveManager.GrantAnalytic(Agent, SaveManager.ANALYTIC_LEVEL_20)
        } else if(OldLevel < 24 and NewLevel >= 24) {
            SaveManager.GrantAchievement(Agent, SaveManager.ACHIEVEMENT_LEVEL_25)
        } else if(OldLevel < 29 and NewLevel >= 29) {
            SaveManager.GrantAnalytic(Agent, SaveManager.ANALYTIC_LEVEL_30)
        } else if(OldLevel < 39 and NewLevel >= 39) {
            SaveManager.GrantAnalytic(Agent, SaveManager.ANALYTIC_LEVEL_40)
        } else if(OldLevel < 49 and NewLevel >= 49) {
            SaveManager.GrantAchievement(Agent, SaveManager.ACHIEVEMENT_LEVEL_50)
            SaveManager.GrantAnalytic(Agent, SaveManager.ANALYTIC_LEVEL_50)
        } else if(OldLevel < 59 and NewLevel >= 59) {
            SaveManager.GrantAnalytic(Agent, SaveManager.ANALYTIC_LEVEL_60)
        } else if(OldLevel < 69 and NewLevel >= 69) {
            SaveManager.GrantAnalytic(Agent, SaveManager.ANALYTIC_LEVEL_70)
        } else if(OldLevel < 74 and NewLevel >= 74) {
            SaveManager.GrantAchievement(Agent, SaveManager.ACHIEVEMENT_LEVEL_75)
        } else if(OldLevel < 79 and NewLevel >= 79) {
            SaveManager.GrantAnalytic(Agent, SaveManager.ANALYTIC_LEVEL_80)
        } else if(OldLevel < 89 and NewLevel >= 89) {
            SaveManager.GrantAnalytic(Agent, SaveManager.ANALYTIC_LEVEL_90)
        } else if(OldLevel < 99 and NewLevel >= 99) {
            SaveManager.GrantAnalytic(Agent, SaveManager.ANALYTIC_LEVEL_100)
        }
        
        SaveManager.GrantAccolade(Agent, SaveManager.ACCOLADE_LEVEL_UP)
        #>
    
        AccoladeManager.AwardAccolade(Agent, 3)

        Task := spawn{HandleLevelUpSuspends()}
    }

    HandleLevelUpSuspends()<suspends>: void = {
        GameManager := GetGameManager()
        Utility := GetUtility()
        if {
            Agent := MaybePlayer?
        } then {
            GameManager.LevelUpVfx.Pickup(Agent)
            GameManager.PlayerLevelUpHud.SetText(Utility.FormatMessage("LEVEL " + ToString(CurLevel + 1)))
            GameManager.PlayerLevelUpHud.Show(Agent)
            # Fully heal the player
            SetHealth(Agent)
            SetShield(Agent)
            # Update tycoon level billboard
            if(TycoonManager := MaybeTycoonManager?) {
                TycoonManager.UpdateOwnerLevelBillboard()
            }
        }
        
        Sleep(5.0)

        if {
            Agent := MaybePlayer?
            FortChar := Agent.GetFortCharacter[]
        } then {
            GameManager.LevelUpVfxRemover.Pickup(Agent)
        }
    }

    OnHeroPowerupBonusCoinCollected(SpawnTrans: transform): void = {
        GameConfig := GetGameConfig()
        AddCreditCardCoin(1)
        AddXp(GameConfig.HeroPowerupBonusCoinXpOnPickup, true)
        Task := spawn{OnHeroPowerupBonusCoinCollectedSuspends(SpawnTrans)}
    }

    OnHeroPowerupBonusCoinCollectedSuspends(SpawnTrans: transform)<suspends>: void = {
        GameConfig := GetGameConfig()
        GameManager := GetGameManager()
        # Spawn Prop with sound and VFX
        SpawnTuple := SpawnProp(GameManager.BpCreditCardCoinCollected, SpawnTrans)
        if(SpawnedProp := SpawnTuple(0)?) {
            Sleep(GameConfig.HeroPowerupBonusCoinCollectedVfxBpDuration)

            if(SpawnedProp.IsValid[]) {
                SpawnedProp.Dispose()
            }
        } else {
            Print("ERR: OnHeroPowerupBonusCoinCollectedSuspends: Couldn't spawn a prop")
        }
    }

    PlayCinematic(Cinematic: cinematic_sequence_device): void = {
        if {
            Agent := MaybePlayer?
        } then {
            # Cancel the previous cinematic (there can only be one active)
            StopCurrentCinematic()
            # Start the new cinematic
            Cinematic.Play(Agent)
            set MaybeCurCinematic = option{Cinematic}
        }
    }

    StopCurrentCinematic(): void = {
        if {
            Agent := MaybePlayer?
            Cinematic := MaybeCurCinematic?
        } then {
            Cinematic.Stop(Agent)
            set MaybeCurCinematic = false
        }
    }

    <# TODO: Outdated methods from RPG. Might be used for quests. Merge with new PlayCinematic?
    ActivateCinematicStasis(CinematicList: []cinematic_sequence_device, CinematicDuration: float)<suspends>: void = {
        Print("ActivateCinematicStasis")
        GameManager := GetGameManager()
        # IntroClassIndex := GameManager.GetFreeCinematicClassIndex()
        IntroClassIndex := 0
        if {
            Agent := MaybePlayer?
            FortChar := Agent.GetFortCharacter[]
            Cinematic := CinematicList[IntroClassIndex]
            IntroClass := GameManager.IntroClassSelectorList[IntroClassIndex]
        } then {
            IntroClass.ChangeClass(Agent)
            Cinematic.Play(Agent)
            # Stasis disabled since the Cinematic doesn't play sometimes 
            # FortChar.PutInStasis(stasis_args{AllowFalling := true})
            PlayerLocalUi.RemoveUiFromPlayer(Agent)
            set MaybeCurCinematic = option{Cinematic}
            # Wait for the cinematic to end
            Sleep(CinematicDuration)
            EndCinematicStasis()
        }
    }

    EndCinematicStasis(): void = {
        Print("EndCinematicStasis")
        GameManager := GetGameManager()
        if {
            Agent := MaybePlayer?
            FortChar := Agent.GetFortCharacter[]
            Cinematic := MaybeCurCinematic?
        } then {
            Cinematic.Stop(Agent)
            FortChar.ReleaseFromStasis()
            GameManager.NoClassClassSelector.ChangeClass(Agent)
            ShowPlayerLocalUi(Agent)
            # Clean up variables
            set MaybeCurCinematic = false
        }
    }
        #>
    
    OnPlayerHealthDamaged(DamageResult: damage_result): void = {
        Print("PlayerData - OnPlayerHealthDamaged")
        if {
            Agent := MaybePlayer?
        } then {
            # PlayerLocalUi.UpdateHealthBar(Agent)
        }

        <# Not in use atm - Boss Zombies extra damage
        <# We can't DamageResult.Source? since it is always set for fortnite zombies
        if(DamageResult.Source?) {
            # This is necessary to prevent the endless loop from causing damage via script
            Print("DamageResult.Source IS SET!!")
        }
        #>
        if {
            # Check if in bossfight
            BossManager := MaybeBossManager?
            # Check if we were damaged by a zombie
            Instigator := DamageResult.Instigator?
            DamagingPlayer := Instigator.GetInstigatorAgent[]
            DamagingFortChar := DamagingPlayer.GetFortCharacter[]
            # Check if it's a zombie and not a player that is causing the damage
            DamagingFortChar.GetMaxShield() = 0.0
            # Get the current zombie bonus damage
            ZombieBonusDamage := BossManager.ZombieBonusDamageList[BossManager.BossDifficulty]
            # Get the owner agent
            OwnerAgent := MaybePlayer?
            OwnerFortChar := OwnerAgent.GetFortCharacter[]
            OwnerFortChar.IsActive[]
        } then {
            Print("DamagingFortChar.GetMaxShield(): {DamagingFortChar.GetMaxShield()}")
            Print("OwnerFortChar.GetMaxShield(): {OwnerFortChar.GetMaxShield()}")

            Print("OnPlayerDamaged - Damaged by bossfight zombie! Deal Bonus Damage: {ZombieBonusDamage}")
            OwnerFortChar.Damage(damage_args{Amount := ZombieBonusDamage})
        }
        #>
    }

    OnPlayerShieldDamaged(DamageResult: damage_result): void = {
        if {
            Agent := MaybePlayer?
        } then {
            # PlayerLocalUi.UpdateHealthBar(Agent)
        }
    }

    OnPlayerHealthHealed(HealingResult: healing_result): void = {
        if {
            Agent := MaybePlayer?
        } then {
            # PlayerLocalUi.UpdateHealthBar(Agent)
        }
    }

    OnPlayerShieldHealed(HealingResult: healing_result): void = {
        if {
            Agent := MaybePlayer?
        } then {
            # PlayerLocalUi.UpdateHealthBar(Agent)
        }
    }

    OnPlayerEliminated(EliminationResult: elimination_result): void = {
        Print("PlayerData - OnPlayerEliminated")
        GameManager := GetGameManager()
        SaveManager := GetSaveManager()
        if {
            Player := MaybePlayer?
        } then {
            # Stop ui ticks - This also removes the UIs from ActiveUiList
            UiChangelog.RemoveUiFromPlayer(Player)
            # UiAcceptQuest.RemoveUiFromPlayer(Player)
            # UiAcceptQuestButtons.RemoveUiFromPlayer(Player)
            UiScamNpc.RemoveUiFromPlayer(Player)
            UiElevator.RemoveUiFromPlayer(Player)
            # Remove all other UIs from the player - UIs with mouse cursor enabled crash Verse UIs on Respawn
            Print("OnPlayerEliminated: ActiveUiList.Length: {ActiveUiList.Length}")
            if {
                PlayerUI := GetPlayerUI[Player]
            } then {
                for(ActiveUi: ActiveUiList) {
                    ActiveUi.RemoveUiFromPlayer(Player)
                }
            }
            set ActiveUiList = map {}
            # Stop the current cinematic (if there is one)
            StopCurrentCinematic()
            # Reactivate tycoon minigame buttons
            for(Button: ButtonListEnableOnDeath) {
                Button.Enable()
            }
            <#
            # Disable no dragon mount flag since player respawns in the town
            set IsInNoDragonMountArea = false
            # Check for current cinematic
            if {
                Cinematic := MaybeCurCinematic?
            } then {
                EndCinematicStasis()
            }
            # Check if in bossfight
            if(BossManager := MaybeBossManager?) {
                BossManager.OnPlayerEliminatedDuringBossfight(Agent, Self)
            }
            # Handle async tasks
            Task := spawn{OnPlayerEliminatedSuspends(Agent)}
            #>
        }
        # Handle killer rewards
        if {
            KillerAgent := EliminationResult.EliminatingCharacter?.GetAgent[]
            # Check if the killer was a player
            KillerPlayerData := KillerAgent.GetPlayerData[]
            # Only grant rewards if this player was killed by a different player
            KillerPlayerData <> Self
            KillerFortChar := KillerAgent.GetFortCharacter[]
            TycoonManager := MaybeTycoonManager?
        } then {
            # Fully heal the killer player
            KillerFortChar.SetHealth(KillerFortChar.GetMaxHealth())
            KillerFortChar.SetShield(KillerFortChar.GetMaxShield())
            # Grant XP reward equal to the highest mask level of the killer
            var KillXpReward: int = TycoonManager.TycoonConfig.NpcSpawnerManager.CalculateXpForKill(KillerAgent, Min(TotalScamLevel, 9))
            KillerPlayerData.AddXp(KillXpReward, true)
            # Achievement: Get 5 Player Kills
            SaveManager.IncrementAchievement(KillerAgent, _Global.achievement_type.Kill, 5, 1)
        }
    }

    RegisterButtonEnableOnDeath(NewButton: button_device): void = {
        # Add button to the list
        set ButtonListEnableOnDeath = ButtonListEnableOnDeath + array {NewButton}
    }

    RestoreTycoonButtonsOwned()<suspends>: void = {
        SaveManager := GetSaveManager()
        if {
            TycoonManager := MaybeTycoonManager?
        } then {
            if {
                Player := MaybePlayer?
                Save := SaveManager.ProxySaves[Player]
            } then {
                for(TycoonButtonId: Save.TycoonButtonsOwned) {
                    if(TycoonButton := TycoonManager.TycoonButtons[TycoonButtonId]) {
                        if(TycoonButton.ButtonId = Save.CurQuestTycoonButtonId) {
                            # Reassign the quest
                            TycoonButton.OnPropBought(false)
                        } else {
                            # Tell TycoonButton that we are loading SaveData
                            TycoonButton.OnPropBought(true)
                        }
                        Sleep(0.1)
                    } else {
                        Print("ERR: RestoreTycoonButtonsOwned: Found no Tycoon Button for ID: {TycoonButtonId}")
                    }
                }
            } else {
                Print("WARN: RestoreTycoonButtonsOwned - Found no save data")
            }
            # CleanUpActiveButtonList is now included in each UpdateButtonStates() call
            # TycoonManager.CleanUpActiveButtonList()
            # Check if we have to show new buttons
            TycoonManager.UpdateButtonStates()
            Print("UpdateButtonStates 222")
        }
    }

    <#
    RestoreTycoonWallColor(): void = {
        SaveManager := GetSaveManager()
        if {
            Player := MaybePlayer?
            TycoonManager := MaybeTycoonManager?
            Save := SaveManager.ProxySaves[Player]
        } then {
            # Restore shop states
            if(Save.HasUnlockedColor0 = true) {
                if(set ColorShopStateList[0] = 1) {}
            }
            if(Save.HasUnlockedColor1 = true) {
                if(set ColorShopStateList[1] = 1) {}
            }
            if(Save.HasUnlockedColor2 = true) {
                if(set ColorShopStateList[2] = 1) {}
            }
            if(Save.HasUnlockedColor3 = true) {
                if(set ColorShopStateList[3] = 1) {}
            }
            if(Save.HasUnlockedColor4 = true) {
                if(set ColorShopStateList[4] = 1) {}
            }
            if(Save.HasUnlockedColor5 = true) {
                if(set ColorShopStateList[5] = 1) {}
            }
            if(Save.HasUnlockedColor6 = true) {
                if(set ColorShopStateList[6] = 1) {}
            }
            if(Save.HasUnlockedColor7 = true) {
                if(set ColorShopStateList[7] = 1) {}
            }
            if(Save.HasUnlockedColor8 = true) {
                if(set ColorShopStateList[8] = 1) {}
            }
            if(Save.HasUnlockedColor9 = true) {
                if(set ColorShopStateList[9] = 1) {}
            }
            # Restore the wall color
            TycoonManager.ColorShopDevice.ChangeWallColor(Player, Save.WallColorId)
        }
    }
    #>

    <#
    RestorePcUpgradesCase(): void = {
        SaveManager := GetSaveManager()
        if {
            Player := MaybePlayer?
            TycoonManager := MaybeTycoonManager?
            Save := SaveManager.ProxySaves[Player]
        } then {
            # Restore shop states
            if(Save.HasUnlockedPcCase0 = true) {
                if(set PcCaseShopStateList[0] = 1) {}
            }
            if(Save.HasUnlockedPcCase1 = true) {
                if(set PcCaseShopStateList[1] = 1) {}
            }
            if(Save.HasUnlockedPcCase2 = true) {
                if(set PcCaseShopStateList[2] = 1) {}
            }
            # Restore the equipped Pc Case
            TycoonManager.PcButtonDevice.UpgradesShopPcCase.EquipPcCase(Player, Save.PcCaseId)
        }
    }

    RestorePcUpgradesDisplay(): void = {
        SaveManager := GetSaveManager()
        if {
            Player := MaybePlayer?
            TycoonManager := MaybeTycoonManager?
            Save := SaveManager.ProxySaves[Player]
        } then {
            # Restore shop states
            if(Save.HasUnlockedPcDisplay0 = true) {
                if(set PcDisplayShopStateList[0] = 1) {}
            }
            if(Save.HasUnlockedPcDisplay1 = true) {
                if(set PcDisplayShopStateList[1] = 1) {}
            }
            if(Save.HasUnlockedPcDisplay2 = true) {
                if(set PcDisplayShopStateList[2] = 1) {}
            }
            # Restore the equipped Pc Display
            TycoonManager.PcButtonDevice.UpgradesShopPcDisplay.EquipPcDisplay(Player, Save.PcDisplayId)
        }
    }

    RestorePcUpgradesDesk(): void = {
        SaveManager := GetSaveManager()
        if {
            Player := MaybePlayer?
            TycoonManager := MaybeTycoonManager?
            Save := SaveManager.ProxySaves[Player]
        } then {
            # Restore shop states
            if(Save.HasUnlockedPcDesk0 = true) {
                if(set PcDeskShopStateList[0] = 1) {}
            }
            if(Save.HasUnlockedPcDesk1 = true) {
                if(set PcDeskShopStateList[1] = 1) {}
            }
            if(Save.HasUnlockedPcDesk2 = true) {
                if(set PcDeskShopStateList[2] = 1) {}
            }
            # Restore the equipped Pc Desk
            TycoonManager.PcButtonDevice.UpgradesShopPcDesk.EquipPcDesk(Player, Save.PcDeskId)
        }
    }
    #>

    DoRebirth(): void = {
        QuestManager := GetQuestManager()
        SaveManager := GetSaveManager()
        if {
            Player := MaybePlayer?
            FortChar := Player.GetFortCharacter[]
            Save := SaveManager.ProxySaves[Player]
        } then {
            # Check for analytic
            if(NumOfRebirths = 0) {
                SaveManager.GrantAnalytic(Player, SaveManager.ANALYTIC_FIRST_REBIRTH)
            }
            # Increase rebirth number
            UpdateCurRebirths(NumOfRebirths + 1)
            # Increase Achievement progress
            # Use Rebirth
            SaveManager.IncrementAchievement(Player, _Global.achievement_type.Rebirth, 0, 1)
            # Reach Rebirth 2
            if(NumOfRebirths >= 2) {
                SaveManager.IncrementAchievement(Player, _Global.achievement_type.Rebirth, 1, 1)
            }
            # Reach Rebirth 5
            if(NumOfRebirths >= 5) {
                SaveManager.IncrementAchievement(Player, _Global.achievement_type.Rebirth, 3, 1)
            }
            # Reach Rebirth 7
            if(NumOfRebirths >= 7) {
                SaveManager.IncrementAchievement(Player, _Global.achievement_type.Rebirth, 5, 1)
            }
            # Reach Rebirth 10
            if(NumOfRebirths >= 10) {
                SaveManager.IncrementAchievement(Player, _Global.achievement_type.Rebirth, 7, 1)
            }
            # Reach Rebirth 15
            if(NumOfRebirths >= 15) {
                SaveManager.IncrementAchievement(Player, _Global.achievement_type.Rebirth, 9, 1)
            }
            # Reset some values
            set CurrentQuestId = -1
            set HighestElevatorFloorIdUnlocked = -1
            set HighestFbiKeyUnlocked = -1
            set HasUnlockedGarage = false
            set HasUnlockedTycoonAutoSave = false
            UpdateCurMoney(0)
            UpdateCurFans(0.0)
            UpdateCurDiamonds(0.0)
            UpdateCurCreditCardCoin(0)
            UpdateResourceUi()
            UpdateRebirthsUi()
            UpdateMoneyMultiplier()
            UpdateFansMultiplier()
            UpdateDiamondsMultiplier()
            UpdateXpMultiplier()
            # Check if the player had an active quest
            if(PrevQuest := MaybeActiveQuest?) {
                QuestManager.CleanUpQuestData(Player, PrevQuest)
            }
            # Reset playtime rewards
            for(ItemIndex -> PlaytimeReward: PlaytimeRewardStateList) {
                SetPlaytimeRewardItemState(ItemIndex, 0)
            }
            <# Reset game pass
            for(ItemIndex -> GamePassItem: GamePassStateList) {
                SetGamePassItemState(ItemIndex, 0)
            }
            #>
            # Reset city owner states
            for(ItemIndex -> CityOwnerState: CityOwnerStateList) {
                SetCityOwnerItemState(ItemIndex, 0)
            }
            # Reset PC Shop values
            for(ItemIndex -> ListItem: PcCaseShopStateList) {
                SetUpgradesPcCaseShopItemState(ItemIndex, 0)
            }
            for(ItemIndex -> ListItem: PcDisplayShopStateList) {
                SetUpgradesPcDisplayShopItemState(ItemIndex, 0)
            }
            for(ItemIndex -> ListItem: PcDeskShopStateList) {
                SetUpgradesPcDeskShopItemState(ItemIndex, 0)
            }
            # Save rebirth + clear Tycoon Buttons save data
            set Save.TycoonButtonsOwned = array{}
            set Save.TycoonButtonsOwned = array{}
            set Save.CurQuestTycoonButtonId = -1
            set Save.CarShopStateList = array {
                0,
                0,
                0,
                0,
                0,
                0,
                0
            }
            # Reset some upgrades from permanent upgrade shop
            set Save.UpgradeCoinsPerKillLevel = 0
            set Save.UpgradeMoneyScamLevel = 0
            # Perform a manual save after each rebirth
            SaveManager.PersistPlayerSave(Player)
            # Respawn the player
            FortChar.Damage(9999999999999999.0)
        }
    }

    SetPlaytimeRewardItemState(ItemIndex: int, NewState: int): void = {
        if(set PlaytimeRewardStateList[ItemIndex] = NewState) {
            # Do nothing here
        } else {
            Print("ERR: SetPlaytimeRewardItemState - Found no StateList for ID: {ItemIndex}")
        }
    }

    SetGamePassItemState(ItemIndex: int, NewState: int): void = {
        Print("SetGamePassItemState: ItemIndex: {ItemIndex}, NewState: {NewState}")
        SaveManager := GetSaveManager()
        if {
            set GamePassStateList[ItemIndex] = NewState
            Player := MaybePlayer?
            PlayerData := Player.GetPlayerData[]
            Save := SaveManager.ProxySaves[Player]
        } then {
            # If we purchased an item --> Set the next item to "Can Buy"
            if {
                NewState = 2
                set GamePassStateList[ItemIndex + 1] = 1
            } then {
                Print("set Save.GamePassLevel = {ItemIndex}")
                # Update the game pass save
                set Save.GamePassLevel = ItemIndex
                # Give reward specific to ItemIndex
                <#
                Reward 0: Dauerhaft 50 XP jede 5 sekunden
                Reward 1: 1x 50.000 $
                Reward 2: 1 Free Rebirth
                Reward 3: Dauerhaft 250 XP (+200)
                Reward 4: 2 Free Rebirth
                Reward 5: Dauerhaft 500 XP (+250)
                Reward 6: 3 Free Rebirth
                Reward 7: VFX
                #>
                case(ItemIndex) {
                    <# TODO: Implement rewards
                    0 =>
                        set GamePassXpIncomePerSec = GamePassXpIncomePerSec + 50
                    1 =>
                        if {
                            Save.ClaimedGamePassMoney = false
                        } then {
                            set Save.ClaimedGamePassMoney = true
                            AddMoney(50000, true)
                        }
                    2 =>
                        if {
                            Save.ClaimedGamePassRebirths1 = false
                        } then {
                            set Save.ClaimedGamePassRebirths1 = true
                            UpdateCurRebirths(NumOfRebirths + 1)
                        }
                    3 =>
                        set GamePassXpIncomePerSec = GamePassXpIncomePerSec + 200
                    4 =>
                        if {
                            Save.ClaimedGamePassRebirths2 = false
                        } then {
                            set Save.ClaimedGamePassRebirths2 = true
                            UpdateCurRebirths(NumOfRebirths + 2)
                        }
                    5 =>
                        set GamePassXpIncomePerSec = GamePassXpIncomePerSec + 250
                    6 =>
                        if {
                            Save.ClaimedGamePassRebirths3 = false
                        } then {
                            set Save.ClaimedGamePassRebirths3 = true
                            UpdateCurRebirths(NumOfRebirths + 3)
                        }
                    7 =>
                        # Change item state so the player can equip the VFX
                        SetGamePassItemState(ItemIndex, 3)
                    #>
                    _ =>
                        Print("ERR: SetGamePassItemState: ItemIndex not found: {ItemIndex}")
                }
            }
        } else {
            Print("ERR: SetGamePassItemState - Found no StateList for ID: {ItemIndex}")
        }
    }

    GetColorShopItemState(ColorShopId: int, ItemIndex: int): int = {
        if {
            ShopItemsStateList := ColorShopStateList[ColorShopId]
            ItemState := ShopItemsStateList[ItemIndex]
        } then {
            return ItemState
        } else {
            Print("ERR: GetColorShopItemState: Found no state for ColorShopId: {ColorShopId}, ItemIndex: {ItemIndex}")
        }
        return -1
    }

    SetColorShopItemState(ColorShopId: int, ItemIndex: int, NewState: int): void = {
        if {
            var ShopItemsStateList: []int = ColorShopStateList[ColorShopId]
            set ShopItemsStateList[ItemIndex] = NewState
            # Must also update the main list
            set ColorShopStateList[ColorShopId] = ShopItemsStateList
        } then {
            # Do nothing here
        } else {
            Print("ERR: SetColorShopItemState - Found no StateList for ColorShopId: {ColorShopId}, ItemIndex: {ItemIndex}")
        }
    }

    SetUpgradesPcCaseShopItemState(ItemIndex: int, NewState: int): void = {
        if(set PcCaseShopStateList[ItemIndex] = NewState) {
            # Do nothing here
        } else {
            Print("ERR: SetUpgradesPcShopItemState - Found no StateList for ID: {ItemIndex}")
        }
    }

    SetUpgradesPcDisplayShopItemState(ItemIndex: int, NewState: int): void = {
        if(set PcDisplayShopStateList[ItemIndex] = NewState) {
            # Do nothing here
        } else {
            Print("ERR: SetUpgradesDisplayShopItemState - Found no StateList for ID: {ItemIndex}")
        }
    }

    SetUpgradesPcDeskShopItemState(ItemIndex: int, NewState: int): void = {
        if(set PcDeskShopStateList[ItemIndex] = NewState) {
            # Do nothing here
        } else {
            Print("ERR: SetUpgradesDeskShopItemState - Found no StateList for ID: {ItemIndex}")
        }
    }

    SetCityOwnerItemState(ItemIndex: int, NewState: int): void = {
        if(set CityOwnerStateList[ItemIndex] = NewState) {
            # Do nothing here
        } else {
            Print("ERR: SetCityOwnerItemState - Found no StateList for ID: {ItemIndex}")
        }
    }

    ApplyTycoonCosmetic(TycoonButtonCosmetic: _Tycoon.tycoon_button_cosmetic): void = {
        Print("ApplyTycoonCosmetic")
        if {
            Player := MaybePlayer?
        } then {
            # Check if we already have an active cosmetic and remove it
            RemoveCurrentCosmetic(TycoonButtonCosmetic.CosmeticSocket, true)
            # Apply the new cosmetic
            for(CosmeticIndex -> CosmeticVfx: TycoonButtonCosmetic.CosmeticVfxList) {
                Print("ApplyTycoonCosmetic: CosmeticIndex: {CosmeticIndex}")
                CosmeticVfx.Pickup(Player)
            }
            # Update the list
            if(set CosmeticManagerList[TycoonButtonCosmetic.CosmeticSocket] = TycoonButtonCosmetic) {}
            # Update income multiplier
            UpdateCosmeticIncomeMultiplier(TycoonButtonCosmetic)
        }
    }

    RemoveCurrentCosmetic(CosmeticSocket: _Tycoon.cosmetic_socket, IsPermanentRemove: logic): void = {
        Print("RemoveCurrentCosmetic")
        if {
            Player := MaybePlayer?
            CosmeticManager := CosmeticManagerList[CosmeticSocket]
        } then {
            for(CosmeticIndex -> CosmeticRemoverVfx: CosmeticManager.CosmeticVfxRemoverList) {
                Print("RemoveCurrentCosmetic: CosmeticIndex: {CosmeticIndex}")
                CosmeticRemoverVfx.Pickup(Player)
            }
            if(IsPermanentRemove = true) {
                set CosmeticManagerList = CosmeticManagerList.RemoveKey(CosmeticSocket)
            }
        }
    }

    ReapplyAllCosmetics(): void = {
        Print("ReapplyAllCosmetics")
        for(CosmeticSocket -> CosmeticManager: CosmeticManagerList) {
            ApplyTycoonCosmetic(CosmeticManager)
        }
    }

    UnequipAllCosmetics(IsPermanentRemove: logic): void = {
        Print("UnequipAllCosmetic")
        for(CosmeticSocket -> CosmeticManager: CosmeticManagerList) {
            RemoveCurrentCosmetic(CosmeticSocket, IsPermanentRemove)
        }
        if(IsPermanentRemove = true) {
            SetScamLevelHead(0)
            SetScamLevelBack(0)
        }
    }

    OnParkourGoal(ParmMoneyMultiplier: float): void = {
        if(ParkourMoneyMultiplier > ParmMoneyMultiplier) {
            # The player already has a stronger multiplier active, do nothing
        } else if(ParkourMoneyMultiplier <= ParmMoneyMultiplier) {
            # Apply new multiplier
            set ParkourMoneyMultiplier = ParmMoneyMultiplier
            UpdateMoneyMultiplier()
            # Reset the multiplier duration
            UpdateParkourMultiplierTimeout()
        }
    }

    UpdateParkourMultiplierTimeout(): void = {
        set TimestampLastParkourCompleted = GetSimulationElapsedTime()
        if(Routine := RoutineParkourTimeoutTick?) {
            # Do nothing, routine is already running
        } else {
            MyRoutine := SpawnRoutine(ParkourMultiplierTimeoutTick)
            set RoutineParkourTimeoutTick = option{MyRoutine}
        }
    }

    OnPlayerParkourMultiplierTimeout(): void = {
        set ParkourMoneyMultiplier = 0.0
        UpdateMoneyMultiplier()
    }

    OnPcBuffActivated(BuffMoneyMultiplier: float): void = {
        Print("OnPcBuffActivated")
        set PcBuffMoneyMultiplier = BuffMoneyMultiplier
        UpdateMoneyMultiplier()
    }

    OnPcBuffTimeout(): void = {
        set PcBuffMoneyMultiplier = 0.0
        UpdateMoneyMultiplier()
    }

    ParkourMultiplierTimeoutTick()<suspends>: void = {
        loop {
            Timestamp := GetSimulationElapsedTime()
            TimePassedSinceEvent := Timestamp - TimestampLastParkourCompleted
            TimeRemainingForEvent := ParkourMultiplierDuration - TimePassedSinceEvent
            if(TimeRemainingForEvent <= 0.0) {
                break
            } else {
                Sleep(TimeRemainingForEvent)
            }
        }

        Print("ParkourMultiplierTimeoutTick - Timeout")
        OnPlayerParkourMultiplierTimeout()
    }

    
    CheckAchievementPetsMounts(): void = {
        Print("PlayerData: CheckAchievementPetsMounts")
        SaveManager := GetSaveManager()
        if(Agent := MaybePlayer?) {
            var FoundMissingItem: logic = false
            for(MountState: MountShopStateList) {
                if(MountState <= 0) {
                    set FoundMissingItem = true
                }
            }
            for(PetState: PetShopStateList) {
                if(PetState <= 0) {
                    set FoundMissingItem = true
                }
            }
            if(FoundMissingItem = false) {
                # SaveManager.GrantAchievement(Agent, SaveManager.ACHIEVEMENT_PETS_MOUNTS_ALL)
            }       
        }
    }

    SetScamLevelHead(NewLevel: int): void = {
        set ScamLevelHead = NewLevel
        UpdateScamLevel()
    }

    SetScamLevelBack(NewLevel: int): void = {
        set ScamLevelBack = NewLevel
        UpdateScamLevel()
    }

    UpdateScamLevel(): void = {
        set TotalScamLevel = ScamLevelHead + ScamLevelBack
        PlayerLocalUi.UpdateScamLevelUi(TotalScamLevel)
    }

    UpdateScamLevelUi(): void = {
        Utility := GetUtility()
        if {
            # Only update UI if we have a tycoon assigned
            TycoonManager := MaybeTycoonManager?
        } then {
            PlayerLocalUi.UpdateScamLevelUi(
                TotalScamLevel
            )
        }
    }

    <#
    TeleportToJail()<suspends>: void = {
        GameManager := GetGameManager()
        set IsInJail = true
        CurTimestamp := GetSimulationElapsedTime()
        set TimestampJailed = CurTimestamp
        if {
            Agent := MaybePlayer?
            TycoonManager := MaybeTycoonManager?
        } then {
            PlayCinematic(GameManager.CinematicJail)
            # Destroy the car so the player can get teleported to jail
            TycoonManager.CustomCarDevice.CarSpawner.DestroyVehicle()
            # The AgentExitsVehicleEvent is bugged so we have to manually call it
            TycoonManager.CustomCarDevice.OnCarExited(Agent)
        }

        Sleep(2.0)

        if {
            Agent := MaybePlayer?
        } then {
            GameManager.TeleporterJail.Teleport(Agent)
            GameManager.HudJail.Show(Agent)
            GameManager.TimerEscapeJail.Start(Agent)
        }

        Sleep(GameManager.AutoEscapeJailDuration)

        if {
            Agent := MaybePlayer?
            # Check if the player got jailed again in the meantime
            TimestampJailed = CurTimestamp
        } then {
            TryToEscapeFromJail(Agent)
        }
    }

    TryToEscapeFromJail(Agent: agent): void = {
        Print("TryToEscapeFromJail")
        if {
            IsInJail = true
            TycoonManager := MaybeTycoonManager?
        } then {
            set IsInJail = false
            Task := spawn{TycoonManager.PlayIntroCinematic()}
        }
    }
    #>

    HandleQuestProgressAgent(Agent: agent): void = {
        if {
            OwnerAgent := MaybePlayer?
            Agent = OwnerAgent
        } then {
            IncreaseQuestProgress()
        }
    }

    HandleQuestProgressAiInteraction(InteractionResult: device_ai_interaction_result): void = {
        Print("HandleQuestProgressAiInteraction")
        if {
            OwnerAgent := MaybePlayer?
            SourceAgent := InteractionResult.Source?
            SourceAgent = OwnerAgent
        } then {
            IncreaseQuestProgress()
        }
    }

    IncreaseQuestProgress(): void = {
        QuestManager := GetQuestManager()
        SaveManager := GetSaveManager()
        if {
            Agent := MaybePlayer?
            Quest := MaybeActiveQuest?
            Save := SaveManager.ProxySaves[Agent]
        } then {
            # This must be separate, so it's not rolled back
            set Save.QuestProgressCount = Save.QuestProgressCount + 1
            if {
                Save.QuestProgressCount >= Quest.QuestGoalCountAmount
            } then {
                QuestManager.HandleQuestCompleted(Agent, Quest)
            }
        }
    }

    StartDragonMount(DragonMountManager: dragon_mount_device, ChairIndex: int): void = {
        InputManager := GetInputManager()
        if(Agent := MaybePlayer?) {
            InputManager.DragonMountRegisterAgent(Agent)
            set MaybeDragonMountManager = option{DragonMountManager}
            set DragonMountIndex = ChairIndex
            MyRoutine := SpawnRoutine(DragonMountTick)
            set RoutineDragonMountTick = option{MyRoutine}
        }
    }

    StopDragonMount(): void = {
        InputManager := GetInputManager()
        set MaybeDragonMountManager = false
        if(Routine := RoutineDragonMountTick?) {
            Routine.Cancel()
        }
        set RoutineDragonMountTick = false
        if(Agent := MaybePlayer?) {
            InputManager.DragonMountUnregisterAgent(Agent)
        }
    }

    StartDragonMountBoost(DragonManager: dragon_mount_device)<suspends>: void = {
        # Check again to be safe that the task is not spawned twice
        if(DragonMountBoostState = dragon_mount_boost_state.Available) {
            set DragonMountBoostState = dragon_mount_boost_state.Active
            Sleep(DragonManager.BoostDuration)
            set DragonMountBoostState = dragon_mount_boost_state.Cooldown
            Sleep(DragonManager.BoostCooldown)
            set DragonMountBoostState = dragon_mount_boost_state.Available
        }
    }

    DragonMountTick()<suspends>: void = {
        GameManager := GetGameManager()
        InputManager := GetInputManager()
        Utility := GetUtility()
        loop {
            if {
                Agent := MaybePlayer?
                FortChar := Agent.GetFortCharacter[]
                DragonManager := MaybeDragonMountManager?
                ChairDevice := DragonManager.ChairDeviceList[DragonMountIndex]
            } then {
                CurPos := ChairDevice.GetTransform().Translation
                CurRot := ChairDevice.GetTransform().Rotation
                var GoalPos: vector3 = CurPos
                GoalRot := FortChar.GetViewRotation()

                var MovementRot: rotation = FortChar.GetViewRotation()
                var MovementSpeedAdj: float = DragonManager.TickRate * DragonManager.MovementSpeed
                if(DragonMountBoostState = dragon_mount_boost_state.Active) {
                    set MovementSpeedAdj = MovementSpeedAdj * DragonManager.MovementBoostRate
                }

                if(InputManager.InputSprint.IsHeld[Agent]) {
                    if(DragonMountBoostState = dragon_mount_boost_state.Available) {
                        Task := spawn{StartDragonMountBoost(DragonManager)}
                    }
                }

                # Movement could be improved by reducing the total distance if multiple inputs are held at the same time
                if(InputManager.InputForward.IsHeld[Agent]) {
                    # Print("InputForward.IsHeld")
                    RotatedMovementVector := MovementRot.RotateVector(Utility.ForwardVector * MovementSpeedAdj)
                    set GoalPos = GoalPos + RotatedMovementVector
                }
                if(InputManager.InputBackward.IsHeld[Agent]) {
                    # Print("InputBackward.IsHeld")
                    RotatedMovementVector := MovementRot.RotateVector(Utility.BackwardVector * MovementSpeedAdj)
                    set GoalPos = GoalPos + RotatedMovementVector
                }
                if(InputManager.InputLeft.IsHeld[Agent]) {
                    # Print("InputLeft.IsHeld")
                    RotatedMovementVector := MovementRot.RotateVector(Utility.LeftVector * MovementSpeedAdj)
                    set GoalPos = GoalPos + RotatedMovementVector
                }
                if(InputManager.InputRight.IsHeld[Agent]) {
                    # Print("InputRight.IsHeld")
                    RotatedMovementVector := MovementRot.RotateVector(Utility.RightVector * MovementSpeedAdj)
                    set GoalPos = GoalPos + RotatedMovementVector
                }
                if(InputManager.InputUp.IsHeld[Agent]) {
                    # Print("InputUp.IsHeld")
                    set GoalPos = GoalPos + (Utility.UpVector * MovementSpeedAdj)
                }
                if(InputManager.InputCrouch.IsHeld[Agent]) {
                    # Print("InputCrouch.IsHeld")
                    set GoalPos = GoalPos + (Utility.DownVector * MovementSpeedAdj)
                }

                # Teleport to rotate instantly
                # if(ChairDevice.TeleportTo[CurBoatPos, CurRot]) {}
                
                # MoveTo
                ChairDevice.MoveTo(GoalPos, GoalRot, DragonManager.TickRate)
            }
            Sleep(0.0)
        }
    }

    ServerTickSecond()<suspends>: void = {
        GameConfig := GetGameConfig()
        loop {
            # Update killstreak
            if {
                TimestampLastKill > 0.0
            } then {
                Timestamp := GetSimulationElapsedTime()
                TimePassedSinceKill := Timestamp - TimestampLastKill
                if {
                    TimePassedSinceKill > GameConfig.KillStreakDuration
                } then {
                    set TimestampLastKill = 0.0
                    set KillStreakMultiplier = 0
                    # GameManager.HudMainUiKillStreak.Hide(Player)
                }
            }
            Sleep(1.0)
        }
    }

    GamePassXpTick()<suspends>: void = {
        loop {
            AddXp(GamePassXpIncomePerSec, false)
            Sleep(GamePassXpTickRate)
        }
    }

    UpdateQuestProgressHud(Agent: agent, Quest: _Tycoon.tycoon_button): void = {
        # Print("UpdateQuestProgressHud")
        SaveManager := GetSaveManager()
        QuestManager := GetQuestManager()
        Utility := GetUtility()
        if {
            Save := SaveManager.ProxySaves[Agent]
        } then {
            var ProgressMessage: string = ""
            set ProgressMessage = ProgressMessage + Quest.QuestObjectiveText
            ProgressCount := Save.QuestProgressCount
            TargetToReach := Quest.QuestGoalCountAmount
            set ProgressMessage = ProgressMessage + " " + ToString(ProgressCount) + "/" + ToString(TargetToReach)
            QuestManager.PlayerQuestProgressHud.Show(Agent, Utility.FormatMessage(ProgressMessage))
        }
    }

    SetWeaponDamage(NewWeaponDamage: float): void = {
        set CurWeaponDamage = NewWeaponDamage
    }

    ProcessWeaponDamage(): float = {
        # TODO: Show UI which shows the damage number
        
        return CurWeaponDamage
    }

    UpdateQuestProgressTick()<suspends>: void = {
        loop {
            if {
                Agent := MaybePlayer?
                Quest := MaybeActiveQuest?
            } then {
                UpdateQuestProgressHud(Agent, Quest)
                Sleep(QuestProgressTickRate)
            } else {
                break
            }
        }
    }

    StartScammingNpc(Agent: agent): void = {
        set IsScammingNpc = true
        UiScamNpc.AddUiToPlayer(Agent, false)
    }

    StopScammingNpc(Agent: agent): void = {
        set IsScammingNpc = false
        UiScamNpc.RemoveUiFromPlayer(Agent)
    }

    OnElevatorFloorUnlocked(Agent: agent, FloorId: int): void = {
        if {
            FloorId > HighestElevatorFloorIdUnlocked
        } then {
            set HighestElevatorFloorIdUnlocked = FloorId
        }
    }

    OnFbiKeyUnlocked(Agent: agent, FbiKeyLevel: int): void = {
        if {
            FbiKeyLevel > HighestFbiKeyUnlocked
        } then {
            set HighestFbiKeyUnlocked = FbiKeyLevel
        }
    }

    StartAchievementTimerKill(AchievementTimer: float): void = {
        # Only start a timer, if there is not already one running
        if(MaybeRoutineAchievementTimerKill?) {
            return
        }
        set HackAchievementTimer = AchievementTimer
        MyRoutine := SpawnRoutine(AchievementTimerKillSuspends)
        set MaybeRoutineAchievementTimerKill = option{MyRoutine}
    }

    AchievementTimerKillSuspends()<suspends>: void = {
        SaveManager := GetSaveManager()
        Sleep(HackAchievementTimer)
        if {
            Player := MaybePlayer?
            Save := SaveManager.ProxySaves[Player]
        } then {
            # Player has failed, reset achievement progress
            set Save.AchievementProgressKill = 0
        }
        set MaybeRoutineAchievementTimerKill = false
    }

    StartAchievementTimerScam(AchievementTimer: float): void = {
        # Only start a timer, if there is not already one running
        if(MaybeRoutineAchievementTimerScam?) {
            return
        }
        set HackAchievementTimer = AchievementTimer
        MyRoutine := SpawnRoutine(AchievementTimerScamSuspends)
        set MaybeRoutineAchievementTimerScam = option{MyRoutine}
    }

    AchievementTimerScamSuspends()<suspends>: void = {
        SaveManager := GetSaveManager()
        Sleep(HackAchievementTimer)
        if {
            Player := MaybePlayer?
            Save := SaveManager.ProxySaves[Player]
        } then {
            # Player has failed, reset achievement progress
            set Save.AchievementProgressScam = 0
        }
        set MaybeRoutineAchievementTimerScam = false
    }

    ServerTickFloatingUi()<suspends>: void = {
        GameConfig := GetGameConfig()
        loop {
            PlayerLocalUi.DoServerTick()
            Sleep(GameConfig.FloatingUiMessageDelay)
        }
    }
}