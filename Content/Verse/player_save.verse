player_save := class<persistable><final> {
    SessionId: int = 0
    SaveVersion: int = 0
    # Int values
    TotalXp: int = 0
    Money: int = 0
    Fans: float = 0.0
    Diamonds: float = 0.0
    CurCreditCardCoins: int = 0
    GamePassLevel: int = -1
    GamePassCoins: int = 0
    CurrentQuestId: int = 0
    QuestProgressCount: int = 0
    Rebirths: int = 0
    XpBoostLevel: int = 0
    UnlockedInstantReload: logic = false
    UpgradeHealthIncreaseLevel: int = 0
    UpgradeCoinsPerKillLevel: int = 0
    UpgradeScamSpeedLevel: int = 0
    UpgradeMoneyScamLevel: int = 0
    UpgradeFasterSaveLevel: int = 0
    # TycoonButton values
    TycoonButtonsOwned: []int = array {}
    CurQuestTycoonButtonId: int = -1
    # -1 = Achievement board not unlocked, set to 0 after Achievement board was unlocked
    NumOfAchievementsCompletedKill: int = -1
    AchievementProgressKill: int = 0
    NumOfAchievementsCompletedScam: int = -1
    AchievementProgressScam: int = 0
    NumOfAchievementsCompletedMoney: int = -1
    AchievementProgressMoney: int = 0
    NumOfAchievementsCompletedRebirth: int = -1
    AchievementProgressRebirth: int = 0
    # Logic values
    # 0 not owned / can buy, 1 owned
    CarShopStateList: []int = array {
        0,
        0,
        0,
        0,
        0,
        0,
        0
    }
    # 0 not found, 1 found
    Puzzle0StateList: []int = array {
        0,
        0,
        0,
        0
    }
    Puzzle1StateList: []int = array {
        0,
        0,
        0,
        0
    }
}

player_save_proxy := class {
    var SessionId: int = 0
    var SaveVersion: int = 0
    # Int values
    var TotalXp: int = 0
    var Money: int = 0
    var Fans: float = 0.0
    var Diamonds: float = 0.0
    var CurCreditCardCoins: int = 0
    var GamePassLevel: int = -1
    var GamePassCoins: int = 0
    var CurrentQuestId: int = 0
    var QuestProgressCount: int = 0
    var Rebirths: int = 0
    var XpBoostLevel: int = 0
    var UnlockedInstantReload: logic = false
    var UpgradeHealthIncreaseLevel: int = 0
    var UpgradeCoinsPerKillLevel: int = 0
    var UpgradeScamSpeedLevel: int = 0
    var UpgradeMoneyScamLevel: int = 0
    var UpgradeFasterSaveLevel: int = 0
    var NumOfAchievementsCompletedKill: int = -1
    var AchievementProgressKill: int = 0
    var NumOfAchievementsCompletedScam: int = -1
    var AchievementProgressScam: int = 0
    var NumOfAchievementsCompletedMoney: int = -1
    var AchievementProgressMoney: int = 0
    var NumOfAchievementsCompletedRebirth: int = -1
    var AchievementProgressRebirth: int = 0
    # TycoonButton values
    var TycoonButtonsOwned: []int = array {}
    var CurQuestTycoonButtonId: int = -1
    # Logic values
    # Car logic values
    var CarShopStateList: []int = array {
        0,
        0,
        0,
        0,
        0,
        0,
        0
    }
    var Puzzle0StateList: []int = array {
        0,
        0,
        0,
        0
    }
    var Puzzle1StateList: []int = array {
        0,
        0,
        0,
        0
    }

    Bake<public>()<transacts>: player_save = {
        player_save {
            SessionId := SessionId
            SaveVersion := SaveVersion
            # Int values
            TotalXp := TotalXp
            Money := Money
            Fans := Fans
            Diamonds := Diamonds
            CurCreditCardCoins := CurCreditCardCoins
            GamePassLevel := GamePassLevel
            GamePassCoins := GamePassCoins
            CurrentQuestId := CurrentQuestId
            QuestProgressCount := QuestProgressCount
            Rebirths := Rebirths
            XpBoostLevel := XpBoostLevel
            UnlockedInstantReload := UnlockedInstantReload
            UpgradeHealthIncreaseLevel := UpgradeHealthIncreaseLevel
            UpgradeCoinsPerKillLevel := UpgradeCoinsPerKillLevel
            UpgradeScamSpeedLevel := UpgradeScamSpeedLevel
            UpgradeMoneyScamLevel := UpgradeMoneyScamLevel
            UpgradeFasterSaveLevel := UpgradeFasterSaveLevel
            NumOfAchievementsCompletedKill := NumOfAchievementsCompletedKill
            AchievementProgressKill := AchievementProgressKill
            NumOfAchievementsCompletedScam := NumOfAchievementsCompletedScam
            AchievementProgressScam := AchievementProgressScam
            NumOfAchievementsCompletedMoney := NumOfAchievementsCompletedMoney
            AchievementProgressMoney := AchievementProgressMoney
            NumOfAchievementsCompletedRebirth := NumOfAchievementsCompletedRebirth
            AchievementProgressRebirth := AchievementProgressRebirth
            TycoonButtonsOwned := TycoonButtonsOwned
            CurQuestTycoonButtonId := CurQuestTycoonButtonId
            # Logic values
            CarShopStateList := CarShopStateList
            Puzzle0StateList := Puzzle0StateList
            Puzzle1StateList := Puzzle1StateList
        }
    }
}
            
CreateProxyFromSave<constructor>(Save: player_save)<transacts> := {
    player_save_proxy {
        SessionId := Save.SessionId + 1
        SaveVersion := Save.SaveVersion
        # Int values
        TotalXp := Save.TotalXp
        Money := Save.Money
        Fans := Save.Fans
        Diamonds := Save.Diamonds
        CurCreditCardCoins := Save.CurCreditCardCoins
        GamePassLevel := Save.GamePassLevel
        GamePassCoins := Save.GamePassCoins
        CurrentQuestId := Save.CurrentQuestId
        QuestProgressCount := Save.QuestProgressCount
        Rebirths := Save.Rebirths
        XpBoostLevel := Save.XpBoostLevel
        UnlockedInstantReload := Save.UnlockedInstantReload
        UpgradeHealthIncreaseLevel := Save.UpgradeHealthIncreaseLevel
        UpgradeCoinsPerKillLevel := Save.UpgradeCoinsPerKillLevel
        UpgradeScamSpeedLevel := Save.UpgradeScamSpeedLevel
        UpgradeMoneyScamLevel := Save.UpgradeMoneyScamLevel
        UpgradeFasterSaveLevel := Save.UpgradeFasterSaveLevel
        NumOfAchievementsCompletedKill := Save.NumOfAchievementsCompletedKill
        AchievementProgressKill := Save.AchievementProgressKill
        NumOfAchievementsCompletedScam := Save.NumOfAchievementsCompletedScam
        AchievementProgressScam := Save.AchievementProgressScam
        NumOfAchievementsCompletedMoney := Save.NumOfAchievementsCompletedMoney
        AchievementProgressMoney := Save.AchievementProgressMoney
        NumOfAchievementsCompletedRebirth := Save.NumOfAchievementsCompletedRebirth
        AchievementProgressRebirth := Save.AchievementProgressRebirth
        TycoonButtonsOwned := Save.TycoonButtonsOwned
        CurQuestTycoonButtonId := Save.CurQuestTycoonButtonId
        # Logic values
        CarShopStateList := Save.CarShopStateList
        Puzzle0StateList := Save.Puzzle0StateList
        Puzzle1StateList := Save.Puzzle1StateList
    }
}