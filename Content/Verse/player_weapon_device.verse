using { /Fortnite.com/Game }
using { /Fortnite.com/Devices }
using { /Fortnite.com/Devices/CreativeAnimation }
using { /Fortnite.com/Characters }
using { /Verse.org/Simulation }
using { /Verse.org/Simulation/Tags }
using { /Verse.org/Random }
using { /Verse.org/Concurrency }
using { /UnrealEngine.com/Temporary/Diagnostics }
using { /UnrealEngine.com/Temporary/SpatialMath }

player_weapon_device := class(creative_device) {
    @editable
    CondButtonList: []conditional_button_device = array {}

    # Constants
    WeaponDamageList: []float = array {
        7.0, # Dual Fiend Hunters (Grey) 
        13.0, # Dual Fiend Hunters (Legendary)
        9.0, # Dual Fiend Hunters (Blue)
        30.0, # Makeshift Bow (Green)
        60.0, # Cluster Bomb New Version
        50.0, # Sticky Grenade Old Version
        55.0, # sense grau*************************************************
        200.0, # Golden Fish Trophy
        180.0, # C4 
        51.0, # Boom Bow (Legendary)
        35.0, # Mechanical Bow (Blue)
        39.0, # Primal Stink Bow (Epic)
        7.5, # Dual Pistols (Blue)+++++++++
        2.9, # Primal SMG (Green) ######
        6.2, # Dual Suppressed Pistols (Green)
        10.0, # Combat Pistol (Legendary)++++++++++
        39.0, # Mammoth Pistol (Grey) ######
        18.0, # Revolver (Epic)
        1.9, # Lever Action Shotgun (Grey)
        1.1, # Drum Shotgun (Blue)
        36.0, # Wood Stake Shotgun (Blue)
        29.0, # Heavy Shotgun (Green)
        1.5, # Tactical Shotgun (Green)
        2.5, # Combat Shotgun (Legendary)
        12.0, # Explosive Repeater Rifle (Legendary)
        2.6, # Tactical Shotgun (Legendary)
        8.0, # Explosive Repeater Rifle (Green)
        3.1, # Machine SMG (Grey) ####
        7.2, # Twin Mag SMG (Blue) ####
        4.7, # Scoped Burst SMG (Blue)
        3.8, # Combat SMG (Green)
        5.3, # Minigun SMG (Legendary)
        7.0, # Submachine Gun 1 (Legendary)
        7.5, # Charge SMG (Legendary)
        7.2, # Tactical Submachine SMG (Epic)
        3.0, # Light Machine SMG (Epic)
        5.0, # Submachine Gun 1 (Blue)<<<<<<<<<<<<<<<<<<<<<<<
        6.25, # Burst SMG (Blue)
        7.3, # P90 (Legendary)
        5.25, # Makeshit AR (Grey)
        5.0, # Scar AR (Blue)
        7.5, # Famas AR -> Burst Assault Rifle(Blue)
        2.6, # Flapjack AR (Green)++++++++++++++++++++++
        35.0, # Ex-Caliber AR (Green)
        3.5, # MK-Seven AR (Green)
        12.0, # Sideways AR (Epic)
        6.0, # Havoc AR (Epic)
        9.0, # Twin Mag AR (Epic)
        7.0, # Tactical AR (Legendary)
        7.0, # M4 AR (Legendary)
        10.5, # Scoped AR (Legendary)
        55.0, # Ex-Caliber AR (Epic)
        8.4, # Ranger AR (Legendary)
        45.0, # Ex-Caliber AR (Blue)
        70.0, # Grenade Launcher (Blue)
        15.0, # Gas Grenade (Blue) won't make boss dmg
        80.0, # sense legendary*********************************************
        77.0, # Egg Launcher (Legendary)
        100.0, # Pumpkin Launcher (Epic)
        65.0, # Sticky Grenade Launcher (Legendary)
        152.0, # Heavy Sniper (Legendary)
        5.5, # Kymera Ray Gun (Legendary)
        110.0, # Anvil Rocket Launcher (Legendary)
        199.5, # Rail Gun (Legendary)
        6.0, # Combat Assault Rifle (Legendary)
        7.0, # Primal Rifle AR (Legendary)
        4.9, # Sideways Minigun (Mythic)+++++++++++++++
        12.0, # Cobra DMR (Mythic)
        6.0, # Shadow Midas' Drum Gun (Mythic)
        8.0, # Relik's MK-Alpha (Mythic)
        60.0, # High Stakes' Shotgun (Mythic)
        75.0, # Boom Sniper Rifle (Exotic)
        21.0, # Heisted Explosive Assault Rifle (Exotic)
        51.0, # Heisted Breacher Shotgun (Exotic)
        12.0, # Heisted Blink Mag SMG (Exotic)
        13.0, # Burst Pulse Rifle (Exotic)
        12.0, # Cobra DMR (Legendary)
        5.5, # Thunder Burst SMG (BLUE)----------NEW GUNS----
        6.6, # Thunder Burst SMG (EPIC)
        7.7, # Thunder Burst SMG (LEGENDARY)
        5.0, # Hyper SMG (BLUE)
        6.0, # Hyper SMG (EPIC)
        7.0, # Hyper SMG (LEGENDARY)
        6.5, # Valeria's Hyper SMG (MYTHIC)
        6.0, # Striker AR (BLUE)
        7.2, # Striker AR (EPIC)
        8.4, # Striker AR (LEGENDARY)
        9.6, # Striker AR (MYTHIC)
        1.9, # x10 Frenzy Auto Shotgun (BLUE)
        2.3, # x10 Frenzy Auto Shotgun (EPIC)
        2.7, # x10 Frenzy Auto Shotgun (LEGENDARY)
        2.0, # x10 Frenzy Auto Shotgun (MYTHIC)
        38.0, # Heavy Shotgun (EPIC)
        6.0, # Suppressed Scar (Legendary)
        60.0, # sense blau**********************************************
        70.0, # sense lila*************************************************
        10.0, # Legion Weapon 1 (SunRose AR)
        5.0, # Legion Weapon 2 (Ceberus Shotgun)
        7.5, # Legion Weapon 3 (SunRose Drumgun)
        7.5, # Legion Weapon 4 (SunRose DPS SMG)
        20.0, # Legion Weapon 5 (Chain Weapon)
        500.0 # Legion Weapon 6 (Zeus)


    }

    TickRate: float = 1.0

    # Runtime variables
    var RoutineServerTick: ?routine = false

    # Runs when the device is started in a running game
    OnBegin<override>()<suspends>: void = {
        MyRoutine := SpawnRoutine(ServerTick)
        set RoutineServerTick = option{MyRoutine}
    }

    ServerTick()<suspends>: void = {
        GameManager := GetGameManager()
        loop {
            for(Player: GetPlayspace().GetPlayers()) {
                if {
                    PlayerData := Player.GetPlayerData[]
                } then {
                    var FoundItemIndex: int = -1
                    for(ButtonIndex -> CondButton: CondButtonList) {
                        # Check if we already found the item
                        if(FoundItemIndex = -1) {
                            for(ItemIndex := 0 .. 2) {
                                if(CondButton.IsHoldingItem[Player, ItemIndex]) {
                                    set FoundItemIndex = (ButtonIndex * 3) + ItemIndex
                                }
                            }
                        }
                    }
                    if(FoundItemIndex > -1) {
                        # Print("Found Item with Index: {FoundItemIndex}")
                        if(WeaponDamage := WeaponDamageList[FoundItemIndex]) {
                            PlayerData.SetWeaponDamage(WeaponDamage)
                        } else {
                            Print("ERR: Found no weapon damage for ID: {FoundItemIndex}")
                        }
                    } else {
                        Print("player_weapon_device - Found no Item!")
                        PlayerData.SetWeaponDamage(1.0)
                    }
                }
            }

            Sleep(TickRate)
        }
    }
}