using { /Verse.org/Simulation }
using { /Verse.org/Colors }
using { /Verse.org/Colors/NamedColors }
using { /Verse.org/Concurrency }
using { /Verse.org/Verse }
using { /Verse.org/Assets }
using { /Fortnite.com/Devices }
using { /Fortnite.com/UI }
using { /Fortnite.com/Characters }
using { /UnrealEngine.com/Temporary/Diagnostics }
using { /UnrealEngine.com/Temporary/SpatialMath }
using { /UnrealEngine.com/Temporary/UI}

playtime_reward_device := class(creative_device) {
    @editable
    OpenUiButtonList: []button_device = array {}

    @editable
    HudRewardReady: hud_message_device = hud_message_device {}

    @editable
    AudioOnRewardClaim: audio_player_device = audio_player_device {}

    # Constants
    # How long you have to wait
    PlaytimeRewardTimerList: []float = array {
        1 * 1.0,
        1 * 60.0,
        4.5 * 60.0,
        11 * 60.0,
        18 * 60.0,
        26 * 60.0,
        35 * 60.0,
        45 * 60.0,
        60 * 60.0,
        120 * 60.0,
        180 * 60.0,
        210 * 60.0
    }

    # How much money you get from the reward
    PlaytimeRewardMoneyList: []int = array {
        1 * 245,
        1 * 625,
        1 * 1250,
        1 * 2500,
        1 * 3500,
        1 * 4500,
        1 * 6500,
        1 * 35000,
        1 * 85000,
        1 * 50000,
        1 * 100000,
        1 * 500000
    }

    # Runtime variables
    var AgentUiMap: [agent]_UI.ui_playtime_reward = map {}
    var AgentRoutineMap: [agent]routine = map {}
    # We need to pass the agent to the routine
    var MaybePrevAgent: ?agent = false

    OnBegin<override>()<suspends>: void = {
        (super:)OnBegin()

        for(OpenUiButton: OpenUiButtonList) {
            OpenUiButton.InteractedWithEvent.Subscribe(OnOpenUiButton)
        }

        # GetPlayspace().PlayerAddedEvent().Subscribe(OnPlayerAdded)
        GetPlayspace().PlayerRemovedEvent().Subscribe(OnPlayerRemoved)

        <# Call logic for initial players
        for(Player : GetPlayspace().GetPlayers()) {
            OnPlayerAdded(Player)
        }
        #>
    }

    # This is called by TycoonManager
    OnDeviceUnlocked(Agent: agent): void = {
        if {
            PlayerData := Agent.GetPlayerData[]
        } then {
            set PlayerData.TimestampPlayerUnlockedPlaytimeRewards = GetSimulationElapsedTime()
            Ui := _UI.ui_playtime_reward {}
            Ui.InitWidgetPlaytimeReward(Agent, Self)
            if(set AgentUiMap[Agent] = Ui) {}
            # Alert the player when rewards are ready
            set MaybePrevAgent = option{Agent}
            NewRoutine := SpawnRoutine(AlertPlayerWhenRewardReadySuspends)
            if(set AgentRoutineMap[Agent] = NewRoutine) {}
        }
    }

    OnPlayerRemoved(Agent: agent): void = {
        # Delete the element from AgentUiMap to avoid memory leak
        if(Ui := AgentUiMap[Agent]) {
            Ui.RemoveUiFromPlayer(Agent)
        }
        set AgentUiMap = AgentUiMap.RemoveKey(Agent)
        # Cancel routine
        if (Routine := AgentRoutineMap[Agent]) {
            Routine.Cancel()
        }
        set AgentRoutineMap = AgentRoutineMap.RemoveKey(Agent)
    }

    OnOpenUiButton(Agent: agent): void = {
        if(Ui := AgentUiMap[Agent]) {
            Ui.AddUiToPlayer(Agent, false)
        }
    }

    # 0 have to wait, 1 can buy, 2 already received
    GetItemState(Agent: agent, ItemIndex: int): int = {
        GameManager := GetGameManager()
        var ButtonText: string = ""
        if {
            PlayerData := Agent.GetPlayerData[]
            ItemState := PlayerData.PlaytimeRewardStateList[ItemIndex]
        } then {
            if(ItemState = 0) {
                # Check if we have to update the item state since enough time passed
            }
            return ItemState
        } else {
            Print("ERR: PlaytimeRewardShop: GetItemState: Didn't find ItemState for Index: {ItemIndex}")
            return 0
        }
    }

    GetButtonText(Agent: agent, ItemIndex: int): string = {
        var ButtonText: string = ""
        ItemState := GetItemState(Agent, ItemIndex)
        case(ItemState) {
        0 =>
            # Use the default shop button text
            set ButtonText = "WAIT..."
        1 =>
            set ButtonText = "GET"
        2 =>
            set ButtonText = "OK!"
        _ =>
            set ButtonText = ""
        }
        return ButtonText
    }

    TryToClaimReward(Agent: agent, Ui: _UI.ui_playtime_reward, ItemIndex: int): void = {
        GameManager := GetGameManager()
        ItemState := GetItemState(Agent, ItemIndex)
        if(ItemState = 1) {
            GivePlaytimeReward(Agent, Ui, ItemIndex)
        }
    }

    GivePlaytimeReward(Agent: agent, Ui: _UI.ui_playtime_reward, ItemIndex: int): void = {
        if {
            PlayerData := Agent.GetPlayerData[]
            RewardMoney := PlaytimeRewardMoneyList[ItemIndex]
            TextureCheckmark := Ui.TextureCheckmarkList[ItemIndex]
        } then {
            PlayerData.SetPlaytimeRewardItemState(ItemIndex, 2)
            PlayerData.AddMoney(RewardMoney, true, false)
            TextureCheckmark.SetVisibility(widget_visibility.Visible)
            AudioOnRewardClaim.Play(Agent)
            # Refresh the UI
            Ui.RefreshUi()
        }
    }

    AlertPlayerWhenRewardReadySuspends()<suspends>: void = {
        # Pass the agent to the routine
        MaybeAgent := MaybePrevAgent
        var LastTimerDuration: float = 0.0
        for(RewardTimer: PlaytimeRewardTimerList) {
            Sleep(RewardTimer - LastTimerDuration)
            # Alert the player that the reward is ready
            if(Agent := MaybeAgent?) {
                HudRewardReady.Show(Agent)
            }
            # Update var for next timer
            set LastTimerDuration = RewardTimer
        }
    }
}