using { /Verse.org/Simulation }
using { /Verse.org/Colors }
using { /Verse.org/Concurrency }
using { /Verse.org/Verse }
using { /Verse.org/Random }
using { /Fortnite.com/Devices }
using { /Fortnite.com/UI }
using { /Fortnite.com/Game }
using { /Fortnite.com/Characters }
using { /Fortnite.com/Devices/CreativeAnimation }
using { /UnrealEngine.com/Temporary/Diagnostics }
using { /UnrealEngine.com/Temporary/SpatialMath }
using { /UnrealEngine.com/Temporary/UI}

police_patrol_device := class(creative_device) {
    @editable
    PolicePatrolId: int = 0

    @editable
    PoliceCarProp: creative_prop = creative_prop {}

    @editable
    PoliceWaypointList: []_Npc.npc_waypoint_device = array {}

    @editable
    OfficerGuardSpawner: npc_spawner_device = npc_spawner_device {}

    @editable
    PoliceCarMovementSpeed: float = 5000.0

    @editable
    PoliceMaxRange: float = 950.0

    @editable
    AudioOfficerSpawned: audio_player_device = audio_player_device {}

    # Constants
    # Make the police officer spawn next to the driver door
    # Warning: Must be adjusted for the current police car rotation
    PoliceOfficerOffset: vector3 = vector3{X := 33.0, Y := -163.0, Z := 100.0}

    # Runtime
    var RoutineServerTick: ?routine = false
    var MaybeRoutineDespawnOfficer: ?routine = false
    var TimestampOfficerSpawn: float = 0.0
    # var TimestampOfficerDeath: float = 0.0
    var MaybeCurWaypoint: ?_Npc.npc_waypoint_device = false
    var MaybeAnimController: ?animation_controller = false
    var MaybePlayerThatSpawnedOfficer: ?player = false
    
    OnBegin<override>()<suspends>: void = {
        # Init variables
        if(AnimationController := PoliceCarProp.GetAnimationController[]) {
            set MaybeAnimController = option{AnimationController}
        }

        # Listen to events
        OfficerGuardSpawner.SpawnedEvent.Subscribe(OnOfficerSpawned)
        # OfficerGuardSpawner.EliminatedEvent.Subscribe(OnOfficerEliminated)
    
        if {
            FirstWaypoint := PoliceWaypointList[0]
        } then {
            MoveToWaypoint(FirstWaypoint)
        }

        MyRoutine := SpawnRoutine(ServerTick)
        set RoutineServerTick = option{MyRoutine}

        # Activate spawner
        # Sleep(1.0)
        # OfficerGuardSpawner.Enable()
        # Sleep(5.0)
        # Remove the first automatic spawn, since it was not on purpose
        # OfficerGuardSpawner.DespawnAll(false)
    }

    OnOfficerSpawned(NpcAgent: agent): void = {
        GameManager := GetGameManager()
        # Prevent triggering the event multiple times, by checking if this device spawned the officer
        if(PolicePatrolId = GameManager.HackPoliceSpawnerId) {
            Print("OnOfficerSpawned")
            Utility := GetUtility()
            if {
                NpcFortChar := NpcAgent.GetFortCharacter[]
                NpcFortChar.IsActive[]
            } then {
                RotatedMovementVector := PoliceCarProp.GetTransform().Rotation.RotateVector(PoliceOfficerOffset)
                OfficerSpawnPos := PoliceCarProp.GetTransform().Translation + RotatedMovementVector
                # OfficerSpawnPos := PoliceCarProp.GetTransform().Translation - (2.0 * Utility.PlayerOffset)
                if(NpcFortChar.TeleportTo[OfficerSpawnPos, PoliceCarProp.GetTransform().Rotation]) {}
                # Assign NpcDataGuard for Elimination rewards
                NpcDataGuard := _Npc.npc_data_guard {}
                NpcDataGuard.InitGuardNpc(NpcAgent, PolicePatrolId, 0)
                NpcDataGuard.SetPolicePatrolManager(Self)
                NpcAgent.SetNpcDataGuard(NpcDataGuard)
            } else {
                Print("ERR: OnOfficerSpawned: Didn't find NpcFortChar")
            }
        }
    }

    HandleNpcGuardEliminated(NpcAgentGuard: agent, GuardSpawnerIndex: int, EliminationResult: elimination_result): void = {
        SaveManager := GetSaveManager()
        GameConfig := GetGameConfig()
        if(Routine := MaybeRoutineDespawnOfficer?) {
            Routine.Cancel()
        }
        set MaybeRoutineDespawnOfficer = false
        set TimestampOfficerSpawn = 0.0
        if {
            PlayerThatSpawnedOfficer := MaybePlayerThatSpawnedOfficer?
            PlayerData := PlayerThatSpawnedOfficer.GetPlayerData[]
        } then {
            set PlayerData.TimestampOfficerDeath = GetSimulationElapsedTime()
        }
        # set TimestampOfficerDeath = GetSimulationElapsedTime()
        # This must be after setting TimestampOfficerDeath
        set MaybePlayerThatSpawnedOfficer = false
        # Reward the killer player
        if {
            NpcDataGuard := NpcAgentGuard.GetNpcDataGuard[]
        } then {
            # Workaround for killing NPCs via script (dragon mount)
            var MaybeKillerAgent: ?agent = false
            if {
                KillerFortCharElimResult := EliminationResult.EliminatingCharacter?
                KillerAgentElimResult := KillerFortCharElimResult.GetAgent[]
                PlayerDataValidAgentTest := KillerAgentElimResult.GetPlayerData[]
            } then {
                # Print("HandleNpcEliminated - Found killer from ElimResult")
                set MaybeKillerAgent = option{KillerAgentElimResult}
            } else {
                if {
                    KillerAgentNpcData := NpcDataGuard.MaybeLastAgentThatDamagedNpc?
                    PlayerDataValidAgentTest := KillerAgentNpcData.GetPlayerData[]
                } then {
                    Timestamp := GetSimulationElapsedTime()
                    TimeSinceLastDamage := Timestamp - NpcDataGuard.TimestampLastNpcDamage
                    if(TimeSinceLastDamage < 3.0) {
                        # Print("HandleNpcEliminated - Found killer from DamagedEvent")
                        set MaybeKillerAgent = option{KillerAgentNpcData}
                    }
                }
            }
            if {
                KillerAgent := MaybeKillerAgent?
                PlayerData := KillerAgent.GetPlayerData[]
                Save := SaveManager.ProxySaves[KillerAgent]
                # CoinsPerKillIncreasePerUpgrade := GameConfig.PermanentUpgradeRewardPerLevelList[1]
                # Rewards at index 0 are for TotalScamLevel = 1
                RewardScamLevel := Max(PlayerData.TotalScamLevel - 1, 0)
                MoneyPerKillBaseRewardMin := GameConfig.MoneyPerKillNpcRankMin[Min(RewardScamLevel, GameConfig.MoneyPerKillNpcRankMin.Length)]
                MoneyPerKillBaseRewardMax := GameConfig.MoneyPerKillNpcRankMax[Min(RewardScamLevel, GameConfig.MoneyPerKillNpcRankMax.Length)]
                DiamondsPerKillBaseRewardMin := GameConfig.BaseKillpointsPerKillNpcRankMin[Min(RewardScamLevel, GameConfig.BaseKillpointsPerKillNpcRankMin.Length)]
                DiamondsPerKillBaseRewardMax := GameConfig.BaseKillpointsPerKillNpcRankMax[Min(RewardScamLevel, GameConfig.BaseKillpointsPerKillNpcRankMax.Length)]
                DiamondsPerKillPoliceBonusMin := GameConfig.BonusKillpointsPerKillPoliceScamLevelMin[Min(RewardScamLevel, GameConfig.BonusKillpointsPerKillPoliceScamLevelMin.Length)]
                DiamondsPerKillPoliceBonusMax := GameConfig.BonusKillpointsPerKillPoliceScamLevelMax[Min(RewardScamLevel, GameConfig.BonusKillpointsPerKillPoliceScamLevelMax.Length)]
        } then {
                MoneyPerKillBaseReward := GetRandomInt(MoneyPerKillBaseRewardMin, MoneyPerKillBaseRewardMax)
                DiamondsPerKillPoliceBonus := GetRandomFloat(DiamondsPerKillPoliceBonusMin, DiamondsPerKillPoliceBonusMax)
                DiamondsPerKillBaseReward := DiamondsPerKillPoliceBonus + GetRandomFloat(DiamondsPerKillBaseRewardMin, DiamondsPerKillBaseRewardMax)
                
                XpForKill := CalculateXpForKill(KillerAgent, NpcDataGuard.NpcRarity)
                PlayerData.AddXp(XpForKill, true)
                PlayerData.AddMoney(MoneyPerKillBaseReward, true, false)
                var DiamondReward: float = DiamondsPerKillBaseReward
                if(DiamondsKillBonusFromUpgrade := GameConfig.PermanentUpgradeRewardKillpointsPerKill[Save.UpgradeCoinsPerKillLevel]) {
                    set DiamondReward = DiamondReward + DiamondsKillBonusFromUpgrade
                }
                # set DiamondReward = DiamondReward + (Save.UpgradeCoinsPerKillLevel * CoinsPerKillIncreasePerUpgrade)
                PlayerData.AddDiamonds(DiamondReward, true)
                PlayerData.AddGamePassCoins(1)
                PlayerData.IncreaseKillStreak(KillerAgent, 1)
            }
        }
    }

    # This can not be used, since it would be called for the officer from the other police car!!
    <#
    OnOfficerEliminated(AiInteractionResult: device_ai_interaction_result): void = {
    }
    #>

    SpawnOfficer(): void = {
        Print("SpawnOfficer")
        GameManager := GetGameManager()
        # Stop the police car movement
        # Teleport is not needed, since we pause the animation
        # if(PoliceCarProp.TeleportTo[PoliceCarProp.GetTransform().Translation, PoliceCarProp.GetTransform().Rotation]) {}
        if(AnimationController := MaybeAnimController?) {
            AnimationController.Pause()
        }
        # Spawn officer
        set GameManager.HackPoliceSpawnerId = PolicePatrolId
        set TimestampOfficerSpawn = GetSimulationElapsedTime()
        OfficerGuardSpawner.Spawn()
        # Play spawn audio
        if(AudioOfficerSpawned.TeleportTo[PoliceCarProp.GetTransform()]) {}
        spawn{PlayAudioOfficerSpawnedSuspends()}

        MyRoutine := SpawnRoutine(DespawnOfficerSuspends)
        set MaybeRoutineDespawnOfficer = option{MyRoutine}
    }

    PlayAudioOfficerSpawnedSuspends()<suspends>: void = {
        # Sleep must be higher than 0.0
        Sleep(0.1)
        AudioOfficerSpawned.Play()
    }

    DespawnOfficerSuspends()<suspends>: void = {
        GameConfig := GetGameConfig()
        Sleep(GameConfig.OfficerDespawnDuration)
        Print("DespawnOfficer")
        set TimestampOfficerSpawn = 0.0
        if {
            PlayerThatSpawnedOfficer := MaybePlayerThatSpawnedOfficer?
            PlayerData := PlayerThatSpawnedOfficer.GetPlayerData[]
        } then {
            set PlayerData.TimestampOfficerDeath = GetSimulationElapsedTime()
        }
        # This must be after setting TimestampOfficerDeath
        set MaybePlayerThatSpawnedOfficer = false
        
        OfficerGuardSpawner.DespawnAll(false)
    }

    MoveToWaypoint(NextWaypoint: _Npc.npc_waypoint_device): void = {
        set MaybeCurWaypoint = option{NextWaypoint}
        NextTransform := NextWaypoint.WaypointProp.GetTransform()
        CurTransform := PoliceCarProp.GetTransform()
        DistanceToWaypoint := Distance(NextTransform.Translation, CurTransform.Translation)
        MovementDuration := DistanceToWaypoint / PoliceCarMovementSpeed
        # Make Rotation
        if (VectorToGoal := (NextTransform.Translation - CurTransform.Translation).MakeUnitVector[]) {
            Yaw := RadiansToDegrees(ArcTan(VectorToGoal.Y, VectorToGoal.X)) # - (PiFloat * 0.5))
            Pitch := 0.0
            Roll := 0.0
            GoalRot := MakeRotationFromYawPitchRollDegrees(Yaw, Pitch, Roll)

            # Teleport to rotate instantly and face the goal
            if(PoliceCarProp.TeleportTo[CurTransform.Translation, GoalRot]) {}
            Task := spawn{PoliceCarProp.MoveTo(NextTransform.Translation, GoalRot, MovementDuration)}
        }
    }

    GetNextWaypoint(NpcWaypointList: []_Npc.npc_waypoint_device, CurWaypoint: _Npc.npc_waypoint_device): _Npc.npc_waypoint_device = {
        # Get a random connected waypoint
        RandInt := GetRandomInt(0, CurWaypoint.WaypointsInReach.Length - 1)
        if {
            WaypointIndex := CurWaypoint.WaypointsInReach[RandInt]
            NextWaypoint := NpcWaypointList[WaypointIndex]
        } then {
            return NextWaypoint  
        } else {
            Print("ERR: NPC GetNextWaypoint - Found no linked waypoint")
            return CurWaypoint
        }
    }

    CalculateXpForKill<public>(Agent: agent, NpcRarity: int): int = {
        GameConfig := GetGameConfig()
        var XpForKill: int = 0
        if {
            PlayerData := Agent.GetPlayerData[]
            XpPerKillNpcMin := GameConfig.XpPerKillNpcMin[NpcRarity]
            XpPerKillNpcMax := GameConfig.XpPerKillNpcMax[NpcRarity]
        } then {
            set XpForKill = GetRandomInt(XpPerKillNpcMin, XpPerKillNpcMax)
            # set XpForKill = GameConfig.BaseXpPerKillNpc + (PlayerData.CurLevel * 10)
        } else {
            Print("ERR: police_patrol_device: CalculateXpForKill")
        }
        return XpForKill
    }


    ServerTick()<suspends>: void = {
        GameConfig := GetGameConfig()
        Utility := GetUtility()
        loop {
            if {
                # Only move the police car if there is currently no officer active
                TimestampOfficerSpawn = 0.0
            } then {
                # Move the police car
                if(AnimationController := MaybeAnimController?) {
                    AnimState := AnimationController.GetState()
                    if(AnimState = animation_controller_state.Playing) {
                        # Do nothing, the car is already moving
                    } else if(AnimState = animation_controller_state.Paused) {
                        # Continue the movement
                        AnimationController.Play()
                    } else if(AnimState = animation_controller_state.Stopped) {
                        # Reached waypoint, move to next waypoint
                        if {
                            CurWaypoint := MaybeCurWaypoint?
                        } then {
                            NextWaypoint := GetNextWaypoint(PoliceWaypointList, CurWaypoint)
                            MoveToWaypoint(NextWaypoint)
                        } else {
                            Print("ERR: police_patrol_device: MaybeCurWaypoint is not set!")
                        }
                    }
                }
                <# Check if officer spawn is on cooldown
                Timestamp := GetSimulationElapsedTime()
                TimeSinceOfficerDeath := Timestamp - TimestampOfficerDeath
                if(TimeSinceOfficerDeath >= GameConfig.PoliceCooldownAfterOfficerDied) {
                #>
                # Check if there is a player in the radius
                # Utility.DebugDraw.DrawSphere(Center := PoliceCarProp.GetTransform().Translation, ?Radius := PoliceMaxRange, ?DrawDurationPolicy := debug_draw_duration_policy.FiniteDuration, ?Duration := 0.04)
                for(Player: GetPlayspace().GetPlayers()) {
                    if {
                        # Abort the loop if we already spawned an officer
                        TimestampOfficerSpawn = 0.0
                        FortChar := Player.GetFortCharacter[]
                        FortChar.IsActive[]
                        PlayerData := Player.GetPlayerData[]
                        # PlayerData.IsScammingNpc = true
                        # PlayerData.TotalScamLevel > 0
                        # Police now only attacks from scam level 6 upwards
                        PlayerData.TotalScamLevel > 5
                    } then {
                        # Check if officer spawn is on cooldown for this player
                        Timestamp := GetSimulationElapsedTime()
                        TimeSinceOfficerDeath := Timestamp - PlayerData.TimestampOfficerDeath
                        if(TimeSinceOfficerDeath >= GameConfig.PoliceCooldownAfterOfficerDied) {
                            DistanceToPolice := Distance(FortChar.GetTransform().Translation, PoliceCarProp.GetTransform().Translation)
                            if(DistanceToPolice <= PoliceMaxRange) {
                                set MaybePlayerThatSpawnedOfficer = option{Player}
                                SpawnOfficer()
                            }
                        }
                    }
                }
            }
            Sleep(0.0)
        }
    }
}
