using { /Fortnite.com/Game }
using { /Fortnite.com/Devices }
using { /Fortnite.com/Devices/CreativeAnimation }
using { /Fortnite.com/Characters }
using { /Verse.org/Simulation }
using { /Verse.org/Simulation/Tags }
using { /Verse.org/Random }
using { /Verse.org/Concurrency }
using { /UnrealEngine.com/Temporary/Diagnostics }
using { /UnrealEngine.com/Temporary/SpatialMath }

# See https://dev.epicgames.com/documentation/en-us/uefn/create-your-own-device-in-verse for how to create a verse device.

<#
# A Verse-authored creative device that can be placed in a level
quest_config_device := class(creative_device) {
    @editable
    QuestTracker: tracker_device = tracker_device {}

    @editable
    RequiredPlayerLevel: int = 1

    @editable
    ShowProgressInHud: logic = false

    @editable
    StartWithoutQuestgiver: logic = false

    @editable
    StartWithoutQuestgiverDelay: float = 0.1

    @editable
    QuestGiverButton: button_device = button_device {}

    @editable
    ExclamationMarkBeacon: beacon_device = beacon_device {}

    @editable
    QuestStoryText: string = ""

    @editable
    QuestAcceptObjectiveText: string = ""

    @editable
    QuestObjectiveText: string = ""

    @editable
    QuestObjectiveBeacon: beacon_device = beacon_device {}

    @editable
    RewardXpAmount: int = 10

    <# Optional properties below #>
    @editable
    RewardItem: item_granter_device = item_granter_device {}

    # Zeitlich begrenzt + Player Stasis. Dafür muss man auch CutsceneDuration setzen
    @editable
    QuestCompletedCutsceneList: []cinematic_sequence_device = array {}

    @editable
    QuestCompletedCutsceneDuration: float = 0.0

    @editable
    SwitchOnStartList: []switch_device = array {}

    @editable
    GrantItemOnKill: logic = false
    
    @editable
    KillItemZombieSpawnerList: []creature_spawner_device = array {}

    @editable
    KillItemGranter: item_granter_device = item_granter_device {}

    # Not implemented yet, copy from OnlyFans
    @editable
    QuestGoalCountAmount: int = 1

    <# QuestGoal properties below #>
    @editable
    QuestGoalButton: ?button_device = false

    @editable
    QuestGoalMutatorZone: ?mutator_zone_device = false

    @editable
    QuestGoalRemotePrimary: ?signal_remote_manager_device = false

    @editable
    QuestGoalRemoteSecondary: ?signal_remote_manager_device = false

    # QuestGoalCondButton is also used for buying items from shops
    @editable
    QuestGoalCondButtonList<public>: []conditional_button_device = array {}

    @editable
    QuestGoalCreatureSpawnerList: []creature_spawner_device = array {}

    @editable
    QuestGoalSwitchList: []switch_device = array {}

    @editable
    QuestGoalFishingZoneList: []fishing_zone_device = array {}

    @editable
    QuestGoalPropManipulatorHarvestingList: []prop_manipulator_device = array {}

    # Constants

    # Runtime variables
    var QuestId: int = -1

    # Runs when the device is started in a running game
    OnBegin<override>()<suspends>:void = {
        
    }

    OnQuestComplete(Agent: agent): void = {
        
    }
}
#>