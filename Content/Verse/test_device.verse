using { /Fortnite.com/Game }
using { /Fortnite.com/AI }
using { /Fortnite.com/FortPlayerUtilities }
using { /Fortnite.com/Devices }
# using { /Fortnite.com/Devices/CreativeAnimation }
using { /Fortnite.com/Characters }
using { /Verse.org/Simulation }
using { /Verse.org/Simulation/Tags }
using { /Verse.org/Random }
using { /Fortnite.com/Teams }
# using { /Verse.org/Assets }
# using { /Verse.org/Concurrency }
# using { /UnrealEngine.com/Temporary/Diagnostics }
using { /UnrealEngine.com/Temporary/SpatialMath }

# See https://dev.epicgames.com/documentation/en-us/uefn/create-your-own-device-in-verse for how to create a verse device.

# A Verse-authored creative device that can be placed in a level
test_device := class(creative_device) {
    @editable
    ButtonGetMoneyAndXp: button_device = button_device {}

    @editable
    ButtonSaveTycoon: button_device = button_device {}

    @editable
    ButtonIncreaseQuestProgress: button_device = button_device {}

    @editable
    ButtonRebirth: button_device = button_device {}

    @editable
    ButtonScamRank10: button_device = button_device {}

    @editable
    Button12160Scams: button_device = button_device {}

    @editable
    Button2OG: button_device = button_device {}

    @editable
    Button3OG: button_device = button_device {}

    @editable
    ButtonLevel7: button_device = button_device {}

    @editable
    ButtonUpdateIncomeBillboards: button_device = button_device {}

    @editable
    IncomeBillboardList: []billboard_device = array {}

    @editable
    ButtonScamLevel3: button_device = button_device {}

    @editable
    ButtonScamLevel4: button_device = button_device {}

    @editable
    ButtonScamLevel5: button_device = button_device {}

    @editable
    ButtonScamLevel7: button_device = button_device {}

    @editable
    ButtonScamLevel8: button_device = button_device {}

    @editable
    ButtonScamLevel9: button_device = button_device {}

    @editable
    ButtonScamLevel10: button_device = button_device {}

    @editable
    ButtonAchievementKill: button_device = button_device {}

    @editable
    ButtonAchievementScam: button_device = button_device {}

    @editable
    ButtonAchievementMoney: button_device = button_device {}

    @editable
    ButtonAchievementRebirth: button_device = button_device {}

    @editable
    ButtonActivateXpBoost: button_device = button_device {}

    @editable
    ButtonScamLevelPlus1: button_device = button_device {}

    @editable
    ButtonBuildWholeTycoon: button_device = button_device {}

    @editable
    BillboardBuildWholeTycoonList: []billboard_device = array {}

    # Constants

    # Runtime
    var CurBuildWholeTycoonIndex: int = 0

    # Runs when the device is started in a running game
    OnBegin<override>()<suspends>:void = {
        ButtonGetMoneyAndXp.InteractedWithEvent.Subscribe(OnTestTrigger)
        ButtonSaveTycoon.InteractedWithEvent.Subscribe(OnTestTriggerSave)
        ButtonIncreaseQuestProgress.InteractedWithEvent.Subscribe(OnButtonIncreaseQuestProgress)
        ButtonRebirth.InteractedWithEvent.Subscribe(OnButtonRebirth)
        ButtonScamRank10.InteractedWithEvent.Subscribe(OnButtonScamRank10)
        Button12160Scams.InteractedWithEvent.Subscribe(OnButton12160Scams)
        Button2OG.InteractedWithEvent.Subscribe(OnButton2OG)
        Button3OG.InteractedWithEvent.Subscribe(OnButton3OG)
        ButtonLevel7.InteractedWithEvent.Subscribe(OnButtonLevel7)
        ButtonUpdateIncomeBillboards.InteractedWithEvent.Subscribe(OnButtonUpdateIncomeBillboards)
        ButtonScamLevel3.InteractedWithEvent.Subscribe(OnButtonScamLevel3)
        ButtonScamLevel4.InteractedWithEvent.Subscribe(OnButtonScamLevel4)
        ButtonScamLevel5.InteractedWithEvent.Subscribe(OnButtonScamLevel5)
        ButtonScamLevel7.InteractedWithEvent.Subscribe(OnButtonScamLevel7)
        ButtonScamLevel8.InteractedWithEvent.Subscribe(OnButtonScamLevel8)
        ButtonScamLevel9.InteractedWithEvent.Subscribe(OnButtonScamLevel9)
        ButtonScamLevel10.InteractedWithEvent.Subscribe(OnButtonScamLevel10)
        ButtonAchievementKill.InteractedWithEvent.Subscribe(OnButtonAchievementKill)
        ButtonAchievementScam.InteractedWithEvent.Subscribe(OnButtonAchievementScam)
        ButtonAchievementMoney.InteractedWithEvent.Subscribe(OnButtonAchievementMoney)
        ButtonAchievementRebirth.InteractedWithEvent.Subscribe(OnButtonAchievementRebirth)
        ButtonActivateXpBoost.InteractedWithEvent.Subscribe(OnButtonActivateXpBoost)
        ButtonBuildWholeTycoon.InteractedWithEvent.Subscribe(OnButtonBuildWholeTycoon)
        ButtonScamLevelPlus1.InteractedWithEvent.Subscribe(OnButtonScamLevelPlus1)
    }

    OnTestTrigger(Agent: agent): void = {
        if {
            FortChar := Agent.GetFortCharacter[]
            PlayerData := Agent.GetPlayerData[]
        } then {
            # PlayerData.AddXp(10000 * 10, true)
            PlayerData.AddXp(19750, true)
            PlayerData.AddFans(150000000.0, true)
            PlayerData.AddDiamonds(150000000.0, true)
            PlayerData.AddMoney(150000000, true, false)
            PlayerData.UpdateCurRebirths(PlayerData.NumOfRebirths + 20)
        }
    }

    OnTestTriggerSave(Agent: agent): void = {
        SaveManager := GetSaveManager()
        if {
            PlayerData := Agent.GetPlayerData[]
        } then {
            SaveManager.PersistPlayerSave(Agent)
        }
    }

    OnButtonIncreaseQuestProgress(Agent: agent): void = {
        if {
            PlayerData := Agent.GetPlayerData[]
        } then {
            PlayerData.IncreaseQuestProgress()
        }
    }

    OnButtonRebirth(Agent: agent): void = {
        if {
            PlayerData := Agent.GetPlayerData[]
            TycoonManager := PlayerData.MaybeTycoonManager?
        } then {
            TycoonManager.OnRebirthButton(Agent)
        }
    }

    OnButtonScamLevelPlus1(Agent: agent): void = {
        if {
            PlayerData := Agent.GetPlayerData[]
        } then {
            Print("OnButtonScamLevelPlus1: New Level: {PlayerData.ScamLevelBack + 1}")
            PlayerData.SetScamLevelBack(PlayerData.ScamLevelBack + 1)
        }
    }

    OnButtonScamRank10(Agent: agent): void = {
        if {
            PlayerData := Agent.GetPlayerData[]
        } then {
            PlayerData.SetScamLevelHead(10)
        }
    }

    OnButton12160Scams(Agent: agent): void = {
        if {
            PlayerData := Agent.GetPlayerData[]
        } then {
            PlayerData.AddMoney(5000, true, false)
            PlayerData.AddFans(12160.0, true)
            PlayerData.AddDiamonds(500.0, true)
            PlayerData.SetScamLevelHead(2)
        }
    }

    OnButton2OG(Agent: agent): void = {
        if {
            PlayerData := Agent.GetPlayerData[]
        } then {
            PlayerData.AddMoney(150000, true, false)
            PlayerData.AddFans(350000.0, true)
            PlayerData.AddDiamonds(400000.0, true)
            PlayerData.SetScamLevelHead(4)
        }
    }

    OnButton3OG(Agent: agent): void = {
        if {
            PlayerData := Agent.GetPlayerData[]
        } then {
            PlayerData.AddMoney(7180000, true, false)
            PlayerData.AddFans(6910000.0, true)
            PlayerData.AddDiamonds(3610000.0, true)
            PlayerData.SetScamLevelHead(6)
        }
    }

    OnButtonLevel7(Agent: agent): void = {
        if {
            PlayerData := Agent.GetPlayerData[]
        } then {
            PlayerData.AddMoney(40000000, true, false)
            PlayerData.AddFans(30000000.0, true)
            PlayerData.AddDiamonds(20000000.0, true)
            PlayerData.SetScamLevelHead(7)
        }
    }

    OnButtonScamLevel3(Agent: agent): void = {
        if {
            PlayerData := Agent.GetPlayerData[]
        } then {
            PlayerData.SetScamLevelHead(3)
        }
    }
    OnButtonScamLevel4(Agent: agent): void = {
        if {
            PlayerData := Agent.GetPlayerData[]
        } then {
            PlayerData.SetScamLevelHead(4)
        }
    }
    OnButtonScamLevel5(Agent: agent): void = {
        if {
            PlayerData := Agent.GetPlayerData[]
        } then {
            PlayerData.SetScamLevelHead(5)
        }
    }

    OnButtonScamLevel7(Agent: agent): void = {
        if {
            PlayerData := Agent.GetPlayerData[]
        } then {
            PlayerData.SetScamLevelHead(7)
        }
    }

    
    OnButtonScamLevel8(Agent: agent): void = {
        if {
            PlayerData := Agent.GetPlayerData[]
        } then {
            PlayerData.SetScamLevelHead(8)
        }
    }
        
    OnButtonScamLevel9(Agent: agent): void = {
        if {
            PlayerData := Agent.GetPlayerData[]
        } then {
            PlayerData.SetScamLevelHead(9)
        }
    }
        
    OnButtonScamLevel10(Agent: agent): void = {
        if {
            PlayerData := Agent.GetPlayerData[]
        } then {
            PlayerData.SetScamLevelHead(10)
        }
    }
    OnButtonUpdateIncomeBillboards(Agent: agent): void = {
        Utility := GetUtility()
        if {
            PlayerData := Agent.GetPlayerData[]
            TycoonManager := PlayerData.MaybeTycoonManager?
        } then {
            if(Billboard := IncomeBillboardList[0]) {
                Billboard.SetText(Utility.FormatMessage(Utility.FormatFloat(TycoonManager.MoneyIncomePerSec * PlayerData.CurMoneyMultiplier, 2, 0)))
            }
            if(Billboard := IncomeBillboardList[1]) {
                Billboard.SetText(Utility.FormatMessage(Utility.FormatFloat(TycoonManager.FansIncomePerSec, 2, 0)))
            }
            if(Billboard := IncomeBillboardList[2]) {
                Billboard.SetText(Utility.FormatMessage(Utility.FormatFloat(TycoonManager.DiamondsIncomePerSec, 2, 0)))
            }
        }
    }

    OnButtonAchievementKill(Agent: agent): void = {
        SaveManager := GetSaveManager()
        if {
            Save := SaveManager.ProxySaves[Agent]
        } then {
            SaveManager.IncrementAchievement(Agent, _Global.achievement_type.Kill, Save.NumOfAchievementsCompletedKill, 10000000000000000)
        }
    }

    OnButtonAchievementScam(Agent: agent): void = {
        SaveManager := GetSaveManager()
        if {
            Save := SaveManager.ProxySaves[Agent]
        } then {
            SaveManager.IncrementAchievement(Agent, _Global.achievement_type.Scam, Save.NumOfAchievementsCompletedScam, 10000000000000000)
        }
    }

    OnButtonAchievementMoney(Agent: agent): void = {
        SaveManager := GetSaveManager()
        if {
            Save := SaveManager.ProxySaves[Agent]
        } then {
            SaveManager.IncrementAchievement(Agent, _Global.achievement_type.Money, Save.NumOfAchievementsCompletedMoney, 10000000000000000)
        }
    }

    OnButtonAchievementRebirth(Agent: agent): void = {
        SaveManager := GetSaveManager()
        if {
            Save := SaveManager.ProxySaves[Agent]
        } then {
            SaveManager.IncrementAchievement(Agent, _Global.achievement_type.Rebirth, Save.NumOfAchievementsCompletedRebirth, 10000000000000000)
        }
    }

    OnButtonActivateXpBoost(Agent: agent): void = {
        if {
            PlayerData := Agent.GetPlayerData[]
        } then {
            PlayerData.SetXpBoostLevel(1)
        }
    }

    OnButtonBuildWholeTycoon(Agent: agent): void = {
        spawn{OnButtonBuildWholeTycoonSuspends(Agent)}
    }

    OnButtonBuildWholeTycoonSuspends(Agent: agent)<suspends>: void = {
        Utility := GetUtility()
        if {
            PlayerData := Agent.GetPlayerData[]
            TycoonManager := PlayerData.MaybeTycoonManager?
        } then {
            if(Billboard := BillboardBuildWholeTycoonList[CurBuildWholeTycoonIndex]) {
                set CurBuildWholeTycoonIndex = CurBuildWholeTycoonIndex + 1
                var NumOfButtonsBought: int = 0
                for(Button: TycoonManager.TycoonButtons) {
                    TycoonManager.TryToBuyProp(Agent, Button)
                    set NumOfButtonsBought = NumOfButtonsBought + 1
                    Billboard.SetText(Utility.FormatMessage(ToString(NumOfButtonsBought)))
                    Sleep(1.0)
                }
            }
        }
    }
}