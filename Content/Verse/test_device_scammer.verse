using { /Fortnite.com/Game }
using { /Fortnite.com/AI }
using { /Fortnite.com/FortPlayerUtilities }
using { /Fortnite.com/Devices }
# using { /Fortnite.com/Devices/CreativeAnimation }
using { /Fortnite.com/Characters }
using { /Verse.org/Simulation }
using { /Verse.org/Simulation/Tags }
using { /Verse.org/Random }
using { /Fortnite.com/Teams }
# using { /Verse.org/Assets }
# using { /Verse.org/Concurrency }
# using { /UnrealEngine.com/Temporary/Diagnostics }
using { /UnrealEngine.com/Temporary/SpatialMath }
using { /UnrealEngine.com/Temporary/UI}
using { /Fortnite.com/UI }

# See https://dev.epicgames.com/documentation/en-us/uefn/create-your-own-device-in-verse for how to create a verse device.

# A Verse-authored creative device that can be placed in a level
test_device_scammer := class(creative_device) {
    @editable
    TestButton: button_device = button_device {}

    @editable
    TestButtonUpgradeMenu: button_device = button_device {}

    # Constants
    UiSlot: player_ui_slot = player_ui_slot {
        ZOrder := 99999
        InputMode := ui_input_mode.All
    }

    # Runtime
    var MaybeCanvas: ?canvas = false

    # Runs when the device is started in a running game
    OnBegin<override>()<suspends>:void = {
        TestButton.InteractedWithEvent.Subscribe(OnTestButton)

        TestButtonUpgradeMenu.InteractedWithEvent.Subscribe(OnTestButtonUpgradeMenu)
    }

    OnTestButtonUpgradeMenu(Agent: agent): void = {
        Print("OnTestButtonUpgradeMenu")
        
        if {
            PlayerData := Agent.GetPlayerData[]
        } then {
            # PlayerData.UiUpgradeMenu.AddUiToPlayer(Agent, false)
        }
    }

    OnTestButton(Agent: agent): void = {
        Print("OnTestButton")
        
        AddUiToPlayer(Agent)
    }

    AddUiToPlayer(Agent: agent): void = {
        CanvasUi: canvas = canvas {
            Slots := CreateSlots(Agent)
        }

        set MaybeCanvas = option{CanvasUi}

        if {
            PlayerUI := GetPlayerUI[player[Agent]]
            Canvas := MaybeCanvas?
        } then {
            PlayerUI.AddWidget(Canvas, UiSlot)
        }
    }

    RemoveUiFromPlayer(Agent: agent): void = {
        if {
            PlayerUI := GetPlayerUI[player[Agent]]
            Canvas := MaybeCanvas?
        } then {
            Print("TEST: RemoveUiFromPlayer: Found Canvas")
            PlayerUI.RemoveWidget(Canvas)
            set MaybeCanvas = false
        } else {
            Print("TEST: RemoveUiFromPlayer: Didn't find Canvas!!")
        }
    }

    CreateSlots(Agent: agent): []canvas_slot = {
        Print("TEST: CreateSlots")
        Utility := GetUtility()
        var Slots: []canvas_slot = array {}

        # Init buttons
        CloseShopButton := button_loud{DefaultText := Utility.FormatMessage("EXIT")}
        CloseShopButton.OnClick().Subscribe(OnExitButtonClicked)

        set Slots = Slots + array {
            <# Background image
            canvas_slot {
                Anchors := anchors{Minimum := vector2{X := 0.5, Y := 0.5}, Maximum := vector2{X := 0.5, Y := 0.5}}
                Offsets := margin{Top := -792.540527, Left := -712.960938, Right := 1509.323608, Bottom := 1510.134521}
                Alignment := vector2{X := 0.0, Y := 0.0}
                SizeToContent := false
                Widget := TextureBlockBackground
                ZOrder := 900
            },
            #>
            # Close shop button
            canvas_slot {
                Anchors := anchors{Minimum := vector2{X := 0.5, Y := 0.5}, Maximum := vector2{X := 0.5, Y := 0.5}}
                Offsets := margin{Top := 397.871246, Left := -42.173973, Right := 163.563293, Bottom := 40.921097}
                Alignment := vector2{X := 0.0, Y := 0.0}
                SizeToContent := false
                Widget := CloseShopButton
                ZOrder := 900
            }
        }

        # Shop item button slots
        ShopItemButton1 := button_loud{}
        ShopItemButton2 := button_loud{}
        ShopItemButton3 := button_loud{}
        ShopItemButton4 := button_loud{}
        ShopItemButton5 := button_loud{}
        ShopItemButton6 := button_loud{}
        ShopItemButton7 := button_loud{}
        ShopItemButton8 := button_loud{}
        ShopItemButton9 := button_loud{}
        ShopItemButton10 := button_loud{}

        ShopItemButtonList: []button_loud = array {
            ShopItemButton1,
            ShopItemButton2,
            ShopItemButton3,
            ShopItemButton4,
            ShopItemButton5,
            ShopItemButton6,
            ShopItemButton7,
            ShopItemButton8,
            ShopItemButton9,
            ShopItemButton10
        }

        for(ItemIndex := 0 .. 1000 - 1) {
            set Slots = Slots + array {
                canvas_slot {
                    Anchors := anchors{Minimum := vector2{X := 0.5, Y := 0.5}, Maximum := vector2{X := 0.5, Y := 0.5}}
                    Offsets := margin{Top := 297.871246, Left := -42.173973, Right := 163.563293, Bottom := 40.921097}
                    Alignment := vector2{X := 0.0, Y := 0.0}
                    SizeToContent := false
                    Widget := ShopItemButton1
                    ZOrder := 900
                }
            }
        }
        
        return Slots
    }

    OnExitButtonClicked(WidgetMessage: widget_message): void = {
        Agent := WidgetMessage.Player
        CloseShop(Agent)
    }

    CloseShop(Agent: agent): void = {
        Print("CloseShop UI")
        RemoveUiFromPlayer(Agent)
    }
}