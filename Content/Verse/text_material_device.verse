using { /Fortnite.com/Game }
using { /Fortnite.com/Devices }
using { /Fortnite.com/Devices/CreativeAnimation }
using { /Fortnite.com/Characters }
using { /Verse.org/Simulation }
using { /Verse.org/Simulation/Tags }
using { /Verse.org/Random }
using { /Verse.org/Assets }
using { /Verse.org/Colors }
using { /Verse.org/Concurrency }
using { /UnrealEngine.com/Temporary/Diagnostics }
using { /UnrealEngine.com/Temporary/SpatialMath }

<#
text_material_device := class(creative_device) {
    @editable
    TextProp: creative_prop = creative_prop {}

    @editable
    TextLine1: string = ""

    @editable
    TextLine2: string = ""

    @editable
    MaterialColorLine2: color = color{R:= 0.0, G:= 128.0, B:= 0.0}

    OnBegin<override>()<suspends>: void = {
        SetMaterialText()
    }

    SetMaterialText(): void = {
        Utility := GetUtility()
        Utility.SetMaterialText(TextProp, TextLine1, TextLine2, array{}, option{MaterialColorLine2})
    }
}
#>