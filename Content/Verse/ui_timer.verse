# using { /Fortnite.com/Game }
using { /Fortnite.com/FortPlayerUtilities }
using { /Fortnite.com/Devices }
# using { /Fortnite.com/Devices/CreativeAnimation }
using { /Fortnite.com/Characters }
using { /Fortnite.com/UI }
using { /Verse.org/Simulation }
using { /Verse.org/Simulation/Tags }
using { /Verse.org/Random }
using { /Fortnite.com/Teams }
# using { /Verse.org/Assets }
# using { /Verse.org/Concurrency }
using { /UnrealEngine.com/Temporary/UI }
# using { /UnrealEngine.com/Temporary/SpatialMath }

StringToMessage<public><localizes>(StringValue:string): message = "{StringValue}"

ui_timer := class() {
    TimerUpdateRate: float = 1.0
    var TextBlock: text_block = text_block {}
    var MaybeRoutine: ?routine = false
    var IsCountingDown: logic = false
    var CountDownDuration: float = 0.0

    var MaybeListener: ?(tuple()->void) = false
    var TextToPrefix: string = ""

    StartTimer(ParmTextBlock: text_block, MaybeDuration: ?float, ParmTextToPrefix: string): void = {
        StopTimer()
        
        if(InDuration := MaybeDuration?) {
            set CountDownDuration = InDuration * 1.0
            set IsCountingDown = true
        }
        
        set TextBlock = ParmTextBlock
        set TextToPrefix = ParmTextToPrefix
        TextBlock.SetVisibility(widget_visibility.Visible)
        NewRoutine := SpawnRoutine(TimerTick)
        set MaybeRoutine = option{NewRoutine}
    }

    StopTimer(): void = {
        TextBlock.SetVisibility(widget_visibility.Hidden)
        if (Routine := MaybeRoutine?) {
            Routine.Cancel()
            set MaybeRoutine = false
        }
    }

    UpdateTimer(Seconds: int): void = {
        Utility := GetUtility()
        FormattedTime := Utility.FormatTimePrefixText(Seconds, TextToPrefix)
        TextBlock.SetText(FormattedTime)
    }

    TimerTick()<suspends>: void = {
        var StartTime: float = GetSimulationElapsedTime()
        loop {
            Sleep(TimerUpdateRate)
            TimePassed := GetSimulationElapsedTime() - StartTime
            if(IsCountingDown = true) {
                if (TimePassed >= CountDownDuration ){
                    OnTimerFinished()
                    return
                } else {
                    if (SecondsLeft := Floor[CountDownDuration - TimePassed]) {
                        UpdateTimer(SecondsLeft)
                    }
                }
            } else {
                if (SecondsPassed := Floor[TimePassed]) {
                    UpdateTimer(SecondsPassed)
                }
            }
        }
    }

    SetCallback(ListenerFunction(): void): void = {
        set MaybeListener = option{ListenerFunction}
    }

    RemoveCallback(): void = {
        set MaybeListener = false
    }

    OnTimerFinished():void = {
        TextBlock.SetVisibility(widget_visibility.Hidden)
        if (Listener := MaybeListener?){
            Listener()
        }
    }
}