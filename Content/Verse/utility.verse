using { /Verse.org/Simulation }
using { /Verse.org/Colors }
# using { /Verse.org/Verse }
using { /Verse.org/Random }
using { /Fortnite.com/Devices }
# using { /Fortnite.com/UI }
using { /Fortnite.com/Characters }
using { /UnrealEngine.com/Temporary/Diagnostics }
using { /UnrealEngine.com/Temporary/SpatialMath }
using { /UnrealEngine.com/Temporary/UI}
using { /Verse.org/Concurrency }

my_channel := class(log_channel) {}
my_debug_draw := class(debug_draw_channel) {}

# Global variables
var GlobalGameManager: weak_map(session, _Global.game_manager_device) = map {}

SetGameManager(GameManager: _Global.game_manager_device): void = {
    if(set GlobalGameManager[GetSession()] = GameManager) {}
}

GetGameManager(): _Global.game_manager_device = {
    if (GameManager := GlobalGameManager[GetSession()]) {
        return GameManager
    } else {
        TaggedActorsGameManager := GetCreativeObjectsWithTag(_Global.verse_game_manager_device{})
        GameManagerDevicesTotal := for (Object: TaggedActorsGameManager, Obj := _Global.game_manager_device[Object]) {
            Obj
        }

        if(GameManager := GameManagerDevicesTotal[0]) {
            SetGameManager(GameManager)
            return GameManager
        } else {
            Print("ERR - Found no Game Manager")
            return _Global.game_manager_device {}
        }
    }
}

var GlobalSaveManager: weak_map(session, _Global.save_manager_device) = map {}

SetSaveManager(SaveManager: _Global.save_manager_device): void = {
    if(set GlobalSaveManager[GetSession()] = SaveManager) {}
}

GetSaveManager(): _Global.save_manager_device = {
    if (SaveManager := GlobalSaveManager[GetSession()]) {
        return SaveManager
    } else {
        TaggedActorsSaveManager := GetCreativeObjectsWithTag(_Global.verse_save_manager_device{})
        SaveManagerDevicesTotal := for (Object: TaggedActorsSaveManager, Obj := _Global.save_manager_device[Object]) {
            Obj
        }

        if(SaveManager := SaveManagerDevicesTotal[0]) {
            SetSaveManager(SaveManager)
            return SaveManager
        } else {
            Print("ERR - Found no Save Manager")
            return _Global.save_manager_device {}
        }
    }
}

var GlobalMessageManager: weak_map(session, message_manager_device) = map {}

SetMessageManager(MessageManager: message_manager_device): void = {
    if(set GlobalMessageManager[GetSession()] = MessageManager) {}
}

GetMessageManager(): message_manager_device = {
    if (MessageManager := GlobalMessageManager[GetSession()]) {
        return MessageManager
    } else {
        TaggedActors := GetCreativeObjectsWithTag(verse_message_manager_device{})
        DevicesTotal := for (Object: TaggedActors, Obj := message_manager_device[Object]) {
            Obj
        }

        if(MessageManager := DevicesTotal[0]) {
            SetMessageManager(MessageManager)
            return MessageManager
        } else {
            Print("ERR - Found no Message Manager")
            return message_manager_device {}
        }
    }
}

var GlobalInputManager: weak_map(session, _Global.input_manager_device) = map {}

SetInputManager(InputManager: _Global.input_manager_device): void = {
    if(set GlobalInputManager[GetSession()] = InputManager) {}
}

GetInputManager(): _Global.input_manager_device = {
    if (InputManager := GlobalInputManager[GetSession()]) {
        return InputManager
    } else {
        TaggedActorsInputManager := GetCreativeObjectsWithTag(_Global.verse_input_manager_device{})
        InputManagerDevicesTotal := for (Object: TaggedActorsInputManager, Obj := _Global.input_manager_device[Object]) {
            Obj
        }

        if(InputManager := InputManagerDevicesTotal[0]) {
            SetInputManager(InputManager)
            return InputManager
        } else {
            Print("ERR - Found no Input Manager")
            return _Global.input_manager_device {}
        }
    }
}

var GlobalQuestManager: weak_map(session, _Global.quest_manager_device) = map {}

SetQuestManager(QuestManager: _Global.quest_manager_device): void = {
    if(set GlobalQuestManager[GetSession()] = QuestManager) {}
}

GetQuestManager(): _Global.quest_manager_device = {
    if (QuestManager := GlobalQuestManager[GetSession()]) {
        return QuestManager
    } else {
        TaggedActorsQuestManager := GetCreativeObjectsWithTag(_Global.verse_quest_manager_device{})
        QuestManagerDevicesTotal := for (Object: TaggedActorsQuestManager, Obj := _Global.quest_manager_device[Object]) {
            Obj
        }

        if(QuestManager := QuestManagerDevicesTotal[0]) {
            SetQuestManager(QuestManager)
            return QuestManager
        } else {
            Print("ERR - Found no Quest Manager")
            return _Global.quest_manager_device {}
        }
    }
}

var GlobalGameConfig: weak_map(session, _Config.game_config) = map {}

SetGameConfig(GameConfig: _Config.game_config): void = {
    if(set GlobalGameConfig[GetSession()] = GameConfig) {}
}

GetGameConfig(): _Config.game_config = {
    if (GameConfig := GlobalGameConfig[GetSession()]) {
        return GameConfig
    } else {
        NewGameConfig := _Config.game_config {}
        SetGameConfig(NewGameConfig)
        return NewGameConfig
    }
}

var GlobalAccoladeManager: weak_map(session, _Global.accolade_manager_device) = map {}

SetAccoladeManager(AccoladeManager: _Global.accolade_manager_device): void = {
    if(set GlobalAccoladeManager[GetSession()] = AccoladeManager) {}
}

GetAccoladeManager(): _Global.accolade_manager_device = {
    if (AccoladeManager := GlobalAccoladeManager[GetSession()]) {
        return AccoladeManager
    } else {
        TaggedActorsAccoladeManager := GetCreativeObjectsWithTag(_Global.verse_accolade_manager_device{})
        AccoladeManagerDevicesTotal := for (Object: TaggedActorsAccoladeManager, Obj := _Global.accolade_manager_device[Object]) {
            Obj
        }

        if(AccoladeManager := AccoladeManagerDevicesTotal[0]) {
            SetAccoladeManager(AccoladeManager)
            return AccoladeManager
        } else {
            Print("ERR - Found no Accolade Manager")
            return _Global.accolade_manager_device {}
        }
    }
}

var GlobalUtility: weak_map(session, utility) = map {}

SetUtility(Utility: utility): void = {
    if(set GlobalUtility[GetSession()] = Utility) {}
}

GetUtility(): utility = {
    if (Utility := GlobalUtility[GetSession()]) {
        return Utility
    } else {
        NewUtility := utility {}
        SetUtility(NewUtility)
        return NewUtility
    }
}

var PlayerDataMap: weak_map(session, [agent]player_data) = map {}

(Agent: agent).SetPlayerData(Data: player_data): void = {
    if(not PlayerDataMap[GetSession()], set PlayerDataMap[GetSession()] = map {}) {}
    if(set PlayerDataMap[GetSession()][Agent] = Data) {}
}

(Agent: agent).GetPlayerData()<decides><transacts>: player_data = {
    return PlayerDataMap[GetSession()][Agent]
}

(Agent: agent).DestroyPlayerData(): void = {
    Session := GetSession()
    var NewMap: [agent]player_data = map {}
    for(Key -> Value: PlayerDataMap[Session], Key <> Agent) {
        if(set NewMap[Key] = Value) {}
    }
    if(set PlayerDataMap[Session] = NewMap) {}
}

# Global npc_data map
var NpcDataMap: weak_map(session, [agent]_Npc.npc_data) = map {}

(Agent: agent).SetNpcData(Data: _Npc.npc_data): void = {
    if(not NpcDataMap[GetSession()], set NpcDataMap[GetSession()] = map {}) {}
    if(set NpcDataMap[GetSession()][Agent] = Data) {}
}

(Agent: agent).GetNpcData()<decides><transacts>: _Npc.npc_data = {
    return NpcDataMap[GetSession()][Agent]
}

(Agent: agent).DestroyNpcData(): void = {
    Session := GetSession()
    var NewMap: [agent]_Npc.npc_data = map {}
    for(Key -> Value: NpcDataMap[Session], Key <> Agent) {
        if(set NewMap[Key] = Value) {}
    }
    if(set NpcDataMap[Session] = NewMap) {}
}

# Global npc_data_guard map
var NpcDataMapGuard: weak_map(session, [agent]_Npc.npc_data_guard) = map {}

(Agent: agent).SetNpcDataGuard(Data: _Npc.npc_data_guard): void = {
    if(not NpcDataMapGuard[GetSession()], set NpcDataMapGuard[GetSession()] = map {}) {}
    if(set NpcDataMapGuard[GetSession()][Agent] = Data) {}
}

(Agent: agent).GetNpcDataGuard()<decides><transacts>: _Npc.npc_data_guard = {
    return NpcDataMapGuard[GetSession()][Agent]
}

(Agent: agent).DestroyNpcDataGuard(): void = {
    Session := GetSession()
    var NewMap: [agent]_Npc.npc_data_guard = map {}
    for(Key -> Value: NpcDataMapGuard[Session], Key <> Agent) {
        if(set NewMap[Key] = Value) {}
    }
    if(set NpcDataMapGuard[Session] = NewMap) {}
}

# Global pet npc_data map
var NpcPetDataMap: weak_map(session, [agent]_Npc.npc_pet_data) = map {}

(Agent: agent).SetNpcPetData(Data: _Npc.npc_pet_data): void = {
    if(not NpcPetDataMap[GetSession()], set NpcPetDataMap[GetSession()] = map {}) {}
    if(set NpcPetDataMap[GetSession()][Agent] = Data) {}
}

(Agent: agent).GetNpcPetData()<decides><transacts>: _Npc.npc_pet_data = {
    return NpcPetDataMap[GetSession()][Agent]
}

(Agent: agent).DestroyNpcPetData(): void = {
    Session := GetSession()
    var NewMap: [agent]_Npc.npc_pet_data = map {}
    for(Key -> Value: NpcPetDataMap[Session], Key <> Agent) {
        if(set NewMap[Key] = Value) {}
    }
    if(set NpcPetDataMap[Session] = NewMap) {}
}

utility := class() {
    # How to use Log:
    # Log.Print("Could not obtain 'fort_character' from 'player'.", ?Level := log_level.Warning)
    _Log: log = log { Channel := my_channel }
    
    PlayerOffset: vector3 = vector3{X := 0.0, Y := 0.0, Z := -80.0}

    UnitVector: vector3 = vector3{X := 1.0, Y := 1.0, Z := 1.0}
    ForwardVector: vector3 = vector3{X := 1.0, Y := 0.0, Z := 0.0}
    BackwardVector: vector3 = vector3{X := -1.0, Y := 0.0, Z := 0.0}
    LeftVector: vector3 = vector3{X := 0.0, Y := -1.0, Z := 0.0}
    RightVector: vector3 = vector3{X := 0.0, Y := 1.0, Z := 0.0}
    UpVector: vector3 = vector3{X := 0.0, Y := 0.0, Z := 1.0}
    DownVector: vector3 = vector3{X := 0.0, Y := 0.0, Z := -1.0}

    # HiddenPos: vector3 = vector3 {X := 0.0, Y := 0.0, Z := -10000.00}
    HiddenPos: vector3 = vector3 {X := -78686.375593, Y := 62761.129924, Z := 9175.250928}
    HiddenTrans: transform = transform {
        # Translation := vector3 {X := 0.0, Y := 0.0, Z := -10000.00},
        Translation := vector3 {X := -78686.375593, Y := 62761.129924, Z := 9175.250928},
        Rotation := rotation{}
    }

    LetterToNumberMap: [char]int = map {
        ' ' => 0,
        'a' => 1, 'A' => 1,
        'b' => 2, 'B' => 2,
        'c' => 3, 'C' => 3,
        'd' => 4, 'D' => 4,
        'e' => 5, 'E' => 5,
        'f' => 6, 'F' => 6,
        'g' => 7, 'G' => 7,
        'h' => 8, 'H' => 8,
        'i' => 9, 'I' => 9,
        'j' => 10, 'J' => 10,
        'k' => 11, 'K' => 11,
        'l' => 12, 'L' => 12,
        'm' => 13, 'M' => 13,
        'n' => 14, 'N' => 14,
        'o' => 15, 'O' => 15,
        'p' => 16, 'P' => 16,
        'q' => 17, 'Q' => 17,
        'r' => 18, 'R' => 18,
        's' => 19, 'S' => 19,
        't' => 20, 'T' => 20,
        'u' => 21, 'U' => 21,
        'v' => 22, 'V' => 22,
        'w' => 23, 'W' => 23,
        'x' => 24, 'X' => 24,
        'y' => 25, 'Y' => 25,
        'z' => 26, 'Z' => 26,
        '0' => 27, '1' => 28, '2' => 29, '3' => 30, '4' => 31,
        '5' => 32, '6' => 33, '7' => 34, '8' => 35, '9' => 36,
        ':' => 37
    }

    ConvertStringToTextNumArray(InputString: string): []float = {
        var RemainingString: string = InputString
        var NumArray: []float = array {}
        for(CurPower := 0 .. InputString.Length - 1) {
            if {
                CurLetterChar := RemainingString[RemainingString.Length - 1]
            } then {
                ResultNum := ConvertCharToTextNumber(CurLetterChar)
                set NumArray = NumArray + array{ResultNum}
            }
            # Remove the last letter, since we processed it
            if {
                Temp := RemainingString.Slice[0, RemainingString.Length - 1]
            } then {
                set RemainingString = Temp
            }
        }
        ReverseNumArray := NumArray.Reverse()
        return ReverseNumArray
    }

    ConvertCharToTextNumber(InputChar: char): float = {
        if {
            CurLetterNumber := LetterToNumberMap[InputChar]
        } then {
            return CurLetterNumber * 1.0
        } else {
            Print("ERR: ConvertCharToNumber: Couldn't convert Letter: {InputChar}")
            return 0.0
        }
    }

    # Generic type doesn't work, need different generic type specification
    CloneArray(InputArray: []type): []type = {
        var NewArray: []type = InputArray
        <#
        for(Obj:InputArray) {
            set NewArray = NewArray + array{Obj}
        }
        #>
        return NewArray
    }

    # Check which player is closest to the prop
    GetClosestFortCharacter(Prop: creative_prop, PlayerList: []player, MaxDistance: float, MaybeTargetToIgnore: ?fort_character): ?fort_character = {
        var ClosestFortCharacter : ?fort_character = false

        PropPos := Prop.GetTransform().Translation
        for(Player: PlayerList) {
            if {
                FortChar := Player.GetFortCharacter[]
                FortChar.IsActive[]
            } then {
                # Check if we are told to ignore a target and if we are currently trying to target that character
                var IgnoreFortChar : logic = false
                if(TargetToIgnore := MaybeTargetToIgnore?) {
                    if(TargetToIgnore = FortChar) {
                        set IgnoreFortChar = true
                    }
                }

                if(IgnoreFortChar = false) {
                    DistanceToCurCharacter := Distance(FortChar.GetTransform().Translation, PropPos)
                    if(DistanceToCurCharacter <= MaxDistance) {
                        if(TargetFortCharacter := ClosestFortCharacter?) {
                            DistanceToTarget := Distance(TargetFortCharacter.GetTransform().Translation, PropPos)
        
                            if(DistanceToCurCharacter < DistanceToTarget) {
                                set ClosestFortCharacter = option{FortChar}
                            }
                        } else {
                            set ClosestFortCharacter = option{FortChar}
                        }
                    }
                }
            }
        }

        return ClosestFortCharacter
    }

    GetAllPlayersInRange(CompareProp: creative_prop, Players: []agent, ?MaxRange: float=100.0)<transacts>: []agent = {
        var PlayersInRange: []agent = array{}
        for(Player: Players) {
            if {
                FortChar := Player.GetFortCharacter[]
                FortChar.IsActive[]
            } then {
                Print("Distance: {Distance(FortChar.GetTransform().Translation, CompareProp.GetTransform().Translation)}")
                if(MyDistance := Distance(FortChar.GetTransform().Translation, CompareProp.GetTransform().Translation), MyDistance < MaxRange) {
                    set PlayersInRange = PlayersInRange + array{Player}
                }
            }
        }
        return PlayersInRange
    }

    GetClosestPlayerInRange(CompareTrans: transform, Players: []agent, ?MaxRange:float=100.0)<transacts>: ?agent = {
        var ClosestDistance: float = 9999999999999.0
        var Closest: ?agent = false
        for(Player: Players) {
            if {
                FortChar := Player.GetFortCharacter[]
                FortChar.IsActive[]
            } then {
                if(MyDistance := Distance(FortChar.GetTransform().Translation, CompareTrans.Translation), MyDistance < ClosestDistance) {
                    set Closest = option{Player}
                    set ClosestDistance = MyDistance
                }
            }
        }
        # Print("ClosestDistance: {ClosestDistance}")
        if(ClosestDistance <= MaxRange) {
            return Closest
        } else {
            return false
        }
    }

    # Returns the furthest player and the distance to the player
    GetFurthestPlayer(CompareProp: creative_prop, Players: []agent)<transacts>: tuple(?agent, float) = {
        var FurthestDistance : float = 0.0
        var Furthest: ?agent = false
        for(Player: Players) {
            if {
                FortChar := Player.GetFortCharacter[]
                FortChar.IsActive[]
            } then {
                if(MyDistance := Distance(FortChar.GetTransform().Translation, CompareProp.GetTransform().Translation), MyDistance > FurthestDistance) {
                    set Furthest = option{Player}
                    set FurthestDistance = MyDistance
                }
            }
        }
        Print("FurthestDistance: {FurthestDistance}")
        return (Furthest, FurthestDistance)
    }

    # Returns relative X and Y coordinates
    GetRandomPointInCircle(Radius: float): tuple(float, float) = {
        Theta := GetRandomFloat(0.0, 1.0) * 2 * PiFloat
        var R : float = GetRandomFloat(0.0, 1.0)
        X := GetRandomFloat(0.0, 1.0)
        if(X > R) {
            set R = X
        }
        return (Radius * R * Cos(Theta), Radius * R * Sin(Theta))
    }

    # Returns relative X and Y coordinates
    # TODO: Test if this function actually works correctly
    GetRandomPointOnCircleEdge(Radius: float): tuple(float, float) = {
        Theta := GetRandomFloat(0.0, 1.0) * 2 * PiFloat
        X := Radius * Cos(Theta)
        Y := Radius * Sin(Theta)
        return (X, Y)
    }

    # Returns relative X and Y coordinates
    # TODO: Test if this function actually works correctly
    # Theta: Angle (in radians, not degrees)
    GetPointOnCircleEdge(Theta: float, Radius: float): tuple(float, float) = {
        X := Radius * Cos(Theta)
        Y := Radius * Sin(Theta)
        return (X, Y)
    }

    GetRandomItemWeightedRollChance(RollChanceList: []float): int = {
        var ChosenItemIndex : int = -1
        var TotalRollChance : float = 0.0
        var CurRollValue : float = 0.0
        for(RollChance : RollChanceList) {
            set TotalRollChance = TotalRollChance + RollChance
        }
        RandomFloat := GetRandomFloat(0.0, TotalRollChance)
        for(RollIndex -> RollChance : RollChanceList) {
            # Check if we already chose an item
            if(ChosenItemIndex = -1) {
                set CurRollValue = CurRollValue + RollChance
                if(RandomFloat <= CurRollValue) {
                    set ChosenItemIndex = RollIndex
                }
            }
        }

        if(ChosenItemIndex = -1) {
            Print("ERR: GetRandomItemWeightedRollChance: Found no item!")
        }

        return ChosenItemIndex
    }

    FormatFloat(Num: float, Fractions: int, Decimals: int): string = {
        var FormattedNum : string = ""
        # 1840 E is maximum (64 bit)
        if(Num >= 1000000000000000000.0) {
            # Convert to E
            set FormattedNum = FloatToTruncatedString(Num / 1000000000000000000.0, Fractions) + " E"
        } else if(Num >= 1000000000000000.0) {
            # Convert to P
            set FormattedNum = FloatToTruncatedString(Num / 1000000000000000.0, Fractions) + " P"
        } else if(Num >= 1000000000000.0) {
            # Convert to trillions
            set FormattedNum = FloatToTruncatedString(Num / 1000000000000.0, Fractions) + " T"
        } else if(Num >= 1000000000.0) {
            # Convert to billions
            set FormattedNum = FloatToTruncatedString(Num / 1000000000.0, Fractions) + " B"
        } else if(Num >= 1000000.0) {
            # Convert to millions
            set FormattedNum = FloatToTruncatedString(Num / 1000000.0, Fractions) + " M"
        } else if(Num >= 1000.0) {
            # Convert to thousands
            set FormattedNum = FloatToTruncatedString(Num / 1000.0, Fractions) + " K"
        } else {
            set FormattedNum = FloatToTruncatedString(Num, Decimals)
        }
        return FormattedNum
    }

    FormatTimeToString(InputSeconds: int): string = {
        var TimeString: string = ""
        if {
            Minutes := Floor(InputSeconds / 60)
            Seconds := Mod[InputSeconds, 60]
        } then {
            if(Minutes < 10) {
                set TimeString = TimeString + "0"
            }
            set TimeString = TimeString + ToString(Minutes) + ":"
            if(Seconds < 10) {
                set TimeString = TimeString + "0"
            }
            set TimeString = TimeString + ToString(Seconds)
        }
        
        return TimeString
    }

    FormatTimeToMessage(InputSeconds: int): message = {
        TimeString := FormatTimeToString(InputSeconds)
        return FormatMessage(TimeString)
    }

    FormatTimePrefixText(InputSeconds: int, TextToPrefix: string): message = {
        TimeString := FormatTimeToString(InputSeconds)
        return FormatMessage(TextToPrefix + TimeString)
    }

    FormatInt(InputNum: int): message = {
        return FormatMessage(ToString(InputNum))
    }

    FormatIntWithDots(InputNum: int): string = {
        NumAsStr := ToString(InputNum)
        var OutputStr: string = ""
        var CurIndex: int = 0
        loop {
            if {
                # Check if we reached the end of the input string
                Char := NumAsStr[CurIndex]
            } then {
                # Check if we have to add a dot (every 3 digits)
                if {
                    Mod[NumAsStr.Length - CurIndex, 3] = 0
                    # Don't add dots before the first char
                    OutputStr.Length > 0
                } then {
                    set OutputStr = OutputStr + array{'.'}
                }
                # Add the current sign
                set OutputStr = OutputStr + array{Char}
                set CurIndex = CurIndex + 1
            } else {
                break
            }
        }
        return OutputStr
    }

    # Returns float converted to a string, truncated to Decimals places
    FloatToTruncatedString<public>(Number: float, ParmDecimals: int): []char = {
        var Decimals: int = ParmDecimals
        <# Old method: Check if the decimals are equal to 0, then omit them
        if {
            Floor[Number] * 1.0 = Number
        } then {
            set Decimals = 0
        }
        #>
        
        var Str: []char = ToString(Number)
        if {
            DotIndex := Str.Find['.']
            StopIndex := if(Decimals > 0) then Min(DotIndex+Decimals+1,Str.Length) else DotIndex
            Temp := Str.Slice[0, StopIndex]
            set Str = Temp
        }

        # New method: Cut off every decimal, which is equal to 0 (starting from the right)
        loop {
            if {
                Decimals > 0
                LastDigit := Str[Str.Length - 1]
                LastDigit = '0'
                Temp := Str.Slice[0, Str.Length - 1]
            } then {
                set Str = Temp
                set Decimals = Decimals - 1
                # Remove the dot, if we have no more decimals
                if {
                    Decimals = 0
                    TempFinal := Str.Slice[0, Str.Length - 1]
                } then {
                    set Str = TempFinal
                    break
                }
            } else {
                break
            }
        }
        return Str
    }

    ConvertFloatToPercentageString(Number: float): string = {
        var ReturnString: string = ToString(Number)
        if(Percentage := Round[Number * 100.0]) {
            set ReturnString = ToString(Percentage)
        }
        return ToString(ReturnString + "%")
    }

    <# LetterMaterial is currently bugged and uses too much memory, so it can't be used until fixed
    SetMaterialText(MaterialProp: creative_prop, Line1String: string, Line2String: string, Line2NumArrayPrefixString: []float, MaybeLine2TextColor: ?color): void = {
        Utility := GetUtility()
        MyMaterial := MaterialText.LetterMaterial_material {}

        Line1NumArray := Utility.ConvertStringToTextNumArray(Line1String)
        Line2NumArraySuffix := Utility.ConvertStringToTextNumArray(Line2String)
        Line2NumArrayTotal := Line2NumArrayPrefixString + Line2NumArraySuffix

        # Center the Line1 and Line2 text, by adding empty spaces at the beginning
        Line1MaxLength := 16
        Line2MaxLength := 9
        
        var Line1NumArrayCentered: []float = array {}
        var Line2NumArrayCentered: []float = array {}

        if(Line1NumArray.Length > Line1MaxLength or Line2NumArrayTotal.Length > Line2MaxLength) then {
            Print("ERR: SetMaterialText: Input string is too long!")
        } else {
            # Line 1 Padding
            if {
                Line1PaddingLeft: int = Floor((Line1MaxLength - Line1NumArray.Length) / 2)
            } then {
                for(ItemIndex := 0 .. Line1PaddingLeft - 1) {
                    # Insert blank space
                    set Line1NumArrayCentered = Line1NumArrayCentered + array {0.0}
                }
                set Line1NumArrayCentered = Line1NumArrayCentered + Line1NumArray
            }
            # Line 2 Padding
            if {
                Line2PaddingLeft: int = Floor((Line2MaxLength - Line2NumArrayTotal.Length) / 2)
            } then {
                for(ItemIndex := 0 .. Line2PaddingLeft - 1) {
                    # Insert blank space
                    set Line2NumArrayCentered = Line2NumArrayCentered + array {0.0}
                }
                set Line2NumArrayCentered = Line2NumArrayCentered + Line2NumArrayTotal
            }
        }

        for(SlotId -> LetterKey: Line1NumArrayCentered) {
            # Print("SlotId: {SlotId}, LetterKey: {LetterKey}")
            case(SlotId) {
                0 =>
                    set MyMaterial.Line1Letter1 = LetterKey
                1 =>
                    set MyMaterial.Line1Letter2 = LetterKey
                2 =>
                    set MyMaterial.Line1Letter3 = LetterKey
                3 =>
                    set MyMaterial.Line1Letter4 = LetterKey
                4 =>
                    set MyMaterial.Line1Letter5 = LetterKey
                5 =>
                    set MyMaterial.Line1Letter6 = LetterKey
                6 =>
                    set MyMaterial.Line1Letter7 = LetterKey
                7 =>
                    set MyMaterial.Line1Letter8 = LetterKey
                8 =>
                    set MyMaterial.Line1Letter9 = LetterKey
                9 =>
                    set MyMaterial.Line1Letter10 = LetterKey
                10 =>
                    set MyMaterial.Line1Letter11 = LetterKey
                11 =>
                    set MyMaterial.Line1Letter12 = LetterKey
                12 =>
                    set MyMaterial.Line1Letter13 = LetterKey
                13 =>
                    set MyMaterial.Line1Letter14 = LetterKey
                14 =>
                    set MyMaterial.Line1Letter15 = LetterKey
                15 =>
                    set MyMaterial.Line1Letter16 = LetterKey
                _ =>
                    Print("ERR: Line1NumArrayCentered is too big: {Line1NumArrayCentered.Length}")
            }
        }

        for(SlotId -> LetterKey: Line2NumArrayCentered) {
            # Print("SlotId: {SlotId}, LetterKey: {LetterKey}")
            case(SlotId) {
                0 =>
                    set MyMaterial.Line2Letter1 = LetterKey
                1 =>
                    set MyMaterial.Line2Letter2 = LetterKey
                2 =>
                    set MyMaterial.Line2Letter3 = LetterKey
                3 =>
                    set MyMaterial.Line2Letter4 = LetterKey
                4 =>
                    set MyMaterial.Line2Letter5 = LetterKey
                5 =>
                    set MyMaterial.Line2Letter6 = LetterKey
                6 =>
                    set MyMaterial.Line2Letter7 = LetterKey
                7 =>
                    set MyMaterial.Line2Letter8 = LetterKey
                8 =>
                    set MyMaterial.Line2Letter9 = LetterKey
                _ =>
                    Print("ERR: Line2NumArrayCentered is too big: {Line2NumArrayCentered.Length}")
            }
        }

        if {
            Line2TextColor := MaybeLine2TextColor?
        } then {
            set MyMaterial.MatTextColor = Line2TextColor
        }

        MaterialProp.SetMaterial(MyMaterial)
    }
    #>

    GetBillboardTextFloat(MyFloat: float, Fractions: int, Decimals: int): message = {
        FormattedFloat := FormatFloat(MyFloat, Fractions, Decimals)
        return FormatMessage(FormattedFloat)
    }

    GetBillboardTextFloatAppend(MyFloat: float, Fractions: int, Decimals: int, StringToAppend: string): message = {
        FormattedFloat := FormatFloat(MyFloat, Fractions, Decimals)
        return FormatMessage(FormattedFloat + StringToAppend)
    }

    GetBillboardTextFloatPrefix(MyFloat: float, Fractions: int, Decimals: int, StringToAppend: string): message = {
        FormattedFloat := FormatFloat(MyFloat, Fractions, Decimals)
        return FormatMessage(StringToAppend + FormattedFloat)
    }

    GetBillboardTextIntPrefix(MyInt: int, Fractions: int, Decimals: int, StringToPrefix: string): message = {
        FormattedInt := FormatFloat(MyInt * 1.0, Fractions, Decimals)
        return FormatMessage(StringToPrefix + FormattedInt)
    }

    GetBillboardTextIntAppend(MyInt: int, Fractions: int, Decimals: int, StringToAppend: string): message = {
        FormattedInt := FormatFloat(MyInt * 1.0, Fractions, Decimals)
        return FormatMessage(FormattedInt + StringToAppend)
    }

    FormatMessage<localizes>(MyString: string) : message = "{MyString}"
    PlayerNameMessage<localizes>(Agent: agent, MyString: string): message = "{Agent} {MyString}"
    
    DebugPrintRotation(Rotation: rotation, PrintString: string): void = {
        Print("{PrintString}")
        MyArray := Rotation.GetYawPitchRollDegrees()
        for(Axis : MyArray) {
            Print("{Axis}")
        }
    }

    DebugDraw: debug_draw = debug_draw{Channel := my_debug_draw}

    # Used for specifying a point offset from the NPC pelvis to the head to draw the look at arrow from.
    VerticalOffsetToNPCHead<private>:float = 55.0

    # This function draws a box around a specified position for a finite amount of time.
    # NOTE: To see this in game, Verse Debug Draw must be enabled in Island Settings.
    DrawDebugLocation(Location:vector3):void = {
        DebugDraw.DrawSphere(Location, 
            ?Color := NamedColors.SteelBlue, 
            ?Thickness := 100.0, 
            ?DrawDurationPolicy := debug_draw_duration_policy.FiniteDuration, 
            ?Duration := 5.0)
    }

    # This function draws an arrow from the Agent's head to its look at point every half a second. 
    # NOTE: To see this in game, Verse Debug Draw must be enabled in Island Settings.
    DrawDebugLookAt(Character: fort_character, LookAtPoint: vector3)<suspends>:void = {
        DebugDraw.DrawArrow(Character.GetTransform().Translation + vector3{Z := VerticalOffsetToNPCHead}, 
            LookAtPoint, 
            ?ArrowSize := 50.0, 
            ?Color := NamedColors.Yellow,
            ?Thickness := 5.0, 
            ?DrawDurationPolicy := debug_draw_duration_policy.FiniteDuration, 
            ?Duration := 5.0)
    }
}

# https://dev.epicgames.com/community/snippets/4dl/fortnite-cancelable-spawned-coroutines
routine<public> := class:
    Event<protected> : event() = event(){}
    var MaybeTask<protected> : ?task(void) = false 
    var _IsRunning<protected> : logic = false
    Cancel<public>():void = Event.Signal()
 
    GetTask<public>()<decides><transacts>:task(void) = MaybeTask?
 
    IsRunning<public>()<decides><transacts>:void = _IsRunning?

routine_typed(t: type) := class(routine):
    Func<public> : type{_(:t)<suspends>:void} 
    Start<internal>(Args: t):void = 
        Task := spawn {RunRoutine(Args)}
        set MaybeTask = option {Task}
 
    RunRoutine<protected>(Args: t)<suspends>: void = 
        set _IsRunning = true
        race:
            Func(Args)
            Event.Await()
        set _IsRunning = false
 
routine_no_args<public> := class(routine):
    Func<public> : type{_()<suspends>:void} 
    Start<internal>():void = 
        Task := spawn {RunRoutine()}
        set MaybeTask = option {Task}
 
    RunRoutine<protected>()<suspends>: void = 
        set _IsRunning = true
        race:
            Func()
            Event.Await()
        set _IsRunning = false
 
SpawnRoutineWithArgs<public>(Func(:t)<suspends>:void, Args: t where t:type):routine_typed(t) = 
    Routine := routine_typed(t){Func := Func}
    Routine.Start(Args)
    return Routine

SpawnRoutine<public>(Func()<suspends>:void):routine_no_args =
    Routine := routine_no_args{Func := Func}
    Routine.Start()
    return Routine

# https://dev.epicgames.com/community/snippets/rmpp/fortnite-generic-queue-implementation
queue<public>(t:type) := class:
    Elements<internal>:[]t = array{}
    
    Enqueue<public>(NewElement:t):queue(t)=
        queue(t){Elements := Elements + array{NewElement}}
    
    Dequeue<public>()<decides><transacts>:tuple(queue(t),t)=
        FirstElement := Front[]
        (queue(t){Elements := Elements.RemoveElement[0]}, FirstElement)
    
    Size<public>()<transacts>:int=
        Elements.Length
    
    IsEmpty<public>()<decides><transacts>:void=
        Size() = 0
    
    Front<public>()<decides><transacts>:t=
        Elements[0]
    
    Rear<public>()<decides><transacts>:t=
        Elements[Elements.Length - 1]
        
CreateQueue<public><constructor>(InitialElements:[]t where t:type) := queue(t):
    Elements := InitialElements

event_handler(a: type, b: type) := class {
    A: a
    B: b
    CallbackFunc: (tuple(a, b) -> void)

    EventSubscription(): void = {
        CallbackFunc(A, B)
    }
}

event_handler_agent(a: type) := class {
    A: a
    CallbackFunc: (tuple(agent, a) -> void)

    EventSubscription(Agent: agent): void = {
        CallbackFunc(Agent, A)
    }
}

event_handler_widget_message(a: type) := class {
    A: a
    CallbackFunc: (tuple(widget_message, a) -> void)

    EventSubscription(WidgetMessage: widget_message): void = {
        CallbackFunc(WidgetMessage, A)
    }
}

event_handler_interaction_result(a: type) := class {
    A: a
    CallbackFunc: (tuple(device_ai_interaction_result, a) -> void)

    EventSubscription(InteractionResult: device_ai_interaction_result): void = {
        CallbackFunc(InteractionResult, A)
    }
}

event_handler_maybe_agent(a: type) := class {
    A: a
    CallbackFunc: (tuple(?agent, a) -> void)

    EventSubscription(Agent: ?agent): void = {
        CallbackFunc(Agent, A)
    }
}

# Removes a key from a map
# var MyMap: [int]string = map{0=>"Test", 3=>"Test2"}
# set MyMap = MyMap.RemoveKey(3)
(Map:[key_type]value_type where key_type:subtype(comparable), value_type:type).RemoveKey(Key:key_type): [key_type]value_type =
    var NewMap: [key_type]value_type = map{}
    for(
        K->V: Map
        K <> Key
        set NewMap[K] = V
    ){}
    NewMap

# Reverses the order of elements in an array
(Input:[]t where t:type).Reverse()<computes>:[]t=
    for (Index -> Unused : Input, Element := Input[Input.Length - Index - 1]). Element