using { /Fortnite.com/Game }
using { /Fortnite.com/Devices }
using { /Fortnite.com/Devices/CreativeAnimation }
using { /Fortnite.com/Characters }
using { /Verse.org/Simulation }
using { /Verse.org/Simulation/Tags }
using { /Verse.org/Random }
using { /Verse.org/Concurrency }
using { /UnrealEngine.com/Temporary/Diagnostics }
using { /UnrealEngine.com/Temporary/SpatialMath }

wildlife_mount_device := class(creative_device) {
    @editable
    SignalRemoteManager: signal_remote_manager_device = signal_remote_manager_device {}

    @editable
    var WildlifeSpawnerList: []wildlife_spawner_device = array {}

    # Constants
    
    # Runtime variables
    WildlifeSpawnersInUseList: []wildlife_spawner_device = array {}

    var LastSpawnPos: transform = transform {}
    
    # Runs when the device is started in a running game
    OnBegin<override>()<suspends>:void = {
        SignalRemoteManager.PrimarySignalEvent.Subscribe(OnPrimarySignal)
        SignalRemoteManager.SecondarySignalEvent.Subscribe(OnPrimarySignal)

        for(SpawnerIndex -> WildlifeSpawner : WildlifeSpawnerList) {
            # Spawned event
            WildlifeSpawner.SpawnedEvent.Subscribe(OnWildlifeSpawned)

            # Dismounted event
            EventHandlerInteractionResult := event_handler_interaction_result(wildlife_spawner_device) {
                A := WildlifeSpawner
                CallbackFunc := OnDismountedEvent
            }

            WildlifeSpawner.DismountedEvent.Subscribe(EventHandlerInteractionResult.EventSubscription)
        }
    }

    OnPrimarySignal(Agent: agent): void = {
        Print("OnPrimarySignal")
        GameManager := GetGameManager()
        if {
            FortChar := Agent.GetFortCharacter[]
            PlayerData := Agent.GetPlayerData[]
        } then {
            # TODO
            <# Check if player has unlocked the mount
            if(PlayerData.MountShopStateList[0] <> 1) {
                Print("WARN: WolfMount: OnPrimarySignal: Didn't spawn the mount, since player didn't buy it")
                return
            }
            #>
            
            set LastSpawnPos = FortChar.GetTransform()
            
            # This needs to be delayed
            Task := spawn{RideMount(option{Agent})}
        }       
    }

    OnWildlifeSpawned(Agent: agent): void = {
        Print("OnWildlifeSpawned")
        if {
            FortCharWolf := Agent.GetFortCharacter[]
            FortCharWolf.IsActive[]
        } then {
            if(FortCharWolf.TeleportTo[LastSpawnPos.Translation, LastSpawnPos.Rotation]) {}
        }
    }

    RideMount(MaybeAgent: ?agent)<suspends>: void = {
        GameManager := GetGameManager()
        if {
            Agent := MaybeAgent?
            PlayerData := Agent.GetPlayerData[]
        } then {
            # Check if the player already has a wildlife spawner assigned
            if(PlayerData.MaybeWildlifeMountManager?) {
                # Do nothing here
            } else {
                set PlayerData.MaybeWildlifeMountManager = option{Self}
                WildlifeSpawner := GetUnusedWildlifeSpawner()
                set PlayerData.MaybeWildlifeSpawner = option{WildlifeSpawner}
            }

            if(WildlifeSpawner := PlayerData.MaybeWildlifeSpawner?) {
                # WildlifeSpawner.Enable()
                WildlifeSpawner.Reset()
                WildlifeSpawner.Spawn()

                Sleep(1.0)

                if(Agent2 := MaybeAgent?) {
                    WildlifeSpawner.Ride(Agent2)
                }
            }
        }    
    }

    # `Source` is the `agent` that dismounted the wildlife.
    # `Target` is the wildlife that was dismounted.
    OnDismountedEvent(InteractionResult: device_ai_interaction_result, WildlifeSpawner: wildlife_spawner_device): void = {
        Print("OnDismountedEvent")
        GameManager := GetGameManager()
        # Deactivate the spawner to destroy the wildlife
        if(Agent := InteractionResult.Source?) {
            WildlifeSpawner.Despawn(Agent)
        }
        
        # WildlifeSpawner.Disable()
    }

    GetUnusedWildlifeSpawner(): wildlife_spawner_device = {
        Print("GetUnusedWildlifeSpawner")
        var WildlifeSpawner: wildlife_spawner_device = wildlife_spawner_device {}
        if {
            Temp := WildlifeSpawnerList[0]
            set WildlifeSpawner = Temp
            # Remove the item from the pool
            Temp2 := WildlifeSpawnerList.RemoveElement[0]
            set WildlifeSpawnerList = Temp2
        } then {
            # WildlifeSpawner.Enable()
            # WildlifeSpawner.Reset()
        } else {
            Print("ERR GetUnusedWildlifeSpawner - Found no unused spawner!")
        }
        return WildlifeSpawner
    }

    ReturnWildlifeSpawner(WildlifeSpawner: wildlife_spawner_device): void = {
        Print("ReturnWildlifeSpawner")
        # Restore lists
        set WildlifeSpawnerList = WildlifeSpawnerList + array{WildlifeSpawner}
    }
}
