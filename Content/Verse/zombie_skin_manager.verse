using { /Fortnite.com/Game }
using { /Fortnite.com/Devices }
using { /Fortnite.com/Devices/CreativeAnimation }
using { /Fortnite.com/Characters }
using { /Verse.org/Simulation }
using { /Verse.org/Simulation/Tags }
using { /Verse.org/Random }
using { /Verse.org/Concurrency }
using { /UnrealEngine.com/Temporary/Diagnostics }
using { /UnrealEngine.com/Temporary/SpatialMath }

# See https://dev.epicgames.com/documentation/en-us/uefn/create-your-own-device-in-verse for how to create a verse device.

# A Verse-authored creative device that can be placed in a level
zombie_skin_manager := class(creative_device) {
    @editable
    ZombieSkinIdle: creative_prop_asset = DefaultCreativePropAsset

    @editable
    ZombieSkinWalk: creative_prop_asset = DefaultCreativePropAsset

    @editable
    ZombieSkinAttack: creative_prop_asset = DefaultCreativePropAsset

    @editable
    ZombieSpawnerList: []creature_spawner_device = array {}

    # Constants
    ZombieActionDespawnDelay: float = 7.0
    ZombieEliminatedDespawnDelay: float = 30.0
    AttackAnimMaxDistance: float = 150.0
    WalkAnimMinDistance: float = 0.001

    # Runtime variables
    var RoutineZombieAction: ?routine = false
    var TimestampLastZombieAction: float = 0.0
    
    var RoutineZombieEliminated: ?routine = false
    var TimestampLastZombieEliminated: float = 0.0

    # Runs when the device is started in a running game
    OnBegin<override>()<suspends>:void = {
        # Listen to events
        for(Spawner : ZombieSpawnerList) {
            Spawner.SpawnedEvent.Subscribe(OnZombieSpawned)
            # Spawner.EliminatedEvent.Subscribe(OnZombieEliminated)
        }
    }

    OnZombieSpawned(Agent: agent): void = {
        if(FortChar := Agent.GetFortCharacter[]) {
            FortChar.Hide()

            IdlePropResult := SpawnProp(ZombieSkinIdle, FortChar.GetTransform().Translation, FortChar.GetTransform().Rotation) 
            if(IdleProp := IdlePropResult(0)?) {
                WalkPropResult := SpawnProp(ZombieSkinWalk, FortChar.GetTransform().Translation, FortChar.GetTransform().Rotation) 
                if(WalkProp := WalkPropResult(0)?) {
                    AttackPropResult := SpawnProp(ZombieSkinAttack, FortChar.GetTransform().Translation, FortChar.GetTransform().Rotation) 
                    if(AttackProp := AttackPropResult(0)?) {
                        spawn{ZombieTick(FortChar, IdleProp, WalkProp, AttackProp)}
                    }
                }
            }
        }
        UpdateZombieAction()
        UpdateZombieEliminatedAction()
    }

    <# Not needed atm
    OnZombieEliminated(InteractionResult: device_ai_interaction_result): void = {
        UpdateZombieAction()
        UpdateZombieEliminatedAction()
    }
    #>

    UpdateZombieAction(): void = {
        set TimestampLastZombieAction = GetSimulationElapsedTime()
        # Spawn zombie action routine if it's not already running
        if(Routine := RoutineZombieAction?) {
            # Do nothing, routine is already running
        } else {
            MyRoutine := SpawnRoutine(ZombieActionTick)
            set RoutineZombieAction = option{MyRoutine}
        }
    }

    UpdateZombieEliminatedAction(): void = {
        set TimestampLastZombieEliminated = GetSimulationElapsedTime()
        # Spawn zombie action routine if it's not already running
        if(Routine := RoutineZombieEliminated?) {
            # Do nothing, routine is already running
        } else {
            MyRoutine := SpawnRoutine(ZombieActionTick)
            set RoutineZombieEliminated = option{MyRoutine}
        }
    }

    ZombieActionTick()<suspends>: void = {
        loop {
            Timestamp := GetSimulationElapsedTime()
            TimePassedSinceEvent := Timestamp - TimestampLastZombieAction
            TimeRemainingForEvent := ZombieActionDespawnDelay - TimePassedSinceEvent
            if(TimeRemainingForEvent <= 0.0) {
                break
            } else {
                Sleep(TimeRemainingForEvent)
            }
        }

        Print("ZombieAction - Timeout")
        OnZombieActionTimeout()
    }

    ZombieEliminatedActionTick()<suspends>: void = {
        loop {
            Timestamp := GetSimulationElapsedTime()
            TimePassedSinceEvent := Timestamp - TimestampLastZombieEliminated
            TimeRemainingForEvent := ZombieEliminatedDespawnDelay - TimePassedSinceEvent
            if(TimeRemainingForEvent <= 0.0) {
                break
            } else {
                Sleep(TimeRemainingForEvent)
            }
        }

        Print("ZombieEliminatedAction - Timeout")
        OnZombieActionTimeout()
    }

    OnZombieActionTimeout(): void = {
        # Despawn all zombies
        for(Spawner : ZombieSpawnerList) {
            Spawner.EliminateCreatures()
        }

        # Clean up routines
        if(Routine := RoutineZombieAction?) {
            set RoutineZombieAction = false
            Routine.Cancel()
        }
        if(Routine := RoutineZombieEliminated?) {
            set RoutineZombieEliminated = false
            Routine.Cancel()
        }
    }
                
    ZombieTick(ZombieCharacter: fort_character, IdleProp: creative_prop, WalkProp: creative_prop, AttackProp: creative_prop)<suspends>: void = {
        Utility := GetUtility()
        var PrevZombiePos: vector3 = vector3 {}
        var IsAttacking: logic = false
        var IsIdle: logic = true
        loop {
            if (not ZombieCharacter.IsActive[]) {
                IdleProp.Dispose()
                WalkProp.Dispose()
                AttackProp.Dispose()
                # These are the same as eliminated event
                UpdateZombieAction()
                UpdateZombieEliminatedAction()
                break
            }

            # Init tick variables
            set IsAttacking = false
            set IsIdle = true
            ZombiePos := ZombieCharacter.GetTransform().Translation

            # Check if the zombie is attacking (close to a player)
            for(Player : GetPlayspace().GetPlayers()) {
                if(IsAttacking = false) {
                    if(PlayerFortChar := Player.GetFortCharacter[]) {
                        PlayerPos := PlayerFortChar.GetTransform().Translation
                        if(Distance(ZombiePos, PlayerPos) <= AttackAnimMaxDistance) {
                            set IsAttacking = true
                        }
                    }
                }
            }

            # Check if the zombie is idle (not moving)
            if(IsAttacking = false) {
                if(Distance(ZombiePos, PrevZombiePos) >= WalkAnimMinDistance) {
                    set IsIdle = false
                } else {
                    set IsIdle = true
                }
            }

            set PrevZombiePos = ZombiePos
            GoalPos := ZombiePos + Utility.PlayerOffset
            GoalRot := ZombieCharacter.GetTransform().Rotation
            if(IsAttacking = true) {
                # Zombie is attacking
                AttackProp.Show()
                IdleProp.Hide()
                WalkProp.Hide()
                if(IdleProp.TeleportTo[GoalPos, GoalRot]) {}
                if(WalkProp.TeleportTo[GoalPos, GoalRot]) {}
                AttackProp.MoveTo(GoalPos, GoalRot, 0.1)
                UpdateZombieAction()
            } else if(IsIdle = true) {
                # Zombie is idle
                IdleProp.Show()
                WalkProp.Hide()
                AttackProp.Hide()
                if(WalkProp.TeleportTo[GoalPos, GoalRot]) {}
                if(AttackProp.TeleportTo[GoalPos, GoalRot]) {}
                IdleProp.MoveTo(GoalPos, GoalRot, 0.1)
            } else {
                # Zombie is moving
                WalkProp.Show()
                IdleProp.Hide()
                AttackProp.Hide()
                if(IdleProp.TeleportTo[GoalPos, GoalRot]) {}
                if(AttackProp.TeleportTo[GoalPos, GoalRot]) {}
                WalkProp.MoveTo(GoalPos, GoalRot, 0.1)
                UpdateZombieAction()
            }
        }
    }
}
