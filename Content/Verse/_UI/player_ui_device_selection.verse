using { /Verse.org/Simulation }
using { /Verse.org/Colors }
using { /Verse.org/Colors/NamedColors }
# using { /Verse.org/Concurrency }
# using { /Verse.org/Verse }
using { /Verse.org/Assets }
using { /Fortnite.com/Devices }
using { /Fortnite.com/UI }
# using { /Fortnite.com/Characters }
# using { /UnrealEngine.com/Temporary/Diagnostics }
using { /UnrealEngine.com/Temporary/SpatialMath }
using { /UnrealEngine.com/Temporary/UI}

<# TODO: Use new UI parent class
player_ui_device_selection := class() {
    # Constants
    MyUiSlot: player_ui_slot = player_ui_slot {
        ZOrder := 3313
        InputMode := ui_input_mode.All
    }

    # Runtime
    var AssignedUiIndex: int = -1
    var MaybeAgent: ?agent = false
    var MaybeCanvas: ?canvas = false
    var MaybePcManager: ?interface_ui_manager = false
    var TimeoutManager: timeout_manager = timeout_manager {}

    InitWidget(Agent: agent): void = {
        Print("player_ui_device_selection - InitWidget")
        set MaybeAgent = option{Agent}
        
        Canvas := canvas {
            Slots := CreateSlots()
        }

        set MaybeCanvas = option{Canvas}
    }

    AddUiToPlayer(Agent: agent, PcManager: interface_ui_manager): void = {
        Print("player_ui_device_selection - AddUiToPlayer")
        set MaybePcManager = option{PcManager}
        
        if(PlayerData := Agent.GetPlayerData[]) {
            if(Canvas := MaybeCanvas?) {
                set AssignedUiIndex = PlayerData.AddUiToPlayer(Canvas, MyUiSlot)
                Print("ADDED UI DEVICE SELECTION: ASSIGNED INDEX: {AssignedUiIndex}")
            }
            TimeoutManager.StartTimeoutTimer(UiTimeoutCallback)
        }
    }

    RemoveUiFromPlayer(Agent: agent): void = {
        if(PlayerData := Agent.GetPlayerData[]) {
            if(Canvas := MaybeCanvas?) {
                Print("REMOVING UI DEVICE SELECTION: ASSIGNED INDEX: {AssignedUiIndex}")
                PlayerData.RemoveUiFromPlayer(AssignedUiIndex)
            }
        }

        TimeoutManager.CancelTimeout()
    }

    CreateCanvasUi(): canvas = {
        CanvasUi: canvas = canvas {
            Slots := CreateSlots()
        }
        return CanvasUi
    }

    CreateSlots(): []canvas_slot = {
        Utility := GetUtility()

        # Buttons need to be defined inside the function
        ButtonClose: button_loud = button_loud{DefaultText:=Utility.FormatMessage("EXIT")}

        # Reward item button slots
        ItemButton1 := button_loud{DefaultText:=Utility.FormatMessage("SELECT")}
        ItemButton2 := button_loud{DefaultText:=Utility.FormatMessage("SELECT")}
        ItemButton3 := button_loud{DefaultText:=Utility.FormatMessage("SELECT")}
        ItemButton4 := button_loud{DefaultText:=Utility.FormatMessage("SELECT")}

        ItemButtonList := array {
            ItemButton1,
            ItemButton2,
            ItemButton3,
            ItemButton4
        }

        # Listen for click events
        for(ButtonIndex -> ItemButton: ItemButtonList) {
            ButtonEventHandler := event_handler_widget_message(int) {
                A := ButtonIndex
                CallbackFunc := OnButtonClick
            }

            SpawnerEventSub := ItemButton.OnClick().Subscribe(ButtonEventHandler.EventSubscription)
        }

        ButtonClose.OnClick().Subscribe(OnExitButtonClicked)

        var TextureBackground: texture_block = texture_block {
            DefaultImage := UI.pcplayer_ui
            DefaultDesiredSize := vector2 {
                X := 187.29332
                Y := 179.594254
            }
        }   
        var TextureBackground2: texture_block = texture_block {
            DefaultImage := UI.psplayer_ui
            DefaultDesiredSize := vector2 {
                X := 187.29332
                Y := 179.594254
            }
        }  
        var TextureBackground3: texture_block = texture_block {
            DefaultImage := UI.xboxplayer_ui
            DefaultDesiredSize := vector2 {
                X := 187.29332
                Y := 179.594254
            }
        }  
        var TextureBackground4: texture_block = texture_block {
            DefaultImage := UI.switchplayer_ui
            DefaultDesiredSize := vector2 {
                X := 187.29332
                Y := 179.594254
            }
        }  
        # Init slots
        var Slots: []canvas_slot = array {
            # Button 1
            canvas_slot {
                Anchors := anchors{Minimum := vector2{X := 0.5, Y := 0.5}, Maximum := vector2{X := 0.5, Y := 0.5}}
                Offsets := margin{Top := 59.459473, Left := -812.960938, Right := 273.814758, Bottom := 67.307922}
                Alignment := vector2{X := 0.0, Y := 0.0}
                SizeToContent := false
                Widget := ItemButton1
                ZOrder := 900
            },
            # Button 2
            canvas_slot {
                Anchors := anchors{Minimum := vector2{X := 0.5, Y := 0.5}, Maximum := vector2{X := 0.5, Y := 0.5}}
                Offsets := margin{Top := 59.459473, Left := -352.960938, Right := 273.814758, Bottom := 67.307922}
                Alignment := vector2{X := 0.0, Y := 0.0}
                SizeToContent := false
                Widget := ItemButton2
                ZOrder := 900
            },
            # Button 3
            canvas_slot {
                Anchors := anchors{Minimum := vector2{X := 0.5, Y := 0.5}, Maximum := vector2{X := 0.5, Y := 0.5}}
                Offsets := margin{Top := 59.459473, Left := 115.039062, Right := 273.814758, Bottom := 67.307922}
                Alignment := vector2{X := 0.0, Y := 0.0}
                SizeToContent := false
                Widget := ItemButton3
                ZOrder := 900
            },
            # Button 4
            canvas_slot {
                Anchors := anchors{Minimum := vector2{X := 0.5, Y := 0.5}, Maximum := vector2{X := 0.5, Y := 0.5}}
                Offsets := margin{Top := 59.459473, Left := 571.039062, Right := 273.814758, Bottom := 67.307922}
                Alignment := vector2{X := 0.0, Y := 0.0}
                SizeToContent := false
                Widget := ItemButton4
                ZOrder := 900
            },
            # Exit Button
            canvas_slot {
                Anchors := anchors{Minimum := vector2{X := 0.5, Y := 0.5}, Maximum := vector2{X := 0.5, Y := 0.5}}
                Offsets := margin{Top := 199.459473, Left := -116.960938, Right := 225.439468, Bottom := 56.450912}
                Alignment := vector2{X := 0.0, Y := 0.0}
                SizeToContent := false
                Widget := ButtonClose
                ZOrder := 900
            },
            canvas_slot {
                Anchors := anchors{Minimum := vector2{X := 0.5, Y := 0.5}, Maximum := vector2{X := 0.5, Y := 0.5}}
                Offsets := margin{Top := -412.540527, Left := -955.082397, Right := 555.188599, Bottom := 585.045105}
                Alignment := vector2{X := 0.0, Y := 0.0}
                SizeToContent := false
                Widget := TextureBackground
                ZOrder := 800
            },
            canvas_slot {
                Anchors := anchors{Minimum := vector2{X := 0.5, Y := 0.5}, Maximum := vector2{X := 0.5, Y := 0.5}}
                Offsets := margin{Top := -412.540527, Left := -504.019379, Right := 570.804199, Bottom := 585.045105}
                Alignment := vector2{X := 0.0, Y := 0.0}
                SizeToContent := false
                Widget := TextureBackground2
                ZOrder := 800
                
            },
            canvas_slot {
                Anchors := anchors{Minimum := vector2{X := 0.5, Y := 0.5}, Maximum := vector2{X := 0.5, Y := 0.5}}
                Offsets := margin{Top := -412.540527, Left := -37.07793, Right := 570.804199, Bottom := 585.045105}
                Alignment := vector2{X := 0.0, Y := 0.0}
                SizeToContent := false
                Widget := TextureBackground3
                ZOrder := 800
            },
            canvas_slot {
                Anchors := anchors{Minimum := vector2{X := 0.5, Y := 0.5}, Maximum := vector2{X := 0.5, Y := 0.5}}
                Offsets := margin{Top := -412.540527, Left := 430.247864, Right := 558.792175, Bottom := 585.045105}
                Alignment := vector2{X := 0.0, Y := 0.0}
                SizeToContent := false
                Widget := TextureBackground4
                ZOrder := 800
            }
        }

        return Slots
    }

    OnButtonClick(WidgetMessage: widget_message, ButtonId: int): void = {
        if {
            PcManager := MaybePcManager?
            Agent := WidgetMessage.Player
        } then {
            RemoveUiFromPlayer(Agent)
            PcManager.OnUiButtonClick(Agent, ButtonId)
        }
    }

    OnExitButtonClicked(WidgetMessage: widget_message): void = {
        Print("OnExitButtonClicked")
        if {
            Agent := MaybeAgent?
            Manager := MaybePcManager?
        } then {
            RemoveUiFromPlayer(Agent)
            Manager.OnExitButtonClick(Agent)
        }
    }

    UiTimeoutCallback(): void = {
        if {
            Agent := MaybeAgent?
            Manager := MaybePcManager?
        } then {
            RemoveUiFromPlayer(Agent)
            Manager.OnExitButtonClick(Agent)
        }
    }
}
#>