# using { /Fortnite.com/Game }
using { /Fortnite.com/FortPlayerUtilities }
using { /Fortnite.com/Devices }
# using { /Fortnite.com/Devices/CreativeAnimation }
using { /Fortnite.com/Characters }
using { /Fortnite.com/UI }
using { /Verse.org/Simulation }
using { /Verse.org/Simulation/Tags }
using { /Verse.org/Random }
using { /Fortnite.com/Teams }
# using { /Verse.org/Assets }
# using { /Verse.org/Concurrency }
using { /UnrealEngine.com/Temporary/UI }
# using { /UnrealEngine.com/Temporary/SpatialMath }

timeout_manager<public> := class() {
    # Constants

    # Runtime variables
    var TimeoutDuration: float = 10.0
    var RoutineTimeoutTick: ?routine = false
    var TimestampLastPlayerAction: float = 0.0
    var MaybeListener: ?(tuple()->void) = false

    StartTimeoutTimer(ListenerFunction(): void): void = {
        set MaybeListener = option{ListenerFunction}
        UpdatePlayerTimeoutAction()
    }

    StartTimeoutTimerWithDuration(ListenerFunction(): void, ParmTimeoutDuration: float): void = {
        set TimeoutDuration = ParmTimeoutDuration
        StartTimeoutTimer(ListenerFunction)
    }

    CancelTimeout(): void = {
        set MaybeListener = false
        # Clean up routines
        if(Routine := RoutineTimeoutTick?) {
            set RoutineTimeoutTick = false
            Routine.Cancel()
        }
    }

    UpdatePlayerTimeoutAction(): void = {
        set TimestampLastPlayerAction = GetSimulationElapsedTime()
        if(Routine := RoutineTimeoutTick?) {
            # Do nothing, routine is already running
        } else {
            MyRoutine := SpawnRoutine(TimeoutTick)
            set RoutineTimeoutTick = option{MyRoutine}
        }
    }

    OnPlayerActionTimeout():void = {
        Print("timeout_manager - Timeout")
        # Clean up routines
        if(Routine := RoutineTimeoutTick?) {
            set RoutineTimeoutTick = false
            Routine.Cancel()
        }
        # Trigger callback function
        if (Listener := MaybeListener?){
            Listener()
        }
    }

    TimeoutTick()<suspends>: void = {
        loop {
            Timestamp := GetSimulationElapsedTime()
            TimePassedSinceEvent := Timestamp - TimestampLastPlayerAction
            TimeRemainingForEvent := TimeoutDuration - TimePassedSinceEvent
            if(TimeRemainingForEvent <= 0.0) {
                break
            } else {
                Sleep(TimeRemainingForEvent)
            }
        }

        OnPlayerActionTimeout()
    }
}