using { /Verse.org/Simulation }
using { /Verse.org/Colors }
using { /Verse.org/Colors/NamedColors }
# using { /Verse.org/Concurrency }
# using { /Verse.org/Verse }
# using { /Fortnite.com/Devices }
using { /Fortnite.com/UI }
# using { /Fortnite.com/Characters }
# using { /UnrealEngine.com/Temporary/Diagnostics }
using { /UnrealEngine.com/Temporary/SpatialMath }
using { /UnrealEngine.com/Temporary/UI}

<#
ui_accept_quest<public> := class(ui_base) {
    # Constants
    UiSlot<override>: player_ui_slot = player_ui_slot {
        ZOrder := 500
        InputMode := ui_input_mode.None
    }

    <# TODO: Add images
    TextureBackground: texture_block = texture_block {
        DefaultImage := UI.quest_template_ui
        DefaultDesiredSize := vector2 {
            X := 176.740372
            Y := 159.328415
        }
    }
    #>

    # Runtime
    var MaybeQuest<public>: ?quest_config_device = false

    AddCustomUiToPlayer<public>(Agent: agent, Manager: _Interface.interface_ui_manager, Quest: quest_config_device): void = {
        Print("ui_accept_quest - AddCustomUiToPlayer")
        AddUiToPlayer(Agent, option{Manager})

        set MaybeQuest = option{Quest}
    }

    CreateSlots<override>(Agent: agent): []canvas_slot = {
        Utility := GetUtility()

        # Init slots
        var Slots: []canvas_slot = array {
            <# TODO: Add images
            canvas_slot {
                Widget := TextureBackground
                SizeToContent:= false
                Offsets:= margin{Left := -516.960938,Top := -484.540527,Right := 1082.108154,Bottom := 1082.108154}
                Anchors:= anchors{Minimum := vector2{X := 0.5,Y := 0.5},Maximum := vector2{X := 0.5,Y := 0.5}}
                ZOrder := 899
            }
            #>
        }

        return Slots
    }
}
#>