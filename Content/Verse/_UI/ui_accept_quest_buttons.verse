using { /Verse.org/Simulation }
using { /Verse.org/Colors }
using { /Verse.org/Colors/NamedColors }
# using { /Verse.org/Concurrency }
# using { /Verse.org/Verse }
# using { /Fortnite.com/Devices }
using { /Fortnite.com/UI }
# using { /Fortnite.com/Characters }
# using { /UnrealEngine.com/Temporary/Diagnostics }
using { /UnrealEngine.com/Temporary/SpatialMath }
using { /UnrealEngine.com/Temporary/UI}

<#
ui_accept_quest_buttons<public> := class(ui_base) {
    # Constants
    UiSlot<override>: player_ui_slot = player_ui_slot {
        ZOrder := 1501
        InputMode := ui_input_mode.All
    }

    # Runtime
    var MaybeQuest: ?quest_config_device = false

    AddCustomUiToPlayer<public>(Agent: agent, Manager: _Interface.interface_ui_manager, Quest: quest_config_device): void = {
        Print("ui_accept_quest - AddCustomUiToPlayer")
        AddUiToPlayer(Agent, option{Manager})

        set MaybeQuest = option{Quest}
    }

    CreateSlots<override>(Agent: agent): []canvas_slot = {
        Utility := GetUtility()

        # Init buttons
        AcceptButton := button_quiet{DefaultText := Utility.FormatMessage("ACCEPT")}
        AcceptButtonEventHandler := event_handler_widget_message(int) {
            A := 0
            CallbackFunc := OnItemClicked
        }
        AcceptButton.OnClick().Subscribe(AcceptButtonEventHandler.EventSubscription)

        DeclineButton := button_quiet{DefaultText := Utility.FormatMessage("DECLINE")}
        DeclineButtonEventHandler := event_handler_widget_message(int) {
            A := 1
            CallbackFunc := OnItemClicked
        }
        DeclineButton.OnClick().Subscribe(DeclineButtonEventHandler.EventSubscription)

        # Init slots
        var Slots: []canvas_slot = array {
            # Accept button
            canvas_slot {
                Anchors := anchors{Minimum := vector2{X := 0.5, Y := 1.0}, Maximum := vector2{X := 0.5, Y := 1.0}}
                Offsets := margin{Top := -162.08606, Left := -184.734741, Right := 156.56308, Bottom := 45.585739}
                Alignment := vector2{X := 0.0, Y := 0.0}
                SizeToContent := false
                Widget := AcceptButton
                ZOrder := 900
            },
            # Decline button
            canvas_slot {
                Anchors := anchors{Minimum := vector2{X := 0.5, Y := 1.0}, Maximum := vector2{X := 0.5, Y := 1.0}}
                Offsets := margin{Top := -161.490555, Left := 95.039062, Right := 157.38208, Bottom := 45.071239}
                Alignment := vector2{X := 0.0, Y := 0.0}
                SizeToContent := false
                Widget := DeclineButton
                ZOrder := 900
            }
        }

        return Slots
    }
}
#>