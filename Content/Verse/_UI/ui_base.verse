using { /Verse.org/Simulation }
using { /Verse.org/Colors }
using { /Verse.org/Colors/NamedColors }
# using { /Verse.org/Concurrency }
# using { /Verse.org/Verse }
using { /Verse.org/Assets }
using { /Fortnite.com/Devices }
using { /Fortnite.com/UI }
# using { /Fortnite.com/Characters }
# using { /UnrealEngine.com/Temporary/Diagnostics }
using { /UnrealEngine.com/Temporary/SpatialMath }
using { /UnrealEngine.com/Temporary/UI}

# Base class for all UI classes
ui_base<public> := class() {
    # Constants
    UiSlot: player_ui_slot = player_ui_slot {
        ZOrder := 1
        InputMode := ui_input_mode.None
    }

    # Runtime
    var AssignedUiIndex: int = -1
    var MaybeAgent: ?agent = false
    var MaybeCanvas<public>: ?canvas = false
    var MaybeManager: ?_Interface.interface_ui_manager = false
    var RoutineTick: ?routine = false
    var TimeoutManager: timeout_manager = timeout_manager {}

    InitWidget<public>(Agent: agent): void = {
        Print("ui_base - InitWidget")
        set MaybeAgent = option{Agent}
        
        Canvas := canvas {
            Slots := CreateSlots(Agent)
        }

        set MaybeCanvas = option{Canvas}
    }

    AddUiToPlayer<public>(Agent: agent, ParmMaybeManager: ?_Interface.interface_ui_manager): void = {
        GameManager := GetGameManager()
        set MaybeManager = ParmMaybeManager
        
        if {
            PlayerData := Agent.GetPlayerData[]
            Canvas := MaybeCanvas?
        } then {
            set AssignedUiIndex = PlayerData.AddUiToPlayer(Self, Canvas, UiSlot)
        }
    }

    RemoveUiFromPlayer<public>(Agent: agent): void = {
        GameManager := GetGameManager()
        if {
            PlayerData := Agent.GetPlayerData[]
        } then {
            # This must be handled by PlayerData, so the widget is removed from the ActiveUiList
            PlayerData.RemoveUiFromPlayer(Self, AssignedUiIndex)
        } else {
            Print("ERR: ui_base: RemoveUiFromPlayer")
        }
    }

    # CleanUpUiOnRemove is called by player_data on remove
    CleanUpUiOnRemove<public>(): void = {
        # Clean up routines
        if(Routine := RoutineTick?) {
            set RoutineTick = false
            Routine.Cancel()
        }

        TimeoutManager.CancelTimeout()
    }

    CreateSlots<public>(Agent: agent): []canvas_slot = {
        # Build canvas
        var Slots: []canvas_slot = array {

        }

        return Slots
    }

    OnItemClicked(WidgetMessage: widget_message, ButtonIndex: int): void = {
        Print("OnItemClicked - ItemIndex: {ButtonIndex}")
        Agent := WidgetMessage.Player
        RemoveUiFromPlayer(Agent)
        if {
            Manager := MaybeManager?
        } then {
            Manager.OnUiButtonClick(Agent, ButtonIndex)
        }
    }

    OnExitButtonClicked(WidgetMessage: widget_message): void = {
        Print("OnExitButtonClicked")
        Agent := WidgetMessage.Player
        RemoveUiFromPlayer(Agent)
        if {
            Manager := MaybeManager?
        } then {
            Manager.OnExitButtonClick(Agent)
        }
    }

    UiTimeoutCallback(): void = {
        if {
            Agent := MaybeAgent?
        } then {
            RemoveUiFromPlayer(Agent)
            if {
                Manager := MaybeManager?
            } then {
                Manager.OnExitButtonClick(Agent)
            }
        }
    }

    GetItemState(Agent: agent, ItemIndex: int): int = {
        return 0
    }

    GetButtonText(Agent: agent, ItemIndex: int): string = {
        var ButtonText: string = ""
        return ButtonText
    }
}