using { /Verse.org/Simulation }
using { /Verse.org/Colors }
using { /Verse.org/Colors/NamedColors }
# using { /Verse.org/Concurrency }
# using { /Verse.org/Verse }
# using { /Fortnite.com/Devices }
using { /Fortnite.com/UI }
# using { /Fortnite.com/Characters }
# using { /UnrealEngine.com/Temporary/Diagnostics }
using { /UnrealEngine.com/Temporary/SpatialMath }
using { /UnrealEngine.com/Temporary/UI}

ui_changelog<public> := class(ui_base) {
    # Constants
    UiSlot<override>: player_ui_slot = player_ui_slot {
        ZOrder := 523
        InputMode := ui_input_mode.All
    }

    <# TODO: Add images
    TextureBackground: texture_block = texture_block {
        DefaultImage := UI.whatsnew_bg
        DefaultDesiredSize := vector2 {
            X := 176.740372
            Y := 159.328415
        }
    }
    #>

    # Runtime
    CreateSlots<override>(Agent: agent): []canvas_slot = {
        Utility := GetUtility()

        # Init buttons
        AcceptButton := button_quiet{DefaultText := Utility.FormatMessage("V2.0.0")}
        AcceptButtonEventHandler := event_handler_widget_message(int) {
            A := 0
            CallbackFunc := OnItemClicked
        }
        AcceptButton.OnClick().Subscribe(AcceptButtonEventHandler.EventSubscription)

        DeclineButton := button_quiet{DefaultText := Utility.FormatMessage("EXIT")}
        DeclineButtonEventHandler := event_handler_widget_message(int) {
            A := 1
            CallbackFunc := OnItemClicked
        }
        DeclineButton.OnClick().Subscribe(DeclineButtonEventHandler.EventSubscription)

        # Init slots
        var Slots: []canvas_slot = array {
            <# TODO: Add images
            canvas_slot{
                Widget := TextureBackground,
                SizeToContent:= false,
                Offsets:= margin{Left := -572.960938,Top := -576.540527,Right := 1122.522583,Bottom := 1122.522583},
                Anchors:= anchors{Minimum := vector2{X := 0.5,Y := 0.5},Maximum := vector2{X := 0.5,Y := 0.5}}
            },
            #>
            # Accept button
            canvas_slot {
                Anchors := anchors{Minimum := vector2{X := 0.5, Y := 0.5}, Maximum := vector2{X := 0.5, Y := 0.5}}
                Offsets := margin{Top := -334.605103, Left := -533.336304, Right := 205.699692, Bottom := 59.866364}
                Alignment := vector2{X := 0.0, Y := 0.0}
                SizeToContent := false
                Widget := AcceptButton
                ZOrder := 900
            },
            # Close button
            canvas_slot {
                Anchors := anchors{Minimum := vector2{X := 0.5, Y := 0.5}, Maximum := vector2{X := 0.5, Y := 0.5}}
                Offsets := margin{Top := 359.459473, Left := -532.960938, Right := 205.32431, Bottom := 61.638969}
                Alignment := vector2{X := 0.0, Y := 0.0}
                SizeToContent := false
                Widget := DeclineButton
                ZOrder := 900
            }
        }

        return Slots
    }
}