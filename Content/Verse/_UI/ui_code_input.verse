using { /Verse.org/Simulation }
using { /Verse.org/Colors }
using { /Verse.org/Colors/NamedColors }
# using { /Verse.org/Concurrency }
# using { /Verse.org/Verse }
using { /Verse.org/Assets }
using { /Fortnite.com/Devices }
using { /Fortnite.com/UI }
# using { /Fortnite.com/Characters }
# using { /UnrealEngine.com/Temporary/Diagnostics }
using { /UnrealEngine.com/Temporary/SpatialMath }
using { /UnrealEngine.com/Temporary/UI}

ui_code_input<public> := class(ui_base) {
    # Constants
    UiSlot<override>: player_ui_slot = player_ui_slot {
        ZOrder := 503
        InputMode := ui_input_mode.All
    }

    var TextBlockCurCode: text_block = text_block{
        DefaultTextColor:= color{R:=0.0, G:=0.0, B:=0.0}
    }

    
    var TextureBlockBackground: texture_block = texture_block {
        DefaultImage := UI.xp_code_UI
        DefaultDesiredSize := vector2{X := 10.315323, Y := 10.027031}
    }
    

    ButtonCurText: button_quiet = button_quiet{}  #Where the input appears
    
    # Runtime
    var MaybeCodeManager: ?code_input_device = false

    var CurCode: string = ""
    # var UiIsActive: logic = false

    InitWidgetCodeInput<public>(Agent: agent, CodeManager: code_input_device): void = {
        Print("player_code_input_ui - InitWidget")
        set MaybeCodeManager = option{CodeManager}
        
        # Call parent class InitWidget
        InitWidget(Agent)
    }

    AddUiToPlayer<override>(Agent: agent, ParmMaybeManager: ?_Interface.interface_ui_manager): void = {
        (super:)AddUiToPlayer(Agent, ParmMaybeManager)

        set CurCode = ""
        UpdateCurCodeText()
    }

    CreateSlots<override>(Agent: agent): []canvas_slot = {
        Print("player_ui_code_input - CreateSlots")
        Utility := GetUtility()

        # Buttons need to be defined inside the function
        ButtonBackspace: button_loud = button_loud{DefaultText:=Utility.FormatMessage("DELETE")}
        ButtonClose: button_loud = button_loud{DefaultText:=FormatBillboardText("EXIT")}
        ButtonConfirm: button_loud = button_loud{DefaultText:=Utility.FormatMessage("CONFIRM CODE")}

        # Build canvas
        var Slots: []canvas_slot = array {
            canvas_slot {
                Anchors := anchors{Minimum := vector2{X := 0.5, Y := 0.5}, Maximum := vector2{X := 0.5, Y := 0.5}}
                Offsets := margin{Top := -76.540527, Left := -176.960938, Right := 74.224228, Bottom := 67.717728}
                Alignment := vector2{X := 0.0, Y := 0.0}
                SizeToContent := false
                Widget := button_loud{DefaultText := Utility.FormatMessage("1")}
                ZOrder := 901
            },
            canvas_slot {
                Anchors := anchors{Minimum := vector2{X := 0.5, Y := 0.5}, Maximum := vector2{X := 0.5, Y := 0.5}}
                Offsets := margin{Top := -76.540527, Left := -56.960938, Right := 74.224228, Bottom := 67.717728}
                Alignment := vector2{X := 0.0, Y := 0.0}
                SizeToContent := false
                Widget := button_loud{DefaultText := Utility.FormatMessage("2")}
                ZOrder := 901
            },
            canvas_slot {
                Anchors := anchors{Minimum := vector2{X := 0.5, Y := 0.5}, Maximum := vector2{X := 0.5, Y := 0.5}}
                Offsets := margin{Top := -76.540527, Left := 59.039062, Right := 74.224228, Bottom := 67.717728}
                Alignment := vector2{X := 0.0, Y := 0.0}
                SizeToContent := false
                Widget := button_loud{DefaultText := Utility.FormatMessage("3")}
                ZOrder := 901
            },
            
            canvas_slot {
                Anchors := anchors{Minimum := vector2{X := 0.5, Y := 0.5}, Maximum := vector2{X := 0.5, Y := 0.5}}
                Offsets := margin{Top := 35.459473, Left := -176.960938, Right := 74.224228, Bottom := 67.717728}
                Alignment := vector2{X := 0.0, Y := 0.0}
                SizeToContent := false
                Widget := button_loud{DefaultText := Utility.FormatMessage("4")}
                ZOrder := 901
            },
            
            canvas_slot {
                Anchors := anchors{Minimum := vector2{X := 0.5, Y := 0.5}, Maximum := vector2{X := 0.5, Y := 0.5}}
                Offsets := margin{Top := 35.459473, Left := -56.960938, Right := 74.224228, Bottom := 67.717728}
                Alignment := vector2{X := 0.0, Y := 0.0}
                SizeToContent := false
                Widget := button_loud{DefaultText := Utility.FormatMessage("5")}
                ZOrder := 901
            },
           
            canvas_slot {
                Anchors := anchors{Minimum := vector2{X := 0.5, Y := 0.5}, Maximum := vector2{X := 0.5, Y := 0.5}}
                Offsets := margin{Top := 35.459473, Left := 59.039062, Right := 74.224228, Bottom := 67.717728}
                Alignment := vector2{X := 0.0, Y := 0.0}
                SizeToContent := false
                Widget := button_loud{DefaultText := Utility.FormatMessage("6")}
                ZOrder := 901
            },
            canvas_slot {
                Anchors := anchors{Minimum := vector2{X := 0.5, Y := 0.5}, Maximum := vector2{X := 0.5, Y := 0.5}}
                Offsets := margin{Top := 147.459473, Left := -176.960938, Right := 74.224228, Bottom := 67.717728}
                Alignment := vector2{X := 0.0, Y := 0.0}
                SizeToContent := false
                Widget := button_loud{DefaultText := Utility.FormatMessage("7")}
                ZOrder := 901
            },
            canvas_slot {
                Anchors := anchors{Minimum := vector2{X := 0.5, Y := 0.5}, Maximum := vector2{X := 0.5, Y := 0.5}}
                Offsets := margin{Top := 147.459473, Left := -56.960938, Right := 74.224228, Bottom := 67.717728}
                Alignment := vector2{X := 0.0, Y := 0.0}
                SizeToContent := false
                Widget := button_loud{DefaultText := Utility.FormatMessage("8")}
                ZOrder := 901
            },
            canvas_slot {
                Anchors := anchors{Minimum := vector2{X := 0.5, Y := 0.5}, Maximum := vector2{X := 0.5, Y := 0.5}}
                Offsets := margin{Top := 147.459473, Left := 59.039062, Right := 74.224228, Bottom := 67.717728}
                Alignment := vector2{X := 0.0, Y := 0.0}
                SizeToContent := false
                Widget := button_loud{DefaultText := Utility.FormatMessage("9")}
                ZOrder := 901
            }
        }

        # Listen for click events
        for(
            Slot : Slots, 
            Button := text_button_base[Slot.Widget]
        ) {
            Button.OnClick().Subscribe(OnClick)
        }

        FunctionalSlots := array {
           
            # Background image
            canvas_slot {
                Anchors := anchors{Minimum := vector2{X := 0.5, Y := 0.5}, Maximum := vector2{X := 0.5, Y := 0.5}}
                Offsets := margin{Top := -472.540527, Left := -492.960938, Right := 967.839172, Bottom := 908.582947}
                Alignment := vector2{X := 0.0, Y := 0.0}
                SizeToContent := false
                Widget := TextureBlockBackground
                ZOrder := 900
            }
            
            canvas_slot {
                Anchors := anchors{Minimum := vector2{X := 0.5, Y := 0.5}, Maximum := vector2{X := 0.5, Y := 0.5}}
                Offsets := margin{Top := -192.540527, Left := -56.960938, Right := 71.139061, Bottom := 64.313316}
                Alignment := vector2{X := 0.0, Y := 0.0}
                SizeToContent := true
                Widget := ButtonCurText
                ZOrder := 901
            }
            canvas_slot {
                Anchors := anchors{Minimum := vector2{X := 0.5, Y := 0.5}, Maximum := vector2{X := 0.5, Y := 0.5}}
                Offsets := margin{Top := -76.540527, Left := 171.039062, Right := 187.405991, Bottom := 64.313316}
                Alignment := vector2{X := 0.0, Y := 0.0}
                SizeToContent := false
                Widget := ButtonBackspace
                ZOrder := 901
            }
            canvas_slot {
                Anchors := anchors{Minimum := vector2{X := 0.5, Y := 0.5}, Maximum := vector2{X := 0.5, Y := 0.5}}
                Offsets := margin{Top := 319.459473, Left := 111.039062, Right := 179.929657, Bottom := 64.313316}
                Alignment := vector2{X := 0.0, Y := 0.0}
                SizeToContent := false
                Widget := ButtonClose
                ZOrder := 901
            }
            canvas_slot {
                Anchors := anchors{Minimum := vector2{X := 0.5, Y := 0.5}, Maximum := vector2{X := 0.5, Y := 0.5}}
                Offsets := margin{Top := 319.459473, Left := -300.960938, Right := 374.292877, Bottom := 64.313316}
                Alignment := vector2{X := 0.0, Y := 0.0}
                SizeToContent := false
                Widget := ButtonConfirm
                ZOrder := 901
            }
        }
        set Slots += FunctionalSlots

        # Listen for click events
        ButtonBackspace.OnClick().Subscribe(OnClickBackspace)
        ButtonClose.OnClick().Subscribe(OnClickClose)
        ButtonConfirm.OnClick().Subscribe(OnClickConfirm)

        return Slots
    }

    OnClick(WidgetMessage: widget_message): void = {
        Print("ON CLICK")
        if(Button := text_button_base[WidgetMessage.Source]) {
            AddSign(Localize(Button.DefaultText))
        }
    }

    OnClickBackspace(WidgetMessage: widget_message): void = {
        Print("ON CLICK Backspace")
        # Update our data
        if(Temp := CurCode.Slice[0, CurCode.Length - 1]) {
            set CurCode = Temp
        }
        # Update UI
        UpdateCurCodeText()
    }

    OnClickClose(WidgetMessage: widget_message): void = {
        Print("ON CLICK Close")
        if(Agent := MaybeAgent?) {
            RemoveUiFromPlayer(Agent)
        }
    }

    OnClickConfirm(WidgetMessage: widget_message): void = {
        Print("ON CLICK Confirm")
        ConfirmCodeInput()
    }

    ConfirmCodeInput(): void = {
        GameManager := GetGameManager()
        if {
            Agent := MaybeAgent?
            CodeManager := MaybeCodeManager?
        } then {
            if(CurCode = CodeManager.Password) {
                CodeManager.AudioCorrectCode.Play(Agent)
                GameManager.ActivateXpBoostCode(Agent)
                RemoveUiFromPlayer(Agent)
            } else if(CurCode = CodeManager.CodeOneTimeReward) {
                CodeManager.AudioCorrectCode.Play(Agent)
                GameManager.ActivateCodeOneTimeReward(Agent)
                RemoveUiFromPlayer(Agent)
            } else {
                CodeManager.AudioWrongCode.Play(Agent)
            }
        }
    }

    AddSign(SignToAdd: string): void = {
        Print("AddSign: {SignToAdd}")
        if {
            Agent := MaybeAgent?
            CodeManager := MaybeCodeManager?
        } then {
            if(CurCode.Length < CodeManager.Password.Length) {
                # Update our data
                set CurCode = CurCode + SignToAdd
                # Update UI
                UpdateCurCodeText()
            }
            CodeManager.AudioButtonClick.Play(Agent)
        }
    }

    UpdateCurCodeText(): void = {
        # Update UI
        ButtonCurText.SetText(SetDynamicText(ToString(CurCode)))
    }

    SetDynamicText<localizes>(MyString: string) : message = "{MyString}"

    FormatBillboardText<localizes>(MyString: string) : message = "{MyString}"
}