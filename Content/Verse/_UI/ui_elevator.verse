using { /Verse.org/Simulation }
using { /Verse.org/Colors }
using { /Verse.org/Colors/NamedColors }
# using { /Verse.org/Concurrency }
# using { /Verse.org/Verse }
# using { /Fortnite.com/Devices }
using { /Fortnite.com/UI }
# using { /Fortnite.com/Characters }
# using { /UnrealEngine.com/Temporary/Diagnostics }
using { /UnrealEngine.com/Temporary/SpatialMath }
using { /UnrealEngine.com/Temporary/UI}

ui_elevator<public> := class(ui_base) {
    # Constants
    UiSlot<override>: player_ui_slot = player_ui_slot {
        ZOrder := 1023
        InputMode := ui_input_mode.All
    }

    TextureBackground: texture_block = texture_block {
        DefaultImage := UI.placeholder_verse
        DefaultDesiredSize := vector2 {
            X := 176.740372
            Y := 159.328415
        }
    }

    ButtonClose: button_loud = button_loud{}

    ItemButton1: button_loud = button_loud{}
    ItemButton2: button_loud = button_loud{}
    ItemButton3: button_loud = button_loud{}
    ItemButton4: button_loud = button_loud{}
    ItemButton5: button_loud = button_loud{}
    ItemButton6: button_loud = button_loud{}
    ItemButton7: button_loud = button_loud{}
    ItemButton8: button_loud = button_loud{}
    ItemButton9: button_loud = button_loud{}
    ItemButton10: button_loud = button_loud{}
    ItemButton11: button_loud = button_loud{}
    ItemButton12: button_loud = button_loud{}
    ItemButton13: button_loud = button_loud{}
    ItemButton14: button_loud = button_loud{}
    ItemButton15: button_loud = button_loud{}

    # Runtime
    var ElevatorButtonList: []button_loud = array {}

    InitWidget<override>(Agent: agent): void = {
        (super:)InitWidget(Agent)

        Utility := GetUtility()
        
        # Button texts need to be defined inside the function
        ButtonClose.SetText(Utility.FormatMessage("EXIT"))
        
        # Elevator button slots
        ItemButton1.SetText(Utility.FormatMessage("-1"))
        ItemButton2.SetText(Utility.FormatMessage("0"))
        ItemButton3.SetText(Utility.FormatMessage("1"))
        ItemButton4.SetText(Utility.FormatMessage("2"))
        ItemButton5.SetText(Utility.FormatMessage("3"))
        ItemButton6.SetText(Utility.FormatMessage("4"))
        ItemButton7.SetText(Utility.FormatMessage("5"))
        ItemButton8.SetText(Utility.FormatMessage("6"))
        ItemButton9.SetText(Utility.FormatMessage("7"))
        ItemButton10.SetText(Utility.FormatMessage("8"))
        ItemButton11.SetText(Utility.FormatMessage("9"))
        ItemButton12.SetText(Utility.FormatMessage("10"))
        ItemButton13.SetText(Utility.FormatMessage("11"))
        ItemButton14.SetText(Utility.FormatMessage("12"))
        ItemButton15.SetText(Utility.FormatMessage("???"))

        set ElevatorButtonList = array {
            ItemButton1,
            ItemButton2,
            ItemButton3,
            ItemButton4,
            ItemButton5,
            ItemButton6,
            ItemButton7,
            ItemButton8,
            ItemButton9,
            ItemButton10,
            ItemButton11,
            ItemButton12,
            ItemButton13,
            ItemButton14,
            ItemButton15
        }

        # Listen for click events
        for(ButtonIndex -> ItemButton: ElevatorButtonList) {
            ButtonEventHandler := event_handler_widget_message(int) {
                A := ButtonIndex
                CallbackFunc := OnItemClicked
            }

            SpawnerEventSub := ItemButton.OnClick().Subscribe(ButtonEventHandler.EventSubscription)
        }

        ButtonClose.OnClick().Subscribe(OnExitButtonClicked)
    }

    # Refresh the UI every time it is added
    AddUiToPlayer<override>(Agent: agent, ParmMaybeManager: ?_Interface.interface_ui_manager): void = {
        (super:)AddUiToPlayer(Agent, ParmMaybeManager)
        RefreshUi(Agent)
    }

    # Runtime
    CreateSlots<override>(Agent: agent): []canvas_slot = {
        Utility := GetUtility()

        # Init slots
        var Slots: []canvas_slot = array {
            # Button 1
            canvas_slot {
                Anchors := anchors{Minimum := vector2{X := 0.5, Y := 0.5}, Maximum := vector2{X := 0.5, Y := 0.5}}
                Offsets := margin{Top := -284.540527, Left := -324.960938, Right := 123.648651, Bottom := 82.042038}
                Alignment := vector2{X := 0.0, Y := 0.0}
                SizeToContent := false
                Widget := ItemButton1
                ZOrder := 900
            },
            # Button 2
            canvas_slot {
                Anchors := anchors{Minimum := vector2{X := 0.5, Y := 0.5}, Maximum := vector2{X := 0.5, Y := 0.5}}
                Offsets := margin{Top := -284.540527, Left := -148.960938, Right := 123.648651, Bottom := 82.042038}
                Alignment := vector2{X := 0.0, Y := 0.0}
                SizeToContent := false
                Widget := ItemButton2
                ZOrder := 900
            },
            # Button 3
            canvas_slot {
                Anchors := anchors{Minimum := vector2{X := 0.5, Y := 0.5}, Maximum := vector2{X := 0.5, Y := 0.5}}
                Offsets := margin{Top := -284.540527, Left := 27.039062, Right := 123.648651, Bottom := 82.042038}
                Alignment := vector2{X := 0.0, Y := 0.0}
                SizeToContent := false
                Widget := ItemButton3
                ZOrder := 900
            },
            # Button 4
            canvas_slot {
                Anchors := anchors{Minimum := vector2{X := 0.5, Y := 0.5}, Maximum := vector2{X := 0.5, Y := 0.5}}
                Offsets := margin{Top := -284.540527, Left := 199.039062, Right := 123.648651, Bottom := 82.042038}
                Alignment := vector2{X := 0.0, Y := 0.0}
                SizeToContent := false
                Widget := ItemButton4
                ZOrder := 900
            },
            canvas_slot {
                Anchors := anchors{Minimum := vector2{X := 0.5, Y := 0.5}, Maximum := vector2{X := 0.5, Y := 0.5}}
                Offsets := margin{Top := -160.540527, Left := -324.960938, Right := 123.648651, Bottom := 82.042038}
                Alignment := vector2{X := 0.0, Y := 0.0}
                SizeToContent := false
                Widget := ItemButton5
                ZOrder := 900
            },
            canvas_slot {
                Anchors := anchors{Minimum := vector2{X := 0.5, Y := 0.5}, Maximum := vector2{X := 0.5, Y := 0.5}}
                Offsets := margin{Top := -160.540527, Left := -148.960938, Right := 123.648651, Bottom := 82.042038}
                Alignment := vector2{X := 0.0, Y := 0.0}
                SizeToContent := false
                Widget := ItemButton6
                ZOrder := 900
            },
            canvas_slot {
                Anchors := anchors{Minimum := vector2{X := 0.5, Y := 0.5}, Maximum := vector2{X := 0.5, Y := 0.5}}
                Offsets := margin{Top := -160.540527, Left := 27.039062, Right := 123.648651, Bottom := 82.042038}
                Alignment := vector2{X := 0.0, Y := 0.0}
                SizeToContent := false
                Widget := ItemButton7
                ZOrder := 900
            },
            canvas_slot {
                Anchors := anchors{Minimum := vector2{X := 0.5, Y := 0.5}, Maximum := vector2{X := 0.5, Y := 0.5}}
                Offsets := margin{Top := -160.540527, Left := 199.039062, Right := 123.648651, Bottom := 82.042038}
                Alignment := vector2{X := 0.0, Y := 0.0}
                SizeToContent := false
                Widget := ItemButton8
                ZOrder := 900
            },
            canvas_slot {
                Anchors := anchors{Minimum := vector2{X := 0.5, Y := 0.5}, Maximum := vector2{X := 0.5, Y := 0.5}}
                Offsets := margin{Top := -36.540527, Left := -324.960938, Right := 123.648651, Bottom := 82.042038}
                Alignment := vector2{X := 0.0, Y := 0.0}
                SizeToContent := false
                Widget := ItemButton9
                ZOrder := 900
            },
            canvas_slot {
                Anchors := anchors{Minimum := vector2{X := 0.5, Y := 0.5}, Maximum := vector2{X := 0.5, Y := 0.5}}
                Offsets := margin{Top := -36.540527, Left := -148.960938, Right := 123.648651, Bottom := 82.042038}
                Alignment := vector2{X := 0.0, Y := 0.0}
                SizeToContent := false
                Widget := ItemButton10
                ZOrder := 900
            },
            canvas_slot {
                Anchors := anchors{Minimum := vector2{X := 0.5, Y := 0.5}, Maximum := vector2{X := 0.5, Y := 0.5}}
                Offsets := margin{Top := -36.540527, Left := 27.039062, Right := 123.648651, Bottom := 82.042038}
                Alignment := vector2{X := 0.0, Y := 0.0}
                SizeToContent := false
                Widget := ItemButton11
                ZOrder := 900
            },
            canvas_slot {
                Anchors := anchors{Minimum := vector2{X := 0.5, Y := 0.5}, Maximum := vector2{X := 0.5, Y := 0.5}}
                Offsets := margin{Top := -36.540527, Left := 199.039062, Right := 123.648651, Bottom := 82.042038}
                Alignment := vector2{X := 0.0, Y := 0.0}
                SizeToContent := false
                Widget := ItemButton12
                ZOrder := 900
            },
            canvas_slot {
                Anchors := anchors{Minimum := vector2{X := 0.5, Y := 0.5}, Maximum := vector2{X := 0.5, Y := 0.5}}
                Offsets := margin{Top := 91.459473, Left := -324.960938, Right := 123.648651, Bottom := 82.042038}
                Alignment := vector2{X := 0.0, Y := 0.0}
                SizeToContent := false
                Widget := ItemButton13
                ZOrder := 900
            },
            canvas_slot {
                Anchors := anchors{Minimum := vector2{X := 0.5, Y := 0.5}, Maximum := vector2{X := 0.5, Y := 0.5}}
                Offsets := margin{Top := 91.459473, Left := -148.960938, Right := 123.648651, Bottom := 82.042038}
                Alignment := vector2{X := 0.0, Y := 0.0}
                SizeToContent := false
                Widget := ItemButton14
                ZOrder := 900
            },
            canvas_slot {
                Anchors := anchors{Minimum := vector2{X := 0.5, Y := 0.5}, Maximum := vector2{X := 0.5, Y := 0.5}}
                Offsets := margin{Top := 91.459473, Left := 27.039062, Right := 123.648651, Bottom := 82.042038}
                Alignment := vector2{X := 0.0, Y := 0.0}
                SizeToContent := false
                Widget := ItemButton15
                ZOrder := 900
            },
            # Exit Button
            canvas_slot {
                Anchors := anchors{Minimum := vector2{X := 0.5, Y := 0.5}, Maximum := vector2{X := 0.5, Y := 0.5}}
                Offsets := margin{Top := 249.2995, Left := -323.199158, Right := 645.795776, Bottom := 60.202026}
                Alignment := vector2{X := 0.0, Y := 0.0}
                SizeToContent := false
                Widget := ButtonClose
                ZOrder := 900
            }
        }

        return Slots
    }

    RefreshUi(Agent: agent): void = {
        if {
            PlayerData := Agent.GetPlayerData[]
        } then {
            for(ItemIndex -> ElevatorButton: ElevatorButtonList) {
                if {
                    PlayerData.HighestElevatorFloorIdUnlocked >= ItemIndex
                } then {
                    ElevatorButton.SetEnabled(true)
                } else {
                    ElevatorButton.SetEnabled(false)
                }
            }
        }
    }
}