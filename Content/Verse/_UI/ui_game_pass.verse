using { /Verse.org/Simulation }
using { /Verse.org/Colors }
using { /Verse.org/Colors/NamedColors }
# using { /Verse.org/Concurrency }
# using { /Verse.org/Verse }
# using { /Fortnite.com/Devices }
using { /Verse.org/Assets }
using { /Fortnite.com/UI }
# using { /Fortnite.com/Characters }
# using { /UnrealEngine.com/Temporary/Diagnostics }
using { /UnrealEngine.com/Temporary/SpatialMath }
using { /UnrealEngine.com/Temporary/UI}

ui_game_pass<public> := class(ui_base) {
    # Constants
    UiSlot<override>: player_ui_slot = player_ui_slot {
        ZOrder := 1224
        InputMode := ui_input_mode.All
    }

    var TextBlockCostList: []text_block = array {}
    var TexturePresentList: []texture_block = array {}
    var TextureCheckmarkList: []texture_block = array {}
    var ItemButtonList: []button_loud = array {}

    <# TODO: Add images
    TextureRewardClaimedList: []texture = array {
        UI.xp_icon_pass,
        UI.money_icon_pass,
        UI.rebirth_free_1,
        UI.xp_icon_pass_250,
        UI.rebirth_free_2,
        UI.xp_icon_pass_500,
        UI.rebirth_free_3,
        UI.killerking_icon
    }
    #>

    TextBlockGamePassCoins: text_block = text_block {
        DefaultTextColor:= color{R:=1.0, G:=1.0, B:=1.0}
    }

    TextBlockCost1: text_block = text_block {
        DefaultTextColor:= color{R:=1.0, G:=1.0, B:=1.0}
        DefaultJustification := text_justification.Center
    }

    TextBlockCost2: text_block = text_block {
        DefaultTextColor:= color{R:=1.0, G:=1.0, B:=1.0}
        DefaultJustification := text_justification.Center
    }

    TextBlockCost3: text_block = text_block {
        DefaultTextColor:= color{R:=1.0, G:=1.0, B:=1.0}
        DefaultJustification := text_justification.Center
    }

    TextBlockCost4: text_block = text_block {
        DefaultTextColor:= color{R:=1.0, G:=1.0, B:=1.0}
        DefaultJustification := text_justification.Center
    }

    TextBlockCost5: text_block = text_block {
        DefaultTextColor:= color{R:=1.0, G:=1.0, B:=1.0}
        DefaultJustification := text_justification.Center
    }

    TextBlockCost6: text_block = text_block {
        DefaultTextColor:= color{R:=1.0, G:=1.0, B:=1.0}
        DefaultJustification := text_justification.Center
    }

    TextBlockCost7: text_block = text_block {
        DefaultTextColor:= color{R:=1.0, G:=1.0, B:=1.0}
        DefaultJustification := text_justification.Center
    }

    TextBlockCost8: text_block = text_block {
        DefaultTextColor:= color{R:=1.0, G:=1.0, B:=1.0}
        DefaultJustification := text_justification.Center
    }

    TextBlockCost9: text_block = text_block {
        DefaultTextColor:= color{R:=1.0, G:=1.0, B:=1.0}
        DefaultJustification := text_justification.Center
    }

    TextBlockCost10: text_block = text_block {
        DefaultTextColor:= color{R:=1.0, G:=1.0, B:=1.0}
        DefaultJustification := text_justification.Center
    }

    TextBlockCost11: text_block = text_block {
        DefaultTextColor:= color{R:=1.0, G:=1.0, B:=1.0}
        DefaultJustification := text_justification.Center
    }

    TextBlockCost12: text_block = text_block {
        DefaultTextColor:= color{R:=1.0, G:=1.0, B:=1.0}
        DefaultJustification := text_justification.Center
    }

    <# TODO: Add images
    TextureBackground: texture_block = texture_block {
        DefaultImage := UI.tycoonpass_bg
        DefaultDesiredSize := vector2 {
            X := 506.206177
            Y := 487.087097
        }
    }

    TexturePresent1: texture_block = texture_block {
        DefaultImage := UI.questionicon
        DefaultDesiredSize := vector2 {
            X := 506.206177
            Y := 487.087097
        }
    }

    TexturePresent2: texture_block = texture_block {
        DefaultImage := UI.questionicon
        DefaultDesiredSize := vector2 {
            X := 506.206177
            Y := 487.087097
        }
    }

    TexturePresent3: texture_block = texture_block {
        DefaultImage := UI.questionicon
        DefaultDesiredSize := vector2 {
            X := 506.206177
            Y := 487.087097
        }
    }

    TexturePresent4: texture_block = texture_block {
        DefaultImage := UI.questionicon
        DefaultDesiredSize := vector2 {
            X := 506.206177
            Y := 487.087097
        }
    }

    TexturePresent5: texture_block = texture_block {
        DefaultImage := UI.questionicon
        DefaultDesiredSize := vector2 {
            X := 506.206177
            Y := 487.087097
        }
    }

    TexturePresent6: texture_block = texture_block {
        DefaultImage := UI.questionicon
        DefaultDesiredSize := vector2 {
            X := 506.206177
            Y := 487.087097
        }
    }

    TexturePresent7: texture_block = texture_block {
        DefaultImage := UI.questionicon
        DefaultDesiredSize := vector2 {
            X := 506.206177
            Y := 487.087097
        }
    }

    TexturePresent8: texture_block = texture_block {
        DefaultImage := UI.questionicon
        DefaultDesiredSize := vector2 {
            X := 506.206177
            Y := 487.087097
        }
    }
    
    TextureCheck1: texture_block = texture_block {
        DefaultImage := UI.checkmark
        DefaultDesiredSize := vector2 {
            X := 506.206177
            Y := 487.087097
        }
    }
    TextureCheck2: texture_block = texture_block {
        DefaultImage := UI.checkmark
        DefaultDesiredSize := vector2 {
            X := 506.206177
            Y := 487.087097
        }
    }
    TextureCheck3: texture_block = texture_block {
        DefaultImage := UI.checkmark
        DefaultDesiredSize := vector2 {
            X := 506.206177
            Y := 487.087097
        }
    }
    TextureCheck4: texture_block = texture_block {
        DefaultImage := UI.checkmark
        DefaultDesiredSize := vector2 {
            X := 506.206177
            Y := 487.087097
        }
    }
    TextureCheck5: texture_block = texture_block {
        DefaultImage := UI.checkmark
        DefaultDesiredSize := vector2 {
            X := 506.206177
            Y := 487.087097
        }
    }
    TextureCheck6: texture_block = texture_block {
        DefaultImage := UI.checkmark
        DefaultDesiredSize := vector2 {
            X := 506.206177
            Y := 487.087097
        }
    }
    TextureCheck7: texture_block = texture_block {
        DefaultImage := UI.checkmark
        DefaultDesiredSize := vector2 {
            X := 506.206177
            Y := 487.087097
        }
    }
    TextureCheck8: texture_block = texture_block {
        DefaultImage := UI.checkmark
        DefaultDesiredSize := vector2 {
            X := 506.206177
            Y := 487.087097
        }
    }
    TextureCheck9: texture_block = texture_block {
        DefaultImage := UI.checkmark
        DefaultDesiredSize := vector2 {
            X := 506.206177
            Y := 487.087097
        }
    }
    TextureCheck10: texture_block = texture_block {
        DefaultImage := UI.checkmark
        DefaultDesiredSize := vector2 {
            X := 506.206177
            Y := 487.087097
        }
    }
    TextureCheck11: texture_block = texture_block {
        DefaultImage := UI.checkmark
        DefaultDesiredSize := vector2 {
            X := 506.206177
            Y := 487.087097
        }
    }
    TextureCheck12: texture_block = texture_block {
        DefaultImage := UI.checkmark
        DefaultDesiredSize := vector2 {
            X := 506.206177
            Y := 487.087097
        }
    }
    #>

    # Runtime
    var MaybeGamePassManager: ?game_pass_device = false

    # var UiIsActive: logic = false
    var TimestampLastPlayerAction: float = 0.0

    InitWidgetGamePass<public>(Agent: agent, GamePassManager: game_pass_device): void = {
        Print("InitWidgetGamePass")
        set MaybeGamePassManager = option{GamePassManager}

        set TextBlockCostList = TextBlockCostList + array {
            TextBlockCost1,
            TextBlockCost2,
            TextBlockCost3,
            TextBlockCost4,
            TextBlockCost5,
            TextBlockCost6,
            TextBlockCost7,
            TextBlockCost8,
            TextBlockCost9,
            TextBlockCost10,
            TextBlockCost11,
            TextBlockCost12
        }

        <# TODO: Add images
        set TexturePresentList = TexturePresentList + array {
            TexturePresent1,
            TexturePresent2,
            TexturePresent3,
            TexturePresent4,
            TexturePresent5,
            TexturePresent6,
            TexturePresent7,
            TexturePresent8
        }

        set TextureCheckmarkList = TextureCheckmarkList + array {
            TextureCheck1,
            TextureCheck2,
            TextureCheck3,
            TextureCheck4,
            TextureCheck5,
            TextureCheck6,
            TextureCheck7,
            TextureCheck8,
            TextureCheck9,
            TextureCheck10,
            TextureCheck11,
            TextureCheck12
        }
        #>

        # Call parent class InitWidget
        InitWidget(Agent)
    }

    AddUiToPlayer<override>(Agent: agent, ParmMaybeManager: ?_Interface.interface_ui_manager): void = {
        (super:)AddUiToPlayer(Agent, ParmMaybeManager)

        RefreshUi()
    }

    CreateSlots<override>(Agent: agent): []canvas_slot = {
        Utility := GetUtility()

        # Buttons need to be defined inside the function
        ButtonClose: button_loud = button_loud{DefaultText:=Utility.FormatMessage("EXIT")}

        # Reward item button slots
        ItemButton1 := button_loud{}
        ItemButton2 := button_loud{}
        ItemButton3 := button_loud{}
        ItemButton4 := button_loud{}
        ItemButton5 := button_loud{}
        ItemButton6 := button_loud{}
        ItemButton7 := button_loud{}
        ItemButton8 := button_loud{}
        ItemButton9 := button_loud{}
        ItemButton10 := button_loud{}
        ItemButton11 := button_loud{}
        ItemButton12 := button_loud{}

        set ItemButtonList = array {
            ItemButton1,
            ItemButton2,
            ItemButton3,
            ItemButton4,
            ItemButton5,
            ItemButton6,
            ItemButton7,
            ItemButton8,
            ItemButton9,
            ItemButton10,
            ItemButton11,
            ItemButton12
        }

        for(ButtonIndex -> ItemButton: ItemButtonList) {
            ButtonEventHandler := event_handler_widget_message(int) {
                A := ButtonIndex
                CallbackFunc := OnItemClicked
            }

            SpawnerEventSub := ItemButton.OnClick().Subscribe(ButtonEventHandler.EventSubscription)
        }

        # Init slots
        var Slots: []canvas_slot = array {
            # Reward present images
            <# TODO: Add images
            canvas_slot {
                Anchors := anchors{Minimum := vector2{X := 0.5, Y := 0.5}, Maximum := vector2{X := 0.5, Y := 0.5}}
                Offsets := margin{Top := -78.268555, Left := -736.960938, Right := 169.498062, Bottom := 169.558136}
                Alignment := vector2{X := 0.0, Y := 0.0}
                SizeToContent := false
                Widget := TexturePresent1
                ZOrder := 900
            },
            canvas_slot {
                Anchors := anchors{Minimum := vector2{X := 0.5, Y := 0.5}, Maximum := vector2{X := 0.5, Y := 0.5}}
                Offsets := margin{Top := -78.268555, Left := -548.960938, Right := 169.498062, Bottom := 169.558136}
                Alignment := vector2{X := 0.0, Y := 0.0}
                SizeToContent := false
                Widget := TexturePresent2
                ZOrder := 900
            },
            canvas_slot {
                Anchors := anchors{Minimum := vector2{X := 0.5, Y := 0.5}, Maximum := vector2{X := 0.5, Y := 0.5}}
                Offsets := margin{Top := -78.268555, Left := -360.960938, Right := 169.498062, Bottom := 169.558136}
                Alignment := vector2{X := 0.0, Y := 0.0}
                SizeToContent := false
                Widget := TexturePresent3
                ZOrder := 900
            },
            canvas_slot {
                Anchors := anchors{Minimum := vector2{X := 0.5, Y := 0.5}, Maximum := vector2{X := 0.5, Y := 0.5}}
                Offsets := margin{Top := -78.268555, Left := -168.960938, Right := 169.498062, Bottom := 169.558136}
                Alignment := vector2{X := 0.0, Y := 0.0}
                SizeToContent := false
                Widget := TexturePresent4
                ZOrder := 900
            },
            canvas_slot {
                Anchors := anchors{Minimum := vector2{X := 0.5, Y := 0.5}, Maximum := vector2{X := 0.5, Y := 0.5}}
                Offsets := margin{Top := -78.268555, Left := 11.039062, Right := 169.498062, Bottom := 169.558136}
                Alignment := vector2{X := 0.0, Y := 0.0}
                SizeToContent := false
                Widget := TexturePresent5
                ZOrder := 900
            },
            canvas_slot {
                Anchors := anchors{Minimum := vector2{X := 0.5, Y := 0.5}, Maximum := vector2{X := 0.5, Y := 0.5}}
                Offsets := margin{Top := -78.268555, Left := 199.039062, Right := 169.498062, Bottom := 169.558136}
                Alignment := vector2{X := 0.0, Y := 0.0}
                SizeToContent := false
                Widget := TexturePresent6
                ZOrder := 900
            },
            canvas_slot {
                Anchors := anchors{Minimum := vector2{X := 0.5, Y := 0.5}, Maximum := vector2{X := 0.5, Y := 0.5}}
                Offsets := margin{Top := -78.268555, Left := 387.039062, Right := 169.498062, Bottom := 169.558136}
                Alignment := vector2{X := 0.0, Y := 0.0}
                SizeToContent := false
                Widget := TexturePresent7
                ZOrder := 900
            },
            canvas_slot {
                Anchors := anchors{Minimum := vector2{X := 0.5, Y := 0.5}, Maximum := vector2{X := 0.5, Y := 0.5}}
                Offsets := margin{Top := -108.540527, Left := 571.039062, Right := 256.585144, Bottom := 244.633209}
                Alignment := vector2{X := 0.0, Y := 0.0}
                SizeToContent := false
                Widget := TexturePresent8
                ZOrder := 900
            },
            #>
            # Claim reward buttons
            canvas_slot {
                Anchors := anchors{Minimum := vector2{X := 0.5, Y := 0.5}, Maximum := vector2{X := 0.5, Y := 0.5}}
                Offsets := margin{Top := 151.459473, Left := -736.960938, Right := 169.419159, Bottom := 46.841301}
                Alignment := vector2{X := 0.0, Y := 0.0}
                SizeToContent := false
                Widget := ItemButton1
                ZOrder := 950
            },
            canvas_slot {
                Anchors := anchors{Minimum := vector2{X := 0.5, Y := 0.5}, Maximum := vector2{X := 0.5, Y := 0.5}}
                Offsets := margin{Top := 151.459473, Left := -548.960938, Right := 169.419159, Bottom := 46.841301}
                Alignment := vector2{X := 0.0, Y := 0.0}
                SizeToContent := false
                Widget := ItemButton2
                ZOrder := 950
            },
            canvas_slot {
                Anchors := anchors{Minimum := vector2{X := 0.5, Y := 0.5}, Maximum := vector2{X := 0.5, Y := 0.5}}
                Offsets := margin{Top := 151.459473, Left := -356.960938, Right := 169.419159, Bottom := 46.841301}
                Alignment := vector2{X := 0.0, Y := 0.0}
                SizeToContent := false
                Widget := ItemButton3
                ZOrder := 950
            },
            canvas_slot {
                Anchors := anchors{Minimum := vector2{X := 0.5, Y := 0.5}, Maximum := vector2{X := 0.5, Y := 0.5}}
                Offsets := margin{Top := 151.459473, Left := -172.960938, Right := 169.419159, Bottom := 46.841301}
                Alignment := vector2{X := 0.0, Y := 0.0}
                SizeToContent := false
                Widget := ItemButton4
                ZOrder := 950
            },
            canvas_slot {
                Anchors := anchors{Minimum := vector2{X := 0.5, Y := 0.5}, Maximum := vector2{X := 0.5, Y := 0.5}}
                Offsets := margin{Top := 151.459473, Left := 11.039062, Right := 169.419159, Bottom := 46.841301}
                Alignment := vector2{X := 0.0, Y := 0.0}
                SizeToContent := false
                Widget := ItemButton5
                ZOrder := 950
            },
            canvas_slot {
                Anchors := anchors{Minimum := vector2{X := 0.5, Y := 0.5}, Maximum := vector2{X := 0.5, Y := 0.5}}
                Offsets := margin{Top := 151.459473, Left := 203.039062, Right := 169.419159, Bottom := 46.841301}
                Alignment := vector2{X := 0.0, Y := 0.0}
                SizeToContent := false
                Widget := ItemButton6
                ZOrder := 950
            },
            canvas_slot {
                Anchors := anchors{Minimum := vector2{X := 0.5, Y := 0.5}, Maximum := vector2{X := 0.5, Y := 0.5}}
                Offsets := margin{Top := 151.459473, Left := 387.039062, Right := 169.419159, Bottom := 46.841301}
                Alignment := vector2{X := 0.0, Y := 0.0}
                SizeToContent := false
                Widget := ItemButton7
                ZOrder := 950
            },
            canvas_slot {
                Anchors := anchors{Minimum := vector2{X := 0.5, Y := 0.5}, Maximum := vector2{X := 0.5, Y := 0.5}}
                Offsets := margin{Top := 179.459473, Left := 611.039062, Right := 169.419159, Bottom := 46.841301}
                Alignment := vector2{X := 0.0, Y := 0.0}
                SizeToContent := false
                Widget := ItemButton8
                ZOrder := 950
            },
            # Reward timers
            canvas_slot {
                Anchors := anchors{Minimum := vector2{X := 0.5, Y := 0.5}, Maximum := vector2{X := 0.5, Y := 0.5}}
                Offsets := margin{Top := 91.459473, Left := -732.960938, Right := 165.0, Bottom := 41.0}
                Alignment := vector2{X := 0.0, Y := 0.0}
                SizeToContent := false
                Widget := TextBlockCost1
                ZOrder := 1900
            },
            canvas_slot {
                Anchors := anchors{Minimum := vector2{X := 0.5, Y := 0.5}, Maximum := vector2{X := 0.5, Y := 0.5}}
                Offsets := margin{Top := 91.459473, Left := -544.960938, Right := 165.0, Bottom := 41.0}
                Alignment := vector2{X := 0.0, Y := 0.0}
                SizeToContent := false
                Widget := TextBlockCost2
                ZOrder := 1900
            },
            canvas_slot {
                Anchors := anchors{Minimum := vector2{X := 0.5, Y := 0.5}, Maximum := vector2{X := 0.5, Y := 0.5}}
                Offsets := margin{Top := 91.459473, Left := -348.960938, Right := 165.0, Bottom := 41.0}
                Alignment := vector2{X := 0.0, Y := 0.0}
                SizeToContent := false
                Widget := TextBlockCost3
                ZOrder := 1900
            },
            canvas_slot {
                Anchors := anchors{Minimum := vector2{X := 0.5, Y := 0.5}, Maximum := vector2{X := 0.5, Y := 0.5}}
                Offsets := margin{Top := 91.459473, Left := -160.960938, Right := 165.0, Bottom := 41.0}
                Alignment := vector2{X := 0.0, Y := 0.0}
                SizeToContent := false
                Widget := TextBlockCost4
                ZOrder := 1900
            },
            canvas_slot {
                Anchors := anchors{Minimum := vector2{X := 0.5, Y := 0.5}, Maximum := vector2{X := 0.5, Y := 0.5}}
                Offsets := margin{Top := 91.459473, Left := 15.039062, Right := 165.0, Bottom := 41.0}
                Alignment := vector2{X := 0.0, Y := 0.0}
                SizeToContent := false
                Widget := TextBlockCost5
                ZOrder := 1900
            },
            canvas_slot {
                Anchors := anchors{Minimum := vector2{X := 0.5, Y := 0.5}, Maximum := vector2{X := 0.5, Y := 0.5}}
                Offsets := margin{Top := 91.459473, Left := 203.039062, Right := 165.0, Bottom := 41.0}
                Alignment := vector2{X := 0.0, Y := 0.0}
                SizeToContent := false
                Widget := TextBlockCost6
                ZOrder := 1900
            },
            canvas_slot {
                Anchors := anchors{Minimum := vector2{X := 0.5, Y := 0.5}, Maximum := vector2{X := 0.5, Y := 0.5}}
                Offsets := margin{Top := 91.459473, Left := 391.039062, Right := 165.0, Bottom := 41.0}
                Alignment := vector2{X := 0.0, Y := 0.0}
                SizeToContent := false
                Widget := TextBlockCost7
                ZOrder := 1900
            },
            canvas_slot {
                Anchors := anchors{Minimum := vector2{X := 0.5, Y := 0.5}, Maximum := vector2{X := 0.5, Y := 0.5}}
                Offsets := margin{Top := 127.459473, Left := 619.039062, Right := 165.0, Bottom := 41.0}
                Alignment := vector2{X := 0.0, Y := 0.0}
                SizeToContent := false
                Widget := TextBlockCost8
                ZOrder := 1900
            },
            # Current coins
            canvas_slot {
                Anchors := anchors{Minimum := vector2{X := 0.5, Y := 0.5}, Maximum := vector2{X := 0.5, Y := 0.5}}
                Offsets := margin{Top := 287.459473, Left := -496.960938, Right := 165.0, Bottom := 41.0}
                Alignment := vector2{X := 0.0, Y := 0.0}
                SizeToContent := false
                Widget := TextBlockGamePassCoins
                ZOrder := 1900
            },
            # Close button
            canvas_slot {
                Anchors := anchors{Minimum := vector2{X := 0.5, Y := 0.5}, Maximum := vector2{X := 0.5, Y := 0.5}}
                Offsets := margin{Top := 455.459473, Left := -76.960938, Right := 170.92067, Bottom := 51.145607}
                Alignment := vector2{X := 0.0, Y := 0.0}
                SizeToContent := false
                Widget := ButtonClose
                ZOrder := 901
            }
            # Checkmarks
            <# TODO: Add images
            canvas_slot {
                Anchors := anchors{Minimum := vector2{X := 0.5, Y := 0.5}, Maximum := vector2{X := 0.5, Y := 0.5}}
                Offsets := margin{Top := 35.169468, Left := -604.960938, Right := 68.695969, Bottom := 71.702148}
                Alignment := vector2{X := 0.0, Y := 0.0}
                SizeToContent := false
                Widget := TextureCheck1
                ZOrder := 900
            },
            canvas_slot {
                Anchors := anchors{Minimum := vector2{X := 0.5, Y := 0.5}, Maximum := vector2{X := 0.5, Y := 0.5}}
                Offsets := margin{Top := 35.459473, Left := -416.960938, Right := 68.695969, Bottom := 71.702148}
                Alignment := vector2{X := 0.0, Y := 0.0}
                SizeToContent := false
                Widget := TextureCheck2
                ZOrder := 900
            },
            canvas_slot {
                Anchors := anchors{Minimum := vector2{X := 0.5, Y := 0.5}, Maximum := vector2{X := 0.5, Y := 0.5}}
                Offsets := margin{Top := 35.459473, Left := -228.960938, Right := 68.695969, Bottom := 71.702148}
                Alignment := vector2{X := 0.0, Y := 0.0}
                SizeToContent := false
                Widget := TextureCheck3
                ZOrder := 900
            },
            canvas_slot {
                Anchors := anchors{Minimum := vector2{X := 0.5, Y := 0.5}, Maximum := vector2{X := 0.5, Y := 0.5}}
                Offsets := margin{Top := 35.459473, Left := -36.960938, Right := 68.695969, Bottom := 71.702148}
                Alignment := vector2{X := 0.0, Y := 0.0}
                SizeToContent := false
                Widget := TextureCheck4
                ZOrder := 900
            },
            canvas_slot {
                Anchors := anchors{Minimum := vector2{X := 0.5, Y := 0.5}, Maximum := vector2{X := 0.5, Y := 0.5}}
                Offsets := margin{Top := 35.459473, Left := 143.039062, Right := 68.695969, Bottom := 71.702148}
                Alignment := vector2{X := 0.0, Y := 0.0}
                SizeToContent := false
                Widget := TextureCheck5
                ZOrder := 900
            },
            canvas_slot {
                Anchors := anchors{Minimum := vector2{X := 0.5, Y := 0.5}, Maximum := vector2{X := 0.5, Y := 0.5}}
                Offsets := margin{Top := 35.459473, Left := 331.039062, Right := 68.695969, Bottom := 71.702148}
                Alignment := vector2{X := 0.0, Y := 0.0}
                SizeToContent := false
                Widget := TextureCheck6
                ZOrder := 900
            },
            canvas_slot {
                Anchors := anchors{Minimum := vector2{X := 0.5, Y := 0.5}, Maximum := vector2{X := 0.5, Y := 0.5}}
                Offsets := margin{Top := 35.459473, Left := 519.039062, Right := 68.695969, Bottom := 71.702148}
                Alignment := vector2{X := 0.0, Y := 0.0}
                SizeToContent := false
                Widget := TextureCheck7
                ZOrder := 900
            },
            canvas_slot {
                Anchors := anchors{Minimum := vector2{X := 0.5, Y := 0.5}, Maximum := vector2{X := 0.5, Y := 0.5}}
                Offsets := margin{Top := 31.459473, Left := 759.039062, Right := 100.060669, Bottom := 103.734177}
                Alignment := vector2{X := 0.0, Y := 0.0}
                SizeToContent := false
                Widget := TextureCheck8
                ZOrder := 900
            },
            # Background
            canvas_slot {
                Anchors := anchors{Minimum := vector2{X := 0.5, Y := 0.5}, Maximum := vector2{X := 0.5, Y := 0.5}}
                Offsets := margin{Top := -398.438416, Left := -940.960938, Right := 1886.743896, Bottom := 909.909912}
                Alignment := vector2{X := 0.0, Y := 0.0}
                SizeToContent := false
                Widget := TextureBackground
                ZOrder := 899
            }
            #>
        }

        # Initially hide some slots
        for(TextureCheckmark: TextureCheckmarkList) {
            TextureCheckmark.SetVisibility(widget_visibility.Hidden)
        }

        # Listen for click events
        ButtonClose.OnClick().Subscribe(OnExitButtonClicked)

        return Slots
    }

    RefreshUi<public>(): void = {
        Utility := GetUtility()
        if {
            Agent := MaybeAgent?
            PlayerData := Agent.GetPlayerData[]
            GamePassManager := MaybeGamePassManager?
        } then {
            # Check every reward state
            for(ItemIndex -> PlaytimeRequired: GamePassManager.ItemPriceList) {
                ItemState := GamePassManager.GetItemState(Agent, ItemIndex)
                if {
                    ItemButton := ItemButtonList[ItemIndex]
                    TextBlockCost := TextBlockCostList[ItemIndex]
                    ItemPrice := GamePassManager.ItemPriceList[ItemIndex]
                    TexturePresent := TexturePresentList[ItemIndex]
                    TextureCheckmark := TextureCheckmarkList[ItemIndex]
                    # TODO: Add images TextureRewardClaimed := TextureRewardClaimedList[ItemIndex]
                } then {
                    NewButtonText := GamePassManager.GetButtonText(Agent, ItemIndex)
                    ItemButton.SetText(Utility.FormatMessage(NewButtonText))
                    TextBlockCost.SetText(Utility.FormatMessage(Utility.FormatFloat(ItemPrice * 1.0, 2, 0) + " P"))
                    if(ItemState >= 2) {
                        # TODO: Add imagesTexturePresent.SetImage(TextureRewardClaimed)
                        TextureCheckmark.SetVisibility(widget_visibility.Visible)
                    }
                }
            }
            # Update currency display
            TextBlockGamePassCoins.SetText(Utility.FormatMessage(Utility.FormatFloat(PlayerData.CurGamePassCoins * 1.0, 2, 0)))
        }
    }

    OnItemClicked<override>(WidgetMessage: widget_message, ButtonIndex: int): void = {
        Print("OnItemClicked - ItemIndex: {ButtonIndex}")
        if {
            GamePassManager := MaybeGamePassManager?
        } then {
            Agent := WidgetMessage.Player
            GamePassManager.HandleItemClicked(Agent, Self, ButtonIndex)
        }
        TimeoutManager.UpdatePlayerTimeoutAction()
    }
}