using { /Verse.org/Simulation }
using { /Verse.org/Colors }
using { /Verse.org/Colors/NamedColors }
# using { /Verse.org/Concurrency }
# using { /Verse.org/Verse }
using { /Verse.org/Assets }
using { /Fortnite.com/Devices }
using { /Fortnite.com/UI }
using { /Fortnite.com/Characters }
# using { /UnrealEngine.com/Temporary/Diagnostics }
using { /UnrealEngine.com/Temporary/SpatialMath }
using { /UnrealEngine.com/Temporary/UI}

ui_main<public> := class(ui_base) {
    # Constants
    UiSlot<override>: player_ui_slot = player_ui_slot {
        ZOrder := 504
        InputMode := ui_input_mode.None
    }

    LevelBarWidth: float = 42.460342
    LevelBarHeight: float = 21.783981

    MaxLevelReqExp: int = 99999999999999999
    XpBarWidth: float = 558.113892
    XpBarHeight: float = 27.612614
    MaxHealthBarWidth: float = 348.042542
    HealthBarHeight: float = 24.822857
    MaxShieldBarWidth: float = 330.603394
    ShieldBarHeight: float = 24.822857

    # New order: XP, Money, Fans(Scams), Diamonds(Kills), GamePassCoinGained
    FloatingTextIconList: []texture = array {
        UI.xpicon
        UI.money_ui_icon
        UI.scam_ui
        UI.killcoin_ui
        UI.invisible_img
    }

    FloatingTextColorList: []color = array {
        color{R:= 1.0, G:= 1.0, B:= 1.0},
        color{R:= 1.0, G:= 1.0, B:= 1.0},
        color{R:= 1.0, G:= 1.0, B:= 1.0},
        color{R:= 1.0, G:= 1.0, B:= 1.0},
        color{R:= 1.0, G:= 1.0, B:= 1.0}
    }
    
    # Runtime
    XpConfig: xp_config = xp_config {}
    var MaybeProgressOverlay: ?overlay = false
    var MaybeBossPortalTimer: ?ui_timer = false
    var CanvasSlots: []canvas_slot = array {}

    var CurLevel: int = 0
    var TotalXp: int = 0

    var FloatingTextList: []text_block = array {}
    var FloatingIconList: []texture_block = array {}

    var FloatingTextAmountQueue: []int = array {}
    var FloatingTextTypeQueue: []int = array {}

    var LevelBarColorBlockList: []color_block = array {}

    LevelBarBackgroundColorBlock: color_block = color_block {
        DefaultColor := color{R:=0.036458, G:=0.036458, B:=0.036458}
        DefaultDesiredSize := vector2 {
            X := 35.89743
            Y := 13.088473
        }
    }

    Level_Overlay: texture_block = texture_block {
        DefaultImage := UI.levelbar_new_bg_2
        DefaultDesiredSize := vector2{X := 509.438873, Y := 421.078094}
    }

    LevelBarColorBlock1: color_block = color_block {
        DefaultColor := color{R:=1.0, G:=0.760268, B:=0.0}
        DefaultDesiredSize := vector2 {
            X := 35.89743
            Y := 13.088473
        }
    }

    LevelBarColorBlock2: color_block = color_block {
        DefaultColor := color{R:=1.0, G:=0.760268, B:=0.0}
        DefaultDesiredSize := vector2 {
            X := 35.89743
            Y := 13.088473
        }
    }

    LevelBarColorBlock3: color_block = color_block {
        DefaultColor := color{R:=1.0, G:=0.760268, B:=0.0}
        DefaultDesiredSize := vector2 {
            X := 35.89743
            Y := 13.088473
        }
    }

    LevelBarColorBlock4: color_block = color_block {
        DefaultColor := color{R:=1.0, G:=0.760268, B:=0.0}
        DefaultDesiredSize := vector2 {
            X := 35.89743
            Y := 13.088473
        }
    }

    LevelBarColorBlock5: color_block = color_block {
        DefaultColor := color{R:=1.0, G:=0.760268, B:=0.0}
        DefaultDesiredSize := vector2 {
            X := 35.89743
            Y := 13.088473
        }
    }

    LevelBarColorBlock6: color_block = color_block {
        DefaultColor := color{R:=1.0, G:=0.760268, B:=0.0}
        DefaultDesiredSize := vector2 {
            X := 35.89743
            Y := 13.088473
        }
    }

    LevelBarColorBlock7: color_block = color_block {
        DefaultColor := color{R:=1.0, G:=0.760268, B:=0.0}
        DefaultDesiredSize := vector2 {
            X := 35.89743
            Y := 13.088473
        }
    }

    LevelBarColorBlock8: color_block = color_block {
        DefaultColor := color{R:=1.0, G:=0.760268, B:=0.0}
        DefaultDesiredSize := vector2 {
            X := 35.89743
            Y := 13.088473
        }
    }

    LevelBarColorBlock9: color_block = color_block {
        DefaultColor := color{R:=1.0, G:=0.760268, B:=0.0}
        DefaultDesiredSize := vector2 {
            X := 35.89743
            Y := 13.088473
        }
    }

    LevelBarColorBlock10: color_block = color_block {
        DefaultColor := color{R:=1.0, G:=0.760268, B:=0.0}
        DefaultDesiredSize := vector2 {
            X := 35.89743
            Y := 13.088473
        }
    }

    var XpBoost2xTexture: texture_block = texture_block {
        DefaultImage := UI.xpboost
        DefaultDesiredSize := vector2 {
            X := 187.29332
            Y := 179.594254
        }
    }

    <#
    var XpBoost3xTexture: texture_block = texture_block {
        DefaultImage := UI.xpboostx3
        DefaultDesiredSize := vector2 {
            X := 187.29332
            Y := 179.594254
        }
    }

    var XpBoost4xTexture: texture_block = texture_block {
        DefaultImage := UI.xpboost4
        DefaultDesiredSize := vector2 {
            X := 187.29332
            Y := 179.594254
        }
    }

    var XpBoost5xTexture: texture_block = texture_block {
        DefaultImage := UI.xpboost5
        DefaultDesiredSize := vector2 {
            X := 187.29332
            Y := 179.594254
        }
    }

    var XpBoost6xTexture: texture_block = texture_block {
        DefaultImage := UI.xpboost6
        DefaultDesiredSize := vector2 {
            X := 187.29332
            Y := 179.594254
        }
    }
    #>

    var TextureMaskLevelStar1: texture_block = texture_block {
        DefaultImage := UI.invisible_img
        DefaultDesiredSize := vector2 {
            X := 187.29332
            Y := 179.594254
        }
    }
    var TextureMaskLevelStar2: texture_block = texture_block {
        DefaultImage := UI.invisible_img
        DefaultDesiredSize := vector2 {
            X := 187.29332
            Y := 179.594254
        }
    }
    var TextureMaskLevelStar3: texture_block = texture_block {
        DefaultImage := UI.invisible_img
        DefaultDesiredSize := vector2 {
            X := 187.29332
            Y := 179.594254
        }
    }
    var TextureMaskLevelStar4: texture_block = texture_block {
        DefaultImage := UI.invisible_img
        DefaultDesiredSize := vector2 {
            X := 187.29332
            Y := 179.594254
        }
    }
    var TextureMaskLevelStar5: texture_block = texture_block {
        DefaultImage := UI.invisible_img
        DefaultDesiredSize := vector2 {
            X := 187.29332
            Y := 179.594254
        }
    }
    var TextureMaskLevelStar6: texture_block = texture_block {
        DefaultImage := UI.invisible_img
        DefaultDesiredSize := vector2 {
            X := 187.29332
            Y := 179.594254
        }
    }
    
    InitWidget<override>(Agent: agent): void = {
        Print("TEST - InitWidget")
        (super:)InitWidget(Agent)
    }

    AddUiToPlayer<override>(Agent: agent, ParmMaybeManager: ?_Interface.interface_ui_manager): void = {
        (super:)AddUiToPlayer(Agent, ParmMaybeManager)

        GameManager := GetGameManager()
        Utility := GetUtility()
        # Initially show Player frame name HUD
        GameManager.HudPlayerFramePlayerName.Show(Agent, Utility.PlayerNameMessage(Agent, ""))

        Print("TEST: AddUiToPlayer: AssignedUiIndex: {AssignedUiIndex}")
    }

    CreateSlots<override>(Agent: agent): []canvas_slot = {
        Print("CreateSlots")
        Utility := GetUtility()

        set LevelBarColorBlockList = array {
            LevelBarColorBlock1,
            LevelBarColorBlock2,
            LevelBarColorBlock3,
            LevelBarColorBlock4,
            LevelBarColorBlock5,
            LevelBarColorBlock6,
            LevelBarColorBlock7,
            LevelBarColorBlock8,
            LevelBarColorBlock9,
            LevelBarColorBlock10
        }

        # MONEY POPUP --------------
        Money_Icon_Pos_1: texture_block = texture_block {  
            DefaultImage := UI.invisible_img
            DefaultDesiredSize := vector2{X := 75.510513, Y := 41.0}
        }
        
        Money_Icon_Pos_2: texture_block = texture_block {  
            DefaultImage := UI.invisible_img 
            DefaultDesiredSize := vector2{X := 75.510513, Y := 41.0}
        }
        
        Money_Icon_Pos_3: texture_block = texture_block {  
            DefaultImage := UI.invisible_img 
            DefaultDesiredSize := vector2{X := 75.510513, Y := 41.0}
        }
        
        Money_Icon_Pos_4: texture_block = texture_block {  
            DefaultImage := UI.invisible_img 
            DefaultDesiredSize := vector2{X := 75.510513, Y := 41.0}
        }
        
        Money_Icon_Pos_5: texture_block = texture_block {  
            DefaultImage := UI.invisible_img 
            DefaultDesiredSize := vector2{X := 75.510513, Y := 41.0}
        }
        
        Money_Icon_Pos_6: texture_block = texture_block {  
            DefaultImage := UI.invisible_img 
            DefaultDesiredSize := vector2{X := 75.510513, Y := 41.0}
        }
        
        Money_Icon_Pos_7: texture_block = texture_block {  
            DefaultImage := UI.invisible_img 
            DefaultDesiredSize := vector2{X := 75.510513, Y := 41.0}
        }
        
        Money_Icon_Pos_8: texture_block = texture_block {  
            DefaultImage := UI.invisible_img 
            DefaultDesiredSize := vector2{X := 75.510513, Y := 41.0}
        }
        
        Money_Icon_Pos_9: texture_block = texture_block {  
            DefaultImage := UI.invisible_img 
            DefaultDesiredSize := vector2{X := 75.510513, Y := 41.0}
        }
        
        Money_Icon_Pos_10: texture_block = texture_block {  
            DefaultImage := UI.invisible_img 
            DefaultDesiredSize := vector2{X := 75.510513, Y := 41.0}
        }
        
        Money_Icon_Pos_11: texture_block = texture_block {  
            DefaultImage := UI.invisible_img 
            DefaultDesiredSize := vector2{X := 75.510513, Y := 41.0}
        }
        
        Money_Icon_Pos_12: texture_block = texture_block {  
            DefaultImage := UI.invisible_img 
            DefaultDesiredSize := vector2{X := 75.510513, Y := 41.0}
        }
        
        Money_Icon_Pos_13: texture_block = texture_block {  
            DefaultImage := UI.invisible_img 
            DefaultDesiredSize := vector2{X := 75.510513, Y := 41.0}
        }
        
        Money_Icon_Pos_14: texture_block = texture_block {  
            DefaultImage := UI.invisible_img 
            DefaultDesiredSize := vector2{X := 75.510513, Y := 41.0}
        }
        
        Money_Icon_Pos_15: texture_block = texture_block {  
            DefaultImage := UI.invisible_img 
            DefaultDesiredSize := vector2{X := 75.510513, Y := 41.0}
        }
        
        Money_Icon_Pos_16: texture_block = texture_block {  
            DefaultImage := UI.invisible_img 
            DefaultDesiredSize := vector2{X := 75.510513, Y := 41.0}
        }
        
        Money_Icon_Pos_17: texture_block = texture_block {  
            DefaultImage := UI.invisible_img 
            DefaultDesiredSize := vector2{X := 75.510513, Y := 41.0}
        }
        
        Money_Icon_Pos_18: texture_block = texture_block {  
            DefaultImage := UI.invisible_img 
            DefaultDesiredSize := vector2{X := 75.510513, Y := 41.0}
        }
        
        Money_Icon_Pos_19: texture_block = texture_block {  
            DefaultImage := UI.invisible_img 
            DefaultDesiredSize := vector2{X := 75.510513, Y := 41.0}
        }
        
        Money_Icon_Pos_20: texture_block = texture_block {  
            DefaultImage := UI.invisible_img 
            DefaultDesiredSize := vector2{X := 75.510513, Y := 41.0}
        }
        
        Money_Icon_Pos_21: texture_block = texture_block {  
            DefaultImage := UI.invisible_img 
            DefaultDesiredSize := vector2{X := 75.510513, Y := 41.0}
        }
        
        Money_Icon_Pos_22: texture_block = texture_block {  
            DefaultImage := UI.invisible_img 
            DefaultDesiredSize := vector2{X := 75.510513, Y := 41.0}
        }
        
        Money_Icon_Pos_23: texture_block = texture_block {  
            DefaultImage := UI.invisible_img 
            DefaultDesiredSize := vector2{X := 75.510513, Y := 41.0}
        }
        
        Money_Icon_Pos_24: texture_block = texture_block {  
            DefaultImage := UI.invisible_img 
            DefaultDesiredSize := vector2{X := 75.510513, Y := 41.0}
        }
        
        Money_Icon_Pos_25: texture_block = texture_block {  
            DefaultImage := UI.invisible_img 
            DefaultDesiredSize := vector2{X := 75.510513, Y := 41.0}
        }
        
        Money_Icon_Pos_26: texture_block = texture_block {  
            DefaultImage := UI.invisible_img 
            DefaultDesiredSize := vector2{X := 75.510513, Y := 41.0}
        }
        
        Money_Icon_Pos_27: texture_block = texture_block {  
            DefaultImage := UI.invisible_img 
            DefaultDesiredSize := vector2{X := 75.510513, Y := 41.0}
        }
        
        Money_Icon_Pos_28: texture_block = texture_block {  
            DefaultImage := UI.invisible_img 
            DefaultDesiredSize := vector2{X := 75.510513, Y := 41.0}
        }
        
        Money_Icon_Pos_29: texture_block = texture_block {  
            DefaultImage := UI.invisible_img 
            DefaultDesiredSize := vector2{X := 75.510513, Y := 41.0}
        }
        
        Money_Icon_Pos_30: texture_block = texture_block {  
            DefaultImage := UI.invisible_img 
            DefaultDesiredSize := vector2{X := 75.510513, Y := 41.0}
        }
        
        Money_Icon_Pos_31: texture_block = texture_block {  
            DefaultImage := UI.invisible_img 
            DefaultDesiredSize := vector2{X := 75.510513, Y := 41.0}
        }
        
        Money_Icon_Pos_32: texture_block = texture_block {  
            DefaultImage := UI.invisible_img 
            DefaultDesiredSize := vector2{X := 75.510513, Y := 41.0}
        }
        
        Money_Icon_Pos_33: texture_block = texture_block {  
            DefaultImage := UI.invisible_img 
            DefaultDesiredSize := vector2{X := 75.510513, Y := 41.0}
        }
        
        Money_Icon_Pos_34: texture_block = texture_block {  
            DefaultImage := UI.invisible_img 
            DefaultDesiredSize := vector2{X := 75.510513, Y := 41.0}
        }
        
        Money_Icon_Pos_35: texture_block = texture_block {  
            DefaultImage := UI.invisible_img 
            DefaultDesiredSize := vector2{X := 75.510513, Y := 41.0}
        }
        
        Money_Icon_Pos_36: texture_block = texture_block {  
            DefaultImage := UI.invisible_img 
            DefaultDesiredSize := vector2{X := 75.510513, Y := 41.0}
        }
        
        Money_Icon_Pos_37: texture_block = texture_block {  
            DefaultImage := UI.invisible_img 
            DefaultDesiredSize := vector2{X := 75.510513, Y := 41.0}
        }
        
        Money_Icon_Pos_38: texture_block = texture_block {  
            DefaultImage := UI.invisible_img 
            DefaultDesiredSize := vector2{X := 75.510513, Y := 41.0}
        }
        
        Money_Icon_Pos_39: texture_block = texture_block {  
            DefaultImage := UI.invisible_img 
            DefaultDesiredSize := vector2{X := 75.510513, Y := 41.0}
        }
        
        Money_Icon_Pos_40: texture_block = texture_block {  
            DefaultImage := UI.invisible_img 
            DefaultDesiredSize := vector2{X := 75.510513, Y := 41.0}
        }
        
        Money_Icon_Pos_41: texture_block = texture_block {  
            DefaultImage := UI.invisible_img 
            DefaultDesiredSize := vector2{X := 75.510513, Y := 41.0}
        }
        
        Money_Icon_Pos_42: texture_block = texture_block {  
            DefaultImage := UI.invisible_img 
            DefaultDesiredSize := vector2{X := 75.510513, Y := 41.0}
        }
        
        Money_Icon_Pos_43: texture_block = texture_block {  
            DefaultImage := UI.invisible_img 
            DefaultDesiredSize := vector2{X := 75.510513, Y := 41.0}
        }
        
        Money_Icon_Pos_44: texture_block = texture_block {  
            DefaultImage := UI.invisible_img 
            DefaultDesiredSize := vector2{X := 75.510513, Y := 41.0}
        }
        
        Money_Icon_Pos_45: texture_block = texture_block {  
            DefaultImage := UI.invisible_img 
            DefaultDesiredSize := vector2{X := 75.510513, Y := 41.0}
        }
        
        Money_Icon_Pos_46: texture_block = texture_block {  
            DefaultImage := UI.invisible_img 
            DefaultDesiredSize := vector2{X := 75.510513, Y := 41.0}
        }
        
        Money_Icon_Pos_47: texture_block = texture_block {  
            DefaultImage := UI.invisible_img 
            DefaultDesiredSize := vector2{X := 75.510513, Y := 41.0}
        }
        
        Money_Icon_Pos_48: texture_block = texture_block {  
            DefaultImage := UI.invisible_img 
            DefaultDesiredSize := vector2{X := 75.510513, Y := 41.0}
        }
        
        Money_Icon_Pos_49: texture_block = texture_block {  
            DefaultImage := UI.invisible_img 
            DefaultDesiredSize := vector2{X := 75.510513, Y := 41.0}
        }
        
        Money_Icon_Pos_50: texture_block = texture_block {  
            DefaultImage := UI.invisible_img 
            DefaultDesiredSize := vector2{X := 75.510513, Y := 41.0}
        }
        
        Money_Icon_Pos_51: texture_block = texture_block {  
            DefaultImage := UI.invisible_img 
            DefaultDesiredSize := vector2{X := 75.510513, Y := 41.0}
        }
        
        Money_Icon_Pos_52: texture_block = texture_block {  
            DefaultImage := UI.invisible_img 
            DefaultDesiredSize := vector2{X := 75.510513, Y := 41.0}
        }
        
        Money_Icon_Pos_53: texture_block = texture_block {  
            DefaultImage := UI.invisible_img 
            DefaultDesiredSize := vector2{X := 75.510513, Y := 41.0}
        }
        
        Money_Icon_Pos_54: texture_block = texture_block {  
            DefaultImage := UI.invisible_img 
            DefaultDesiredSize := vector2{X := 75.510513, Y := 41.0}
        }
        
        Money_Icon_Pos_55: texture_block = texture_block {  
            DefaultImage := UI.invisible_img 
            DefaultDesiredSize := vector2{X := 75.510513, Y := 41.0}
        }
        
        Money_Icon_Pos_56: texture_block = texture_block {  
            DefaultImage := UI.invisible_img 
            DefaultDesiredSize := vector2{X := 75.510513, Y := 41.0}
        }
        
        Money_Icon_Pos_57: texture_block = texture_block {  
            DefaultImage := UI.invisible_img 
            DefaultDesiredSize := vector2{X := 75.510513, Y := 41.0}
        }
        
        Money_Icon_Pos_58: texture_block = texture_block {  
            DefaultImage := UI.invisible_img 
            DefaultDesiredSize := vector2{X := 75.510513, Y := 41.0}
        }
        
        Money_Icon_Pos_59: texture_block = texture_block {  
            DefaultImage := UI.invisible_img 
            DefaultDesiredSize := vector2{X := 75.510513, Y := 41.0}
        }
        
        Money_Icon_Pos_60: texture_block = texture_block {  
            DefaultImage := UI.invisible_img 
            DefaultDesiredSize := vector2{X := 75.510513, Y := 41.0}
        }

        #-------------- MONEY NUMBER TEXT -------------- 
        Pos_1:= text_block{DefaultText := Utility.FormatMessage("50M")
            DefaultJustification := text_justification.Left
            DefaultTextColor := color{R:= 1.0, G:= 1.0, B:= 1.0}
            DefaultTextOpacity := 1.000}
        
        
        Pos_2:= text_block{DefaultText := Utility.FormatMessage("50M")
            DefaultJustification := text_justification.Left
            DefaultTextColor := color{R:= 1.0, G:= 1.0, B:= 1.0}
            DefaultTextOpacity := 0.976}
        
        
        Pos_3:= text_block{DefaultText := Utility.FormatMessage("50M")
            DefaultJustification := text_justification.Left
            DefaultTextColor := color{R:= 1.0, G:= 1.0, B:= 1.0}
            DefaultTextOpacity := 0.951}
        
        
        Pos_4:= text_block{DefaultText := Utility.FormatMessage("50M")
            DefaultJustification := text_justification.Left
            DefaultTextColor := color{R:= 1.0, G:= 1.0, B:= 1.0}
            DefaultTextOpacity := 0.927}
        
        
        Pos_5:= text_block{DefaultText := Utility.FormatMessage("50M")
            DefaultJustification := text_justification.Left
            DefaultTextColor := color{R:= 1.0, G:= 1.0, B:= 1.0}
            DefaultTextOpacity := 0.902}
        
        
        Pos_6:= text_block{DefaultText := Utility.FormatMessage("50M")
            DefaultJustification := text_justification.Left
            DefaultTextColor := color{R:= 1.0, G:= 1.0, B:= 1.0}
            DefaultTextOpacity := 0.878}
        
        
        Pos_7:= text_block{DefaultText := Utility.FormatMessage("50M")
            DefaultJustification := text_justification.Left
            DefaultTextColor := color{R:= 1.0, G:= 1.0, B:= 1.0}
            DefaultTextOpacity := 0.854}
        
        
        Pos_8:= text_block{DefaultText := Utility.FormatMessage("50M")
            DefaultJustification := text_justification.Left
            DefaultTextColor := color{R:= 1.0, G:= 1.0, B:= 1.0}
            DefaultTextOpacity := 0.829}
        
        
        Pos_9:= text_block{DefaultText := Utility.FormatMessage("50M")
            DefaultJustification := text_justification.Left
            DefaultTextColor := color{R:= 1.0, G:= 1.0, B:= 1.0}
            DefaultTextOpacity := 0.805}
        
        
        Pos_10:= text_block{DefaultText := Utility.FormatMessage("50M")
            DefaultJustification := text_justification.Left
            DefaultTextColor := color{R:= 1.0, G:= 1.0, B:= 1.0}
            DefaultTextOpacity := 0.780}
        
    
        Pos_11:= text_block{DefaultText := Utility.FormatMessage("50M")
            DefaultJustification := text_justification.Left
            DefaultTextColor := color{R:= 1.0, G:= 1.0, B:= 1.0}
            DefaultTextOpacity := 0.756}
        
        
        Pos_12:= text_block{DefaultText := Utility.FormatMessage("50M")
            DefaultJustification := text_justification.Left
            DefaultTextColor := color{R:= 1.0, G:= 1.0, B:= 1.0}
            DefaultTextOpacity := 0.732}
        
        
        Pos_13:= text_block{DefaultText := Utility.FormatMessage("50M")
            DefaultJustification := text_justification.Left
            DefaultTextColor := color{R:= 1.0, G:= 1.0, B:= 1.0}
            DefaultTextOpacity := 0.707}
            
    
        Pos_14:= text_block{DefaultText := Utility.FormatMessage("50M")
            DefaultJustification := text_justification.Left
            DefaultTextColor := color{R:= 1.0, G:= 1.0, B:= 1.0}
            DefaultTextOpacity := 0.683}
        
        
        Pos_15:= text_block{DefaultText := Utility.FormatMessage("50M")
            DefaultJustification := text_justification.Left
            DefaultTextColor := color{R:= 1.0, G:= 1.0, B:= 1.0}
            DefaultTextOpacity := 0.658}
        
        
        Pos_16:= text_block{DefaultText := Utility.FormatMessage("50M")
            DefaultJustification := text_justification.Left
            DefaultTextColor := color{R:= 1.0, G:= 1.0, B:= 1.0}
            DefaultTextOpacity := 0.634}
        
        
        Pos_17:= text_block{DefaultText := Utility.FormatMessage("50M")
            DefaultJustification := text_justification.Left
            DefaultTextColor := color{R:= 1.0, G:= 1.0, B:= 1.0}
            DefaultTextOpacity := 0.610}
        
        
        Pos_18:= text_block{DefaultText := Utility.FormatMessage("50M")
            DefaultJustification := text_justification.Left
            DefaultTextColor := color{R:= 1.0, G:= 1.0, B:= 1.0}
            DefaultTextOpacity := 0.585}
        
        
        Pos_19:= text_block{DefaultText := Utility.FormatMessage("50M")
            DefaultJustification := text_justification.Left
            DefaultTextColor := color{R:= 1.0, G:= 1.0, B:= 1.0}
            DefaultTextOpacity := 0.561}
        
        
        Pos_20:= text_block{DefaultText := Utility.FormatMessage("50M")
            DefaultJustification := text_justification.Left
            DefaultTextColor := color{R:= 1.0, G:= 1.0, B:= 1.0}
            DefaultTextOpacity := 0.536}
        
        
        Pos_21:= text_block{DefaultText := Utility.FormatMessage("50M")
            DefaultJustification := text_justification.Left
            DefaultTextColor := color{R:= 1.0, G:= 1.0, B:= 1.0}
            DefaultTextOpacity := 0.512}
        
        
        Pos_22:= text_block{DefaultText := Utility.FormatMessage("50M")
            DefaultJustification := text_justification.Left
            DefaultTextColor := color{R:= 1.0, G:= 1.0, B:= 1.0}
            DefaultTextOpacity := 0.488}
        
        
        Pos_23:= text_block{DefaultText := Utility.FormatMessage("50M")
            DefaultJustification := text_justification.Left
            DefaultTextColor := color{R:= 1.0, G:= 1.0, B:= 1.0}
            DefaultTextOpacity := 0.463}
        
        
        Pos_24:= text_block{DefaultText := Utility.FormatMessage("50M")
            DefaultJustification := text_justification.Left
            DefaultTextColor := color{R:= 1.0, G:= 1.0, B:= 1.0}
            DefaultTextOpacity := 0.439}
        
        
        Pos_25:= text_block{DefaultText := Utility.FormatMessage("50M")
            DefaultJustification := text_justification.Left
            DefaultTextColor := color{R:= 1.0, G:= 1.0, B:= 1.0}
            DefaultTextOpacity := 0.414}
        
        
        Pos_26:= text_block{DefaultText := Utility.FormatMessage("50M")
            DefaultJustification := text_justification.Left
            DefaultTextColor := color{R:= 1.0, G:= 1.0, B:= 1.0}
            DefaultTextOpacity := 0.390}
        
        
        Pos_27:= text_block{DefaultText := Utility.FormatMessage("50M")
            DefaultJustification := text_justification.Left
            DefaultTextColor := color{R:= 1.0, G:= 1.0, B:= 1.0}
            DefaultTextOpacity := 0.366}
        
        
        Pos_28:= text_block{DefaultText := Utility.FormatMessage("50M")
            DefaultJustification := text_justification.Left
            DefaultTextColor := color{R:= 1.0, G:= 1.0, B:= 1.0}
            DefaultTextOpacity := 0.341}
        
        
        Pos_29:= text_block{DefaultText := Utility.FormatMessage("50M")
            DefaultJustification := text_justification.Left
            DefaultTextColor := color{R:= 1.0, G:= 1.0, B:= 1.0}
            DefaultTextOpacity := 0.317}
        
        
        Pos_30:= text_block{DefaultText := Utility.FormatMessage("50M")
            DefaultJustification := text_justification.Left
            DefaultTextColor := color{R:= 1.0, G:= 1.0, B:= 1.0}
            DefaultTextOpacity := 0.300}
        
        Pos_31:= text_block{DefaultText := Utility.FormatMessage("50M")
            DefaultJustification := text_justification.Left
            DefaultTextColor := color{R:= 1.0, G:= 1.0, B:= 1.0}
            DefaultTextOpacity := 0.290}
        
        
        Pos_32:= text_block{DefaultText := Utility.FormatMessage("50M")
            DefaultJustification := text_justification.Left
            DefaultTextColor := color{R:= 1.0, G:= 1.0, B:= 1.0}
            DefaultTextOpacity := 0.280}
        
        
        Pos_33:= text_block{DefaultText := Utility.FormatMessage("50M")
            DefaultJustification := text_justification.Left
            DefaultTextColor := color{R:= 1.0, G:= 1.0, B:= 1.0}
            DefaultTextOpacity := 0.271}
        
        
        Pos_34:= text_block{DefaultText := Utility.FormatMessage("50M")
            DefaultJustification := text_justification.Left
            DefaultTextColor := color{R:= 1.0, G:= 1.0, B:= 1.0}
            DefaultTextOpacity := 0.261}
        
    
        Pos_35:= text_block{DefaultText := Utility.FormatMessage("50M")
            DefaultJustification := text_justification.Left
            DefaultTextColor := color{R:= 1.0, G:= 1.0, B:= 1.0}
            DefaultTextOpacity := 0.251}
        
        
        Pos_36:= text_block{DefaultText := Utility.FormatMessage("50M")
            DefaultJustification := text_justification.Left
            DefaultTextColor := color{R:= 1.0, G:= 1.0, B:= 1.0}
            DefaultTextOpacity := 0.241}
        
        
        Pos_37:= text_block{DefaultText := Utility.FormatMessage("50M")
            DefaultJustification := text_justification.Left
            DefaultTextColor := color{R:= 1.0, G:= 1.0, B:= 1.0}
            DefaultTextOpacity := 0.232}
        
        
        Pos_38:= text_block{DefaultText := Utility.FormatMessage("50M")
            DefaultJustification := text_justification.Left
            DefaultTextColor := color{R:= 1.0, G:= 1.0, B:= 1.0}
            DefaultTextOpacity := 0.222}
        
        
        Pos_39:= text_block{DefaultText := Utility.FormatMessage("50M")
            DefaultJustification := text_justification.Left
            DefaultTextColor := color{R:= 1.0, G:= 1.0, B:= 1.0}
            DefaultTextOpacity := 0.212}
        
        
        Pos_40:= text_block{DefaultText := Utility.FormatMessage("50M")
            DefaultJustification := text_justification.Left
            DefaultTextColor := color{R:= 1.0, G:= 1.0, B:= 1.0}
            DefaultTextOpacity := 0.202}
        
        
        Pos_41:= text_block{DefaultText := Utility.FormatMessage("50M")
            DefaultJustification := text_justification.Left
            DefaultTextColor := color{R:= 1.0, G:= 1.0, B:= 1.0}
            DefaultTextOpacity := 0.193}
        
        
        Pos_42:= text_block{DefaultText := Utility.FormatMessage("50M")
            DefaultJustification := text_justification.Left
            DefaultTextColor := color{R:= 1.0, G:= 1.0, B:= 1.0}
            DefaultTextOpacity := 0.183}
        
        
        Pos_43:= text_block{DefaultText := Utility.FormatMessage("50M")
            DefaultJustification := text_justification.Left
            DefaultTextColor := color{R:= 1.0, G:= 1.0, B:= 1.0}
            DefaultTextOpacity := 0.173}
        
        
        Pos_44:= text_block{DefaultText := Utility.FormatMessage("50M")
            DefaultJustification := text_justification.Left
            DefaultTextColor := color{R:= 1.0, G:= 1.0, B:= 1.0}
            DefaultTextOpacity := 0.163}
        
        
        Pos_45:= text_block{DefaultText := Utility.FormatMessage("50M")
            DefaultJustification := text_justification.Left
            DefaultTextColor := color{R:= 1.0, G:= 1.0, B:= 1.0}
            DefaultTextOpacity := 0.154}
        
        
        Pos_46:= text_block{DefaultText := Utility.FormatMessage("50M")
            DefaultJustification := text_justification.Left
            DefaultTextColor := color{R:= 1.0, G:= 1.0, B:= 1.0}
            DefaultTextOpacity := 0.144}
        
        Pos_47:= text_block{DefaultText := Utility.FormatMessage("50M")
            DefaultJustification := text_justification.Left
            DefaultTextColor := color{R:= 1.0, G:= 1.0, B:= 1.0}
            DefaultTextOpacity := 0.134}
        
        Pos_48:= text_block{DefaultText := Utility.FormatMessage("50M")
            DefaultJustification := text_justification.Left
            DefaultTextColor := color{R:= 1.0, G:= 1.0, B:= 1.0}
            DefaultTextOpacity := 0.124}
        
        Pos_49:= text_block{DefaultText := Utility.FormatMessage("50M")
            DefaultJustification := text_justification.Left
            DefaultTextColor := color{R:= 1.0, G:= 1.0, B:= 1.0}
            DefaultTextOpacity := 0.115}
        
        Pos_50:= text_block{DefaultText := Utility.FormatMessage("50M")
            DefaultJustification := text_justification.Left
            DefaultTextColor := color{R:= 1.0, G:= 1.0, B:= 1.0}
            DefaultTextOpacity := 0.105}
        
        Pos_51:= text_block{DefaultText := Utility.FormatMessage("50M")
            DefaultJustification := text_justification.Left
            DefaultTextColor := color{R:= 1.0, G:= 1.0, B:= 1.0}
            DefaultTextOpacity := 0.095}
        
        Pos_52:= text_block{DefaultText := Utility.FormatMessage("50M")
            DefaultJustification := text_justification.Left
            DefaultTextColor := color{R:= 1.0, G:= 1.0, B:= 1.0}
            DefaultTextOpacity := 0.085}
        
        Pos_53:= text_block{DefaultText := Utility.FormatMessage("50M")
            DefaultJustification := text_justification.Left
            DefaultTextColor := color{R:= 1.0, G:= 1.0, B:= 1.0}
            DefaultTextOpacity := 0.076}
        
        Pos_54:= text_block{DefaultText := Utility.FormatMessage("50M")
            DefaultJustification := text_justification.Left
            DefaultTextColor := color{R:= 1.0, G:= 1.0, B:= 1.0}
            DefaultTextOpacity := 0.066}
        
        Pos_55:= text_block{DefaultText := Utility.FormatMessage("50M")
            DefaultJustification := text_justification.Left
            DefaultTextColor := color{R:= 1.0, G:= 1.0, B:= 1.0}
            DefaultTextOpacity := 0.056}
        
        Pos_56:= text_block{DefaultText := Utility.FormatMessage("50M")
            DefaultJustification := text_justification.Left
            DefaultTextColor := color{R:= 1.0, G:= 1.0, B:= 1.0}
            DefaultTextOpacity := 0.046}
        
        Pos_57:= text_block{DefaultText := Utility.FormatMessage("50M")
            DefaultJustification := text_justification.Left
            DefaultTextColor := color{R:= 1.0, G:= 1.0, B:= 1.0}
            DefaultTextOpacity := 0.037}
        
        Pos_58:= text_block{DefaultText := Utility.FormatMessage("50M")
            DefaultJustification := text_justification.Left
            DefaultTextColor := color{R:= 1.0, G:= 1.0, B:= 1.0}
            DefaultTextOpacity := 0.027}
        
        Pos_59:= text_block{DefaultText := Utility.FormatMessage("50M")
            DefaultJustification := text_justification.Left
            DefaultTextColor := color{R:= 1.0, G:= 1.0, B:= 1.0}
            DefaultTextOpacity := 0.01}
    
        Pos_60:= text_block{DefaultText := Utility.FormatMessage("50M")        
            DefaultJustification := text_justification.Left
            DefaultTextColor := color{R:= 1.0, G:= 1.0, B:= 1.0}
            DefaultTextOpacity := 0.0}

        set FloatingTextList = array {
            Pos_1,
            Pos_2,
            Pos_3,
            Pos_4,
            Pos_5,
            Pos_6,
            Pos_7,
            Pos_8,
            Pos_9,
            Pos_10,
            Pos_11,
            Pos_12,
            Pos_13,
            Pos_14,
            Pos_15,
            Pos_16,
            Pos_17,
            Pos_18,
            Pos_19,
            Pos_20,
            Pos_21,
            Pos_22,
            Pos_23,
            Pos_24,
            Pos_25,
            Pos_26,
            Pos_27,
            Pos_28,
            Pos_29,
            Pos_30,
            Pos_31,
            Pos_32,
            Pos_33,
            Pos_34,
            Pos_35,
            Pos_36,
            Pos_37,
            Pos_38,
            Pos_39,
            Pos_40,
            Pos_41,
            Pos_42,
            Pos_43,
            Pos_44,
            Pos_45,
            Pos_46,
            Pos_47,
            Pos_48,
            Pos_49,
            Pos_50,
            Pos_51,
            Pos_52,
            Pos_53,
            Pos_54,
            Pos_55,
            Pos_56,
            Pos_57,
            Pos_58,
            Pos_59,
            Pos_60
        }

        set FloatingIconList = array {
            Money_Icon_Pos_1,
            Money_Icon_Pos_2,
            Money_Icon_Pos_3,
            Money_Icon_Pos_4,
            Money_Icon_Pos_5,
            Money_Icon_Pos_6,
            Money_Icon_Pos_7,
            Money_Icon_Pos_8,
            Money_Icon_Pos_9,
            Money_Icon_Pos_10,
            Money_Icon_Pos_11,
            Money_Icon_Pos_12,
            Money_Icon_Pos_13,
            Money_Icon_Pos_14,
            Money_Icon_Pos_15,
            Money_Icon_Pos_16,
            Money_Icon_Pos_17,
            Money_Icon_Pos_18,
            Money_Icon_Pos_19,
            Money_Icon_Pos_20,
            Money_Icon_Pos_21,
            Money_Icon_Pos_22,
            Money_Icon_Pos_23,
            Money_Icon_Pos_24,
            Money_Icon_Pos_25,
            Money_Icon_Pos_26,
            Money_Icon_Pos_27,
            Money_Icon_Pos_28,
            Money_Icon_Pos_29,
            Money_Icon_Pos_30,
            Money_Icon_Pos_31,
            Money_Icon_Pos_32,
            Money_Icon_Pos_33,
            Money_Icon_Pos_34,
            Money_Icon_Pos_35,
            Money_Icon_Pos_36,
            Money_Icon_Pos_37,
            Money_Icon_Pos_38,
            Money_Icon_Pos_39,
            Money_Icon_Pos_40,
            Money_Icon_Pos_41,
            Money_Icon_Pos_42,
            Money_Icon_Pos_43,
            Money_Icon_Pos_44,
            Money_Icon_Pos_45,
            Money_Icon_Pos_46,
            Money_Icon_Pos_47,
            Money_Icon_Pos_48,
            Money_Icon_Pos_49,
            Money_Icon_Pos_50,
            Money_Icon_Pos_51,
            Money_Icon_Pos_52,
            Money_Icon_Pos_53,
            Money_Icon_Pos_54,
            Money_Icon_Pos_55,
            Money_Icon_Pos_56,
            Money_Icon_Pos_57,
            Money_Icon_Pos_58,
            Money_Icon_Pos_59,
            Money_Icon_Pos_60
        }

        var Slots: []canvas_slot = array {}
        if {
            PlayerData := Agent.GetPlayerData[]
        } then {
            Print("CreateSlots 222")
            # Level bar stuff
            set Slots = Slots + array {
                # Canvas slot for current level
                canvas_slot {
                    Anchors := anchors{Minimum := vector2{X := 0.0, Y := 0.5}, Maximum := vector2{X := 0.0, Y := 0.5}}
                    Offsets := margin{Top := -28.540527, Left := 24.0, Right := 38.236313, Bottom := 37.401249}
                    Alignment := vector2{X := 0.0, Y := 0.0}
                    SizeToContent := false
                    Widget := TextureMaskLevelStar1
                    ZOrder := 1050
                },
                canvas_slot {
                    Anchors := anchors{Minimum := vector2{X := 0.0, Y := 0.5}, Maximum := vector2{X := 0.0, Y := 0.5}}
                    Offsets := margin{Top := -28.540527, Left := 60.0, Right := 38.236313, Bottom := 37.401249}
                    Alignment := vector2{X := 0.0, Y := 0.0}
                    SizeToContent := false
                    Widget := TextureMaskLevelStar2
                    ZOrder := 1050
                },
                canvas_slot {
                    Anchors := anchors{Minimum := vector2{X := 0.0, Y := 0.5}, Maximum := vector2{X := 0.0, Y := 0.5}}
                    Offsets := margin{Top := -28.540527, Left := 97.523415, Right := 38.236313, Bottom := 37.401249}
                    Alignment := vector2{X := 0.0, Y := 0.0}
                    SizeToContent := false
                    Widget := TextureMaskLevelStar3
                    ZOrder := 1050
                },
                canvas_slot {
                    Anchors := anchors{Minimum := vector2{X := 0.0, Y := 0.5}, Maximum := vector2{X := 0.0, Y := 0.5}}
                    Offsets := margin{Top := -28.540527, Left := 133.764191, Right := 38.236313, Bottom := 37.401249}
                    Alignment := vector2{X := 0.0, Y := 0.0}
                    SizeToContent := false
                    Widget := TextureMaskLevelStar4
                    ZOrder := 1050
                },
                canvas_slot {
                    Anchors := anchors{Minimum := vector2{X := 0.0, Y := 0.5}, Maximum := vector2{X := 0.0, Y := 0.5}}
                    Offsets := margin{Top := -28.540527, Left := 172.0, Right := 38.236313, Bottom := 37.401249}
                    Alignment := vector2{X := 0.0, Y := 0.0}
                    SizeToContent := false
                    Widget := TextureMaskLevelStar5
                    ZOrder := 1050
                },
                canvas_slot {
                    Anchors := anchors{Minimum := vector2{X := 0.0, Y := 0.5}, Maximum := vector2{X := 0.0, Y := 0.5}}
                    Offsets := margin{Top := -28.540527, Left := 208.0, Right := 38.236313, Bottom := 37.401249}
                    Alignment := vector2{X := 0.0, Y := 0.0}
                    SizeToContent := false
                    Widget := TextureMaskLevelStar6
                    ZOrder := 1050
                },
                # Slots for the level bar
                <#
                canvas_slot {
                    Anchors := anchors{Minimum := vector2{X := 0.5, Y := 1.0}, Maximum := vector2{X := 0.5, Y := 1.0}}
                    Offsets := margin{Top := -109.081055, Left := -268.960938, Right := 165.0, Bottom := 41.0}
                    Alignment := vector2{X := 0.0, Y := 0.0}
                    SizeToContent := false
                    Widget := TextBlockCurLevel
                    ZOrder := 1030
                },
                
                canvas_slot {
                    Anchors := anchors{Minimum := vector2{X := 0.0, Y := 0.5}, Maximum := vector2{X := 0.0, Y := 0.5}}
                    Offsets := margin{Top := 38.020287, Left := 80.0, Right := 165.0, Bottom := 41.0}
                    Alignment := vector2{X := 0.0, Y := 0.0}
                    SizeToContent := false
                    Widget := TextBlockCurMoney
                    ZOrder := 910
                },
                #>
                canvas_slot {
                    Anchors := anchors{Minimum := vector2{X := 0.0, Y := 1.0}, Maximum := vector2{X := 0.0, Y := 1.0}}
                    Offsets := margin{Top := -40.333496, Left := 39.738506, Right := 391.753296, Bottom := 13.088473}
                    Alignment := vector2{X := 0.0, Y := 0.0}
                    SizeToContent := false
                    Widget := LevelBarBackgroundColorBlock
                    ZOrder := 1009
                },
                canvas_slot {
                    Anchors := anchors{Minimum := vector2{X := 0.0, Y := 1.0}, Maximum := vector2{X := 0.0, Y := 1.0}}
                    Offsets := margin{Top := -293.081055, Left := -16.0, Right := 509.438873, Bottom := 421.078094}
                    Alignment := vector2{X := 0.0, Y := 0.0}
                    SizeToContent := false
                    Widget := Level_Overlay
                    ZOrder := 901
                },
                canvas_slot {
                    Anchors := anchors{Minimum := vector2{X := 0.0, Y := 1.0}, Maximum := vector2{X := 0.0, Y := 1.0}}
                    Offsets := margin{Top := -41.081055, Left := 41.039501, Right := 35.89743, Bottom := 13.088473}
                    Alignment := vector2{X := 0.0, Y := 0.0}
                    SizeToContent := true
                    Widget := LevelBarColorBlock1
                    ZOrder := 1010
                },
                canvas_slot {
                    Anchors := anchors{Minimum := vector2{X := 0.0, Y := 1.0}, Maximum := vector2{X := 0.0, Y := 1.0}}
                    Offsets := margin{Top := -41.081055, Left := 80.0, Right := 36.243931, Bottom := 13.665974}
                    Alignment := vector2{X := 0.0, Y := 0.0}
                    SizeToContent := true
                    Widget := LevelBarColorBlock2
                    ZOrder := 1010
                },
                canvas_slot {
                    Anchors := anchors{Minimum := vector2{X := 0.0, Y := 1.0}, Maximum := vector2{X := 0.0, Y := 1.0}}
                    Offsets := margin{Top := -41.081055, Left := 119.653503, Right := 36.012932, Bottom := 13.665974}
                    Alignment := vector2{X := 0.0, Y := 0.0}
                    SizeToContent := true
                    Widget := LevelBarColorBlock3
                    ZOrder := 1010
                },
                canvas_slot {
                    Anchors := anchors{Minimum := vector2{X := 0.0, Y := 1.0}, Maximum := vector2{X := 0.0, Y := 1.0}}
                    Offsets := margin{Top := -41.081055, Left := 158.960495, Right := 35.781933, Bottom := 13.665974}
                    Alignment := vector2{X := 0.0, Y := 0.0}
                    SizeToContent := true
                    Widget := LevelBarColorBlock4
                    ZOrder := 1010
                },
                canvas_slot {
                    Anchors := anchors{Minimum := vector2{X := 0.0, Y := 1.0}, Maximum := vector2{X := 0.0, Y := 1.0}}
                    Offsets := margin{Top := -41.081055, Left := 198.036499, Right := 36.359432, Bottom := 13.665974}
                    Alignment := vector2{X := 0.0, Y := 0.0}
                    SizeToContent := true
                    Widget := LevelBarColorBlock5
                    ZOrder := 1010
                },
                canvas_slot {
                    Anchors := anchors{Minimum := vector2{X := 0.0, Y := 1.0}, Maximum := vector2{X := 0.0, Y := 1.0}}
                    Offsets := margin{Top := -41.081055, Left := 237.386002, Right := 36.128433, Bottom := 13.665974}
                    Alignment := vector2{X := 0.0, Y := 0.0}
                    SizeToContent := true
                    Widget := LevelBarColorBlock6
                    ZOrder := 1010
                },
                canvas_slot {
                    Anchors := anchors{Minimum := vector2{X := 0.0, Y := 1.0}, Maximum := vector2{X := 0.0, Y := 1.0}}
                    Offsets := margin{Top := -41.081055, Left := 276.577515, Right := 36.128433, Bottom := 13.665974}
                    Alignment := vector2{X := 0.0, Y := 0.0}
                    SizeToContent := true
                    Widget := LevelBarColorBlock7
                    ZOrder := 1010
                },
                canvas_slot {
                    Anchors := anchors{Minimum := vector2{X := 0.0, Y := 1.0}, Maximum := vector2{X := 0.0, Y := 1.0}}
                    Offsets := margin{Top := -41.081055, Left := 316.0, Right := 36.128433, Bottom := 13.665974}
                    Alignment := vector2{X := 0.0, Y := 0.0}
                    SizeToContent := true
                    Widget := LevelBarColorBlock8
                    ZOrder := 1010
                },
                canvas_slot {
                    Anchors := anchors{Minimum := vector2{X := 0.0, Y := 1.0}, Maximum := vector2{X := 0.0, Y := 1.0}}
                    Offsets := margin{Top := -41.081055, Left := 355.537994, Right := 35.666435, Bottom := 13.665974}
                    Alignment := vector2{X := 0.0, Y := 0.0}
                    SizeToContent := true
                    Widget := LevelBarColorBlock9
                    ZOrder := 1010
                },
                canvas_slot {
                    Anchors := anchors{Minimum := vector2{X := 0.0, Y := 1.0}, Maximum := vector2{X := 0.0, Y := 1.0}}
                    Offsets := margin{Top := -41.081055, Left := 394.656494, Right := 36.012936, Bottom := 13.665974}
                    Alignment := vector2{X := 0.0, Y := 0.0}
                    SizeToContent := true
                    Widget := LevelBarColorBlock10
                    ZOrder := 1010
                },
                # --------------MONEY ICON POSITIONS --------------
                canvas_slot {
                    Anchors := anchors{Minimum := vector2{X := 0.5, Y := 0.5}, Maximum := vector2{X := 0.5, Y := 0.5}}
                    Offsets := margin{Top := 84.634567, Left := 135.039062, Right := 38.808941, Bottom := 38.808941}
                    Alignment := vector2{X := 0.0, Y := 0.0}
                    SizeToContent := false
                    Widget := Money_Icon_Pos_1
                    ZOrder := 900
                },
                canvas_slot {
                    Anchors := anchors{Minimum := vector2{X := 0.5, Y := 0.5}, Maximum := vector2{X := 0.5, Y := 0.5}}
                    Offsets := margin{Top := 80.634567, Left := 135.039062, Right := 38.808941, Bottom := 38.808941}
                    Alignment := vector2{X := 0.0, Y := 0.0}
                    SizeToContent := false
                    Widget := Money_Icon_Pos_2
                    ZOrder := 901
                },
                canvas_slot {
                    Anchors := anchors{Minimum := vector2{X := 0.5, Y := 0.5}, Maximum := vector2{X := 0.5, Y := 0.5}}
                    Offsets := margin{Top := 76.634567, Left := 135.039062, Right := 38.808941, Bottom := 38.808941}
                    Alignment := vector2{X := 0.0, Y := 0.0}
                    SizeToContent := false
                    Widget := Money_Icon_Pos_3
                    ZOrder := 902
                },
                canvas_slot {
                    Anchors := anchors{Minimum := vector2{X := 0.5, Y := 0.5}, Maximum := vector2{X := 0.5, Y := 0.5}}
                    Offsets := margin{Top := 72.634567, Left := 135.039062, Right := 38.808941, Bottom := 38.808941}
                    Alignment := vector2{X := 0.0, Y := 0.0}
                    SizeToContent := false
                    Widget := Money_Icon_Pos_4
                    ZOrder := 903
                },
                canvas_slot {
                    Anchors := anchors{Minimum := vector2{X := 0.5, Y := 0.5}, Maximum := vector2{X := 0.5, Y := 0.5}}
                    Offsets := margin{Top := 68.634567, Left := 135.039062, Right := 38.808941, Bottom := 38.808941}
                    Alignment := vector2{X := 0.0, Y := 0.0}
                    SizeToContent := false
                    Widget := Money_Icon_Pos_5
                    ZOrder := 904
                },
                canvas_slot {
                    Anchors := anchors{Minimum := vector2{X := 0.5, Y := 0.5}, Maximum := vector2{X := 0.5, Y := 0.5}}
                    Offsets := margin{Top := 64.634567, Left := 135.039062, Right := 38.808941, Bottom := 38.808941}
                    Alignment := vector2{X := 0.0, Y := 0.0}
                    SizeToContent := false
                    Widget := Money_Icon_Pos_6
                    ZOrder := 905
                },
                canvas_slot {
                    Anchors := anchors{Minimum := vector2{X := 0.5, Y := 0.5}, Maximum := vector2{X := 0.5, Y := 0.5}}
                    Offsets := margin{Top := 60.634567, Left := 135.039062, Right := 38.808941, Bottom := 38.808941}
                    Alignment := vector2{X := 0.0, Y := 0.0}
                    SizeToContent := false
                    Widget := Money_Icon_Pos_7
                    ZOrder := 906
                },
                canvas_slot {
                    Anchors := anchors{Minimum := vector2{X := 0.5, Y := 0.5}, Maximum := vector2{X := 0.5, Y := 0.5}}
                    Offsets := margin{Top := 56.634567, Left := 135.039062, Right := 38.808941, Bottom := 38.808941}
                    Alignment := vector2{X := 0.0, Y := 0.0}
                    SizeToContent := false
                    Widget := Money_Icon_Pos_8
                    ZOrder := 907
                },
                canvas_slot {
                    Anchors := anchors{Minimum := vector2{X := 0.5, Y := 0.5}, Maximum := vector2{X := 0.5, Y := 0.5}}
                    Offsets := margin{Top := 52.634567, Left := 135.039062, Right := 38.808941, Bottom := 38.808941}
                    Alignment := vector2{X := 0.0, Y := 0.0}
                    SizeToContent := false
                    Widget := Money_Icon_Pos_9
                    ZOrder := 908
                },
                canvas_slot {
                    Anchors := anchors{Minimum := vector2{X := 0.5, Y := 0.5}, Maximum := vector2{X := 0.5, Y := 0.5}}
                    Offsets := margin{Top := 48.634567, Left := 135.039062, Right := 38.808941, Bottom := 38.808941}
                    Alignment := vector2{X := 0.0, Y := 0.0}
                    SizeToContent := false
                    Widget := Money_Icon_Pos_10
                    ZOrder := 909
                },
                canvas_slot {
                    Anchors := anchors{Minimum := vector2{X := 0.5, Y := 0.5}, Maximum := vector2{X := 0.5, Y := 0.5}}
                    Offsets := margin{Top := 44.634567, Left := 135.039062, Right := 38.808941, Bottom := 38.808941}
                    Alignment := vector2{X := 0.0, Y := 0.0}
                    SizeToContent := false
                    Widget := Money_Icon_Pos_11
                    ZOrder := 910
                },
                canvas_slot {
                    Anchors := anchors{Minimum := vector2{X := 0.5, Y := 0.5}, Maximum := vector2{X := 0.5, Y := 0.5}}
                    Offsets := margin{Top := 40.634567, Left := 135.039062, Right := 38.808941, Bottom := 38.808941}
                    Alignment := vector2{X := 0.0, Y := 0.0}
                    SizeToContent := false
                    Widget := Money_Icon_Pos_12
                    ZOrder := 911
                },
                canvas_slot {
                    Anchors := anchors{Minimum := vector2{X := 0.5, Y := 0.5}, Maximum := vector2{X := 0.5, Y := 0.5}}
                    Offsets := margin{Top := 36.634567, Left := 135.039062, Right := 38.808941, Bottom := 38.808941}
                    Alignment := vector2{X := 0.0, Y := 0.0}
                    SizeToContent := false
                    Widget := Money_Icon_Pos_13
                    ZOrder := 912
                },
                canvas_slot {
                    Anchors := anchors{Minimum := vector2{X := 0.5, Y := 0.5}, Maximum := vector2{X := 0.5, Y := 0.5}}
                    Offsets := margin{Top := 32.634567, Left := 135.039062, Right := 38.808941, Bottom := 38.808941}
                    Alignment := vector2{X := 0.0, Y := 0.0}
                    SizeToContent := false
                    Widget := Money_Icon_Pos_14
                    ZOrder := 913
                },
                canvas_slot {
                    Anchors := anchors{Minimum := vector2{X := 0.5, Y := 0.5}, Maximum := vector2{X := 0.5, Y := 0.5}}
                    Offsets := margin{Top := 28.634567, Left := 135.039062, Right := 38.808941, Bottom := 38.808941}
                    Alignment := vector2{X := 0.0, Y := 0.0}
                    SizeToContent := false
                    Widget := Money_Icon_Pos_15
                    ZOrder := 914
                },
                canvas_slot {
                    Anchors := anchors{Minimum := vector2{X := 0.5, Y := 0.5}, Maximum := vector2{X := 0.5, Y := 0.5}}
                    Offsets := margin{Top := 24.634567, Left := 135.039062, Right := 38.808941, Bottom := 38.808941}
                    Alignment := vector2{X := 0.0, Y := 0.0}
                    SizeToContent := false
                    Widget := Money_Icon_Pos_16
                    ZOrder := 915
                },
                canvas_slot {
                    Anchors := anchors{Minimum := vector2{X := 0.5, Y := 0.5}, Maximum := vector2{X := 0.5, Y := 0.5}}
                    Offsets := margin{Top := 20.634567, Left := 135.039062, Right := 38.808941, Bottom := 38.808941}
                    Alignment := vector2{X := 0.0, Y := 0.0}
                    SizeToContent := false
                    Widget := Money_Icon_Pos_17
                    ZOrder := 916
                },
                canvas_slot {
                    Anchors := anchors{Minimum := vector2{X := 0.5, Y := 0.5}, Maximum := vector2{X := 0.5, Y := 0.5}}
                    Offsets := margin{Top := 16.634567, Left := 135.039062, Right := 38.808941, Bottom := 38.808941}
                    Alignment := vector2{X := 0.0, Y := 0.0}
                    SizeToContent := false
                    Widget := Money_Icon_Pos_18
                    ZOrder := 917
                },
                canvas_slot {
                    Anchors := anchors{Minimum := vector2{X := 0.5, Y := 0.5}, Maximum := vector2{X := 0.5, Y := 0.5}}
                    Offsets := margin{Top := 12.634567, Left := 135.039062, Right := 38.808941, Bottom := 38.808941}
                    Alignment := vector2{X := 0.0, Y := 0.0}
                    SizeToContent := false
                    Widget := Money_Icon_Pos_19
                    ZOrder := 918
                },
                canvas_slot {
                    Anchors := anchors{Minimum := vector2{X := 0.5, Y := 0.5}, Maximum := vector2{X := 0.5, Y := 0.5}}
                    Offsets := margin{Top := 8.634567, Left := 135.039062, Right := 38.808941, Bottom := 38.808941}
                    Alignment := vector2{X := 0.0, Y := 0.0}
                    SizeToContent := false
                    Widget := Money_Icon_Pos_20
                    ZOrder := 919
                },
                canvas_slot {
                    Anchors := anchors{Minimum := vector2{X := 0.5, Y := 0.5}, Maximum := vector2{X := 0.5, Y := 0.5}}
                    Offsets := margin{Top := 4.634567, Left := 135.039062, Right := 38.808941, Bottom := 38.808941}
                    Alignment := vector2{X := 0.0, Y := 0.0}
                    SizeToContent := false
                    Widget := Money_Icon_Pos_21
                    ZOrder := 920
                },
                canvas_slot {
                    Anchors := anchors{Minimum := vector2{X := 0.5, Y := 0.5}, Maximum := vector2{X := 0.5, Y := 0.5}}
                    Offsets := margin{Top := 0.634567, Left := 135.039062, Right := 38.808941, Bottom := 38.808941}
                    Alignment := vector2{X := 0.0, Y := 0.0}
                    SizeToContent := false
                    Widget := Money_Icon_Pos_22
                    ZOrder := 921
                },
                canvas_slot {
                    Anchors := anchors{Minimum := vector2{X := 0.5, Y := 0.5}, Maximum := vector2{X := 0.5, Y := 0.5}}
                    Offsets := margin{Top := -3.365433, Left := 135.039062, Right := 38.808941, Bottom := 38.808941}
                    Alignment := vector2{X := 0.0, Y := 0.0}
                    SizeToContent := false
                    Widget := Money_Icon_Pos_23
                    ZOrder := 922
                },
                canvas_slot {
                    Anchors := anchors{Minimum := vector2{X := 0.5, Y := 0.5}, Maximum := vector2{X := 0.5, Y := 0.5}}
                    Offsets := margin{Top := -7.365433, Left := 135.039062, Right := 38.808941, Bottom := 38.808941}
                    Alignment := vector2{X := 0.0, Y := 0.0}
                    SizeToContent := false
                    Widget := Money_Icon_Pos_24
                    ZOrder := 923
                },
                canvas_slot {
                    Anchors := anchors{Minimum := vector2{X := 0.5, Y := 0.5}, Maximum := vector2{X := 0.5, Y := 0.5}}
                    Offsets := margin{Top := -11.365433, Left := 135.039062, Right := 38.808941, Bottom := 38.808941}
                    Alignment := vector2{X := 0.0, Y := 0.0}
                    SizeToContent := false
                    Widget := Money_Icon_Pos_25
                    ZOrder := 924
                },
                canvas_slot {
                    Anchors := anchors{Minimum := vector2{X := 0.5, Y := 0.5}, Maximum := vector2{X := 0.5, Y := 0.5}}
                    Offsets := margin{Top := -15.365433, Left := 135.039062, Right := 38.808941, Bottom := 38.808941}
                    Alignment := vector2{X := 0.0, Y := 0.0}
                    SizeToContent := false
                    Widget := Money_Icon_Pos_26
                    ZOrder := 925
                },
                canvas_slot {
                    Anchors := anchors{Minimum := vector2{X := 0.5, Y := 0.5}, Maximum := vector2{X := 0.5, Y := 0.5}}
                    Offsets := margin{Top := -19.365433, Left := 135.039062, Right := 38.808941, Bottom := 38.808941}
                    Alignment := vector2{X := 0.0, Y := 0.0}
                    SizeToContent := false
                    Widget := Money_Icon_Pos_27
                    ZOrder := 926
                },
                canvas_slot {
                    Anchors := anchors{Minimum := vector2{X := 0.5, Y := 0.5}, Maximum := vector2{X := 0.5, Y := 0.5}}
                    Offsets := margin{Top := -23.365433, Left := 135.039062, Right := 38.808941, Bottom := 38.808941}
                    Alignment := vector2{X := 0.0, Y := 0.0}
                    SizeToContent := false
                    Widget := Money_Icon_Pos_28
                    ZOrder := 927
                },
                canvas_slot {
                    Anchors := anchors{Minimum := vector2{X := 0.5, Y := 0.5}, Maximum := vector2{X := 0.5, Y := 0.5}}
                    Offsets := margin{Top := -27.365433, Left := 135.039062, Right := 38.808941, Bottom := 38.808941}
                    Alignment := vector2{X := 0.0, Y := 0.0}
                    SizeToContent := false
                    Widget := Money_Icon_Pos_29
                    ZOrder := 928
                },
                canvas_slot {
                    Anchors := anchors{Minimum := vector2{X := 0.5, Y := 0.5}, Maximum := vector2{X := 0.5, Y := 0.5}}
                    Offsets := margin{Top := -31.365433, Left := 135.039062, Right := 38.808941, Bottom := 38.808941}
                    Alignment := vector2{X := 0.0, Y := 0.0}
                    SizeToContent := false
                    Widget := Money_Icon_Pos_30
                    ZOrder := 929
                },
                canvas_slot {
                    Anchors := anchors{Minimum := vector2{X := 0.5, Y := 0.5}, Maximum := vector2{X := 0.5, Y := 0.5}}
                    Offsets := margin{Top := -35.365433, Left := 135.039062, Right := 38.808941, Bottom := 38.808941}
                    Alignment := vector2{X := 0.0, Y := 0.0}
                    SizeToContent := false
                    Widget := Money_Icon_Pos_31
                    ZOrder := 930
                },
                canvas_slot {
                    Anchors := anchors{Minimum := vector2{X := 0.5, Y := 0.5}, Maximum := vector2{X := 0.5, Y := 0.5}}
                    Offsets := margin{Top := -39.365433, Left := 135.039062, Right := 38.808941, Bottom := 38.808941}
                    Alignment := vector2{X := 0.0, Y := 0.0}
                    SizeToContent := false
                    Widget := Money_Icon_Pos_32
                    ZOrder := 931
                },
                canvas_slot {
                    Anchors := anchors{Minimum := vector2{X := 0.5, Y := 0.5}, Maximum := vector2{X := 0.5, Y := 0.5}}
                    Offsets := margin{Top := -43.365433, Left := 135.039062, Right := 38.808941, Bottom := 38.808941}
                    Alignment := vector2{X := 0.0, Y := 0.0}
                    SizeToContent := false
                    Widget := Money_Icon_Pos_33
                    ZOrder := 932
                },
                canvas_slot {
                    Anchors := anchors{Minimum := vector2{X := 0.5, Y := 0.5}, Maximum := vector2{X := 0.5, Y := 0.5}}
                    Offsets := margin{Top := -47.365433, Left := 135.039062, Right := 38.808941, Bottom := 38.808941}
                    Alignment := vector2{X := 0.0, Y := 0.0}
                    SizeToContent := false
                    Widget := Money_Icon_Pos_34
                    ZOrder := 933
                },
                canvas_slot {
                    Anchors := anchors{Minimum := vector2{X := 0.5, Y := 0.5}, Maximum := vector2{X := 0.5, Y := 0.5}}
                    Offsets := margin{Top := -51.365433, Left := 135.039062, Right := 38.808941, Bottom := 38.808941}
                    Alignment := vector2{X := 0.0, Y := 0.0}
                    SizeToContent := false
                    Widget := Money_Icon_Pos_35
                    ZOrder := 934
                },
                canvas_slot {
                    Anchors := anchors{Minimum := vector2{X := 0.5, Y := 0.5}, Maximum := vector2{X := 0.5, Y := 0.5}}
                    Offsets := margin{Top := -55.365433, Left := 135.039062, Right := 38.808941, Bottom := 38.808941}
                    Alignment := vector2{X := 0.0, Y := 0.0}
                    SizeToContent := false
                    Widget := Money_Icon_Pos_36
                    ZOrder := 935
                },
                canvas_slot {
                    Anchors := anchors{Minimum := vector2{X := 0.5, Y := 0.5}, Maximum := vector2{X := 0.5, Y := 0.5}}
                    Offsets := margin{Top := -59.365433, Left := 135.039062, Right := 38.808941, Bottom := 38.808941}
                    Alignment := vector2{X := 0.0, Y := 0.0}
                    SizeToContent := false
                    Widget := Money_Icon_Pos_37
                    ZOrder := 936
                },
                canvas_slot {
                    Anchors := anchors{Minimum := vector2{X := 0.5, Y := 0.5}, Maximum := vector2{X := 0.5, Y := 0.5}}
                    Offsets := margin{Top := -63.365433, Left := 135.039062, Right := 38.808941, Bottom := 38.808941}
                    Alignment := vector2{X := 0.0, Y := 0.0}
                    SizeToContent := false
                    Widget := Money_Icon_Pos_38
                    ZOrder := 937
                },
                canvas_slot {
                    Anchors := anchors{Minimum := vector2{X := 0.5, Y := 0.5}, Maximum := vector2{X := 0.5, Y := 0.5}}
                    Offsets := margin{Top := -67.365433, Left := 135.039062, Right := 38.808941, Bottom := 38.808941}
                    Alignment := vector2{X := 0.0, Y := 0.0}
                    SizeToContent := false
                    Widget := Money_Icon_Pos_39
                    ZOrder := 938
                },
                canvas_slot {
                    Anchors := anchors{Minimum := vector2{X := 0.5, Y := 0.5}, Maximum := vector2{X := 0.5, Y := 0.5}}
                    Offsets := margin{Top := -71.365433, Left := 135.039062, Right := 38.808941, Bottom := 38.808941}
                    Alignment := vector2{X := 0.0, Y := 0.0}
                    SizeToContent := false
                    Widget := Money_Icon_Pos_40
                    ZOrder := 939
                },
                canvas_slot {
                    Anchors := anchors{Minimum := vector2{X := 0.5, Y := 0.5}, Maximum := vector2{X := 0.5, Y := 0.5}}
                    Offsets := margin{Top := -75.365433, Left := 135.039062, Right := 38.808941, Bottom := 38.808941}
                    Alignment := vector2{X := 0.0, Y := 0.0}
                    SizeToContent := false
                    Widget := Money_Icon_Pos_41
                    ZOrder := 940
                },
                canvas_slot {
                    Anchors := anchors{Minimum := vector2{X := 0.5, Y := 0.5}, Maximum := vector2{X := 0.5, Y := 0.5}}
                    Offsets := margin{Top := -79.365433, Left := 135.039062, Right := 38.808941, Bottom := 38.808941}
                    Alignment := vector2{X := 0.0, Y := 0.0}
                    SizeToContent := false
                    Widget := Money_Icon_Pos_42
                    ZOrder := 941
                },
                canvas_slot {
                    Anchors := anchors{Minimum := vector2{X := 0.5, Y := 0.5}, Maximum := vector2{X := 0.5, Y := 0.5}}
                    Offsets := margin{Top := -83.365433, Left := 135.039062, Right := 38.808941, Bottom := 38.808941}
                    Alignment := vector2{X := 0.0, Y := 0.0}
                    SizeToContent := false
                    Widget := Money_Icon_Pos_43
                    ZOrder := 942
                },
                canvas_slot {
                    Anchors := anchors{Minimum := vector2{X := 0.5, Y := 0.5}, Maximum := vector2{X := 0.5, Y := 0.5}}
                    Offsets := margin{Top := -87.365433, Left := 135.039062, Right := 38.808941, Bottom := 38.808941}
                    Alignment := vector2{X := 0.0, Y := 0.0}
                    SizeToContent := false
                    Widget := Money_Icon_Pos_44
                    ZOrder := 943
                },
                canvas_slot {
                    Anchors := anchors{Minimum := vector2{X := 0.5, Y := 0.5}, Maximum := vector2{X := 0.5, Y := 0.5}}
                    Offsets := margin{Top := -91.365433, Left := 135.039062, Right := 38.808941, Bottom := 38.808941}
                    Alignment := vector2{X := 0.0, Y := 0.0}
                    SizeToContent := false
                    Widget := Money_Icon_Pos_45
                    ZOrder := 944
                },
                canvas_slot {
                    Anchors := anchors{Minimum := vector2{X := 0.5, Y := 0.5}, Maximum := vector2{X := 0.5, Y := 0.5}}
                    Offsets := margin{Top := -95.365433, Left := 135.039062, Right := 38.808941, Bottom := 38.808941}
                    Alignment := vector2{X := 0.0, Y := 0.0}
                    SizeToContent := false
                    Widget := Money_Icon_Pos_46
                    ZOrder := 945
                },
                canvas_slot {
                    Anchors := anchors{Minimum := vector2{X := 0.5, Y := 0.5}, Maximum := vector2{X := 0.5, Y := 0.5}}
                    Offsets := margin{Top := -99.365433, Left := 135.039062, Right := 38.808941, Bottom := 38.808941}
                    Alignment := vector2{X := 0.0, Y := 0.0}
                    SizeToContent := false
                    Widget := Money_Icon_Pos_47
                    ZOrder := 946
                },
                canvas_slot {
                    Anchors := anchors{Minimum := vector2{X := 0.5, Y := 0.5}, Maximum := vector2{X := 0.5, Y := 0.5}}
                    Offsets := margin{Top := -103.365433, Left := 135.039062, Right := 38.808941, Bottom := 38.808941}
                    Alignment := vector2{X := 0.0, Y := 0.0}
                    SizeToContent := false
                    Widget := Money_Icon_Pos_48
                    ZOrder := 947
                },
                canvas_slot {
                    Anchors := anchors{Minimum := vector2{X := 0.5, Y := 0.5}, Maximum := vector2{X := 0.5, Y := 0.5}}
                    Offsets := margin{Top := -107.365433, Left := 135.039062, Right := 38.808941, Bottom := 38.808941}
                    Alignment := vector2{X := 0.0, Y := 0.0}
                    SizeToContent := false
                    Widget := Money_Icon_Pos_49
                    ZOrder := 948
                },
                canvas_slot {
                    Anchors := anchors{Minimum := vector2{X := 0.5, Y := 0.5}, Maximum := vector2{X := 0.5, Y := 0.5}}
                    Offsets := margin{Top := -111.365433, Left := 135.039062, Right := 38.808941, Bottom := 38.808941}
                    Alignment := vector2{X := 0.0, Y := 0.0}
                    SizeToContent := false
                    Widget := Money_Icon_Pos_50
                    ZOrder := 949
                },
                canvas_slot {
                    Anchors := anchors{Minimum := vector2{X := 0.5, Y := 0.5}, Maximum := vector2{X := 0.5, Y := 0.5}}
                    Offsets := margin{Top := -115.365433, Left := 135.039062, Right := 38.808941, Bottom := 38.808941}
                    Alignment := vector2{X := 0.0, Y := 0.0}
                    SizeToContent := false
                    Widget := Money_Icon_Pos_51
                    ZOrder := 950
                },
                canvas_slot {
                    Anchors := anchors{Minimum := vector2{X := 0.5, Y := 0.5}, Maximum := vector2{X := 0.5, Y := 0.5}}
                    Offsets := margin{Top := -119.365433, Left := 135.039062, Right := 38.808941, Bottom := 38.808941}
                    Alignment := vector2{X := 0.0, Y := 0.0}
                    SizeToContent := false
                    Widget := Money_Icon_Pos_52
                    ZOrder := 951
                },
                canvas_slot {
                    Anchors := anchors{Minimum := vector2{X := 0.5, Y := 0.5}, Maximum := vector2{X := 0.5, Y := 0.5}}
                    Offsets := margin{Top := -123.365433, Left := 135.039062, Right := 38.808941, Bottom := 38.808941}
                    Alignment := vector2{X := 0.0, Y := 0.0}
                    SizeToContent := false
                    Widget := Money_Icon_Pos_53
                    ZOrder := 952
                },
                canvas_slot {
                    Anchors := anchors{Minimum := vector2{X := 0.5, Y := 0.5}, Maximum := vector2{X := 0.5, Y := 0.5}}
                    Offsets := margin{Top := -127.365433, Left := 135.039062, Right := 38.808941, Bottom := 38.808941}
                    Alignment := vector2{X := 0.0, Y := 0.0}
                    SizeToContent := false
                    Widget := Money_Icon_Pos_54
                    ZOrder := 953
                },
                canvas_slot {
                    Anchors := anchors{Minimum := vector2{X := 0.5, Y := 0.5}, Maximum := vector2{X := 0.5, Y := 0.5}}
                    Offsets := margin{Top := -131.365433, Left := 135.039062, Right := 38.808941, Bottom := 38.808941}
                    Alignment := vector2{X := 0.0, Y := 0.0}
                    SizeToContent := false
                    Widget := Money_Icon_Pos_55
                    ZOrder := 954
                },
                canvas_slot {
                    Anchors := anchors{Minimum := vector2{X := 0.5, Y := 0.5}, Maximum := vector2{X := 0.5, Y := 0.5}}
                    Offsets := margin{Top := -135.365433, Left := 135.039062, Right := 38.808941, Bottom := 38.808941}
                    Alignment := vector2{X := 0.0, Y := 0.0}
                    SizeToContent := false
                    Widget := Money_Icon_Pos_56
                    ZOrder := 955
                },
                canvas_slot {
                    Anchors := anchors{Minimum := vector2{X := 0.5, Y := 0.5}, Maximum := vector2{X := 0.5, Y := 0.5}}
                    Offsets := margin{Top := -139.365433, Left := 135.039062, Right := 38.808941, Bottom := 38.808941}
                    Alignment := vector2{X := 0.0, Y := 0.0}
                    SizeToContent := false
                    Widget := Money_Icon_Pos_57
                    ZOrder := 956
                },
                canvas_slot {
                    Anchors := anchors{Minimum := vector2{X := 0.5, Y := 0.5}, Maximum := vector2{X := 0.5, Y := 0.5}}
                    Offsets := margin{Top := -143.365433, Left := 135.039062, Right := 38.808941, Bottom := 38.808941}
                    Alignment := vector2{X := 0.0, Y := 0.0}
                    SizeToContent := false
                    Widget := Money_Icon_Pos_58
                    ZOrder := 957
                },
                canvas_slot {
                    Anchors := anchors{Minimum := vector2{X := 0.5, Y := 0.5}, Maximum := vector2{X := 0.5, Y := 0.5}}
                    Offsets := margin{Top := -147.365433, Left := 135.039062, Right := 38.808941, Bottom := 38.808941}
                    Alignment := vector2{X := 0.0, Y := 0.0}
                    SizeToContent := false
                    Widget := Money_Icon_Pos_59
                    ZOrder := 958
                },
                canvas_slot {
                    Anchors := anchors{Minimum := vector2{X := 0.5, Y := 0.5}, Maximum := vector2{X := 0.5, Y := 0.5}}
                    Offsets := margin{Top := -151.365433, Left := 135.039062, Right := 38.808941, Bottom := 38.808941}
                    Alignment := vector2{X := 0.0, Y := 0.0}
                    SizeToContent := false
                    Widget := Money_Icon_Pos_60
                    ZOrder := 959
                },

                # --------------MONEY POPUP TEXT --------------
                canvas_slot {
                    Anchors := anchors{Minimum := vector2{X := 0.5, Y := 0.5}, Maximum := vector2{X := 0.5, Y := 0.5}}
                    Offsets := margin{Top := 84.634567, Left := 195.039062, Right := 75.510513, Bottom := 41.0}
                    Alignment := vector2{X := 0.0, Y := 0.0}
                    SizeToContent := false
                    Widget := Pos_1
                    ZOrder := 901
                },
                canvas_slot {
                    Anchors := anchors{Minimum := vector2{X := 0.5, Y := 0.5}, Maximum := vector2{X := 0.5, Y := 0.5}}
                    Offsets := margin{Top := 80.634567, Left := 195.039062, Right := 75.510513, Bottom := 41.0}
                    Alignment := vector2{X := 0.0, Y := 0.0}
                    SizeToContent := false
                    Widget := Pos_2
                    ZOrder := 902
                },
                canvas_slot {
                    Anchors := anchors{Minimum := vector2{X := 0.5, Y := 0.5}, Maximum := vector2{X := 0.5, Y := 0.5}}
                    Offsets := margin{Top := 76.634567, Left := 195.039062, Right := 75.510513, Bottom := 41.0}
                    Alignment := vector2{X := 0.0, Y := 0.0}
                    SizeToContent := false
                    Widget := Pos_3
                    ZOrder := 903
                },
                canvas_slot {
                    Anchors := anchors{Minimum := vector2{X := 0.5, Y := 0.5}, Maximum := vector2{X := 0.5, Y := 0.5}}
                    Offsets := margin{Top := 72.634567, Left := 195.039062, Right := 75.510513, Bottom := 41.0}
                    Alignment := vector2{X := 0.0, Y := 0.0}
                    SizeToContent := false
                    Widget := Pos_4
                    ZOrder := 904
                },
                canvas_slot {
                    Anchors := anchors{Minimum := vector2{X := 0.5, Y := 0.5}, Maximum := vector2{X := 0.5, Y := 0.5}}
                    Offsets := margin{Top := 68.634567, Left := 195.039062, Right := 75.510513, Bottom := 41.0}
                    Alignment := vector2{X := 0.0, Y := 0.0}
                    SizeToContent := false
                    Widget := Pos_5
                    ZOrder := 905
                },
                canvas_slot {
                    Anchors := anchors{Minimum := vector2{X := 0.5, Y := 0.5}, Maximum := vector2{X := 0.5, Y := 0.5}}
                    Offsets := margin{Top := 64.634567, Left := 195.039062, Right := 75.510513, Bottom := 41.0}
                    Alignment := vector2{X := 0.0, Y := 0.0}
                    SizeToContent := false
                    Widget := Pos_6
                    ZOrder := 906
                },
                canvas_slot {
                    Anchors := anchors{Minimum := vector2{X := 0.5, Y := 0.5}, Maximum := vector2{X := 0.5, Y := 0.5}}
                    Offsets := margin{Top := 60.634567, Left := 195.039062, Right := 75.510513, Bottom := 41.0}
                    Alignment := vector2{X := 0.0, Y := 0.0}
                    SizeToContent := false
                    Widget := Pos_7
                    ZOrder := 907
                },
                canvas_slot {
                    Anchors := anchors{Minimum := vector2{X := 0.5, Y := 0.5}, Maximum := vector2{X := 0.5, Y := 0.5}}
                    Offsets := margin{Top := 56.634567, Left := 195.039062, Right := 75.510513, Bottom := 41.0}
                    Alignment := vector2{X := 0.0, Y := 0.0}
                    SizeToContent := false
                    Widget := Pos_8
                    ZOrder := 908
                },
                canvas_slot {
                    Anchors := anchors{Minimum := vector2{X := 0.5, Y := 0.5}, Maximum := vector2{X := 0.5, Y := 0.5}}
                    Offsets := margin{Top := 52.634567, Left := 195.039062, Right := 75.510513, Bottom := 41.0}
                    Alignment := vector2{X := 0.0, Y := 0.0}
                    SizeToContent := false
                    Widget := Pos_9
                    ZOrder := 909
                },
                canvas_slot {
                    Anchors := anchors{Minimum := vector2{X := 0.5, Y := 0.5}, Maximum := vector2{X := 0.5, Y := 0.5}}
                    Offsets := margin{Top := 48.634567, Left := 195.039062, Right := 75.510513, Bottom := 41.0}
                    Alignment := vector2{X := 0.0, Y := 0.0}
                    SizeToContent := false
                    Widget := Pos_10
                    ZOrder := 910
                },
                canvas_slot {
                    Anchors := anchors{Minimum := vector2{X := 0.5, Y := 0.5}, Maximum := vector2{X := 0.5, Y := 0.5}}
                    Offsets := margin{Top := 44.634567, Left := 195.039062, Right := 75.510513, Bottom := 41.0}
                    Alignment := vector2{X := 0.0, Y := 0.0}
                    SizeToContent := false
                    Widget := Pos_11
                    ZOrder := 911
                },
                canvas_slot {
                    Anchors := anchors{Minimum := vector2{X := 0.5, Y := 0.5}, Maximum := vector2{X := 0.5, Y := 0.5}}
                    Offsets := margin{Top := 40.634567, Left := 195.039062, Right := 75.510513, Bottom := 41.0}
                    Alignment := vector2{X := 0.0, Y := 0.0}
                    SizeToContent := false
                    Widget := Pos_12
                    ZOrder := 912
                },
                canvas_slot {
                    Anchors := anchors{Minimum := vector2{X := 0.5, Y := 0.5}, Maximum := vector2{X := 0.5, Y := 0.5}}
                    Offsets := margin{Top := 36.634567, Left := 195.039062, Right := 75.510513, Bottom := 41.0}
                    Alignment := vector2{X := 0.0, Y := 0.0}
                    SizeToContent := false
                    Widget := Pos_13
                    ZOrder := 913
                },
                canvas_slot {
                    Anchors := anchors{Minimum := vector2{X := 0.5, Y := 0.5}, Maximum := vector2{X := 0.5, Y := 0.5}}
                    Offsets := margin{Top := 32.634567, Left := 195.039062, Right := 75.510513, Bottom := 41.0}
                    Alignment := vector2{X := 0.0, Y := 0.0}
                    SizeToContent := false
                    Widget := Pos_14
                    ZOrder := 914
                },
                canvas_slot {
                    Anchors := anchors{Minimum := vector2{X := 0.5, Y := 0.5}, Maximum := vector2{X := 0.5, Y := 0.5}}
                    Offsets := margin{Top := 28.634567, Left := 195.039062, Right := 75.510513, Bottom := 41.0}
                    Alignment := vector2{X := 0.0, Y := 0.0}
                    SizeToContent := false
                    Widget := Pos_15
                    ZOrder := 915
                },
                canvas_slot {
                    Anchors := anchors{Minimum := vector2{X := 0.5, Y := 0.5}, Maximum := vector2{X := 0.5, Y := 0.5}}
                    Offsets := margin{Top := 24.634567, Left := 195.039062, Right := 75.510513, Bottom := 41.0}
                    Alignment := vector2{X := 0.0, Y := 0.0}
                    SizeToContent := false
                    Widget := Pos_16
                    ZOrder := 916
                },
                canvas_slot {
                    Anchors := anchors{Minimum := vector2{X := 0.5, Y := 0.5}, Maximum := vector2{X := 0.5, Y := 0.5}}
                    Offsets := margin{Top := 20.634567, Left := 195.039062, Right := 75.510513, Bottom := 41.0}
                    Alignment := vector2{X := 0.0, Y := 0.0}
                    SizeToContent := false
                    Widget := Pos_17
                    ZOrder := 917
                },
                canvas_slot {
                    Anchors := anchors{Minimum := vector2{X := 0.5, Y := 0.5}, Maximum := vector2{X := 0.5, Y := 0.5}}
                    Offsets := margin{Top := 16.634567, Left := 195.039062, Right := 75.510513, Bottom := 41.0}
                    Alignment := vector2{X := 0.0, Y := 0.0}
                    SizeToContent := false
                    Widget := Pos_18
                    ZOrder := 918
                },
                canvas_slot {
                    Anchors := anchors{Minimum := vector2{X := 0.5, Y := 0.5}, Maximum := vector2{X := 0.5, Y := 0.5}}
                    Offsets := margin{Top := 12.634567, Left := 195.039062, Right := 75.510513, Bottom := 41.0}
                    Alignment := vector2{X := 0.0, Y := 0.0}
                    SizeToContent := false
                    Widget := Pos_19
                    ZOrder := 919
                },
                canvas_slot {
                    Anchors := anchors{Minimum := vector2{X := 0.5, Y := 0.5}, Maximum := vector2{X := 0.5, Y := 0.5}}
                    Offsets := margin{Top := 8.634567, Left := 195.039062, Right := 75.510513, Bottom := 41.0}
                    Alignment := vector2{X := 0.0, Y := 0.0}
                    SizeToContent := false
                    Widget := Pos_20
                    ZOrder := 920
                },
                canvas_slot {
                    Anchors := anchors{Minimum := vector2{X := 0.5, Y := 0.5}, Maximum := vector2{X := 0.5, Y := 0.5}}
                    Offsets := margin{Top := 4.634567, Left := 195.039062, Right := 75.510513, Bottom := 41.0}
                    Alignment := vector2{X := 0.0, Y := 0.0}
                    SizeToContent := false
                    Widget := Pos_21
                    ZOrder := 921
                },
                canvas_slot {
                    Anchors := anchors{Minimum := vector2{X := 0.5, Y := 0.5}, Maximum := vector2{X := 0.5, Y := 0.5}}
                    Offsets := margin{Top := 0.634567, Left := 195.039062, Right := 75.510513, Bottom := 41.0}
                    Alignment := vector2{X := 0.0, Y := 0.0}
                    SizeToContent := false
                    Widget := Pos_22
                    ZOrder := 922
                },
                canvas_slot {
                    Anchors := anchors{Minimum := vector2{X := 0.5, Y := 0.5}, Maximum := vector2{X := 0.5, Y := 0.5}}
                    Offsets := margin{Top := -3.365433, Left := 195.039062, Right := 75.510513, Bottom := 41.0}
                    Alignment := vector2{X := 0.0, Y := 0.0}
                    SizeToContent := false
                    Widget := Pos_23
                    ZOrder := 923
                },
                canvas_slot {
                    Anchors := anchors{Minimum := vector2{X := 0.5, Y := 0.5}, Maximum := vector2{X := 0.5, Y := 0.5}}
                    Offsets := margin{Top := -7.365433, Left := 195.039062, Right := 75.510513, Bottom := 41.0}
                    Alignment := vector2{X := 0.0, Y := 0.0}
                    SizeToContent := false
                    Widget := Pos_24
                    ZOrder := 924
                },
                canvas_slot {
                    Anchors := anchors{Minimum := vector2{X := 0.5, Y := 0.5}, Maximum := vector2{X := 0.5, Y := 0.5}}
                    Offsets := margin{Top := -11.365433, Left := 195.039062, Right := 75.510513, Bottom := 41.0}
                    Alignment := vector2{X := 0.0, Y := 0.0}
                    SizeToContent := false
                    Widget := Pos_25
                    ZOrder := 925
                },
                canvas_slot {
                    Anchors := anchors{Minimum := vector2{X := 0.5, Y := 0.5}, Maximum := vector2{X := 0.5, Y := 0.5}}
                    Offsets := margin{Top := -15.365433, Left := 195.039062, Right := 75.510513, Bottom := 41.0}
                    Alignment := vector2{X := 0.0, Y := 0.0}
                    SizeToContent := false
                    Widget := Pos_26
                    ZOrder := 926
                },
                canvas_slot {
                    Anchors := anchors{Minimum := vector2{X := 0.5, Y := 0.5}, Maximum := vector2{X := 0.5, Y := 0.5}}
                    Offsets := margin{Top := -19.365433, Left := 195.039062, Right := 75.510513, Bottom := 41.0}
                    Alignment := vector2{X := 0.0, Y := 0.0}
                    SizeToContent := false
                    Widget := Pos_27
                    ZOrder := 927
                },
                canvas_slot {
                    Anchors := anchors{Minimum := vector2{X := 0.5, Y := 0.5}, Maximum := vector2{X := 0.5, Y := 0.5}}
                    Offsets := margin{Top := -23.365433, Left := 195.039062, Right := 75.510513, Bottom := 41.0}
                    Alignment := vector2{X := 0.0, Y := 0.0}
                    SizeToContent := false
                    Widget := Pos_28
                    ZOrder := 928
                },
                canvas_slot {
                    Anchors := anchors{Minimum := vector2{X := 0.5, Y := 0.5}, Maximum := vector2{X := 0.5, Y := 0.5}}
                    Offsets := margin{Top := -27.365433, Left := 195.039062, Right := 75.510513, Bottom := 41.0}
                    Alignment := vector2{X := 0.0, Y := 0.0}
                    SizeToContent := false
                    Widget := Pos_29
                    ZOrder := 929
                },
                canvas_slot {
                    Anchors := anchors{Minimum := vector2{X := 0.5, Y := 0.5}, Maximum := vector2{X := 0.5, Y := 0.5}}
                    Offsets := margin{Top := -31.365433, Left := 195.039062, Right := 75.510513, Bottom := 41.0}
                    Alignment := vector2{X := 0.0, Y := 0.0}
                    SizeToContent := false
                    Widget := Pos_30
                    ZOrder := 930
                },
                canvas_slot {
                    Anchors := anchors{Minimum := vector2{X := 0.5, Y := 0.5}, Maximum := vector2{X := 0.5, Y := 0.5}}
                    Offsets := margin{Top := -35.365433, Left := 195.039062, Right := 75.510513, Bottom := 41.0}
                    Alignment := vector2{X := 0.0, Y := 0.0}
                    SizeToContent := false
                    Widget := Pos_31
                    ZOrder := 931
                },
                canvas_slot {
                    Anchors := anchors{Minimum := vector2{X := 0.5, Y := 0.5}, Maximum := vector2{X := 0.5, Y := 0.5}}
                    Offsets := margin{Top := -39.365433, Left := 195.039062, Right := 75.510513, Bottom := 41.0}
                    Alignment := vector2{X := 0.0, Y := 0.0}
                    SizeToContent := false
                    Widget := Pos_32
                    ZOrder := 932
                },
                canvas_slot {
                    Anchors := anchors{Minimum := vector2{X := 0.5, Y := 0.5}, Maximum := vector2{X := 0.5, Y := 0.5}}
                    Offsets := margin{Top := -43.365433, Left := 195.039062, Right := 75.510513, Bottom := 41.0}
                    Alignment := vector2{X := 0.0, Y := 0.0}
                    SizeToContent := false
                    Widget := Pos_33
                    ZOrder := 933
                },
                canvas_slot {
                    Anchors := anchors{Minimum := vector2{X := 0.5, Y := 0.5}, Maximum := vector2{X := 0.5, Y := 0.5}}
                    Offsets := margin{Top := -47.365433, Left := 195.039062, Right := 75.510513, Bottom := 41.0}
                    Alignment := vector2{X := 0.0, Y := 0.0}
                    SizeToContent := false
                    Widget := Pos_34
                    ZOrder := 934
                },
                canvas_slot {
                    Anchors := anchors{Minimum := vector2{X := 0.5, Y := 0.5}, Maximum := vector2{X := 0.5, Y := 0.5}}
                    Offsets := margin{Top := -51.365433, Left := 195.039062, Right := 75.510513, Bottom := 41.0}
                    Alignment := vector2{X := 0.0, Y := 0.0}
                    SizeToContent := false
                    Widget := Pos_35
                    ZOrder := 935
                },
                canvas_slot {
                    Anchors := anchors{Minimum := vector2{X := 0.5, Y := 0.5}, Maximum := vector2{X := 0.5, Y := 0.5}}
                    Offsets := margin{Top := -55.365433, Left := 195.039062, Right := 75.510513, Bottom := 41.0}
                    Alignment := vector2{X := 0.0, Y := 0.0}
                    SizeToContent := false
                    Widget := Pos_36
                    ZOrder := 936
                },
                canvas_slot {
                    Anchors := anchors{Minimum := vector2{X := 0.5, Y := 0.5}, Maximum := vector2{X := 0.5, Y := 0.5}}
                    Offsets := margin{Top := -59.365433, Left := 195.039062, Right := 75.510513, Bottom := 41.0}
                    Alignment := vector2{X := 0.0, Y := 0.0}
                    SizeToContent := false
                    Widget := Pos_37
                    ZOrder := 937
                },
                canvas_slot {
                    Anchors := anchors{Minimum := vector2{X := 0.5, Y := 0.5}, Maximum := vector2{X := 0.5, Y := 0.5}}
                    Offsets := margin{Top := -63.365433, Left := 195.039062, Right := 75.510513, Bottom := 41.0}
                    Alignment := vector2{X := 0.0, Y := 0.0}
                    SizeToContent := false
                    Widget := Pos_38
                    ZOrder := 938
                },
                canvas_slot {
                    Anchors := anchors{Minimum := vector2{X := 0.5, Y := 0.5}, Maximum := vector2{X := 0.5, Y := 0.5}}
                    Offsets := margin{Top := -67.365433, Left := 195.039062, Right := 75.510513, Bottom := 41.0}
                    Alignment := vector2{X := 0.0, Y := 0.0}
                    SizeToContent := false
                    Widget := Pos_39
                    ZOrder := 939
                },
                canvas_slot {
                    Anchors := anchors{Minimum := vector2{X := 0.5, Y := 0.5}, Maximum := vector2{X := 0.5, Y := 0.5}}
                    Offsets := margin{Top := -71.365433, Left := 195.039062, Right := 75.510513, Bottom := 41.0}
                    Alignment := vector2{X := 0.0, Y := 0.0}
                    SizeToContent := false
                    Widget := Pos_40
                    ZOrder := 940
                },
                canvas_slot {
                    Anchors := anchors{Minimum := vector2{X := 0.5, Y := 0.5}, Maximum := vector2{X := 0.5, Y := 0.5}}
                    Offsets := margin{Top := -75.365433, Left := 195.039062, Right := 75.510513, Bottom := 41.0}
                    Alignment := vector2{X := 0.0, Y := 0.0}
                    SizeToContent := false
                    Widget := Pos_41
                    ZOrder := 941
                },
                canvas_slot {
                    Anchors := anchors{Minimum := vector2{X := 0.5, Y := 0.5}, Maximum := vector2{X := 0.5, Y := 0.5}}
                    Offsets := margin{Top := -79.365433, Left := 195.039062, Right := 75.510513, Bottom := 41.0}
                    Alignment := vector2{X := 0.0, Y := 0.0}
                    SizeToContent := false
                    Widget := Pos_42
                    ZOrder := 942
                },
                canvas_slot {
                    Anchors := anchors{Minimum := vector2{X := 0.5, Y := 0.5}, Maximum := vector2{X := 0.5, Y := 0.5}}
                    Offsets := margin{Top := -83.365433, Left := 195.039062, Right := 75.510513, Bottom := 41.0}
                    Alignment := vector2{X := 0.0, Y := 0.0}
                    SizeToContent := false
                    Widget := Pos_43
                    ZOrder := 943
                },
                canvas_slot {
                    Anchors := anchors{Minimum := vector2{X := 0.5, Y := 0.5}, Maximum := vector2{X := 0.5, Y := 0.5}}
                    Offsets := margin{Top := -87.365433, Left := 195.039062, Right := 75.510513, Bottom := 41.0}
                    Alignment := vector2{X := 0.0, Y := 0.0}
                    SizeToContent := false
                    Widget := Pos_44
                    ZOrder := 944
                },
                canvas_slot {
                    Anchors := anchors{Minimum := vector2{X := 0.5, Y := 0.5}, Maximum := vector2{X := 0.5, Y := 0.5}}
                    Offsets := margin{Top := -91.365433, Left := 195.039062, Right := 75.510513, Bottom := 41.0}
                    Alignment := vector2{X := 0.0, Y := 0.0}
                    SizeToContent := false
                    Widget := Pos_45
                    ZOrder := 945
                },
                canvas_slot {
                    Anchors := anchors{Minimum := vector2{X := 0.5, Y := 0.5}, Maximum := vector2{X := 0.5, Y := 0.5}}
                    Offsets := margin{Top := -95.365433, Left := 195.039062, Right := 75.510513, Bottom := 41.0}
                    Alignment := vector2{X := 0.0, Y := 0.0}
                    SizeToContent := false
                    Widget := Pos_46
                    ZOrder := 946
                },
                canvas_slot {
                    Anchors := anchors{Minimum := vector2{X := 0.5, Y := 0.5}, Maximum := vector2{X := 0.5, Y := 0.5}}
                    Offsets := margin{Top := -99.365433, Left := 195.039062, Right := 75.510513, Bottom := 41.0}
                    Alignment := vector2{X := 0.0, Y := 0.0}
                    SizeToContent := false
                    Widget := Pos_47
                    ZOrder := 947
                },
                canvas_slot {
                    Anchors := anchors{Minimum := vector2{X := 0.5, Y := 0.5}, Maximum := vector2{X := 0.5, Y := 0.5}}
                    Offsets := margin{Top := -103.365433, Left := 195.039062, Right := 75.510513, Bottom := 41.0}
                    Alignment := vector2{X := 0.0, Y := 0.0}
                    SizeToContent := false
                    Widget := Pos_48
                    ZOrder := 948
                },
                canvas_slot {
                    Anchors := anchors{Minimum := vector2{X := 0.5, Y := 0.5}, Maximum := vector2{X := 0.5, Y := 0.5}}
                    Offsets := margin{Top := -107.365433, Left := 195.039062, Right := 75.510513, Bottom := 41.0}
                    Alignment := vector2{X := 0.0, Y := 0.0}
                    SizeToContent := false
                    Widget := Pos_49
                    ZOrder := 949
                },
                canvas_slot {
                    Anchors := anchors{Minimum := vector2{X := 0.5, Y := 0.5}, Maximum := vector2{X := 0.5, Y := 0.5}}
                    Offsets := margin{Top := -111.365433, Left := 195.039062, Right := 75.510513, Bottom := 41.0}
                    Alignment := vector2{X := 0.0, Y := 0.0}
                    SizeToContent := false
                    Widget := Pos_50
                    ZOrder := 950
                },
                canvas_slot {
                    Anchors := anchors{Minimum := vector2{X := 0.5, Y := 0.5}, Maximum := vector2{X := 0.5, Y := 0.5}}
                    Offsets := margin{Top := -115.365433, Left := 195.039062, Right := 75.510513, Bottom := 41.0}
                    Alignment := vector2{X := 0.0, Y := 0.0}
                    SizeToContent := false
                    Widget := Pos_51
                    ZOrder := 951
                },
                canvas_slot {
                    Anchors := anchors{Minimum := vector2{X := 0.5, Y := 0.5}, Maximum := vector2{X := 0.5, Y := 0.5}}
                    Offsets := margin{Top := -119.365433, Left := 195.039062, Right := 75.510513, Bottom := 41.0}
                    Alignment := vector2{X := 0.0, Y := 0.0}
                    SizeToContent := false
                    Widget := Pos_52
                    ZOrder := 952
                },
                canvas_slot {
                    Anchors := anchors{Minimum := vector2{X := 0.5, Y := 0.5}, Maximum := vector2{X := 0.5, Y := 0.5}}
                    Offsets := margin{Top := -123.365433, Left := 195.039062, Right := 75.510513, Bottom := 41.0}
                    Alignment := vector2{X := 0.0, Y := 0.0}
                    SizeToContent := false
                    Widget := Pos_53
                    ZOrder := 953
                },
                canvas_slot {
                    Anchors := anchors{Minimum := vector2{X := 0.5, Y := 0.5}, Maximum := vector2{X := 0.5, Y := 0.5}}
                    Offsets := margin{Top := -127.365433, Left := 195.039062, Right := 75.510513, Bottom := 41.0}
                    Alignment := vector2{X := 0.0, Y := 0.0}
                    SizeToContent := false
                    Widget := Pos_54
                    ZOrder := 954
                },
                canvas_slot {
                    Anchors := anchors{Minimum := vector2{X := 0.5, Y := 0.5}, Maximum := vector2{X := 0.5, Y := 0.5}}
                    Offsets := margin{Top := -131.365433, Left := 195.039062, Right := 75.510513, Bottom := 41.0}
                    Alignment := vector2{X := 0.0, Y := 0.0}
                    SizeToContent := false
                    Widget := Pos_55
                    ZOrder := 955
                },
                canvas_slot {
                    Anchors := anchors{Minimum := vector2{X := 0.5, Y := 0.5}, Maximum := vector2{X := 0.5, Y := 0.5}}
                    Offsets := margin{Top := -135.365433, Left := 195.039062, Right := 75.510513, Bottom := 41.0}
                    Alignment := vector2{X := 0.0, Y := 0.0}
                    SizeToContent := false
                    Widget := Pos_56
                    ZOrder := 956
                },
                canvas_slot {
                    Anchors := anchors{Minimum := vector2{X := 0.5, Y := 0.5}, Maximum := vector2{X := 0.5, Y := 0.5}}
                    Offsets := margin{Top := -139.365433, Left := 195.039062, Right := 75.510513, Bottom := 41.0}
                    Alignment := vector2{X := 0.0, Y := 0.0}
                    SizeToContent := false
                    Widget := Pos_57
                    ZOrder := 957
                },
                canvas_slot {
                    Anchors := anchors{Minimum := vector2{X := 0.5, Y := 0.5}, Maximum := vector2{X := 0.5, Y := 0.5}}
                    Offsets := margin{Top := -143.365433, Left := 195.039062, Right := 75.510513, Bottom := 41.0}
                    Alignment := vector2{X := 0.0, Y := 0.0}
                    SizeToContent := false
                    Widget := Pos_58
                    ZOrder := 958
                },
                canvas_slot {
                    Anchors := anchors{Minimum := vector2{X := 0.5, Y := 0.5}, Maximum := vector2{X := 0.5, Y := 0.5}}
                    Offsets := margin{Top := -147.365433, Left := 195.039062, Right := 75.510513, Bottom := 41.0}
                    Alignment := vector2{X := 0.0, Y := 0.0}
                    SizeToContent := false
                    Widget := Pos_59
                    ZOrder := 959
                },
                canvas_slot {
                    Anchors := anchors{Minimum := vector2{X := 0.5, Y := 0.5}, Maximum := vector2{X := 0.5, Y := 0.5}}
                    Offsets := margin{Top := -151.365433, Left := 195.039062, Right := 75.510513, Bottom := 41.0}
                    Alignment := vector2{X := 0.0, Y := 0.0}
                    SizeToContent := false
                    Widget := Pos_60
                    ZOrder := 960
                }
            }

            var XpBoostWidget: widget = XpBoost2xTexture
            <#
            if(PlayerData.XpBoostLevel = 2) {
                set XpBoostWidget = XpBoost3xTexture
            } else if(PlayerData.XpBoostLevel = 3) {
                set XpBoostWidget = XpBoost4xTexture
            } else if(PlayerData.XpBoostLevel = 4) {
                set XpBoostWidget = XpBoost5xTexture
            } else if(PlayerData.XpBoostLevel = 5) {
                set XpBoostWidget = XpBoost6xTexture
            }
                #>

            if(PlayerData.XpBoostLevel >= 1) {
                set Slots = Slots + array {
                    canvas_slot {
                        Anchors := anchors{Minimum := vector2{X := 0.5, Y := 1.0}, Maximum := vector2{X := 0.5, Y := 1.0}}
                        Offsets := margin{Top := -201.081055, Left := -320.960938, Right := 145.864044, Bottom := 137.188095}
                        Alignment := vector2{X := 0.0, Y := 0.0}
                        SizeToContent := false
                        Widget := XpBoostWidget
                        ZOrder := 1000
                    }
                }
            }
        }

        # Initially hide some slots
        TextureMaskLevelStar1.SetVisibility(widget_visibility.Hidden)
        TextureMaskLevelStar2.SetVisibility(widget_visibility.Hidden)
        TextureMaskLevelStar3.SetVisibility(widget_visibility.Hidden)
        TextureMaskLevelStar4.SetVisibility(widget_visibility.Hidden)
        TextureMaskLevelStar5.SetVisibility(widget_visibility.Hidden)
        TextureMaskLevelStar6.SetVisibility(widget_visibility.Hidden)

        for(FloatingText: FloatingTextList) {
            FloatingText.SetVisibility(widget_visibility.Hidden)
        }

        for(FloatingIcon: FloatingIconList) {
            FloatingIcon.SetVisibility(widget_visibility.Hidden)
        }

        set CanvasSlots = Slots
        
        return Slots
    }

    RefreshPlayerLocalUi<public>(Agent: agent): void = {
        Print("RefreshPlayerLocalUi")

        # TODO: We shouldn't Remove and Add Widget just to refresh
        # TODO: Handle all Widget actions in PlayerData
        <#
        if(Canvas := MaybeCanvas?) {
            for(OldSlot: CanvasSlots) {
                Canvas.RemoveWidget(OldSlot.Widget)
            }
            NewSlots := CreateSlots(Agent)
            for(NewSlot:NewSlots) {
                Canvas.AddWidget(NewSlot)
            }
        }
        #>
    }

    # Returns the new level
    UpdateXpUi<public>(ParmTotalXp: int): int = {
        # Update our data
        set TotalXp = ParmTotalXp
        
        set CurLevel = GetLevelFromXp()

        # Update UI
        UpdateLevelBarUi()

        return CurLevel
    }

    UpdateLevelBarUi(): void = {
        Print("UpdateLevelBarUi: TotalXp: {TotalXp}")
        Utility := GetUtility()
        GameManager := GetGameManager()
        if {
            Agent := MaybeAgent?
        } then {
            # Calculate values
            XpTowardsNextLevel := TotalXp - GetCumulativeXpForLevel(CurLevel - 1)
            XpRequiredForNextLevel := GetRequiredXpForLevel(CurLevel)
            # Progress until the next level. Values from 0 to 1
            var XpProgressPercentage: float = Min((XpTowardsNextLevel * 1.0) / (XpRequiredForNextLevel * 1.0), 1.0)

            # Update UI
            GameManager.HudMainUiCurrentLevel.Show(Agent, Utility.FormatMessage("LEVEL " + ToString(CurLevel + 1)))
            GameManager.HudMainUiXpTowardsNextLevel.Show(Agent, Utility.FormatMessage(ToString(XpTowardsNextLevel) + " / " + ToString(XpRequiredForNextLevel) + " XP"))
            NumOfHuds := LevelBarColorBlockList.Length
            PercPerHud := 1.0 / (NumOfHuds * 1.0)
            for(BarIndex -> LevelBarColorBlock: LevelBarColorBlockList) {
                var BarPercentage: float = 0.0
                if(XpProgressPercentage >= PercPerHud) {
                    set BarPercentage = 1.0
                } else {
                    if(XpProgressPercentage > 0.0) {
                        set BarPercentage = XpProgressPercentage / PercPerHud
                    }
                }

                LevelBarCurrentWidth := BarPercentage * LevelBarWidth
                LevelBarColorBlock.SetDesiredSize(vector2{X := LevelBarCurrentWidth, Y := LevelBarHeight})

                set XpProgressPercentage = XpProgressPercentage - PercPerHud
            }
        }
    }

    HandleLevelUpUiMain<public>(): void = {
        GameManager := GetGameManager()
        Utility := GetUtility()
        if {
            Agent := MaybeAgent?
            PlayerData := Agent.GetPlayerData[]
            LevelUpHud := PlayerData.MaybeLevelUpHud?
        } then {
            LevelUpHud.Show(Agent, Utility.FormatMessage("LEVEL " + ToString(CurLevel + 1)))
        }
    }

    GetRequiredXpForLevel(Level: int): int = {
        var ReqExp: int = MaxLevelReqExp
        if(Temp := XpConfig.RequiredXpForLevelList[Level]) {
            set ReqExp = Temp
        } else {
            Print("WARN: GetRequiredXpForLevel, found no XP value for Level: {Level}")
            set ReqExp = 99999
        }
        return ReqExp
    }

    # Returns the level which the passed in XP corresponds to
    GetLevelFromXp(): int = {
        var XpLeft: int = TotalXp
        var TempLevel: int = 0
        loop {
            # Print("GetLevelFromXp: TempLevel {TempLevel}")
            var XpForNextLevel: int = GetRequiredXpForLevel(TempLevel)
            if(XpLeft < XpForNextLevel) {
                return TempLevel
            } else {
                # The user has enough XP for this level, check the next level
                set XpLeft = XpLeft - XpForNextLevel
                set TempLevel = TempLevel + 1
            }

            if(XpLeft <= 0) {
                break
            }
        }
        return TempLevel
    }

    GetCumulativeXpForLevel(Level: int): int = {
        if(Level < 0) {
            return 0
        }

        # Check if we already calculated this level
        if(CumulativeXp := XpConfig.CumulativeXpForLevelList[Level]) {
            return CumulativeXp
        } else {
            Print("ERR: GetCumulativeXpForLevel: Player is above max level")
            return MaxLevelReqExp
        }
    }

    UpdateResourceUi<public>(CurMoney: int, CurFans: int, CurDiamonds: int): void = {
        GameManager := GetGameManager()
        Utility := GetUtility()
        if {
            Agent := MaybeAgent?
        } then {
            GameManager.CurMoneyHud.Show(Agent, Utility.FormatMessage(Utility.FormatFloat(CurMoney * 1.0, 2, 0)))
            GameManager.CurFansHud.Show(Agent, Utility.FormatMessage(Utility.FormatFloat(CurFans * 1.0, 2, 0)))
            GameManager.CurDiamondsHud.Show(Agent, Utility.FormatMessage(Utility.FormatFloat(CurDiamonds * 1.0, 2, 0)))
        }
    }

    UpdateRebirthsUi<public>(CurRebirths: int): void = {
        GameManager := GetGameManager()
        Utility := GetUtility()
        if {
            Agent := MaybeAgent?
        } then {
            GameManager.CurRebirthsHud.Show(Agent, Utility.FormatMessage(Utility.FormatFloat(CurRebirths * 1.0, 0, 0) + " REBIRTH"))
        }
    }

    UpdateMoneyMultiplier<public>(CurMoneyMultiplier: float): void = {
        Utility := GetUtility()
        # TextBlockCurMoneyMultiplier.SetText(Utility.FormatMessage("x" + Utility.FormatFloat(CurMoneyMultiplier, 2, 1)))
    }

    UpdateIncomeMoneyUi<public>(CurIncomeMoney: float): void = {
        Utility := GetUtility()
        # TextBlockCurIncomeMoney.SetText(Utility.FormatMessage(Utility.FormatFloat(CurIncomeMoney, 2, 0) + "/s"))
    }

    UpdateScamLevelUi<public>(ParmScamLevel: int): void = {
        Print("UpdateScamLevelUi: ParmScamLevel: {ParmScamLevel}")
        GameManager := GetGameManager()
        Utility := GetUtility()
        if {
            Agent := MaybeAgent?
        } then {
            GameManager.CurScamRankHud.Show(Agent, Utility.FormatMessage("RANK: " + ToString(ParmScamLevel)))
        }
        <#
        if(ParmScamLevel >= 1) {
            TextureMaskLevelStar1.SetVisibility(widget_visibility.Visible)
        } else {
            TextureMaskLevelStar1.SetVisibility(widget_visibility.Hidden)
        }

        if(ParmScamLevel >= 2) {
            TextureMaskLevelStar2.SetVisibility(widget_visibility.Visible)
        } else {
            TextureMaskLevelStar2.SetVisibility(widget_visibility.Hidden)
        }

        if(ParmScamLevel >= 4) {
            TextureMaskLevelStar3.SetVisibility(widget_visibility.Visible)
        } else {
            TextureMaskLevelStar3.SetVisibility(widget_visibility.Hidden)
        }

        if(ParmScamLevel >= 5) {
            TextureMaskLevelStar4.SetVisibility(widget_visibility.Visible)
        } else {
            TextureMaskLevelStar4.SetVisibility(widget_visibility.Hidden)
        }

        if(ParmScamLevel >= 6) {
            TextureMaskLevelStar5.SetVisibility(widget_visibility.Visible)
        } else {
            TextureMaskLevelStar5.SetVisibility(widget_visibility.Hidden)
        }

        if(ParmScamLevel >= 7) {
            TextureMaskLevelStar6.SetVisibility(widget_visibility.Visible)
        } else {
            TextureMaskLevelStar6.SetVisibility(widget_visibility.Hidden)
        }
        #>
    }

    AddFloatingMessage<public>(Amount: int, MessageTypeId: int): void = {
        if(FloatingTextAmountQueue.Length < 20) {
            set FloatingTextAmountQueue = FloatingTextAmountQueue + array {Amount}
            set FloatingTextTypeQueue = FloatingTextTypeQueue + array {MessageTypeId}
            # Messages are now processed by the queue
            # Task := spawn{AddFloatingMessageSuspends(Amount, MessageTypeId)}
        }
    }

    AddFloatingMessageSuspends(Amount: int, MessageTypeId: int)<suspends>: void = {
        Utility := GetUtility()
        if {
            Agent := MaybeAgent?
            PlayerData := Agent.GetPlayerData[]
        } then {
            var MessagePrefix: string = ""
            case(MessageTypeId) {
                _ =>
                    set MessagePrefix = "+" # XP
                }
            var SlotId: int = 0
            loop {
                if {
                    TextSlot := FloatingTextList[SlotId]
                    IconSlot := FloatingIconList[SlotId]
                    TextColor := FloatingTextColorList[MessageTypeId]
                    IconTexture := FloatingTextIconList[MessageTypeId]
                } then {
                    var TextSuffix: string = ""
                    if {
                        MessageTypeId = 0
                        PlayerData.XpBoostLevel > 0
                    } then {
                        set TextSuffix = " (x2)"
                    }
                    TextToDisplay := Utility.FormatMessage(MessagePrefix + Utility.FormatFloat(Amount * 1.0, 2, 0) + TextSuffix) 
                    TextSlot.SetText(TextToDisplay)
                    TextSlot.SetTextColor(TextColor)
                    IconSlot.SetImage(IconTexture)
                    TextSlot.SetVisibility(widget_visibility.Visible)
                    IconSlot.SetVisibility(widget_visibility.Visible)
                    Sleep(0.0)
                    TextSlot.SetVisibility(widget_visibility.Hidden)
                    IconSlot.SetVisibility(widget_visibility.Hidden)
                    set SlotId = SlotId + 1
                } else {
                    break
                }
            }
        }
    }

    DoServerTick<public>(): void = {
        if {
            FloatingTextAmount := FloatingTextAmountQueue[0]
            FloatingTextType := FloatingTextTypeQueue[0]
        } then {
            if {
                NewFloatingTextAmountQueue := FloatingTextAmountQueue.RemoveElement[0]
                set FloatingTextAmountQueue = NewFloatingTextAmountQueue
                NewFloatingTextTypeQueue := FloatingTextTypeQueue.RemoveElement[0]
                set FloatingTextTypeQueue = NewFloatingTextTypeQueue
            } else {
                Print("ERR: NewFloatingTextAmountQueue")
            }
            Task := spawn{AddFloatingMessageSuspends(FloatingTextAmount, FloatingTextType)}
        }
    }
}
