using { /Verse.org/Simulation }
using { /Verse.org/Colors }
using { /Verse.org/Colors/NamedColors }
# using { /Verse.org/Concurrency }
# using { /Verse.org/Verse }
# using { /Fortnite.com/Devices }
using { /Fortnite.com/UI }
# using { /Fortnite.com/Characters }
# using { /UnrealEngine.com/Temporary/Diagnostics }
using { /UnrealEngine.com/Temporary/SpatialMath }
using { /UnrealEngine.com/Temporary/UI}

ui_playtime_reward<public> := class(ui_base) {
    # Constants
    RefreshRate: float = 1.0

    UiSlot<override>: player_ui_slot = player_ui_slot {
        ZOrder := 1200
        InputMode := ui_input_mode.All
    }

    var TextBlockTimerList<public>: []text_block = array {}
    var TextureCheckmarkList<public>: []texture_block = array {}
    var ItemButtonList<public>: []button_loud = array {}

    TextBlockTimer1: text_block = text_block {
        DefaultTextColor:= color{R:=1.0, G:=1.0, B:=1.0}
    }

    TextBlockTimer2: text_block = text_block {
        DefaultTextColor:= color{R:=1.0, G:=1.0, B:=1.0}
    }

    TextBlockTimer3: text_block = text_block {
        DefaultTextColor:= color{R:=1.0, G:=1.0, B:=1.0}
    }

    TextBlockTimer4: text_block = text_block {
        DefaultTextColor:= color{R:=1.0, G:=1.0, B:=1.0}
    }

    TextBlockTimer5: text_block = text_block {
        DefaultTextColor:= color{R:=1.0, G:=1.0, B:=1.0}
    }

    TextBlockTimer6: text_block = text_block {
        DefaultTextColor:= color{R:=1.0, G:=1.0, B:=1.0}
    }

    TextBlockTimer7: text_block = text_block {
        DefaultTextColor:= color{R:=1.0, G:=1.0, B:=1.0}
    }

    TextBlockTimer8: text_block = text_block {
        DefaultTextColor:= color{R:=1.0, G:=1.0, B:=1.0}
    }

    TextBlockTimer9: text_block = text_block {
        DefaultTextColor:= color{R:=1.0, G:=1.0, B:=1.0}
    }

    TextBlockTimer10: text_block = text_block {
        DefaultTextColor:= color{R:=1.0, G:=1.0, B:=1.0}
    }

    TextBlockTimer11: text_block = text_block {
        DefaultTextColor:= color{R:=1.0, G:=1.0, B:=1.0}
    }

    TextBlockTimer12: text_block = text_block {
        DefaultTextColor:= color{R:=1.0, G:=1.0, B:=1.0}
    }

    TextureBackground: texture_block = texture_block {
        DefaultImage := UI.free_rewardUI
        DefaultDesiredSize := vector2 {
            X := 506.206177
            Y := 487.087097
        }
    }

    TexturePresent1: texture_block = texture_block {
        DefaultImage := UI.giftbox
        DefaultDesiredSize := vector2 {
            X := 506.206177
            Y := 487.087097
        }
    }

    TexturePresent2: texture_block = texture_block {
        DefaultImage := UI.giftbox
        DefaultDesiredSize := vector2 {
            X := 506.206177
            Y := 487.087097
        }
    }

    TexturePresent3: texture_block = texture_block {
        DefaultImage := UI.giftbox
        DefaultDesiredSize := vector2 {
            X := 506.206177
            Y := 487.087097
        }
    }

    TexturePresent4: texture_block = texture_block {
        DefaultImage := UI.giftbox
        DefaultDesiredSize := vector2 {
            X := 506.206177
            Y := 487.087097
        }
    }

    TexturePresent5: texture_block = texture_block {
        DefaultImage := UI.giftbox
        DefaultDesiredSize := vector2 {
            X := 506.206177
            Y := 487.087097
        }
    }

    TexturePresent6: texture_block = texture_block {
        DefaultImage := UI.chest
        DefaultDesiredSize := vector2 {
            X := 506.206177
            Y := 487.087097
        }
    }

    TexturePresent7: texture_block = texture_block {
        DefaultImage := UI.chest
        DefaultDesiredSize := vector2 {
            X := 506.206177
            Y := 487.087097
        }
    }

    TexturePresent8: texture_block = texture_block {
        DefaultImage := UI.chest
        DefaultDesiredSize := vector2 {
            X := 506.206177
            Y := 487.087097
        }
    }

    TexturePresent9: texture_block = texture_block {
        DefaultImage := UI.chest
        DefaultDesiredSize := vector2 {
            X := 506.206177
            Y := 487.087097
        }
    }

    TexturePresent10: texture_block = texture_block {
        DefaultImage := UI.chest
        DefaultDesiredSize := vector2 {
            X := 506.206177
            Y := 487.087097
        }
    }

    TexturePresent11: texture_block = texture_block {
        DefaultImage := UI.chestbig
        DefaultDesiredSize := vector2 {
            X := 506.206177
            Y := 487.087097
        }
    }

    TexturePresent12: texture_block = texture_block {
        DefaultImage := UI.chestbig
        DefaultDesiredSize := vector2 {
            X := 506.206177
            Y := 487.087097
        }
    }
    
    TextureCheck1: texture_block = texture_block {
        DefaultImage := UI.checkmark
        DefaultDesiredSize := vector2 {
            X := 506.206177
            Y := 487.087097
        }
    }
    TextureCheck2: texture_block = texture_block {
        DefaultImage := UI.checkmark
        DefaultDesiredSize := vector2 {
            X := 506.206177
            Y := 487.087097
        }
    }
    TextureCheck3: texture_block = texture_block {
        DefaultImage := UI.checkmark
        DefaultDesiredSize := vector2 {
            X := 506.206177
            Y := 487.087097
        }
    }
    TextureCheck4: texture_block = texture_block {
        DefaultImage := UI.checkmark
        DefaultDesiredSize := vector2 {
            X := 506.206177
            Y := 487.087097
        }
    }
    TextureCheck5: texture_block = texture_block {
        DefaultImage := UI.checkmark
        DefaultDesiredSize := vector2 {
            X := 506.206177
            Y := 487.087097
        }
    }
    TextureCheck6: texture_block = texture_block {
        DefaultImage := UI.checkmark
        DefaultDesiredSize := vector2 {
            X := 506.206177
            Y := 487.087097
        }
    }
    TextureCheck7: texture_block = texture_block {
        DefaultImage := UI.checkmark
        DefaultDesiredSize := vector2 {
            X := 506.206177
            Y := 487.087097
        }
    }
    TextureCheck8: texture_block = texture_block {
        DefaultImage := UI.checkmark
        DefaultDesiredSize := vector2 {
            X := 506.206177
            Y := 487.087097
        }
    }
    TextureCheck9: texture_block = texture_block {
        DefaultImage := UI.checkmark
        DefaultDesiredSize := vector2 {
            X := 506.206177
            Y := 487.087097
        }
    }
    TextureCheck10: texture_block = texture_block {
        DefaultImage := UI.checkmark
        DefaultDesiredSize := vector2 {
            X := 506.206177
            Y := 487.087097
        }
    }
    TextureCheck11: texture_block = texture_block {
        DefaultImage := UI.checkmark
        DefaultDesiredSize := vector2 {
            X := 506.206177
            Y := 487.087097
        }
    }
    TextureCheck12: texture_block = texture_block {
        DefaultImage := UI.checkmark
        DefaultDesiredSize := vector2 {
            X := 506.206177
            Y := 487.087097
        }
    }
    

    # Runtime
    var MaybePlaytimeRewardManager: ?playtime_reward_device = false

    InitWidgetPlaytimeReward<public>(Agent: agent, PlaytimeRewardManager: playtime_reward_device): void = {
        set MaybePlaytimeRewardManager = option{PlaytimeRewardManager}

        set TextBlockTimerList = TextBlockTimerList + array {
            TextBlockTimer1,
            TextBlockTimer2,
            TextBlockTimer3,
            TextBlockTimer4,
            TextBlockTimer5,
            TextBlockTimer6,
            TextBlockTimer7,
            TextBlockTimer8,
            TextBlockTimer9,
            TextBlockTimer10,
            TextBlockTimer11,
            TextBlockTimer12
        }

        # TODO: Add images
        set TextureCheckmarkList = TextureCheckmarkList + array {
            TextureCheck1,
            TextureCheck2,
            TextureCheck3,
            TextureCheck4,
            TextureCheck5,
            TextureCheck6,
            TextureCheck7,
            TextureCheck8,
            TextureCheck9,
            TextureCheck10,
            TextureCheck11,
            TextureCheck12
        }
        

        # Call super class InitWidget
        InitWidget(Agent)
    }

    AddUiToPlayer<override>(Agent: agent, ParmMaybeManager: ?_Interface.interface_ui_manager): void = {
        (super:)AddUiToPlayer(Agent, ParmMaybeManager)

        RefreshUi()

        # Start refresh tick
        MyRoutine := SpawnRoutine(RefreshTick)
        set RoutineTick = option{MyRoutine}
    }

    CreateSlots<override>(Agent: agent): []canvas_slot = {
        Utility := GetUtility()

        # Buttons need to be defined inside the function
        ButtonClose: button_loud = button_loud{DefaultText:=Utility.FormatMessage("EXIT")}

        # Reward item button slots
        ItemButton1 := button_loud{}
        ItemButton2 := button_loud{}
        ItemButton3 := button_loud{}
        ItemButton4 := button_loud{}
        ItemButton5 := button_loud{}
        ItemButton6 := button_loud{}
        ItemButton7 := button_loud{}
        ItemButton8 := button_loud{}
        ItemButton9 := button_loud{}
        ItemButton10 := button_loud{}
        ItemButton11 := button_loud{}
        ItemButton12 := button_loud{}

        set ItemButtonList = array {
            ItemButton1,
            ItemButton2,
            ItemButton3,
            ItemButton4,
            ItemButton5,
            ItemButton6,
            ItemButton7,
            ItemButton8,
            ItemButton9,
            ItemButton10,
            ItemButton11,
            ItemButton12
        }

        for(ButtonIndex -> ItemButton: ItemButtonList) {
            ButtonEventHandler := event_handler_widget_message(int) {
                A := ButtonIndex
                CallbackFunc := OnItemClicked
            }

            SpawnerEventSub := ItemButton.OnClick().Subscribe(ButtonEventHandler.EventSubscription)
        }

        # Init slots
        var Slots: []canvas_slot = array {
            # Reward present images
            # TODO: Add images
            canvas_slot {
                Anchors := anchors{Minimum := vector2{X := 0.5, Y := 0.5}, Maximum := vector2{X := 0.5, Y := 0.5}}
                Offsets := margin{Top := -280.176361, Left := -528.932251, Right := 145.474045, Bottom := 133.95108}
                Alignment := vector2{X := 0.0, Y := 0.0}
                SizeToContent := false
                Widget := TexturePresent1
                ZOrder := 900
            },
            canvas_slot {
                Anchors := anchors{Minimum := vector2{X := 0.5, Y := 0.5}, Maximum := vector2{X := 0.5, Y := 0.5}}
                Offsets := margin{Top := -280.176361, Left := -297.423401, Right := 145.474045, Bottom := 133.95108}
                Alignment := vector2{X := 0.0, Y := 0.0}
                SizeToContent := false
                Widget := TexturePresent2
                ZOrder := 900
            },
            canvas_slot {
                Anchors := anchors{Minimum := vector2{X := 0.5, Y := 0.5}, Maximum := vector2{X := 0.5, Y := 0.5}}
                Offsets := margin{Top := -280.176361, Left := -66.960938, Right := 145.474045, Bottom := 133.95108}
                Alignment := vector2{X := 0.0, Y := 0.0}
                SizeToContent := false
                Widget := TexturePresent3
                ZOrder := 900
            },
            canvas_slot {
                Anchors := anchors{Minimum := vector2{X := 0.5, Y := 0.5}, Maximum := vector2{X := 0.5, Y := 0.5}}
                Offsets := margin{Top := -280.176361, Left := 164.039062, Right := 145.474045, Bottom := 133.95108}
                Alignment := vector2{X := 0.0, Y := 0.0}
                SizeToContent := false
                Widget := TexturePresent4
                ZOrder := 900
            },
            canvas_slot {
                Anchors := anchors{Minimum := vector2{X := 0.5, Y := 0.5}, Maximum := vector2{X := 0.5, Y := 0.5}}
                Offsets := margin{Top := -280.176361, Left := 395.039062, Right := 145.474045, Bottom := 133.95108}
                Alignment := vector2{X := 0.0, Y := 0.0}
                SizeToContent := false
                Widget := TexturePresent5
                ZOrder := 900
            },
            canvas_slot {
                Anchors := anchors{Minimum := vector2{X := 0.5, Y := 0.5}, Maximum := vector2{X := 0.5, Y := 0.5}}
                Offsets := margin{Top := -16.540527, Left := -528.960938, Right := 145.474045, Bottom := 133.95108}
                Alignment := vector2{X := 0.0, Y := 0.0}
                SizeToContent := false
                Widget := TexturePresent6
                ZOrder := 900
            },
            canvas_slot {
                Anchors := anchors{Minimum := vector2{X := 0.5, Y := 0.5}, Maximum := vector2{X := 0.5, Y := 0.5}}
                Offsets := margin{Top := -16.540527, Left := -296.960938, Right := 145.474045, Bottom := 133.95108}
                Alignment := vector2{X := 0.0, Y := 0.0}
                SizeToContent := false
                Widget := TexturePresent7
                ZOrder := 900
            },
            canvas_slot {
                Anchors := anchors{Minimum := vector2{X := 0.5, Y := 0.5}, Maximum := vector2{X := 0.5, Y := 0.5}}
                Offsets := margin{Top := -16.540527, Left := -64.960938, Right := 145.474045, Bottom := 133.95108}
                Alignment := vector2{X := 0.0, Y := 0.0}
                SizeToContent := false
                Widget := TexturePresent8
                ZOrder := 900
            },
            canvas_slot {
                Anchors := anchors{Minimum := vector2{X := 0.5, Y := 0.5}, Maximum := vector2{X := 0.5, Y := 0.5}}
                Offsets := margin{Top := -16.540527, Left := 163.039062, Right := 145.474045, Bottom := 133.95108}
                Alignment := vector2{X := 0.0, Y := 0.0}
                SizeToContent := false
                Widget := TexturePresent9
                ZOrder := 900
            },
            canvas_slot {
                Anchors := anchors{Minimum := vector2{X := 0.5, Y := 0.5}, Maximum := vector2{X := 0.5, Y := 0.5}}
                Offsets := margin{Top := -16.540527, Left := 395.039062, Right := 145.474045, Bottom := 133.95108}
                Alignment := vector2{X := 0.0, Y := 0.0}
                SizeToContent := false
                Widget := TexturePresent10
                ZOrder := 900
            },
            canvas_slot {
                Anchors := anchors{Minimum := vector2{X := 0.5, Y := 0.5}, Maximum := vector2{X := 0.5, Y := 0.5}}
                Offsets := margin{Top := 242.857483, Left := -528.960938, Right := 145.474045, Bottom := 133.95108}
                Alignment := vector2{X := 0.0, Y := 0.0}
                SizeToContent := false
                Widget := TexturePresent11
                ZOrder := 900
            },
            canvas_slot {
                Anchors := anchors{Minimum := vector2{X := 0.5, Y := 0.5}, Maximum := vector2{X := 0.5, Y := 0.5}}
                Offsets := margin{Top := 242.857483, Left := -296.960938, Right := 145.474045, Bottom := 133.95108}
                Alignment := vector2{X := 0.0, Y := 0.0}
                SizeToContent := false
                Widget := TexturePresent12
                ZOrder := 900
            },
            #
            # Claim reward buttons
            canvas_slot {
                Anchors := anchors{Minimum := vector2{X := 0.5, Y := 0.5}, Maximum := vector2{X := 0.5, Y := 0.5}}
                Offsets := margin{Top := -119.719635, Left := -539.624268, Right := 169.419159, Bottom := 46.841301}
                Alignment := vector2{X := 0.0, Y := 0.0}
                SizeToContent := false
                Widget := ItemButton1
                ZOrder := 950
            },
            canvas_slot {
                Anchors := anchors{Minimum := vector2{X := 0.5, Y := 0.5}, Maximum := vector2{X := 0.5, Y := 0.5}}
                Offsets := margin{Top := -119.719635, Left := -308.960938, Right := 169.419159, Bottom := 46.841301}
                Alignment := vector2{X := 0.0, Y := 0.0}
                SizeToContent := false
                Widget := ItemButton2
                ZOrder := 950
            },
            canvas_slot {
                Anchors := anchors{Minimum := vector2{X := 0.5, Y := 0.5}, Maximum := vector2{X := 0.5, Y := 0.5}}
                Offsets := margin{Top := -119.719635, Left := -77.960938, Right := 169.419159, Bottom := 46.841301}
                Alignment := vector2{X := 0.0, Y := 0.0}
                SizeToContent := false
                Widget := ItemButton3
                ZOrder := 950
            },
            canvas_slot {
                Anchors := anchors{Minimum := vector2{X := 0.5, Y := 0.5}, Maximum := vector2{X := 0.5, Y := 0.5}}
                Offsets := margin{Top := -119.719635, Left := 153.039062, Right := 169.419159, Bottom := 46.841301}
                Alignment := vector2{X := 0.0, Y := 0.0}
                SizeToContent := false
                Widget := ItemButton4
                ZOrder := 950
            },
            canvas_slot {
                Anchors := anchors{Minimum := vector2{X := 0.5, Y := 0.5}, Maximum := vector2{X := 0.5, Y := 0.5}}
                Offsets := margin{Top := -119.719635, Left := 384.039062, Right := 169.419159, Bottom := 46.841301}
                Alignment := vector2{X := 0.0, Y := 0.0}
                SizeToContent := false
                Widget := ItemButton5
                ZOrder := 950
            },
            canvas_slot {
                Anchors := anchors{Minimum := vector2{X := 0.5, Y := 0.5}, Maximum := vector2{X := 0.5, Y := 0.5}}
                Offsets := margin{Top := 147.459473, Left := -540.960938, Right := 169.419159, Bottom := 46.841301}
                Alignment := vector2{X := 0.0, Y := 0.0}
                SizeToContent := false
                Widget := ItemButton6
                ZOrder := 950
            },
            canvas_slot {
                Anchors := anchors{Minimum := vector2{X := 0.5, Y := 0.5}, Maximum := vector2{X := 0.5, Y := 0.5}}
                Offsets := margin{Top := 147.459473, Left := -308.960938, Right := 169.419159, Bottom := 46.841301}
                Alignment := vector2{X := 0.0, Y := 0.0}
                SizeToContent := false
                Widget := ItemButton7
                ZOrder := 950
            },
            canvas_slot {
                Anchors := anchors{Minimum := vector2{X := 0.5, Y := 0.5}, Maximum := vector2{X := 0.5, Y := 0.5}}
                Offsets := margin{Top := 147.459473, Left := -76.960938, Right := 169.419159, Bottom := 46.841301}
                Alignment := vector2{X := 0.0, Y := 0.0}
                SizeToContent := false
                Widget := ItemButton8
                ZOrder := 950
            },
            canvas_slot {
                Anchors := anchors{Minimum := vector2{X := 0.5, Y := 0.5}, Maximum := vector2{X := 0.5, Y := 0.5}}
                Offsets := margin{Top := 147.459473, Left := 155.039062, Right := 169.419159, Bottom := 46.841301}
                Alignment := vector2{X := 0.0, Y := 0.0}
                SizeToContent := false
                Widget := ItemButton9
                ZOrder := 950
            },
            canvas_slot {
                Anchors := anchors{Minimum := vector2{X := 0.5, Y := 0.5}, Maximum := vector2{X := 0.5, Y := 0.5}}
                Offsets := margin{Top := 147.459473, Left := 383.039062, Right := 169.419159, Bottom := 46.841301}
                Alignment := vector2{X := 0.0, Y := 0.0}
                SizeToContent := false
                Widget := ItemButton10
                ZOrder := 950
            },
            canvas_slot {
                Anchors := anchors{Minimum := vector2{X := 0.5, Y := 0.5}, Maximum := vector2{X := 0.5, Y := 0.5}}
                Offsets := margin{Top := 391.459473, Left := -540.960938, Right := 169.419159, Bottom := 46.841301}
                Alignment := vector2{X := 0.0, Y := 0.0}
                SizeToContent := false
                Widget := ItemButton11
                ZOrder := 950
            },
            canvas_slot {
                Anchors := anchors{Minimum := vector2{X := 0.5, Y := 0.5}, Maximum := vector2{X := 0.5, Y := 0.5}}
                Offsets := margin{Top := 391.459473, Left := -312.960938, Right := 169.419159, Bottom := 46.841301}
                Alignment := vector2{X := 0.0, Y := 0.0}
                SizeToContent := false
                Widget := ItemButton12
                ZOrder := 950
            },
            # Reward timers
            canvas_slot {
                Anchors := anchors{Minimum := vector2{X := 0.5, Y := 0.5}, Maximum := vector2{X := 0.5, Y := 0.5}}
                Offsets := margin{Top := -316.540527, Left := -496.960938, Right := 94.262596, Bottom := 41.0}
                Alignment := vector2{X := 0.0, Y := 0.0}
                SizeToContent := false
                Widget := TextBlockTimer1
                ZOrder := 1900
            },
            canvas_slot {
                Anchors := anchors{Minimum := vector2{X := 0.5, Y := 0.5}, Maximum := vector2{X := 0.5, Y := 0.5}}
                Offsets := margin{Top := -316.540527, Left := -260.960938, Right := 94.262596, Bottom := 41.0}
                Alignment := vector2{X := 0.0, Y := 0.0}
                SizeToContent := false
                Widget := TextBlockTimer2
                ZOrder := 1900
            },
            canvas_slot {
                Anchors := anchors{Minimum := vector2{X := 0.5, Y := 0.5}, Maximum := vector2{X := 0.5, Y := 0.5}}
                Offsets := margin{Top := -316.540527, Left := -36.960938, Right := 94.262596, Bottom := 41.0}
                Alignment := vector2{X := 0.0, Y := 0.0}
                SizeToContent := false
                Widget := TextBlockTimer3
                ZOrder := 1900
            },
            canvas_slot {
                Anchors := anchors{Minimum := vector2{X := 0.5, Y := 0.5}, Maximum := vector2{X := 0.5, Y := 0.5}}
                Offsets := margin{Top := -316.540527, Left := 195.039062, Right := 94.262596, Bottom := 41.0}
                Alignment := vector2{X := 0.0, Y := 0.0}
                SizeToContent := false
                Widget := TextBlockTimer4
                ZOrder := 1900
            },
            canvas_slot {
                Anchors := anchors{Minimum := vector2{X := 0.5, Y := 0.5}, Maximum := vector2{X := 0.5, Y := 0.5}}
                Offsets := margin{Top := -316.540527, Left := 427.039062, Right := 94.262596, Bottom := 41.0}
                Alignment := vector2{X := 0.0, Y := 0.0}
                SizeToContent := false
                Widget := TextBlockTimer5
                ZOrder := 1900
            },
            canvas_slot {
                Anchors := anchors{Minimum := vector2{X := 0.5, Y := 0.5}, Maximum := vector2{X := 0.5, Y := 0.5}}
                Offsets := margin{Top := -52.540527, Left := -492.960938, Right := 94.262596, Bottom := 41.0}
                Alignment := vector2{X := 0.0, Y := 0.0}
                SizeToContent := false
                Widget := TextBlockTimer6
                ZOrder := 1900
            },
            canvas_slot {
                Anchors := anchors{Minimum := vector2{X := 0.5, Y := 0.5}, Maximum := vector2{X := 0.5, Y := 0.5}}
                Offsets := margin{Top := -52.540527, Left := -264.960938, Right := 94.262596, Bottom := 41.0}
                Alignment := vector2{X := 0.0, Y := 0.0}
                SizeToContent := false
                Widget := TextBlockTimer7
                ZOrder := 1900
            },
            canvas_slot {
                Anchors := anchors{Minimum := vector2{X := 0.5, Y := 0.5}, Maximum := vector2{X := 0.5, Y := 0.5}}
                Offsets := margin{Top := -52.540527, Left := -36.960938, Right := 94.262596, Bottom := 41.0}
                Alignment := vector2{X := 0.0, Y := 0.0}
                SizeToContent := false
                Widget := TextBlockTimer8
                ZOrder := 1900
            },
            canvas_slot {
                Anchors := anchors{Minimum := vector2{X := 0.5, Y := 0.5}, Maximum := vector2{X := 0.5, Y := 0.5}}
                Offsets := margin{Top := -52.540527, Left := 195.039062, Right := 94.262596, Bottom := 41.0}
                Alignment := vector2{X := 0.0, Y := 0.0}
                SizeToContent := false
                Widget := TextBlockTimer9
                ZOrder := 1900
            },
            canvas_slot {
                Anchors := anchors{Minimum := vector2{X := 0.5, Y := 0.5}, Maximum := vector2{X := 0.5, Y := 0.5}}
                Offsets := margin{Top := -52.540527, Left := 427.039062, Right := 94.262596, Bottom := 41.0}
                Alignment := vector2{X := 0.0, Y := 0.0}
                SizeToContent := false
                Widget := TextBlockTimer10
                ZOrder := 1900
            },
            canvas_slot {
                Anchors := anchors{Minimum := vector2{X := 0.5, Y := 0.5}, Maximum := vector2{X := 0.5, Y := 0.5}}
                Offsets := margin{Top := 203.459473, Left := -524.960938, Right := 94.262596, Bottom := 41.0}
                Alignment := vector2{X := 0.0, Y := 0.0}
                SizeToContent := false
                Widget := TextBlockTimer11
                ZOrder := 1900
            },
            canvas_slot {
                Anchors := anchors{Minimum := vector2{X := 0.5, Y := 0.5}, Maximum := vector2{X := 0.5, Y := 0.5}}
                Offsets := margin{Top := 203.459473, Left := -292.960938, Right := 94.262596, Bottom := 41.0}
                Alignment := vector2{X := 0.0, Y := 0.0}
                SizeToContent := false
                Widget := TextBlockTimer12
                ZOrder := 1900
            },
            # Close button
            canvas_slot {
                Anchors := anchors{Minimum := vector2{X := 0.5, Y := 0.5}, Maximum := vector2{X := 0.5, Y := 0.5}}
                Offsets := margin{Top := 455.459473, Left := -76.960938, Right := 170.92067, Bottom := 51.145607}
                Alignment := vector2{X := 0.0, Y := 0.0}
                SizeToContent := false
                Widget := ButtonClose
                ZOrder := 901
            },
            # Checkmarks
            # TODO: Add images
            canvas_slot {
                Anchors := anchors{Minimum := vector2{X := 0.5, Y := 0.5}, Maximum := vector2{X := 0.5, Y := 0.5}}
                Offsets := margin{Top := -220.540527, Left := -440.960938, Right := 68.695969, Bottom := 71.702148}
                Alignment := vector2{X := 0.0, Y := 0.0}
                SizeToContent := false
                Widget := TextureCheck1
                ZOrder := 900
            },
            canvas_slot {
                Anchors := anchors{Minimum := vector2{X := 0.5, Y := 0.5}, Maximum := vector2{X := 0.5, Y := 0.5}}
                Offsets := margin{Top := -216.540527, Left := -208.960938, Right := 68.695969, Bottom := 71.702148}
                Alignment := vector2{X := 0.0, Y := 0.0}
                SizeToContent := false
                Widget := TextureCheck2
                ZOrder := 900
            },
            canvas_slot {
                Anchors := anchors{Minimum := vector2{X := 0.5, Y := 0.5}, Maximum := vector2{X := 0.5, Y := 0.5}}
                Offsets := margin{Top := -216.540527, Left := 19.039062, Right := 68.695969, Bottom := 71.702148}
                Alignment := vector2{X := 0.0, Y := 0.0}
                SizeToContent := false
                Widget := TextureCheck3
                ZOrder := 900
            },
            canvas_slot {
                Anchors := anchors{Minimum := vector2{X := 0.5, Y := 0.5}, Maximum := vector2{X := 0.5, Y := 0.5}}
                Offsets := margin{Top := -216.540527, Left := 255.039062, Right := 68.695969, Bottom := 71.702148}
                Alignment := vector2{X := 0.0, Y := 0.0}
                SizeToContent := false
                Widget := TextureCheck4
                ZOrder := 900
            },
            canvas_slot {
                Anchors := anchors{Minimum := vector2{X := 0.5, Y := 0.5}, Maximum := vector2{X := 0.5, Y := 0.5}}
                Offsets := margin{Top := -216.540527, Left := 483.039062, Right := 68.695969, Bottom := 71.702148}
                Alignment := vector2{X := 0.0, Y := 0.0}
                SizeToContent := false
                Widget := TextureCheck5
                ZOrder := 900
            },
            canvas_slot {
                Anchors := anchors{Minimum := vector2{X := 0.5, Y := 0.5}, Maximum := vector2{X := 0.5, Y := 0.5}}
                Offsets := margin{Top := 47.459473, Left := -432.960938, Right := 68.695969, Bottom := 71.702148}
                Alignment := vector2{X := 0.0, Y := 0.0}
                SizeToContent := false
                Widget := TextureCheck6
                ZOrder := 900
            },
            canvas_slot {
                Anchors := anchors{Minimum := vector2{X := 0.5, Y := 0.5}, Maximum := vector2{X := 0.5, Y := 0.5}}
                Offsets := margin{Top := 47.459473, Left := -200.960938, Right := 68.695969, Bottom := 71.702148}
                Alignment := vector2{X := 0.0, Y := 0.0}
                SizeToContent := false
                Widget := TextureCheck7
                ZOrder := 900
            },
            canvas_slot {
                Anchors := anchors{Minimum := vector2{X := 0.5, Y := 0.5}, Maximum := vector2{X := 0.5, Y := 0.5}}
                Offsets := margin{Top := 47.459473, Left := 31.039062, Right := 68.695969, Bottom := 71.702148}
                Alignment := vector2{X := 0.0, Y := 0.0}
                SizeToContent := false
                Widget := TextureCheck8
                ZOrder := 900
            },
            canvas_slot {
                Anchors := anchors{Minimum := vector2{X := 0.5, Y := 0.5}, Maximum := vector2{X := 0.5, Y := 0.5}}
                Offsets := margin{Top := 47.459473, Left := 263.039062, Right := 68.695969, Bottom := 71.702148}
                Alignment := vector2{X := 0.0, Y := 0.0}
                SizeToContent := false
                Widget := TextureCheck9
                ZOrder := 900
            },
            canvas_slot {
                Anchors := anchors{Minimum := vector2{X := 0.5, Y := 0.5}, Maximum := vector2{X := 0.5, Y := 0.5}}
                Offsets := margin{Top := 47.459473, Left := 491.039062, Right := 68.695969, Bottom := 71.702148}
                Alignment := vector2{X := 0.0, Y := 0.0}
                SizeToContent := false
                Widget := TextureCheck10
                ZOrder := 900
            },
            canvas_slot {
                Anchors := anchors{Minimum := vector2{X := 0.5, Y := 0.5}, Maximum := vector2{X := 0.5, Y := 0.5}}
                Offsets := margin{Top := 299.459473, Left := -436.960938, Right := 68.695969, Bottom := 71.702148}
                Alignment := vector2{X := 0.0, Y := 0.0}
                SizeToContent := false
                Widget := TextureCheck11
                ZOrder := 900
            },
            canvas_slot {
                Anchors := anchors{Minimum := vector2{X := 0.5, Y := 0.5}, Maximum := vector2{X := 0.5, Y := 0.5}}
                Offsets := margin{Top := 299.459473, Left := -204.960938, Right := 68.695969, Bottom := 71.702148}
                Alignment := vector2{X := 0.0, Y := 0.0}
                SizeToContent := false
                Widget := TextureCheck12
                ZOrder := 900
            },
            # Background
            canvas_slot {
                Anchors := anchors{Minimum := vector2{X := 0.5, Y := 0.5}, Maximum := vector2{X := 0.5, Y := 0.5}}
                Offsets := margin{Top := -644.540527, Left := -709.025452, Right := 1394.866211, Bottom := 1341.215576}
                Alignment := vector2{X := 0.0, Y := 0.0}
                SizeToContent := false
                Widget := TextureBackground
                ZOrder := 899
            }
        }

        # Initially hide some slots
        for(TextureCheckmark: TextureCheckmarkList) {
            TextureCheckmark.SetVisibility(widget_visibility.Hidden)
        }

        # Listen for click events
        ButtonClose.OnClick().Subscribe(OnExitButtonClicked)

        return Slots
    }

    RefreshUi<public>(): void = {
        Utility := GetUtility()
        if {
            Agent := MaybeAgent?
            PlaytimeRewardManager := MaybePlaytimeRewardManager?
            PlayerData := Agent.GetPlayerData[]
        } then {
            # Check every playtime reward if it is ready to be picked up
            Timestamp := GetSimulationElapsedTime()
            TimePlayed := Timestamp - PlayerData.TimestampPlayerUnlockedPlaytimeRewards
            for(ItemIndex -> PlaytimeRequired: PlaytimeRewardManager.PlaytimeRewardTimerList) {
                if {
                    ItemButton := ItemButtonList[ItemIndex]
                    TextBlockTimer := TextBlockTimerList[ItemIndex]
                    TextureCheckmark := TextureCheckmarkList[ItemIndex]
                } then {
                    TimeRemaining := PlaytimeRequired - TimePlayed
                    ItemState := PlaytimeRewardManager.GetItemState(Agent, ItemIndex)
                    if {
                        ItemState = 0
                        TimeRemainingInt := Floor[TimeRemaining]
                    } then {
                        ItemButton.SetText(Utility.FormatMessage("WAIT..."))
                        if(TimeRemaining <= 0.0) {
                            PlayerData.SetPlaytimeRewardItemState(ItemIndex, 1)
                            # Hide the time remaining timer if the item is ready / already received
                            TextBlockTimer.SetText(Utility.FormatMessage(""))
                        } else {
                            # Update the time remaining timer
                            TextBlockTimer.SetText(Utility.FormatTimeToMessage(TimeRemainingInt))
                        }
                    } else {
                        # Hide the time remaining timer if the item is ready / already received
                        TextBlockTimer.SetText(Utility.FormatMessage(""))
                    }
                    if {
                        ItemState = 1
                    } then {
                        ItemButton.SetText(Utility.FormatMessage("GET"))
                    }
                    if {
                        ItemState = 2
                    } then {
                        ItemButton.SetText(Utility.FormatMessage("OK!"))
                    }
                }
            }
        }
    }

    OnItemClicked<override>(WidgetMessage: widget_message, ButtonIndex: int): void = {
        Print("OnItemClicked - ItemIndex: {ButtonIndex}")
        if {
            PlaytimeRewardManager := MaybePlaytimeRewardManager?
        } then {
            Agent := WidgetMessage.Player
            PlaytimeRewardManager.TryToClaimReward(Agent, Self, ButtonIndex)
        }
        TimeoutManager.UpdatePlayerTimeoutAction()
    }

    RefreshTick()<suspends>: void = {
        loop {
            RefreshUi()
            Sleep(RefreshRate)
        }
    }
}