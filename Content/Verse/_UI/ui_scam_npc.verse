using { /Verse.org/Simulation }
using { /Verse.org/Colors }
using { /Verse.org/Colors/NamedColors }
# using { /Verse.org/Concurrency }
# using { /Verse.org/Verse }
# using { /Fortnite.com/Devices }
using { /Fortnite.com/UI }
# using { /Fortnite.com/Characters }
# using { /UnrealEngine.com/Temporary/Diagnostics }
using { /UnrealEngine.com/Temporary/SpatialMath }
using { /UnrealEngine.com/Temporary/UI}

ui_scam_npc<public> := class(ui_base) {
    # Constants
    UiSlot<override>: player_ui_slot = player_ui_slot {
        ZOrder := 1001
        InputMode := ui_input_mode.None
    }

    ActionBarWidth: float = 775.246582
    ActionBarHeight: float = 53.728012


    TextBlockScammingProgress: text_block = text_block{
        DefaultTextColor := color{R:=1.0, G:=1.0, B:=1.0}
    }

    ProgressBlockScamming: color_block = color_block {
        DefaultColor := color{R:=0.110038, G:=1.0, B:=0.0}
        DefaultDesiredSize := vector2 {
            X := 775.246582
            Y := 53.728012
        }
    }

    ProgressBackground: color_block = color_block {
        DefaultColor := Black
        DefaultDesiredSize := vector2 {
            X := 775.246582
            Y := 53.728012
        }
    }

    TextBlockTimerNpcEscaping: text_block = text_block{
        DefaultTextColor := color{R:=1.0, G:=1.0, B:=1.0}
    }

    # Runtime
    var MaybeNpcEscapingTimer: ?ui_timer = false

    CreateSlots<override>(Agent: agent): []canvas_slot = {
        Utility := GetUtility()

        # Init slots
        var Slots: []canvas_slot = array {
            # Progress Background
            canvas_slot {
                Anchors := anchors{Minimum := vector2{X := 0.5, Y := 1.0}, Maximum := vector2{X := 0.5, Y := 1.0}}
                Offsets := margin{Top := -205.081055, Left := -380.960938, Right := 775.246582, Bottom := 53.728012}
                Alignment := vector2{X := 0.0, Y := 0.0}
                SizeToContent := true
                Widget := ProgressBackground
                ZOrder := 1010
            },
            # Progress bars
            canvas_slot {
                Anchors := anchors{Minimum := vector2{X := 0.5, Y := 1.0}, Maximum := vector2{X := 0.5, Y := 1.0}}
                Offsets := margin{Top := -205.081055, Left := -380.960938, Right := 775.246582, Bottom := 53.728012}
                Alignment := vector2{X := 0.0, Y := 0.0}
                SizeToContent := true
                Widget := ProgressBlockScamming
                ZOrder := 1011
            },
            canvas_slot {
                Anchors := anchors{Minimum := vector2{X := 0.5, Y := 1.0}, Maximum := vector2{X := 0.5, Y := 1.0}}
                Offsets := margin{Top := -205.081055, Left := -380.960938, Right := 775.246582, Bottom := 53.728012}
                Alignment := vector2{X := 0.0, Y := 0.0}
                SizeToContent := true
                Widget := TextBlockScammingProgress
                ZOrder := 1012
            },
            canvas_slot {
                Anchors := anchors{Minimum := vector2{X := 0.5, Y := 1.0}, Maximum := vector2{X := 0.5, Y := 1.0}}
                Offsets := margin{Top := -245.081055, Left := -380.960938, Right := 775.246582, Bottom := 53.728012}
                Alignment := vector2{X := 0.0, Y := 0.0}
                SizeToContent := true
                Widget := TextBlockTimerNpcEscaping
                ZOrder := 1100
            }
        }

        # Initially hide some slots
        TextBlockTimerNpcEscaping.SetVisibility(widget_visibility.Hidden)

        return Slots
    }

    UpdateScamProgressBar<public>(ScamProgress: float): void = {
        Utility := GetUtility()
        if {
            NpcEscapeTimer := MaybeNpcEscapingTimer?
        } then {
            NpcEscapeTimer.StopTimer()
        }

        ScammingBarWidth := ScamProgress * ActionBarWidth
        ProgressBlockScamming.SetDesiredSize(vector2 {X := ScammingBarWidth, Y := ActionBarHeight})
        TextBlockScammingProgress.SetText(Utility.FormatMessage("Scamming... " + Utility.FormatFloat(ScamProgress * 100.0, 0, 0) + "%"))
    }

    HandleNpcStartsEscaping<public>(ScamOutOfRangeResetDuration: float): void = {
        NpcEscapeTimer: ui_timer = ui_timer {}
        NpcEscapeTimer.StartTimer(TextBlockTimerNpcEscaping, option{ScamOutOfRangeResetDuration}, "Victim is escaping in ")
        set MaybeNpcEscapingTimer = option{NpcEscapeTimer}
    }
}