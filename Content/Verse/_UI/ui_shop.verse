using { /Verse.org/Simulation }
using { /Verse.org/Colors }
using { /Verse.org/Colors/NamedColors }
# using { /Verse.org/Concurrency }
# using { /Verse.org/Verse }
using { /Verse.org/Assets }
using { /Fortnite.com/Devices }
using { /Fortnite.com/UI }
# using { /Fortnite.com/Characters }
# using { /UnrealEngine.com/Temporary/Diagnostics }
using { /UnrealEngine.com/Temporary/SpatialMath }
using { /UnrealEngine.com/Temporary/UI}

ui_shop<public> := class(ui_base) {
    # Constants
    UiSlot<override>: player_ui_slot = player_ui_slot {
        ZOrder := 99999
        InputMode := ui_input_mode.All
    }

    # Runtime
    var MaybeShopManager: ?_Shop.shop_base_device = false
    var CanvasSlots: []canvas_slot = array {}
    var CurShopScrollIndex: int = 0
    var ImageList: []texture = array {}
    var ShopItemButtonSlotList: []canvas_slot = array {}
    var ShopItemImageSlotList: []canvas_slot = array {}

    # Must be in shop base class unfortunately
    var AchievementType<public>: _Global.achievement_type = _Global.achievement_type.Kill

    # Shops use InitShop instead of InitWidget from UI base class
    InitShop<public>(ShopManager: _Shop.shop_base_device, Agent: agent): void = {
        Print("shop_ui_player - InitShop")
        set MaybeShopManager = option{ShopManager}
        set MaybeAgent = option{Agent}

        if(CorrectImages := ShopManager.ImageListAllShops[ShopManager.ShopId]) {
            set ImageList = CorrectImages
        } else {
            Print("ERR: shop_ui_device: Found no images for ShopId: {ShopManager.ShopId}")
        }

        Canvas := ShopManager.CreateCanvasUi(Self, Agent)
        set MaybeCanvas = option{Canvas}
    }

    AddUiToPlayer<override>(Agent: agent, ParmMaybeManager: ?_Interface.interface_ui_manager): void = {
        (super:)AddUiToPlayer(Agent, ParmMaybeManager)
        RefreshPlayerShopUi(Agent)
    }

    CreateSlots<override>(Agent: agent): []canvas_slot = {
        Print("player_ui_shop - CreateSlots")
        Utility := GetUtility()
        var Slots: []canvas_slot = array {}

        if(ShopManager := MaybeShopManager?) {
            # Init background
            var TextureBlockBackground: texture_block = texture_block {
                DefaultImage := ShopManager.BackgroundImageTexture
                DefaultDesiredSize := vector2{X := 10.315323, Y := 10.027031}
            }

            # Init buttons
            CloseShopButton := button_loud{DefaultText := Utility.FormatMessage("EXIT")}
            CloseShopButton.OnClick().Subscribe(OnExitButtonClicked)

            set Slots = Slots + array {
                # Background image
                canvas_slot {
                    Anchors := ShopManager.AnchorsBackgroundImage
                    Offsets := ShopManager.OffsetsBackgroundImage
                    Alignment := vector2{X := 0.0, Y := 0.0}
                    SizeToContent := false
                    Widget := TextureBlockBackground
                    ZOrder := 900
                },
                # Close shop button
                canvas_slot {
                    Anchors := ShopManager.AnchorsCloseShopButton
                    Offsets := ShopManager.OffsetsCloseShopButton
                    Alignment := vector2{X := 0.0, Y := 0.0}
                    SizeToContent := false
                    Widget := CloseShopButton
                    ZOrder := 900
                }
            }

            # Shop item button slots
            ShopItemButton1 := button_loud{}
            ShopItemButton2 := button_loud{}
            ShopItemButton3 := button_loud{}
            ShopItemButton4 := button_loud{}
            ShopItemButton5 := button_loud{}
            ShopItemButton6 := button_loud{}
            ShopItemButton7 := button_loud{}
            ShopItemButton8 := button_loud{}
            ShopItemButton9 := button_loud{}
            ShopItemButton10 := button_loud{}

            ShopItemButtonList: []button_loud = array {
                ShopItemButton1,
                ShopItemButton2,
                ShopItemButton3,
                ShopItemButton4,
                ShopItemButton5,
                ShopItemButton6,
                ShopItemButton7,
                ShopItemButton8,
                ShopItemButton9,
                ShopItemButton10
            }

            for(ButtonIndex -> ShopItemButton: ShopItemButtonList) {
                ButtonEventHandler := event_handler_widget_message(int) {
                    A := ButtonIndex
                    CallbackFunc := OnItemClicked
                }
    
                SpawnerEventSub := ShopItemButton.OnClick().Subscribe(ButtonEventHandler.EventSubscription)
            }

            # Buy button slots (10x)
            for(ItemIndex -> AnchorsBuyButton: ShopManager.AnchorsBuyButton) {
                if {
                    OffsetsBuyButton := ShopManager.OffsetsBuyButton[ItemIndex]
                    ShopItemButton := ShopItemButtonList[ItemIndex]
                } then {
                    set ShopItemButtonSlotList = ShopItemButtonSlotList + array {
                        canvas_slot {
                            Anchors := AnchorsBuyButton
                            Offsets := OffsetsBuyButton
                            Alignment := vector2{X := 0.0, Y := 0.0}
                            SizeToContent := false
                            Widget := ShopItemButton
                            ZOrder := 900
                        }
                    }
                }
            }

            set Slots = Slots + ShopItemButtonSlotList

            # Init shop item textures
            var DummyTexture: texture = UI.checkmark

            var ShopImageTextureBlock: texture_block = texture_block {
                DefaultImage := DummyTexture
                DefaultDesiredSize := vector2{X := 10.315323, Y := 10.027031}
            }

            var ShopImageTextureBlock2: texture_block = texture_block {
                DefaultImage := DummyTexture
                DefaultDesiredSize := vector2{X := 10.315323, Y := 10.027031}
            }

            var ShopImageTextureBlock3: texture_block = texture_block {
                DefaultImage := DummyTexture
                DefaultDesiredSize := vector2{X := 10.315323, Y := 10.027031}
            }

            var ShopImageTextureBlock4: texture_block = texture_block {
                DefaultImage := DummyTexture
                DefaultDesiredSize := vector2{X := 10.315323, Y := 10.027031}
            }

            var ShopImageTextureBlock5: texture_block = texture_block {
                DefaultImage := DummyTexture
                DefaultDesiredSize := vector2{X := 10.315323, Y := 10.027031}
            }

            var ShopImageTextureBlock6: texture_block = texture_block {
                DefaultImage := DummyTexture
                DefaultDesiredSize := vector2{X := 10.315323, Y := 10.027031}
            }

            var ShopImageTextureBlock7: texture_block = texture_block {
                DefaultImage := DummyTexture
                DefaultDesiredSize := vector2{X := 10.315323, Y := 10.027031}
            }

            var ShopImageTextureBlock8: texture_block = texture_block {
                DefaultImage := DummyTexture
                DefaultDesiredSize := vector2{X := 10.315323, Y := 10.027031}
            }

            var ShopImageTextureBlock9: texture_block = texture_block {
                DefaultImage := DummyTexture
                DefaultDesiredSize := vector2{X := 10.315323, Y := 10.027031}
            }

            var ShopImageTextureBlock10: texture_block = texture_block {
                DefaultImage := DummyTexture
                DefaultDesiredSize := vector2{X := 10.315323, Y := 10.027031}
            }

            ShopItemImageList: []texture_block = array {
                ShopImageTextureBlock,
                ShopImageTextureBlock2,
                ShopImageTextureBlock3,
                ShopImageTextureBlock4,
                ShopImageTextureBlock5,
                ShopImageTextureBlock6,
                ShopImageTextureBlock7,
                ShopImageTextureBlock8,
                ShopImageTextureBlock9,
                ShopImageTextureBlock10
            }

            # Shop item image slots (x10)
            for(ItemIndex -> AnchorsItemImage: ShopManager.AnchorsItemImage) {
                if {
                    OffsetsItemImage := ShopManager.OffsetsItemImage[ItemIndex]
                    ShopItemImage := ShopItemImageList[ItemIndex]
                } then {
                    set ShopItemImageSlotList = ShopItemImageSlotList + array {
                        canvas_slot {
                            Anchors := AnchorsItemImage
                            Offsets := OffsetsItemImage
                            Alignment := vector2{X := 0.0, Y := 0.0}
                            SizeToContent := false
                            Widget := ShopItemImage
                            ZOrder := 900
                        }
                    }
                }
            }

            set Slots = Slots + ShopItemImageSlotList

            # Hide buttons + images initially
            for(ShopItemButton: ShopItemButtonSlotList) {
                ShopItemButton.Widget.SetVisibility(widget_visibility.Hidden)
            }
            
            for(ShopItemImage: ShopItemImageSlotList) {
                ShopItemImage.Widget.SetVisibility(widget_visibility.Hidden)
            }

            set CanvasSlots = Slots
        }
        
        return Slots
    }

    RefreshPlayerShopUi<public>(Agent: agent): void = {
        Print("RefreshPlayerShopUi")
        Print("CurShopScrollIndex: {CurShopScrollIndex}")

        Utility := GetUtility()

        <# Old method
        RemoveUiFromPlayer(Agent)
        AddUiToPlayer(Agent)
        #>
        
        <# Another old method
        if {
            Canvas := MaybeCanvas?
        } then {
            Print("CanvasSlots.Length = {CanvasSlots.Length}")
            for(OldSlot: CanvasSlots) {
                Canvas.RemoveWidget(OldSlot.Widget)
            }
            NewSlots := CreateSlots(Agent)
            Print("NewSlots.Length = {NewSlots.Length}")
            for(NewSlot: NewSlots) {
                Canvas.AddWidget(NewSlot)
            }
        }
        #>

        if(ShopManager := MaybeShopManager?) {
            <# Hide previous slots
            for(ItemIndex := 0 .. ShopManager.ShopItemsPerPage) {
                if {
                    ShopItemButtonSlot := text_button_base[ShopItemButtonSlotList[ItemIndex].Widget]
                    ShopItemImageSlot := texture_block[ShopItemImageSlotList[ItemIndex].Widget]
                } then {
                    ShopItemButtonSlot.SetVisibility(widget_visibility.Hidden)
                    ShopItemImageSlot.SetVisibility(widget_visibility.Hidden)
                }
            }
            #>

            # Show current slots
            for(ItemSlotIndex := 0 .. ShopManager.ShopItemsPerPage - 1) {
                if {
                    ShopItemButtonSlot := text_button_base[ShopItemButtonSlotList[ItemSlotIndex].Widget]
                    ShopItemImageSlot := texture_block[ShopItemImageSlotList[ItemSlotIndex].Widget]
                } then {
                    # Check if we have a config for this item
                    ItemIndex := ItemSlotIndex + CurShopScrollIndex
                    if {
                        ItemPrice := ShopManager.PriceList[ItemIndex]
                        ItemImage := ImageList[ItemIndex]
                    } then {
                        ShopItemButtonSlot.SetVisibility(widget_visibility.Visible)
                        ShopItemImageSlot.SetVisibility(widget_visibility.Visible)

                        # Shop item image
                        ShopItemImageSlot.SetImage(ItemImage)

                        # Check for overwrite with custom button text
                        var ButtonText: string = ShopManager.GetButtonText(Agent, ItemIndex)
                        if(ButtonText.Length = 0) {
                            # Default text
                            set ButtonText = ToString(ItemPrice) + ShopManager.ShopItemPriceSuffix
                        }
                        ShopItemButtonSlot.SetText(Utility.FormatMessage(ButtonText))
                    } else {
                        ShopItemButtonSlot.SetVisibility(widget_visibility.Hidden)
                        ShopItemImageSlot.SetVisibility(widget_visibility.Hidden)
                    }
                    
                }
            }

            # Call event for parent script
            ShopManager.HandleShopUiRefresh(Agent)
        }
    }

    OnArrowButtonClick(WidgetMessage: widget_message, IncreasePage: logic): void = {
        Print("OnArrowButtonClick")
        if(ShopManager := MaybeShopManager?) {
            Agent := WidgetMessage.Player
            Print("CurShopScrollIndex: {CurShopScrollIndex}")
            MaxScrollIndex := CurShopScrollIndex + ShopManager.ShopItemsPerPage - 1
            PrevShopScrollIndex := CurShopScrollIndex
            if(IncreasePage = true) {
                Print("IncreasePage")
                # if(CurShopScrollIndex < PriceList.Length - 1) {
                if(MaxScrollIndex < ShopManager.PriceList.Length - 1) {
                    set CurShopScrollIndex = CurShopScrollIndex + ShopManager.ShopItemsPerPage
                }
            } else {
                Print("DecreasePage")
                if(CurShopScrollIndex > 0) {
                    set CurShopScrollIndex = CurShopScrollIndex - ShopManager.ShopItemsPerPage
                }
            }

            # Check if the index has changed
            Print("CurShopScrollIndex: {CurShopScrollIndex}")
            if(PrevShopScrollIndex <> CurShopScrollIndex) {
                # Build new UI
                RefreshPlayerShopUi(Agent)
            }
        }
    }

    OnItemClicked<override>(WidgetMessage: widget_message, ButtonIndex: int): void = {
        ItemIndex := ButtonIndex + CurShopScrollIndex
        Print("OnItemClicked - ItemIndex: {ItemIndex}")
        if(ShopManager := MaybeShopManager?) {
            Agent := WidgetMessage.Player
            if(ItemPrice := ShopManager.PriceList[ItemIndex]) {
                ShopManager.TryToPurchaseItem(Agent, ItemIndex, ItemPrice)
            } else {
                Print("ERR: OnItemClicked - Found no ItemPrice for ItemIndex: {ItemIndex}")
            }
        }
    }

    OnExitButtonClicked<override>(WidgetMessage: widget_message): void = {
        Agent := WidgetMessage.Player
        CloseShop(Agent)
    }

    CloseShop<public>(Agent: agent): void = {
        Print("CloseShop UI")
        GameManager := GetGameManager()
        set CurShopScrollIndex = 0
        RemoveUiFromPlayer(Agent)
        # Check if the current quest is fulfilled
        # GameManager.CheckIfQuestItemsFulfilled(Agent)
        if(ShopManager := MaybeShopManager?) {
            ShopManager.OnShopClosed(Agent)
        }
    }

    CleanUpUiOnRemove<override>(): void = {
        (super:)CleanUpUiOnRemove()
        if {
            Agent := MaybeAgent?
            ShopManager := MaybeShopManager?
        } then {
            ShopManager.OnShopClosed(Agent)
        }
    }
}