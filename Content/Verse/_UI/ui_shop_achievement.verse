using { /Verse.org/Simulation }
using { /Verse.org/Colors }
using { /Verse.org/Colors/NamedColors }
# using { /Verse.org/Concurrency }
# using { /Verse.org/Verse }
using { /Verse.org/Assets }
using { /Fortnite.com/Devices }
using { /Fortnite.com/UI }
# using { /Fortnite.com/Characters }
# using { /UnrealEngine.com/Temporary/Diagnostics }
using { /UnrealEngine.com/Temporary/SpatialMath }
using { /UnrealEngine.com/Temporary/UI}

ui_shop_achievement<public> := class(ui_shop) {
    # Runtime
    var ItemObjectiveTextList: []text_block = array {}
    var ItemObjectiveCounterList: []text_block = array {}
    var ProgressBarList: []color_block = array {}
    var ProgressBarBackgroundList: []color_block = array {}

    CreateSlots<override>(Agent: agent): []canvas_slot = {
        Utility := GetUtility()
        # Get common shop slots
        # var Slots: []canvas_slot = (super:)CreateSlots(Agent)
        var Slots: []canvas_slot = array {}

        if(ShopManager := MaybeShopManager?) {
            CloseShopButton := button_loud{DefaultText := Utility.FormatMessage("EXIT")}
            CloseShopButton.OnClick().Subscribe(OnExitButtonClicked)

            set Slots = Slots + array {
                # Close shop button
                canvas_slot {
                    Anchors := anchors{Minimum := vector2{X := 0.04125, Y := 0.945556}, Maximum := vector2{X := 0.98125, Y := 1.0}}
                    Offsets := margin{Left := 8.036049, Top := 5.777832, Right := 4.128906, Bottom := 9.117065}
                    Alignment := vector2{X := 0.0, Y := 0.0}
                    SizeToContent := false
                    Widget := CloseShopButton
                    ZOrder := 901
                }
            }
        }

        <#
        Achievement_Title_1 := text_block{DefaultText := Utility.FormatMessage("TERMINATOR 1")
        DefaultJustification := text_justification.Left
        DefaultTextColor := color{R:= 1.0, G:= 0.653278, B:= 0.0647}}
       
        Achievement_ObjectiveText_1 := text_block{DefaultText := Utility.FormatMessage("Eliminate 10 enemies")
        DefaultJustification := text_justification.Left
        DefaultTextColor := color{R:= 0.609375, G:= 0.551962, B:= 0.520962}}
        Achievement_ObjectiveText_2 := text_block{DefaultText := Utility.FormatMessage("Eliminate 10 enemies")
        DefaultJustification := text_justification.Left
        DefaultTextColor := color{R:= 0.609375, G:= 0.551962, B:= 0.520962}}

        Achievement_Title_2 := text_block{DefaultText := Utility.FormatMessage("Eliminate NPCs in the Battle Area!")
        DefaultJustification := text_justification.Left
        DefaultTextColor := color{R:= 1.0, G:= 0.653278, B:= 0.0647}}
        Achievement_Title_3 := text_block{DefaultText := Utility.FormatMessage("Eliminate NPCs in the Battle Area!")
        DefaultJustification := text_justification.Left
        DefaultTextColor := color{R:= 1.0, G:= 0.653278, B:= 0.0647}}
        
        Achievement_ObjectiveText_3 := text_block{DefaultText := Utility.FormatMessage("Eliminate 10 enemies1")
        DefaultJustification := text_justification.Left
        DefaultTextColor := color{R:= 0.609375, G:= 0.551962, B:= 0.520962}}
        Achievement_ObjectiveText_4 := text_block{DefaultText := Utility.FormatMessage("Eliminate 10 enemies")
        DefaultJustification := text_justification.Left
        DefaultTextColor := color{R:= 0.609375, G:= 0.551962, B:= 0.520962}}
        
        Achievement_Title_4 := text_block{DefaultText := Utility.FormatMessage("Eliminate NPCs in the Battle Area!")
        DefaultJustification := text_justification.Left
        DefaultTextColor := color{R:= 1.0, G:= 0.653278, B:= 0.0647}}
        Achievement_Title_5 := text_block{DefaultText := Utility.FormatMessage("Eliminate NPCs in the Battle Area!")
        DefaultJustification := text_justification.Left
        DefaultTextColor := color{R:= 1.0, G:= 0.653278, B:= 0.0647}}
        
        Achievement_ObjectiveText_5 := text_block{DefaultText := Utility.FormatMessage("Eliminate 10 enemies")
        DefaultJustification := text_justification.Left
        DefaultTextColor := color{R:= 0.609375, G:= 0.551962, B:= 0.520962}}
        Achievement_ObjectiveText_10 := text_block{DefaultText := Utility.FormatMessage("Eliminate 10 enemies")
        DefaultJustification := text_justification.Left
        DefaultTextColor := color{R:= 0.609375, G:= 0.551962, B:= 0.520962}}
        
        Achievement_Title_10 := text_block{DefaultText := Utility.FormatMessage("Eliminate NPCs in the Battle Area!")
        DefaultJustification := text_justification.Left
        DefaultTextColor := color{R:= 1.0, G:= 0.653278, B:= 0.0647}}
        Achievement_Title_9 := text_block{DefaultText := Utility.FormatMessage("Eliminate NPCs in the Battle Area!")
        DefaultJustification := text_justification.Left
        DefaultTextColor := color{R:= 1.0, G:= 0.653278, B:= 0.0647}}
       
        Achievement_ObjectiveText_9 := text_block{DefaultText := Utility.FormatMessage("Eliminate 10 enemies")
        DefaultJustification := text_justification.Left
        DefaultTextColor := color{R:= 0.609375, G:= 0.551962, B:= 0.520962}}
        Achievement_ObjectiveText_8 := text_block{DefaultText := Utility.FormatMessage("Eliminate 10 enemies")
        DefaultJustification := text_justification.Left
        DefaultTextColor := color{R:= 0.609375, G:= 0.551962, B:= 0.520962}}
       
        Achievement_Title_8 := text_block{DefaultText := Utility.FormatMessage("Eliminate NPCs in the Battle Area!")
        DefaultJustification := text_justification.Left
        DefaultTextColor := color{R:= 1.0, G:= 0.653278, B:= 0.0647}}
        Achievement_Title_7 := text_block{DefaultText := Utility.FormatMessage("Eliminate NPCs in the Battle Area!")
        DefaultJustification := text_justification.Left
        DefaultTextColor := color{R:= 1.0, G:= 0.653278, B:= 0.0647}}
       
        Achievement_ObjectiveText_7 := text_block{DefaultText := Utility.FormatMessage("Eliminate 10 enemies")
        DefaultJustification := text_justification.Left
        DefaultTextColor := color{R:= 0.609375, G:= 0.551962, B:= 0.520962}}
        Achievement_ObjectiveText_6 := text_block{DefaultText := Utility.FormatMessage("Eliminate 10 enemies")
        DefaultJustification := text_justification.Left
        DefaultTextColor := color{R:= 0.609375, G:= 0.551962, B:= 0.520962}}
       
        Achievement_Title_6 := text_block{DefaultText := Utility.FormatMessage("TERMINATOR 1")
        DefaultJustification := text_justification.Left
        DefaultTextColor := color{R:= 1.0, G:= 0.653278, B:= 0.0647}}
        #>

        # DefaultImage is set in GameConfig, not here!
        Achievement_ProgressBar_1: color_block = color_block {
            DefaultColor := color{R:= 0.066964, G:= 0.390625, B:= 0.0}
            DefaultDesiredSize := vector2{X := 525.271423, Y := 36.303986}
        }
        Achievement_ProgressBar_2: color_block = color_block {
            DefaultColor := color{R:= 0.066964, G:= 0.390625, B:= 0.0}
            DefaultDesiredSize := vector2{X := 525.271423, Y := 36.303986}
        }
        Achievement_ProgressBar_3: color_block = color_block {
            DefaultColor := color{R:= 0.066964, G:= 0.390625, B:= 0.0}
            DefaultDesiredSize := vector2{X := 525.271423, Y := 36.303986}
        }
        Achievement_ProgressBar_4: color_block = color_block {
            DefaultColor := color{R:= 0.066964, G:= 0.390625, B:= 0.0}
            DefaultDesiredSize := vector2{X := 525.271423, Y := 36.303986}
        }
        Achievement_ProgressBar_5: color_block = color_block {
            DefaultColor := color{R:= 0.066964, G:= 0.390625, B:= 0.0}
            DefaultDesiredSize := vector2{X := 516.782166, Y := 36.303986}
        }
        Achievement_ProgressBar_6: color_block = color_block {
            DefaultColor := color{R:= 0.066964, G:= 0.390625, B:= 0.0}
            DefaultDesiredSize := vector2{X := 525.271423, Y := 36.303986}
        }
        Achievement_ProgressBar_7: color_block = color_block {
            DefaultColor := color{R:= 0.066964, G:= 0.390625, B:= 0.0}
            DefaultDesiredSize := vector2{X := 525.271423, Y := 36.303986}
        }
        Achievement_ProgressBar_8: color_block = color_block {
            DefaultColor := color{R:= 0.066964, G:= 0.390625, B:= 0.0}
            DefaultDesiredSize := vector2{X := 525.271423, Y := 36.304016}
        }
        Achievement_ProgressBar_9: color_block = color_block {
            DefaultColor := color{R:= 0.066964, G:= 0.390625, B:= 0.0}
            DefaultDesiredSize := vector2{X := 524.827759, Y := 36.304016}
        }
        Achievement_ProgressBar_10: color_block = color_block {
            DefaultColor := color{R:= 0.066964, G:= 0.390625, B:= 0.0}
            DefaultDesiredSize := vector2{X := 525.271423, Y := 36.304016}
        }
        <#
        Achievement_Icon_6: texture_block = texture_block {
            DefaultImage := UI.bonuscoins_icons
            DefaultDesiredSize := vector2{X := 1.840454, Y := 0.785583}
        }
        Achievement_Icon_7: texture_block = texture_block {
            DefaultImage := UI.bonuscoins_icons
            DefaultDesiredSize := vector2{X := 0.897583, Y := 0.953796}
        }
        Achievement_Icon_1: texture_block = texture_block {
            DefaultImage := UI.bonuscoins_icons
            DefaultDesiredSize := vector2{X := 0.583923, Y := 0.635773}
        }
        Achievement_Icon_2: texture_block = texture_block {
            DefaultImage := UI.bonuscoins_icons
            DefaultDesiredSize := vector2{X := -0.304535, Y := 3.435153}
        }
        Achievement_Icon_3: texture_block = texture_block {
            DefaultImage := UI.bonuscoins_icons
            DefaultDesiredSize := vector2{X := -0.927413, Y := 4.152405}
        }
        Achievement_Icon_4: texture_block = texture_block {
            DefaultImage := UI.bonuscoins_icons
            DefaultDesiredSize := vector2{X := -1.138748, Y := 4.443663}
        }
        Achievement_Icon_5: texture_block = texture_block {
            DefaultImage := UI.bonuscoins_icons
            DefaultDesiredSize := vector2{X := 3.438828, Y := 3.51416}
        }
        Achievement_Icon_8: texture_block = texture_block {
            DefaultImage := UI.bonuscoins_icons
            DefaultDesiredSize := vector2{X := 2.398921, Y := 0.52143}
        }
        Achievement_Icon_9: texture_block = texture_block {
            DefaultImage := UI.bonuscoins_icons
            DefaultDesiredSize := vector2{X := 1.927856, Y := 0.936829}
        }
        Achievement_Icon_10: texture_block = texture_block {
            DefaultImage := UI.bonuscoins_icons
            DefaultDesiredSize := vector2{X := -0.80371, Y := 3.310181}
        }
        #>
        Achievement_ProgressBarBackground_1: color_block = color_block {
            DefaultColor := color{R:= 0.005208, G:= 0.005208, B:= 0.005208}
            DefaultDesiredSize := vector2{X := 1.203247, Y := 0.020325}
        }
        Achievement_ProgressBarBackground_2: color_block = color_block {
            DefaultColor := color{R:= 0.005208, G:= 0.005208, B:= 0.005208}
            DefaultDesiredSize := vector2{X := -0.663818, Y := 1.904541}
        }
        Achievement_ProgressBarBackground_3: color_block = color_block {
            DefaultColor := color{R:= 0.005208, G:= 0.005208, B:= 0.005208}
            DefaultDesiredSize := vector2{X := 0.902161, Y := 0.356628}
        }
        Achievement_ProgressBarBackground_4: color_block = color_block {
            DefaultColor := color{R:= 0.005208, G:= 0.005208, B:= 0.005208}
            DefaultDesiredSize := vector2{X := 525.271423, Y := 36.304016}
        }
        Achievement_ProgressBarBackground_5: color_block = color_block {
            DefaultColor := color{R:= 0.005208, G:= 0.005208, B:= 0.005208}
            DefaultDesiredSize := vector2{X := 1.127747, Y := 2.530762}
        }
        Achievement_ProgressBarBackground_6: color_block = color_block {
            DefaultColor := color{R:= 0.005208, G:= 0.005208, B:= 0.005208}
            DefaultDesiredSize := vector2{X := 525.271423, Y := 36.303986}
        }
        Achievement_ProgressBarBackground_7: color_block = color_block {
            DefaultColor := color{R:= 0.005208, G:= 0.005208, B:= 0.005208}
            DefaultDesiredSize := vector2{X := 525.271423, Y := 36.303986}
        }
        Achievement_ProgressBarBackground_8: color_block = color_block {
            DefaultColor := color{R:= 0.005208, G:= 0.005208, B:= 0.005208}
            DefaultDesiredSize := vector2{X := 525.271423, Y := 36.304016}
        }
        Achievement_ProgressBarBackground_9: color_block = color_block {
            DefaultColor := color{R:= 0.005208, G:= 0.005208, B:= 0.005208}

            DefaultDesiredSize := vector2{X := 524.827759, Y := 36.304016}
        }
        Achievement_ProgressBarBackground_10: color_block = color_block {
            DefaultColor := color{R:= 0.005208, G:= 0.005208, B:= 0.005208}
            DefaultDesiredSize := vector2{X := 525.271423, Y := 36.304016}
        }

        Achievement_ObjectiveText_1 := text_block{
            DefaultText := Utility.FormatMessage("Kill 50 NPCs"),
            DefaultJustification := text_justification.Left,
            DefaultTextColor := color{R := 0.609375, G := 0.551962, B := 0.520962}
        }

        Achievement_ObjectiveText_2 := text_block{
            DefaultText := Utility.FormatMessage("Kill 50 NPCs"),
            DefaultJustification := text_justification.Left,
            DefaultTextColor := color{R := 0.609375, G := 0.551962, B := 0.520962}
        }

        Achievement_ObjectiveText_3 := text_block{
            DefaultText := Utility.FormatMessage("Kill 50 NPCs"),
            DefaultJustification := text_justification.Left,
            DefaultTextColor := color{R := 0.609375, G := 0.551962, B := 0.520962}
        }

        Achievement_ObjectiveText_4 := text_block{
            DefaultText := Utility.FormatMessage("Kill 50 NPCs"),
            DefaultJustification := text_justification.Left,
            DefaultTextColor := color{R := 0.609375, G := 0.551962, B := 0.520962}
        }

        Achievement_ObjectiveText_5 := text_block{
            DefaultText := Utility.FormatMessage("Kill 50 NPCs"),
            DefaultJustification := text_justification.Left,
            DefaultTextColor := color{R := 0.609375, G := 0.551962, B := 0.520962}
        }

        Achievement_ObjectiveText_6 := text_block{
            DefaultText := Utility.FormatMessage("Kill 50 NPCs"),
            DefaultJustification := text_justification.Left,
            DefaultTextColor := color{R := 0.609375, G := 0.551962, B := 0.520962}
        }

        Achievement_ObjectiveText_7 := text_block{
            DefaultText := Utility.FormatMessage("Kill 50 NPCs"),
            DefaultJustification := text_justification.Left,
            DefaultTextColor := color{R := 0.609375, G := 0.551962, B := 0.520962}
        }

        Achievement_ObjectiveText_8 := text_block{
            DefaultText := Utility.FormatMessage("Kill 50 NPCs"),
            DefaultJustification := text_justification.Left,
            DefaultTextColor := color{R := 0.609375, G := 0.551962, B := 0.520962}
        }

        Achievement_ObjectiveText_9 := text_block{
            DefaultText := Utility.FormatMessage("Kill 50 NPCs"),
            DefaultJustification := text_justification.Left,
            DefaultTextColor := color{R := 0.609375, G := 0.551962, B := 0.520962}
        }

        Achievement_ObjectiveText_10 := text_block{
            DefaultText := Utility.FormatMessage("Kill 50 NPCs"),
            DefaultJustification := text_justification.Right,
            DefaultTextColor := color{R := 0.609375, G := 0.551962, B := 0.520962}
        }

        Achievement_ObjectiveCounter_1 := text_block{
            DefaultText := Utility.FormatMessage("0/10"),
            DefaultJustification := text_justification.Right,
            DefaultTextColor := color{R := 0.609375, G := 0.551962, B := 0.520962}
        }

        Achievement_ObjectiveCounter_2 := text_block{
            DefaultText := Utility.FormatMessage("0/10"),
            DefaultJustification := text_justification.Right,
            DefaultTextColor := color{R := 0.609375, G := 0.551962, B := 0.520962}
        }

        Achievement_ObjectiveCounter_3 := text_block{
            DefaultText := Utility.FormatMessage("0/10"),
            DefaultJustification := text_justification.Right,
            DefaultTextColor := color{R := 0.609375, G := 0.551962, B := 0.520962}
        }

        Achievement_ObjectiveCounter_4 := text_block{
            DefaultText := Utility.FormatMessage("0/10"),
            DefaultJustification := text_justification.Right,
            DefaultTextColor := color{R := 0.609375, G := 0.551962, B := 0.520962}
        }

        Achievement_ObjectiveCounter_5 := text_block{
            DefaultText := Utility.FormatMessage("0/10"),
            DefaultJustification := text_justification.Right,
            DefaultTextColor := color{R := 0.609375, G := 0.551962, B := 0.520962}
        }

        Achievement_ObjectiveCounter_6 := text_block{
            DefaultText := Utility.FormatMessage("0/10"),
            DefaultJustification := text_justification.Right,
            DefaultTextColor := color{R := 0.609375, G := 0.551962, B := 0.520962}
        }

        Achievement_ObjectiveCounter_7 := text_block{
            DefaultText := Utility.FormatMessage("0/10"),
            DefaultJustification := text_justification.Right,
            DefaultTextColor := color{R := 0.609375, G := 0.551962, B := 0.520962}
        }

        Achievement_ObjectiveCounter_8 := text_block{
            DefaultText := Utility.FormatMessage("0/10"),
            DefaultJustification := text_justification.Right,
            DefaultTextColor := color{R := 0.609375, G := 0.551962, B := 0.520962}
        }

        Achievement_ObjectiveCounter_9 := text_block{
            DefaultText := Utility.FormatMessage("0/10"),
            DefaultJustification := text_justification.Right,
            DefaultTextColor := color{R := 0.609375, G := 0.551962, B := 0.520962}
        }

        Achievement_ObjectiveCounter_10 := text_block{
            DefaultText := Utility.FormatMessage("0/10"),
            DefaultJustification := text_justification.Right,
            DefaultTextColor := color{R := 0.609375, G := 0.551962, B := 0.520962}
        }

        <#
        set ItemNameList = array {
            Achievement_Title_1,
            Achievement_Title_2,
            Achievement_Title_3,
            Achievement_Title_4,
            Achievement_Title_5,
            Achievement_Title_6,
            Achievement_Title_7,
            Achievement_Title_8,
            Achievement_Title_9,
            Achievement_Title_10
        }

        set ItemIconList = array {
            Achievement_Icon_1,
            Achievement_Icon_2,
            Achievement_Icon_3,
            Achievement_Icon_4,
            Achievement_Icon_5,
            Achievement_Icon_6,
            Achievement_Icon_7,
            Achievement_Icon_8,
            Achievement_Icon_9,
            Achievement_Icon_10
        }
        #>

        set ItemObjectiveTextList = array {
            Achievement_ObjectiveText_1,
            Achievement_ObjectiveText_2,
            Achievement_ObjectiveText_3,
            Achievement_ObjectiveText_4,
            Achievement_ObjectiveText_5,
            Achievement_ObjectiveText_6,
            Achievement_ObjectiveText_7,
            Achievement_ObjectiveText_8,
            Achievement_ObjectiveText_9,
            Achievement_ObjectiveText_10
        }

        set ItemObjectiveCounterList = array {
            Achievement_ObjectiveCounter_1,
            Achievement_ObjectiveCounter_2,
            Achievement_ObjectiveCounter_3,
            Achievement_ObjectiveCounter_4,
            Achievement_ObjectiveCounter_5,
            Achievement_ObjectiveCounter_6,
            Achievement_ObjectiveCounter_7,
            Achievement_ObjectiveCounter_8,
            Achievement_ObjectiveCounter_9,
            Achievement_ObjectiveCounter_10
        }

        set ProgressBarList = array {
            Achievement_ProgressBar_1,
            Achievement_ProgressBar_2,
            Achievement_ProgressBar_3,
            Achievement_ProgressBar_4,
            Achievement_ProgressBar_5,
            Achievement_ProgressBar_6,
            Achievement_ProgressBar_7,
            Achievement_ProgressBar_8,
            Achievement_ProgressBar_9,
            Achievement_ProgressBar_10
        }

        set ProgressBarBackgroundList = array{
            Achievement_ProgressBarBackground_1,
            Achievement_ProgressBarBackground_2,
            Achievement_ProgressBarBackground_3,
            Achievement_ProgressBarBackground_4,
            Achievement_ProgressBarBackground_5,
            Achievement_ProgressBarBackground_6,
            Achievement_ProgressBarBackground_7,
            Achievement_ProgressBarBackground_8,
            Achievement_ProgressBarBackground_9,
            Achievement_ProgressBarBackground_10
        }

        # Make slots
        set Slots = Slots + array {
            canvas_slot {
                Anchors := anchors{Minimum := vector2{X := 0.2, Y := 0.178571}, Maximum := vector2{X := 0.311161, Y := 0.221429}}
                Offsets := margin{Left := -0.384369, Top := 2.950272, Right := -2.898132, Bottom := 2.381775}
                Alignment := vector2{X := 0.0, Y := 0.0}
                SizeToContent := false
                Widget := Achievement_ObjectiveText_1
                ZOrder := 900
            },
            canvas_slot {
                Anchors := anchors{Minimum := vector2{X := 0.2, Y := 0.336806}, Maximum := vector2{X := 0.476562, Y := 0.380556}}
                Offsets := margin{Left := 3.615631, Top := 3.885864, Right := 1.954956, Bottom := 2.411377}
                Alignment := vector2{X := 0.0, Y := 0.0}
                SizeToContent := false
                Widget := Achievement_ObjectiveText_2
                ZOrder := 900
            },
            canvas_slot {
                Anchors := anchors{Minimum := vector2{X := 0.2, Y := 0.489815}, Maximum := vector2{X := 0.476042, Y := 0.535185}}
                Offsets := margin{Left := 3.615631, Top := 6.470459, Right := 0.953979, Bottom := 1.578552}
                Alignment := vector2{X := 0.0, Y := 0.0}
                SizeToContent := false
                Widget := Achievement_ObjectiveText_3
                ZOrder := 900
            },
            canvas_slot {
                Anchors := anchors{Minimum := vector2{X := 0.2, Y := 0.652381}, Maximum := vector2{X := 0.477679, Y := 0.7}}
                Offsets := margin{Left := 7.615631, Top := 2.723328, Right := 0.099976, Bottom := 7.756775}
                Alignment := vector2{X := 0.0, Y := 0.0}
                SizeToContent := false
                Widget := Achievement_ObjectiveText_4
                ZOrder := 900
            },
            canvas_slot {
                Anchors := anchors{Minimum := vector2{X := 0.2, Y := 0.810494}, Maximum := vector2{X := 0.479861, Y := 0.851111}}
                Offsets := margin{Left := 7.615631, Top := -0.209595, Right := 4.294678, Bottom := 3.120117}
                Alignment := vector2{X := 0.0, Y := 0.0}
                SizeToContent := false
                Widget := Achievement_ObjectiveText_5
                ZOrder := 900
            },
            canvas_slot {
                Anchors := anchors{Minimum := vector2{X := 0.660227, Y := 0.179237}, Maximum := vector2{X := 0.775253, Y := 0.220988}}
                Offsets := margin{Left := 3.094849, Top := 2.230392, Right := 1.050049, Bottom := 1.905701}
                Alignment := vector2{X := 0.0, Y := 0.0}
                SizeToContent := false
                Widget := Achievement_ObjectiveText_6
                ZOrder := 900
            },
            canvas_slot {
                Anchors := anchors{Minimum := vector2{X := 0.660764, Y := 0.338272}, Maximum := vector2{X := 0.936111, Y := 0.379012}}
                Offsets := margin{Left := 2.063599, Top := 2.301025, Right := 1.171509, Bottom := 0.743042}
                Alignment := vector2{X := 0.0, Y := 0.0}
                SizeToContent := false
                Widget := Achievement_ObjectiveText_7
                ZOrder := 900
            },
            canvas_slot {
                Anchors := anchors{Minimum := vector2{X := 0.661607, Y := 0.5}, Maximum := vector2{X := 0.938839, Y := 0.534921}}
                Offsets := margin{Left := 4.442749, Top := -4.540527, Right := 2.414795, Bottom := 1.292603}
                Alignment := vector2{X := 0.0, Y := 0.0}
                SizeToContent := false
                Widget := Achievement_ObjectiveText_8
                ZOrder := 900
            },
            canvas_slot {
                Anchors := anchors{Minimum := vector2{X := 0.661607, Y := 0.651587}, Maximum := vector2{X := 0.936161, Y := 0.7}}
                Offsets := margin{Left := 4.442993, Top := 3.581665, Right := -2.733398, Bottom := 7.75647}
                Alignment := vector2{X := 0.0, Y := 0.0}
                SizeToContent := false
                Widget := Achievement_ObjectiveText_9
                ZOrder := 900
            },
            canvas_slot {
                Anchors := anchors{Minimum := vector2{X := 0.660417, Y := 0.811111}, Maximum := vector2{X := 0.935417, Y := 0.848148}}
                Offsets := margin{Left := 2.730713, Top := -0.876892, Right := -0.16333, Bottom := -0.08313}
                Alignment := vector2{X := 0.0, Y := 0.0}
                SizeToContent := false
                Widget := Achievement_ObjectiveText_10
                ZOrder := 900
            },
            canvas_slot {
                Anchors := anchors{Minimum := vector2{X := 0.432589, Y := 0.218254}, Maximum := vector2{X := 0.476339, Y := 0.259524}}
                Offsets := margin{Left := 4.597168, Top := 4.049713, Right := 2.897034, Bottom := -0.433746}
                Alignment := vector2{X := 0.0, Y := 0.0}
                SizeToContent := false
                Widget := Achievement_ObjectiveCounter_1
                ZOrder := 900
            },
            canvas_slot {
                Anchors := anchors{Minimum := vector2{X := 0.435547, Y := 0.375}, Maximum := vector2{X := 0.478516, Y := 0.416667}}
                Offsets := margin{Left := -1.08728, Top := 2.594604, Right := 7.079895, Bottom := 1.450438}
                Alignment := vector2{X := 0.0, Y := 0.0}
                SizeToContent := false
                Widget := Achievement_ObjectiveCounter_2
                ZOrder := 900
            },
            canvas_slot {
                Anchors := anchors{Minimum := vector2{X := 0.434375, Y := 0.532099}, Maximum := vector2{X := 0.474306, Y := 0.571605}}
                Offsets := margin{Left := 1.165222, Top := 0.758118, Right := -1.011658, Bottom := 0.951233}
                Alignment := vector2{X := 0.0, Y := 0.0}
                SizeToContent := false
                Widget := Achievement_ObjectiveCounter_3
                ZOrder := 900
            },
            canvas_slot {
                Anchors := anchors{Minimum := vector2{X := 0.432812, Y := 0.688889}, Maximum := vector2{X := 0.475, Y := 0.731944}}
                Offsets := margin{Left := 4.168213, Top := 3.255249, Right := 0.323059, Bottom := 2.291321}
                Alignment := vector2{X := 0.0, Y := 0.0}
                SizeToContent := false
                Widget := Achievement_ObjectiveCounter_4
                ZOrder := 900
            },
            canvas_slot {
                Anchors := anchors{Minimum := vector2{X := 0.432813, Y := 0.84537}, Maximum := vector2{X := 0.475521, Y := 0.887037}}
                Offsets := margin{Left := 4.168091, Top := 2.086182, Right := 1.324036, Bottom := 1.958984}
                Alignment := vector2{X := 0.0, Y := 0.0}
                SizeToContent := false
                Widget := Achievement_ObjectiveCounter_5
                ZOrder := 900
            },
            canvas_slot {
                Anchors := anchors{Minimum := vector2{X := 0.9, Y := 0.843519}, Maximum := vector2{X := 0.936979, Y := 0.887963}}
                Offsets := margin{Left := -5.729736, Top := 4.088074, Right := 0.21106, Bottom := 2.9599}
                Alignment := vector2{X := 0.0, Y := 0.0}
                SizeToContent := false
                Widget := Achievement_ObjectiveCounter_10
                ZOrder := 900
            },
            canvas_slot {
                Anchors := anchors{Minimum := vector2{X := 0.9, Y := 0.689815}, Maximum := vector2{X := 0.9375, Y := 0.730556}}
                Offsets := margin{Left := -5.729736, Top := 2.254211, Right := 1.211914, Bottom := 0.789795}
                Alignment := vector2{X := 0.0, Y := 0.0}
                SizeToContent := false
                Widget := Achievement_ObjectiveCounter_9
                ZOrder := 900
            },
            canvas_slot {
                Anchors := anchors{Minimum := vector2{X := 0.5, Y := 0.5}, Maximum := vector2{X := 0.940179, Y := 0.573016}}
                Offsets := margin{Left := 759.039062, Top := 35.459473, Right := 10.359863, Bottom := 2.476685}
                Alignment := vector2{X := 0.0, Y := 0.0}
                SizeToContent := false
                Widget := Achievement_ObjectiveCounter_8
                ZOrder := 900
            },
            canvas_slot {
                Anchors := anchors{Minimum := vector2{X := 0.888462, Y := 0.373016}, Maximum := vector2{X := 0.936161, Y := 0.417094}}
                Offsets := margin{Left := 12.446289, Top := 4.739594, Right := 2.637939, Bottom := 1.912445}
                Alignment := vector2{X := 0.0, Y := 0.0}
                SizeToContent := false
                Widget := Achievement_ObjectiveCounter_7
                ZOrder := 900
            },
            canvas_slot {
                Anchors := anchors{Minimum := vector2{X := 0.889732, Y := 0.218254}, Maximum := vector2{X := 0.937054, Y := 0.25873}}
                Offsets := margin{Left := 14.004395, Top := 4.049805, Right := 0.354004, Bottom := -1.291748}
                Alignment := vector2{X := 0.0, Y := 0.0}
                SizeToContent := false
                Widget := Achievement_ObjectiveCounter_6
                ZOrder := 900
            },
            canvas_slot {
                Anchors := anchors{Minimum := vector2{X := 0.5, Y := 0.5}, Maximum := vector2{X := 0.5, Y := 0.5}}
                Offsets := margin{Top := -252.540527, Left := -592.960938, Right := 525.271423, Bottom := 36.304016}
                Alignment := vector2{X := 0.0, Y := 0.0}
                SizeToContent := true
                Widget := Achievement_ProgressBar_1
                ZOrder := 901
            },
            canvas_slot {
                Anchors := anchors{Minimum := vector2{X := 0.5, Y := 0.5}, Maximum := vector2{X := 0.5, Y := 0.5}}
                Offsets := margin{Top := -84.540527, Left := -592.960938, Right := 525.271423, Bottom := 36.304016}
                Alignment := vector2{X := 0.0, Y := 0.0}
                SizeToContent := true
                Widget := Achievement_ProgressBar_2
                ZOrder := 901
            },
            canvas_slot {
                Anchors := anchors{Minimum := vector2{X := 0.5, Y := 0.5}, Maximum := vector2{X := 0.5, Y := 0.5}}
                Offsets := margin{Top := 83.459473, Left := -592.960938, Right := 525.271423, Bottom := 36.304016}
                Alignment := vector2{X := 0.0, Y := 0.0}
                SizeToContent := true
                Widget := Achievement_ProgressBar_3
                ZOrder := 901
            },
            canvas_slot {
                Anchors := anchors{Minimum := vector2{X := 0.5, Y := 0.5}, Maximum := vector2{X := 0.5, Y := 0.5}}
                Offsets := margin{Top := 255.459473, Left := -592.960938, Right := 525.271423, Bottom := 36.304016}
                Alignment := vector2{X := 0.0, Y := 0.0}
                SizeToContent := true
                Widget := Achievement_ProgressBar_4
                ZOrder := 901
            },
            canvas_slot {
                Anchors := anchors{Minimum := vector2{X := 0.5, Y := 0.5}, Maximum := vector2{X := 0.5, Y := 0.5}}
                Offsets := margin{Top := 423.459473, Left := -592.960938, Right := 516.782166, Bottom := 36.303986}
                Alignment := vector2{X := 0.0, Y := 0.0}
                SizeToContent := true
                Widget := Achievement_ProgressBar_5
                ZOrder := 901
            },
            canvas_slot {
                Anchors := anchors{Minimum := vector2{X := 0.5, Y := 0.5}, Maximum := vector2{X := 0.5, Y := 0.5}}
                Offsets := margin{Top := -252.540527, Left := 318.400818, Right := 525.271423, Bottom := 36.303986}
                Alignment := vector2{X := 0.0, Y := 0.0}
                SizeToContent := true
                Widget := Achievement_ProgressBar_6
                ZOrder := 901
            },
            canvas_slot {
                Anchors := anchors{Minimum := vector2{X := 0.5, Y := 0.5}, Maximum := vector2{X := 0.5, Y := 0.5}}
                Offsets := margin{Top := -84.540527, Left := 318.400818, Right := 525.271423, Bottom := 36.303986}
                Alignment := vector2{X := 0.0, Y := 0.0}
                SizeToContent := true
                Widget := Achievement_ProgressBar_7
                ZOrder := 901
            },
            canvas_slot {
                Anchors := anchors{Minimum := vector2{X := 0.5, Y := 0.5}, Maximum := vector2{X := 0.5, Y := 0.5}}
                Offsets := margin{Top := 83.459473, Left := 318.400818, Right := 525.271423, Bottom := 36.304016}
                Alignment := vector2{X := 0.0, Y := 0.0}
                SizeToContent := true
                Widget := Achievement_ProgressBar_8
                ZOrder := 901
            },
            canvas_slot {
                Anchors := anchors{Minimum := vector2{X := 0.5, Y := 0.5}, Maximum := vector2{X := 0.5, Y := 0.5}}
                Offsets := margin{Top := 255.459473, Left := 318.400818, Right := 525.271423, Bottom := 36.304016}
                Alignment := vector2{X := 0.0, Y := 0.0}
                SizeToContent := true
                Widget := Achievement_ProgressBar_9
                ZOrder := 901
            },
            canvas_slot {
                Anchors := anchors{Minimum := vector2{X := 0.5, Y := 0.5}, Maximum := vector2{X := 0.5, Y := 0.5}}
                Offsets := margin{Top := 423.459473, Left := 318.400818, Right := 525.271423, Bottom := 36.304016}
                Alignment := vector2{X := 0.0, Y := 0.0}
                SizeToContent := true
                Widget := Achievement_ProgressBar_10
                ZOrder := 901
            },
            
            canvas_slot {
                Anchors := anchors{Minimum := vector2{X := 0.5, Y := 0.5}, Maximum := vector2{X := 0.5, Y := 0.5}}
                Offsets := margin{Top := -252.540527, Left := -592.960938, Right := 525.271423, Bottom := 36.304016}
                Alignment := vector2{X := 0.0, Y := 0.0}
                SizeToContent := false
                Widget := Achievement_ProgressBarBackground_1
                ZOrder := 900
            },
            canvas_slot {
                Anchors := anchors{Minimum := vector2{X := 0.5, Y := 0.5}, Maximum := vector2{X := 0.5, Y := 0.5}}
                Offsets := margin{Top := -84.540527, Left := -592.960938, Right := 525.271423, Bottom := 36.304016}
                Alignment := vector2{X := 0.0, Y := 0.0}
                SizeToContent := false
                Widget := Achievement_ProgressBarBackground_2
                ZOrder := 900
            },
            canvas_slot {
                Anchors := anchors{Minimum := vector2{X := 0.5, Y := 0.5}, Maximum := vector2{X := 0.5, Y := 0.5}}
                Offsets := margin{Top := 83.459473, Left := -592.960938, Right := 525.271423, Bottom := 36.304016}
                Alignment := vector2{X := 0.0, Y := 0.0}
                SizeToContent := false
                Widget := Achievement_ProgressBarBackground_3
                ZOrder := 900
            },
            canvas_slot {
                Anchors := anchors{Minimum := vector2{X := 0.5, Y := 0.5}, Maximum := vector2{X := 0.5, Y := 0.5}}
                Offsets := margin{Top := 255.459473, Left := -592.960938, Right := 525.271423, Bottom := 36.304016}
                Alignment := vector2{X := 0.0, Y := 0.0}
                SizeToContent := false
                Widget := Achievement_ProgressBarBackground_4
                ZOrder := 900
            },
            canvas_slot {
                Anchors := anchors{Minimum := vector2{X := 0.5, Y := 0.5}, Maximum := vector2{X := 0.5, Y := 0.5}}
                Offsets := margin{Top := 423.459473, Left := -592.960938, Right := 516.782166, Bottom := 36.303986}
                Alignment := vector2{X := 0.0, Y := 0.0}
                SizeToContent := false
                Widget := Achievement_ProgressBarBackground_5
                ZOrder := 900
            },
            canvas_slot {
                Anchors := anchors{Minimum := vector2{X := 0.5, Y := 0.5}, Maximum := vector2{X := 0.5, Y := 0.5}}
                Offsets := margin{Top := -252.540527, Left := 318.400818, Right := 525.271423, Bottom := 36.303986}
                Alignment := vector2{X := 0.0, Y := 0.0}
                SizeToContent := false
                Widget := Achievement_ProgressBarBackground_6
                ZOrder := 900
            },
            canvas_slot {
                Anchors := anchors{Minimum := vector2{X := 0.5, Y := 0.5}, Maximum := vector2{X := 0.5, Y := 0.5}}
                Offsets := margin{Top := -84.540527, Left := 318.400818, Right := 525.271423, Bottom := 36.303986}
                Alignment := vector2{X := 0.0, Y := 0.0}
                SizeToContent := false
                Widget := Achievement_ProgressBarBackground_7
                ZOrder := 900
            },
            canvas_slot {
                Anchors := anchors{Minimum := vector2{X := 0.5, Y := 0.5}, Maximum := vector2{X := 0.5, Y := 0.5}}
                Offsets := margin{Top := 83.459473, Left := 318.400818, Right := 525.271423, Bottom := 36.304016}
                Alignment := vector2{X := 0.0, Y := 0.0}
                SizeToContent := false
                Widget := Achievement_ProgressBarBackground_8
                ZOrder := 900
            },
            canvas_slot {
                Anchors := anchors{Minimum := vector2{X := 0.5, Y := 0.5}, Maximum := vector2{X := 0.5, Y := 0.5}}
                Offsets := margin{Top := 255.459473, Left := 318.400818, Right := 525.271423, Bottom := 36.304016}
                Alignment := vector2{X := 0.0, Y := 0.0}
                SizeToContent := false
                Widget := Achievement_ProgressBarBackground_9
                ZOrder := 900
            },
            canvas_slot {
                Anchors := anchors{Minimum := vector2{X := 0.5, Y := 0.5}, Maximum := vector2{X := 0.5, Y := 0.5}}
                Offsets := margin{Top := 423.459473, Left := 318.400818, Right := 525.271423, Bottom := 36.304016}
                Alignment := vector2{X := 0.0, Y := 0.0}
                SizeToContent := false
                Widget := Achievement_ProgressBarBackground_10
                ZOrder := 900
            }
        }
          
        return Slots
    }

    RefreshPlayerShopUi<override>(Agent: agent): void = {
        Utility := GetUtility()
        GameConfig := GetGameConfig()
        GameManager := GetGameManager()
        SaveManager := GetSaveManager()
        if {
            PlayerData := Agent.GetPlayerData[]
            Save := SaveManager.ProxySaves[Agent]
        } then {
            var AchievementLevel: int = 0
            var HudHideAchievementList: []hud_message_device = array {}
            case(AchievementType) {
                _Global.achievement_type.Kill =>
                    set AchievementLevel = Save.NumOfAchievementsCompletedKill
                    set HudHideAchievementList = GameManager.HudHideAchievementsListKill
                _Global.achievement_type.Scam =>
                    set AchievementLevel = Save.NumOfAchievementsCompletedScam
                    set HudHideAchievementList = GameManager.HudHideAchievementsListScam
                _Global.achievement_type.Money =>
                    set AchievementLevel = Save.NumOfAchievementsCompletedMoney
                    set HudHideAchievementList = GameManager.HudHideAchievementsListMoney
                _Global.achievement_type.Rebirth =>
                    set AchievementLevel = Save.NumOfAchievementsCompletedRebirth
                    set HudHideAchievementList = GameManager.HudHideAchievementsListRebirth
            }
            # Show a HUD to hide the items we didn't unlock yet
            for(HudHideAchievements: HudHideAchievementList) {
                HudHideAchievements.Hide(Agent)
            }
            if {
                HudHideAchievements := HudHideAchievementList[AchievementLevel]
            } then {
                HudHideAchievements.Show(Agent)
            }
            for(ItemIndex -> ItemObjectiveCounterSlot: ItemObjectiveCounterList) {
                if {
                    ItemObjectiveText := ItemObjectiveTextList[ItemIndex]
                    ProgressBar := ProgressBarList[ItemIndex]
                    ProgressBarBackground := ProgressBarBackgroundList[ItemIndex]
                } then {
                    # Hide the text if we didn't unlock this achievement yet
                    if(AchievementLevel < ItemIndex) {
                        ItemObjectiveCounterSlot.SetText(Utility.FormatMessage(""))
                        ProgressBar.SetDesiredSize(vector2{X := 0.0, Y := ProgressBar.DefaultDesiredSize.Y})
                        ItemObjectiveText.SetVisibility(widget_visibility.Hidden)
                        ProgressBarBackground.SetVisibility(widget_visibility.Hidden)
                    } else {
                        var AchievementGoalText: string = ""
                        var AchievementProgress: int = 0
                        var AchievementGoal: int = 0
                        case(AchievementType) {
                            _Global.achievement_type.Kill =>
                                set AchievementProgress = Save.AchievementProgressKill
                                if {
                                    AchievementGoalTemp := GameConfig.AchievementGoalListKill[ItemIndex]
                                    set AchievementGoal = AchievementGoalTemp
                                    AchievementGoalTextTemp := GameConfig.AchievementGoalTextList[0][ItemIndex]
                                    set AchievementGoalText = AchievementGoalTextTemp
                                }
                            _Global.achievement_type.Scam =>
                                set AchievementProgress = Save.AchievementProgressScam
                                if {
                                    AchievementGoalTemp := GameConfig.AchievementGoalListScam[ItemIndex]
                                    set AchievementGoal = AchievementGoalTemp
                                    AchievementGoalTextTemp := GameConfig.AchievementGoalTextList[1][ItemIndex]
                                    set AchievementGoalText = AchievementGoalTextTemp
                                }
                            _Global.achievement_type.Money =>
                                set AchievementProgress = Save.AchievementProgressMoney
                                if {
                                    AchievementGoalTemp := GameConfig.AchievementGoalListMoney[ItemIndex]
                                    set AchievementGoal = AchievementGoalTemp
                                    AchievementGoalTextTemp := GameConfig.AchievementGoalTextList[2][ItemIndex]
                                    set AchievementGoalText = AchievementGoalTextTemp
                                }
                            _Global.achievement_type.Rebirth =>
                                set AchievementProgress = Save.AchievementProgressRebirth
                                if {
                                    AchievementGoalTemp := GameConfig.AchievementGoalListRebirth[ItemIndex]
                                    set AchievementGoal = AchievementGoalTemp
                                    AchievementGoalTextTemp := GameConfig.AchievementGoalTextList[3][ItemIndex]
                                    set AchievementGoalText = AchievementGoalTextTemp
                                }
                        }
                        ItemObjectiveText.SetVisibility(widget_visibility.Visible)
                        ItemObjectiveText.SetText(Utility.FormatMessage(AchievementGoalText))
                        ProgressBarBackground.SetVisibility(widget_visibility.Visible)
                        if(AchievementLevel = ItemIndex) {
                            # Set Progress bar progress --> Get percentage of each achievement, related to goal
                            ProgressPercentage: float = Min((AchievementProgress * 1.0) / (AchievementGoal * 1.0), 1.0)
                            ProgressBarMaxSize := ProgressBar.DefaultDesiredSize
                            ProgressBarWidth := ProgressBarMaxSize.X * ProgressPercentage
                            ProgressBar.SetDesiredSize(vector2{X := ProgressBarWidth, Y := ProgressBarMaxSize.Y})
                        } else {
                            # Set previous achievements to completed
                            set AchievementProgress = AchievementGoal
                            ProgressBar.SetDesiredSize(ProgressBar.DefaultDesiredSize)
                        }
                        ItemObjectiveCounterSlot.SetText(Utility.FormatMessage(ToString(AchievementProgress) + "/" + ToString(AchievementGoal)))
                    }
                } else {
                    Print("ERR: ui_shop_achievement: RefreshUi: Failed to init variable for ItemIndex: {ItemIndex}")
                }
            }
        } else {
            Print("ERR: ui_shop_achievement: RefreshUi: Failed to init variables")
        }
    }

    CleanUpUiOnRemove<override>(): void = {
        (super:)CleanUpUiOnRemove()
        GameManager := GetGameManager()
        var HudHideAchievementList: []hud_message_device = array {}
        case(AchievementType) {
            _Global.achievement_type.Kill =>
                set HudHideAchievementList = GameManager.HudHideAchievementsListKill
            _Global.achievement_type.Scam =>
                set HudHideAchievementList = GameManager.HudHideAchievementsListScam
            _Global.achievement_type.Money =>
                set HudHideAchievementList = GameManager.HudHideAchievementsListMoney
            _Global.achievement_type.Rebirth =>
                set HudHideAchievementList = GameManager.HudHideAchievementsListRebirth
        }
        if {
            Agent := MaybeAgent?
        } then {
            for(HudHideAchievements: HudHideAchievementList) {
                HudHideAchievements.Hide(Agent)
            }
        } else {
            Print("ERR: ui_shop_achievement: CleanUpUiOnRemove")
        }
        <#
        for(ItemIndexBottom -> Item: ItemIconList) {
            if {
                ItemIcon := ItemIconList[ItemIndexBottom]
            } then {
                ItemIcon.SetImage(UI.invisible_img)
            }
        }
        #>
    }
}
