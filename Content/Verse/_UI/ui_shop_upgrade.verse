using { /Verse.org/Simulation }
using { /Verse.org/Colors }
using { /Verse.org/Colors/NamedColors }
# using { /Verse.org/Concurrency }
# using { /Verse.org/Verse }
# using { /Fortnite.com/Devices }
using { /Fortnite.com/UI }
# using { /Fortnite.com/Characters }
# using { /UnrealEngine.com/Temporary/Diagnostics }
using { /UnrealEngine.com/Temporary/SpatialMath }
using { /UnrealEngine.com/Temporary/UI}

ui_shop_upgrade<public> := class(ui_shop) {
    # Constants
    UiSlot<override>: player_ui_slot = player_ui_slot {
        ZOrder := 1026
        InputMode := ui_input_mode.All
    }

    CurFixedRewardPrefixTextList: []string = array {
        "Cur: +", 
        "Cur: +",
        "Cur: +",
        "Cur: +",
        "Cur: -"
    }

    NextFixedRewardPrefixTextList: []string = array {
        "Next: +", 
        "Next: +",
        "Next: +",
        "Next: +",
        "Next: -"
    }

    FixedRewardSuffixTextList: []string = array {
        " HP", 
        " Kill Coins",
        " Speed",
        " Money",
        " seconds"
    }

    Background: texture_block = texture_block {
        DefaultImage := UI.upgrades_bg
        DefaultDesiredSize := vector2{X := 165.0, Y := 41.0}
    }
    Logo: texture_block = texture_block {
        DefaultImage := UI.upgrades
        DefaultDesiredSize := vector2{X := 247.306137, Y := 55.925388}
    }
    Health_IMG: texture_block = texture_block {
        DefaultImage := UI.health_increase_icon
        DefaultDesiredSize := vector2{X := 217.318253, Y := 204.70192}
    }
    CoinKill_IMG: texture_block = texture_block {
        DefaultImage := UI.coinperkill
        DefaultDesiredSize := vector2{X := 217.318253, Y := 204.70192}
    }
    ScamSpeed_IMG: texture_block = texture_block {
        DefaultImage := UI.scamspeed
        DefaultDesiredSize := vector2{X := 217.318253, Y := 204.70192}
    }
    MoneyScam_IMG: texture_block = texture_block {
        DefaultImage := UI.moneyperscam
        DefaultDesiredSize := vector2{X := 217.318253, Y := 204.70192}
    }
    FasterSave_IMG: texture_block = texture_block {
        DefaultImage := UI.faster_save
        DefaultDesiredSize := vector2{X := 217.318253, Y := 204.70192}
    }

    ButtonClose: button_loud = button_loud {}

    Health_Increase_Button: button_loud = button_loud {}
    Coins_per_Kill_Button: button_loud = button_loud {}
    ScamSpeed_Button: button_loud = button_loud {}
    MoneyScam_Button: button_loud = button_loud {}
    FasterSave_Button: button_loud = button_loud {}

    # Runtime
    var UpgradeButtonList: []text_button_base = array {}
    var CurLevelList: []text_block = array {}
    var CurRewardList: []text_block = array {}
    var NextRewardList: []text_block = array {}

    # Refresh the UI every time it is added
    AddUiToPlayer<override>(Agent: agent, ParmMaybeManager: ?_Interface.interface_ui_manager): void = {
        (super:)AddUiToPlayer(Agent, ParmMaybeManager)
        RefreshUi(Agent)
    }

    # Runtime
    CreateSlots<override>(Agent: agent): []canvas_slot = {
        Utility := GetUtility()
        
        Text_MoneyScam_CurrentLevel := text_block{DefaultText := Utility.FormatMessage("Level 0")
        DefaultJustification := text_justification.Center
        DefaultTextColor := color{R:= 1.0, G:= 1.0, B:= 1.0}}
        Text_MoneyScam_CurrentStat := text_block{DefaultText := Utility.FormatMessage("Current Stat: +0")
        DefaultJustification := text_justification.Center
        DefaultTextColor := color{R:= 1.0, G:= 1.0, B:= 1.0}}
        Text_MoneyScam_NextStat := text_block{DefaultText := Utility.FormatMessage("Level 1")
        DefaultJustification := text_justification.Center
        DefaultTextColor := color{R:= 1.0, G:= 1.0, B:= 1.0}}
        Text_MoneyScam_NextLevel := text_block{DefaultText := Utility.FormatMessage("Level 1")
        DefaultJustification := text_justification.Center
        DefaultTextColor := color{R:= 1.0, G:= 1.0, B:= 1.0}}
        Text_ScamSpeed_CurrentLevel := text_block{DefaultText := Utility.FormatMessage("Level 0")
        DefaultJustification := text_justification.Center
        DefaultTextColor := color{R:= 1.0, G:= 1.0, B:= 1.0}}
        Text_ScamSpeed_CurrentStat := text_block{DefaultText := Utility.FormatMessage("Current Stat: +0")
        DefaultJustification := text_justification.Center
        DefaultTextColor := color{R:= 1.0, G:= 1.0, B:= 1.0}}
        Text_ScamSpeed_NextStat := text_block{DefaultText := Utility.FormatMessage("Next Stat: +10")
        DefaultJustification := text_justification.Center
        DefaultTextColor := color{R:= 1.0, G:= 1.0, B:= 1.0}}
        Text_ScamSpeed_NextLevel := text_block{DefaultText := Utility.FormatMessage("Level 1")
        DefaultJustification := text_justification.Center
        DefaultTextColor := color{R:= 1.0, G:= 1.0, B:= 1.0}}
        Text_Health_Increase_Titles := text_block{DefaultText := Utility.FormatMessage("Health Increase")
        DefaultJustification := text_justification.Center
        DefaultTextColor := color{R:= 1.0, G:= 1.0, B:= 1.0}}
        Text_Health_Increase_CurrentStat := text_block{DefaultText := Utility.FormatMessage("Current Stat: +0")
        DefaultJustification := text_justification.Center
        DefaultTextColor := color{R:= 1.0, G:= 1.0, B:= 1.0}}
        Text_Health_Increase_CurrentLevel := text_block{DefaultText := Utility.FormatMessage("Level 0")
        DefaultJustification := text_justification.Center
        DefaultTextColor := color{R:= 1.0, G:= 1.0, B:= 1.0}}
        Text_Health_Increase_NextLevel := text_block{DefaultText := Utility.FormatMessage("Level 1")
        DefaultJustification := text_justification.Center
        DefaultTextColor := color{R:= 1.0, G:= 1.0, B:= 1.0}}
        Text_Health_Increase_NextStat := text_block{DefaultText := Utility.FormatMessage("Level 1")
        DefaultJustification := text_justification.Center
        DefaultTextColor := color{R:= 1.0, G:= 1.0, B:= 1.0}}
        Text_CoinKill_NextLevel := text_block{DefaultText := Utility.FormatMessage("Level 1")
        DefaultJustification := text_justification.Center
        DefaultTextColor := color{R:= 1.0, G:= 1.0, B:= 1.0}}
        Text_CoinKill_NextStat := text_block{DefaultText := Utility.FormatMessage("Next Stat: +10")
        DefaultJustification := text_justification.Center
        DefaultTextColor := color{R:= 1.0, G:= 1.0, B:= 1.0}}
        Text_CoinKill_CurrentStat := text_block{DefaultText := Utility.FormatMessage("Current Stat: +0")
        DefaultJustification := text_justification.Center
        DefaultTextColor := color{R:= 1.0, G:= 1.0, B:= 1.0}}
        Text_CoinKill_CurrentLevel := text_block{DefaultText := Utility.FormatMessage("Level 0")
        DefaultJustification := text_justification.Center
        DefaultTextColor := color{R:= 1.0, G:= 1.0, B:= 1.0}}
        Text_CoinKill_Title := text_block{DefaultText := Utility.FormatMessage("Coins per Kill")
        DefaultJustification := text_justification.Center
        DefaultTextColor := color{R:= 1.0, G:= 1.0, B:= 1.0}}
        Text_ScamSpeed_Title := text_block{DefaultText := Utility.FormatMessage("Scam Speed")
        DefaultJustification := text_justification.Center
        DefaultTextColor := color{R:= 1.0, G:= 1.0, B:= 1.0}}
        Text_MoneyScam_Title := text_block{DefaultText := Utility.FormatMessage("Money per Scam")
        DefaultJustification := text_justification.Center
        DefaultTextColor := color{R:= 1.0, G:= 1.0, B:= 1.0}}
        Text_FasterSave_Title := text_block{DefaultText := Utility.FormatMessage("Faster Save")
        DefaultJustification := text_justification.Center
        DefaultTextColor := color{R:= 1.0, G:= 1.0, B:= 1.0}}
        Text_FasterSave_CurrentLevel := text_block{DefaultText := Utility.FormatMessage("Level 0")
        DefaultJustification := text_justification.Center
        DefaultTextColor := color{R:= 1.0, G:= 1.0, B:= 1.0}}
        Text_FasterSave_CurrentStat := text_block{DefaultText := Utility.FormatMessage("Current Stat: +0")
        DefaultJustification := text_justification.Center
        DefaultTextColor := color{R:= 1.0, G:= 1.0, B:= 1.0}}
        Text_FasterSave_NextStat := text_block{DefaultText := Utility.FormatMessage("Next Stat: +10")
        DefaultJustification := text_justification.Center
        DefaultTextColor := color{R:= 1.0, G:= 1.0, B:= 1.0}}
        Text_FasterSave_NextLevel := text_block{DefaultText := Utility.FormatMessage("Level 1")
        DefaultJustification := text_justification.Center
        DefaultTextColor := color{R:= 1.0, G:= 1.0, B:= 1.0}}

        # Button texts need to be defined inside the function
        ButtonClose.SetText(Utility.FormatMessage("EXIT"))
    
        Health_Increase_Button.SetText(Utility.FormatMessage("↑ $5600"))
        Coins_per_Kill_Button.SetText(Utility.FormatMessage("↑ $5600"))
        ScamSpeed_Button.SetText(Utility.FormatMessage("↑ $5600"))
        MoneyScam_Button.SetText(Utility.FormatMessage("↑ $5600"))
        FasterSave_Button.SetText(Utility.FormatMessage("↑ $5600"))

        set UpgradeButtonList = array {
            Health_Increase_Button,
            Coins_per_Kill_Button,
            ScamSpeed_Button,
            MoneyScam_Button,
            FasterSave_Button
        }

        set CurLevelList = array {
            Text_Health_Increase_CurrentLevel,
            Text_CoinKill_CurrentLevel,
            Text_ScamSpeed_CurrentLevel,
            Text_MoneyScam_CurrentLevel,
            Text_FasterSave_CurrentLevel
        }

        set CurRewardList = array {
            Text_Health_Increase_CurrentStat,
            Text_CoinKill_CurrentStat,
            Text_ScamSpeed_CurrentStat,
            Text_MoneyScam_CurrentStat,
            Text_FasterSave_CurrentStat
        }

        set NextRewardList = array {
            Text_Health_Increase_NextStat,
            Text_CoinKill_NextStat,
            Text_ScamSpeed_NextStat,
            Text_MoneyScam_NextStat,
            Text_FasterSave_NextStat
        }

        # Listen for click events
        for(ButtonIndex -> ItemButton: UpgradeButtonList) {
            ButtonEventHandler := event_handler_widget_message(int) {
                A := ButtonIndex
                CallbackFunc := OnItemClicked
            }

            SpawnerEventSub := ItemButton.OnClick().Subscribe(ButtonEventHandler.EventSubscription)
        }

        ButtonClose.OnClick().Subscribe(OnExitButtonClicked)

        # Init slots
        var Slots: []canvas_slot = array {
            canvas_slot {
                Anchors := anchors{Minimum := vector2{X := 0.0, Y := 0.0}, Maximum := vector2{X := 1.0, Y := 1.0}}
                Offsets := margin{Top := -121.56163, Left := 148.312317, Right := 121.401123, Bottom := -124.598145}
                Alignment := vector2{X := 0.0, Y := 0.0}
                SizeToContent := false
                Widget := Background
                ZOrder := 899
            },
            canvas_slot {
                Anchors := anchors{Minimum := vector2{X := 0.0, Y := 0.0}, Maximum := vector2{X := 1.0, Y := 1.0}}
                Offsets := margin{Top := -392.0, Left := 332.0, Right := 333.55127, Bottom := 350.927612}
                Alignment := vector2{X := 0.0, Y := 0.0}
                SizeToContent := false
                Widget := Logo
                ZOrder := 900
            },
            canvas_slot {
                Anchors := anchors{Minimum := vector2{X := 0.5, Y := 0.5}, Maximum := vector2{X := 0.5, Y := 0.5}}
                Offsets := margin{Top := -276.540527, Left := -732.960938, Right := 217.318253, Bottom := 204.70192}
                Alignment := vector2{X := 0.0, Y := 0.0}
                SizeToContent := false
                Widget := Health_IMG
                ZOrder := 900
            },
            canvas_slot {
                Anchors := anchors{Minimum := vector2{X := 0.5, Y := 0.5}, Maximum := vector2{X := 0.5, Y := 0.5}}
                Offsets := margin{Top := -136.540527, Left := -732.960938, Right := 217.318253, Bottom := 204.70192}
                Alignment := vector2{X := 0.0, Y := 0.0}
                SizeToContent := false
                Widget := CoinKill_IMG
                ZOrder := 900
            },
            canvas_slot {
                Anchors := anchors{Minimum := vector2{X := 0.5, Y := 0.5}, Maximum := vector2{X := 0.5, Y := 0.5}}
                Offsets := margin{Top := 7.459473, Left := -732.960938, Right := 217.318253, Bottom := 204.70192}
                Alignment := vector2{X := 0.0, Y := 0.0}
                SizeToContent := false
                Widget := ScamSpeed_IMG
                ZOrder := 900
            },
            canvas_slot {
                Anchors := anchors{Minimum := vector2{X := 0.5, Y := 0.5}, Maximum := vector2{X := 0.5, Y := 0.5}}
                Offsets := margin{Top := 159.459473, Left := -732.960938, Right := 217.318253, Bottom := 204.70192}
                Alignment := vector2{X := 0.0, Y := 0.0}
                SizeToContent := false
                Widget := MoneyScam_IMG
                ZOrder := 900
            },
            canvas_slot {
                Anchors := anchors{Minimum := vector2{X := 0.5, Y := 0.5}, Maximum := vector2{X := 0.5, Y := 0.5}}
                Offsets := margin{Top := 307.459473, Left := -732.960938, Right := 217.318253, Bottom := 204.70192}
                Alignment := vector2{X := 0.0, Y := 0.0}
                SizeToContent := false
                Widget := FasterSave_IMG
                ZOrder := 900
            },
            canvas_slot {
                Anchors := anchors{Minimum := vector2{X := 0.5, Y := 0.5}, Maximum := vector2{X := 0.5, Y := 0.5}}
                Offsets := margin{Top := 223.459473, Left := -504.960938, Right := 260.095093, Bottom := 41.0}
                Alignment := vector2{X := 0.0, Y := 0.0}
                SizeToContent := false
                Widget := Text_MoneyScam_Title
                ZOrder := 900
            },
            canvas_slot {
                Anchors := anchors{Minimum := vector2{X := 0.5, Y := 0.5}, Maximum := vector2{X := 0.5, Y := 0.5}}
                Offsets := margin{Top := 203.459473, Left := -76.960938, Right := 114.749741, Bottom := 41.0}
                Alignment := vector2{X := 0.0, Y := 0.0}
                SizeToContent := false
                Widget := Text_MoneyScam_CurrentLevel
                ZOrder := 900
            },
            canvas_slot {
                Anchors := anchors{Minimum := vector2{X := 0.5, Y := 0.5}, Maximum := vector2{X := 0.5, Y := 0.5}}
                Offsets := margin{Top := 243.459473, Left := -172.960938, Right := 290.125122, Bottom := 41.0}
                Alignment := vector2{X := 0.0, Y := 0.0}
                SizeToContent := false
                Widget := Text_MoneyScam_CurrentStat
                ZOrder := 900
            },
            canvas_slot {
                Anchors := anchors{Minimum := vector2{X := 0.5, Y := 0.5}, Maximum := vector2{X := 0.5, Y := 0.5}}
                Offsets := margin{Top := 223.459473, Left := 203.039062, Right := 220.455444, Bottom := 41.0}
                Alignment := vector2{X := 0.0, Y := 0.0}
                SizeToContent := false
                Widget := Text_MoneyScam_NextStat
                ZOrder := 900
            },
            canvas_slot {
                Anchors := anchors{Minimum := vector2{X := 0.5, Y := 0.5}, Maximum := vector2{X := 0.5, Y := 0.5}}
                Offsets := margin{Top := 67.459473, Left := -508.960938, Right := 190.425415, Bottom := 34.093094}
                Alignment := vector2{X := 0.0, Y := 0.0}
                SizeToContent := false
                Widget := Text_ScamSpeed_Title
                ZOrder := 900
            },
            canvas_slot {
                Anchors := anchors{Minimum := vector2{X := 0.5, Y := 0.5}, Maximum := vector2{X := 0.5, Y := 0.5}}
                Offsets := margin{Top := 47.459473, Left := -84.960938, Right := 114.749741, Bottom := 41.0}
                Alignment := vector2{X := 0.0, Y := 0.0}
                SizeToContent := false
                Widget := Text_ScamSpeed_CurrentLevel
                ZOrder := 900
            },
            canvas_slot {
                Anchors := anchors{Minimum := vector2{X := 0.5, Y := 0.5}, Maximum := vector2{X := 0.5, Y := 0.5}}
                Offsets := margin{Top := 83.459473, Left := -176.960938, Right := 290.125122, Bottom := 41.0}
                Alignment := vector2{X := 0.0, Y := 0.0}
                SizeToContent := false
                Widget := Text_ScamSpeed_CurrentStat
                ZOrder := 900
            },
            canvas_slot {
                Anchors := anchors{Minimum := vector2{X := 0.5, Y := 0.5}, Maximum := vector2{X := 0.5, Y := 0.5}}
                Offsets := margin{Top := 67.459473, Left := 199.039062, Right := 220.455444, Bottom := 41.0}
                Alignment := vector2{X := 0.0, Y := 0.0}
                SizeToContent := false
                Widget := Text_ScamSpeed_NextStat
                ZOrder := 900
            },
            canvas_slot {
                Anchors := anchors{Minimum := vector2{X := 0.5, Y := 0.5}, Maximum := vector2{X := 0.5, Y := 0.5}}
                Offsets := margin{Top := -216.540527, Left := -508.960938, Right := 247.537079, Bottom := 41.0}
                Alignment := vector2{X := 0.0, Y := 0.0}
                SizeToContent := false
                Widget := Text_Health_Increase_Titles
                ZOrder := 905
            },
            canvas_slot {
                Anchors := anchors{Minimum := vector2{X := 0.5, Y := 0.5}, Maximum := vector2{X := 0.5, Y := 0.5}}
                Offsets := margin{Top := -204.540527, Left := -176.960938, Right := 290.125122, Bottom := 41.0}
                Alignment := vector2{X := 0.0, Y := 0.0}
                SizeToContent := false
                Widget := Text_Health_Increase_CurrentStat
                ZOrder := 900
            },
            canvas_slot {
                Anchors := anchors{Minimum := vector2{X := 0.5, Y := 0.5}, Maximum := vector2{X := 0.5, Y := 0.5}}
                Offsets := margin{Top := -240.540527, Left := -84.960938, Right := 114.749741, Bottom := 41.0}
                Alignment := vector2{X := 0.0, Y := 0.0}
                SizeToContent := false
                Widget := Text_Health_Increase_CurrentLevel
                ZOrder := 900
            },
            canvas_slot {
                Anchors := anchors{Minimum := vector2{X := 0.5, Y := 0.5}, Maximum := vector2{X := 0.5, Y := 0.5}}
                Offsets := margin{Top := -216.540527, Left := 203.039062, Right := 220.455444, Bottom := 34.093094}
                Alignment := vector2{X := 0.0, Y := 0.0}
                SizeToContent := false
                Widget := Text_Health_Increase_NextStat
                ZOrder := 900
            },
            canvas_slot {
                Anchors := anchors{Minimum := vector2{X := 0.5, Y := 0.5}, Maximum := vector2{X := 0.5, Y := 0.5}}
                Offsets := margin{Top := -84.540527, Left := 199.039062, Right := 220.455444, Bottom := 41.0}
                Alignment := vector2{X := 0.0, Y := 0.0}
                SizeToContent := false
                Widget := Text_CoinKill_NextStat
                ZOrder := 900
            },
            canvas_slot {
                Anchors := anchors{Minimum := vector2{X := 0.5, Y := 0.5}, Maximum := vector2{X := 0.5, Y := 0.5}}
                Offsets := margin{Top := -64.540527, Left := -176.960938, Right := 290.125122, Bottom := 41.0}
                Alignment := vector2{X := 0.0, Y := 0.0}
                SizeToContent := false
                Widget := Text_CoinKill_CurrentStat
                ZOrder := 900
            },
            canvas_slot {
                Anchors := anchors{Minimum := vector2{X := 0.5, Y := 0.5}, Maximum := vector2{X := 0.5, Y := 0.5}}
                Offsets := margin{Top := -100.540527, Left := -84.960938, Right := 114.749741, Bottom := 34.531994}
                Alignment := vector2{X := 0.0, Y := 0.0}
                SizeToContent := false
                Widget := Text_CoinKill_CurrentLevel
                ZOrder := 900
            },
            canvas_slot {
                Anchors := anchors{Minimum := vector2{X := 0.5, Y := 0.5}, Maximum := vector2{X := 0.5, Y := 0.5}}
                Offsets := margin{Top := -84.540527, Left := -508.960938, Right := 211.189041, Bottom := 41.0}
                Alignment := vector2{X := 0.0, Y := 0.0}
                SizeToContent := false
                Widget := Text_CoinKill_Title
                ZOrder := 900
            },
            canvas_slot {
                Anchors := anchors{Minimum := vector2{X := 0.5, Y := 0.5}, Maximum := vector2{X := 0.5, Y := 0.5}}
                Offsets := margin{Top := 375.459473, Left := -508.960938, Right := 182.617615, Bottom := 41.0}
                Alignment := vector2{X := 0.0, Y := 0.0}
                SizeToContent := false
                Widget := Text_FasterSave_Title
                ZOrder := 900
            },
            canvas_slot {
                Anchors := anchors{Minimum := vector2{X := 0.5, Y := 0.5}, Maximum := vector2{X := 0.5, Y := 0.5}}
                Offsets := margin{Top := 355.459473, Left := -80.960938, Right := 114.749741, Bottom := 41.0}
                Alignment := vector2{X := 0.0, Y := 0.0}
                SizeToContent := false
                Widget := Text_FasterSave_CurrentLevel
                ZOrder := 900
            },
            canvas_slot {
                Anchors := anchors{Minimum := vector2{X := 0.5, Y := 0.5}, Maximum := vector2{X := 0.5, Y := 0.5}}
                Offsets := margin{Top := 395.459473, Left := -176.960938, Right := 290.125122, Bottom := 41.0}
                Alignment := vector2{X := 0.0, Y := 0.0}
                SizeToContent := false
                Widget := Text_FasterSave_CurrentStat
                ZOrder := 900
            },
            canvas_slot {
                Anchors := anchors{Minimum := vector2{X := 0.5, Y := 0.5}, Maximum := vector2{X := 0.5, Y := 0.5}}
                Offsets := margin{Top := 375.459473, Left := 199.039062, Right := 220.455444, Bottom := 41.0}
                Alignment := vector2{X := 0.0, Y := 0.0}
                SizeToContent := false
                Widget := Text_FasterSave_NextStat
                ZOrder := 900
            },
            # Button 1
            canvas_slot {
                Anchors := anchors{Minimum := vector2{X := 0.5, Y := 0.5}, Maximum := vector2{X := 0.5, Y := 0.5}}
                Offsets := margin{Top := -224.540527, Left := 499.039062, Right := 247.306137, Bottom := 58.307312}
                Alignment := vector2{X := 0.0, Y := 0.0}
                SizeToContent := false
                Widget := Health_Increase_Button
                ZOrder := 900
            },
            # Button 2
            canvas_slot {
                Anchors := anchors{Minimum := vector2{X := 0.5, Y := 0.5}, Maximum := vector2{X := 0.5, Y := 0.5}}
                Offsets := margin{Top := -84.540527, Left := 499.039062, Right := 247.306137, Bottom := 58.307312}
                Alignment := vector2{X := 0.0, Y := 0.0}
                SizeToContent := false
                Widget := Coins_per_Kill_Button
                ZOrder := 900
            },
            # Button 3
            canvas_slot {
                Anchors := anchors{Minimum := vector2{X := 0.5, Y := 0.5}, Maximum := vector2{X := 0.5, Y := 0.5}}
                Offsets := margin{Top := 59.459473, Left := 499.039062, Right := 247.306137, Bottom := 58.307312}
                Alignment := vector2{X := 0.0, Y := 0.0}
                SizeToContent := false
                Widget := ScamSpeed_Button
                ZOrder := 900
            },
            # Button 3
            canvas_slot {
                Anchors := anchors{Minimum := vector2{X := 0.5, Y := 0.5}, Maximum := vector2{X := 0.5, Y := 0.5}}
                Offsets := margin{Top := 223.459473, Left := 499.039062, Right := 247.306137, Bottom := 58.307312}
                Alignment := vector2{X := 0.0, Y := 0.0}
                SizeToContent := false
                Widget := MoneyScam_Button
                ZOrder := 900
            },
            # Button 3
            canvas_slot {
                Anchors := anchors{Minimum := vector2{X := 0.5, Y := 0.5}, Maximum := vector2{X := 0.5, Y := 0.5}}
                Offsets := margin{Top := 387.459473, Left := 499.039062, Right := 247.306137, Bottom := 58.307312}
                Alignment := vector2{X := 0.0, Y := 0.0}
                SizeToContent := false
                Widget := FasterSave_Button
                ZOrder := 900
            },
            # Exit Button
            canvas_slot {
                Anchors := anchors{Minimum := vector2{X := 0.5, Y := 0.5}, Maximum := vector2{X := 0.5, Y := 0.5}}
                Offsets := margin{Top := -356.540527, Left := 499.039062, Right := 247.306137, Bottom := 55.925388}
                Alignment := vector2{X := 0.0, Y := 0.0}
                SizeToContent := false
                Widget := ButtonClose
                ZOrder := 900
            }
        }

        return Slots
    }

    OnItemClicked<override>(WidgetMessage: widget_message, ButtonIndex: int): void = {
        Print("ui_upgrade_menu: OnItemClicked - ItemIndex: {ButtonIndex}")
        Agent := WidgetMessage.Player
        TryToPurchaseUpgrade(Agent, ButtonIndex)
    }

    TryToPurchaseUpgrade(Agent: agent, UpgradeIndex: int): void = {
        SaveManager := GetSaveManager()
        GameConfig := GetGameConfig()
        UpgradeLevel := FindUpgradeLevel(Agent, UpgradeIndex)
        if {
            PlayerData := Agent.GetPlayerData[]
            Save := SaveManager.ProxySaves[Agent]
            TycoonManager := PlayerData.MaybeTycoonManager?
            TycoonConfig := TycoonManager.TycoonConfig
            MaxUpgradeLevel := GameConfig.PermanentUpgradeCostList[UpgradeIndex].Length
            UpgradeLevel < MaxUpgradeLevel
        } then {
            UpgradeCost := FindUpgradeCost(Agent, UpgradeIndex)
            case(UpgradeIndex) {
                0 =>
                    if {
                        PlayerData.CurMoney >= UpgradeCost
                    } then {
                        PlayerData.RemoveMoney(UpgradeCost)
                        set Save.UpgradeHealthIncreaseLevel = Save.UpgradeHealthIncreaseLevel + 1
                        PlayerData.SetHealth(Agent)
                    }
                1 =>
                    if {
                        PlayerData.CurMoney >= UpgradeCost
                    } then {
                        PlayerData.RemoveMoney(UpgradeCost)
                        set Save.UpgradeCoinsPerKillLevel = Save.UpgradeCoinsPerKillLevel + 1
                    }
                2 =>
                    if {
                        PlayerData.CurMoney >= UpgradeCost
                    } then {
                        PlayerData.RemoveMoney(UpgradeCost)
                        set Save.UpgradeScamSpeedLevel = Save.UpgradeScamSpeedLevel + 1
                    }
                3 =>
                    if {
                        PlayerData.CurMoney >= UpgradeCost
                    } then {
                        PlayerData.RemoveMoney(UpgradeCost)
                        set Save.UpgradeMoneyScamLevel = Save.UpgradeMoneyScamLevel + 1
                    }
                4 =>
                    if {
                        PlayerData.CurMoney >= UpgradeCost
                    } then {
                        PlayerData.RemoveMoney(UpgradeCost)
                        set Save.UpgradeFasterSaveLevel = Save.UpgradeFasterSaveLevel + 1
                        TycoonManager.TycoonButtonAutoSave.CalculateRemainingTime()
                        # Increment achievements
                        # Spend $ 500K for Progress Save Upgrades
                        SaveManager.IncrementAchievement(Agent, _Global.achievement_type.Money, 4, UpgradeCost)
                    }
                _ =>
                    Print("ERR: ui_upgrade_menu: Unknown UpgradeIndex: {UpgradeIndex}")
            }
            # This must be last
            RefreshUi(Agent)
        }
    }

    IncreaseUpgradeQuestProgress(Agent: agent): void = {
        if {
            PlayerData := Agent.GetPlayerData[]
            Quest := PlayerData.MaybeActiveQuest?
            # Check if we are currently on the correct quest
            QuestUpgradeShop := _Tycoon.tycoon_button_upgrade_shop[Quest]
        } then {
            Print("IncreaseUpgradeQuestProgress: Correct Quest")
            PlayerData.IncreaseQuestProgress()
        } else {
            Print("IncreaseUpgradeQuestProgress: Invalid Quest")
        }
    }

    FindUpgradeLevel(Agent: agent, UpgradeIndex: int): int = {
        SaveManager := GetSaveManager()
        var UpgradeLevel: int = 0
        if {
            PlayerData := Agent.GetPlayerData[]
            Save := SaveManager.ProxySaves[Agent]
            TycoonManager := PlayerData.MaybeTycoonManager?
            TycoonConfig := TycoonManager.TycoonConfig
        } then {
            case(UpgradeIndex) {
                0 =>
                    set UpgradeLevel = Save.UpgradeHealthIncreaseLevel
                1 =>
                    set UpgradeLevel = Save.UpgradeCoinsPerKillLevel
                2 =>
                    set UpgradeLevel = Save.UpgradeScamSpeedLevel
                3 =>
                    set UpgradeLevel = Save.UpgradeMoneyScamLevel
                4 =>
                    set UpgradeLevel = Save.UpgradeFasterSaveLevel
                _ =>
                    Print("ERR: FindUpgradeLevel: Unknown UpgradeIndex: {UpgradeIndex}")
            }
            if(UpgradeLevel > 0) {
                IncreaseUpgradeQuestProgress(Agent)
            }
            # Achievement safety fallback: Spend $ 500K for Progress Save Upgrades
            if {
                Save.UpgradeFasterSaveLevel >= 47
            } then {
                SaveManager.IncrementAchievement(Agent, _Global.achievement_type.Money, 4, 99999999)
            }
        }
        return UpgradeLevel
    }

    FindUpgradeCost(Agent: agent, UpgradeIndex: int): int = {
        SaveManager := GetSaveManager()
        var UpgradeCost: int = 0
        UpgradeLevel := FindUpgradeLevel(Agent, UpgradeIndex)
        return GetUpgradeCost(UpgradeIndex, UpgradeLevel)
        <#
        if {
            PlayerData := Agent.GetPlayerData[]
            Save := SaveManager.ProxySaves[Agent]
            TycoonManager := PlayerData.MaybeTycoonManager?
            TycoonConfig := TycoonManager.TycoonConfig
        } then {
            case(UpgradeIndex) {
                0 =>
                    set UpgradeCost = GetUpgradeCost(TycoonConfig.UpgradeBaseCostHealthIncrease, UpgradeLevel)
                1 =>
                    set UpgradeCost = GetUpgradeCost(TycoonConfig.UpgradeBaseCostCoinsPerKill, UpgradeLevel)
                2 =>
                    set UpgradeCost = GetUpgradeCost(TycoonConfig.UpgradeBaseCostScamSpeed, UpgradeLevel)
                3 =>
                    set UpgradeCost = GetUpgradeCost(TycoonConfig.UpgradeBaseCostMoneyScam, UpgradeLevel)
                4 =>
                    set UpgradeCost =  GetUpgradeCost(TycoonConfig.UpgradeBaseCostFasterSave, UpgradeLevel)
                _ =>
                    Print("ERR: FindUpgradeCost: Unknown UpgradeIndex: {UpgradeIndex}")
            }
        }
        return UpgradeCost
        #>
    }

    GetUpgradeCost(UpgradeId: int, UpgradeLevel: int): int = {
        GameConfig := GetGameConfig()
        if {
            UpgradeCostList := GameConfig.PermanentUpgradeCostList[UpgradeId]
            UpgradeCost := UpgradeCostList[UpgradeLevel]
        } then {
            return UpgradeCost
        }
        return 0
    }

    <#
    GetUpgradeCost(BaseCost: int, UpgradeLevel: int): int = {
        if(UpgradeLevel = 0) {
            return BaseCost
        } else {
            # BaseCost * ((1 + 0.25) ^ UpgradeLevel)
            # BaseCost = 100, UpgradeLevel = 0 --> 100, 125, 156.25
            if {
                UpgradeCost: int = Round[BaseCost * Pow(1.0 + 0.25, UpgradeLevel * 1.0)]
            } then {
                return UpgradeCost
            } else {
                Print("ERR: GetUpgradeCost: Couldn't round number")
                return BaseCost
            }
        }
    }
    #>

    RefreshUi(Agent: agent): void = {
        Utility := GetUtility()
        SaveManager := GetSaveManager()
        GameConfig := GetGameConfig()
        if {
            PlayerData := Agent.GetPlayerData[]
            Save := SaveManager.ProxySaves[Agent]
            TycoonManager := PlayerData.MaybeTycoonManager?
            TycoonConfig := TycoonManager.TycoonConfig
        } then {
            for(ItemIndex -> Button: UpgradeButtonList) {
                if {
                    CurLevelTextSlot := CurLevelList[ItemIndex]
                    CurRewardTextSlot := CurRewardList[ItemIndex]
                    NextRewardTextSlot := NextRewardList[ItemIndex]
                } then {
                    case(ItemIndex) {
                        0 =>
                            Health_Increase_Button.SetText(Utility.FormatMessage("Level " + ToString(Save.UpgradeHealthIncreaseLevel + 1)))
                        1 =>
                            Coins_per_Kill_Button.SetText(Utility.FormatMessage("Level " + ToString(Save.UpgradeCoinsPerKillLevel + 1)))
                        2 =>
                            ScamSpeed_Button.SetText(Utility.FormatMessage("Level " + ToString(Save.UpgradeScamSpeedLevel + 1)))
                        3 =>
                            MoneyScam_Button.SetText(Utility.FormatMessage("Level " + ToString(Save.UpgradeMoneyScamLevel + 1)))
                        4 =>
                            FasterSave_Button.SetText(Utility.FormatMessage("Level " + ToString(Save.UpgradeFasterSaveLevel + 1)))
                        _ =>
                            Print("ERR: ui_upgrade_menu: RefreshUi: ItemIndex not implemented: {ItemIndex}")
                    }

                    UpgradeLevel := FindUpgradeLevel(Agent, ItemIndex)
                    UpgradeCost := FindUpgradeCost(Agent, ItemIndex)
                    if {
                        RewardPerLevel := GameConfig.PermanentUpgradeRewardPerLevelList[ItemIndex]
                        CurFixedRewardPrefixText := CurFixedRewardPrefixTextList[ItemIndex]
                        NextFixedRewardPrefixText := NextFixedRewardPrefixTextList[ItemIndex]
                        FixedRewardSuffixText := FixedRewardSuffixTextList[ItemIndex]
                    } then {
                        CurLevelTextSlot.SetText(Utility.FormatMessage("Level " + ToString(UpgradeLevel)))

                        var CurTotalAmountFromUpgrade: float = UpgradeLevel * RewardPerLevel
                        # Special case for CoinsPerKill
                        if(ItemIndex = 1) {
                            if(CoinReward := GameConfig.PermanentUpgradeRewardKillpointsPerKill[UpgradeLevel]) {
                                set CurTotalAmountFromUpgrade = CoinReward
                            }
                        }
                        # Special case for MoneyScam
                        if(ItemIndex = 3) {
                            if(ScamReward := GameConfig.PermanentUpgradeRewardMoneyScam[UpgradeLevel]) {
                                set CurTotalAmountFromUpgrade = ScamReward
                            }
                        }

                        if {
                            UpgradeLevel > 0
                        } then {
                            CurRewardTextSlot.SetText(Utility.FormatMessage(CurFixedRewardPrefixText + Utility.FormatFloat(CurTotalAmountFromUpgrade, 2, 1) + FixedRewardSuffixText))
                        } else {
                            CurRewardTextSlot.SetText(Utility.FormatMessage(""))
                        }
                        
                        if {
                            MaxUpgradeLevel := GameConfig.PermanentUpgradeCostList[ItemIndex].Length
                            UpgradeLevel < MaxUpgradeLevel
                        } then {
                            var NextRewardAmount: float = CurTotalAmountFromUpgrade + RewardPerLevel
                            # Special case for CoinsPerKill
                            if(ItemIndex = 1) {
                                if(CoinReward := GameConfig.PermanentUpgradeRewardKillpointsPerKill[UpgradeLevel + 1]) {
                                    set NextRewardAmount = CoinReward
                                }
                            }
                            # Special case for MoneyScam
                            if(ItemIndex = 3) {
                                if(ScamReward := GameConfig.PermanentUpgradeRewardMoneyScam[UpgradeLevel + 1]) {
                                    set NextRewardAmount = ScamReward
                                }
                            }

                            NextRewardTextSlot.SetText(Utility.FormatMessage(NextFixedRewardPrefixText + Utility.FormatFloat(NextRewardAmount, 2, 1) + FixedRewardSuffixText))
                            # Button.SetEnabled(true)
                            Button.SetText(Utility.FormatMessage("↑ $" + Utility.FormatFloat(UpgradeCost * 1.0, 2, 0)))
                        } else {
                            NextRewardTextSlot.SetText(Utility.FormatMessage(""))
                            # Button.SetEnabled(false)
                            Button.SetText(Utility.FormatMessage("MAX LEVEL"))
                        }
                    }
                }
            }
        }
    }
}
