using { /Fortnite.com/Game }
using { /Fortnite.com/Devices }
using { /Fortnite.com/Devices/CreativeAnimation }
using { /Fortnite.com/Characters }
using { /Fortnite.com/Vehicles }
using { /Verse.org/Simulation }
using { /Verse.org/Simulation/Tags }
using { /Verse.org/Random }
using { /Verse.org/Concurrency }
using { /UnrealEngine.com/Temporary/Diagnostics }
using { /UnrealEngine.com/Temporary/SpatialMath }

city_owner_device := class(creative_device) {
    @editable
    BuildingSwitchList: []switch_device = array {}

    @editable
    BuyCurrencyList: []tycoon_currency = array {}

    @editable
    PriceToBuyList: []int = array {}

    @editable
    CinematicOnBuyList: []cinematic_sequence_device = array {}

    @editable
    SwitchTurnOffOnBuyList: []switch_device = array {}

    @editable
    MoneyIncomeIncreaseAmountAfterBuy: []int = array {}

    @editable
    FansIncomeIncreaseAmountAfterBuy: []float = array {}

    @editable
    DiamondsIncomeIncreaseAmountAfterBuy: []float = array {}

    @editable
    HudDidntUnlockCityOwner: hud_message_device = hud_message_device {}

    @editable
    HudNotEnoughCurrency: hud_message_device = hud_message_device {}

    @editable
    HudBuildingAlreadyOwned: hud_message_device = hud_message_device {}

    @editable
    AudioPurchaseSuccess: audio_player_device = audio_player_device {}
    
    # Runtime
    var MaybeOwnerAgent: ?agent = false
    var MaybeTycoonManager: ?tycoon_manager_device = false

    # Runs when the device is started in a running game
    OnBegin<override>()<suspends>:void = {
        # Listen to events
        # SwitchList
        for(ItemIndex -> BuildingSwitch: BuildingSwitchList) {
            EventHandler := event_handler_agent(int) {
                A := ItemIndex
                CallbackFunc := OnBuildingSwitch
            }

            EventSub := BuildingSwitch.TurnedOffEvent.Subscribe(EventHandler.EventSubscription)
        }
    }

    AssignOwner(Agent: agent, TycoonManager: tycoon_manager_device): void = {
        set MaybeOwnerAgent = option{Agent}
        set MaybeTycoonManager = option{TycoonManager}
    }

    OnTycoonReset(): void = {
        if {
            Agent := MaybeOwnerAgent?
        } then {
            for(Switch: BuildingSwitchList) {
                Switch.TurnOn(Agent)
            }
        }
        set MaybeOwnerAgent = false
        set MaybeTycoonManager = false
    }

    OnBuildingSwitch(Agent: agent, ItemIndex: int): void = {
        Print("City owner device: OnBuildingSwitch")
        Print("WARN: Not implemented!")
        <#
        if {
            OwnerAgent := MaybeOwnerAgent?
            Agent = OwnerAgent
            PlayerData := Agent.GetPlayerData[]
        } then {
            Print("City owner device: OnBuildingSwitch 222")
            if(PlayerData.HasUnlockedCityOwner = false) {
                HudDidntUnlockCityOwner.Show(Agent)
                return
            }
            
            if {
                TycoonManager := MaybeTycoonManager?
                BuyCurrency := BuyCurrencyList[ItemIndex]
                PriceToBuy := PriceToBuyList[ItemIndex]
                CinematicOnBuy := CinematicOnBuyList[ItemIndex]
                SwitchTurnOffOnBuy := SwitchTurnOffOnBuyList[ItemIndex]
                MoneyIncrease := MoneyIncomeIncreaseAmountAfterBuy[ItemIndex]
                FansIncrease := FansIncomeIncreaseAmountAfterBuy[ItemIndex]
                DiamondsIncrease := DiamondsIncomeIncreaseAmountAfterBuy[ItemIndex]
            } then {
                Print("City owner device: OnBuildingSwitch 333")
                # Check if the player already owns this city building
                if {
                    BuildingState := PlayerData.CityOwnerStateList[ItemIndex]
                    BuildingState = 0
                } then {
                    # Check if the player has enough money
                    var PurchaseSucceeded: logic = false
                    case(BuyCurrency) {
                        tycoon_currency.Money =>
                            if(PlayerData.CurMoney >= PriceToBuy) {
                                PlayerData.RemoveMoney(PriceToBuy)
                                set PurchaseSucceeded = true
                            }
                        tycoon_currency.Fans =>
                            if(PlayerData.CurFans >= PriceToBuy * 1.0) {
                                # Fans are not removed
                                set PurchaseSucceeded = true
                            }
                        tycoon_currency.Diamonds =>
                            if(PlayerData.CurDiamonds >= PriceToBuy * 1.0) {
                                PlayerData.RemoveDiamonds(PriceToBuy)
                                set PurchaseSucceeded = true
                            }
                        tycoon_currency.Rebirths =>
                            if(PlayerData.NumOfRebirths >= PriceToBuy) {
                                set PurchaseSucceeded = true
                            }
                        tycoon_currency.PlayerLevel =>
                            # Levels start with 0, so we have to add 1 to it
                            if((PlayerData.CurLevel + 1) >= PriceToBuy) {
                                set PurchaseSucceeded = true
                            }
                    }
                    # Check if we could purchase the button
                    if(PurchaseSucceeded = true) {
                        Print("City owner device: OnBuildingSwitch 444")
                        PlayerData.SetCityOwnerItemState(ItemIndex, 1)
                        CinematicOnBuy.Play(Agent)
                        AudioPurchaseSuccess.Play(Agent)
                        SwitchTurnOffOnBuy.TurnOn(Agent)
                        TycoonManager.IncreaseMoneyIncome(Agent, MoneyIncrease)
                        TycoonManager.IncreaseFansIncome(FansIncrease)
                        TycoonManager.IncreaseDiamondsIncome(DiamondsIncrease)
                    } else {
                        HudNotEnoughCurrency.Show(Agent)
                    }
                } else {
                    HudBuildingAlreadyOwned.Show(Agent)
                    Print("WARN: City owner device: The player already owns this building")
                }
            } else {
                Print("ERR: City owner device: OnBuildingSwitch: You have to enter a value for each increase property")
            }
        } else {
            Print("ERR: City owner device: Wrong player data")
        }
        #>
    }
}