using { /Fortnite.com/Game }
using { /Fortnite.com/Devices }
using { /Fortnite.com/Devices/CreativeAnimation }
using { /Fortnite.com/Characters }
using { /Verse.org/Simulation }
using { /Verse.org/Simulation/Tags }
using { /Verse.org/Random }
using { /Verse.org/Colors }
using { /Verse.org/Concurrency }
using { /UnrealEngine.com/Temporary/Diagnostics }
using { /UnrealEngine.com/Temporary/SpatialMath }
using { /Fortnite.com/Devices/CreativeAnimation/InterpolationTypes }

button_activation_type := enum {
    Teleport,
    Visibility
}

button_state := enum {
    Hidden,
    Visible,
    QuestActive,
    Purchased
}

tycoon_button<public> := class(creative_device) {
    # Devices
    @editable
    BuyButtonLocation: creative_prop = creative_prop {}

    @editable
    PropToBuy: creative_prop = creative_prop {}

    @editable
    PropBuyImage: creative_prop_asset = DefaultCreativePropAsset

    @editable
    BuyCurrency: tycoon_currency = tycoon_currency.Money

    @editable
    PriceToBuy: int = 0

    @editable
    BuyText: string = ""

    @editable
    DependencyList: []int = array {}

    @editable
    XpReward: int = 0

    @editable
    MoneyIncomeIncreaseAmountAfterBuy: int = 0

    @editable
    FansIncomeIncreaseAmountAfterBuy: float = 0.0

    @editable
    DiamondsIncomeIncreaseAmountAfterBuy: float = 0.0

    @editable
    ButtonActivation: button_activation_type = button_activation_type.Teleport

    @editable
    PropToHideAfterPurchaseList: []creative_prop = array {}

    @editable
    CinematicAfterPurchaseList: []cinematic_sequence_device = array {}

    # MutatorZoneBuy is optional, TycoonButtons normally use a trigger
    @editable
    MutatorZoneBuy: mutator_zone_device = mutator_zone_device {}

    @editable
    HudOnPurchase: ?hud_message_device = false

    @editable
    StartQuestOnBuy: logic = false

    # FollowUpQuestButton is not in use for Scammer Tycoon, and might not work correctly when restoring the tycoon from save data
    @editable
    FollowUpQuestButton: ?tycoon_button = false

    @editable
    ShowDependenciesBeforeQuestCompletion: logic = false

    @editable
    QuestPlayerLevelRequirement<public>: int = 0

    @editable
    QuestObjectiveText<public>: string = ""

    @editable
    QuestObjectiveBeaconPos<public>: creative_prop = creative_prop {}

    @editable
    QuestRewardXpAmount<public>: int = 0

    @editable
    QuestRewardMoneyAmount<public>: int = 0

    @editable
    QuestRewardFansAmount<public>: int = 0

    @editable
    QuestRewardDiamondsAmount<public>: int = 0

    <# Optional properties below #>
    @editable
    QuestRewardItemOnComplete<public>: ?item_granter_device = false

    @editable
    QuestGoalCountAmount<public>: int = 1

    @editable
    QuestGoalKillNpcRarity<public>: ?int = false

    @editable
    QuestGoalScamNpcRarity<public>: ?int = false

    <#
    @editable
    QuestGoalRecruitNpcRarity<public>: ?int = false

    @editable
    QuestGoalBringBackNpcRarity<public>: ?int = false

    @editable
    QuestGoalBringBackNpcStreamerSlotId<public>: ?int = false
    
    @editable
    QuestPlaceHelperInMutatorZone: ?mutator_zone_device = false
    #>

    @editable
    QuestGoalButton<public>: ?button_device = false

    @editable
    QuestGoalTycoonButtonOnPurchase: ?tycoon_button = false

    @editable
    QuestGoalMutatorZone<public>: ?mutator_zone_device = false

    @editable
    QuestGoalSwitchList<public>: []switch_device = array {}

    # Constants
    SpawnAnimOffset: vector3 = vector3{X:=0.0, Y:=0.0, Z:=-400.0}
    SmokePropAfterBuyDuration: float = 2.0

    # Runtime variables
    var ButtonId<public>: int = -1
    
    var MaybeTycoonManager: ?tycoon_manager_device = false
    var MaybeButtonProp: ?creative_prop = false
    var MaybeButtonImageProp: ?creative_prop = false
    var MaybeButtonBuyTextProp: ?creative_prop = false

    var ButtonState: button_state = button_state.Hidden

    var ButtonStartPos: vector3 = vector3 {}
    var ButtonStartRot: rotation = rotation {}
    var PropToBuyStartPos: vector3 = vector3 {}
    var PropToBuyStartRot: rotation = rotation {}
    var PropToHideAfterPurchaseStartPosList: []vector3 = array {}
    var PropToHideAfterPurchaseStartRotList: []rotation = array {}

    # Runs when the device is started in a running game
    OnBegin<override>()<suspends>: void = {
        # Init variables
        if(BuyButtonLocation.IsValid[]) {
            set ButtonStartPos = BuyButtonLocation.GetTransform().Translation
            set ButtonStartRot = BuyButtonLocation.GetTransform().Rotation
        }

        if(PropToBuy.IsValid[]) {
            set PropToBuyStartPos = PropToBuy.GetTransform().Translation
            set PropToBuyStartRot = PropToBuy.GetTransform().Rotation
        }

        for(PropToHide: PropToHideAfterPurchaseList) {
            if(PropToHide.IsValid[]) {
                set PropToHideAfterPurchaseStartPosList = PropToHideAfterPurchaseStartPosList + array{PropToHide.GetTransform().Translation}
                set PropToHideAfterPurchaseStartRotList = PropToHideAfterPurchaseStartRotList + array{PropToHide.GetTransform().Rotation}
            }
        }

        # Listen to events
        MutatorZoneBuy.AgentEntersEvent.Subscribe(OnMutatorZoneBuy)

        # Reset button - Needs sleep to save correct elevator door pos
        Sleep(0.1)
        ResetButton()
    }

    InitButton(TycoonManager: tycoon_manager_device, ParmButtonIndex: int): void = {
        set MaybeTycoonManager = option {TycoonManager}
        set ButtonId = ParmButtonIndex
    }

    # MutatorZone event is bugged and triggers again, if we move the mutator zone
    OnMutatorZoneBuy(Agent: agent): void = {
        if {
            TycoonManager := MaybeTycoonManager?
            OwnerAgent := TycoonManager.MaybeOwnerAgent?
            Agent = OwnerAgent
        } then {
            TycoonManager.TryToBuyProp(Agent, Self)
        }
    }

    ResetButton(): void = {
        # Print("ResetButton")
        Utility := GetUtility()
        set ButtonState = button_state.Hidden
        HidePurchaseButton()
        HidePropToBuy()
        # ResetPropsToHideToStartPos()
        # Optional parameter: Stop cinematics
        if {
            TycoonManager := MaybeTycoonManager?
            Agent := TycoonManager.MaybeOwnerAgent?
        } then {
            for(Cinematic: CinematicAfterPurchaseList) {
                Cinematic.Stop(Agent)
            }
        }
    }

    HidePropToBuy(): void = {
        if(not PropToBuy.IsValid[]) {
            return
        }
        Utility := GetUtility()
        if(ButtonActivation = button_activation_type.Teleport) {
            if(PropToBuy.TeleportTo[Utility.HiddenPos, rotation{}]) {}
        } else {
            PropToBuy.Hide()
        }
    }

    HidePurchaseButton(): void = {
        # Print("HidePurchaseButton")
        if {
            TycoonManager := MaybeTycoonManager?
        } then {
            # Hide the button prop
            if(Button := MaybeButtonProp?) {
                if(Button.IsValid[]) {
                    Button.Dispose()
                }
                set MaybeButtonProp = false
            }
            # Hide the button image prop
            if(ButtonImage := MaybeButtonImageProp?) {
                if(ButtonImage.IsValid[]) {
                    ButtonImage.Dispose()
                }
                set MaybeButtonImageProp = false
            }
            # Hide the button buy text
            if(ButtonBuyText := MaybeButtonBuyTextProp?) {
                if(ButtonBuyText.IsValid[]) {
                    ButtonBuyText.Dispose()
                }
                set MaybeButtonBuyTextProp = false
            }
        }
    }

    ShowButton(): void = {
        # Print("ShowButton")
        Utility := GetUtility()
        if {
            TycoonManager := MaybeTycoonManager?
            CurrencyId := TycoonManager.TycoonConfig.GetTycoonCurrencyId(BuyCurrency)
            ButtonBlueprint := TycoonManager.TycoonConfig.TycoonButtonBlueprintList[CurrencyId]
        } then {
            set ButtonState = button_state.Visible

            # Spawn the button prop
            SpawnTupleButton := SpawnProp(ButtonBlueprint, ButtonStartPos, ButtonStartRot)
            if(SpawnedPropButton := SpawnTupleButton(0)?) {
                set MaybeButtonProp = option {SpawnedPropButton}
            } else {
                Print("WARN: Couldn't spawn a button prop")
            }
            # Spawn the button image
            SpawnTupleImage := SpawnProp(PropBuyImage, ButtonStartPos - Utility.PlayerOffset, ButtonStartRot)
            if(SpawnedPropImage := SpawnTupleImage(0)?) {
                set MaybeButtonImageProp = option{SpawnedPropImage}
            } else {
                Print("WARN: Couldn't spawn a button image prop")
            }
            # SpawnBuyText()
        }
    }

    ResetPropsToHideToStartPos(): void = {
        # Print("ResetPropsToHideToStartPos, ButtonId: {ButtonId}")
        for(ItemIndex -> PropToHide: PropToHideAfterPurchaseList) {
            if {
                StartPos := PropToHideAfterPurchaseStartPosList[ItemIndex]
                StartRot := PropToHideAfterPurchaseStartRotList[ItemIndex]
                PropToHide.IsValid[]
            } then {
                if(PropToHide.TeleportTo[StartPos, StartRot]) {}
            }
        }
    }

    OnPropBought<public>(IsLoadingSaveData: logic): void = {
        Print("OnPropBought, ButtonId: {ButtonId}")
        if {
            TycoonManager := MaybeTycoonManager?
            Agent := TycoonManager.MaybeOwnerAgent?
            PlayerData := Agent.GetPlayerData[]
        } then {
            Utility := GetUtility()
            # Increase income
            TycoonManager.IncreaseMoneyIncome(Agent, MoneyIncomeIncreaseAmountAfterBuy)
            TycoonManager.IncreaseFansIncome(FansIncomeIncreaseAmountAfterBuy)
            TycoonManager.IncreaseDiamondsIncome(DiamondsIncomeIncreaseAmountAfterBuy)
            # Check if we have to init the playtime reward device
            if(TycoonManager.TycoonConfig.PlaytimeRewardButtonId = ButtonId) {
                TycoonManager.TycoonConfig.PlaytimeRewardDevice.OnDeviceUnlocked(Agent)
            }
            if(IsLoadingSaveData = false) {
                if {
                    StartQuestOnBuy = true
                    ShowDependenciesBeforeQuestCompletion = false
                } then {
                    # Overwrite button state until the player completes the quest - This must be first!
                    set ButtonState = button_state.QuestActive
                } else {
                    set ButtonState = button_state.Purchased
                }
                # Show the prop that we bought
                ShowPropWithAnimation()
                # Hide the button prop
                HidePurchaseButton()
                # Hide props if set
                for(PropToHide: PropToHideAfterPurchaseList) {
                    if {
                        PropToHide.IsValid[]
                    } then {
                        Print("HIDE NON SAVE PropToHide, ButtonId: {ButtonId}")
                        if(PropToHide.TeleportTo[Utility.HiddenPos, rotation{}]) {}
                    }
                }
                # Check if the player has a quest to buy this TycoonButton
                if {
                    Quest := PlayerData.MaybeActiveQuest?
                    QuestGoalTycoonButton := Quest.QuestGoalTycoonButtonOnPurchase?
                    QuestGoalTycoonButton.ButtonId = ButtonId
                } then {
                    PlayerData.IncreaseQuestProgress()
                }
                # Optional parameter: Show hud
                if(Hud := HudOnPurchase?) {
                    Hud.Show(Agent)
                }
                # Optional parameter: Start cinematics
                for(Cinematic: CinematicAfterPurchaseList) {
                    Cinematic.Play(Agent)
                }
                # Show smoke effect after purchase
                Task := spawn{ShowSmokePropAfterBuySuspends()}
                # Assign a quest (if property is set)
                if {
                    StartQuestOnBuy = true
                } then {
                    TryToAssignQuest()
                }
            } else {
                # Loading Save Data - Faster load
                set ButtonState = button_state.Purchased
                # Show the prop that we bought
                ShowPropWithoutAnimation()
                # Hide the button prop
                HidePurchaseButton()
                # Hide props if set
                for(PropToHide: PropToHideAfterPurchaseList) {
                    if {
                        PropToHide.IsValid[]
                    } then {
                        Print("HIDE SAVE PropToHide, ButtonId: {ButtonId}")
                        if(PropToHide.TeleportTo[Utility.HiddenPos, rotation{}]) {}
                    }
                }
                # Optional parameter: Start cinematics
                for(Cinematic: CinematicAfterPurchaseList) {
                    Cinematic.Play(Agent)
                }
            }
        }
    }

    ShowSmokePropAfterBuySuspends()<suspends>: void = {
        Utility := GetUtility()
        if {
            TycoonManager := MaybeTycoonManager?
            Agent := TycoonManager.MaybeOwnerAgent?
        } then {
            # Show the prop
            SpawnTuple := SpawnProp(TycoonManager.TycoonConfig.SmokeAfterBuyBp, ButtonStartPos, ButtonStartRot)
            if(SpawnedProp := SpawnTuple(0)?) {
                Sleep(SmokePropAfterBuyDuration)
                if(SpawnedProp.IsValid[]) {
                    SpawnedProp.Dispose()
                }
            } else {
                Print("ERR: ShowSmokePropAfterBuySuspends: Couldn't spawn a prop")
            }
            # Old method
            # if(PropToMove.TeleportTo[ButtonStartPos, ButtonStartRot]) {}
            # Sleep(SmokePropAfterBuyDuration)
            # Hide the prop
            # if(PropToMove.TeleportTo[Utility.HiddenPos, ButtonStartRot]) {}
        }
    }

    OnTycoonReset()<suspends>: void = {
        # Print("TycoonButton - OnTycoonReset")
        ResetButton()
        ResetPropsToHideToStartPos()
    }

    ShowPropWithAnimation(): void = {
        if(not PropToBuy.IsValid[]) {
            return
        }
        if(ButtonActivation = button_activation_type.Teleport) {
            #if(PropToBuy.TeleportTo[PropToBuyStartPos, PropToBuyStartRot]) {}
            if(PropToBuy.TeleportTo[PropToBuyStartPos + SpawnAnimOffset, PropToBuyStartRot]) {}

            Keyframes := array{
                keyframe_delta{
                    DeltaLocation:=vector3{},
                    DeltaRotation:=rotation{},
                    DeltaScale:=vector3{X:=1.0, Y:= 1.0, Z:=1.0},
                    Time:=0.0,
                    Interpolation:=InterpolationTypes.EaseInOut
                },
                keyframe_delta{
                    DeltaLocation:=vector3{Z:=400.0},
                    DeltaRotation:=rotation{},
                    DeltaScale:=vector3{X:=1.0, Y:= 1.0, Z:=1.0},
                    Time:=2.0,
                    Interpolation:=InterpolationTypes.EaseInOut
                }
            }

            <# OneBrick keyframes, which bounce a bit above the finish line
            Keyframes := array{
                keyframe_delta{
                    DeltaLocation:=vector3{},
                    DeltaRotation:=rotation{},
                    DeltaScale:=vector3{X:=1.0, Y:= 1.0, Z:=1.0},
                    Time:=0.0,
                    Interpolation:=InterpolationTypes.EaseInOut
                },
                keyframe_delta{
                    DeltaLocation:=vector3{Z:=300.0},
                    DeltaRotation:=rotation{},
                    DeltaScale:=vector3{X:=1.7,Y:=1.4,Z:=1.1},
                    Time:=0.7,
                    Interpolation:=InterpolationTypes.EaseInOut
                },
                keyframe_delta{
                    DeltaLocation:=vector3{Z:=-100.0},
                    DeltaRotation:=rotation{},
                    DeltaScale:=vector3{X:=1.0/1.7,Y:=1.0/1.4,Z:=1.0/1.1},
                    Time:=1.0,
                    Interpolation:=InterpolationTypes.EaseInOut
                }
            }
            #>

            if(AnimationController:=PropToBuy.GetAnimationController[]) {
                AnimationController.SetAnimation(Keyframes, ?Mode:=animation_mode.OneShot)
                AnimationController.Play()
                # AnimationController.MovementCompleteEvent.Await()
            }
        } else {
            PropToBuy.Show()
        }
    }

    ShowPropWithoutAnimation(): void = {
        if(ButtonActivation = button_activation_type.Teleport) {
            if(PropToBuy.TeleportTo[PropToBuyStartPos, PropToBuyStartRot]) {}
        } else {
            PropToBuy.Show()
        }
    }

    TryToAssignQuest(): void = {
        QuestManager := GetQuestManager()
        if {
            TycoonManager := MaybeTycoonManager?
            Agent := TycoonManager.MaybeOwnerAgent?
            PlayerData := Agent.GetPlayerData[]
        } then {
            # CurrentQuestId check disabled for this game!
            # if(PlayerData.CurrentQuestId <= ButtonId) {
                # Check if the player already had an active quest
                if(PrevQuest := PlayerData.MaybeActiveQuest?) {
                    QuestManager.CleanUpQuestData(Agent, PrevQuest)
                }
                # Assign the quest
                QuestManager.OnQuestAccepted(Agent, Self)
            <#
            } else {
                # Print("WARN: tycoon_button_quest: Not starting quest. The player already has a higher CurrentQuestId: {ButtonId} - {PlayerData.CurrentQuestId}")
            }
            #>
        }
    }

    OnQuestCompleted<public>(): void = {
        if {
            TycoonManager := MaybeTycoonManager?
            OwnerAgent := TycoonManager.MaybeOwnerAgent?
        } then {
            Print("Tycoon Button: OnQuestCompleted: ButtonID: {ButtonId}")
            # Set this button to purchased to enable dependency buttons
            set ButtonState = button_state.Purchased
            # Check if we have to show new buttons
            TycoonManager.UpdateButtonStates()
            # Directly accept the follow up quest (if there is one)
            # Button will be purchased to trigger the quest (must be 0 cost)
            if(NextQuestButton := FollowUpQuestButton?) {
                TycoonManager.TryToBuyProp(OwnerAgent, NextQuestButton)
            }
        }
    }

    <# BuyText disabled for now
    SpawnBuyText(): void = {
        Utility := GetUtility()
        if {
            TycoonManager := MaybeTycoonManager?
            MaterialTextBp := TycoonManager.TycoonConfig.MaterialTextBp
        } then {
            # Adjust Line2NumArray and Line2TextColor based on BuyCurrency
            var Line2NumArrayPrefix: []float = array {}
            var Line2TextColor: color = TycoonManager.TycoonConfig.MaterialColorMoney
            case(BuyCurrency) {
                tycoon_currency.Money =>
                    set Line2TextColor = TycoonManager.TycoonConfig.MaterialColorMoney
                    set Line2NumArrayPrefix = array{TycoonManager.TycoonConfig.MaterialIconIdMoney}
                tycoon_currency.Fans =>
                    set Line2TextColor = TycoonManager.TycoonConfig.MaterialColorScams
                    set Line2NumArrayPrefix = array{TycoonManager.TycoonConfig.MaterialIconIdFans}
                tycoon_currency.Diamonds =>
                    set Line2TextColor = TycoonManager.TycoonConfig.MaterialColorKills
                    set Line2NumArrayPrefix = array{TycoonManager.TycoonConfig.MaterialIconIdDiamonds}
                tycoon_currency.Rebirths =>
                    set Line2TextColor = TycoonManager.TycoonConfig.MaterialColorRebirths
                    set Line2NumArrayPrefix = array{TycoonManager.TycoonConfig.MaterialIconIdRebirths}
                _ =>
                    Print("ERR: TycoonButton: SpawnBuyText: Unknown BuyCurrency")
            }
            
            SpawnTuple := SpawnProp(MaterialTextBp, ButtonStartPos - Utility.PlayerOffset, ButtonStartRot)
            if(SpawnedProp := SpawnTuple(0)?) {
                set MaybeButtonBuyTextProp = option{SpawnedProp}
                # Convert to number with "K"
                Line2String := Utility.FormatFloat(PriceToBuy * 1.0, 0, 0)
                # Utility.SetMaterialText(SpawnedProp, BuyText, Line2String, Line2NumArrayPrefix, option{Line2TextColor})
            } else {
                Print("ERR: UpdateTransform: Couldn't spawn a prop")
            }
        } else {
            Print("ERR: SpawnBuyText: Found no TycoonManager")
        }
    }
    #>
}