using { /Fortnite.com/Game }
using { /Fortnite.com/Devices }
using { /Fortnite.com/Devices/CreativeAnimation }
using { /Fortnite.com/Characters }
using { /Verse.org/Simulation }
using { /Verse.org/Simulation/Tags }
using { /Verse.org/Random }
using { /Verse.org/Assets }
using { /Verse.org/Colors }
using { /Verse.org/Concurrency }
using { /UnrealEngine.com/Temporary/Diagnostics }
using { /UnrealEngine.com/Temporary/SpatialMath }

tycoon_button_achievement_board<public> := class(tycoon_button) {
    @editable
    ActivateChestSwitchList: []switch_device = array {}

    @editable
    ChestLockIconPropList: []creative_prop = array {}

    @editable
    RewardTextPropList: []creative_prop = array {}

    @editable
    ChestLockAllAchievementsCompleted: creative_prop = creative_prop {}

    @editable
    RewardTextAllAchievementsCompleted: creative_prop = creative_prop {}

    @editable
    PropToHideOnAllAchievementsCompleted: creative_prop = creative_prop {}

    # Properties for XP pickaxe block
    @editable
    PickaxeBlockTrigger: trigger_device = trigger_device {}

    @editable
    PickaxeBlockToAnimate: creative_prop = creative_prop {}

    @editable
    XpRewardAmount: int = 0

    # Runtime
    var Reward0Active: logic = false
    var Reward1Active: logic = false
    var Reward2Active: logic = false
    var Reward3Active: logic = false
    var AllRewardsActive: logic = false

    var PickaxeBlockStartPos: vector3 = vector3 {}
    var PickaxeBlockStartRot: rotation = rotation {}

    OnBegin<override>()<suspends>: void = {
        # This must be before super:OnBegin
        if(PickaxeBlockToAnimate.IsValid[]) {
            set PickaxeBlockStartPos = PickaxeBlockToAnimate.GetTransform().Translation
            set PickaxeBlockStartRot = PickaxeBlockToAnimate.GetTransform().Rotation
        }

        (super:)OnBegin()

        for(ItemIndex -> Switch: ActivateChestSwitchList) {
            # TurnedOnEvent
            EventHandler := event_handler_agent(int) {
                A := ItemIndex
                CallbackFunc := ActivateChest
            }

            EventSub := Switch.TurnedOnEvent.Subscribe(EventHandler.EventSubscription)

            # TurnedOffEvent
            EventHandler2 := event_handler_agent(int) {
                A := ItemIndex
                CallbackFunc := DeactivateChest
            }

            EventSub2 := Switch.TurnedOffEvent.Subscribe(EventHandler2.EventSubscription)

            Switch.Disable()
        }
        
        PickaxeBlockTrigger.TriggeredEvent.Subscribe(OnPickaxeBlockTrigger)

        for(Prop: RewardTextPropList) {
            Prop.Hide()
        }
        RewardTextAllAchievementsCompleted.Hide()
    }

    OnPropBought<override>(IsLoadingSaveData: logic): void = {
        (super:)OnPropBought(IsLoadingSaveData)
        SaveManager := GetSaveManager()
        if {
            TycoonManager := MaybeTycoonManager?
            Agent := TycoonManager.MaybeOwnerAgent?
            Save := SaveManager.ProxySaves[Agent]
        } then {
            # Init achievements so they can be incremented
            if(Save.NumOfAchievementsCompletedKill = -1) {
                set Save.NumOfAchievementsCompletedKill = 0
            }
            if(Save.NumOfAchievementsCompletedScam = -1) {
                set Save.NumOfAchievementsCompletedScam = 0
            }
            if(Save.NumOfAchievementsCompletedMoney = -1) {
                set Save.NumOfAchievementsCompletedMoney = 0
            }
            if(Save.NumOfAchievementsCompletedRebirth = -1) {
                set Save.NumOfAchievementsCompletedRebirth = 0
            }
            UpdateAchievementChestState(Agent)
        }
    }

    ActivateChest(Agent: agent, ItemIndex: int): void = {
        SaveManager := GetSaveManager()
        if {
            TycoonManager := MaybeTycoonManager?
            OwnerAgent := TycoonManager.MaybeOwnerAgent?
            Agent = OwnerAgent
            PlayerData := Agent.GetPlayerData[]
            Save := SaveManager.ProxySaves[Agent]
        } then {
            Print("AchievementChest activated: ItemIndex: {ItemIndex}")
            # Activate specific bonus
            if(ItemIndex = 0) {
                set Reward0Active = true
                ChangeAchievementClass(Agent)
            } else if(ItemIndex = 1) {
                set Reward1Active = true
                ChangeAchievementClass(Agent)
            } else if(ItemIndex = 2) {
                set Reward2Active = true
                TycoonManager.ClaimIncome(Agent)
                set PlayerData.RebirthAutomaticIncomeLevel = 0
            } else if(ItemIndex = 3) {
                set Reward3Active = true
            }
        }
    }

    DeactivateChest(Agent: agent, ItemIndex: int): void = {
        GameManager := GetGameManager()
        if {
            TycoonManager := MaybeTycoonManager?
            OwnerAgent := TycoonManager.MaybeOwnerAgent?
            Agent = OwnerAgent
            PlayerData := Agent.GetPlayerData[]
        } then {
            Print("AchievementChest deactivated: ItemIndex: {ItemIndex}")
            # Deactivate bonus based on index
            if(ItemIndex = 0) {
                set Reward0Active = false
                ChangeAchievementClass(Agent)
            } else if(ItemIndex = 1) {
                set Reward1Active = false
                ChangeAchievementClass(Agent)
            } else if(ItemIndex = 2) {
                set Reward2Active = false
                set PlayerData.RebirthAutomaticIncomeLevel = -1
            } else if(ItemIndex = 3) {
                set Reward3Active = false
            }
        }
    }

    ChangeAchievementClass(Agent: agent): void = {
        GameManager := GetGameManager()
        if {
            Reward0Active = true
            Reward1Active = true
        } then {
            GameManager.ClassInstantReloadAndIncreasedMovementSpeed.ChangeClass(Agent)
        } else if(Reward0Active = true) {
            GameManager.ClassInstantReload.ChangeClass(Agent)
        } else if(Reward1Active = true) {
            GameManager.ClassIncreasedMovementSpeed.ChangeClass(Agent)
        } else {
            GameManager.NoClassClassSelector.ChangeClass(Agent)
        }
    }

    # UpdateAchievementChestState is called once you got a new achievement
    UpdateAchievementChestState<public>(Agent: agent): void = {
        SaveManager := GetSaveManager()
        if {
            Save := SaveManager.ProxySaves[Agent]
        } then {
            if(Save.NumOfAchievementsCompletedKill = 10) {
                ActivateAchievementChest(Agent, 0)
            }
            if(Save.NumOfAchievementsCompletedScam = 10) {
                ActivateAchievementChest(Agent, 1)
            }
            if(Save.NumOfAchievementsCompletedMoney = 10) {
                ActivateAchievementChest(Agent, 2)
            }
            if(Save.NumOfAchievementsCompletedRebirth = 10) {
                ActivateAchievementChest(Agent, 3)
            }
        }
    }

    ActivateAchievementChest(Agent: agent, ChestId: int): void = {
        if {
            ActivateChestSwitch := ActivateChestSwitchList[ChestId]
            ChestLockIcon := ChestLockIconPropList[ChestId]
            RewardTextProp := RewardTextPropList[ChestId]
        } then {
            ActivateChestSwitch.Enable()
            ChestLockIcon.Hide()
            RewardTextProp.Show()
            CheckBarrierCondition(Agent)
        } else {
            Print("ERR: ActivateAchievementChest: Unknown ChestId: {ChestId}")
        }
    }

    CheckBarrierCondition(Agent: agent): void = {
        SaveManager := GetSaveManager()
        if {
            Save := SaveManager.ProxySaves[Agent]
            Save.NumOfAchievementsCompletedKill = 10
            Save.NumOfAchievementsCompletedScam = 10
            Save.NumOfAchievementsCompletedMoney = 10
            Save.NumOfAchievementsCompletedRebirth = 10
        } then {
            set AllRewardsActive = true
            RewardTextAllAchievementsCompleted.Show()
            PropToHideOnAllAchievementsCompleted.Hide()
            ChestLockAllAchievementsCompleted.Hide()
        }
    }

    OnPickaxeBlockTrigger(MaybeAgent: ?agent): void = {
        Utility := GetUtility()
        AccoladeManager := GetAccoladeManager()
        if {
            AllRewardsActive = true
            TycoonManager := MaybeTycoonManager?
            OwnerAgent := TycoonManager.MaybeOwnerAgent?
            Agent := MaybeAgent?
            OwnerAgent = Agent
            PlayerData := OwnerAgent.GetPlayerData[]
        } then {
            PlayerData.AddXp(XpRewardAmount, true)
            AccoladeManager.AwardAccolade(Agent, 3)

            # Teleport back to start pos
            if(PickaxeBlockToAnimate.TeleportTo[transform{Translation:=PickaxeBlockStartPos, Rotation:=PickaxeBlockStartRot, Scale:=Utility.UnitVector}]) {}

            # Animate pickaxe block
            Keyframes := array{
                keyframe_delta{
                    DeltaLocation:=vector3{},
                    DeltaRotation:=rotation{},
                    DeltaScale:=vector3{X:=1.0, Y:= 1.0, Z:=1.0},
                    Time:=0.0,
                    Interpolation:=InterpolationTypes.EaseInOut
                },
                keyframe_delta{
                    DeltaLocation:=vector3{Z:=25.0},
                    DeltaRotation:=rotation{},
                    DeltaScale:=vector3{X:=0.3,Y:=0.6,Z:=0.9},
                    Time:=0.25,
                    Interpolation:=InterpolationTypes.EaseInOut
                },
                keyframe_delta{
                    DeltaLocation:=vector3{Z:=-25.0},
                    DeltaRotation:=rotation{},
                    DeltaScale:=vector3{X:=1.0/0.3,Y:=1.0/0.6,Z:=1.0/0.9},
                    Time:=0.25,
                    Interpolation:=InterpolationTypes.EaseInOut
                }
            }

            if(AnimationController := PickaxeBlockToAnimate.GetAnimationController[]) {
                AnimationController.SetAnimation(Keyframes, ?Mode:=animation_mode.OneShot)
                AnimationController.Play()
            }
        }
    }
}