using { /Fortnite.com/Game }
using { /Fortnite.com/Devices }
using { /Fortnite.com/Devices/CreativeAnimation }
using { /Fortnite.com/Characters }
using { /Verse.org/Simulation }
using { /Verse.org/Simulation/Tags }
using { /Verse.org/Random }
using { /Verse.org/Assets }
using { /Verse.org/Concurrency }
using { /UnrealEngine.com/Temporary/Diagnostics }
using { /UnrealEngine.com/Temporary/SpatialMath }

tycoon_button_auto_save<public> := class(tycoon_button) {
    @editable
    SaveTimer: timer_device = timer_device {}

    # Constants

    # Runtime variables

    OnBegin<override>()<suspends>: void = {
        (super:)OnBegin()
        SaveTimer.SuccessEvent.Subscribe(OnSaveTimerSuccess)
    }

    OnPropBought<override>(IsLoadingSaveData: logic): void = {
        (super:)OnPropBought(IsLoadingSaveData)
        CalculateRemainingTime()
        if(IsLoadingSaveData = false) {
            # Immediately perform a save upon first buying this tycoon_button
            SaveTycoonProgress()
        }
        if {
            TycoonManager := MaybeTycoonManager?
            OwnerAgent := TycoonManager.MaybeOwnerAgent?
            PlayerData := OwnerAgent.GetPlayerData[]
        } then {
            set PlayerData.HasUnlockedTycoonAutoSave = true
        }
    }

    OnTycoonReset<override>()<suspends>: void = {
        (super:)OnTycoonReset()
        SaveTimer.ResetForAll()
    }

    # CalculateRemainingTime also starts the timer
    CalculateRemainingTime<public>(): void = {
        SaveManager := GetSaveManager()
        GameConfig := GetGameConfig()
        if {
            TycoonManager := MaybeTycoonManager?
            OwnerAgent := TycoonManager.MaybeOwnerAgent?
            Save := SaveManager.ProxySaves[OwnerAgent]
            SaveUpgradePerLevel := GameConfig.PermanentUpgradeRewardPerLevelList[4]
        } then {
            # 5 sec minimum save duration
            SaveInterval := Max(GameConfig.AutoSaveBaseInterval - (Save.UpgradeFasterSaveLevel * SaveUpgradePerLevel), 5.0)
            CurTimerPercentage := SaveTimer.GetActiveDuration() / SaveTimer.GetMaxDuration()
            RemainingSaveInterval := SaveInterval * CurTimerPercentage
            # SaveTimer.Reset(OwnerAgent)
            SaveTimer.SetMaxDuration(SaveInterval)
            SaveTimer.SetActiveDuration(RemainingSaveInterval)
            spawn{StartSaveTimerSuspends()}
        }
    }

    OnSaveTimerSuccess(MaybeAgent: ?agent): void = {
        SaveManager := GetSaveManager()
        if {
            TycoonManager := MaybeTycoonManager?
            OwnerAgent := TycoonManager.MaybeOwnerAgent?
        } then {
            # SaveTimer.Reset(OwnerAgent)
            SaveTycoonProgress()
            spawn{StartSaveTimerSuspends()}
        }
    }

    SaveTycoonProgress(): void = {
        SaveManager := GetSaveManager()
        if {
            TycoonManager := MaybeTycoonManager?
            OwnerAgent := TycoonManager.MaybeOwnerAgent?
        } then {
            SaveManager.PersistPlayerSave(OwnerAgent)
        }
    }

    # Use a delay to avoid timer getting stuck (?)
    StartSaveTimerSuspends()<suspends>: void = {
        Sleep(1.0)
        if {
            TycoonManager := MaybeTycoonManager?
            OwnerAgent := TycoonManager.MaybeOwnerAgent?
        } then {
            SaveTimer.Start(OwnerAgent)
        }
    }
}