using { /Fortnite.com/Game }
using { /Fortnite.com/Devices }
using { /Fortnite.com/Devices/CreativeAnimation }
using { /Fortnite.com/Characters }
using { /Verse.org/Simulation }
using { /Verse.org/Simulation/Tags }
using { /Verse.org/Random }
using { /Verse.org/Assets }
using { /Verse.org/Concurrency }
using { /UnrealEngine.com/Temporary/Diagnostics }
using { /UnrealEngine.com/Temporary/SpatialMath }

tycoon_button_city_owner := class(tycoon_button) {
    @editable
    SwitchTurnOffOnBuy: switch_device = switch_device {}

    # Constants

    # Runtime variables

    OnBegin<override>()<suspends>: void = {
        (super:)OnBegin()
    }

    OnPropBought<override>(IsLoadingSaveData: logic): void = {
        (super:)OnPropBought(IsLoadingSaveData)

        if {
            TycoonManager := MaybeTycoonManager?
            Agent := TycoonManager.MaybeOwnerAgent?
            PlayerData := Agent.GetPlayerData[]
        } then {
            set PlayerData.HasUnlockedCityOwner = true
            SwitchTurnOffOnBuy.TurnOff(Agent)
        }
    }
}