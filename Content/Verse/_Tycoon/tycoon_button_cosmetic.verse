using { /Fortnite.com/Game }
using { /Fortnite.com/Devices }
using { /Fortnite.com/Devices/CreativeAnimation }
using { /Fortnite.com/Characters }
using { /Verse.org/Simulation }
using { /Verse.org/Simulation/Tags }
using { /Verse.org/Random }
using { /Verse.org/Assets }
using { /Verse.org/Concurrency }
using { /UnrealEngine.com/Temporary/Diagnostics }
using { /UnrealEngine.com/Temporary/SpatialMath }

cosmetic_socket<public> := enum {
    Head
    Back
    Hands
}

tycoon_button_cosmetic<public> := class(tycoon_button) {
    @editable
    ApplyCosmeticOnPurchase: logic = false

    @editable
    CosmeticButton: button_device = button_device {}

    @editable
    ButtonHideCosmetic: button_device = button_device {}

    @editable
    CosmeticSocket<public>: cosmetic_socket = cosmetic_socket.Head

    @editable
    CosmeticVfxList<public>: []visual_effect_powerup_device = array {}

    @editable
    CosmeticVfxRemoverList<public>: []visual_effect_powerup_device = array {}

    @editable
    CosmeticMoneyMultiplier<public>: float = 0.0

    @editable
    CosmeticFansMultiplier<public>: float = 0.0

    @editable
    CosmeticDiamondsMultiplier<public>: float = 0.0

    # Constants

    # Runtime variables

    OnBegin<override>()<suspends>: void = {
        (super:)OnBegin()

        CosmeticButton.InteractedWithEvent.Subscribe(OnButtonPressed)
        ButtonHideCosmetic.InteractedWithEvent.Subscribe(OnButtonHideCosmetic)
    }

    OnPropBought<override>(IsLoadingSaveData: logic): void = {
        (super:)OnPropBought(IsLoadingSaveData)

        if {
            ApplyCosmeticOnPurchase = true
            TycoonManager := MaybeTycoonManager?
            Agent := TycoonManager.MaybeOwnerAgent?
        } then {
            OnButtonPressed(Agent)
        }
    }

    OnButtonPressed(Agent: agent): void = {
        if {
            ButtonState = button_state.QuestActive or ButtonState = button_state.Purchased
            TycoonManager := MaybeTycoonManager?
            OwnerAgent := TycoonManager.MaybeOwnerAgent?
            OwnerAgent = Agent
            PlayerData := Agent.GetPlayerData[]
        } then {
            PlayerData.ApplyTycoonCosmetic(Self)
        }
    }

    OnButtonHideCosmetic(Agent: agent): void = {
        Print("OnButtonHideCosmetic")
        if {
            ButtonState = button_state.QuestActive or ButtonState = button_state.Purchased
            TycoonManager := MaybeTycoonManager?
            OwnerAgent := TycoonManager.MaybeOwnerAgent?
            OwnerAgent = Agent
            PlayerData := Agent.GetPlayerData[]
        } then {
            PlayerData.RemoveCurrentCosmetic(cosmetic_socket.Head, false)
        }
    }
}