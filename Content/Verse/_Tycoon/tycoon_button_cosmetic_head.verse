using { /Fortnite.com/Game }
using { /Fortnite.com/Devices }
using { /Fortnite.com/Devices/CreativeAnimation }
using { /Fortnite.com/Characters }
using { /Verse.org/Simulation }
using { /Verse.org/Simulation/Tags }
using { /Verse.org/Random }
using { /Verse.org/Assets }
using { /Verse.org/Concurrency }
using { /UnrealEngine.com/Temporary/Diagnostics }
using { /UnrealEngine.com/Temporary/SpatialMath }

tycoon_button_cosmetic_head := class(tycoon_button_cosmetic) {
    @editable
    ScamLevelHead: int = 0

    # Constants

    # Runtime variables

    OnBegin<override>()<suspends>: void = {
        (super:)OnBegin()
    }

    OnButtonPressed<override>(Agent: agent): void = {
        (super:)OnButtonPressed(Agent)
        if {
            ButtonState = button_state.QuestActive or ButtonState = button_state.Purchased
            TycoonManager := MaybeTycoonManager?
            OwnerAgent := TycoonManager.MaybeOwnerAgent?
            OwnerAgent = Agent
            PlayerData := Agent.GetPlayerData[]
        } then {
            PlayerData.SetScamLevelHead(ScamLevelHead)
        }
    }
}