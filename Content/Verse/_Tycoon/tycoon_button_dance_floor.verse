using { /Fortnite.com/Game }
using { /Fortnite.com/Devices }
using { /Fortnite.com/Devices/CreativeAnimation }
using { /Fortnite.com/Characters }
using { /Verse.org/Simulation }
using { /Verse.org/Simulation/Tags }
using { /Verse.org/Random }
using { /Verse.org/Assets }
using { /Verse.org/Concurrency }
using { /UnrealEngine.com/Temporary/Diagnostics }
using { /UnrealEngine.com/Temporary/SpatialMath }

tycoon_button_dance_floor := class(tycoon_button) {
    # Devices
    @editable
    DanceMutatorZone: mutator_zone_device = mutator_zone_device {}

    @editable
    RewardCurrency: tycoon_currency = tycoon_currency.Money

    @editable
    RewardFactor: float = 1.0

    # Constants
    RewardTickRate: float = 1.0

    # Runtime variables
    var TaskRewardTick: ?routine = false

    OnBegin<override>()<suspends>: void = {
        (super:)OnBegin()

        DanceMutatorZone.AgentBeginsEmotingEvent.Subscribe(OnStartDance)
        DanceMutatorZone.AgentEndsEmotingEvent.Subscribe(OnStopDance)
    }

    OnStartDance(Agent: agent): void = {
        if {
            ButtonState = button_state.QuestActive or ButtonState = button_state.Purchased
            TycoonManager := MaybeTycoonManager?
            OwnerAgent := TycoonManager.MaybeOwnerAgent?
            OwnerAgent = Agent
        } then {
            CancelRewardTick()
            # Start reward tick
            Temp := SpawnRoutine(RewardTick)
            set TaskRewardTick = option{Temp}
        }
    }

    OnStopDance(Agent: agent): void = {
        if {
            TycoonManager := MaybeTycoonManager?
            OwnerAgent := TycoonManager.MaybeOwnerAgent?
            OwnerAgent = Agent
        } then {
            CancelRewardTick()
        }
    }

    CancelRewardTick(): void = {
        if(Task := TaskRewardTick?) {
            Task.Cancel()
        }
        set TaskRewardTick = false
    }

    RewardTick()<suspends>: void = {
        loop {
            Sleep(RewardTickRate)
            if {
                TycoonManager := MaybeTycoonManager?
                OwnerAgent := TycoonManager.MaybeOwnerAgent?
                PlayerData := OwnerAgent.GetPlayerData[]
            } then {
                PlayerData.RewardPlayerIncomeFactor(RewardCurrency, RewardFactor)
            }
        }
    }
}