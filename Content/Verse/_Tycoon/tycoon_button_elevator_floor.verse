using { /Fortnite.com/Game }
using { /Fortnite.com/Devices }
using { /Fortnite.com/Devices/CreativeAnimation }
using { /Fortnite.com/Characters }
using { /Verse.org/Simulation }
using { /Verse.org/Simulation/Tags }
using { /Verse.org/Random }
using { /Verse.org/Assets }
using { /Verse.org/Concurrency }
using { /UnrealEngine.com/Temporary/Diagnostics }
using { /UnrealEngine.com/Temporary/SpatialMath }

tycoon_button_elevator_floor := class(tycoon_button) {
    @editable
    FloorId: int = 0

    # Constants

    # Runtime variables

    OnBegin<override>()<suspends>: void = {
        (super:)OnBegin()
    }

    OnPropBought<override>(IsLoadingSaveData: logic): void = {
        (super:)OnPropBought(IsLoadingSaveData)

        if {
            TycoonManager := MaybeTycoonManager?
            Agent := TycoonManager.MaybeOwnerAgent?
            PlayerData := Agent.GetPlayerData[]
        } then {
            PlayerData.OnElevatorFloorUnlocked(Agent, FloorId)
        }
    }
}