using { /Fortnite.com/Game }
using { /Fortnite.com/Devices }
using { /Fortnite.com/Devices/CreativeAnimation }
using { /Fortnite.com/Characters }
using { /Verse.org/Simulation }
using { /Verse.org/Simulation/Tags }
using { /Verse.org/Random }
using { /Verse.org/Assets }
using { /Verse.org/Concurrency }
using { /UnrealEngine.com/Temporary/Diagnostics }
using { /UnrealEngine.com/Temporary/SpatialMath }

<#
tycoon_button_emote_show_dance := class(tycoon_button) {
    @editable
    DanceToUnlockIndex: int = 0

    # Constants

    # Runtime variables

    OnBegin<override>()<suspends>: void = {
        (super:)OnBegin()
    }

    OnPropBought<override>(): void = {
        (super:)OnPropBought()

        if {
            TycoonManager := MaybeTycoonManager?
            Agent := TycoonManager.MaybeOwnerAgent?
            PlayerData := Agent.GetPlayerData[]
            set PlayerData.EmoteShowDanceStateList[DanceToUnlockIndex] = 1
        } then {
            # Do nothing here
        } else {
            Print("ERR: tycoon_button_emote_show_dance - DanceToUnlockIndex: {DanceToUnlockIndex}")
        }
    }
}
#>