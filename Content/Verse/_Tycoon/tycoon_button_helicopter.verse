using { /Fortnite.com/Game }
using { /Fortnite.com/Devices }
using { /Fortnite.com/Devices/CreativeAnimation }
using { /Fortnite.com/Characters }
using { /Verse.org/Simulation }
using { /Verse.org/Simulation/Tags }
using { /Verse.org/Random }
using { /Verse.org/Assets }
using { /Verse.org/Concurrency }
using { /UnrealEngine.com/Temporary/Diagnostics }
using { /UnrealEngine.com/Temporary/SpatialMath }

tycoon_button_helicopter := class(tycoon_button) {
    @editable
    RocketDragonMountDevice: dragon_mount_device = dragon_mount_device {}

    @editable
    ButtonActivateRocketMount: button_device = button_device {}

    @editable
    HelicopterSwitchList: []switch_device = array {}

    @editable
    HelicopterCinematicOnSwitchList: []cinematic_sequence_device = array {}

    @editable
    CinematicHelicopterComplete: cinematic_sequence_device = cinematic_sequence_device {}

    @editable
    PropToHideOnActivationList: []creative_prop = array {}

    @editable
    TurnOnAfterPurchaseSwitchList: []switch_device = array {}

    # Constants

    # Runtime variables
    var NumOfPartsFound: int = 0

    var PropToHideOnActivationStartPosList: []vector3 = array {}
    var PropToHideOnActivationStartRotList: []rotation = array {}

    OnBegin<override>()<suspends>: void = {
        (super:)OnBegin()

        for(ItemIndex -> Switch: HelicopterSwitchList) {
            EventHandler := event_handler_agent(int) {
                A := ItemIndex
                CallbackFunc := OnPuzzlePieceFound
            }

            EventSub := Switch.TurnedOffEvent.Subscribe(EventHandler.EventSubscription)
        }

        for(PropToHide: PropToHideOnActivationList) {
            if(PropToHide.IsValid[]) {
                set PropToHideOnActivationStartPosList = PropToHideOnActivationStartPosList + array{PropToHide.GetTransform().Translation}
                set PropToHideOnActivationStartRotList = PropToHideOnActivationStartRotList + array{PropToHide.GetTransform().Rotation}
            }
        }

        ButtonActivateRocketMount.InteractedWithEvent.Subscribe(OnButtonActivateRocketMount)
    }

    OnPropBought<override>(IsLoadingSaveData: logic): void = {
        (super:)OnPropBought(IsLoadingSaveData)

        if {
            TycoonManager := MaybeTycoonManager?
            Agent := TycoonManager.MaybeOwnerAgent?
        } then {
            for(Switch: HelicopterSwitchList) {
                Switch.TurnOn(Agent)
            }
            for(Switch2: TurnOnAfterPurchaseSwitchList) {
                Switch2.TurnOn(Agent)
            }
        }
    }

    OnButtonActivateRocketMount(Agent: agent): void = {
        RocketDragonMountDevice.StartFlying(Agent)
    }

    OnPuzzlePieceFound(Agent: agent, HelicopterPartId: int): void = {
        Utility := GetUtility()
        if {
            ButtonState = button_state.QuestActive or ButtonState = button_state.Purchased
            TycoonManager := MaybeTycoonManager?
            OwnerAgent := TycoonManager.MaybeOwnerAgent?
            OwnerAgent = Agent
            Cinematic := HelicopterCinematicOnSwitchList[HelicopterPartId]
        } then {
            set NumOfPartsFound = NumOfPartsFound + 1
            Cinematic.Play(Agent)
            if(NumOfPartsFound >= HelicopterSwitchList.Length) {
                for(PropToHide: PropToHideOnActivationList) {
                    if(PropToHide.IsValid[]) {
                        if(PropToHide.TeleportTo[Utility.HiddenPos, rotation{}]) {}
                    }
                }
                Task := spawn{EnableHelicopterSuspends(Agent)}
            }
        }
    }

    EnableHelicopterSuspends(Agent: agent)<suspends>: void = {
        Sleep(6.0)
        CinematicHelicopterComplete.Play(Agent)
        Sleep(5.0)
        ButtonActivateRocketMount.Enable()
    }

    OnTycoonReset<override>()<suspends>: void = {
        (super:)OnTycoonReset()

        if {
            TycoonManager := MaybeTycoonManager?
            Agent := TycoonManager.MaybeOwnerAgent?
        } then {
            for(Switch: HelicopterSwitchList) {
                Switch.TurnOff(Agent)
            }
        }

        # Move PropsToHide back to original pos
        for(ItemIndex -> PropToHide: PropToHideOnActivationList) {
            if {
                StartPos := PropToHideOnActivationStartPosList[ItemIndex]
                StartRot := PropToHideOnActivationStartRotList[ItemIndex]
                PropToHide.IsValid[]
            } then {
                if(PropToHide.TeleportTo[StartPos, StartRot]) {}
            }
        }

        ButtonActivateRocketMount.Disable()
    }
}