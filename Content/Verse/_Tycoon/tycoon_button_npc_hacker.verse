using { /Fortnite.com/Game }
using { /Fortnite.com/Devices }
using { /Fortnite.com/Devices/CreativeAnimation }
using { /Fortnite.com/Characters }
using { /Verse.org/Simulation }
using { /Verse.org/Simulation/Tags }
using { /Verse.org/Random }
using { /Verse.org/Assets }
using { /Verse.org/Concurrency }
using { /UnrealEngine.com/Temporary/Diagnostics }
using { /UnrealEngine.com/Temporary/SpatialMath }

tycoon_button_npc_hacker := class(tycoon_button) {
    # Devices
    @editable
    PropToSetMaterial: creative_prop = creative_prop {}

    @editable
    PropToShowOnComplete: creative_prop = creative_prop {}

    @editable
    ChangeMaterialTickRate: float = 10.0

    @editable
    CompletePropShowDuration: float = 2.0

    # Constants
    NpcHackerMaterialList: []material = array{
        # Images.Job_Searching,
        # Images.Job_Progress,
        # Images.Job_Completed
    }

    # Runtime variables
    var ChangeMaterialTick: ?routine = false
    var CurMaterialIndex: int = 0
    var PropOnCompleteStartPos: vector3 = vector3 {}

    OnBegin<override>()<suspends>: void = {
        # This must be first
        if(PropToShowOnComplete.IsValid[]) {
            set PropOnCompleteStartPos = PropToShowOnComplete.GetTransform().Translation
            # Initially hide the prop
            # HideOnCompleteProp()
        }

        (super:)OnBegin()
    }

    OnPropBought<override>(IsLoadingSaveData: logic): void = {
        (super:)OnPropBought(IsLoadingSaveData)
        # Start tick
        Temp := SpawnRoutine(MaterialTick)
        set ChangeMaterialTick = option{Temp}
    }

    OnTycoonReset<override>()<suspends>: void = {
        (super:)OnTycoonReset()

        <#
        HideOnCompleteProp()
        # Cancel tick
        CancelTick()
        #>
    }

    ResetButton<override>(): void = {
        (super:)ResetButton()

        HideOnCompleteProp()
        # Cancel tick
        CancelTick()
    }

    CancelTick(): void = {
        if(Task := ChangeMaterialTick?) {
            Task.Cancel()
        }
        set ChangeMaterialTick = false
    }

    MaterialTick()<suspends>: void = {
        if(not PropToSetMaterial.IsValid[]) {
            return
        }

        loop {
            Sleep(ChangeMaterialTickRate)
            # Change the material
            if(CurMaterialIndex >= NpcHackerMaterialList.Length) {
                set CurMaterialIndex = 0
            }

            if(NewMaterial := NpcHackerMaterialList[CurMaterialIndex]) {
                PropToSetMaterial.SetMaterial(NewMaterial)
            } else {
                Print("ERR: tycoon_button_npc_hacker: MaterialTick: Didn't find material with Id: {CurMaterialIndex}")
            }

            # Show / hide the Completed Blueprint
            # It has sound so we have to move it
            if(CurMaterialIndex = 0) {
                Task := spawn{HideOnCompletePropSuspends()}
            } else if(CurMaterialIndex = 2) {
                if(PropToShowOnComplete.IsValid[]) {
                    if(PropToShowOnComplete.TeleportTo[PropOnCompleteStartPos, PropToShowOnComplete.GetTransform().Rotation]) {}
                }
            }

            set CurMaterialIndex = CurMaterialIndex + 1
        }
    }

    HideOnCompletePropSuspends()<suspends>: void = {
        Sleep(CompletePropShowDuration)
        HideOnCompleteProp()
    }

    HideOnCompleteProp(): void = {
        Utility := GetUtility()
        if {
            PropToShowOnComplete.IsValid[]
            PropToShowOnComplete.TeleportTo[Utility.HiddenPos, PropToShowOnComplete.GetTransform().Rotation]
        }
    }
}