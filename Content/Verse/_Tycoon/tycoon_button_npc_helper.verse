using { /Fortnite.com/Game }
using { /Fortnite.com/Devices }
using { /Fortnite.com/Devices/CreativeAnimation }
using { /Fortnite.com/Characters }
using { /Verse.org/Simulation }
using { /Verse.org/Simulation/Tags }
using { /Verse.org/Random }
using { /Verse.org/Assets }
using { /Verse.org/Concurrency }
using { /UnrealEngine.com/Temporary/Diagnostics }
using { /UnrealEngine.com/Temporary/SpatialMath }

<#
tycoon_button_npc_helper := class(tycoon_button) {
    @editable
    PhoneSignalRemoteManager: signal_remote_manager_device = signal_remote_manager_device {}

    @editable
    NpcBlueprintList: []creative_prop = array {}

    @editable
    NpcBeaconList: []beacon_device = array {}

    @editable
    NpcBillboardPlayerNameList: []billboard_device = array {}

    @editable
    NpcBillboardVictimFactorList: []billboard_device = array {}

    @editable
    NpcCollectButtonList: []button_device = array {}

    @editable
    VfxNpcAvailable: visual_effect_powerup_device = visual_effect_powerup_device {}

    @editable
    VfxRemoverNpcAvailable: visual_effect_powerup_device = visual_effect_powerup_device {}

    @editable
    RewardCurrency: tycoon_currency = tycoon_currency.Fans

    @editable
    RewardStartPercentage: float = 0.8

    @editable
    RewardDecreaseAmount: float = 0.002

    @editable
    HudNpcHelperLowPercentage: hud_message_device = hud_message_device {}

    @editable
    NpcHelperLowPercentageAmount: float = 0.1

    # Constants
    ServerTickRate: float = 1.0

    PlayerNameMessage<localizes>(MyString: string, Agent: agent): message = "{MyString}{Agent}"

    # Runtime variables
    var NpcIsActiveList: []logic = array {}
    var NpcRewardFactorList: []float = array {}
    var RoutineServerTick: ?routine = false

    OnBegin<override>()<suspends>: void = {
        (super:)OnBegin()

        # Listen to events
        PhoneSignalRemoteManager.PrimarySignalEvent.Subscribe(OnPrimaryPhoneSignal)
        PhoneSignalRemoteManager.SecondarySignalEvent.Subscribe(OnSecondaryPhoneSignal)
        # NPC collect button
        for(ButtonIndex -> Button: NpcCollectButtonList) {
            EventHandler := event_handler_agent(int) {
                A := ButtonIndex
                CallbackFunc := OnReturnNpcButton
            }

            EventSub := Button.InteractedWithEvent.Subscribe(EventHandler.EventSubscription)
        }
    }

    InitNpcLists(): void = {
        set NpcIsActiveList = array {}
        for(NpcIndex -> NpcBlueprint: NpcBlueprintList) {
            set NpcIsActiveList = NpcIsActiveList + array {false}
        }

        set NpcRewardFactorList = array {}
        for(NpcIndex -> NpcBlueprint: NpcBlueprintList) {
            set NpcRewardFactorList = NpcRewardFactorList + array {0.0}
        }
    }

    OnPropBought<override>(): void = {
        (super:)OnPropBought()

        if {
            TycoonManager := MaybeTycoonManager?
            Agent := TycoonManager.MaybeOwnerAgent?
        } then {
            # Init variables
            InitNpcLists()
            # Enable back bling VFX
            EnableBackBlingVfx(Agent)
            # Update UI
            UpdateNumOfActiveNpcsUi(Agent)
        }
    }

    OnTycoonReset<override>()<suspends>: void = {
        (super:)OnTycoonReset()
        
        # If the agent used rebirth and didn't leave, then we pass the agent to remove the VFX
        var MaybeAgent: ?agent = false
        if {
            TycoonManager := MaybeTycoonManager?
            Agent := TycoonManager.MaybeOwnerAgent?
        } then {
            set MaybeAgent = option{Agent}
        }
        # Hide NPCs
        for(NpcIndex -> NpcBlueprint: NpcBlueprintList) {
            # This also updates the UI
            HandleReturnNpcToPool(MaybeAgent, NpcIndex)
        }
        # Reset Npc Lists
        InitNpcLists()
        # Clean up routines
        if(Routine := RoutineServerTick?) {
            Routine.Cancel()
        }
        set RoutineServerTick = false
    }

    OnPrimaryPhoneSignal(Agent: agent): void = {
        Print("OnPrimaryPhoneSignal: NpcBlueprintList.Length: {NpcBlueprintList.Length}")
        Utility := GetUtility()
        if {
            ButtonState = button_state.QuestActive or ButtonState = button_state.Purchased
            TycoonManager := MaybeTycoonManager?
            OwnerAgent := TycoonManager.MaybeOwnerAgent?
            Agent = OwnerAgent
            PlayerData := Agent.GetPlayerData[]
            FortCharacter := Agent.GetFortCharacter[]
        } then {
            UnusedNpcIndex := TryToGetUnusedNpcIndex()
            if {
                UnusedNpcIndex >= 0
                NpcBlueprint := NpcBlueprintList[UnusedNpcIndex]
            } then {
                # Flag the Blueprint as in use
                if(set NpcIsActiveList[UnusedNpcIndex] = true) {}
                if(set NpcRewardFactorList[UnusedNpcIndex] = RewardStartPercentage) {}
                # Teleport NpcBlueprint in player camera direction
                CameraPos := FortCharacter.GetViewLocation()
                LookDir: rotation = FortCharacter.GetViewRotation()
                MovementVector := LookDir.RotateVector(Utility.ForwardVector)
                var LookAtPos: vector3 = CameraPos + (MovementVector * 400.0)
                TargetRot := FortCharacter.GetTransform().Rotation
                # Make Rotation
                GoalRot := TargetRot.ApplyYaw(PiFloat * -0.5)
                # We might have to use FortChar Transform instead of camera pos (?)
                # set LookAtPos.Z = CameraPos.Z
                set LookAtPos.Z = FortCharacter.GetTransform().Translation.Z + Utility.PlayerOffset.Z
                Utility.DebugDraw.DrawSphere(Center := LookAtPos, ?DrawDurationPolicy := debug_draw_duration_policy.FiniteDuration, ?Duration := 1.0)
                if(NpcBlueprint.TeleportTo[LookAtPos, GoalRot]) {}
                # Update billboard
                UpdateNpcOwnerBillboard(Agent, UnusedNpcIndex)
                # Update VFX
                NumOfNpcsAvailable := GetNumOfAvailableNpcs()
                if(NumOfNpcsAvailable = 0) {
                    DisableBackBlingVfx(Agent)
                }
                # Update UI
                UpdateNumOfActiveNpcsUi(Agent)
                # Add beacon
                if(Beacon := NpcBeaconList[UnusedNpcIndex]) {
                    Beacon.AddToShowList(Agent)
                }
                # Check if we have to increment quest progress
                if {
                    Quest := PlayerData.MaybeActiveQuest?
                    QuestMutatorZone := Quest.QuestPlaceHelperInMutatorZone?
                    QuestMutatorZone.IsInVolume[Agent]
                } then {
                    PlayerData.IncreaseQuestProgress()
                }
                # Start server tick (if not already running)
                if(not RoutineServerTick?) {
                    MyRoutine := SpawnRoutine(ServerTick)
                    set RoutineServerTick = option{MyRoutine}
                }
            } else {
                Print("WARN: OnPhoneSignal: Found no UnusedNpc")
            }
        }
    }

    OnSecondaryPhoneSignal(Agent: agent): void = {
        Print("OnSecondaryPhoneSignal: NpcBlueprintList.Length: {NpcBlueprintList.Length}")
        Utility := GetUtility()
        if {
            ButtonState = button_state.QuestActive or ButtonState = button_state.Purchased
            TycoonManager := MaybeTycoonManager?
            OwnerAgent := TycoonManager.MaybeOwnerAgent?
            Agent = OwnerAgent
        } then {
            ActiveNpcIndex := TryToGetActiveNpcIndex()
            if {
                ActiveNpcIndex >= 0
            } then {
                # Return one of the active NPCs
                HandleReturnNpcToPool(option{Agent}, ActiveNpcIndex)
            }
        }
    }

    TryToGetActiveNpcIndex(): int = {
        for(NpcIndex -> IsActive: NpcIsActiveList) {
            if(IsActive = true) {
                return NpcIndex
            }
        }
        return -1
    }

    TryToGetUnusedNpcIndex(): int = {
        for(NpcIndex -> IsActive: NpcIsActiveList) {
            if(IsActive = false) {
                return NpcIndex
            }
        }
        return -1
    }

    OnReturnNpcButton(Agent: agent, NpcIndex: int): void = {
        if {
            TycoonManager := MaybeTycoonManager?
            OwnerAgent := TycoonManager.MaybeOwnerAgent?
            OwnerAgent = Agent
        } then {
            HandleReturnNpcToPool(option{Agent}, NpcIndex)
        }
    }

    HandleReturnNpcToPool(MaybeAgent: ?agent, NpcIndex: int): void = {
        Print("ReturnNpcToPool")
        Utility := GetUtility()
        if {
            NpcBlueprint := NpcBlueprintList[NpcIndex]
        } then {
            # Flag the Blueprint as not in use
            if(set NpcIsActiveList[NpcIndex] = false) {}
            # Hide NPC
            if(NpcBlueprint.TeleportTo[Utility.HiddenPos, rotation {}]) {}
            # Update VFX
            if(Agent := MaybeAgent?) {
                NumOfNpcsAvailable := GetNumOfAvailableNpcs()
                if(NumOfNpcsAvailable = 1) {
                    EnableBackBlingVfx(Agent)
                }
                # Remove beacon
                if(Beacon := NpcBeaconList[NpcIndex]) {
                    Beacon.RemoveFromShowList(Agent)
                }
                # Update UI
                UpdateNumOfActiveNpcsUi(Agent)
            }
            # Check if there are no more active NPCs
            Temp := GetNumOfAvailableNpcs()
            if {
                Temp >= NpcIsActiveList.Length
                Routine := RoutineServerTick?
            } then {
                # Clean up routines
                Routine.Cancel()
                set RoutineServerTick = false
            }
        }
    }

    # This must also be called on each player respawn by GameManager
    EnableBackBlingVfx(Agent: agent): void = {
        Print("VfxNpcAvailable.Pickup")
        VfxNpcAvailable.Pickup(Agent)
    }

    DisableBackBlingVfx(Agent: agent): void = {
        Print("VfxRemoverNpcAvailable.Pickup")
        VfxRemoverNpcAvailable.Pickup(Agent)
    }

    UpdateNumOfActiveNpcsUi(Agent: agent): void = {
        if {
            PlayerData := Agent.GetPlayerData[]
        } then {
            NumOfNpcsAvailable := GetNumOfAvailableNpcs()
            PlayerData.PlayerLocalUi.UpdateNpcHelpersUi(NumOfNpcsAvailable, NpcIsActiveList.Length)
        }
    }

    <#
    UpdateBackBlingVfx(Agent: agent): void = {
        NumOfNpcsAvailable := GetNumOfAvailableNpcs()
        # Update VFX
        if(NumOfNpcsAvailable = 0) {
            Print("VfxRemoverNpcAvailable.Pickup")
            VfxRemoverNpcAvailable.Pickup(Agent)
        }
        Print("VfxNpcAvailable.Pickup")
        VfxNpcAvailable.Pickup(Agent)
    }
    #>

    GetNumOfAvailableNpcs(): int = {
        var NumOfNpcsAvailable: int = 0
        # Count how many NPCs are not active
        for(IsActive: NpcIsActiveList) {
            if(IsActive = false) {
                set NumOfNpcsAvailable = NumOfNpcsAvailable + 1
            }
        }
        return NumOfNpcsAvailable
    }

    UpdateNpcOwnerBillboard(Agent: agent, NpcIndex: int): void = {
        Utility := GetUtility()
        if {
            BillboardOwner := NpcBillboardPlayerNameList[NpcIndex]
        } then {
            BillboardOwner.SetText(PlayerNameMessage("Promoter of ", Agent))
        }
    }

    UpdateNpcVictimBillboard(NpcIndex: int): void = {
        Utility := GetUtility()
        if {
            BillboardVictimFactor := NpcBillboardVictimFactorList[NpcIndex]
            VictimFactor := NpcRewardFactorList[NpcIndex]
        } then {
            # Scam by Senix: Show 10x the actual amount
            BillboardVictimFactor.SetText(Utility.FormatMessage("Fans +" + Utility.ConvertFloatToPercentageString(VictimFactor * 1.0 * 10.0)))
        }
    }

    GetCurrentRewardFactor(): float = {
        var RewardFactor: float = 1.0
        for(NpcIndex -> IsActive: NpcIsActiveList) {
            if(IsActive = true) {
                if(RewardIncrease := NpcRewardFactorList[NpcIndex]) {
                    set RewardFactor = RewardFactor + RewardIncrease
                }
            }
        }
        return RewardFactor
    }

    ServerTick()<suspends>: void = {
        loop {
            for(NpcIndex -> NpcReward: NpcRewardFactorList) {
                NewRewardValue := Max(NpcReward - RewardDecreaseAmount, 0.0)
                if(set NpcRewardFactorList[NpcIndex] = NewRewardValue) {}
                UpdateNpcVictimBillboard(NpcIndex)
                # Check if we have to show the hud
                if {
                    NpcReward > NpcHelperLowPercentageAmount and NewRewardValue <= NpcHelperLowPercentageAmount
                    TycoonManager := MaybeTycoonManager?
                    Agent := TycoonManager.MaybeOwnerAgent?
                } then {
                    HudNpcHelperLowPercentage.Show(Agent)
                }
            }

            Sleep(ServerTickRate)
        }
    }
}
#>