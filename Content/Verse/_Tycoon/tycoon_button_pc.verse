using { /Fortnite.com/Game }
using { /Fortnite.com/Devices }
using { /Fortnite.com/Devices/CreativeAnimation }
using { /Fortnite.com/Characters }
using { /Fortnite.com/AI }
using { /Verse.org/Simulation }
using { /Verse.org/Simulation/Tags }
using { /Verse.org/Random }
using { /Verse.org/Assets }
using { /Verse.org/Concurrency }
using { /UnrealEngine.com/Temporary/Diagnostics }
using { /UnrealEngine.com/Temporary/SpatialMath }

tycoon_pc_state := enum {
    MainMenu
    CommunityChatDeviceSelection
    CommunityChat
    EmoteShow
    # Not in use atm
    UpgradesMenu
}

<#
tycoon_button_pc := class(tycoon_button, _Interface.interface_ui_manager) {
    # Devices
    @editable
    ButtonStartPc: button_device = button_device {}

    @editable
    ChairCommunityChat: chair_device = chair_device {}

    @editable
    CinematicCommunityChat: cinematic_sequence_device = cinematic_sequence_device {}

    @editable
    CommunityChatFansRewardFactor: float = 1.0

    @editable
    CommunityChatHypeGainPerLevel: int = 50

    @editable
    EmoteShowDanceChairList: []chair_device = array {}

    @editable
    DanceAudioList: []audio_player_device = array {}

    @editable
    CinematicEmoteShow: cinematic_sequence_device = cinematic_sequence_device {}

    @editable
    VfxEmoteShow: vfx_spawner_device = vfx_spawner_device {}

    @editable
    EmoteShowFansRewardFactor: float = 1.0

    @editable
    EmoteShowHypeGainPerTick: int = 1

    # Runtime variables
    var PcState: tycoon_pc_state = tycoon_pc_state.MainMenu
    var HasRegisteredButtonEnableOnDeath: logic = false

    var MaybeAgentUiPcMainMenu: ?player_ui_pc = false

    var MaybeCurDanceAudio: ?audio_player_device = false

    OnBegin<override>()<suspends>: void = {
        (super:)OnBegin()
        
        ButtonStartPc.InteractedWithEvent.Subscribe(OnStartPcButton)
        ChairCommunityChat.ExitedEvent.Subscribe(EndCommunityChat)
    }

    OnTycoonReset<override>()<suspends>: void = {
        (super:)OnTycoonReset()
    }

    OnPropBought<override>(): void = {
        (super:)OnPropBought()

        # Add custom logic here
        if {
            TycoonManager := MaybeTycoonManager?
            Agent := TycoonManager.MaybeOwnerAgent?
        } then {
            # Init PC Menu UI
            UiPcMenu := player_ui_pc {}
            UiPcMenu.InitWidget(Agent, Self)
            set MaybeAgentUiPcMainMenu = option{UiPcMenu}
        }
    }

    OnStartPcButton(Agent: agent): void = {
        if {
            ButtonState = button_state.QuestActive or ButtonState = button_state.Purchased
            TycoonManager := MaybeTycoonManager?
            OwnerAgent := TycoonManager.MaybeOwnerAgent?
            OwnerAgent = Agent
        } then {
            ShowPcMainMenu(Agent)
        }
    }

    ShowPcMainMenu(Agent: agent): void = {
        Print("ShowPcMainMenu")
        if(PlayerData := Agent.GetPlayerData[]) {
            ButtonStartPc.Disable()
            if(HasRegisteredButtonEnableOnDeath = false) {
                set HasRegisteredButtonEnableOnDeath = true
                PlayerData.RegisterButtonEnableOnDeath(ButtonStartPc)
            }
            set PcState = tycoon_pc_state.MainMenu
            # Add main menu UI
            if(Ui := MaybeAgentUiPcMainMenu?) {
                Ui.AddUiToPlayer(Agent)
            }
        }
    }

    StartCommunityChatDeviceSelection(Agent: agent): void = {
        Print("StartCommunityChatDeviceSelection")
        set PcState = tycoon_pc_state.CommunityChatDeviceSelection
        if {
            PlayerData := Agent.GetPlayerData[]
        } then {
            # Add Device selection UI
            PlayerData.PlayerUiDeviceSelection.AddUiToPlayer(Agent, Self)
        }
    }

    StartCommunityChat(Agent: agent, DeviceId: int): void = {
        Print("StartCommunityChat")
        if {
            PlayerData := Agent.GetPlayerData[]
            UiCommunityChat := PlayerData.PlayerUiCommunityChat
            FortChar := Agent.GetFortCharacter[]
        } then {
            set PcState = tycoon_pc_state.CommunityChat
            ChairCommunityChat.Seat(Agent)
            # Change Camera
            PlayerData.PlayCinematic(CinematicCommunityChat)
            # Add CommunityChat UI
            UiCommunityChat.AddUiToPlayer(Agent, Self, DeviceId)
            # Add hype bar UI
            PlayerData.PlayerLocalUi.ShowHypeBarUi()
            # Add generic chat UI
            PlayerData.PlayerUiChat.AddUiToPlayer(Agent, Self)
            # Put player in stasis
            FortChar.PutInStasis(stasis_args{AllowFalling := true})
        }
    }

    EndCommunityChat(Agent: agent): void = {
        Print("EndCommunityChat")
        if {
            PlayerData := Agent.GetPlayerData[]
        } then {
            # Eject the player from the chair
            ChairCommunityChat.Eject()
            # Remove CommunityChat UI
            PlayerData.PlayerUiCommunityChat.RemoveUiFromPlayer(Agent)
            # Clean up all PC data
            EndPc(Agent)
        }
    }

    StartEmoteShow(Agent: agent): void = {
        Print("StartEmoteShow")
        if {
            PlayerData := Agent.GetPlayerData[]
            UiEmoteShow := PlayerData.PlayerUiEmoteShow
        } then {
            set PcState = tycoon_pc_state.EmoteShow
            PlayerData.PlayCinematic(CinematicEmoteShow)
            VfxEmoteShow.Enable()
            # Add hype bar UI
            PlayerData.PlayerLocalUi.ShowHypeBarUi()
            # Add emote show UI
            UiEmoteShow.AddUiToPlayer(Agent, Self, EmoteShowFansRewardFactor, EmoteShowHypeGainPerTick)
            # Add generic chat UI
            PlayerData.PlayerUiChat.AddUiToPlayer(Agent, Self)
            # Initially make the player perform the first dance
            StartEmoteShowDance(Agent, 0)
        }
    }

    OnEndEmoteShow(Agent: agent): void = {
        Print("OnEndEmoteShow")
        if {
            PcState = tycoon_pc_state.EmoteShow
            PlayerData := Agent.GetPlayerData[]
        } then {
            VfxEmoteShow.Disable()
            StopDanceAudio(Agent)
            for(Chair: EmoteShowDanceChairList) {
                Chair.Eject()
            }
            # Reapply cosmetics
            PlayerData.ReapplyAllCosmetics()
            EndPc(Agent)
        }
    }

    StopDanceAudio(Agent: agent): void = {
        if {
            CurDanceAudio := MaybeCurDanceAudio?
        } then {
            CurDanceAudio.Stop(Agent)
        }
        set MaybeCurDanceAudio = false
    }

    SelectDeviceForCommunityChat(Agent: agent, ItemIndex: int): void = {
        Print("SelectDeviceForCommunityChat: ItemIndex: {ItemIndex}")
        StartCommunityChat(Agent, ItemIndex)
    }

    StartEmoteShowDance(Agent: agent, ItemIndex: int): void = {
        Print("StartEmoteShowDance: ItemIndex: {ItemIndex}")
        # Stop the previous audio
        StopDanceAudio(Agent)
        if {
            NewChair := EmoteShowDanceChairList[ItemIndex]
            DanceAudio := DanceAudioList[ItemIndex]
            PlayerData := Agent.GetPlayerData[]
        } then {
            # Make sure the previous dance gets cancelled
            for(Chair: EmoteShowDanceChairList) {
                Chair.Eject()
            }
            NewChair.Seat(Agent)
            DanceAudio.Play(Agent)
            set MaybeCurDanceAudio = option{DanceAudio}
            # Hide cosmetics during dance
            PlayerData.UnequipAllCosmetics(false)
        } else {
            Print("ERR: tycoon_pc_state.EmoteShow: OnUiButtonClick: Didn't find dance Chair for ID: {ItemIndex}")
        }
    }

    EndPc(Agent: agent): void = {
        Print("EndPc")
        if {
            PlayerData := Agent.GetPlayerData[]
        } then {
            PlayerData.StopCurrentCinematic()
            # Remove chat UI
            PlayerData.PlayerUiChat.RemoveUiFromPlayer(Agent)
            # Remove hype bar UI
            PlayerData.PlayerLocalUi.HideHypeBarUi()
            # Remove player stasis
            if {
                FortChar := Agent.GetFortCharacter[]
            } then {
                FortChar.ReleaseFromStasis()
            }
            # Reset state
            set PcState = tycoon_pc_state.MainMenu
            ButtonStartPc.Enable()
        }
    }

    OnUiButtonClick<override>(Agent: agent, ParmButtonId: int): void = {
        Print("tycoon_button_pc: Verse OnUiButtonClick: ParmButtonId: {ParmButtonId}")

        case(PcState) {
            tycoon_pc_state.MainMenu =>
                # Remove the main menu UI
                if(Ui := MaybeAgentUiPcMainMenu?) {
                    Ui.RemoveUiFromPlayer(Agent)
                }
                # Check next action
                case(ParmButtonId) {
                    <# Id 0 is not in use atm
                    0 =>
                        ShowUpgradesMenu()
                    #>
                    1 =>
                        StartCommunityChatDeviceSelection(Agent)
                    2 =>
                        StartEmoteShow(Agent)
                    _ =>
                        Print("ERR: PC: OnMainMenuButtonClick: Unknown Button ID: {ParmButtonId}")
                }
            tycoon_pc_state.CommunityChatDeviceSelection =>
                SelectDeviceForCommunityChat(Agent, ParmButtonId)
            tycoon_pc_state.EmoteShow =>
                StartEmoteShowDance(Agent, ParmButtonId)
            _ =>
                Print("ERR: tycoon_button_pc: OnUiButtonClick: Unknown PC State")
        }
    }

    OnExitButtonClick<override>(Agent: agent): void = {
        if {
            PcState = tycoon_pc_state.EmoteShow
        } then {
            OnEndEmoteShow(Agent)
        } else {
            EndPc(Agent)
        }
    }

    OnMinigameSuccess<override>(Agent: agent):void = {
        # Called by community chat on level complete
        if {
            PlayerData := Agent.GetPlayerData[]
            UiCommunityChat := PlayerData.PlayerUiCommunityChat
        } then {
            # Give reward
            PlayerData.RewardPlayerIncomeFactor(tycoon_currency.Fans, UiCommunityChat.CurLevel * CommunityChatFansRewardFactor)
            # Increase hype level
            PlayerData.PlayerLocalUi.AddHype(CommunityChatHypeGainPerLevel)
        }
    }
    
    OnMinigameFailure<override>(Agent: agent):void = {
        # Called by community chat when player wants to exit (on timeout)
        EndCommunityChat(Agent)
    }
}
#>