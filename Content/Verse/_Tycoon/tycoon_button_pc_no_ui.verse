using { /Fortnite.com/Game }
using { /Fortnite.com/Devices }
using { /Fortnite.com/Devices/CreativeAnimation }
using { /Fortnite.com/Characters }
using { /Fortnite.com/AI }
using { /Verse.org/Simulation }
using { /Verse.org/Simulation/Tags }
using { /Verse.org/Random }
using { /Verse.org/Assets }
using { /Verse.org/Colors }
using { /Verse.org/Concurrency }
using { /UnrealEngine.com/Temporary/Diagnostics }
using { /UnrealEngine.com/Temporary/SpatialMath }

tycoon_button_pc_no_ui := class(tycoon_button) {
    # Devices
    @editable
    ButtonStartPc: button_device = button_device {}

    @editable
    BuffDurationTimer: timer_device = timer_device {}

    @editable
    BuffDuration: float = 100.0

    @editable
    BuffMoneyIncomeIncreaseFactor: float = 0.5

    @editable
    ActivationXpReward: int = 10

    # Runtime variables
    var RoutineDurationTick: ?routine = false
    var MaybeTimerTextProp: ?creative_prop = false
    var TimestampPcBuffStarted: float = 0.0
    # var RemainingDuration: float = 0.0

    OnBegin<override>()<suspends>: void = {
        (super:)OnBegin()
        
        ButtonStartPc.InteractedWithEvent.Subscribe(OnStartPcButton)
        BuffDurationTimer.SuccessEvent.Subscribe(HandleBuffTimedOut)
    }

    OnPropBought<override>(IsLoadingSaveData: logic): void = {
        (super:)OnPropBought(IsLoadingSaveData)
        # Spawn timer prop
        # SpawnTimerText()
    }

    OnTycoonReset<override>()<suspends>: void = {
        (super:)OnTycoonReset()

        # Cancel tick
        if(Routine := RoutineDurationTick?) {
            Routine.Cancel()
        }
        set RoutineDurationTick = false
    }

    OnStartPcButton(Agent: agent): void = {
        Print("OnStartPcButton")
        if {
            ButtonState = button_state.QuestActive or ButtonState = button_state.Purchased
            TycoonManager := MaybeTycoonManager?
            OwnerAgent := TycoonManager.MaybeOwnerAgent?
            OwnerAgent = Agent
            PlayerData := OwnerAgent.GetPlayerData[]
            # Only start the buff if it's not on cooldown
            # RemainingDuration <= 0.0
        } then {
            # Increase income
            PlayerData.OnPcBuffActivated(BuffMoneyIncomeIncreaseFactor)
            PlayerData.AddXp(ActivationXpReward, true)
            # Update timestamp
            set TimestampPcBuffStarted = GetSimulationElapsedTime()
            # Remaining duration must be set here to prevent starting multiple scams at the same time
            # set RemainingDuration = BuffDuration
            # Start timer
            BuffDurationTimer.SetMaxDuration(BuffDuration)
            BuffDurationTimer.Start()
        }
    }

    HandleBuffTimedOut(MaybeAgent: ?agent): void = {
        Print("HandleBuffTimedOut")
        Utility := GetUtility()
        if {
            TycoonManager := MaybeTycoonManager?
            OwnerAgent := TycoonManager.MaybeOwnerAgent?
            PlayerData := OwnerAgent.GetPlayerData[]
            TimerProp := MaybeTimerTextProp?
        } then {
            # Stop increasing income
            PlayerData.OnPcBuffTimeout()
            # Hide the timer text
            BuffDurationTimer.Reset()
            # Utility.SetMaterialText(TimerProp, "", "", array{}, false)
        }
    }

    <#
    PcBuffActiveTick()<suspends>: void = {
        Utility := GetUtility()
        loop {
            Timestamp := GetSimulationElapsedTime()
            TimePassedSinceEvent := Timestamp - TimestampPcBuffStarted
            set RemainingDuration = BuffDuration - TimePassedSinceEvent
            # Update timer
            if {
                TimerProp := MaybeTimerTextProp?
                RemainingBuffDurationInt := Round[RemainingDuration]
            } then {
                TimeString := Utility.FormatTimeToString(RemainingBuffDurationInt)
                Utility.SetMaterialText(TimerProp, TimeString, "", array{}, false)
            }
            if(RemainingDuration <= 0.0) {
                HandleBuffTimedOut()
                break
            }
            Sleep(1.0)
        }
    }
    #>

    <#
    SpawnTimerText(): void = {
        Utility := GetUtility()
        if {
            TycoonManager := MaybeTycoonManager?
            MaterialTextBp := TycoonManager.TycoonConfig.MaterialTextBp
        } then {
            SpawnTuple := SpawnProp(MaterialTextBp, ButtonStartPos - Utility.PlayerOffset, ButtonStartRot)
            if(SpawnedProp := SpawnTuple(0)?) {
                set MaybeTimerTextProp = option{SpawnedProp}
            } else {
                Print("ERR: SpawnTimerText: Couldn't spawn a prop")
            }
        } else {
            Print("ERR: SpawnTimerText: Found no TycoonManager")
        }
    }
    #>
}