using { /Fortnite.com/Game }
using { /Fortnite.com/Devices }
using { /Fortnite.com/Devices/CreativeAnimation }
using { /Fortnite.com/Characters }
using { /Verse.org/Simulation }
using { /Verse.org/Simulation/Tags }
using { /Verse.org/Random }
using { /Verse.org/Assets }
using { /Verse.org/Concurrency }
using { /UnrealEngine.com/Temporary/Diagnostics }
using { /UnrealEngine.com/Temporary/SpatialMath }

tycoon_button_pet := class(tycoon_button) {
    @editable
    SpawnPetButton: switch_device = switch_device {}

    @editable
    PetNpcSpawner: npc_spawner_device = npc_spawner_device {}

    <#
    @editable
    PetPropTemplate: creative_prop_asset = DefaultCreativePropAsset
    #>

    @editable
    PetMoneyMultiplier: float = 0.0

    @editable
    PetFansMultiplier: float = 0.0

    @editable
    PetDiamondsMultiplier: float = 0.0

    @editable
    PetRecruitChanceMultiplier: float = 0.0

    # Constants

    # Runtime variables

    OnBegin<override>()<suspends>: void = {
        (super:)OnBegin()

        SpawnPetButton.TurnedOffEvent.Subscribe(OnSpawnPetButton)
        PetNpcSpawner.EliminatedEvent.Subscribe(OnPetNpcEliminated)
    }

    OnSpawnPetButton(Agent: agent): void = {
        Print("OnSpawnPetButton")
        if {
            ButtonState = button_state.QuestActive or ButtonState = button_state.Purchased
            TycoonManager := MaybeTycoonManager?
            OwnerAgent := TycoonManager.MaybeOwnerAgent?
            OwnerAgent = Agent
        } then {
            TycoonManager.SpawnPet(OwnerAgent, PetNpcSpawner, PetMoneyMultiplier, PetFansMultiplier, PetDiamondsMultiplier, PetRecruitChanceMultiplier)
        }
    }

    OnPetNpcEliminated(InteractionResult: device_ai_interaction_result): void = {
        Print("OnPetNpcEliminated")
        if {
            TycoonManager := MaybeTycoonManager?
        } then {
            if(SourceAgent := InteractionResult.Source?) {
                if(NpcData := SourceAgent.GetNpcPetData[]) {
                    TycoonManager.HandlePetNpcEliminated(SourceAgent)
                }
            }
        }
    }
}