using { /Fortnite.com/Game }
using { /Fortnite.com/Devices }
using { /Fortnite.com/Devices/CreativeAnimation }
using { /Fortnite.com/Characters }
using { /Verse.org/Simulation }
using { /Verse.org/Simulation/Tags }
using { /Verse.org/Random }
using { /Verse.org/Assets }
using { /Verse.org/Concurrency }
using { /UnrealEngine.com/Temporary/Diagnostics }
using { /UnrealEngine.com/Temporary/SpatialMath }

tycoon_button_pickaxe_block := class(tycoon_button) {
    @editable
    PickaxeBlockTrigger: trigger_device = trigger_device {}

    @editable
    PickaxeBlockToAnimate: creative_prop = creative_prop {}

    @editable
    RewardCurrency: tycoon_currency = tycoon_currency.Money

    @editable
    RewardFactor: float = 0.0
    
    @editable
    FixedRewardAmount: int = 1

    @editable
    XpRewardAmount: int = 0

    # Constants

    # Runtime variables
    var PickaxeBlockStartPos: vector3 = vector3 {}
    var PickaxeBlockStartRot: rotation = rotation {}

    OnBegin<override>()<suspends>: void = {
        # This must be before super:OnBegin
        if(PickaxeBlockToAnimate.IsValid[]) {
            set PickaxeBlockStartPos = PickaxeBlockToAnimate.GetTransform().Translation
            set PickaxeBlockStartRot = PickaxeBlockToAnimate.GetTransform().Rotation
        }

        (super:)OnBegin()

        PickaxeBlockTrigger.TriggeredEvent.Subscribe(OnPickaxeBlockTrigger)
    }

    OnPickaxeBlockTrigger(MaybeAgent: ?agent): void = {
        Utility := GetUtility()
        AccoladeManager := GetAccoladeManager()
        if {
            ButtonState = button_state.QuestActive or ButtonState = button_state.Purchased
            TycoonManager := MaybeTycoonManager?
            OwnerAgent := TycoonManager.MaybeOwnerAgent?
            Agent := MaybeAgent?
            OwnerAgent = Agent
            PlayerData := OwnerAgent.GetPlayerData[]
        } then {
            PlayerData.AddCurrency(RewardCurrency, FixedRewardAmount, true)
            PlayerData.RewardPlayerIncomeFactor(RewardCurrency, RewardFactor)
            PlayerData.AddXp(XpRewardAmount, true)
            # PickaxeAccolade.Award(Agent)
            AccoladeManager.AwardAccolade(Agent, 0)

            # Teleport back to start pos
            if(PickaxeBlockToAnimate.TeleportTo[transform{Translation:=PickaxeBlockStartPos, Rotation:=PickaxeBlockStartRot, Scale:=Utility.UnitVector}]) {}

            # Animate pickaxe block
            Keyframes := array{
                keyframe_delta{
                    DeltaLocation:=vector3{},
                    DeltaRotation:=rotation{},
                    DeltaScale:=vector3{X:=1.0, Y:= 1.0, Z:=1.0},
                    Time:=0.0,
                    Interpolation:=InterpolationTypes.EaseInOut
                },
                keyframe_delta{
                    DeltaLocation:=vector3{Z:=25.0},
                    DeltaRotation:=rotation{},
                    DeltaScale:=vector3{X:=0.3,Y:=0.6,Z:=0.9},
                    Time:=0.25,
                    Interpolation:=InterpolationTypes.EaseInOut
                },
                keyframe_delta{
                    DeltaLocation:=vector3{Z:=-25.0},
                    DeltaRotation:=rotation{},
                    DeltaScale:=vector3{X:=1.0/0.3,Y:=1.0/0.6,Z:=1.0/0.9},
                    Time:=0.25,
                    Interpolation:=InterpolationTypes.EaseInOut
                }
            }

            if(AnimationController := PickaxeBlockToAnimate.GetAnimationController[]) {
                AnimationController.SetAnimation(Keyframes, ?Mode:=animation_mode.OneShot)
                AnimationController.Play()
            }
        }
    }
}