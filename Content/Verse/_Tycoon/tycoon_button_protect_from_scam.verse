using { /Fortnite.com/Game }
using { /Fortnite.com/Devices }
using { /Fortnite.com/Devices/CreativeAnimation }
using { /Fortnite.com/Characters }
using { /Verse.org/Simulation }
using { /Verse.org/Simulation/Tags }
using { /Verse.org/Random }
using { /Verse.org/Assets }
using { /Verse.org/Concurrency }
using { /UnrealEngine.com/Temporary/Diagnostics }
using { /UnrealEngine.com/Temporary/SpatialMath }

tycoon_button_protect_from_scam := class(tycoon_button) {
    @editable
    ButtonStartScam: button_device = button_device {}

    @editable
    HudOwnerWarning: hud_message_device = hud_message_device {}

    @editable
    ScamMutatorZone: mutator_zone_device = mutator_zone_device {}

    @editable
    ScamBaseDuration: float = 10.0

    @editable
    ScamCooldown: float = 300.0

    @editable
    ScamRewardFactorMoney: float = 10.0

    @editable
    ScamRewardXp: float = 10.0

    # Constants
    
    # Runtime variables
    var RoutineScamInProgressTick: ?routine = false
    var RoutineCooldownTick: ?routine = false
    var MaybeTimerCooldownTextProp: ?creative_prop = false
    var MaybeTimerActiveTextProp: ?creative_prop = false
    var TimestampScamCooldownStarted: float = 0.0
    var RemainingCooldown: float = 0.0
    var ScamProgress: float = 0.0
    var EnemiesInScamZoneList: []agent = array {}
    var OwnerIsInZone: logic = false

    OnBegin<override>()<suspends>: void = {
        (super:)OnBegin()

        # Listen to events
        ScamMutatorZone.AgentEntersEvent.Subscribe(OnAgentEnterScamZone)
        ScamMutatorZone.AgentExitsEvent.Subscribe(OnAgentExitScamZone)

        ButtonStartScam.InteractedWithEvent.Subscribe(OnButtonStartScam)
    }

    OnPropBought<override>(IsLoadingSaveData: logic): void = {
        (super:)OnPropBought(IsLoadingSaveData)

        # Spawn timer prop
        # SpawnTimerActiveText()
        # SpawnTimerCooldownText()
    }

    OnTycoonReset<override>()<suspends>: void = {
        (super:)OnTycoonReset()

        # Cancel tick
        if(Routine := RoutineScamInProgressTick?) {
            Routine.Cancel()
        }
        set RoutineScamInProgressTick = false

        if(Routine := RoutineCooldownTick?) {
            Routine.Cancel()
        }
        set RoutineCooldownTick = false

        # Destroy the timer text
        if(TimerTextProp := MaybeTimerActiveTextProp?) {
            if(TimerTextProp.IsValid[]) {
                TimerTextProp.Dispose()
            }
            set MaybeTimerActiveTextProp = false
        }

        if(TimerTextProp := MaybeTimerCooldownTextProp?) {
            if(TimerTextProp.IsValid[]) {
                TimerTextProp.Dispose()
            }
            set MaybeTimerCooldownTextProp = false
        }
    }

    OnAgentEnterScamZone(Agent: agent): void = {
        if {
            ButtonState = button_state.QuestActive or ButtonState = button_state.Purchased
            TycoonManager := MaybeTycoonManager?
            OwnerAgent := TycoonManager.MaybeOwnerAgent?
        } then {
            if {
                Agent = OwnerAgent
            } then {
                set OwnerIsInZone = true
            } else {
                if(Temp := EnemiesInScamZoneList.RemoveFirstElement[Agent]) {
                    Print("WARN: OnAgentEnterScamZone: Agent entered twice without exit!")
                } else {
                    set EnemiesInScamZoneList = EnemiesInScamZoneList + array {Agent}
                }
            }
        }
    }

    OnAgentExitScamZone(Agent: agent): void = {
        if {
            ButtonState = button_state.QuestActive or ButtonState = button_state.Purchased
            TycoonManager := MaybeTycoonManager?
            OwnerAgent := TycoonManager.MaybeOwnerAgent?
        } then {
            if {
                Agent = OwnerAgent
            } then {
                set OwnerIsInZone = false
            } else {
                if(Temp := EnemiesInScamZoneList.RemoveFirstElement[Agent]) {
                    set EnemiesInScamZoneList = Temp
                } else {
                    Print("ERR: OnAgentExitScamZone: Didn't find Agent to remove!")
                }
                # Cancel the scam if all enemies left the zone
                if(EnemiesInScamZoneList.Length = 0) {
                    # Only cancel the scam if it is currently active
                    if(Routine := RoutineScamInProgressTick?) {
                        CancelScam()
                    }
                }
            }
        }
    }

    OnButtonStartScam(Agent: agent): void = {
        Utility := GetUtility()
        if {
            ButtonState = button_state.QuestActive or ButtonState = button_state.Purchased
            TycoonManager := MaybeTycoonManager?
            OwnerAgent := TycoonManager.MaybeOwnerAgent?
            # The scam can only be started by another player
            Agent <> OwnerAgent
            # Only start the scam if it's not on cooldown
            RemainingCooldown <= 0.0
        } then {
            set ScamProgress = 0.0
            HudOwnerWarning.Show(OwnerAgent)
            # Remaining cooldown must be set here to prevent starting multiple scams at the same time
            set RemainingCooldown = ScamCooldown
            # Start tick
            MyRoutine := SpawnRoutine(ScamInProgressTick)
            set RoutineScamInProgressTick = option{MyRoutine}
        }
    }

    CancelScam(): void = {
        Print("CancelScam")
        OnScamEnded()
    }

    HandleScamSuccess(): void = {
        Print("HandleScamSuccess")
        for(ScammingPlayer: EnemiesInScamZoneList) {
            if {
                PlayerData := ScammingPlayer.GetPlayerData[]
            } then {
                PlayerData.RewardPlayerIncomeFactor(_Tycoon.tycoon_currency.Money, ScamRewardFactorMoney)
                PlayerData.AddXp(XpReward, true)
            }
        }

        OnScamEnded()
    }

    OnScamEnded(): void = {
        Print("OnScamEnded")
        # Clean up routines
        if(Routine := RoutineScamInProgressTick?) {
            Routine.Cancel()
        }
        set RoutineScamInProgressTick = false
        # Handle Scam cooldown
        if(Routine := RoutineCooldownTick?) {
            Routine.Cancel()
        }
        # Hide the timer text
        if(TimerTextProp := MaybeTimerActiveTextProp?) {
            if(TimerTextProp.IsValid[]) {
                TimerTextProp.Dispose()
            }
            set MaybeTimerActiveTextProp = false
        }
        set TimestampScamCooldownStarted = GetSimulationElapsedTime()

        MyRoutine := SpawnRoutine(ScamCooldownTick)
        set RoutineCooldownTick = option{MyRoutine}
    }

    ScamInProgressTick()<suspends>: void = {
        Utility := GetUtility()
        loop {
            # Only progress the scam while the owner is not in the zone
            if {
                OwnerIsInZone = false
            } then {
                Print("ScamInProgress: ScamProgress: {ScamProgress}")
                set ScamProgress = ScamProgress + 0.03
                # Calculate current scam duration (this could also use the upgrade in the future)
                ScamDuration := ScamBaseDuration
                # Update timer
                if {
                    RemainingScamDurationInt := Round[ScamDuration - ScamProgress]
                    TimerActiveTextProp := MaybeTimerActiveTextProp?
                } then {
                    TimeString := Utility.FormatTimeToString(RemainingScamDurationInt)
                    # Utility.SetMaterialText(TimerActiveTextProp, TimeString, "", array{}, false)
                }

                if(ScamProgress >= ScamDuration) {
                    HandleScamSuccess()
                    break
                }
            }
            Sleep(0.0)
        }
    }

    ScamCooldownTick()<suspends>: void = {
        Utility := GetUtility()
        loop {
            Timestamp := GetSimulationElapsedTime()
            TimePassedSinceEvent := Timestamp - TimestampScamCooldownStarted
            set RemainingCooldown = ScamCooldown - TimePassedSinceEvent
            # Update timer
            if {
                RemainingScamCooldownInt := Round[RemainingCooldown]
                TimerCooldownTextProp := MaybeTimerCooldownTextProp?
            } then {
                TimeString := Utility.FormatTimeToString(RemainingScamCooldownInt)
                # Utility.SetMaterialText(TimerCooldownTextProp, TimeString, "", array{}, false)
            }
            if(RemainingCooldown <= 0.0) {
                break
            }
            Sleep(1.0)
        }
    }

    <#
    SpawnTimerActiveText(): void = {
        Utility := GetUtility()
        if {
            TycoonManager := MaybeTycoonManager?
            MaterialTextBp := TycoonManager.TycoonConfig.MaterialTextBp
        } then {
            SpawnTuple := SpawnProp(MaterialTextBp, ButtonStartPos - Utility.PlayerOffset, ButtonStartRot)
            if(SpawnedProp := SpawnTuple(0)?) {
                set MaybeTimerActiveTextProp = option{SpawnedProp}
            } else {
                Print("ERR: SpawnTimerActiveText: Couldn't spawn a prop")
            }
        } else {
            Print("ERR: SpawnTimerActiveText: Found no TycoonManager")
        }
    }

    SpawnTimerCooldownText(): void = {
        Utility := GetUtility()
        if {
            TycoonManager := MaybeTycoonManager?
            MaterialTextBp := TycoonManager.TycoonConfig.MaterialTextBp
        } then {
            SpawnTuple := SpawnProp(MaterialTextBp, ButtonStartPos - (2.0 * Utility.PlayerOffset), ButtonStartRot)
            if(SpawnedProp := SpawnTuple(0)?) {
                set MaybeTimerCooldownTextProp = option{SpawnedProp}
            } else {
                Print("ERR: SpawnTimerCooldownText: Couldn't spawn a prop")
            }
        } else {
            Print("ERR: SpawnTimerCooldownText: Found no TycoonManager")
        }
    }
    #>
}