using { /Fortnite.com/Game }
using { /Fortnite.com/Devices }
using { /Fortnite.com/Devices/CreativeAnimation }
using { /Fortnite.com/Characters }
using { /Verse.org/Simulation }
using { /Verse.org/Simulation/Tags }
using { /Verse.org/Random }
using { /Verse.org/Assets }
using { /Verse.org/Concurrency }
using { /UnrealEngine.com/Temporary/Diagnostics }
using { /UnrealEngine.com/Temporary/SpatialMath }

tycoon_button_puzzle := class(tycoon_button) {
    @editable
    PuzzleId: int = 0

    @editable
    PuzzlePieceSwitchList: []switch_device = array {}

    @editable
    TurnOnAfterPurchaseSwitchList: []switch_device = array {}

    @editable
    CinematicOnSwitchTurnedOffList: []cinematic_sequence_device = array {}

    @editable
    CinematicAllPartsFound: cinematic_sequence_device = cinematic_sequence_device {}

    @editable
    DelayBeforeCinematicAllPartsFound: float = 6.0

    @editable
    PropToHideOnActivationList: []creative_prop = array {}

    @editable
    PropPuzzlePieceShowOnLoadList: []creative_prop = array {}

    # Constants

    # Runtime variables
    var NumOfPartsFound: int = 0

    var PropToHideOnActivationStartPosList: []vector3 = array {}
    var PropToHideOnActivationStartRotList: []rotation = array {}

    OnBegin<override>()<suspends>: void = {
        (super:)OnBegin()

        for(ItemIndex -> Switch: PuzzlePieceSwitchList) {
            EventHandler := event_handler_agent(int) {
                A := ItemIndex
                CallbackFunc := OnHelicopterPartFound
            }

            EventSub := Switch.TurnedOffEvent.Subscribe(EventHandler.EventSubscription)
        }

        for(PropToHide: PropToHideOnActivationList) {
            if(PropToHide.IsValid[]) {
                set PropToHideOnActivationStartPosList = PropToHideOnActivationStartPosList + array{PropToHide.GetTransform().Translation}
                set PropToHideOnActivationStartRotList = PropToHideOnActivationStartRotList + array{PropToHide.GetTransform().Rotation}
            }
        }

        # Initially hide pieces
        for(PropPuzzlePiece: PropPuzzlePieceShowOnLoadList) {
            PropPuzzlePiece.Hide()
        }
    }

    OnPropBought<override>(IsLoadingSaveData: logic): void = {
        (super:)OnPropBought(IsLoadingSaveData)

        if {
            TycoonManager := MaybeTycoonManager?
            Agent := TycoonManager.MaybeOwnerAgent?
        } then {
            for(PuzzlePieceSwitch: PuzzlePieceSwitchList) {
                PuzzlePieceSwitch.TurnOn(Agent)
            }
            for(TurnOnAfterPurchaseSwitch: TurnOnAfterPurchaseSwitchList) {
                TurnOnAfterPurchaseSwitch.TurnOn(Agent)
            }

            LoadPuzzleSaveProgress(Agent)
        }
    }

    OnHelicopterPartFound(Agent: agent, HelicopterPartId: int): void = {
        SaveManager := GetSaveManager()
        Utility := GetUtility()
        if {
            ButtonState = button_state.QuestActive or ButtonState = button_state.Purchased
            TycoonManager := MaybeTycoonManager?
            OwnerAgent := TycoonManager.MaybeOwnerAgent?
            OwnerAgent = Agent
            Cinematic := CinematicOnSwitchTurnedOffList[HelicopterPartId]
            Save := SaveManager.ProxySaves[Agent]
        } then {
            set NumOfPartsFound = NumOfPartsFound + 1
            if(PuzzleId = 0) {
                if(set Save.Puzzle0StateList[HelicopterPartId] = 1) {}
            } else if(PuzzleId = 1) {
                if(set Save.Puzzle1StateList[HelicopterPartId] = 1) {}
            }
            Cinematic.Play(Agent)
            if(NumOfPartsFound >= PuzzlePieceSwitchList.Length) {
                for(PropToHide: PropToHideOnActivationList) {
                    if(PropToHide.IsValid[]) {
                        if(PropToHide.TeleportTo[Utility.HiddenPos, rotation{}]) {}
                    }
                }
                Task := spawn{HandleAllPartsFoundSuspends(Agent)}
            }
        }
    }

    HandleAllPartsFoundSuspends(Agent: agent)<suspends>: void = {
        Sleep(DelayBeforeCinematicAllPartsFound)
        CinematicAllPartsFound.Play(Agent)
        Sleep(5.0)
        # Add custom logic for each puzzle (?)
        # ButtonActivateRocketMount.Enable()
    }

    LoadPuzzleSaveProgress(Agent: agent): void = {
        SaveManager := GetSaveManager()
        if {
            Save := SaveManager.ProxySaves[Agent]
        } then {
            var PuzzleStateList: []int = array {}
            var NumOfPuzzlePiecesFound: int = 0
            if(PuzzleId = 0) {
                set PuzzleStateList = Save.Puzzle0StateList
            } else if(PuzzleId = 1) {
                set PuzzleStateList = Save.Puzzle1StateList
            }
            for(ItemIndex -> PuzzleState: PuzzleStateList) {
                if(PuzzleState = 1) {
                    set NumOfPuzzlePiecesFound = NumOfPuzzlePiecesFound + 1
                    if(PropPuzzlePiece := PropPuzzlePieceShowOnLoadList[ItemIndex]) {
                        PropPuzzlePiece.Show()
                    } else {
                        Print("ERR: LoadPuzzleSaveProgress: ItemIndex: {ItemIndex}")
                    }
                }
            }
            if(NumOfPuzzlePiecesFound >= PuzzlePieceSwitchList.Length) {
                CinematicAllPartsFound.Play(Agent)
            }
        }
    }

    OnTycoonReset<override>()<suspends>: void = {
        (super:)OnTycoonReset()

        if {
            TycoonManager := MaybeTycoonManager?
            Agent := TycoonManager.MaybeOwnerAgent?
        } then {
            for(Switch: PuzzlePieceSwitchList) {
                Switch.TurnOff(Agent)
            }
            for(PropPuzzlePiece: PropPuzzlePieceShowOnLoadList) {
                PropPuzzlePiece.Hide()
            }
        }

        # Move PropsToHide back to original pos
        for(ItemIndex -> PropToHide: PropToHideOnActivationList) {
            if {
                StartPos := PropToHideOnActivationStartPosList[ItemIndex]
                StartRot := PropToHideOnActivationStartRotList[ItemIndex]
                PropToHide.IsValid[]
            } then {
                if(PropToHide.TeleportTo[StartPos, StartRot]) {}
            }
        }
    }
}