using { /Fortnite.com/Game }
using { /Fortnite.com/Devices }
using { /Fortnite.com/Devices/CreativeAnimation }
using { /Fortnite.com/Characters }
using { /Verse.org/Simulation }
using { /Verse.org/Simulation/Tags }
using { /Verse.org/Random }
using { /Verse.org/Assets }
using { /Verse.org/Concurrency }
using { /UnrealEngine.com/Temporary/Diagnostics }
using { /UnrealEngine.com/Temporary/SpatialMath }

tycoon_button_rebirth := class(tycoon_button) {
    @editable
    OnlyShowOnRebirthLevel: ?int = false

    @editable
    RebirthButton: button_device = button_device {}

    # Constants

    # Runtime variables

    OnBegin<override>()<suspends>: void = {
        (super:)OnBegin()

        RebirthButton.InteractedWithEvent.Subscribe(OnButtonPressed)
    }

    OnPropBought<override>(IsLoadingSaveData: logic): void = {
        if {
            RequiredRebirthLevel := OnlyShowOnRebirthLevel?
            TycoonManager := MaybeTycoonManager?
            OwnerAgent := TycoonManager.MaybeOwnerAgent?
            PlayerData := OwnerAgent.GetPlayerData[]
            PlayerData.NumOfRebirths <> RequiredRebirthLevel
        } then {
            # Don't show the button if the player is on a different rebirth level
            set ButtonState = button_state.Purchased
        } else {
            (super:)OnPropBought(IsLoadingSaveData)
        }
    }

    OnButtonPressed(Agent: agent): void = {
        if {
            ButtonState = button_state.QuestActive or ButtonState = button_state.Purchased
            TycoonManager := MaybeTycoonManager?
            OwnerAgent := TycoonManager.MaybeOwnerAgent?
            OwnerAgent = Agent
        } then {
            if {
                RequiredRebirthLevel := OnlyShowOnRebirthLevel?
                PlayerData := OwnerAgent.GetPlayerData[]
                PlayerData.NumOfRebirths <> RequiredRebirthLevel
            } then {
                return
            }
            TycoonManager.OnRebirthButton(Agent)
        }
    }
}