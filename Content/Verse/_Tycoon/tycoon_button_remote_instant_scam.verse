using { /Fortnite.com/Game }
using { /Fortnite.com/Devices }
using { /Fortnite.com/Devices/CreativeAnimation }
using { /Fortnite.com/Characters }
using { /Verse.org/Simulation }
using { /Verse.org/Simulation/Tags }
using { /Verse.org/Random }
using { /Verse.org/Assets }
using { /Verse.org/Concurrency }
using { /UnrealEngine.com/Temporary/Diagnostics }
using { /UnrealEngine.com/Temporary/SpatialMath }

tycoon_button_remote_instant_scam := class(tycoon_button) {
    @editable
    SignalRemoteManager: signal_remote_manager_device = signal_remote_manager_device {}

    @editable
    AbilityMaxRange: float = 2500.0

    # Constants
    
    # Runtime variables

    OnBegin<override>()<suspends>: void = {
        (super:)OnBegin()

        # Listen to events
        SignalRemoteManager.PrimarySignalEvent.Subscribe(OnPrimaryPhoneSignal)
        SignalRemoteManager.SecondarySignalEvent.Subscribe(OnSecondaryPhoneSignal)
    }

    OnPrimaryPhoneSignal(Agent: agent): void = {
        HandlePhoneSignal(Agent)
    }

    OnSecondaryPhoneSignal(Agent: agent): void = {
        HandlePhoneSignal(Agent)
    }

    HandlePhoneSignal(Agent: agent): void = {
        Print("HandlePhoneSignal")
        Utility := GetUtility()
        if {
            ButtonState = button_state.QuestActive or ButtonState = button_state.Purchased
            TycoonManager := MaybeTycoonManager?
            OwnerAgent := TycoonManager.MaybeOwnerAgent?
            Agent = OwnerAgent
            PlayerFortChar := Agent.GetFortCharacter[]
        } then {
            Print("HandlePhoneSignal 222")
            # Find all NPCs around the player
            Utility.DebugDraw.DrawSphere(Center := PlayerFortChar.GetTransform().Translation, ?Radius := AbilityMaxRange, ?DrawDurationPolicy := debug_draw_duration_policy.FiniteDuration, ?Duration := 1.0)    
            for(NpcAgent: TycoonManager.TycoonConfig.NpcSpawnerManager.ActiveNpcList) {
                if {
                    NpcData := NpcAgent.GetNpcData[]
                    NpcFortChar := NpcAgent.GetFortCharacter[]
                } then {
                    DistanceToPlayer := Distance(NpcFortChar.GetTransform().Translation, PlayerFortChar.GetTransform().Translation)
                    if(DistanceToPlayer <= AbilityMaxRange) {
                        NpcData.HandleScamSuccess(Agent)
                    }
                } else {
                    Print("ERR: tycoon_button_remote_instant_scam: Couldn't get NpcData")
                }
            }
        }
    }
}