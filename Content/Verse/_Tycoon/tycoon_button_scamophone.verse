using { /Fortnite.com/Game }
using { /Fortnite.com/Devices }
using { /Fortnite.com/Devices/CreativeAnimation }
using { /Fortnite.com/Characters }
using { /Verse.org/Simulation }
using { /Verse.org/Simulation/Tags }
using { /Verse.org/Random }
using { /Verse.org/Assets }
using { /Verse.org/Concurrency }
using { /UnrealEngine.com/Temporary/Diagnostics }
using { /UnrealEngine.com/Temporary/SpatialMath }

tycoon_button_scamophone := class(tycoon_button) {
    # Devices
    @editable
    PhoneButtonList: []button_device = array {}

    @editable
    PhoneActivePropPosList: []creative_prop = array {}
    
    @editable
    PhoneActivePropBp: creative_prop_asset = DefaultCreativePropAsset

    @editable
    PhoneMinCooldown: float = 5.0

    @editable
    PhoneMaxCooldown: float = 120.0

    @editable
    PhoneActiveDuration: float = 10.0

    @editable
    RewardCurrency: tycoon_currency = tycoon_currency.Money

    @editable
    RewardFactor: float = 1.0

    # Constants
    EmptyCreativeProp: creative_prop = creative_prop {}

    # Runtime variables
    # ActiveProp needs to be spawned / destroyed, since it contains a sound
    var PhoneActivePropList: []creative_prop = array {}

    OnBegin<override>()<suspends>: void = {
        (super:)OnBegin()

        # Init variables
        for(ItemIndex -> Button: PhoneButtonList) {
            set PhoneActivePropList = PhoneActivePropList + array{EmptyCreativeProp}
            # Initially disable buttons
            Button.Disable()
        }

        # Listen to events
        for(ItemIndex -> Button: PhoneButtonList) {
            EventHandler := event_handler_agent(int) {
                A := ItemIndex
                CallbackFunc := OnPhoneButtonPressed
            }

            EventSub := Button.InteractedWithEvent.Subscribe(EventHandler.EventSubscription)
        }
    }

    OnPropBought<override>(IsLoadingSaveData: logic): void = {
        (super:)OnPropBought(IsLoadingSaveData)

        # Start phone tasks
        for(PhoneId -> Button: PhoneButtonList) {
            Task := spawn{PhoneIdleSuspends(PhoneId)}
        }
    }

    StartPhoneRing(PhoneId: int)<suspends>: void = {
        if {
            PhoneActivePropPos := PhoneActivePropPosList[PhoneId]
            PhoneButton := PhoneButtonList[PhoneId]
        } then {
            SpawnTuple := SpawnProp(PhoneActivePropBp, PhoneActivePropPos.GetTransform().Translation, PhoneActivePropPos.GetTransform().Rotation)
            if(SpawnedProp := SpawnTuple(0)?) {
                if(set PhoneActivePropList[PhoneId] = SpawnedProp) {}
            } else {
                Print("ERR: StartPhoneRing: Couldn't spawn a prop")
            }

            PhoneButton.Enable()

            Sleep(PhoneActiveDuration)

            # End phone ring in case the agent didn't press the button in time
            EndPhoneRing(PhoneId)
        }
    }

    EndPhoneRing(PhoneId: int): void = {
        # Destroy the spawned prop
        if {
            PhoneActiveProp := PhoneActivePropList[PhoneId]
            PhoneActiveProp.IsValid[]
            PhoneButton := PhoneButtonList[PhoneId]
        } then {
            PhoneActiveProp.Dispose()
            PhoneButton.Disable()
            # Spawn PhoneIdleSuspends so the phone rings again - Must be in the if clause, so we don't call it twice
            Task := spawn{PhoneIdleSuspends(PhoneId)}
        }
    }

    OnPhoneButtonPressed(Agent: agent, PhoneId: int): void = {
        SaveManager := GetSaveManager()
        if {
            TycoonManager := MaybeTycoonManager?
            OwnerAgent := TycoonManager.MaybeOwnerAgent?
            PlayerData := OwnerAgent.GetPlayerData[]
            # Check if the phone is currently ringing - By checking if the prop is active
            PhoneActiveProp := PhoneActivePropList[PhoneId]
            PhoneActiveProp.IsValid[]
        } then {
            PlayerData.RewardPlayerIncomeFactor(RewardCurrency, RewardFactor)
            EndPhoneRing(PhoneId)
            # Check to complete quest "Use scam phone"
            if {
                Quest := PlayerData.MaybeActiveQuest?
                Quest.ButtonId = 272
            } then {
                PlayerData.IncreaseQuestProgress()
            }
            # Increase Achievement progress
            # Earn Money by using Scam Phones 50 Times
            SaveManager.IncrementAchievement(Agent, _Global.achievement_type.Money, 5, 1)
        }
    }

    PhoneIdleSuspends(PhoneId: int)<suspends>: void = {
        RandSleep := GetRandomFloat(PhoneMinCooldown, PhoneMaxCooldown)
        Sleep(RandSleep)
        # Make sure the tycoon is still active (owner didn't rebirth or leave)
        if {
            ButtonState = button_state.QuestActive or ButtonState = button_state.Purchased
        } then {
            StartPhoneRing(PhoneId)
        }
    }
}