using { /Fortnite.com/Game }
using { /Fortnite.com/Devices }
using { /Fortnite.com/Devices/CreativeAnimation }
using { /Fortnite.com/Characters }
using { /Fortnite.com/Vehicles }
using { /Verse.org/Simulation }
using { /Verse.org/Simulation/Tags }
using { /Verse.org/Random }
using { /Verse.org/Concurrency }
using { /UnrealEngine.com/Temporary/Diagnostics }
using { /UnrealEngine.com/Temporary/SpatialMath }

<#
tycoon_button_squat := class(tycoon_button, _Interface.interface_ui_manager) {
    # Devices
    @editable
    ButtonToStart: button_device = button_device {}

    @editable
    ChairIdle: chair_device = chair_device {}

    @editable
    ChairSquat: chair_device = chair_device {}

    @editable
    NumOfSquatsToComplete: int = 15

    @editable
    CooldownPerPlayer: float = 120.0

    @editable
    MinigameRewardCurrency: tycoon_currency = tycoon_currency.Money

    @editable
    MinigameRewardFactor: float = 0.0

    @editable
    HudOnCooldown: hud_message_device = hud_message_device {}

    @editable
    TimerOnStart: timer_device = timer_device {}
    
    @editable
    CinematicOnStart: cinematic_sequence_device = cinematic_sequence_device {}

    @editable
    BackgroundMusic: audio_player_device = audio_player_device {}

    @editable
    AudioOnCorrectInput: audio_player_device = audio_player_device {}

    @editable
    AudioOnWrongInput: audio_player_device = audio_player_device {}

    @editable
    AudioMinigameSuccess: audio_player_device = audio_player_device {}

    @editable
    SquadAnimDuration: float = 1.75

    # Runtime variables
    var IsMinigameActive: logic = false
    var HasRegisteredButtonEnableOnDeath: logic = false

    # var MaybeInputJumpListener: ?cancelable = false
    # var MaybeInputFireListener: ?cancelable = false

    OnBegin<override>()<suspends>: void = {
        (super:)OnBegin()
        
        ButtonToStart.InteractedWithEvent.Subscribe(OnButtonToStart)
    }

    OnTycoonReset<override>()<suspends>: void = {
        (super:)OnTycoonReset()
    }

    OnPropBought<override>(): void = {
        (super:)OnPropBought()
    }

    OnButtonToStart(Agent: agent): void = {
        Print("tycoon_button_squat: OnButtonToStart")
        if {
            TycoonManager := MaybeTycoonManager?
            OwnerAgent := TycoonManager.MaybeOwnerAgent?
            Agent = OwnerAgent
            PlayerData := Agent.GetPlayerData[]
        } then {
            set IsMinigameActive = true
            ButtonToStart.Disable()
            if(HasRegisteredButtonEnableOnDeath = false) {
                set HasRegisteredButtonEnableOnDeath = true
                PlayerData.RegisterButtonEnableOnDeath(ButtonToStart)
            }
            # Start the minigame
            ChairIdle.Seat(Agent)
            PlayerData.PlayerUiSquatMinigame.AddUiToPlayer(Agent, Self, NumOfSquatsToComplete)
            # Add generic chat UI
            PlayerData.PlayerUiChat.AddUiToPlayer(Agent, Self)
            # Add hype bar UI
            PlayerData.PlayerLocalUi.ShowHypeBarUi()
            TimerOnStart.Start(Agent)
            PlayerData.PlayCinematic(CinematicOnStart)
            # Start background music
            BackgroundMusic.Play(Agent)
        }
    }

    OnUiButtonClick<override>(Agent: agent, ParmButtonId: int): void = {
        Print("tycoon_button_squat: Verse OnSliderMinigameButtonClick: ParmButtonId: {ParmButtonId}")

        # Check next action
        case(ParmButtonId) {
            0 =>
                if {
                    PlayerData := Agent.GetPlayerData[]
                    PlayerData.PlayerUiSquatMinigame.SliderIsEnabled = true
                } then {
                    HandleSquatMinigameInput(Agent)
                }
            _ =>
                Print("ERR: OnUiButtonClick: Unknown Button ID: {ParmButtonId}")
        }
    }

    OnExitButtonClick<override>(Agent: agent): void = {
        HandleMinigameFinished(Agent)
    }

    OnMinigameSuccess<override>(Agent:agent):void = {

    }
    
    OnMinigameFailure<override>(Agent:agent):void = {
        Print("OnMinigameFailure")
        # Do nothing here, the player can immediately try again
        HandleMinigameFinished(Agent)
    }

    HandleSquatMinigameInput(Agent: agent): void = {
        Print("HandleSquatMinigameInput")
        if {
            PlayerData := Agent.GetPlayerData[]
            Ui := PlayerData.PlayerUiSquatMinigame
        } then {
            # Calculate the reward percent, with diminishing returns the further from the middle you are
            SliderPosPercent := Ui.SliderPosX / Ui.SliderMaxPos
            var RewardPercent: float = -1.75 + 11 * SliderPosPercent -11 * Pow(SliderPosPercent, 2.0)
            # Prevent negative values
            set RewardPercent = Max(RewardPercent, 0.0)
            Print("SliderPosPercent: {SliderPosPercent}, RewardPercent: {RewardPercent}")
            set Ui.NumOfInputs = Ui.NumOfInputs + 1
            set Ui.SummedAccuracy = Ui.SummedAccuracy + RewardPercent
            set Ui.TotalAccuracy = Ui.SummedAccuracy / (Ui.NumOfInputs * 1.0)
            # Check if the player pressed correctly
            if(SliderPosPercent < 0.4175 or SliderPosPercent > 0.5825) {
                AudioOnWrongInput.Play(Agent)
            } else {
                AudioOnCorrectInput.Play(Agent)
                Task := spawn{HandleSquatAnimationSuspends(option{Agent}, Ui)}
            }
            # Check if we finished the minigame
            if(Ui.NumOfInputs >= Ui.NumOfInputsForSuccess) {
                HandleSquatMinigameSuccess(Agent, Ui.TotalAccuracy)
            }
            Ui.UpdateSliderUi()
        }
    }

    HandleSquatAnimationSuspends(MaybeAgent: ?agent, Ui: player_ui_squat_minigame)<suspends>: void = {
        if(Agent := MaybeAgent?) {
            ChairIdle.Eject()
            ChairSquat.Seat(Agent)
            Ui.ButtonSquat.SetEnabled(false)
        }
        
        Sleep(SquadAnimDuration)

        if {
            IsMinigameActive = true
            Agent := MaybeAgent?
        } then {
            ChairSquat.Eject()
            ChairIdle.Seat(Agent)
            Ui.ButtonSquat.SetEnabled(true)
        }
    }

    HandleSquatMinigameSuccess(Agent: agent, RewardPercent: float): void = {
        Print("OnMinigameSuccess")
        if {
            PlayerData := Agent.GetPlayerData[]
        } then {
            # Play success sound
            AudioMinigameSuccess.Play(Agent)
            HandleMinigameFinished(Agent)
            # Give reward
            PlayerData.RewardPlayerIncomeFactor(MinigameRewardCurrency, MinigameRewardFactor * RewardPercent)
        }
    }

    HandleMinigameFinished(Agent: agent): void = {
        if {
            PlayerData := Agent.GetPlayerData[]
        } then {
            Print("Squat Minigame: HandleMinigameFinished")
            set IsMinigameActive = false
            TimerOnStart.Reset(Agent)
            PlayerData.StopCurrentCinematic()
            # Stop background music
            BackgroundMusic.Stop(Agent)
            # Remove UI
            PlayerData.PlayerUiSquatMinigame.RemoveUiFromPlayer(Agent)
            PlayerData.PlayerUiChat.RemoveUiFromPlayer(Agent)
            # Remove hype bar UI
            PlayerData.PlayerLocalUi.HideHypeBarUi()
            ChairIdle.Eject()
            ChairSquat.Eject()
            ButtonToStart.Enable()
        }
    }
}
#>