using { /Fortnite.com/Game }
using { /Fortnite.com/Devices }
using { /Fortnite.com/Devices/CreativeAnimation }
using { /Fortnite.com/Characters }
using { /Verse.org/Simulation }
using { /Verse.org/Simulation/Tags }
using { /Verse.org/Random }
using { /Verse.org/Assets }
using { /Verse.org/Concurrency }
using { /UnrealEngine.com/Temporary/Diagnostics }
using { /UnrealEngine.com/Temporary/SpatialMath }

<#
tycoon_button_star_level := class(tycoon_button) {
    @editable
    StarLevelAfterPurchase: int = 0

    # Constants

    # Runtime variables

    OnBegin<override>()<suspends>: void = {
        (super:)OnBegin()
    }

    OnPropBought<override>(): void = {
        (super:)OnPropBought()

        if {
            TycoonManager := MaybeTycoonManager?
            OwnerAgent := TycoonManager.MaybeOwnerAgent?
            PlayerData := OwnerAgent.GetPlayerData[]
            StarLevelAfterPurchase > PlayerData.HitmanMaskLevel
        } then {
            PlayerData.SetHitmanMaskLevel(StarLevelAfterPurchase)
        }
    }
}
#>