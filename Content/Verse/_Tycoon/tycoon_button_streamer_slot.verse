using { /Fortnite.com/Game }
using { /Fortnite.com/Devices }
using { /Fortnite.com/Devices/CreativeAnimation }
using { /Fortnite.com/Characters }
using { /Verse.org/Simulation }
using { /Verse.org/Simulation/Tags }
using { /Verse.org/Random }
using { /Verse.org/Assets }
using { /Verse.org/Concurrency }
using { /UnrealEngine.com/Temporary/Diagnostics }
using { /UnrealEngine.com/Temporary/SpatialMath }

<#
tycoon_button_streamer_slot := class(tycoon_button) {
    # Devices
    @editable
    SlotMutatorZone: mutator_zone_device = mutator_zone_device {}

    @editable
    SeatingPos: creative_prop = creative_prop {}

    @editable
    DancingPos: creative_prop = creative_prop {}

    @editable
    TeleportWhenFollowPos: creative_prop = creative_prop {}

    @editable
    PlaneHideWhenInUse: creative_prop = creative_prop {}

    @editable
    QuestArrowRecruitNpcTemplate: creative_prop_asset = DefaultCreativePropAsset

    # Constants
    RewardTickRate: float = 1.0
    ChangeStreamingPositionTickRate: float = 15.0

    # Runtime variables
    var TaskStreamerTick: ?routine = false
    var MaybeSpawnedPropRarityStars: ?creative_prop = false
    var MaybeSpawnedPropSeating: ?creative_prop = false
    var MaybeSpawnedPropDancing: ?creative_prop = false
    var MaybeSpawnedPropQuestArrowAboveNpc: ?creative_prop = false

    var CurrentStreamingPositionIndex: int = 0
    var TimestampLastPositionChange: float = 0.0
    var StreamerFansIncomeFactor: float = 0.0

    OnBegin<override>()<suspends>: void = {
        (super:)OnBegin()

        if {
            SeatingPos.IsValid[]
            DancingPos.IsValid[]
        } then {
            SlotMutatorZone.AgentEntersEvent.Subscribe(OnEnterSlot)
        } else {
            Print("ERR: tycoon_button_streamer_slot: Properties not set!")
        }
    }

    OnPropBought<override>(): void = {
        (super:)OnPropBought()

        if {
            TycoonManager := MaybeTycoonManager?
            Agent := TycoonManager.MaybeOwnerAgent?
            PlayerData := Agent.GetPlayerData[]
        } then {
            PlayerData.RegisterStreamerSlotBought(Self)
        }
    }

    OnTycoonReset<override>()<suspends>: void = {
        (super:)OnTycoonReset()

        DestroyCurrentStreamer()
        
        PlaneHideWhenInUse.Show()

        # Check if we have to destroy the quest arrow
        if {
            SpawnedProp := MaybeSpawnedPropQuestArrowAboveNpc?
            SpawnedProp.IsValid[]
        } then {
            SpawnedProp.Dispose()
        }
        set MaybeSpawnedPropQuestArrowAboveNpc = false
    }

    OnEnterSlot(Agent: agent): void = {
        if {
            ButtonState = button_state.QuestActive or ButtonState = button_state.Purchased
            TycoonManager := MaybeTycoonManager?
            OwnerAgent := TycoonManager.MaybeOwnerAgent?
            OwnerAgent = Agent
        } then {
            Task := spawn{OnAssignNewStreamer(Agent)}
        }
    }

    SpawnQuestArrowAboveNpc(): void = {
        Utility := GetUtility()
        SpawnTuple := SpawnProp(QuestArrowRecruitNpcTemplate, TeleportWhenFollowPos.GetTransform().Translation - Utility.PlayerOffset, TeleportWhenFollowPos.GetTransform().Rotation)
        if(SpawnedProp := SpawnTuple(0)?) {
            set MaybeSpawnedPropQuestArrowAboveNpc = option{SpawnedProp}
        } else {
            Print("ERR: SpawnQuestArrowAboveNpc: Couldn't spawn a quest arrow prop")
        }
    }

    OnAssignNewStreamer(Agent: agent)<suspends>: void = {
        if {
            PlayerData := Agent.GetPlayerData[]
            NewStreamer := PlayerData.MaybeStreamerRecruit?
            NpcData := PlayerData.MaybeStreamerRecruit?
            NpcData.IsMovingToStreamerSlot = false
            NpcFortChar := NpcData.MaybeNpcFortChar?
        } then {
            Print("STREAMER SLOT: OnAssignNewStreamer")
            # Increase bring back NPCs quest counter
            if {
                Quest := PlayerData.MaybeActiveQuest?
                QuestBringBackNpcRarity := Quest.QuestGoalBringBackNpcRarity?
                NewStreamer.NpcRarity >= QuestBringBackNpcRarity
            } then {
                if {
                    QuestStreamerSlotId := QuestGoalBringBackNpcStreamerSlotId?
                    ButtonId = QuestStreamerSlotId
                } then {
                    PlayerData.IncreaseQuestProgress()
                } else {
                    PlayerData.IncreaseQuestProgress()
                }
            }
            # Move the streamer to the slot
            if(NpcFortChar.TeleportTo[TeleportWhenFollowPos.GetTransform().Translation, TeleportWhenFollowPos.GetTransform().Rotation]) {}
            # NavigateToStreamerSlot has suspends - Wait for the streamer to arrive
            NewStreamer.NavigateToStreamerSlot(SeatingPos)
            OnNewStreamerReachedSlot(Agent)
        }
    }

    OnNewStreamerReachedSlot(Agent: agent): void = {
        Utility := GetUtility()
        if {
            PlayerData := Agent.GetPlayerData[]
            NewStreamer := PlayerData.MaybeStreamerRecruit?
            NpcSpawnerId := NewStreamer.NpcSpawnerId
            TycoonManager := MaybeTycoonManager?
            BlueprintRarityStars := TycoonManager.TycoonConfig.StreamerStarBlueprintList[NewStreamer.NpcRarity]
            BlueprintSeating := TycoonManager.TycoonConfig.StreamerSeatingBlueprintList[NpcSpawnerId]
            BlueprintDancing := TycoonManager.TycoonConfig.StreamerDancingBlueprintList[NpcSpawnerId]
            IncomeFactor := TycoonManager.TycoonConfig.StreamerFansIncomeAmountList[NewStreamer.NpcRarity]
        } then {
            set PlayerData.MaybeStreamerRecruit = false
            # Check if we have to destroy the quest arrow
            if {
                SpawnedProp := MaybeSpawnedPropQuestArrowAboveNpc?
                SpawnedProp.IsValid[]
            } then {
                SpawnedProp.Dispose()
            }
            set MaybeSpawnedPropQuestArrowAboveNpc = false
            # Remove the previous streamer from the slot
            DestroyCurrentStreamer()
            # Spawn new streamer props
            # Each Streamer starts in seating pos
            SpawnTupleRarityStars := SpawnProp(BlueprintRarityStars, SeatingPos.GetTransform().Translation, SeatingPos.GetTransform().Rotation)
            if(SpawnedPropRarityStars := SpawnTupleRarityStars(0)?) {
                set MaybeSpawnedPropRarityStars = option{SpawnedPropRarityStars}
            } else {
                Print("ERR: OnAssignNewStreamer: Couldn't spawn a rarity stars prop")
            }

            SpawnTupleSeating := SpawnProp(BlueprintSeating, SeatingPos.GetTransform().Translation, SeatingPos.GetTransform().Rotation)
            if(SpawnedPropSeating := SpawnTupleSeating(0)?) {
                set MaybeSpawnedPropSeating = option{SpawnedPropSeating}
            } else {
                Print("ERR: OnAssignNewStreamer: Couldn't spawn a seating prop")
            }

            SpawnTupleDancing := SpawnProp(BlueprintDancing, Utility.HiddenPos, rotation{})
            if(SpawnedPropDancing := SpawnTupleDancing(0)?) {
                set MaybeSpawnedPropDancing = option{SpawnedPropDancing}
            } else {
                Print("ERR: OnAssignNewStreamer: Couldn't spawn a dancing prop")
            }

            # Update values
            set StreamerFansIncomeFactor = IncomeFactor
            set TimestampLastPositionChange = GetSimulationElapsedTime()
            PlayerData.UpdateStreamerIncome()
            TycoonManager.TycoonConfig.HudStreamerSlotFilled.Show(Agent)
            PlaneHideWhenInUse.Hide()
            # Start streamer tick
            Temp := SpawnRoutine(StreamerTick)
            set TaskStreamerTick = option{Temp}
        }
    }

    DestroyCurrentStreamer(): void = {
        if(Task := TaskStreamerTick?) {
            Task.Cancel()
        }
        set TaskStreamerTick = false

        # Despawn the previous streamer props
        if {
            SpawnedProp := MaybeSpawnedPropRarityStars?
            SpawnedProp.IsValid[]
        } then {
            SpawnedProp.Dispose()
        }
        set MaybeSpawnedPropRarityStars = false

        if {
            SpawnedProp := MaybeSpawnedPropSeating?
            SpawnedProp.IsValid[]
        } then {
            SpawnedProp.Dispose()
        }
        set MaybeSpawnedPropSeating = false

        if {
            SpawnedProp := MaybeSpawnedPropDancing?
            SpawnedProp.IsValid[]
        } then {
            SpawnedProp.Dispose()
        }
        set MaybeSpawnedPropDancing = false
    }

    ChangeStreamingPosition(): void = {
        Utility := GetUtility()
        if {
            SpawnedPropRarityStars := MaybeSpawnedPropRarityStars?
            SpawnedPropSeating := MaybeSpawnedPropSeating?
            SpawnedPropDancing := MaybeSpawnedPropDancing?
        } then {
            set TimestampLastPositionChange = GetSimulationElapsedTime()
            if(CurrentStreamingPositionIndex = 0) {
                # Change from seating to dancing
                set CurrentStreamingPositionIndex = 1
                # Hide props
                if(SpawnedPropSeating.TeleportTo[Utility.HiddenPos, rotation{}]) {}
                # Shop props
                if(SpawnedPropDancing.TeleportTo[DancingPos.GetTransform().Translation, DancingPos.GetTransform().Rotation]) {}
                if(SpawnedPropRarityStars.TeleportTo[DancingPos.GetTransform().Translation, DancingPos.GetTransform().Rotation]) {}
            } else {
                # Change from dancing to seating
                set CurrentStreamingPositionIndex = 0
                # Hide props
                if(SpawnedPropDancing.TeleportTo[Utility.HiddenPos, rotation{}]) {}
                # Shop props
                if(SpawnedPropSeating.TeleportTo[SeatingPos.GetTransform().Translation, SeatingPos.GetTransform().Rotation]) {}
                if(SpawnedPropRarityStars.TeleportTo[SeatingPos.GetTransform().Translation, SeatingPos.GetTransform().Rotation]) {}
            }
        }
    }

    StreamerTick()<suspends>: void = {
        loop {
            # Sleep(RewardTickRate)
            Sleep(ChangeStreamingPositionTickRate)
            if {
                TycoonManager := MaybeTycoonManager?
                OwnerAgent := TycoonManager.MaybeOwnerAgent?
                PlayerData := OwnerAgent.GetPlayerData[]
            } then {
                ChangeStreamingPosition()

                # Rewards are now handled via player income
                <# Old code
                # PlayerData.RewardPlayerIncomeFactor(tycoon_currency.Fans, StreamerFansIncomeFactor)

                # Check if we have to change the streamer pos
                Timestamp := GetSimulationElapsedTime()
                TimeSinceLastNpcElim := Timestamp - TimestampLastPositionChange
                if(TimeSinceLastNpcElim >= ChangeStreamingPositionTickRate) {
                    ChangeStreamingPosition()
                }
                #>
            }
        }
    }
}
#>