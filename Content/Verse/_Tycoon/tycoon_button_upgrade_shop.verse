using { /Fortnite.com/Game }
using { /Fortnite.com/Devices }
using { /Fortnite.com/Devices/CreativeAnimation }
using { /Fortnite.com/Characters }
using { /Verse.org/Simulation }
using { /Verse.org/Simulation/Tags }
using { /Verse.org/Random }
using { /Verse.org/Assets }
using { /Verse.org/Concurrency }
using { /UnrealEngine.com/Temporary/Diagnostics }
using { /UnrealEngine.com/Temporary/SpatialMath }

# This button exists to assign a quest, which is hardcoded in verse
tycoon_button_upgrade_shop<public> := class(tycoon_button_cosmetic) {
    OnBegin<override>()<suspends>: void = {
        (super:)OnBegin()
    }
}