using { /Fortnite.com/Game }
using { /Fortnite.com/Devices }
using { /Fortnite.com/Devices/CreativeAnimation }
using { /Fortnite.com/Characters }
using { /Fortnite.com/AI }
using { /Verse.org/Simulation }
using { /Verse.org/Simulation/Tags }
using { /Verse.org/Random }
using { /Verse.org/Assets }
using { /Verse.org/Concurrency }
using { /UnrealEngine.com/Temporary/Diagnostics }
using { /UnrealEngine.com/Temporary/SpatialMath }

<#
tycoon_button_whirlpool := class(tycoon_button, interface_ui_manager) {
    # Devices
    @editable
    ChairStartWhirlpool: chair_device = chair_device {}

    @editable
    CinematicWhirlpool: cinematic_sequence_device = cinematic_sequence_device {}

    @editable
    WhirlpoolFansRewardFactor: float = 1.0

    @editable
    WhirlpoolHypeGainPerTick: int = 1

    # Runtime variables
    var RoutinePcTick: ?routine = false
    var IsWhirlpoolActive: logic = false

    # var MaybeInputJumpListener: ?cancelable = false
    # var MaybeInputFireListener: ?cancelable = false

    OnBegin<override>()<suspends>: void = {
        (super:)OnBegin()
        
        ChairStartWhirlpool.SeatedEvent.Subscribe(OnSeatedEvent)
        ChairStartWhirlpool.ExitedEvent.Subscribe(EndWhirlpool)
    }

    OnTycoonReset<override>()<suspends>: void = {
        (super:)OnTycoonReset()
    }

    OnPropBought<override>(): void = {
        (super:)OnPropBought()
    }

    OnSeatedEvent(Agent: agent): void = {
        if {
            ButtonState = button_state.QuestActive or ButtonState = button_state.Purchased
            TycoonManager := MaybeTycoonManager?
            OwnerAgent := TycoonManager.MaybeOwnerAgent?
            OwnerAgent = Agent
        } then {
            StartWhirlpool(Agent)
        }
    }

    StartWhirlpool(Agent: agent): void = {
        Print("StartWhirlpool")
        if {
            PlayerData := Agent.GetPlayerData[]
        } then {
            set IsWhirlpoolActive = true
            PlayerData.PlayCinematic(CinematicWhirlpool)
            # Add generic chat UI
            PlayerData.PlayerUiChat.AddUiToPlayer(Agent, Self)
            # Add hype bar UI
            PlayerData.PlayerLocalUi.ShowHypeBarUi()
            # Start pc tick
            MyRoutine := SpawnRoutine(PcTick)
            set RoutinePcTick = option{MyRoutine}
        }
    }

    EndWhirlpool(Agent: agent): void = {
        Print("EndWhirlpool")
        if {
            IsWhirlpoolActive = true
            TycoonManager := MaybeTycoonManager?
            OwnerAgent := TycoonManager.MaybeOwnerAgent?
            OwnerAgent = Agent
            PlayerData := Agent.GetPlayerData[]
        } then {
            PlayerData.StopCurrentCinematic()
            # Remove chat UI
            PlayerData.PlayerUiChat.RemoveUiFromPlayer(Agent)
            # Remove hype bar UI
            PlayerData.PlayerLocalUi.HideHypeBarUi()
            # Clean up routines
            if(Routine := RoutinePcTick?) {
                Routine.Cancel()
            }
            # Reset state
            set IsWhirlpoolActive = false
        }
    }

    OnUiButtonClick<override>(Agent: agent, ParmButtonId: int): void = {

    }

    OnExitButtonClick<override>(Agent: agent): void = {

    }

    OnMinigameSuccess<override>(Agent:agent):void = {

    }
    
    OnMinigameFailure<override>(Agent:agent):void = {

    }

    PcTick()<suspends>: void = {
        if {
            IsWhirlpoolActive = true
            TycoonManager := MaybeTycoonManager?
            OwnerAgent := TycoonManager.MaybeOwnerAgent?
            PlayerData := OwnerAgent.GetPlayerData[]
        } then {
            loop {
                PlayerData.PlayerLocalUi.AddHype(WhirlpoolHypeGainPerTick)
                PlayerData.RewardPlayerIncomeFactor(tycoon_currency.Fans, WhirlpoolFansRewardFactor + (PlayerData.PlayerLocalUi.HypeCurLevel * PlayerData.HypeTrainIncomeBoostPerLevel))

                Sleep(0.0)
            }
        }
    }
}
#>