using { /Fortnite.com/Game }
using { /Fortnite.com/Devices }
using { /Fortnite.com/Devices/CreativeAnimation }
using { /Fortnite.com/Characters }
using { /Verse.org/Simulation }
using { /Verse.org/Simulation/Tags }
using { /Verse.org/Random }
using { /Verse.org/Assets }
using { /Verse.org/Concurrency }
using { /Verse.org/Colors }
using { /UnrealEngine.com/Temporary/Diagnostics }
using { /UnrealEngine.com/Temporary/SpatialMath }

tycoon_config_device<public> := class(creative_device) {
    # Devices
    @editable
    KingDevice: tycoon_king_device = tycoon_king_device {}

    @editable
    NpcSpawnerManager<public>: _Npc.npc_spawner_manager_device = _Npc.npc_spawner_manager_device {}
    
    @editable
    PlaytimeRewardDevice: playtime_reward_device = playtime_reward_device {}

    @editable
    PlaytimeRewardButtonId: int = -1

    @editable
    FirstStreamerSlotButtonId: int = -1

    # TycoonButtonBlueprintList uses Currency Id as key
    <# Currency Id:
        0: Money
        1: Fans
        2: Diamonds
        3: Rebirths 
        4: Player Level
    #>
    @editable
    TycoonButtonBlueprintList: []creative_prop_asset = array {}

    @editable
    TycoonButtonImageBlueprint: creative_prop_asset = DefaultCreativePropAsset
    
    @editable
    CollectMoneyAudio: audio_player_device = audio_player_device {}

    @editable
    TycoonBaseMusic: audio_player_device = audio_player_device {}

    <#
    @editable
    HudBaseClaimed: hud_message_device = hud_message_device {}
    #>

    @editable
    MessageBuyPropSuccess: hud_message_device = hud_message_device {}

    @editable
    MessageBuyPropFail: hud_message_device = hud_message_device {}

    @editable
    HudNeedDifferentMask<public>: hud_message_device = hud_message_device {}

    @editable
    HudStreamerSlotFilled: hud_message_device = hud_message_device {}

    @editable
    SmokeAfterBuyBp: creative_prop_asset = DefaultCreativePropAsset

    @editable
    StreamerStarBlueprintList: []creative_prop_asset = array {}

    @editable
    StreamerSeatingBlueprintList: []creative_prop_asset = array {}

    @editable
    StreamerDancingBlueprintList: []creative_prop_asset = array {}

    @editable
    StreamerFansIncomeAmountList: []float = array {
        1.0,
        500.0,
        1000.0,
        2000.0,
        5000.0,
        10000.0
    }

    @editable
    StreamerRecruitChanceList: []float = array {
        1.0,
        1.0,
        1.0,
        1.0,
        1.0,
        1.0
    }

    @editable
    ChatMessageIsDonationChance: float = 0.1

    @editable
    ChatDonationMinAmount: int = 1

    @editable
    ChatDonationMaxFactor: float = 2.0

    @editable
    CellarFloorButtonId: int = -1

    @editable
    GarageFloorButtonId: int = -1

    @editable
    SecondFloorButtonId: int = -1

    @editable
    RoofFloorButtonId: int = -1

    <# UnlockCarButtonList must be in TycoonManager, since it is base specific
    @editable
    UnlockCarButtonList: []button_device = array {}
    #>

    @editable
    UnlockCarBuyCurrencyList: []tycoon_currency = array {}

    @editable
    UnlockCarPriceToBuyList: []int = array {}

    <#
    @editable
    MaterialColorMoney: color = color{R:= 0.0, G:= 128.0, B:= 0.0}

    @editable
    MaterialColorScams: color = color{R:= 255.0, G:= 128.0, B:= 0.0}

    @editable
    MaterialColorKills: color = color{R:= 255.0, G:= 255.0, B:= 0.0}

    @editable
    MaterialColorRebirths: color = color{R:= 255.0, G:= 128.0, B:= 0.0}
    #>

    @editable
    UpgradeBaseCostHealthIncrease<public>: int = 100

    @editable
    UpgradeBaseCostCoinsPerKill<public>: int = 200

    @editable
    UpgradeBaseCostScamSpeed<public>: int = 300

    @editable
    UpgradeBaseCostMoneyScam<public>: int = 400

    @editable
    UpgradeBaseCostFasterSave<public>: int = 500

    # Constants
    MaterialIconIdMoney: float = 38.0
    MaterialIconIdFans: float = 39.0
    MaterialIconIdDiamonds: float = 40.0
    MaterialIconIdRebirths: float = 41.0

    # Runs when the device is started in a running game
    OnBegin<override>()<suspends>: void = {
        <#
        for(ItemIndex -> UnlockCarButton: UnlockCarButtonList) {
            EventHandler := event_handler_agent(int) {
                A := ItemIndex
                CallbackFunc := OnUnlockCarButton
            }

            EventSub := UnlockCarButton.InteractedWithEvent.Subscribe(EventHandler.EventSubscription)
        }
        #>
    }

    GetTycoonCurrencyId(Currency: tycoon_currency)<computes>: int = {
        case(Currency) {
            tycoon_currency.Money =>
                return 0
            tycoon_currency.Fans =>
                return 1
            tycoon_currency.Diamonds =>
                return 2
            tycoon_currency.Rebirths =>
                return 3
            tycoon_currency.PlayerLevel =>
                return 4
        }
    }
}