using { /Fortnite.com/Game }
using { /Fortnite.com/Devices }
using { /Fortnite.com/Devices/CreativeAnimation }
using { /Fortnite.com/Characters }
using { /Verse.org/Simulation }
using { /Verse.org/Simulation/Tags }
using { /Verse.org/Random }
using { /Verse.org/Assets }
using { /Verse.org/Concurrency }
using { /UnrealEngine.com/Temporary/Diagnostics }
using { /UnrealEngine.com/Temporary/SpatialMath }

tycoon_elevator_device := class(creative_device, _Interface.interface_ui_manager) {
    @editable
    TycoonId: int = 0

    @editable
    ElevatorProp: creative_prop = creative_prop {}

    @editable
    ButtonCallElevatorList: []button_device = array {}

    @editable
    ElevatorFloorPosList: []creative_prop = array {}

    @editable
    ElevatorLeftDoorList: []creative_prop = array {}

    @editable
    ElevatorRightDoorList: []creative_prop = array {}

    @editable
    ButtonOpenElevatorUi: button_device = button_device {}

    @editable
    ElevatorMovementSpeed: float = 1000.0

    @editable
    DoorOpenDistance: float = 500.0

    @editable
    DoorMovementDuration: float = 3.0

    @editable
    CloseDoorsDelay: float = 5.0

    @editable
    AudioWhileMoving: audio_player_device = audio_player_device {}

    # Runtime
    var MaybeTycoonManager: ?tycoon_manager_device = false
    var LeftDoorClosedPosList: []vector3 = array {}
    var RightDoorClosedPosList: []vector3 = array {}
    var LeftDoorOpenPosList: []vector3 = array {}
    var RightDoorOpenPosList: []vector3 = array {}

    var RoutineCloseDoor: ?routine = false
    var IsMoving: logic = false
    var CurFloorId: int = 0

    OnBegin<override>()<suspends>: void = {
        var LocalLeft: vector3 = vector3{}
        if(TycoonId = 0) {
            set LocalLeft = vector3{X := 0.0, Y := 1.0, Z := 0.0}
        } else if(TycoonId = 1) {
            set LocalLeft = vector3{X := 1.0, Y := 0.0, Z := 0.0}
        } else if(TycoonId = 2) {
            set LocalLeft = vector3{X := 0.0, Y := -1.0, Z := 0.0}
        }
        LocalRight := LocalLeft * -1.0

        # Remember door starting pos
        for(LeftDoor: ElevatorLeftDoorList) {
            DoorClosedPos := LeftDoor.GetTransform().Translation
            set LeftDoorClosedPosList = LeftDoorClosedPosList + array{DoorClosedPos}

            # LocalLeft := LeftDoor.GetTransform().Rotation.Invert().GetLocalRight()
            DoorOpenPos := DoorClosedPos + (DoorOpenDistance * LocalLeft)
            set LeftDoorOpenPosList = LeftDoorOpenPosList + array{DoorOpenPos}
        }

        for(RightDoor: ElevatorRightDoorList) {
            DoorClosedPos := RightDoor.GetTransform().Translation
            set RightDoorClosedPosList = RightDoorClosedPosList + array{DoorClosedPos}

            # LocalRight := RightDoor.GetTransform().Rotation.GetLocalRight()
            DoorOpenPos := DoorClosedPos + (DoorOpenDistance * LocalRight)
            set RightDoorOpenPosList = RightDoorOpenPosList + array{DoorOpenPos}
        }

        # Register events
        for(ItemIndex -> ButtonCallElevator: ButtonCallElevatorList) {
            EventHandler := event_handler_agent(int) {
                A := ItemIndex
                CallbackFunc := OnButtonCallElevator
            }

            EventSub := ButtonCallElevator.InteractedWithEvent.Subscribe(EventHandler.EventSubscription)
        }

        ButtonOpenElevatorUi.InteractedWithEvent.Subscribe(OnButtonOpenElevatorUi)
    }

    InitTycoonManager(TycoonManager: tycoon_manager_device): void = {
        set MaybeTycoonManager = option{TycoonManager}
    }

    OnButtonCallElevator(Agent: agent, FloorId: int): void = {
        if {
            IsMoving = false
        } then {
            CallElevator(Agent, FloorId)
        }
    }

    CallElevator(Agent: agent, FloorId: int): void = {
        Print("CallElevator")
        if {
            TycoonManager := MaybeTycoonManager?
            OwnerAgent := TycoonManager.MaybeOwnerAgent?
            OwnerAgent = Agent
        } then {
            # Teleport the elevator immediately to the call pos
            Task := spawn{TeleportElevator(FloorId)}
        }
    }

    TeleportElevator(FloorId: int)<suspends>: void = {
        if {
            ElevatorPos := ElevatorFloorPosList[FloorId]
        } then {
            Print("TeleportElevator 222")
            CancelCloseDoorAfterDelay()
            set IsMoving = true
            # We don't wait for CloseDoor here
            Task := spawn{CloseDoor(FloorId)}
            if(ElevatorProp.TeleportTo[ElevatorPos.GetTransform()]) {}
            set CurFloorId = FloorId
            OpenDoor(FloorId)
            set IsMoving = false
        } else {
            Print("ERR: Found no ElevatorPos for FloorId: {FloorId}")
        }
    }

    OnButtonOpenElevatorUi(Agent: agent): void = {
        if {
            TycoonManager := MaybeTycoonManager?
            OwnerAgent := TycoonManager.MaybeOwnerAgent?
            OwnerAgent = Agent
        } then {
            if {
                IsMoving = false
                PlayerData := Agent.GetPlayerData[]
            } then {
                PlayerData.UiElevator.AddUiToPlayer(Agent, option{Self})
            }
        }
    }

    OpenDoor(FloorId: int)<suspends>: void = {
        if {
            LeftDoor := ElevatorLeftDoorList[FloorId]
            RightDoor := ElevatorRightDoorList[FloorId]
            LeftDoorClosedPos := LeftDoorClosedPosList[FloorId]
            RightDoorClosedPos := RightDoorClosedPosList[FloorId]
            LeftDoorOpenPos := LeftDoorOpenPosList[FloorId]
            RightDoorOpenPos := RightDoorOpenPosList[FloorId]
        } then {
            # if(LeftDoor.TeleportTo[LeftDoorClosedPos, LeftDoor.GetTransform().Rotation]) {}
            # if(RightDoor.TeleportTo[RightDoorClosedPos, RightDoor.GetTransform().Rotation]) {}
            # We need to move one door asynchronous, so it doesn't block the thread
            Task := spawn{LeftDoor.MoveTo(LeftDoorOpenPos, LeftDoor.GetTransform().Rotation, DoorMovementDuration)}
            RightDoor.MoveTo(RightDoorOpenPos, RightDoor.GetTransform().Rotation, DoorMovementDuration)
        }
    }

    CloseDoor(FloorId: int)<suspends>: void = {
        if {
            LeftDoor := ElevatorLeftDoorList[FloorId]
            RightDoor := ElevatorRightDoorList[FloorId]
            LeftDoorClosedPos := LeftDoorClosedPosList[FloorId]
            RightDoorClosedPos := RightDoorClosedPosList[FloorId]
            LeftDoorOpenPos := LeftDoorOpenPosList[FloorId]
            RightDoorOpenPos := RightDoorOpenPosList[FloorId]
        } then {
            # if(LeftDoor.TeleportTo[LeftDoorOpenPos, LeftDoor.GetTransform().Rotation]) {}
            # if(RightDoor.TeleportTo[RightDoorOpenPos, RightDoor.GetTransform().Rotation]) {}
            # We need to move one door asynchronous, so it doesn't block the thread
            Task := spawn{LeftDoor.MoveTo(LeftDoorClosedPos, LeftDoor.GetTransform().Rotation, DoorMovementDuration)}
            RightDoor.MoveTo(RightDoorClosedPos, RightDoor.GetTransform().Rotation, DoorMovementDuration)
        }
    }

    MoveElevator(GoalFloorId: int)<suspends>: void = {
        if {
            ElevatorFloorPos := ElevatorFloorPosList[GoalFloorId]
        } then {
            CancelCloseDoorAfterDelay()
            set IsMoving = true
            AudioWhileMoving.Play()
            CloseDoor(CurFloorId)
            GoalPos := ElevatorFloorPos.GetTransform()
            CurTransform := ElevatorProp.GetTransform()
            DistanceToGoal := Distance(GoalPos.Translation, CurTransform.Translation)
            MovementDuration := DistanceToGoal / ElevatorMovementSpeed
            # Start movement
            ElevatorProp.MoveTo(GoalPos.Translation, GoalPos.Rotation, MovementDuration)
            # Movement has finished
            set CurFloorId = GoalFloorId
            OpenDoor(CurFloorId)
            set IsMoving = false
            # Stop music after doors have been opened
            AudioWhileMoving.Stop()
            # Close the doors again after delay
            MyRoutine := SpawnRoutine(CloseDoorAfterDelay)
            set RoutineCloseDoor = option{MyRoutine}
        } else {
            Print("ERR: MoveElevator: Found no ElevatorFloorPos for GoalFloorId: {GoalFloorId}")
        }
    }

    CancelCloseDoorAfterDelay(): void = {
        if {
            Routine := RoutineCloseDoor?
        } then {
            Routine.Cancel()
            set RoutineCloseDoor = false
        }
    }

    CloseDoorAfterDelay()<suspends>: void = {
        Sleep(CloseDoorsDelay)
        CloseDoor(CurFloorId)
    }

    OnUiButtonClick<override>(Agent: agent, ParmButtonId: int): void = {
        Print("tycoon_elevator_device: Verse OnUiButtonClick: ParmButtonId: {ParmButtonId}")

        if {
            IsMoving = false
        } then {
            Task := spawn{MoveElevator(ParmButtonId)}
        }
    }

    # This is called by TycoonManager
    ResetElevator(): void = {
        Task := spawn{TeleportElevator(0)}
    }

    OnExitButtonClick<override>(Agent: agent): void = {

    }

    OnMinigameSuccess<override>(Agent: agent):void = {

    }
    
    OnMinigameFailure<override>(Agent: agent):void = {

    }
}