using { /Fortnite.com/Game }
using { /Fortnite.com/Devices }
using { /Fortnite.com/Devices/CreativeAnimation }
using { /Fortnite.com/Characters }
using { /Fortnite.com/Vehicles }
using { /Verse.org/Simulation }
using { /Verse.org/Simulation/Tags }
using { /Verse.org/Random }
using { /Verse.org/Concurrency }
using { /UnrealEngine.com/Temporary/Diagnostics }
using { /UnrealEngine.com/Temporary/SpatialMath }

tycoon_king_device := class(creative_device) {
    @editable
    ButtonToClaim: button_device = button_device {}

    @editable
    BuyCurrency: tycoon_currency = tycoon_currency.Money

    @editable
    PriceToBuy: int = 1

    @editable
    PlayerRefKing: player_reference_device = player_reference_device {}

    @editable
    HudOnClaimed: hud_message_device = hud_message_device {}

    @editable
    HudKingLost: hud_message_device = hud_message_device {}

    @editable
    MoneyIncomeIncreaseAmountAfterBuy: int = 0

    @editable
    FansIncomeIncreaseAmountAfterBuy: float = 0.0

    @editable
    DiamondsIncomeIncreaseAmountAfterBuy: float = 0.0

    @editable
    XpIncomeIncreaseAmountAfterBuy: int = 0

    # Runtime
    var MaybeOwnerAgent: ?agent = false

    # Runs when the device is started in a running game
    OnBegin<override>()<suspends>:void = {
        # Listen to events
        GetPlayspace().PlayerRemovedEvent().Subscribe(OnPlayerRemoved)

        # Claim button
        ButtonToClaim.InteractedWithEvent.Subscribe(AssignOwner)
    }

    OnPlayerRemoved(Player: player): void = {
        TryToClearOwner(Player)
    }

    AssignOwner(Agent: agent): void = {
        if {
            PrevOwner := MaybeOwnerAgent?
        } then {
            # Check if the agent that owns the throne tries to claim it again
            if(Agent = PrevOwner) {
                return
            }
            # Handle different agent as new owner
            RemoveRewardsForPreviousOwner(PrevOwner)
        }
        # Assign new owner
        set MaybeOwnerAgent = option{Agent}
        if {
            PlayerData := Agent.GetPlayerData[]
        }
        GrantKingReward(Agent)
    }

    RemoveRewardsForPreviousOwner(PrevOwner: agent): void = {
        if {
            PlayerData := PrevOwner.GetPlayerData[]
            TycoonManager := PlayerData.MaybeTycoonManager?
        } then {
            TycoonManager.IncreaseMoneyIncome(PrevOwner, MoneyIncomeIncreaseAmountAfterBuy * -1)
            TycoonManager.IncreaseFansIncome(FansIncomeIncreaseAmountAfterBuy * -1)
            TycoonManager.IncreaseDiamondsIncome(DiamondsIncomeIncreaseAmountAfterBuy * -1)
            TycoonManager.IncreaseXpIncome(XpIncomeIncreaseAmountAfterBuy * -1)
        }
    }

    TryToClearOwner(Agent: agent): void = {
        Print("KingDevice: TryToClearOwner")
        if {
            OwnerAgent := MaybeOwnerAgent?
            Agent = OwnerAgent
        } then {
            PlayerRefKing.Clear()
            set MaybeOwnerAgent = false
        }
    }

    GrantKingReward(Agent: agent): void = {
        if {
            PlayerData := Agent.GetPlayerData[]
            TycoonManager := PlayerData.MaybeTycoonManager?
        } then {
            # Check if the player has enough money
            var PurchaseSucceeded: logic = false
            case(BuyCurrency) {
                tycoon_currency.Money =>
                    if(PlayerData.CurMoney >= PriceToBuy) {
                        PlayerData.RemoveMoney(PriceToBuy)
                        set PurchaseSucceeded = true
                    }
                tycoon_currency.Fans =>
                    if(PlayerData.CurFans >= PriceToBuy * 1.0) {
                        # Fans are not removed
                        set PurchaseSucceeded = true
                    }
                tycoon_currency.Diamonds =>
                    if(PlayerData.CurDiamonds >= PriceToBuy * 1.0) {
                        PlayerData.RemoveDiamonds(PriceToBuy)
                        set PurchaseSucceeded = true
                    }
                tycoon_currency.Rebirths =>
                    if(PlayerData.NumOfRebirths >= PriceToBuy) {
                        set PurchaseSucceeded = true
                    }
                tycoon_currency.PlayerLevel =>
                    # Levels start with 0, so we have to add 1 to it
                    if((PlayerData.CurLevel + 1) >= PriceToBuy) {
                        set PurchaseSucceeded = true
                    }
            }
            # Check if we could purchase the button
            if(PurchaseSucceeded = true) {
                if {
                    PrevOwner := MaybeOwnerAgent?
                    PrevOwner <> Agent
                } then {
                    HudKingLost.Show(PrevOwner)
                }
                HudOnClaimed.Show(Agent)
                PlayerRefKing.Register(Agent)
                TycoonManager.IncreaseMoneyIncome(Agent, MoneyIncomeIncreaseAmountAfterBuy)
                TycoonManager.IncreaseFansIncome(FansIncomeIncreaseAmountAfterBuy)
                TycoonManager.IncreaseDiamondsIncome(DiamondsIncomeIncreaseAmountAfterBuy)
                TycoonManager.IncreaseXpIncome(XpIncomeIncreaseAmountAfterBuy)
            }
        } else {
            Print("ERR: GrantKingReward: Didn't find PlayerData")
        }
    }
}