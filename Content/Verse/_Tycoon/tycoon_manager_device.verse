using { /Fortnite.com/Game }
using { /Fortnite.com/Devices }
using { /Fortnite.com/Devices/CreativeAnimation }
using { /Fortnite.com/Characters }
using { /Verse.org/Simulation }
using { /Verse.org/Simulation/Tags }
using { /Verse.org/Random }
using { /Verse.org/Concurrency }
using { /UnrealEngine.com/Temporary/Diagnostics }
using { /UnrealEngine.com/Temporary/SpatialMath }

tycoon_currency<public> := enum {
    Money,
    Fans,
    Diamonds,
    PlayerLevel,
    Rebirths
}

tycoon_manager_device<public> := class(creative_device) {
    # Verse Devices
    @editable
    TycoonConfig<public>: tycoon_config_device = tycoon_config_device {}

    @editable
    AchievementBoard<public>: tycoon_button_achievement_board = tycoon_button_achievement_board {}

    @editable
    TycoonElevator: tycoon_elevator_device = tycoon_elevator_device {}

    @editable
    TycoonButtonAutoSave<public>: tycoon_button_auto_save = tycoon_button_auto_save {}

    @editable
    PcButtonDevice: tycoon_button_pc_no_ui = tycoon_button_pc_no_ui {}

    @editable
    ShopGarage<public>: _Shop.shop_currency_garage = _Shop.shop_currency_garage {}

    @editable
    CustomCarDevice<public>: custom_car_device = custom_car_device {}

    @editable
    CityOwnerDevice: city_owner_device = city_owner_device {}
    
    <#
    @editable
    TycoonButtonNpcHelper: ?tycoon_button_npc_helper = false
    #>
    
    <#
    @editable
    CityNpcsDevice: city_npcs_device = city_npcs_device {}
    #>

    # Devices
    @editable
    IntroCinematic: cinematic_sequence_device = cinematic_sequence_device {}

    @editable
    IntroCinematicTeleportDelay: float = 1.0

    @editable
    IntroCinematicDuration: float = 5.0

    @editable
    PlayerStartPos<public>: teleporter_device = teleporter_device {}

    @editable
    TycoonButtons<public>: []tycoon_button = array {}

    @editable
    TycoonBuyPropTrigger: trigger_device = trigger_device {}

    <#
    @editable
    EntranceObjectiveDevice: objective_device = objective_device {}
    #>

    @editable
    MutatorZoneClaimTycoon: mutator_zone_device = mutator_zone_device {}

    @editable
    MutatorZoneClaimMoneyList: []mutator_zone_device = array {}

    @editable
    BillboardMailboxList: []billboard_device = array {}

    @editable
    ObjectivePrioritizeLowestCost: logic = true

    @editable
    YourHomeBeacon<public>: beacon_device = beacon_device {}

    @editable
    ObjectiveDeviceBeacon: beacon_device = beacon_device {}

    @editable
    OwnerPlayerRefList: []player_reference_device = array {}

    @editable
    OwnerBillboardLevel: billboard_device = billboard_device {}

    @editable
    OwnerBillboardRebirth: billboard_device = billboard_device {}

    @editable
    OwnerBillboardConcertScore: billboard_device = billboard_device {}

    <#
    @editable
    PropWhenInactive: creative_prop = creative_prop {}
    #>

    @editable
    PetAudioPlayerProp<public>: creative_prop = creative_prop {}

    @editable
    MutatorZoneEnableOnResetList: []mutator_zone_device = array {}

    @editable
    MutatorZoneDisableOnResetList: []mutator_zone_device = array {}

    @editable
    CinematicStopOnResetList: []cinematic_sequence_device = array {}

    @editable
    ButtonEnableOnResetList: []button_device = array {}

    @editable
    TriggerOnReset: trigger_device = trigger_device {}

    @editable
    QuestObjectiveBeaconParent<public>: creative_prop = creative_prop {}

    @editable
    QuestObjectiveBeacon<public>: beacon_device = beacon_device {}

    @editable
    UnlockCarButtonList: []button_device = array {}

    <#
    @editable
    TestButtonUnassignTycoon: button_device = button_device {}

    @editable
    MoneyTowerProp: creative_prop = creative_prop {}
    #>

    # Constants
    ButtonBuyTriggerTickRate: float = 0.1
    MoneyTickRate: float = 1.0
    PetFollowUpdateRate: float = 1.0
    # MoneyTowerMoveDuration: float = 1.0
    MaxNumOfActiveButtons: int = 99999
    StartMoneyIncome: int = 0
    StartFansIncome: float = 0.0
    StartDiamondsIncome: float = 0.0
    StartXpIncome: int = 0

    ButtonTriggerOffset: vector3 = vector3{X:=0.0, Y:=0.0, Z:=-10.0}
    ButtonNameBillboardOffset: vector3 = vector3{X:=0.0, Y:=0.0, Z:=100.0}
    ButtonPriceBillboardOffset: vector3 = vector3{X:=0.5, Y:=0.5, Z:=60.0}

    PetOffsets : []vector3 = array {
        vector3{X := -250.0, Y := 0.0, Z := 0.0}
        vector3{X := -125.0, Y := 125.0, Z := 0.0}
        vector3{X := -125.0, Y := -125.0, Z := 0.0}
    }
    
    # Runtime variables
    var TycoonId<public>: int = -1
    var MaybeOwnerAgent<public>: ?agent = false
    var MaybeButtonInView: ?tycoon_button = false
        
    var TycoonInitComplete: logic = false

    # All buttons that are currently shown and ready to buy
    var ActiveButtons: []tycoon_button = array {}

    # var MoneyTowerStartPos: vector3 = vector3 {}
    # var MoneyTowerStartRot: rotation = rotation {}

    var MoneyIncomePerSec<public>: int = 0
    var MoneyInMailbox<public>: int = 0
    var FansIncomePerSec<public>: float = 0.0
    var DiamondsIncomePerSec<public>: float = 0.0
    var StreamerFansIncomePerSec<public>: float = 0.0
    var XpIncomePerSec<public>: int = 0

    var PetTimeSinceLastFollow: float = 0.0
    var CurPetsList<public>: []_Npc.pet = array {}

    # Runs when the device is started in a running game
    OnBegin<override>()<suspends>: void = {
        GameManager := GetGameManager()

        for(ItemIndex -> UnlockCarButton: UnlockCarButtonList) {
            EventHandler := event_handler_agent(int) {
                A := ItemIndex
                CallbackFunc := OnUnlockCarButton
            }

            EventSub := UnlockCarButton.InteractedWithEvent.Subscribe(EventHandler.EventSubscription)
        }

        # Wait until all the buttons are initialized
        Sleep(0.5)

        # Init Tycoon buttons
        for(ButtonIndex -> Button: TycoonButtons) {
            Button.InitButton(Self, ButtonIndex)
        }

        # Init elevator
        TycoonElevator.InitTycoonManager(Self)

        # Listen for events
        TycoonBuyPropTrigger.TriggeredEvent.Subscribe(OnTycoonButtonTrigger)
        MutatorZoneClaimTycoon.AgentEntersEvent.Subscribe(TryToAssignTycoonToAgent)
        for(MutatorZoneClaimMoney: MutatorZoneClaimMoneyList) {
            MutatorZoneClaimMoney.AgentEntersEvent.Subscribe(OnMutatorZoneClaimMoney)
        }
        # TestButtonUnassignTycoon.InteractedWithEvent.Subscribe(OnTestButtonUnassignTycoon)

        # Init variables
        # set MoneyTowerStartPos = MoneyTowerProp.GetTransform().Translation
        # set MoneyTowerStartRot = MoneyTowerProp.GetTransform().Rotation

        # Start the server tick
        Task := spawn {ServerTick()}
        Task2 := spawn {MoneyTick()}

        set TycoonInitComplete = true
    }

    OnTycoonButtonTrigger(MaybeAgent: ?agent): void = {
        # Print("OnTycoonButtonTrigger")
        if {
            OwnerAgent := MaybeOwnerAgent?
            AgentOnTrigger := MaybeAgent?
            OwnerAgent = AgentOnTrigger
            ButtonInView := MaybeButtonInView?
        } then {
            TryToBuyProp(AgentOnTrigger, ButtonInView)
        }
    }

    OnMutatorZoneClaimMoney(Agent: agent): void = {
        Utility := GetUtility()
        if {
            OwnerAgent := MaybeOwnerAgent?
            Agent = OwnerAgent
        } then {
            ClaimIncome(Agent)
            TycoonConfig.CollectMoneyAudio.Play(Agent)
        }
    }

    PlayIntroCinematic<public>()<suspends>: void = {
        GameManager := GetGameManager()
        <# Intro Cinematic is played without Verse for this game
        if {
            Agent := MaybeOwnerAgent?
            PlayerData := Agent.GetPlayerData[]
        } then {
            # Play intro cinematic (base specific)
            PlayerData.PlayCinematic(IntroCinematic) 
            # Hide UI
            PlayerData.PlayerLocalUi.RemoveUiFromPlayer(Agent)
        }

        Sleep(IntroCinematicTeleportDelay)
        #>

        if {
            Agent := MaybeOwnerAgent?
            PlayerData := Agent.GetPlayerData[]
        } then {
            # Teleport player to their tycoon
            PlayerStartPos.Teleport(Agent)
        }

        # Sleep(IntroCinematicDuration - IntroCinematicTeleportDelay)

        if {
            Agent := MaybeOwnerAgent?
            PlayerData := Agent.GetPlayerData[]
        } then {
            # Show UI
            PlayerData.ShowPlayerLocalUi(Agent)
        }
    }

    OnRebirthButton<public>(Agent: agent): void = {
        if {
            OwnerAgent := MaybeOwnerAgent?
            Agent = OwnerAgent
            PlayerData := Agent.GetPlayerData[]
        } then {
            # Clear king device in case the Tycoon Owner was king
            TycoonConfig.KingDevice.TryToClearOwner(OwnerAgent)
            # Reset and unassign the tycoon
            PlayerData.UnassignTycoon()
            # Change player data
            PlayerData.DoRebirth()
            # Reassign the same tycoon directly to the same player
            TryToAssignTycoonToAgent(Agent)
        }
    }

    OnTestButtonUnassignTycoon(Agent: agent): void = {
        if {
            PlayerData := Agent.GetPlayerData[]
        } then {
            PlayerData.UnassignTycoon()
        }
    }

    UpdateButtonStates<public>(): void = {
        Print("UpdateButtonStates")
        # CleanUpActiveButtonList() must be in this order, or you won't be able to buy the first button
        CleanUpActiveButtonList()
        # Separate variable for better performance
        var NumOfButtons: int = ActiveButtons.Length
        for(Button: TycoonButtons) {
            if(NumOfButtons < MaxNumOfActiveButtons) {
                if {
                    Button.ButtonState = button_state.Hidden
                } then {
                    var ShouldShowButton: logic = true
                    for(DependencyIndex : Button.DependencyList) {
                        if(Dependency := TycoonButtons[DependencyIndex]) {
                            if(Dependency.ButtonState <> button_state.Purchased) {
                                set ShouldShowButton = false
                            }
                        } else {
                            Print("WARN: Tycoon ButtonId: {Button.ButtonId} has an invalid dependency ID: {DependencyIndex}")
                            set ShouldShowButton = false
                        }
                    }
                    if(ShouldShowButton = true) {
                        if {
                            # Prevent adding the first 2 special case buttons to ActiveButtons, since they will bug out the Buy Beacon
                            Button.BuyButtonLocation.IsValid[]
                        } then {
                            set ActiveButtons = ActiveButtons + array{Button}
                        }

                        set NumOfButtons = NumOfButtons + 1
                        
                        # Asynchronous to improve performance
                        Task := spawn {ActivateButton(Button)}
                    }
                }
            }
        }
        # CleanUpActiveButtonList here won't fix the buy beacon bug on load
        UpdateButtonInView()
        UpdateObjectiveDevice()
    }

    UpdateObjectiveDevice(): void = {
        # Print("ActiveButtons.Length: {ActiveButtons.Length}")
        var LowestPrice: int = 1000000000000000000
        var CheapestButton: ?tycoon_button = false
        if(ObjectivePrioritizeLowestCost = true) {
            # Prioritize lowest cost
            for(Button : ActiveButtons) {
                var ButtonPrice: int = Button.PriceToBuy
                if {
                    # Some level buttons are quest buttons, which should be showed first
                    (Button.BuyCurrency <> tycoon_currency.PlayerLevel) or Button.StartQuestOnBuy = true
                    Button.BuyCurrency <> tycoon_currency.Rebirths
                    ButtonPrice < LowestPrice
                } then {
                    set LowestPrice = ButtonPrice
                    set CheapestButton = option{Button}
                } else {
                    # Special case: Only have level or rebirth button --> Display it
                    if(ActiveButtons.Length = 1) {
                        set LowestPrice = ButtonPrice
                        set CheapestButton = option{Button}
                    }
                }
            }
        } else {
            # Prioritize lowest currency id
            var LowestButtonCurrencyId: int = 100
            for(Button: ActiveButtons) {
                var ButtonPrice: int = Button.PriceToBuy
                # Prioritize buttons with lower currency id
                var ButtonCurrencyId: int = GetTycoonCurrencyId(Button.BuyCurrency)

                if(ButtonCurrencyId < LowestButtonCurrencyId or (ButtonCurrencyId = LowestButtonCurrencyId and ButtonPrice < LowestPrice)) {
                    set LowestPrice = ButtonPrice
                    set CheapestButton = option {Button}
                    set LowestButtonCurrencyId = ButtonCurrencyId
                }
            }
        }
        if {
            OwnerAgent := MaybeOwnerAgent?
        } then {
            if(Button := CheapestButton?) {
                Print("UpdateObjectiveDevice 222. Button.ButtonId: {Button.ButtonId}. BuyButtonLocation.GetTransform().Translation: {Button.BuyButtonLocation.GetTransform().Translation}")
                if(ObjectiveDeviceBeacon.TeleportTo[Button.ButtonStartPos, Button.ButtonStartRot]) {}
                ObjectiveDeviceBeacon.AddToShowList(OwnerAgent)
            } else {
                ObjectiveDeviceBeacon.RemoveFromShowList(OwnerAgent)
            }
        }
    }

    CleanUpActiveButtonList<public>(): void = {
        # Clean up the ActiveButton List in case we restored a save and didn't remove the purchased button from the list
        for(ButtonIndex -> ActiveButton : ActiveButtons) {
            if(ActiveButton.ButtonState = button_state.Purchased) {
                if {
                    Temp := ActiveButtons.RemoveElement[ButtonIndex]
                    set ActiveButtons = Temp
                }
            }
        }
    }

    ActivateButton(Button: tycoon_button)<suspends>: void = {
        GameManager := GetGameManager()
        Button.ShowButton()
    }

    TryToAssignTycoonToAgent<public>(Agent : agent): void = {
        Print("TryToAssignTycoonToAgent")
        GameManager := GetGameManager()
        Utility := GetUtility()

        if {
            MaybeOwnerAgent = false
            PlayerData := Agent.GetPlayerData[]
            not PlayerData.MaybeTycoonManager?
        } then {
            Print("TryToAssignTycoonToAgent 222")
            # DeactivateTycoonObjectives must be first
            # GameManager.DeactivateTycoonObjectivesForAgent(Agent)

            # Init variables
            set MoneyIncomePerSec = StartMoneyIncome
            set FansIncomePerSec = StartFansIncome
            set DiamondsIncomePerSec = StartDiamondsIncome
            set XpIncomePerSec = StartXpIncome

            # Assign Tycoon
            set MaybeOwnerAgent = option{Agent}
            if(TeamSelector := GameManager.TeamSelectorList[TycoonId]) {
                TeamSelector.ChangeTeam(Agent)
            }
            PlayerData.AssignTycoon(Self)
            for(OwnerPlayerRef: OwnerPlayerRefList) {
                OwnerPlayerRef.Register(Agent)
            }
            CustomCarDevice.AssignOwner(Agent)
            CityOwnerDevice.AssignOwner(Agent, Self)
            TycoonConfig.TycoonBaseMusic.Play(Agent)
            # TycoonConfig.HudBaseClaimed.Show(Agent)

            # Show initial buttons - This is included in PlayerData.AssignTycoon(Self)
            # UpdateButtonStates()
            # Deactivate this tycoon objective for every player
            for(Player : GetPlayspace().GetPlayers()) {
                if(PlayerAgent := agent[Player]) {
                    ObjectiveDeviceBeacon.RemoveFromShowList(PlayerAgent)
                }
            }
            # Show rebirth value
            OwnerBillboardRebirth.SetText(Utility.FormatMessage("REBIRTHS: " + Utility.FormatFloat(PlayerData.NumOfRebirths * 1.0, 3, 0)))
            # Show initial level value
            UpdateOwnerLevelBillboard()
            # Activate objective for the owner
            ObjectiveDeviceBeacon.AddToShowList(Agent)
            # Add home beacon
            YourHomeBeacon.AddToShowList(Agent)
        }
    }

    UpdateOwnerLevelBillboard<public>(): void = {
        Utility := GetUtility()
        if {
            Player := MaybeOwnerAgent?
            PlayerData := Player.GetPlayerData[]
        } then {
            OwnerBillboardLevel.SetText(Utility.FormatMessage("LEVEL: " + Utility.FormatFloat((PlayerData.CurLevel + 1) * 1.0, 2, 0)))
        }
    }

    ResetTycoon<public>(): void = {
        Print("ResetTycoon")
        
        Utility := GetUtility()
        # Reset linked devices
        CustomCarDevice.OnTycoonReset()
        CityOwnerDevice.OnTycoonReset()
        # Reset owner agent - We don't have to reset PlayerData since we come from there
        set MaybeOwnerAgent = false
        # Reset every button
        for(Button : TycoonButtons) {
            # Asynchronous to improve performance
            Task := spawn {Button.OnTycoonReset()}
        }
        # if(MoneyTowerProp.TeleportTo[MoneyTowerStartPos, MoneyTowerStartRot]) {}
        # Reset the whole Tycoon
        set ActiveButtons = array {}
        set MoneyIncomePerSec = StartMoneyIncome
        set FansIncomePerSec = StartFansIncome
        set DiamondsIncomePerSec = StartDiamondsIncome
        set StreamerFansIncomePerSec = 0.0
        set XpIncomePerSec = StartXpIncome
        set MoneyInMailbox = 0
        UpdateMailboxText()
        # Reset pets
        DestroyPets()
        # Reset elevator
        TycoonElevator.ResetElevator()
        # Reset city NPCs
        # CityNpcsDevice.Reset()
        <# Show prop (Must also add code to hide it)
        if(PropWhenInactive.IsValid[]) {
            PropWhenInactive.Show()
        }
        #>
        # Reset devices for Senix
        for(MutatorZoneEnable: MutatorZoneEnableOnResetList) {
            MutatorZoneEnable.Enable()
        }
        for(MutatorZoneDisable: MutatorZoneDisableOnResetList) {
            MutatorZoneDisable.Disable()
        }
        for(CinematicToStop: CinematicStopOnResetList) {
            CinematicToStop.Stop()
        }
        for(ButtonToEnable: ButtonEnableOnResetList) {
            ButtonToEnable.Enable()
        }
        # Turn off switches without player signal
        TriggerOnReset.Trigger()
        # Reactivate tycoon for next player
        for(OwnerPlayerRef: OwnerPlayerRefList) {
            OwnerPlayerRef.Clear()
        }
        OwnerBillboardLevel.SetText(Utility.FormatMessage(""))
        OwnerBillboardRebirth.SetText(Utility.FormatMessage(""))
        OwnerBillboardConcertScore.SetText(Utility.FormatMessage(""))
    }

    DestroyPets(): void = {
        for(Pet: CurPetsList) {
            Pet.DestroyPet()
        }
        set CurPetsList = array{}
    }

    TryToBuyProp<public>(Agent: agent, Button: tycoon_button): void = {
        GameManager := GetGameManager()
        SaveManager := GetSaveManager()
        if {
            PlayerData := Agent.GetPlayerData[]
            # IF COMMENTED OUT: REVERT, ONLY FOR VILVA TEST
            Button.ButtonState = button_state.Visible
        } then {
            # Check if the player has enough money
            var PurchaseSucceeded: logic = false
            case(Button.BuyCurrency) {
                tycoon_currency.Money =>
                    Print("BuyProp: MONEY: {Button.PriceToBuy}")
                    if(PlayerData.CurMoney >= Button.PriceToBuy) {
                        PlayerData.RemoveMoney(Button.PriceToBuy)
                        set PurchaseSucceeded = true
                    }
                tycoon_currency.Fans =>
                    Print("BuyProp: SCAMS: {Button.PriceToBuy}")
                    if(PlayerData.CurFans >= Button.PriceToBuy * 1.0) {
                        # Fans are not removed
                        # PlayerData.RemoveFans(Button.PriceToBuy)
                        set PurchaseSucceeded = true
                    }
                tycoon_currency.Diamonds =>
                    Print("BuyProp: KILL COINS: {Button.PriceToBuy}")
                    if(PlayerData.CurDiamonds >= Button.PriceToBuy * 1.0) {
                        PlayerData.RemoveDiamonds(Button.PriceToBuy)
                        set PurchaseSucceeded = true
                    }
                tycoon_currency.PlayerLevel =>
                    Print("BuyProp: LEVEL: {Button.PriceToBuy}")
                    # Levels start with 0, so we have to add 1 to it
                    if((PlayerData.CurLevel + 1) >= Button.PriceToBuy) {
                        set PurchaseSucceeded = true
                    }
                tycoon_currency.Rebirths =>
                    Print("BuyProp: REBIRTHS: {Button.PriceToBuy}")
                    if(PlayerData.NumOfRebirths >= Button.PriceToBuy) {
                        set PurchaseSucceeded = true
                    }
            }
            # Check if we could purchase the button
            if(PurchaseSucceeded = true) {
                # Give xp to player
                # Scammer Tycoon: Give 200 extra XP for buying a prop
                PlayerData.AddXp(Button.XpReward + 200, true)
                # PlayerData.AddXp(Button.PriceToBuy)
                GrantAccolade(Agent)

                # Remove this button from the list of active buttons
                # This must be before UpdateButtonStates
                var FoundButton: logic = false
                for(ButtonIndex -> ActiveButton : ActiveButtons) {
                    if(FoundButton = false) {
                        if(Button.ButtonId = ActiveButton.ButtonId) {
                            set FoundButton = true
                            if {
                                Temp := ActiveButtons.RemoveElement[ButtonIndex]
                                set ActiveButtons = Temp
                            }
                        }
                    }
                }

                # Update save data - Must be before Button.OnPropBought
                PlayerData.PersistButtonPurchase(Button.ButtonId)

                # Spawn the prop and hide the button
                Button.OnPropBought(false)

                <# Check if we have to spawn a pet
                if(PetTemplate := Button.PetPropTemplate[0]) {
                    SpawnPet(PetTemplate)
                }
                #>

                <# Check for analytics
                else if(TycoonConfig.CellarFloorButtonId = Button.ButtonId) {
                    SaveManager.GrantAnalytic(Agent, SaveManager.ANALYTIC_BUTTON_CELLAR_FLOOR)
                } else if(TycoonConfig.GarageFloorButtonId = Button.ButtonId) {
                    SaveManager.GrantAnalytic(Agent, SaveManager.ANALYTIC_BUTTON_GARAGE_FLOOR)
                } else if(TycoonConfig.SecondFloorButtonId = Button.ButtonId) {
                    SaveManager.GrantAnalytic(Agent, SaveManager.ANALYTIC_BUTTON_SECOND_FLOOR)
                } else if(TycoonConfig.RoofFloorButtonId = Button.ButtonId) {
                    SaveManager.GrantAnalytic(Agent, SaveManager.ANALYTIC_BUTTON_ROOF_FLOOR)
                }
                #>

                # Check if we have to show new buttons
                UpdateButtonStates()

                # UpgradeMoneyTower()
                TycoonConfig.MessageBuyPropSuccess.Show(Agent)
            } else {
                TycoonConfig.MessageBuyPropFail.Show(Agent)
            }
        }
    }

    # Increase functions also work with negative values
    IncreaseMoneyIncome(Agent: agent, Amount: int): void = {
        if {
            PlayerData := Agent.GetPlayerData[]
        } then {
            set MoneyIncomePerSec = MoneyIncomePerSec + Amount
            PlayerData.UpdateIncomeMoney(MoneyIncomePerSec)
        }
    }

    IncreaseFansIncome(Amount: float): void = {
        set FansIncomePerSec = FansIncomePerSec + Amount
    }

    IncreaseDiamondsIncome(Amount: float): void = {
        set DiamondsIncomePerSec = DiamondsIncomePerSec + Amount
    }

    IncreaseXpIncome(Amount: int): void = {
        set XpIncomePerSec = XpIncomePerSec + Amount
    }

    GrantAccolade(Agent: agent): void = {
        AccoladeManager := GetAccoladeManager()
        AccoladeManager.AwardAccolade(Agent,2)
    }

    ClaimIncome(Agent: agent): void = {
        GameManager := GetGameManager()
        if {
            PlayerData := Agent.GetPlayerData[]
        } then {
            PlayerData.AddMoney(MoneyInMailbox, true, true)

            # Increase the money tower height based on mailbox
            # TowerGoalPos := MoneyTowerProp.GetTransform().Translation + vector3{X:=0.0, Y:=0.0, Z:=MoneyInMailbox * 0.25}
            # Task := spawn {MoneyTowerProp.MoveTo(TowerGoalPos, MoneyTowerStartRot, MoneyTowerMoveDuration)}
            
            # Reset money in mailbox
            set MoneyInMailbox = 0
            UpdateMailboxText()
        }
    }

    <# Increase the money tower height by a fixed amount
    UpgradeMoneyTower(): void = {
        TowerGoalPos := MoneyTowerProp.GetTransform().Translation + vector3{X:=0.0, Y:=0.0, Z:=25.0}
        Task := spawn {MoneyTowerProp.MoveTo(TowerGoalPos, MoneyTowerStartRot, MoneyTowerMoveDuration)}
    }
    #>

    <#
    IncreasePopulation(Amount: float): void = {
        Utility := GetUtility()
        set CurPopulation = CurPopulation + Amount
        FormattedPopulation := Utility.GetBillboardTextFloat(CurPopulation, 2)
        if(PopulationFloored := Floor[CurPopulation]) {
            CityNpcsDevice.CheckForNewSpawns(PopulationFloored)
        }
    }
    #>

    UpdateMailboxText(): void = {
        Utility := GetUtility()
        for(BillboardMailbox: BillboardMailboxList) {
            if {
                Agent := MaybeOwnerAgent?
                PlayerData := Agent.GetPlayerData[]
                PlayerData.RebirthAutomaticIncomeLevel >= 0
            } then {
                BillboardMailbox.SetText(Utility.FormatMessage("AUTO COLLECT"))
            } else {
                BillboardMailbox.SetText(Utility.GetBillboardTextIntPrefix(MoneyInMailbox, 2, 0, "$ "))
            }
        }
    }

    DoIncomeTick(): void = {
        GameManager := GetGameManager()
        if {
            OwnerAgent := MaybeOwnerAgent?
            PlayerData := OwnerAgent.GetPlayerData[]
        } then {
            # Update mailbox money
            if(MultipliedIncomeMoney := Round[MoneyIncomePerSec * PlayerData.CurMoneyMultiplier]) {
                # If the player unlocked automatic income, then he gets money directly (without mailbox)
                if(PlayerData.RebirthAutomaticIncomeLevel >= 0) {
                    var AutomaticIncomeMoney: int = 0
                    if(MoneyBonus := PlayerData.AutomaticIncomeMoney[PlayerData.RebirthAutomaticIncomeLevel]) {
                        set AutomaticIncomeMoney = MoneyBonus
                    }
                    PlayerData.AddMoney(MultipliedIncomeMoney + AutomaticIncomeMoney, false, true)
                } else {
                    set MoneyInMailbox = MoneyInMailbox + MultipliedIncomeMoney
                }
            }
            UpdateMailboxText()
            # Directly grant other income types
            PlayerData.AddXp(XpIncomePerSec, true)
            PlayerData.AddFans((FansIncomePerSec + StreamerFansIncomePerSec) * PlayerData.CurFansMultiplier, false)
            # Diamonds can have automatic income
            var AutomaticIncomeDiamonds: float = 0.0
            if(DiamondsBonus := PlayerData.AutomaticIncomeDiamonds[PlayerData.RebirthAutomaticIncomeLevel]) {
                set AutomaticIncomeDiamonds = DiamondsBonus
            }
            PlayerData.AddDiamonds((DiamondsIncomePerSec * PlayerData.CurDiamondsMultiplier) + AutomaticIncomeDiamonds, false)
        }
    }

    SpawnPet(Agent: agent, PetNpcSpawner: npc_spawner_device, ParmMoneyMultiplier: float, ParmFansMultiplier: float, ParmDiamondsMultiplier: float, ParmRecruitChanceMultiplier: float): void = {
        if {
            PlayerData := Agent.GetPlayerData[]
        } then {
            Utility := GetUtility()
            GameManager := GetGameManager()
            # Currently you can only have one pet active, so we clear previous pets
            DestroyPets()
            # Spawn the new pet
            set GameManager.MaybeLastAgentThatSpawnedPet = MaybeOwnerAgent
            var NewPet: _Npc.pet = _Npc.pet {}
            NewPet.SpawnPet(PetNpcSpawner)
    
            set NewPet.PetMultiplierMoney = ParmMoneyMultiplier
            set NewPet.PetMultiplierFans = ParmFansMultiplier
            set NewPet.PetMultiplierDiamonds = ParmDiamondsMultiplier
            set NewPet.PetMultiplierRecruitChance = ParmRecruitChanceMultiplier

            set CurPetsList = CurPetsList + array{NewPet}

            PlayerData.UpdatePetIncomeMultiplier()
        }
    }

    HandlePetNpcEliminated(NpcAgent: agent): void = {
        Print("HandlePetNpcEliminated")
        # Clean up NPC
        DestroyPetNpcScript(NpcAgent)
    }

    DestroyPetNpcScript(NpcAgent: agent): void = {
        Print("DestroyPetNpcScript")
        if(NpcPetData := NpcAgent.GetNpcPetData[]) {
            NpcPetData.HandleDestroy()
        }
        NpcAgent.DestroyNpcPetData()
    }

    GetTycoonCurrencyId(Currency: tycoon_currency): int = {
        case(Currency) {
            tycoon_currency.Money =>
                return 0
            tycoon_currency.Fans =>
                return 1
            tycoon_currency.Diamonds =>
                return 2
            tycoon_currency.PlayerLevel =>
                return 3
            tycoon_currency.Rebirths =>
                return 4
        }
    }

    ServerTick()<suspends>: void = {
        loop {
            Sleep(ButtonBuyTriggerTickRate)

            # Update buy button
            UpdateButtonInView()
            <# Update pets
            for(Pet: CurPetsList) {
                Pet.PetTick(OwnerAgent, FortChar, ServerTickRate)
            }
            #>
        }
    }

    MoneyTick()<suspends>: void = {
        loop {
            Sleep(MoneyTickRate)

            if {
                OwnerAgent := MaybeOwnerAgent?
                FortChar := OwnerAgent.GetFortCharacter[]
            } then {
                DoIncomeTick()
            }
        }
    }

    UpdateButtonInView(): void = {
        Utility := GetUtility()

        if {
            OwnerAgent := MaybeOwnerAgent?
            FortCharacter := OwnerAgent.GetFortCharacter[]
        } then {
            if(ActiveButtons.Length = 0) {
                # Hide the triggers
                if(TycoonBuyPropTrigger.TeleportTo[Utility.HiddenPos, rotation{}]) {}
            }
            
            CameraPos := FortCharacter.GetViewLocation()
            LookDir := FortCharacter.GetViewRotation()
            MovementVector := LookDir.RotateVector(Utility.ForwardVector)
            LookAtPos := CameraPos + (MovementVector * 400.0)
            # Print("LookAtPos: {LookAtPos}")
            # Utility.DebugDraw.DrawSphere(Center := LookAtPos, ?DrawDurationPolicy := debug_draw_duration_policy.FiniteDuration, ?Duration := 1.0)

            # Check which button has the shortest distance to LookAtPos
            var ShortestDistance: float = 100000000000000.0
            var ClosestButton: ?tycoon_button = false
            for(Button: ActiveButtons) {
                CurDistance := Distance(LookAtPos, Button.ButtonStartPos)
                if(CurDistance < ShortestDistance) {
                    set ShortestDistance = CurDistance
                    set ClosestButton = option {Button}
                }
            }
            if(Button := ClosestButton?) {
                # Check if this is a different button, than last look at button
                var IsDifferentButton: logic = false
                if(ButtonInView := MaybeButtonInView?) {
                    # Print("TEST 1 {Button.IdPriceToBuy} {ButtonInView.IdPriceToBuy}")
                    if(Button.ButtonId <> ButtonInView.ButtonId) {
                        set IsDifferentButton = true
                    }
                } else {
                    set IsDifferentButton = true
                }

                BuyButtonPos := Button.ButtonStartPos
                BuyButtonRot := Button.ButtonStartRot
                if(IsDifferentButton = true) {
                    # Move trigger to button pos
                    if(TycoonBuyPropTrigger.TeleportTo[BuyButtonPos + ButtonTriggerOffset, BuyButtonRot]) {}
                    <# Set billboard text
                    if(Button.RequiresSecondCurrency = false) {
                        TycoonNameBillboardDollar.SetText(Utility.FormatMessage("BUY " + Button.PropBillboardName))
                        TycoonPriceBillboardDollar.SetText(Utility.GetBillboardTextIntPrefix(Button.PriceToBuy, 2, "$ "))
                    } else {
                        TycoonNameBillboardPopulation.SetText(Utility.FormatMessage("BUY " + Button.PropBillboardName))
                        TycoonPriceBillboardPopulation.SetText(Utility.GetBillboardTextIntAppend(Button.PriceToBuy, 2, " Population"))
                    }
                    #>

                    <# Move the second trigger to the old position of the first trigger
                    if(ButtonInView := MaybeButtonInView?) {
                        PrevButtonPos := ButtonInView.ButtonStartPos
                        PrevButtonRot := ButtonInView.ButtonStartRot

                        if(TycoonBuyPropTrigger2.TeleportTo[PrevButtonPos + ButtonTriggerOffset, PrevButtonRot]) {}
                    }
                    #>

                    # Update the button in view for next tick
                    set MaybeButtonInView = option{Button}
                }

                <# Rotate the billboards to player look rotation
                if (VectorToBillboard := MovementVector.MakeUnitVector[]) {
                    Yaw := RadiansToDegrees(ArcTan(VectorToBillboard.Y, VectorToBillboard.X) - (PiFloat * 0.5))
                    Pitch := 0.0
                    Roll := 0.0
                    BillboardRot := MakeRotationFromYawPitchRollDegrees(Yaw, Pitch, Roll)

                    if(Button.RequiresSecondCurrency = false) {
                        if(TycoonNameBillboardDollar.TeleportTo[BuyButtonPos + ButtonNameBillboardOffset, BillboardRot]) {}
                        if(TycoonPriceBillboardDollar.TeleportTo[BuyButtonPos + ButtonPriceBillboardOffset, BillboardRot]) {}
                    } else {
                        if(TycoonNameBillboardPopulation.TeleportTo[BuyButtonPos + ButtonNameBillboardOffset, BillboardRot]) {}
                        if(TycoonPriceBillboardPopulation.TeleportTo[BuyButtonPos + ButtonPriceBillboardOffset, BillboardRot]) {}
                    }
                }
                #>
            }
        }
    }

    OnUnlockCarButton(Agent: agent, ItemIndex: int): void = {
        SaveManager := GetSaveManager()
        if {
            PlayerData := Agent.GetPlayerData[]
            PlayerData.HasUnlockedGarage = true
        } then {
            if {
                BuyCurrency := TycoonConfig.UnlockCarBuyCurrencyList[ItemIndex]
                PriceToBuy := TycoonConfig.UnlockCarPriceToBuyList[ItemIndex]
            } then {
                # Check if the player already already owns this car
                if {
                    Save := SaveManager.ProxySaves[Agent]
                    # Only trigger this code if the car was not previously owned
                    Save.CarShopStateList[ItemIndex] = 0
                } then {
                    # Check if the player has enough money
                    var PurchaseSucceeded: logic = false
                    case(BuyCurrency) {
                        tycoon_currency.Money =>
                            if(PlayerData.CurMoney >= PriceToBuy) {
                                PlayerData.RemoveMoney(PriceToBuy)
                                set PurchaseSucceeded = true
                            }
                        tycoon_currency.Fans =>
                            if(PlayerData.CurFans >= PriceToBuy * 1.0) {
                                # Fans are not removed
                                set PurchaseSucceeded = true
                            }
                        tycoon_currency.Diamonds =>
                            if(PlayerData.CurDiamonds >= PriceToBuy * 1.0) {
                                PlayerData.RemoveDiamonds(PriceToBuy)
                                set PurchaseSucceeded = true
                            }
                        tycoon_currency.Rebirths =>
                            if(PlayerData.NumOfRebirths >= PriceToBuy) {
                                set PurchaseSucceeded = true
                            }
                        tycoon_currency.PlayerLevel =>
                            # Levels start with 0, so we have to add 1 to it
                            if((PlayerData.CurLevel + 1) >= PriceToBuy) {
                                set PurchaseSucceeded = true
                            }
                    }
                    # Check if we could purchase the button
                    if(PurchaseSucceeded = true) {
                        PlayerData.SetCarShopItemState(ItemIndex, 1)
                        # Check to complete quest "Buy first car"
                        if {
                            Quest := PlayerData.MaybeActiveQuest?
                            Quest.ButtonId = 298
                        } then {
                            PlayerData.IncreaseQuestProgress()
                        }
                    } else {
                        # HudNotEnoughCurrency.Show(Agent)
                    }
                } else {
                    # HudBuildingAlreadyOwned.Show(Agent)
                    Print("WARN: tycoon manager device: The player already owns this car")
                }
            } else {
                Print("ERR: OnUnlockCarButton: ItemIndex: {ItemIndex}")
            }
        }
    }
}