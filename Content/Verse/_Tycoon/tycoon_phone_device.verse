using { /Fortnite.com/Game }
using { /Fortnite.com/Devices }
using { /Fortnite.com/Devices/CreativeAnimation }
using { /Fortnite.com/Characters }
using { /Verse.org/Simulation }
using { /Verse.org/Simulation/Tags }
using { /Verse.org/Random }
using { /Verse.org/Assets }
using { /Verse.org/Concurrency }
using { /UnrealEngine.com/Temporary/Diagnostics }
using { /UnrealEngine.com/Temporary/SpatialMath }

tycoon_phone_device := class(creative_device) {
    @editable
    PhoneSignalRemoteManager: signal_remote_manager_device = signal_remote_manager_device {}

    @editable
    RewardCurrency: tycoon_currency = tycoon_currency.Money

    @editable
    RewardFactor: float = 0.1
    
    @editable
    FixedRewardAmount: int = 0

    OnBegin<override>()<suspends>: void = {
        PhoneSignalRemoteManager.PrimarySignalEvent.Subscribe(OnPhoneSignal)
        PhoneSignalRemoteManager.SecondarySignalEvent.Subscribe(OnPhoneSignal)
    }

    OnPhoneSignal(Agent: agent): void = {
        if {
            PlayerData := Agent.GetPlayerData[]
        } then {
            PlayerData.AddCurrency(RewardCurrency, FixedRewardAmount, true)
            PlayerData.RewardPlayerIncomeFactor(RewardCurrency, RewardFactor)
        }
    }
}