using { /Verse.org/Simulation }
using { /Verse.org/Colors }
using { /Verse.org/Colors/NamedColors }
# using { /Verse.org/Concurrency }
# using { /Verse.org/Verse }
using { /Verse.org/Assets }
using { /Fortnite.com/Devices }
using { /Fortnite.com/UI }
# using { /Fortnite.com/Characters }
# using { /UnrealEngine.com/Temporary/Diagnostics }
using { /UnrealEngine.com/Temporary/SpatialMath }
using { /UnrealEngine.com/Temporary/UI}

# Needs different verse device for UI class
shop_achievement<public> := class(shop_base_device) {
    @editable
    AchievementType: _Global.achievement_type = _Global.achievement_type.Kill

    @editable
    ShopMainHud: hud_message_device = hud_message_device {}

    GetShopUiClass<override>(): _UI.ui_shop = {
        return _UI.ui_shop_achievement {}
    }

    OnPlayerAdded<override>(Agent: agent): void = {
        (super:)OnPlayerAdded(Agent)
        if(Ui := AgentShopMap[Agent]) {
            set Ui.AchievementType = AchievementType
        } else {
            Print("ERR: shop_achievement")
        }
    }

    OnShopOpened<override>(Agent: agent): void = {
        (super:)OnShopOpened(Agent)
        ShopMainHud.Show(Agent)
    }

    OnShopClosed<override>(Agent: agent): void = {
        (super:)OnShopClosed(Agent)
        ShopMainHud.Hide(Agent)
    }
}