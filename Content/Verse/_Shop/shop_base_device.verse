using { /Verse.org/Simulation }
using { /Verse.org/Colors }
using { /Verse.org/Colors/NamedColors }
# using { /Verse.org/Concurrency }
# using { /Verse.org/Verse }
using { /Verse.org/Assets }
using { /Fortnite.com/Devices }
using { /Fortnite.com/UI }
# using { /Fortnite.com/Characters }
# using { /UnrealEngine.com/Temporary/Diagnostics }
using { /UnrealEngine.com/Temporary/SpatialMath }
using { /UnrealEngine.com/Temporary/UI}

shop_base_device<public> := class(creative_device) {
    @editable
    TycoonId: int = 0

    @editable
    ShopId<public>: int = 0

    @editable
    LevelRequirement: int = 1

    @editable
    TooLowLevelHudMessage: hud_message_device = hud_message_device {}

    @editable
    PurchaseAudio: audio_player_device = audio_player_device {}

    @editable
    NotEnoughGoldAudio: audio_player_device = audio_player_device {}

    @editable
    OpenShopButton: button_device = button_device {}

    @editable
    ConditionalButton: conditional_button_device = conditional_button_device {}

    @editable
    ItemGranter: item_granter_device = item_granter_device {}

    @editable
    PriceList<public>: []int = array {
        100,
        200,
        300
    }

    <# 
    @editable
    OptionalCinematicOnBuy: ?cinematic_sequence_device = false
    #>

    <# Images must be modified in the script #>
    ImageListAllShops<public>: [][]texture = array {
        <# TODO: Add images #>
        # Level 1 Shop ID 0
        array {
            UI.level1_flintknock,
            UI.level1_pistol,
            UI.level1_rapidfire,
            UI.level1_sixshooter,
            UI.level1_shotgun,
            UI.level1_huntingrifle,
            UI.level1_makeshiftrifle,
            UI.level1_ar,
            UI.level1_drumgun,
            UI.level1_hammerar
        },
        # Level 2 Shop Id 1
        array {
            UI.pistol_level1,
            UI.huntingrifle_level1,
            UI.smg_level1,
            UI.hunterdmg_level1,
            UI.thermaldmr_level1,
            UI.scar_level1,
            UI.enforcer_level1,
            UI.shotgun_level1,
            UI.revolver_level1,
            UI.dualpistol_level1
        },
        # Level 3 Shop id 2
        array {
            UI.level3_tactpistol,
            UI.level3_smg,
            UI.level3_shotgun2,
            UI.level3_shotugn,
            UI.level3_ar3,
            UI.level3_sniper2,
            UI.level3_ar4,
            UI.level3_ar,
            UI.level3_sniper,
            UI.level3_ar2
        },
        # Level 4 Shop id 3
        array {
            UI.level4_pistol2,
            UI.level4_pistol,
            UI.level4_shotgun,
            UI.level4_shotgun3,
            UI.level4_shotgun2,
            UI.level4_ar3,
            UI.level4_ar,
            UI.level4_ar2,
            UI.level4_sniper,
            UI.level4_sniper2
        },  
        # Level 5 Shop id 4
        array {
            UI.level5_smg3,
            UI.level5_smg,
            UI.level5_smg2,
            UI.level5_sniper,
            UI.level5_sniper2,
            UI.level5_shotgun3,
            UI.level5_shotgun,
            UI.level5_shotgun2,
            UI.level5_ar,
            UI.level5_ar2
        },  
        # Ammo Shop id 5
        array {
            UI.ar_ammo,
            UI.smg_ammo,
            UI.shotgun_ammo,
            UI.sniper_ammo
        },
        # Garage UI id 6
        array {
            UI.cybertruck_car,
            UI.ferrari_car,
            UI.gold_car,
            UI.lambo_car,
            UI.scam_van,
            UI.tank_car,
            UI.ufo_car
        },
        # Monitor Shop id 7
        array {
            UI.monitor_level1,
            UI.monitor_level2,
            UI.monitor_level3
        },
        # Desk Shop id 8
        array {
            UI.desk_level1,
            UI.desk_level2,
            UI.desk_level3
        },
        # Color Shop id 9
        array {
            UI.blueshop,
            UI.redshop,
            UI.pinkshop,
            UI.greenshop,
            UI.yellowshop,
            UI.purpleshop,
            UI.hackingshop,
            UI.emojishop,
            UI.rainbowshop,
            UI.coinsshop
        }                                      
    }
    <# End Modify #>

    # Constants
    # ImageWidth: float = 500.0
    # ImageHeight: float = 500.0
    ShopItemsPerPage<public>: int = 10

    # Constants you can override for child classes
    BackgroundImageTexture<public>: texture = UI.weaponshop
    ShopItemPriceSuffix<public>: string = " GOLD"
    ShopButtonLeftPosAdjust<public>: float = 0.0

    # Background image
    AnchorsBackgroundImage<public>: anchors = anchors{Minimum := vector2{X := 0.5, Y := 0.5}, Maximum := vector2{X := 0.5, Y := 0.5}}
    OffsetsBackgroundImage<public>: margin = margin{Top := -792.540527, Left := -712.960938, Right := 1509.323608, Bottom := 1510.134521}

    # Close shop button
    AnchorsCloseShopButton<public>: anchors = anchors{Minimum := vector2{X := 0.5, Y := 0.5}, Maximum := vector2{X := 0.5, Y := 0.5}}
    OffsetsCloseShopButton<public>: margin = margin{Top := 397.871246, Left := -42.173973, Right := 163.563293, Bottom := 40.921097}

    # Buy button slots (10x)
    AnchorsBuyButton<public>: []anchors = array {
        anchors{Minimum := vector2{X := 0.5, Y := 0.5}, Maximum := vector2{X := 0.5, Y := 0.5}},
        anchors{Minimum := vector2{X := 0.5, Y := 0.5}, Maximum := vector2{X := 0.5, Y := 0.5}},
        anchors{Minimum := vector2{X := 0.5, Y := 0.5}, Maximum := vector2{X := 0.5, Y := 0.5}},
        anchors{Minimum := vector2{X := 0.5, Y := 0.5}, Maximum := vector2{X := 0.5, Y := 0.5}},
        anchors{Minimum := vector2{X := 0.5, Y := 0.5}, Maximum := vector2{X := 0.5, Y := 0.5}},
        anchors{Minimum := vector2{X := 0.5, Y := 0.5}, Maximum := vector2{X := 0.5, Y := 0.5}},
        anchors{Minimum := vector2{X := 0.5, Y := 0.5}, Maximum := vector2{X := 0.5, Y := 0.5}},
        anchors{Minimum := vector2{X := 0.5, Y := 0.5}, Maximum := vector2{X := 0.5, Y := 0.5}},
        anchors{Minimum := vector2{X := 0.5, Y := 0.5}, Maximum := vector2{X := 0.5, Y := 0.5}},
        anchors{Minimum := vector2{X := 0.5, Y := 0.5}, Maximum := vector2{X := 0.5, Y := 0.5}}
    }

    OffsetsBuyButton<public>: []margin = array {
        margin{Top := 3.459473, Left := -555.57489, Right := 138.224823, Bottom := 38.278355},
        margin{Top := 3.459473, Left := -289.241699, Right := 138.224823, Bottom := 38.278355},
        margin{Top := 3.459473, Left := -23.954376, Right := 138.224823, Bottom := 38.278355},
        margin{Top := 3.459473, Left := 239.039062, Right := 138.224823, Bottom := 38.278355},
        margin{Top := 3.459473, Left := 503.039062, Right := 138.224823, Bottom := 38.278355},
        margin{Top := 307.459473, Left := -555.57489, Right := 138.224823, Bottom := 38.278355},
        margin{Top := 307.459473, Left := -288.960938, Right := 138.224823, Bottom := 38.278355},
        margin{Top := 307.459473, Left := -24.960938, Right := 138.224823, Bottom := 38.278355},
        margin{Top := 307.459473, Left := 239.039062, Right := 138.224823, Bottom := 38.278355},
        margin{Top := 307.459473, Left := 503.039062, Right := 138.224823, Bottom := 38.278355}
    }

    # Shop item image slots (10x)
    AnchorsItemImage<public>: []anchors = array {
        anchors{Minimum := vector2{X := 0.5, Y := 0.5}, Maximum := vector2{X := 0.5, Y := 0.5}},
        anchors{Minimum := vector2{X := 0.5, Y := 0.5}, Maximum := vector2{X := 0.5, Y := 0.5}},
        anchors{Minimum := vector2{X := 0.5, Y := 0.5}, Maximum := vector2{X := 0.5, Y := 0.5}},
        anchors{Minimum := vector2{X := 0.5, Y := 0.5}, Maximum := vector2{X := 0.5, Y := 0.5}},
        anchors{Minimum := vector2{X := 0.5, Y := 0.5}, Maximum := vector2{X := 0.5, Y := 0.5}},
        anchors{Minimum := vector2{X := 0.5, Y := 0.5}, Maximum := vector2{X := 0.5, Y := 0.5}},
        anchors{Minimum := vector2{X := 0.5, Y := 0.5}, Maximum := vector2{X := 0.5, Y := 0.5}},
        anchors{Minimum := vector2{X := 0.5, Y := 0.5}, Maximum := vector2{X := 0.5, Y := 0.5}},
        anchors{Minimum := vector2{X := 0.5, Y := 0.5}, Maximum := vector2{X := 0.5, Y := 0.5}},
        anchors{Minimum := vector2{X := 0.5, Y := 0.5}, Maximum := vector2{X := 0.5, Y := 0.5}}
    }
    
    OffsetsItemImage<public>: []margin = array {
        margin{Top := -256.540527, Left := -616.960938, Right := 267.167175, Bottom := 267.167175},
        margin{Top := -256.540527, Left := -352.960938, Right := 267.167175, Bottom := 267.167175},
        margin{Top := -256.540527, Left := -88.960938, Right := 267.167175, Bottom := 267.167175},
        margin{Top := -256.540527, Left := 176.039062, Right := 267.167175, Bottom := 267.167175},
        margin{Top := -256.540527, Left := 440.039062, Right := 267.167175, Bottom := 267.167175},
        margin{Top := 47.459473, Left := -616.960938, Right := 267.167175, Bottom := 267.167175},
        margin{Top := 47.459473, Left := -352.960938, Right := 267.167175, Bottom := 267.167175},
        margin{Top := 47.459473, Left := -88.960938, Right := 267.167175, Bottom := 267.167175},
        margin{Top := 47.459473, Left := 176.039062, Right := 267.167175, Bottom := 267.167175},
        margin{Top := 47.459473, Left := 440.039062, Right := 267.167175, Bottom := 267.167175}
    }

    # Runtime
    var AgentShopMap: [agent]_UI.ui_shop = map {}
    
    OnBegin<override>()<suspends>: void = {
        GetPlayspace().PlayerAddedEvent().Subscribe(OnPlayerAdded)
        GetPlayspace().PlayerRemovedEvent().Subscribe(OnPlayerRemoved)

        OpenShopButton.InteractedWithEvent.Subscribe(OnOpenShopButton)

        # Call logic for initial players
        for(Player : GetPlayspace().GetPlayers()) {
            OnPlayerAdded(Player)
        }
    }

    GetShopUiClass(): _UI.ui_shop = {
        return _UI.ui_shop {}
    }

    OnPlayerAdded(Agent: agent): void = {
        ShopUi := GetShopUiClass()
        ShopUi.InitShop(Self, Agent)
        if(set AgentShopMap[Agent] = ShopUi) {}
    }

    OnPlayerRemoved(Agent: agent): void = {
        # Delete the element from AgentShopMap to avoid memory leak
        if(Ui := AgentShopMap[Agent]) {
            Ui.RemoveUiFromPlayer(Agent)
        }
        set AgentShopMap = AgentShopMap.RemoveKey(Agent)
    }

    OnOpenShopButton(Agent: agent): void = {
        Print("OnOpenShopButton")
        if {
            PlayerData := Agent.GetPlayerData[]
            TycoonManager := PlayerData.MaybeTycoonManager?
        } then {
            if(TycoonId <> TycoonManager.TycoonId) {
                Print("WARN: Player has a different Tycoon Id than this shop")
                return
            }

            if(PlayerData.CurLevel + 1 < LevelRequirement) {
                Print("Player level is too low for shop")
                TooLowLevelHudMessage.Show(Agent)
            } else {
                OnShopOpened(Agent)
            }
        }
    }

    # Below are overwrite functions for child classes
    OnShopOpened(Agent: agent): void = {
        if(ShopUi := AgentShopMap[Agent]) {
            ShopUi.AddUiToPlayer(Agent, false)
        }
    }

    OnShopClosed<public>(Agent: agent): void = {
        # Add custom code here
    }

    CreateCanvasUi<public>(ShopUi: _UI.ui_shop, Agent: agent): canvas = {
        CanvasUi: canvas = canvas {
            Slots := ShopUi.CreateSlots(Agent)
        }

        return CanvasUi
    }

    GetCondButton<public>(ItemIndex: int): conditional_button_device = {
        return ConditionalButton
    }

    HandleShopUiRefresh<public>(Agent: agent): void = {
        # Overwrite for child classes
    }

    # This must be called manually by the shop that wants to use it
    RefreshUi<public>(Agent: agent): void = {
        if(PlayerShopUi := AgentShopMap[Agent]) {
            PlayerShopUi.RefreshPlayerShopUi(Agent)
        }
    }

    TryToPurchaseItem<public>(Agent: agent, ItemId: int, ItemPrice: int): void = {
        Print("TryToPurchaseItem")
        PlayerCurrencyAmount := GetPlayerCurrencyCount(Agent, ItemId)
        if(PlayerCurrencyAmount >= ItemPrice) {
            PerformItemPurchase(Agent, ItemId, ItemPrice)
        } else {
            Print("TryToPurchaseItem - Player doesn't have enough currency")
            NotEnoughGoldAudio.Play(Agent)
        }
    }

    PerformItemPurchase(Agent: agent, ItemId: int, ItemPrice: int): void = {
        Print("PerformItemPurchase")
        AccoladeManager := GetAccoladeManager()
        RemovePlayerCurrency(Agent, ItemId, ItemPrice)
        ItemGranter.SetNextItem(ItemId)
        ItemGranter.GrantItem(Agent)
        PurchaseAudio.Play(Agent)
        AccoladeManager.AwardAccolade(Agent, 1)

        <# This might need some additional work, since it doesn't close related shops (for example PC)
        if(CinematicOnBuy := OptionalCinematicOnBuy?) {
            CinematicOnBuy.Play(Agent)
            # Close the shop, the player has to manually reopen it
            if(Ui := AgentShopMap[Agent]) {
                _UI.CloseShop(Agent)
            }
        }
        #>
    }

    # Overwrite for child classes
    GetPlayerCurrencyCount(Agent: agent, ItemIndex: int): int = {
        GetCondButton(ItemIndex).GetItemCount(Agent, 0)
    }

    # Overwrite for child classes
    RemovePlayerCurrency(Agent: agent, ItemIndex: int, ItemPrice: int): void = {
        CondButton := GetCondButton(ItemIndex)
        CondButton.SetItemCountRequired(0, ItemPrice)
        CondButton.Activate(Agent)
    }

    # Overwrite for child classes
    GetItemState<public>(Agent: agent, ItemIndex: int): int = {
        return 0
    }

    # Overwrite for child classes
    GetButtonText<public>(Agent: agent, ItemIndex: int): string = {
        return ""
    }
}