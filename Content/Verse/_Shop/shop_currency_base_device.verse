using { /Verse.org/Simulation }
using { /Verse.org/Colors }
using { /Verse.org/Colors/NamedColors }
# using { /Verse.org/Concurrency }
# using { /Verse.org/Verse }
using { /Verse.org/Assets }
using { /Fortnite.com/Devices }
using { /Fortnite.com/UI }
# using { /Fortnite.com/Characters }
# using { /UnrealEngine.com/Temporary/Diagnostics }
using { /UnrealEngine.com/Temporary/SpatialMath }
using { /UnrealEngine.com/Temporary/UI}

shop_currency_base_device := class(shop_base_device) {
    # 0: Money
    # 1: Fans
    # 2: Diamonds
    @editable
    CurrencyList: []int = array {
        0,
        0,
        0,
        0,
        0,
        0,
        0,
        0,
        0,
        0
    }

    @editable
    ConditionalButtonCurrencyList: []conditional_button_device = array {}

    # Constants
    ShopItemPriceSuffix<override>: string = ""

    # Constants you can override for child classes
    # Shop item currency slots (10x)
    AnchorsItemCurrencyImage: []anchors = array {
        anchors{Minimum := vector2{X := 0.5, Y := 0.5}, Maximum := vector2{X := 0.5, Y := 0.5}}
        anchors{Minimum := vector2{X := 0.5, Y := 0.5}, Maximum := vector2{X := 0.5, Y := 0.5}}
        anchors{Minimum := vector2{X := 0.5, Y := 0.5}, Maximum := vector2{X := 0.5, Y := 0.5}}
        anchors{Minimum := vector2{X := 0.5, Y := 0.5}, Maximum := vector2{X := 0.5, Y := 0.5}}
        anchors{Minimum := vector2{X := 0.5, Y := 0.5}, Maximum := vector2{X := 0.5, Y := 0.5}}
        anchors{Minimum := vector2{X := 0.5, Y := 0.5}, Maximum := vector2{X := 0.5, Y := 0.5}}
        anchors{Minimum := vector2{X := 0.5, Y := 0.5}, Maximum := vector2{X := 0.5, Y := 0.5}}
        anchors{Minimum := vector2{X := 0.5, Y := 0.5}, Maximum := vector2{X := 0.5, Y := 0.5}}
        anchors{Minimum := vector2{X := 0.5, Y := 0.5}, Maximum := vector2{X := 0.5, Y := 0.5}}
        anchors{Minimum := vector2{X := 0.5, Y := 0.5}, Maximum := vector2{X := 0.5, Y := 0.5}}
    }
    
    OffsetsItemCurrencyImage: []margin = array {
        margin{Top := -112.540527, Left := -560.960938, Right := 152.439941, Bottom := 146.118607}
        margin{Top := -112.540527, Left := -296.960938, Right := 152.439941, Bottom := 146.118607}
        margin{Top := -112.540527, Left := -32.960938, Right := 152.439941, Bottom := 146.118607}
        margin{Top := -112.540527, Left := 235.039062, Right := 152.439941, Bottom := 146.118607}
        margin{Top := -112.540527, Left := 495.039062, Right := 152.439941, Bottom := 146.118607}
        margin{Top := 191.459473, Left := -560.960938, Right := 152.439941, Bottom := 146.118607}
        margin{Top := 191.459473, Left := -296.960938, Right := 152.439941, Bottom := 146.118607}
        margin{Top := 191.459473, Left := -32.960938, Right := 152.439941, Bottom := 146.118607}
        margin{Top := 191.459473, Left := 235.039062, Right := 152.439941, Bottom := 146.118607}
        margin{Top := 191.459473, Left := 495.039062, Right := 152.439941, Bottom := 146.118607}
    }

    # TextBlock for item cost (x10)
    AnchorsItemCostTextBlock: []anchors = array {
        anchors{Minimum := vector2{X := 0.5, Y := 0.5}, Maximum := vector2{X := 0.5, Y := 0.5}}
        anchors{Minimum := vector2{X := 0.5, Y := 0.5}, Maximum := vector2{X := 0.5, Y := 0.5}}
        anchors{Minimum := vector2{X := 0.5, Y := 0.5}, Maximum := vector2{X := 0.5, Y := 0.5}}
        anchors{Minimum := vector2{X := 0.5, Y := 0.5}, Maximum := vector2{X := 0.5, Y := 0.5}}
        anchors{Minimum := vector2{X := 0.5, Y := 0.5}, Maximum := vector2{X := 0.5, Y := 0.5}}
        anchors{Minimum := vector2{X := 0.5, Y := 0.5}, Maximum := vector2{X := 0.5, Y := 0.5}}
        anchors{Minimum := vector2{X := 0.5, Y := 0.5}, Maximum := vector2{X := 0.5, Y := 0.5}}
        anchors{Minimum := vector2{X := 0.5, Y := 0.5}, Maximum := vector2{X := 0.5, Y := 0.5}}
        anchors{Minimum := vector2{X := 0.5, Y := 0.5}, Maximum := vector2{X := 0.5, Y := 0.5}}
        anchors{Minimum := vector2{X := 0.5, Y := 0.5}, Maximum := vector2{X := 0.5, Y := 0.5}}
    }
    
    OffsetsItemCostTextBlock: []margin = array {
        margin{Top := -60.540527, Left := -496.960938, Right := 66.329903, Bottom := 41.0}
        margin{Top := -60.540527, Left := -232.960938, Right := 66.329903, Bottom := 41.0}
        margin{Top := -60.540527, Left := 31.039062, Right := 66.329903, Bottom := 41.0}
        margin{Top := -60.540527, Left := 296.39621, Right := 66.329903, Bottom := 41.0}
        margin{Top := -60.114609, Left := 558.468567, Right := 66.329903, Bottom := 41.0}
        margin{Top := 244.448944, Left := -496.960938, Right := 66.329903, Bottom := 41.0}
        margin{Top := 244.448944, Left := -232.960938, Right := 66.329903, Bottom := 41.0}
        margin{Top := 244.448944, Left := 31.039062, Right := 66.329903, Bottom := 41.0}
        margin{Top := 244.448944, Left := 296.39621, Right := 66.329903, Bottom := 41.0}
        margin{Top := 244.448944, Left := 558.468567, Right := 66.329903, Bottom := 41.0}
    }

    # Runtime variables
    var TextureMaterialSlotList: []texture_block = array {}
    var TextBlockCostList: []text_block = array {}

    CreateCanvasUi<override>(ShopUi: _UI.ui_shop, Agent: agent): canvas = {
        Utility := GetUtility()

        TextureMoney: texture = UI.dollarcurrency
        TextureFans: texture = UI.fanscurrency
        TextureDiamonds: texture = UI.diamondcurrency

        var TextureSlot1: texture = TextureMoney
        var TextureSlot2: texture = TextureMoney
        var TextureSlot3: texture = TextureMoney
        var TextureSlot4: texture = TextureMoney
        var TextureSlot5: texture = TextureMoney
        var TextureSlot6: texture = TextureMoney
        var TextureSlot7: texture = TextureMoney
        var TextureSlot8: texture = TextureMoney
        var TextureSlot9: texture = TextureMoney
        var TextureSlot10: texture = TextureMoney

        if(CurrencyId := CurrencyList[0]) {
            case(CurrencyId) {
                1 =>
                    set TextureSlot1 = TextureFans
                2 =>
                    set TextureSlot1 = TextureDiamonds
                _ =>
                    set TextureSlot1 = TextureMoney
            }
        }

        if(CurrencyId := CurrencyList[1]) {
            case(CurrencyId) {
                1 =>
                    set TextureSlot2 = TextureFans
                2 =>
                    set TextureSlot2 = TextureDiamonds
                _ =>
                    set TextureSlot2 = TextureMoney
            }
        }

        if(CurrencyId := CurrencyList[2]) {
            case(CurrencyId) {
                1 =>
                    set TextureSlot3 = TextureFans
                2 =>
                    set TextureSlot3 = TextureDiamonds
                _ =>
                    set TextureSlot3 = TextureMoney
            }
        }

        if(CurrencyId := CurrencyList[3]) {
            case(CurrencyId) {
                1 =>
                    set TextureSlot4 = TextureFans
                2 =>
                    set TextureSlot4 = TextureDiamonds
                _ =>
                    set TextureSlot4 = TextureMoney
            }
        }

        if(CurrencyId := CurrencyList[4]) {
            case(CurrencyId) {
                1 =>
                    set TextureSlot5 = TextureFans
                2 =>
                    set TextureSlot5 = TextureDiamonds
                _ =>
                    set TextureSlot5 = TextureMoney
            }
        }

        if(CurrencyId := CurrencyList[5]) {
            case(CurrencyId) {
                1 =>
                    set TextureSlot6 = TextureFans
                2 =>
                    set TextureSlot6 = TextureDiamonds
                _ =>
                    set TextureSlot6 = TextureMoney
            }
        }

        if(CurrencyId := CurrencyList[6]) {
            case(CurrencyId) {
                1 =>
                    set TextureSlot7 = TextureFans
                2 =>
                    set TextureSlot7 = TextureDiamonds
                _ =>
                    set TextureSlot7 = TextureMoney
            }
        }

        if(CurrencyId := CurrencyList[7]) {
            case(CurrencyId) {
                1 =>
                    set TextureSlot8 = TextureFans
                2 =>
                    set TextureSlot8 = TextureDiamonds
                _ =>
                    set TextureSlot8 = TextureMoney
            }
        }

        if(CurrencyId := CurrencyList[8]) {
            case(CurrencyId) {
                1 =>
                    set TextureSlot9 = TextureFans
                2 =>
                    set TextureSlot9 = TextureDiamonds
                _ =>
                    set TextureSlot9 = TextureMoney
            }
        }

        if(CurrencyId := CurrencyList[9]) {
            case(CurrencyId) {
                1 =>
                    set TextureSlot10 = TextureFans
                2 =>
                    set TextureSlot10 = TextureDiamonds
                _ =>
                    set TextureSlot10 = TextureMoney
            }
        }

        # We need a different texture_block for every time we use the image!
        var TextureMaterialSlot1: texture_block = texture_block {
            DefaultImage := TextureSlot1
            DefaultDesiredSize := vector2{X := 10.315323, Y := 10.027031}
        }
        var TextureMaterialSlot2: texture_block = texture_block {
            DefaultImage := TextureSlot2
            DefaultDesiredSize := vector2{X := 10.315323, Y := 10.027031}
        }
        var TextureMaterialSlot3: texture_block = texture_block {
            DefaultImage := TextureSlot3
            DefaultDesiredSize := vector2{X := 10.315323, Y := 10.027031}
        }
        var TextureMaterialSlot4: texture_block = texture_block {
            DefaultImage := TextureSlot4
            DefaultDesiredSize := vector2{X := 10.315323, Y := 10.027031}
        }
        var TextureMaterialSlot5: texture_block = texture_block {
            DefaultImage := TextureSlot5
            DefaultDesiredSize := vector2{X := 10.315323, Y := 10.027031}
        }
        var TextureMaterialSlot6: texture_block = texture_block {
            DefaultImage := TextureSlot6
            DefaultDesiredSize := vector2{X := 10.315323, Y := 10.027031}
        }
        var TextureMaterialSlot7: texture_block = texture_block {
            DefaultImage := TextureSlot7
            DefaultDesiredSize := vector2{X := 10.315323, Y := 10.027031}
        }
        var TextureMaterialSlot8: texture_block = texture_block {
            DefaultImage := TextureSlot8
            DefaultDesiredSize := vector2{X := 10.315323, Y := 10.027031}
        }
        var TextureMaterialSlot9: texture_block = texture_block {
            DefaultImage := TextureSlot9
            DefaultDesiredSize := vector2{X := 10.315323, Y := 10.027031}
        }
        var TextureMaterialSlot10: texture_block = texture_block {
            DefaultImage := TextureSlot10
            DefaultDesiredSize := vector2{X := 10.315323, Y := 10.027031}
        }

        set TextureMaterialSlotList = array {
            TextureMaterialSlot1,
            TextureMaterialSlot2,
            TextureMaterialSlot3,
            TextureMaterialSlot4,
            TextureMaterialSlot5,
            TextureMaterialSlot6,
            TextureMaterialSlot7,
            TextureMaterialSlot8,
            TextureMaterialSlot9,
            TextureMaterialSlot10
        }

        # Add text blocks
        var TextBlockCost1: text_block = text_block{
            DefaultTextColor := color{R:=1.0, G:=1.0, B:=1.0}
        }

        var TextBlockCost2: text_block = text_block{
            DefaultTextColor := color{R:=1.0, G:=1.0, B:=1.0}
        }

        var TextBlockCost3: text_block = text_block{
            DefaultTextColor := color{R:=1.0, G:=1.0, B:=1.0}
        }

        var TextBlockCost4: text_block = text_block{
            DefaultTextColor := color{R:=1.0, G:=1.0, B:=1.0}
        }

        var TextBlockCost5: text_block = text_block{
            DefaultTextColor := color{R:=1.0, G:=1.0, B:=1.0}
        }

        var TextBlockCost6: text_block = text_block{
            DefaultTextColor := color{R:=1.0, G:=1.0, B:=1.0}
        }

        var TextBlockCost7: text_block = text_block{
            DefaultTextColor := color{R:=1.0, G:=1.0, B:=1.0}
        }

        var TextBlockCost8: text_block = text_block{
            DefaultTextColor := color{R:=1.0, G:=1.0, B:=1.0}
        }

        var TextBlockCost9: text_block = text_block{
            DefaultTextColor := color{R:=1.0, G:=1.0, B:=1.0}
        }

        var TextBlockCost10: text_block = text_block{
            DefaultTextColor := color{R:=1.0, G:=1.0, B:=1.0}
        }

        set TextBlockCostList = array {
            TextBlockCost1,
            TextBlockCost2,
            TextBlockCost3,
            TextBlockCost4,
            TextBlockCost5,
            TextBlockCost6,
            TextBlockCost7,
            TextBlockCost8,
            TextBlockCost9,
            TextBlockCost10
        }

        var ExtendedSlots: []canvas_slot = array {}
        set ExtendedSlots = ExtendedSlots + ShopUi.CreateSlots(Agent)

        # Add custom slots
        # Shop item currency slots (10x)
        for(ItemIndex -> AnchorsItem: AnchorsItemCurrencyImage) {
            if {
                OffsetsItem := OffsetsItemCurrencyImage[ItemIndex]
                WidgetItem := TextureMaterialSlotList[ItemIndex]
            } then {
                set ExtendedSlots = ExtendedSlots + array {
                    canvas_slot {
                        Anchors := AnchorsItem
                        Offsets := OffsetsItem
                        Alignment := vector2{X := 0.0, Y := 0.0}
                        SizeToContent := false
                        Widget := WidgetItem
                        ZOrder := 900
                    }
                }
            }
        }

        # TextBlock for item cost (x10)
        for(ItemIndex -> AnchorsItem: AnchorsItemCostTextBlock) {
            if {
                OffsetsItem := OffsetsItemCostTextBlock[ItemIndex]
                WidgetItem := TextBlockCostList[ItemIndex]
            } then {
                set ExtendedSlots = ExtendedSlots + array {
                    canvas_slot {
                        Anchors := AnchorsItem
                        Offsets := OffsetsItem
                        Alignment := vector2{X := 0.0, Y := 0.0}
                        SizeToContent := false
                        Widget := WidgetItem
                        ZOrder := 90000
                    }
                }
            }
        }

        # Set price texts
        for(ItemIndex -> ItemPrice: PriceList) {
            if(TextBlock := TextBlockCostList[ItemIndex]) {
                TextBlock.SetText(Utility.FormatMessage(Utility.FormatFloat(ItemPrice * 1.0, 2, 0)))
            } else {
                Print("ERR: currency_shop_ui_device: Unknown TextBlockCost - ItemIndex: {ItemIndex}")
            }
        }

        # Hide TextureMaterialSlots initially
        for(TextureMaterialSlot: TextureMaterialSlotList) {
            # Old method
            # TextureMaterialSlot.SetVisibility(widget_visibility.Hidden)

            # New method: Hide TextureMaterialSlots, which have no PriceList config
            for(ItemIndex := PriceList.Length .. ShopItemsPerPage) {
                if {
                    TextureMaterialSlotToHide := TextureMaterialSlotList[ItemIndex]
                } then {
                    TextureMaterialSlotToHide.SetVisibility(widget_visibility.Hidden)
                }
            }
        }

        # Create canvas
        CanvasUi: canvas = canvas {
            Slots := ExtendedSlots
        }

        Print("NUM OF Canvas_slots: {ExtendedSlots.Length}")

        return CanvasUi
    }

    HandleShopUiRefresh<override>(Agent: agent): void = {
        (super:)HandleShopUiRefresh(Agent)
        # Add custom logic here
        <# Old method: Not needed anymore, since we only hide the slots which have no config
         # It doesn't work somehow for Tycoon Slots 2 + 3
        for(ItemIndex := 0 .. ShopItemsPerPage) {
            # Check if we have a config for this item
            if {
                ItemPrice := PriceList[ItemIndex]
                TextureMaterialSlot := TextureMaterialSlotList[ItemIndex]
            } then {
                Print("HandleShopUiRefresh: ItemIndex: {ItemIndex}")
                TextureMaterialSlot.SetVisibility(widget_visibility.Visible)
            }
        }
        #>
    }

    GetPlayerCurrencyCount<override>(Agent: agent, ItemIndex: int): int = {
        if {
            ItemCurrencyId := CurrencyList[ItemIndex]
            PlayerData := Agent.GetPlayerData[]
            FansFloored := Floor[PlayerData.CurFans]
            DiamondsFloored := Floor[PlayerData.CurDiamonds]
        } then {
            case(ItemCurrencyId) {
                1 =>
                    return FansFloored
                2 =>
                    return DiamondsFloored
                _ =>
                    return PlayerData.CurMoney
            } 
        } else {
            Print("ERR: GetPlayerCurrencyCount: Found no ItemCurrencyId for ItemIndex: {ItemIndex}")
            return 0
        }
    }

    RemovePlayerCurrency<override>(Agent: agent, ItemIndex: int, ItemPrice: int): void = {
        SaveManager := GetSaveManager()
        if {
            ItemCurrencyId := CurrencyList[ItemIndex]
            PlayerData := Agent.GetPlayerData[]
        } then {
            case(ItemCurrencyId) {
                1 =>
                    PlayerData.RemoveFans(ItemPrice)
                2 =>
                    PlayerData.RemoveDiamonds(ItemPrice)
                _ =>
                    PlayerData.RemoveMoney(ItemPrice)
                    # Increment achievements
                    # Spend $ 250K for Guns/Ammo
                    if {
                        ShopId <= 5
                    } then {
                        SaveManager.IncrementAchievement(Agent, _Global.achievement_type.Money, 2, ItemPrice)
                    }
            } 
        } else {
            Print("ERR: RemovePlayerCurrency: Found no ItemCurrencyId for ItemIndex: {ItemIndex}")
        }
    }

    <# Not in use for tycoons
    Overwrite to get a cond button that uses a special currency
    GetCondButton<override>(ItemIndex: int): conditional_button_device = {
        if {
            CurrencyId := CurrencyList[ItemIndex]
            CurrencyCondButton := ConditionalButtonCurrencyList[CurrencyId]
        } then {
            return CurrencyCondButton
        } else {
            Print("ERR: CurrencyShop: GetCondButton: Found no CondButton ItemIndex: {ItemIndex}")
            return ConditionalButton
        }
    }
    #>

    GetButtonText<override>(Agent: agent, ItemIndex: int): string = {
        return "BUY"
    }
}