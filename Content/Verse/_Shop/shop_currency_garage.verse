using { /Verse.org/Simulation }
using { /Verse.org/Colors }
using { /Verse.org/Colors/NamedColors }
# using { /Verse.org/Concurrency }
# using { /Verse.org/Verse }
using { /Verse.org/Assets }
using { /Fortnite.com/Devices }
using { /Fortnite.com/UI }
# using { /Fortnite.com/Characters }
# using { /UnrealEngine.com/Temporary/Diagnostics }
using { /UnrealEngine.com/Temporary/SpatialMath }
using { /UnrealEngine.com/Temporary/UI}

shop_currency_garage<public> := class(shop_base_device) {
    @editable
    CustomCarDevice: custom_car_device = custom_car_device {}

    @editable
    DragonMountDevice: dragon_mount_device = dragon_mount_device {}

    # 0 can buy, 1 owned
    GetItemState<override>(Agent: agent, ItemIndex: int): int = {
        SaveManager := GetSaveManager()
        var ButtonText: string = ""
        if {
            Save := SaveManager.ProxySaves[Agent]
            ItemState := Save.CarShopStateList[ItemIndex]
        } then {
            return ItemState
        } else {
            Print("ERR: GarageShop: GetItemState: Didn't find ItemState for Index: {ItemIndex}")
            return 0
        }
    }

    GetButtonText<override>(Agent: agent, ItemIndex: int): string = {
        var ButtonText: string = ""
        ItemState := GetItemState(Agent, ItemIndex)
        case(ItemState) {
        0 =>
            return "🔒"
        1 =>
            return "DRIVE"
        _ =>
            return "🔒"
        }
    }

    TryToPurchaseItem<override>(Agent: agent, ItemId: int, ItemPrice: int): void = {
        GameManager := GetGameManager()
        ItemState := GetItemState(Agent, ItemId)
        if(ItemState = 0) {
            # Do nothing here since we can't buy the car in this shop
            # (super:)TryToPurchaseItem(Agent, ItemId, ItemPrice)
        } else {
            # The item is already owned
            HandleCarSummon(Agent, ItemId)
        }
    }

    PerformItemPurchase<override>(Agent: agent, ItemId: int, ItemPrice: int): void = {
        (super:)PerformItemPurchase(Agent, ItemId, ItemPrice)
        HandleCarSummon(Agent, ItemId)
    }

    HandleCarSummon<public>(Agent: agent, ItemId: int): void = {
        GameManager := GetGameManager()
        SaveManager := GetSaveManager()
        if {
            PlayerData := Agent.GetPlayerData[]
        } then {
            if(ItemId <= 5) {
                CustomCarDevice.AssignOwner(Agent)
                CustomCarDevice.OnCarSwitch(Agent, ItemId)
            } else if(ItemId = 6) {
                DragonMountDevice.StartFlying(Agent)
            }
            if(ShopUi := AgentShopMap[Agent]) {
                ShopUi.CloseShop(Agent)
            }
        }
    }
}