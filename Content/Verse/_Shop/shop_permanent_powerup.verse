using { /Verse.org/Simulation }
using { /Verse.org/Colors }
using { /Verse.org/Colors/NamedColors }
# using { /Verse.org/Concurrency }
# using { /Verse.org/Verse }
using { /Verse.org/Assets }
using { /Fortnite.com/Devices }
using { /Fortnite.com/UI }
# using { /Fortnite.com/Characters }
# using { /UnrealEngine.com/Temporary/Diagnostics }
using { /UnrealEngine.com/Temporary/SpatialMath }
using { /UnrealEngine.com/Temporary/UI}

# Needs different verse device for UI class
shop_permanent_powerup<public> := class(shop_base_device) {
    GetShopUiClass<override>(): _UI.ui_shop = {
        return _UI.ui_shop_upgrade {}
    }
}