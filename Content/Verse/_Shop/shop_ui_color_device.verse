using { /Verse.org/Simulation }
using { /Verse.org/Colors }
using { /Verse.org/Colors/NamedColors }
# using { /Verse.org/Concurrency }
# using { /Verse.org/Verse }
using { /Verse.org/Assets }
using { /Fortnite.com/Devices }
using { /Fortnite.com/UI }
# using { /Fortnite.com/Characters }
# using { /UnrealEngine.com/Temporary/Diagnostics }
using { /UnrealEngine.com/Temporary/SpatialMath }
using { /UnrealEngine.com/Temporary/UI}

shop_ui_color_device := class(shop_currency_base_device) {
    # Properties
    @editable
    ColorShopId: int = 0

    @editable
    ShopCinematic: cinematic_sequence_device = cinematic_sequence_device {}

    @editable
    WallsThatChangeColor: []creative_prop = array {}

    # Constants you can modify
    BackgroundImageTexture<override>: texture = UI.colorbackground

    # Background image
    AnchorsBackgroundImage<override>: anchors = anchors{Minimum := vector2{X := 0.5, Y := 0.5}, Maximum := vector2{X := 0.5, Y := 0.5}}
    OffsetsBackgroundImage<override>: margin = margin{Top := -372.540527, Left := -904.960938, Right := 1818.165894, Bottom := 1378.645752}

    # Close shop button
    AnchorsCloseShopButton<override>: anchors = anchors{Minimum := vector2{X := 0.5, Y := 0.5}, Maximum := vector2{X := 0.5, Y := 0.5}}
    OffsetsCloseShopButton<override>: margin = margin{Top := 187.459473, Left := -88.960938, Right := 163.563293, Bottom := 40.921097}

    # Buy button slots (10x)
    AnchorsBuyButton<override>: []anchors = array {
        anchors{Minimum := vector2{X := 0.5, Y := 0.5}, Maximum := vector2{X := 0.5, Y := 0.5}}
        anchors{Minimum := vector2{X := 0.5, Y := 0.5}, Maximum := vector2{X := 0.5, Y := 0.5}}
        anchors{Minimum := vector2{X := 0.5, Y := 0.5}, Maximum := vector2{X := 0.5, Y := 0.5}}
        anchors{Minimum := vector2{X := 0.5, Y := 0.5}, Maximum := vector2{X := 0.5, Y := 0.5}}
        anchors{Minimum := vector2{X := 0.5, Y := 0.5}, Maximum := vector2{X := 0.5, Y := 0.5}}
        anchors{Minimum := vector2{X := 0.5, Y := 0.5}, Maximum := vector2{X := 0.5, Y := 0.5}}
        anchors{Minimum := vector2{X := 0.5, Y := 0.5}, Maximum := vector2{X := 0.5, Y := 0.5}}
        anchors{Minimum := vector2{X := 0.5, Y := 0.5}, Maximum := vector2{X := 0.5, Y := 0.5}}
        anchors{Minimum := vector2{X := 0.5, Y := 0.5}, Maximum := vector2{X := 0.5, Y := 0.5}}
        anchors{Minimum := vector2{X := 0.5, Y := 0.5}, Maximum := vector2{X := 0.5, Y := 0.5}}
    }
    
    OffsetsBuyButton<override>: []margin = array {
        margin{Top := 463.459473, Left := -780.960938, Right := 138.224823, Bottom := 38.278355}
        margin{Top := 463.459473, Left := -624.960938, Right := 138.224823, Bottom := 38.278355}
        margin{Top := 463.459473, Left := -468.960938, Right := 138.224823, Bottom := 38.278355}
        margin{Top := 463.459473, Left := -312.960938, Right := 138.224823, Bottom := 38.278355}
        margin{Top := 463.459473, Left := -156.960938, Right := 138.224823, Bottom := 38.278355}
        margin{Top := 463.459473, Left := 3.039062, Right := 138.224823, Bottom := 38.278355}
        margin{Top := 463.459473, Left := 159.039062, Right := 138.224823, Bottom := 38.278355}
        margin{Top := 463.459473, Left := 315.039062, Right := 138.224823, Bottom := 38.278355}
        margin{Top := 463.459473, Left := 471.039062, Right := 138.224823, Bottom := 38.278355}
        margin{Top := 463.459473, Left := 627.039062, Right := 138.224823, Bottom := 38.278355}
    }

    # Shop item image slots (10x)
    AnchorsItemImage<override>: []anchors = array {
        anchors{Minimum := vector2{X := 0.5, Y := 0.5}, Maximum := vector2{X := 0.5, Y := 0.5}}
        anchors{Minimum := vector2{X := 0.5, Y := 0.5}, Maximum := vector2{X := 0.5, Y := 0.5}}
        anchors{Minimum := vector2{X := 0.5, Y := 0.5}, Maximum := vector2{X := 0.5, Y := 0.5}}
        anchors{Minimum := vector2{X := 0.5, Y := 0.5}, Maximum := vector2{X := 0.5, Y := 0.5}}
        anchors{Minimum := vector2{X := 0.5, Y := 0.5}, Maximum := vector2{X := 0.5, Y := 0.5}}
        anchors{Minimum := vector2{X := 0.5, Y := 0.5}, Maximum := vector2{X := 0.5, Y := 0.5}}
        anchors{Minimum := vector2{X := 0.5, Y := 0.5}, Maximum := vector2{X := 0.5, Y := 0.5}}
        anchors{Minimum := vector2{X := 0.5, Y := 0.5}, Maximum := vector2{X := 0.5, Y := 0.5}}
        anchors{Minimum := vector2{X := 0.5, Y := 0.5}, Maximum := vector2{X := 0.5, Y := 0.5}}
        anchors{Minimum := vector2{X := 0.5, Y := 0.5}, Maximum := vector2{X := 0.5, Y := 0.5}}
    }

    OffsetsItemImage<override>: []margin = array {
        margin{Top := 263.459473, Left := -788.960938, Right := 157.277069, Bottom := 157.277069}
        margin{Top := 263.459473, Left := -632.960938, Right := 157.277069, Bottom := 157.277069}
        margin{Top := 263.459473, Left := -476.960938, Right := 157.277069, Bottom := 157.277069}
        margin{Top := 263.459473, Left := -324.960938, Right := 157.277069, Bottom := 157.277069}
        margin{Top := 263.459473, Left := -168.960938, Right := 157.277069, Bottom := 157.277069}
        margin{Top := 263.459473, Left := -4.960938, Right := 157.277069, Bottom := 157.277069}
        margin{Top := 263.459473, Left := 147.039062, Right := 157.277069, Bottom := 157.277069}
        margin{Top := 263.459473, Left := 303.039062, Right := 157.277069, Bottom := 157.277069}
        margin{Top := 263.459473, Left := 459.039062, Right := 157.277069, Bottom := 157.277069}
        margin{Top := 263.459473, Left := 615.039062, Right := 157.277069, Bottom := 157.277069}
    }

    # Shop item currency slots (10x)
    AnchorsItemCurrencyImage<override>: []anchors = array {
        anchors{Minimum := vector2{X := 0.5, Y := 0.5}, Maximum := vector2{X := 0.5, Y := 0.5}}
        anchors{Minimum := vector2{X := 0.5, Y := 0.5}, Maximum := vector2{X := 0.5, Y := 0.5}}
        anchors{Minimum := vector2{X := 0.5, Y := 0.5}, Maximum := vector2{X := 0.5, Y := 0.5}}
        anchors{Minimum := vector2{X := 0.5, Y := 0.5}, Maximum := vector2{X := 0.5, Y := 0.5}}
        anchors{Minimum := vector2{X := 0.5, Y := 0.5}, Maximum := vector2{X := 0.5, Y := 0.5}}
        anchors{Minimum := vector2{X := 0.5, Y := 0.5}, Maximum := vector2{X := 0.5, Y := 0.5}}
        anchors{Minimum := vector2{X := 0.5, Y := 0.5}, Maximum := vector2{X := 0.5, Y := 0.5}}
        anchors{Minimum := vector2{X := 0.5, Y := 0.5}, Maximum := vector2{X := 0.5, Y := 0.5}}
        anchors{Minimum := vector2{X := 0.5, Y := 0.5}, Maximum := vector2{X := 0.5, Y := 0.5}}
        anchors{Minimum := vector2{X := 0.5, Y := 0.5}, Maximum := vector2{X := 0.5, Y := 0.5}}
    }
    OffsetsItemCurrencyImage<override>: []margin = array {
        margin{Top := 351.459473, Left := -788.960938, Right := 152.439941, Bottom := 146.118607}
        margin{Top := 351.459473, Left := -632.960938, Right := 152.439941, Bottom := 146.118607}
        margin{Top := 351.459473, Left := -480.960938, Right := 152.439941, Bottom := 146.118607}
        margin{Top := 351.459473, Left := -324.960938, Right := 152.439941, Bottom := 146.118607}
        margin{Top := 351.459473, Left := -164.960938, Right := 152.439941, Bottom := 146.118607}
        margin{Top := 351.459473, Left := -4.960938, Right := 152.439941, Bottom := 146.118607}
        margin{Top := 351.459473, Left := 147.039062, Right := 152.439941, Bottom := 146.118607}
        margin{Top := 351.459473, Left := 303.039062, Right := 152.439941, Bottom := 146.118607}
        margin{Top := 351.459473, Left := 459.039062, Right := 152.439941, Bottom := 146.118607}
        margin{Top := 351.459473, Left := 615.039062, Right := 152.439941, Bottom := 146.118607}
    }

    # TextBlock for item cost (x10)
    AnchorsItemCostTextBlock<override>: []anchors = array {
        anchors{Minimum := vector2{X := 0.5, Y := 0.5}, Maximum := vector2{X := 0.5, Y := 0.5}}
        anchors{Minimum := vector2{X := 0.5, Y := 0.5}, Maximum := vector2{X := 0.5, Y := 0.5}}
        anchors{Minimum := vector2{X := 0.5, Y := 0.5}, Maximum := vector2{X := 0.5, Y := 0.5}}
        anchors{Minimum := vector2{X := 0.5, Y := 0.5}, Maximum := vector2{X := 0.5, Y := 0.5}}
        anchors{Minimum := vector2{X := 0.5, Y := 0.5}, Maximum := vector2{X := 0.5, Y := 0.5}}
        anchors{Minimum := vector2{X := 0.5, Y := 0.5}, Maximum := vector2{X := 0.5, Y := 0.5}}
        anchors{Minimum := vector2{X := 0.5, Y := 0.5}, Maximum := vector2{X := 0.5, Y := 0.5}}
        anchors{Minimum := vector2{X := 0.5, Y := 0.5}, Maximum := vector2{X := 0.5, Y := 0.5}}
        anchors{Minimum := vector2{X := 0.5, Y := 0.5}, Maximum := vector2{X := 0.5, Y := 0.5}}
        anchors{Minimum := vector2{X := 0.5, Y := 0.5}, Maximum := vector2{X := 0.5, Y := 0.5}}
    }
    OffsetsItemCostTextBlock<override>: []margin = array {
        margin{Top := 403.459473, Left := -728.960938, Right := 66.329903, Bottom := 41.0}
        margin{Top := 403.459473, Left := -572.960938, Right := 66.329903, Bottom := 41.0}
        margin{Top := 403.459473, Left := -420.960938, Right := 66.329903, Bottom := 41.0}
        margin{Top := 403.459473, Left := -264.960938, Right := 66.329903, Bottom := 41.0}
        margin{Top := 403.459473, Left := -108.960938, Right := 66.329903, Bottom := 41.0}
        margin{Top := 403.459473, Left := 51.039062, Right := 66.329903, Bottom := 41.0}
        margin{Top := 403.459473, Left := 203.039062, Right := 66.329903, Bottom := 41.0}
        margin{Top := 403.459473, Left := 359.039062, Right := 66.329903, Bottom := 41.0}
        margin{Top := 403.459473, Left := 507.039062, Right := 66.329903, Bottom := 41.0}
        margin{Top := 403.459473, Left := 671.039062, Right := 66.329903, Bottom := 41.0}
    }

    # Constants
    WallMaterialList: []material = array{
        Materials.BlueShop,
        Materials.RedShop,
        Materials.PinkShop,
        Materials.GreenShop,
        Materials.YellowShop,
        Materials.PurpleShop,
        Materials.HackingShop,
        Materials.EmojiShop,
        Materials.RainbowShop,
        Materials.CoinsShop
    }

    # Runtime variables
    var CurrentItemEquippedId: int = -1

    OnShopOpened<override>(Agent: agent): void = {
        (super:)OnShopOpened(Agent)

        if {
            PlayerData := Agent.GetPlayerData[]
        } then {
            PlayerData.PlayCinematic(ShopCinematic)
        }
    }

    OnShopClosed<override>(Agent: agent): void = {
        (super:)OnShopClosed(Agent)

        if {
            PlayerData := Agent.GetPlayerData[]
        } then {
            PlayerData.StopCurrentCinematic()
        }
    }

    # 0 can buy, 1 owned, 2 equipped
    GetItemState<override>(Agent: agent, ItemIndex: int): int = {
        GameManager := GetGameManager()
        if {
            PlayerData := Agent.GetPlayerData[]
        } then {
            return PlayerData.GetColorShopItemState(ColorShopId, ItemIndex)
        } else {
            Print("ERR: ColorShop: GetItemState: Didn't find ItemState for Index: {ItemIndex}")
            return 0
        }
    }

    GetButtonText<override>(Agent: agent, ItemIndex: int): string = {
        var ButtonText: string = ""
        ItemState := GetItemState(Agent, ItemIndex)
        case(ItemState) {
        0 =>
            set ButtonText = "BUY"
        1 =>
            set ButtonText = "USE"
        2 =>
            set ButtonText = "EQUIPPED"
        _ =>
            set ButtonText = "BUY"
        }
        return ButtonText
    }

    TryToPurchaseItem<override>(Agent: agent, ItemIndex: int, ItemPrice: int): void = {
        GameManager := GetGameManager()
        ItemState := GetItemState(Agent, ItemIndex)
        if(ItemState = 0) {
            (super:)TryToPurchaseItem(Agent, ItemIndex, ItemPrice)
        } else {
            # The item is already owned
            PerformItemPurchase(Agent, ItemIndex, ItemPrice)
        }
    }

    PerformItemPurchase<override>(Agent: agent, ItemIndex: int, ItemPrice: int): void = {
        GameManager := GetGameManager()
        ItemState := GetItemState(Agent, ItemIndex)
        if {
            PlayerData := Agent.GetPlayerData[]
        } then {
            case(ItemState) {
            0 =>
                # The item wasn't owned before, normal purchase
                (super:)PerformItemPurchase(Agent, ItemIndex, ItemPrice)
                <# Save the purchase
                if(Save := SaveManager.ProxySaves[Agent]) {
                    Save.SetColorValue(ItemIndex, true)
                }
                #>
                ChangeWallColor(Agent, ItemIndex)
            1 =>
                ChangeWallColor(Agent, ItemIndex)
            2 =>
                # Do nothing here
            _ =>
                Print("ERR: color_shop_ui - Unknown ItemState: {ItemState}")
            }
            # Refresh the UI
            RefreshUi(Agent)
        }
    }

    ChangeWallColor(Agent: agent, ItemIndex: int): void = {
        if {
            Material := WallMaterialList[ItemIndex]
            PlayerData := Agent.GetPlayerData[]
        } then {
            # Reset the previous UI value to OWNED
            if(CurrentItemEquippedId >= 0) {
                PlayerData.SetColorShopItemState(ColorShopId, CurrentItemEquippedId, 1)
            }
            PlayerData.SetColorShopItemState(ColorShopId, ItemIndex, 2)
            set CurrentItemEquippedId = ItemIndex
            <# Persist the wall color
            set Save.WallColorId = ItemIndex
            #>
            # Change the wall color
            for(Wall: WallsThatChangeColor) {
                Wall.SetMaterial(Material)
            }
        }
    }
}