using { /Verse.org/Simulation }
using { /Verse.org/Colors }
using { /Verse.org/Colors/NamedColors }
# using { /Verse.org/Concurrency }
# using { /Verse.org/Verse }
# using { /Fortnite.com/Devices }
using { /Fortnite.com/UI }
# using { /Fortnite.com/Characters }
# using { /UnrealEngine.com/Temporary/Diagnostics }
using { /UnrealEngine.com/Temporary/SpatialMath }
using { /UnrealEngine.com/Temporary/UI}

<# Not in use currently, also not tested
ui_save_data_progress_bar<public> := class(ui_base) {
    # Constants
    UiSlot<override>: player_ui_slot = player_ui_slot {
        ZOrder := 1025
        # InputMode := ui_input_mode.All
        InputMode := ui_input_mode.None
    }

    TextureBackground: texture_block = texture_block {
        DefaultImage := UI.placeholder_verse
        DefaultDesiredSize := vector2 {
            X := 176.740372
            Y := 159.328415
        }
    }

    LevelBarProgressColorBlock : color_block = color_block {
        DefaultColor := color{R:=42.0, G:=255.0, B:=0.0}
        DefaultDesiredSize := vector2 {
            X := 391.791779
            Y := 31.491491
        }
    }

    ButtonClose: button_loud = button_loud {}

    # Constants
    XpBarWidth: float = 558.113892
    XpBarHeight: float = 27.612614

    LevelBarAnimationDuration: float = 5.0

    # Runtime
    var LevelBarCurrentPercentage: float = 0.0
    var LevelBarTargetPercentage: float = 0.0
    # Step size can also be negative
    var LevelBarStepSize: float = 0.0
    var RoutineLevelBarAnimation: ?routine = false

    InitWidget<override>(Agent: agent): void = {
        (super:)InitWidget(Agent)

        Utility := GetUtility()
        
        # Button texts need to be defined inside the function
        ButtonClose.SetText(Utility.FormatMessage("ABORT SAVE"))

        ButtonClose.OnClick().Subscribe(OnExitButtonClicked)
    }

    # Refresh the UI every time it is added
    AddUiToPlayer<override>(Agent: agent, ParmMaybeManager: ?_Interface.interface_ui_manager): void = {
        (super:)AddUiToPlayer(Agent, ParmMaybeManager)
        RefreshUi(Agent)
    }

    RemoveUiFromPlayer<override>(Agent: agent): void = {
        (super:)RemoveUiFromPlayer(Agent)
        
        # Clean up routines
        if(Routine := RoutineLevelBarAnimation?) {
            set RoutineLevelBarAnimation = false
            Routine.Cancel()
        }
    }

    # Runtime
    CreateSlots<override>(Agent: agent): []canvas_slot = {
        Utility := GetUtility()

        # Init slots
        var Slots: []canvas_slot = array {
            # Exit Button
            canvas_slot {
                Anchors := anchors{Minimum := vector2{X := 0.5, Y := 0.5}, Maximum := vector2{X := 0.5, Y := 0.5}}
                Offsets := margin{Top := 199.459473, Left := -116.960938, Right := 225.439468, Bottom := 56.450912}
                Alignment := vector2{X := 0.0, Y := 0.0}
                SizeToContent := false
                Widget := ButtonClose
                ZOrder := 900
            },
            canvas_slot {
                Anchors := anchors{Minimum := vector2{X := 0.5, Y := 0.5}, Maximum := vector2{X := 0.5, Y := 0.5}}
                Offsets := margin{Top := -412.540527, Left := -955.082397, Right := 555.188599, Bottom := 585.045105}
                Alignment := vector2{X := 0.0, Y := 0.0}
                SizeToContent := false
                Widget := TextureBackground
                ZOrder := 800
            },
            canvas_slot {
                Anchors := anchors{Minimum := vector2{X := 0.5, Y := 1.0}, Maximum := vector2{X := 0.5, Y := 1.0}}
                Offsets := margin{Top := -300.32428, Left := -292.831787, Right := 558.113892, Bottom := 27.612614}
                Alignment := vector2{X := 0.0, Y := 0.0}
                SizeToContent := true
                Widget := LevelBarProgressColorBlock
                ZOrder := 1010
            }
        }

        return Slots
    }

    RefreshUi(Agent: agent): void = {
        if {
            PlayerData := Agent.GetPlayerData[]
        } then {
            # Reset variables
            set LevelBarCurrentPercentage = 0.0
            set LevelBarTargetPercentage = 0.0
            set LevelBarStepSize = 0.0
            Print("MY TEST - RefreshUi")
            # Spawn AnimateLevelBarTick if it's not already running
            if(Routine := RoutineLevelBarAnimation?) {
                # Do nothing, routine is already running
            } else {
                MyRoutine := SpawnRoutine(AnimateLevelBarTick)
                set RoutineLevelBarAnimation = option{MyRoutine}
            }
        }
    }

    AnimateLevelBarTick()<suspends>: void = {
        loop {
            # Sleep must be first in this case, otherwise routine will not be cancelled
            Sleep(0.0)

            set LevelBarCurrentPercentage = LevelBarCurrentPercentage + LevelBarStepSize
            # Print("AnimateLevelBarTick: LevelBarStepSize: {LevelBarStepSize}, LevelBarCurrentPercentage: {LevelBarCurrentPercentage}, LevelBarTargetPercentage: {LevelBarTargetPercentage}")
            # Check if we reached the target percentage
            if((LevelBarStepSize < 0.0 and LevelBarCurrentPercentage <= LevelBarTargetPercentage) or 
                (LevelBarStepSize >= 0.0 and LevelBarCurrentPercentage >= LevelBarTargetPercentage)) {
                set LevelBarCurrentPercentage = LevelBarTargetPercentage
                LevelBarProgressColorBlock.SetDesiredSize(vector2{X := LevelBarCurrentPercentage, Y := XpBarHeight})
                # Stop the routine
                set RoutineLevelBarAnimation = false
                HandleSaveFinished()
                break
            }

            LevelBarProgressColorBlock.SetDesiredSize(vector2{X := LevelBarCurrentPercentage, Y := XpBarHeight})
        }
    }

    HandleSaveFinished(): void = {
        if {
            Agent := MaybeAgent?
        } then {
            # TODO: Actually update save data (would be best to always update proxy, but disable auto save tick and call it manually)
            RemoveUiFromPlayer(Agent)
        }
    }
}
    #>