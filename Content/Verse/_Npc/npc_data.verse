using { /Fortnite.com/AI }
using { /Fortnite.com/AI/movement_types }
using { /Fortnite.com/Game }
using { /Fortnite.com/Animation }
using { /Fortnite.com/Animation/PlayAnimation }
using { /Fortnite.com/FortPlayerUtilities }
using { /Fortnite.com/Devices }
using { /Fortnite.com/Devices/CreativeAnimation }
using { /Fortnite.com/Characters }
using { /Verse.org/Simulation }
using { /Verse.org/Simulation/Tags }
using { /Verse.org/Random }
using { /Fortnite.com/Teams }
using { /Verse.org/Assets }
using { /Verse.org/Colors }
using { /Verse.org/Concurrency }
using { /UnrealEngine.com/Temporary/Diagnostics }
using { /UnrealEngine.com/Temporary/SpatialMath }

npc_anim_state := enum {
    # Idle is played when the NPC is standing still
    # Idle is the default animation in the Blueprint and doesn't need to be played
    Idle
    # Walk anim is played when walking around idle
    Walk
    # Run is played when a player is chased and when retreating
    Run
    # Attack animation is played during attack and doesn't need to switch anim_state
    # Attack
}

npc_state := enum {
    # Idle means walking around for this game
    Idle
    # Someone has opened the UI to talk to the NPC
    Talking
    # Someone is currently trying to recruit the NPC (slider is running)
    Recruiting
    Chasing
    Retreat
}

npc_data<public> := class() {
    # Constants
    # The speed to play the animation at.
    PlayRate : float = 1.0
    # How long to blend the previous animation into the current one.
    BlendInTime : float = 0.25
    # How long to blend the current animation into the next one.
    BlendOutTime : float = 0.25
    AnimationStartPositionSeconds : float = 0.0

    AlertDistance: float = 300.0
    LeashDistance: float = 5000.0
    IdleMaxWalkDistance: float = 2000.0
    IdleMinSleepTime: float = 1.0
    IdleMaxSleepTime: float = 5.0

    MaintainFocusEvent: event() = event(){}

    # Runtime variables
    var MaybeNpcManager: ?npc_spawner_manager_device = false
    var MaybeNpcAgent: ?agent = false
    var MaybeNpcFortChar: ?fort_character = false
    var MaybeNpcNavigatable: ?navigatable = false
    var MaybeNpcFocusable: ?focus_interface = false
    var MaybeNpcAnimationController: ?play_animation_controller = false
    var MaybeCurAnimationInstance: ?play_animation_instance = false

    var MaybeNpcStartWaypoint: ?npc_waypoint_device = false
    var NpcWaypointList: []npc_waypoint_device = array {}
    var MaybeCurWaypoint: ?npc_waypoint_device = false

    var RoutineServerTick: ?routine = false
    var RoutineIdleTick: ?routine = false
    var RoutineChasingTick: ?routine = false

    var CurrentTarget: ?fort_character = false
    var MaybeRecruitButton: ?button_device = false
    var RecruitButtonSub: ?cancelable = false
    var CurrentTargetEliminatedSub: ?cancelable = false

    var NpcState: npc_state = npc_state.Idle
    var NpcAnimState: npc_anim_state = npc_anim_state.Idle
    var StartPos: vector3 = vector3 {}
    var IdleMovementType: movement_type = Walking

    var NpcRarity: int = -1
    var MaybePropRarityStars: ?creative_prop = false
    var AgentsThatDamagedNpc: []agent = array {}
    var MaybeLastAgentThatDamagedNpc: ?agent = false
    var TimestampLastNpcDamage: float = 0.0
    var TimestampScamStarted: float = 0.0
    var TimestampScamOutOfRange: float = 0.0
    var MaybeScammerAgent: ?agent = false
    var CarScamSpeedFactor: float = 1.0
    var AttackDamage: float = 25.0
    var WaypointVisitedList: []npc_waypoint_device = array {}

    var NpcName: string = "Jonesy"
    var NpcSpawnerId: int = -1
    var NpcId: int = -1
    var AgentTalkedToMap: [agent]logic = map {}
    var IsMovingToStreamerSlot: logic = false
    var ReservedForOwnerPlayerId: int = -1

    # This function runs when the NPC is spawned in the world and ready to follow a behavior.
    InitNpc(Agent: agent, NpcManager: npc_spawner_manager_device, ParmRecruitButton: ?button_device, ParmMaxHp: float, ParmAttackDamage: float, ParmNpcName: string, ParmNpcSpawnerId: int, ParmNpcId: int, ParmNpcRarity: int, ParmReservedForOwnerPlayerId: int): void = {
        # Print("npc_data - InitNpc")
        Utility := GetUtility()
        if {
            # Get the fort_character interface of the agent to access Fortnite character-specific behaviors, events, functions, and interfaces.
            FortChar := Agent.GetFortCharacter[]
            # Get the navigatable interface of the character to set specific targets for the character to travel to.
            Navigatable := FortChar.GetNavigatable[]
            # Get the focus_interface of the character to set specific targets to focus on after traveling to them.
            Focusable := FortChar.GetFocusInterface[]
            # Get Anim Controller
            AnimationController := FortChar.GetPlayAnimationController[]
        } then {
            # Print("InitNpc: NpcId: {ParmNpcId}")
            GameManager := GetGameManager()

            # Init variables
            set MaybeNpcManager = option{NpcManager}
            set MaybeRecruitButton = ParmRecruitButton
            set MaybeNpcAgent = option{Agent}
            set MaybeNpcFortChar = option{FortChar}
            set MaybeNpcNavigatable = option{Navigatable}
            set MaybeNpcFocusable = option{Focusable}
            set MaybeNpcAnimationController = option{AnimationController}
            set NpcName = ParmNpcName
            set NpcSpawnerId = ParmNpcSpawnerId
            set NpcId = ParmNpcId
            set NpcRarity = ParmNpcRarity
            set ReservedForOwnerPlayerId = ParmReservedForOwnerPlayerId

            set StartPos = FortChar.GetTransform().Translation
            set AttackDamage = ParmAttackDamage
            FortChar.SetMaxHealth(ParmMaxHp)
            FortChar.SetHealth(ParmMaxHp)

            # Listen to events
            FortChar.DamagedEvent().Subscribe(OnNpcDamaged)
            FortChar.EliminatedEvent().Subscribe(OnNpcEliminated)

            if(RecruitButton := MaybeRecruitButton?) {
                RegisterRecruitButtonEvent(RecruitButton)
            }

            # Get the rarity star prop for the NPC
            if(NpcRarity >= 0) {
                # Normal NPC
                MaybeStarProp := NpcManager.GetUnusedStarProp(NpcRarity)
                if(ParmStarBlueprint := MaybeStarProp?) {
                    set MaybePropRarityStars = MaybeStarProp
                } else {
                    Print("ERR: InitNpc: Got no Star Prop for RarityLevel: {NpcRarity}")
                }
            }

            Task := spawn{InitNpcSuspends()}
        } else {
            Print("ERR: NPC Script on NPC Setup")
        }
    }

    InitNpcSuspends()<suspends>: void = {
        Sleep(1.0)
        TeleportToFirstWaypoint()
        SetNpcState(npc_state.Idle)

        MyRoutine := SpawnRoutine(ServerTick)
        set RoutineServerTick = option{MyRoutine}
    }

    AssignWaypointList(StartWaypoint: npc_waypoint_device, WaypointList: []npc_waypoint_device): void = {
        # Print("AssignWaypointList")
        set MaybeNpcStartWaypoint = option{StartWaypoint}
        set NpcWaypointList = WaypointList
    }

    # This function must be called when the NPC is despawned or eliminated from the world.
    HandleDestroy():void = {
        # Print("npc - HandleDestroy")
        if {
            NpcManager := MaybeNpcManager?
        } then {
            # Return the rarity star prop
            if(PropRarityStars := MaybePropRarityStars?) {
                NpcManager.ReturnNpcStarProp(PropRarityStars, NpcRarity)
                set MaybePropRarityStars = false
            }
            # Return the recruit button
            if(RecruitButton := MaybeRecruitButton?) {
                CancelRecruitButtonSub()
                NpcManager.ReturnNpcRecruitButton(RecruitButton)
                set MaybeRecruitButton = false
            }
        }
        
        # Clean up routines
        if(Routine := RoutineServerTick?) {
            Routine.Cancel()
        }
        if(Routine := RoutineIdleTick?) {
            Routine.Cancel()
        }
        if(Routine := RoutineChasingTick?) {
            Routine.Cancel()
        }
    }

    OnNpcDamaged(DamageResult: damage_result): void = {
        # Print("OnNpcDamaged")
        if {
            Instigator := DamageResult.Instigator?
            DamagingPlayer := Instigator.GetInstigatorAgent[]
            DamagingPlayerData := DamagingPlayer.GetPlayerData[]
            # DamagingPlayerFortChar := DamagingPlayer.GetFortCharacter[]
            NpcFortChar := MaybeNpcFortChar?
        } then {
            ## Print("OnNpcDamaged - Damaged by player")
            # Check if the player is wearing the same mask as the NPC rarity
            if {
                NpcFortChar.IsVulnerable[]
                # Disabled for this game
                # DamagingPlayerData.HitmanMaskLevel >= NpcRarity
            } then {
                # Damage was legit
                set MaybeLastAgentThatDamagedNpc = option{DamagingPlayer}
                set TimestampLastNpcDamage = GetSimulationElapsedTime()
            } else {
                # Prevent the damage by healing back the same amount
                HealNpcByAmount(DamageResult.Amount)
                <# Inform the player that he needs a different mask to damage this NPC
                if {
                    TycoonManager := DamagingPlayerData.MaybeTycoonManager?
                } then {
                    TycoonManager.TycoonConfig.HudNeedDifferentMask.Show(DamagingPlayer)
                }
                #>
            }
        }
    }

    OnNpcEliminated(EliminationResult: elimination_result): void = {
        # Print("OnNpcEliminated")
        CancelScammingNpc()
        if {
            NpcAgent := MaybeNpcAgent?
            NpcManager := MaybeNpcManager?
        } then {
            # Call NPC Manager
            NpcManager.HandleNpcEliminated(NpcAgent, EliminationResult)
        } else {
            Print("ERR: OnNpcEliminated: Not initialized correctly!")
        }
    }

    OnTargetEliminated(FortCharacter: elimination_result): void = {
        ## Print("OnTargetEliminated")
        CancelSubOnTargetEliminatedEvent()
        set CurrentTarget = false
    }

    HealNpcByAmount(HealAmount: float): void = {
        if {
            NpcFortChar := MaybeNpcFortChar?
        } then {
            # We don't use the heal method, since it might be bugged
            NpcFortChar.SetHealth(NpcFortChar.GetHealth() + HealAmount)
        }
    }

    FullyHealNpc(): void = {
        if {
            NpcFortChar := MaybeNpcFortChar?
        } then {
            NpcFortChar.SetHealth(NpcFortChar.GetMaxHealth())
        }
    }

    RegisterRecruitButtonEvent(RecruitButton: button_device): void = {
        ## Print("RegisterRecruitButtonEvent")
        EventSub := RecruitButton.InteractedWithEvent.Subscribe(OnScamButtonPressed)
        set RecruitButtonSub = option{EventSub}
    }

    CancelRecruitButtonSub(): void = {
        if(EventSub := RecruitButtonSub?) {
            EventSub.Cancel()
        }
        set RecruitButtonSub = false
    }

    RegisterSubOnTargetEliminatedEvent(TargetFortChar: fort_character): void = {
        EliminatedSub := TargetFortChar.EliminatedEvent().Subscribe(OnTargetEliminated)
        set CurrentTargetEliminatedSub = option{EliminatedSub}
    }

    CancelSubOnTargetEliminatedEvent(): void = {
        if(EliminatedSub := CurrentTargetEliminatedSub?) {
            EliminatedSub.Cancel()
        }
        set CurrentTargetEliminatedSub = false
    }

    StopNavigationMovement(): void = {
        if {
            Navigatable := MaybeNpcNavigatable?
        } then {
            Navigatable.StopNavigation()
        }
    }

    CancelIdleTick(): void = {
        # Cancel idle routine
        if(Routine := RoutineIdleTick?) {
            Routine.Cancel()
        }
        set RoutineIdleTick = false
    }

    ResetNpcMovement()<suspends>: void = {
        # Print("ResetNpcMovement")
        if {
            Navigatable := MaybeNpcNavigatable?
        } then {
            CancelIdleTick()
            # Cancel chasing routine
            if(Routine := RoutineChasingTick?) {
                Routine.Cancel()
            }
            set RoutineChasingTick = false
            # Reset movement speed
            Navigatable.SetMovementSpeedMultiplier(1.0)
            # Move back to previous goal pos
            if(CurWaypoint := MaybeCurWaypoint?) {
                NavigateToWaypoint(CurWaypoint)
            }
            # Start movement tick
            SetNpcState(npc_state.Idle)
        }
    }

    SetNewTarget(TargetFortChar: fort_character): void = {
        Utility := GetUtility()
        IsValidTarget := CheckIfTargetIsValid(TargetFortChar)
        if(IsValidTarget = true) {
            if {
                NpcFortChar := MaybeNpcFortChar?
            } then {
                set CurrentTarget = option{TargetFortChar}
                CancelSubOnTargetEliminatedEvent()
                RegisterSubOnTargetEliminatedEvent(TargetFortChar)
                Task := spawn{Utility.DrawDebugLookAt(NpcFortChar, TargetFortChar.GetTransform().Translation)}
                SetNpcState(npc_state.Chasing)
            }
        }
    }

    NavigateToWaypoint(NextWaypoint: npc_waypoint_device)<suspends>: void = {
        Utility := GetUtility()
        if {
            NpcAgent := MaybeNpcAgent?
            NpcManager := MaybeNpcManager?
            Navigatable := MaybeNpcNavigatable?
            NextWaypoint.WaypointProp.IsValid[]
        } then {
            set MaybeCurWaypoint = option{NextWaypoint}
            NextWaypointPos := NextWaypoint.WaypointProp.GetTransform().Translation
            Utility.DrawDebugLocation(NextWaypointPos)
            NavigationResult := Navigatable.NavigateTo(Target := MakeNavigationTarget(NextWaypointPos), ?MovementType := IdleMovementType)
            if(NavigationResult = navigation_result.Reached) {
                # Print("NPC {NpcId} reached waypoint {NextWaypoint.WaypointId}!")
            } else {
                # Print("WARN: NPC {NpcId} couldn't reach waypoint {NextWaypoint.WaypointId}!", ?Color := NamedColors.Red)
            }
            # We reached the waypoint, do some additional processing
            if(NextWaypoint.IsSpawnOrExitPoint = true) {
                # Print("NPC {NpcId} reached exit point {NextWaypoint.WaypointId}")
                # If the NPC is currently being scammed while despawning, then instantly finish the scam instead
                if {
                    ScammerAgent := MaybeScammerAgent?
                } then {
                    HandleScamSuccess(ScammerAgent)
                } else {
                    DespawnNpc()
                }
            } else {
                set WaypointVisitedList = WaypointVisitedList + array{NextWaypoint}
                <# IdleDuration: Disabled for now, seems to be bugged
                # Sleep random amount, based on IdleDuration property
                RandomSleepDuration := GetRandomFloat(NextWaypoint.IdleDuration / 2.0, NextWaypoint.IdleDuration)
                Sleep(RandomSleepDuration)
                #>
            }
        }
    }

    TeleportToFirstWaypoint(): void = {
        Utility := GetUtility()
        if {
            NpcManager := MaybeNpcManager?
            NpcAgent := MaybeNpcAgent?
            NpcFortChar := MaybeNpcFortChar?
            NpcFortChar.IsActive[]
            FirstWaypoint := MaybeNpcStartWaypoint?
            FirstWaypoint.WaypointProp.IsValid[]
        } then {
            # Teleport to the first waypoint
            set MaybeCurWaypoint = option{FirstWaypoint}
            set WaypointVisitedList = WaypointVisitedList + array{FirstWaypoint}

            # Teleport 200 Z upwards, so the position can't be blocked
            # if(NpcFortChar.TeleportTo[FirstWaypoint.WaypointProp.GetTransform().Translation - Utility.PlayerOffset, FirstWaypoint.WaypointProp.GetTransform().Rotation]) {}
            if(NpcFortChar.TeleportTo[FirstWaypoint.WaypointProp.GetTransform().Translation - Utility.PlayerOffset - (2.5 * Utility.PlayerOffset), FirstWaypoint.WaypointProp.GetTransform().Rotation]) {}
        }
    }

    OnScamButtonPressed(ScammerAgent: agent): void = {
        TryScammingNpc(ScammerAgent, 1.0)
    }

    TryScammingNpc<public>(ScammerAgent: agent, ParmScamSpeedFactor: float): void = {
        # Print("TryScammingNpc")
        if {
            TimestampScamStarted = 0.0
            RecruitButton := MaybeRecruitButton?
            ScammerPlayerData := ScammerAgent.GetPlayerData[]
            TycoonManager := ScammerPlayerData.MaybeTycoonManager?
        } then {
            if {
                # 1 Star NPC has NpcRarity = 0
                ScammerPlayerData.TotalScamLevel >= NpcRarity + 1
            } then {
                RecruitButton.Disable()
                set MaybeScammerAgent = option{ScammerAgent}
                set TimestampScamStarted = GetSimulationElapsedTime()
                set CarScamSpeedFactor = ParmScamSpeedFactor
                ScammerPlayerData.StartScammingNpc(ScammerAgent)
            } else {
                Print("INFO: TryScammingNpc: PlayerData.TotalScamLevel is too low for this NPC. TotalScamLevel: {ScammerPlayerData.TotalScamLevel}, Required: {NpcRarity + 1}")
                TycoonManager.TycoonConfig.HudNeedDifferentMask.Show(ScammerAgent)
            }
        }
    }

    CancelScammingNpc(): void = {
        if {
            RecruitButton := MaybeRecruitButton?
        } then {
            RecruitButton.Enable()
        }
        if {
            ScammerAgent := MaybeScammerAgent?
            ScammerPlayerData := ScammerAgent.GetPlayerData[]
        } then {
            ScammerPlayerData.StopScammingNpc(ScammerAgent)
        }
        set MaybeScammerAgent = false
        set TimestampScamStarted = 0.0
        set TimestampScamOutOfRange = 0.0
    }

    UpdateScamAttempt(): void = {
        Utility := GetUtility()
        SaveManager := GetSaveManager()
        GameConfig := GetGameConfig()
        Timestamp := GetSimulationElapsedTime()
        if {
            NpcFortChar := MaybeNpcFortChar?
            NpcManager := MaybeNpcManager?
            ScammerAgent := MaybeScammerAgent?
            ScammerPlayerData := ScammerAgent.GetPlayerData[]
            ScammerSave := SaveManager.ProxySaves[ScammerAgent]
            PlayerFortChar := ScammerAgent.GetFortCharacter[]
            PlayerFortChar.IsActive[]
            TycoonManager := ScammerPlayerData.MaybeTycoonManager?
            ScamSpeedIncreasePerUpgrade := GameConfig.PermanentUpgradeRewardPerLevelList[2]
        } then {
            var ScamMaxRange: float = 0.0
            if(CarScamSpeedFactor > 1.0) {
                set ScamMaxRange = GameConfig.ScamMaxRangeInCar
            } else {
                set ScamMaxRange = GameConfig.ScamMaxRangeNoCar
            }
            Utility.DebugDraw.DrawSphere(Center := NpcFortChar.GetTransform().Translation, ?Radius := ScamMaxRange, ?DrawDurationPolicy := debug_draw_duration_policy.FiniteDuration, ?Duration := 0.04)    
            DistanceToPlayer := Distance(NpcFortChar.GetTransform().Translation, PlayerFortChar.GetTransform().Translation)
            if(DistanceToPlayer <= ScamMaxRange) {
                # Ignore the time where the scammer was outside of range
                if(TimestampScamOutOfRange > 0.0) {
                    ScamTimeProgress := TimestampScamOutOfRange - TimestampScamStarted
                    # Print("ScamTimeProgress: {ScamTimeProgress} TimestampScamOutOfRange {TimestampScamOutOfRange} TimestampScamStarted {TimestampScamStarted}")
                    set TimestampScamStarted = GetSimulationElapsedTime() - ScamTimeProgress
                    set TimestampScamOutOfRange = 0.0
                }
                # Calculate current scam duration (min: 0 seconds)
                ScamDuration := Max((GameConfig.ScamBaseDuration - (ScammerSave.UpgradeScamSpeedLevel * ScamSpeedIncreasePerUpgrade)) / CarScamSpeedFactor, 0.0) * (1.0 + (NpcRarity * GameConfig.ScamDurationIncreasePerNpcRank))
                # Check if enough time has passed for the scam
                TimePassedSinceEvent := Timestamp - TimestampScamStarted
                # Print("Timestamp {Timestamp} TimestampScamStarted {TimestampScamStarted} TimePassedSinceEvent {TimePassedSinceEvent}")
                TimeRemainingForEvent := ScamDuration - TimePassedSinceEvent
                if(TimeRemainingForEvent <= 0.0) {
                    HandleScamSuccess(ScammerAgent)
                } else {
                    # Print("Scam in progress: Time remaining: {TimeRemainingForEvent}")
                    ProgressPercentage := TimePassedSinceEvent / ScamDuration
                    ScammerPlayerData.UiScamNpc.UpdateScamProgressBar(ProgressPercentage)
                }
            } else {
                if(TimestampScamOutOfRange > 0.0) {
                    # Check if enough time has passed to reset the scam
                    TimePassedSinceEvent := Timestamp - TimestampScamOutOfRange
                    TimeRemainingForEvent := GameConfig.ScamOutOfRangeResetDuration - TimePassedSinceEvent
                    if(TimeRemainingForEvent <= 0.0) {
                        CancelScammingNpc()
                    }
                } else {
                    # Init out of range timestamp
                    set TimestampScamOutOfRange = GetSimulationElapsedTime()
                    # Show timer in UI
                    ScammerPlayerData.UiScamNpc.HandleNpcStartsEscaping(GameConfig.ScamOutOfRangeResetDuration)
                }
            }
        }
    }

    HandleScamSuccess<public>(ScammingAgent: agent): void = {
        # Print("Scam succeeded!")
        Utility := GetUtility()
        GameManager := GetGameManager()
        SaveManager := GetSaveManager()
        AccoladeManager := GetAccoladeManager()
        GameConfig := GetGameConfig()
        # Reward the scammer
        if {
            NpcManager := MaybeNpcManager?
            PlayerData := ScammingAgent.GetPlayerData[]
            Save := SaveManager.ProxySaves[ScammingAgent]
            # MoneyScamIncreasePerUpgrade := GameConfig.PermanentUpgradeRewardPerLevelList[3]
            MoneyPerScamBaseRewardMin := GameConfig.MoneyPerScamNpcRankMin[NpcRarity]
            MoneyPerScamBaseRewardMax := GameConfig.MoneyPerScamNpcRankMax[NpcRarity]
            ScamsPerScamNpcRankMin := GameConfig.ScamsPerScamNpcRankMin[NpcRarity]
            ScamsPerScamNpcRankMax := GameConfig.ScamsPerScamNpcRankMax[NpcRarity]
        } then {
            # Scamming NPC gives same rewards as kill - XP must be first for UI popup
            XpForKill := NpcManager.CalculateXpForKill(ScammingAgent, NpcRarity)
            PlayerData.AddXp(XpForKill, true)

            MoneyPerScamBaseReward := GetRandomInt(MoneyPerScamBaseRewardMin, MoneyPerScamBaseRewardMax)
            ScamsPerScamNpcRank := GetRandomInt(ScamsPerScamNpcRankMin, ScamsPerScamNpcRankMax)

            var MoneyReward: float = (MoneyPerScamBaseReward * 1.0)
            # set MoneyReward = MoneyReward + (Save.UpgradeMoneyScamLevel * MoneyScamIncreasePerUpgrade)
            if(MoneyScamBonusFromUpgrade := GameConfig.PermanentUpgradeRewardMoneyScam[Save.UpgradeMoneyScamLevel]) {
                set MoneyReward = MoneyReward + MoneyScamBonusFromUpgrade
            }
            if {
                MoneyRewardInt := Round[MoneyReward]
            } then {
                PlayerData.AddMoney(MoneyRewardInt, true, false)
                # Increase Achievement progress
                # Earn $ 250K by scamming/killing Beast NPCs 
                if {
                    NpcSpawnerId = 0
                } then {
                    SaveManager.IncrementAchievement(ScammingAgent, _Global.achievement_type.Money, 3, MoneyRewardInt)
                }
                # Earn $ 1M by scamming/killing NPCs
                SaveManager.IncrementAchievement(ScammingAgent, _Global.achievement_type.Money, 7, MoneyRewardInt)
            }
            PlayerData.AddFans(ScamsPerScamNpcRank * 1.0, true)
            # Increase Achievement progress
            # Scam 30 NPCs Rank 2 (NpcRarity = 1, OR HIGHER)
            if(NpcRarity >= 1) {
                SaveManager.IncrementAchievement(ScammingAgent, _Global.achievement_type.Scam, 0, 1)
            }
            # Scam 3 Beast NPCs within 60 seconds
            if {
                NpcSpawnerId = 0
            } then {
                SaveManager.IncrementAchievement(ScammingAgent, _Global.achievement_type.Scam, 1, 1)
                # Starting the timer must be after IncrementAchievement
                if(Save.NumOfAchievementsCompletedScam = 1) {
                    PlayerData.StartAchievementTimerScam(60.0)
                }
            }
            # Scam 15 Goku NPCs Rank 4
            if {
                NpcSpawnerId = 5
                NpcRarity >= 3
            } then {
                SaveManager.IncrementAchievement(ScammingAgent, _Global.achievement_type.Scam, 2, 1)
            }
            # Scam 50 NPCs Rank 4
            if(NpcRarity >= 3) {
                SaveManager.IncrementAchievement(ScammingAgent, _Global.achievement_type.Scam, 3, 1)
            }
            # Scam 30 NPCs Rank 5
            if(NpcRarity >= 4) {
                SaveManager.IncrementAchievement(ScammingAgent, _Global.achievement_type.Scam, 4, 1)
            }
            # Scam 15 Travis NPCs Rank 6
            if {
                NpcSpawnerId = 2
                NpcRarity >= 5
            } then {
                SaveManager.IncrementAchievement(ScammingAgent, _Global.achievement_type.Scam, 5, 1)
            }
            # Scam 15 NPCs within 60 seconds
            SaveManager.IncrementAchievement(ScammingAgent, _Global.achievement_type.Scam, 6, 1)
            # Starting the timer must be after IncrementAchievement
            if(Save.NumOfAchievementsCompletedScam = 6) {
                PlayerData.StartAchievementTimerScam(60.0)
            }
            # Scam 25 Myers NPCs with Scam Cars
            if {
                NpcSpawnerId = 3
                CarScamSpeedFactor > 1.0
            } then {
                SaveManager.IncrementAchievement(ScammingAgent, _Global.achievement_type.Scam, 7, 1)
            }
            # Scam 25 Beast NPCs Rank 7
            if {
                NpcSpawnerId = 0
                NpcRarity >= 6
            } then {
                SaveManager.IncrementAchievement(ScammingAgent, _Global.achievement_type.Scam, 8, 1)
            }
            # Scam Thanos NPC Rank 9 with Iron Scam Helmet
            if {
                NpcSpawnerId = 1
                NpcRarity >= 8
                PlayerData.ScamLevelBack = 9
            } then {
                Print("Success: Scam Thanos NPC Rank 9 with Iron Scam Helmet")
                SaveManager.IncrementAchievement(ScammingAgent, _Global.achievement_type.Scam, 9, 1)
            }
            # Scam 50 NPCs in Rebirth 10 or higher
            if {
                PlayerData.NumOfRebirths >= 10
            } then {
                SaveManager.IncrementAchievement(ScammingAgent, _Global.achievement_type.Rebirth, 8, 1)
            }
        }
        # Increase quest progress counter
        if {
            PlayerData := ScammingAgent.GetPlayerData[]
            Quest := PlayerData.MaybeActiveQuest?
            ScamNpcRarity := Quest.QuestGoalScamNpcRarity?
            # Npc rarity must be exact for Senix Quest
            NpcRarity = ScamNpcRarity
            # NpcRarity >= ScamNpcRarity
        } then {
            PlayerData.IncreaseQuestProgress()
        }
        # Grant accolade
        AccoladeManager.AwardAccolade(ScammingAgent, 1)
        if {
            NpcAgent := MaybeNpcAgent?
            NpcFortChar := MaybeNpcFortChar?
            NpcFortChar.IsActive[]
            NpcManager := MaybeNpcManager?
        } then {
            CancelScammingNpc()
            # Play audio
            if(GameManager.AudioScamSucceeded.TeleportTo[NpcFortChar.GetTransform().Translation + (2.0 * Utility.PlayerOffset), NpcFortChar.GetTransform().Rotation]) {}
            spawn{PlayAudioScamSucceededSuspends()}
            # Cancel idle routine
            if(Routine := RoutineIdleTick?) {
                Routine.Cancel()
            }
            # Clean up other routines
            if(Routine := RoutineServerTick?) {
                Routine.Cancel()
            }
            if(Routine := RoutineChasingTick?) {
                Routine.Cancel()
            }
            # Stop movement
            StopNavigationMovement()
            # Grant weapon
            NpcManager.ItemGranterAfterScam.GrantItem(NpcAgent)
            # Spawn guard on NPC pos
            if {
                GuardSpawner := NpcManager.GuardSpawnerList[NpcSpawnerId]
            } then {
                if {
                    LastScamPosList := NpcManager.LastScamPosList[NpcSpawnerId]
                } then {
                    # Add the NPC position at the end of the list for this GuardSpawner
                    if {
                        TempArray := NpcManager.LastScamPosList[NpcSpawnerId] + array {NpcFortChar.GetTransform().Translation}
                        set NpcManager.LastScamPosList[NpcSpawnerId] = TempArray
                    }
                } else {
                    Print("ERR: LastScamPosList not initialized for NpcSpawnerId: {NpcSpawnerId}")
                }
                set NpcManager.HackNextGuardRarityToAssign = NpcRarity
                GuardSpawner.Spawn()
            } else {
                Print("ERR: Found no GuardSpawner for NpcSpawnerId: {NpcSpawnerId}")
            }
            # Hide the NPC before despawn
            if(NpcFortChar.TeleportTo[Utility.HiddenPos, rotation{}]) {}
            Task := spawn{HandleScamSuccessSuspends()}
        }
    }

    PlayAudioScamSucceededSuspends()<suspends>: void = {
        GameManager := GetGameManager()
        # Sleep must be higher than 0.0
        Sleep(0.1)
        GameManager.AudioScamSucceeded.Play()
    }

    HandleScamSuccessSuspends()<suspends>: void = {
        Sleep(0.0)
        DespawnNpc()
    }
    
    ServerTick()<suspends>: void = {
        Utility := GetUtility()
        loop {
            case(NpcState) {
                npc_state.Idle =>
                    # CheckIfPlayerNearby()
                npc_state.Talking =>
                    # Do nothing
                npc_state.Recruiting =>
                    # Do nothing
                npc_state.Chasing =>
                    if {
                        TargetFortChar := CurrentTarget?   
                        CheckIfTargetIsValid(TargetFortChar) = true
                    } then {
                        # Everything okay, keep chasing
                    } else {
                        SetNpcState(npc_state.Retreat)
                    }
                npc_state.Retreat =>
                    # Do nothing here for now
            }

            if {
                NpcFortChar := MaybeNpcFortChar?
                NpcManager := MaybeNpcManager?
            } then {
                # Check if we have to upgrade the NPC
                if {
                    NpcManager.RemainingNpcsToUpgrade > 0
                    # 9 = Max rarity, can't upgrade
                    NpcRarity < 9
                    # Don't upgrade NPCs that are currently being scammed
                    not MaybeScammerAgent?
                } then {
                    set NpcManager.RemainingNpcsToUpgrade = NpcManager.RemainingNpcsToUpgrade - 1
                    # Return the rarity star prop
                    if(PropRarityStars := MaybePropRarityStars?) {
                        NpcManager.ReturnNpcStarProp(PropRarityStars, NpcRarity)
                        set MaybePropRarityStars = false
                    }
                    # Upgrade NPC Rarity to current highest rank
                    Print("Upgrading NPC from NpcRarity {NpcRarity} to {NpcManager.PrevHighestPlayerLevel - 1}")
                    set NpcRarity = NpcManager.PrevHighestPlayerLevel - 1
                    # Get a new rarity star prop
                    MaybeStarProp := NpcManager.GetUnusedStarProp(NpcRarity)
                    if(ParmStarBlueprint := MaybeStarProp?) {
                        set MaybePropRarityStars = MaybeStarProp
                    } else {
                        Print("ERR: InitNpc: Got no Star Prop for RarityLevel: {NpcRarity}")
                    }
                }
                # Move the rarity stars to the NPC pos
                if {
                    PropRarityStars := MaybePropRarityStars?
                    PropRarityStars.IsValid[]
                } then {
                    if(PropRarityStars.TeleportTo[NpcFortChar.GetTransform().Translation + NpcManager.NpcStarPropOffset, NpcFortChar.GetTransform().Rotation]) {}
                }
                # Move the recruit button to the NPC pos
                if {
                    RecruitButton := MaybeRecruitButton?
                } then {
                    if(RecruitButton.TeleportTo[NpcFortChar.GetTransform().Translation, NpcFortChar.GetTransform().Rotation]) {}
                }
            }

            # Constantly checks if there is a player nearby that is trying to scam this NPC
            UpdateScamAttempt()

            Sleep(0.0)
        }
    }

    IdleTick()<suspends>: void = {
        # Print("IdleTick")
        Utility := GetUtility()
        if {
            NpcManager := MaybeNpcManager?
            NpcAgent := MaybeNpcAgent?
            NpcFortChar := MaybeNpcFortChar?
            FirstWaypoint := MaybeNpcStartWaypoint?
        } then {
            loop {
                # Print("IdleTick 333")
                # Move to random adjacent pos (suspends)
                SetAnimState(npc_anim_state.Walk)

                if(CurWaypoint := MaybeCurWaypoint?) {
                    # Print("NpcManager.GetNextWaypoint: NPC Id: {NpcId}")
                    MaybeNextWaypoint := NpcManager.GetNextWaypoint(NpcWaypointList, CurWaypoint, FirstWaypoint, WaypointVisitedList)
                    if {
                        NextWaypoint := MaybeNextWaypoint?
                    } then {
                        NavigateToWaypoint(NextWaypoint)
                    } else {
                        Print("ERR: NpcId: {NpcId}, CurWaypoint: {CurWaypoint.WaypointId}. NPC Waypoint IdleTick: Found no connected waypoint, that the NPC didn't already visit. Non-exit waypoints must have at least 2 connections")
                        DespawnNpc()
                    }
                }

                Sleep(0.0)

                # Sleep random amount
                # SetAnimState(npc_anim_state.Idle)
                # RandomSleepDuration := GetRandomFloat(IdleMinSleepTime, IdleMaxSleepTime)
                # Sleep(RandomSleepDuration)
            }
        }
    }

    ChasingTick()<suspends>: void = {
        Utility := GetUtility()
        loop {
            Print("ChasingTick")
            if {
                Navigatable := MaybeNpcNavigatable?
                TargetFortChar := CurrentTarget?
                TargetAgent := TargetFortChar.GetAgent[]
            } then {
                # Move to the target (suspends)
                Navigatable.SetMovementSpeedMultiplier(2.0)
                NavigationResult := Navigatable.NavigateTo(Target := MakeNavigationTarget(TargetAgent), ?MovementType := Running)
                if(NavigationResult = navigation_result.Reached) {
                    # Print("Reached player")
                } else {
                    Print("Couldn't reach player")
                }
            }
            Sleep(0.0)
        }
    }

    <#
    MoveBackToSpawn()<suspends>: void = {
        Utility := GetUtility()
        if {
            Navigatable := MaybeNpcNavigatable?
            NpcManager := MaybeNpcManager?
        } then {
            NavigationResult := Navigatable.NavigateTo(MakeNavigationTarget(StartPos))
            # Check if we still have open idle slots
            if(NpcManager.NumOfIdleNpcs < NpcManager.MaxNumOfIdleNpcs) {
                SetNpcState(npc_state.Idle)
            } else {
                DespawnNpc()
            }
        }
    }
    #>

    DespawnNpc(): void = {
        # Print("DespawnNpc")
        CancelScammingNpc()
        # Make sure we grant no damage / kill rewards if we despawn the NPC
        set AgentsThatDamagedNpc = array {}
        set MaybeLastAgentThatDamagedNpc = false
        # Since there is no despawn function we have to deal damage
        if {
            NpcFortChar := MaybeNpcFortChar?
            NpcFortChar.IsActive[]
        } then {
            # Print("DespawnNpc: NPC Id: {NpcId}")
            NpcFortChar.SetVulnerability(true)
            NpcFortChar.Damage(999999999999.0)
        }
        # Also calling destroy manually to prevent endless loop where NpcElimEvent isn't called
        HandleDestroy()
    }

    CheckIfPlayerNearby(): void = {
        GameManager := GetGameManager()
        if {
            NpcFortChar := MaybeNpcFortChar?
        } then {
            var MinDistance: float = AlertDistance
            var MaybeClosestPlayer: ?fort_character = false
            for(Player : GameManager.GetPlayspace().GetPlayers()) {
                if(PlayerFortChar := Player.GetFortCharacter[]) {
                    MyDistance := Distance(NpcFortChar.GetTransform().Translation, PlayerFortChar.GetTransform().Translation)
                    if(MyDistance < MinDistance) {
                        set MinDistance = MyDistance
                        set MaybeClosestPlayer = option{PlayerFortChar}
                    }
                }
            }
            if(ClosestPlayer := MaybeClosestPlayer?) {
                SetNewTarget(ClosestPlayer)
            }
        }
    }

    CheckIfTargetIsValid(TargetFortChar: fort_character)<transacts>: logic = {
        if {
            TargetFortChar.IsActive[]
            # TargetPos := TargetFortChar.GetTransform().Translation
            # TargetToStartDistance := Distance(TargetPos, StartPos)
            # TargetToStartDistance < LeashDistance
        } then {
            return true
        } else {
            return false
        }
    }

    PlayNewAnimation(AnimationSequence: animation_sequence): void = {
        if {
            AnimationController := MaybeNpcAnimationController?
        } then {
            AnimationInstance := AnimationController.Play(AnimationSequence, ?PlayRate := PlayRate, ?BlendInTime := BlendInTime, ?BlendOutTime := BlendInTime, ?StartPositionSeconds := AnimationStartPositionSeconds)
            set MaybeCurAnimationInstance = option{AnimationInstance}
        }
    }

    PlayAndAwaitAnimation(AnimationSequence: animation_sequence)<suspends>: void = {
        if {
            AnimationController := MaybeNpcAnimationController?
        } then {
            AnimationInstance := AnimationController.Play(AnimationSequence, ?PlayRate := PlayRate, ?BlendInTime := BlendInTime, ?BlendOutTime := BlendInTime, ?StartPositionSeconds := AnimationStartPositionSeconds)
            set MaybeCurAnimationInstance = option{AnimationInstance}
            AnimationInstance.CompletedEvent.Await()
            set MaybeCurAnimationInstance = false
        }
    }

    SetNpcState(NewState: npc_state): void = {
        if {
            NpcManager := MaybeNpcManager?
            Navigatable := MaybeNpcNavigatable?
        } then {
            set NpcState = NewState

            case(NewState) {
                npc_state.Idle =>
                    # Print("SetNpcState: NewState: Idle")
                    SetAnimState(npc_anim_state.Idle)
                    # Start idle routine
                    MyRoutine := SpawnRoutine(IdleTick)
                    set RoutineIdleTick = option{MyRoutine}
                    # Add the NPC as an idle NPC
                    set NpcManager.NumOfIdleNpcs = NpcManager.NumOfIdleNpcs + 1
                npc_state.Talking =>
                    # Print("SetNpcState: NewState: Talking")
                    # Cancel idle routine
                    if(Routine := RoutineIdleTick?) {
                        Routine.Cancel()
                    }
                    set RoutineIdleTick = false
                    # Remove the NPC as an idle NPC
                    set NpcManager.NumOfIdleNpcs = NpcManager.NumOfIdleNpcs - 1
                    # Stop previous movement
                    StopNavigationMovement()
                npc_state.Recruiting =>
                    # Print("SetNpcState: NewState: Recruiting")
                npc_state.Chasing =>
                    # Print("SetNpcState: NewState: Chasing")
                    SetAnimState(npc_anim_state.Run)
                    # Start chasing routine
                    MyRoutine := SpawnRoutine(ChasingTick)
                    set RoutineChasingTick = option{MyRoutine}
                npc_state.Retreat =>
                    Print("SetNpcState: NewState: Retreat")
                    SetAnimState(npc_anim_state.Run)
                    # Cancel chasing routine
                    if(Routine := RoutineChasingTick?) {
                        Routine.Cancel()
                    }
                    set RoutineChasingTick = false
                    # Despawn NPC
                    DespawnNpc()
                    <# Move back to spawn
                    FullyHealNpc()
                    set AgentsThatDamagedNpc = array {}
                    Task := spawn{MoveBackToSpawn()}
                    #>
            }
        }
    }

    SetAnimState(NewState: npc_anim_state): void = {
        if(NewState <> NpcAnimState) {
            if {
                AnimationController := MaybeNpcAnimationController?
            } then {
                # Stop previous animation
                if(CurAnimation := MaybeCurAnimationInstance?) {
                    CurAnimation.Stop()
                }

                # Start new animation
                if(NewState = 1) {
                    # Idle animation is default in the blueprint, so we don't have to play it
                } else if(NewState = 2) {
                    # PlayNewAnimation(Animations.Fortnite_M_Avg_Player_Skeleton_Sequence)
                } else if(NewState = 3) {
                    # PlayNewAnimation(Animations.Fortnite_M_Avg_Player_Skeleton_Sequence)
                }

                # Set variable
                set NpcAnimState = NewState
            }
        }
    }

    <#
    TryGetClosestFortCharacter()<decides><transacts>: fort_character = {
        var ClosestFortCharacter : ?fort_character = false
        var ClosestFortCharacterDistance: float = MaxChaseDistance * 2
        
        MonsterPosition := MonsterProp.GetTransform().Translation

        for (Player : GetPlayspace().GetPlayers()):
            if:
                Target := Player.GetFortCharacter[],
                Target.IsActive[],
                TargetPosition:= Target.GetTransform().Translation,
                TargetToStartDistance:=Distance(TargetPosition, StartTransform.Translation),
                TargetToStartDistance < MaxChaseDistance
            then:
                DistanceToTarget := Distance(TargetPosition, MonsterPosition)
                if (DistanceToTarget<ClosestFortCharacterDistance):
                    set ClosestFortCharacterDistance = DistanceToTarget
                    set ClosestFortCharacter = option{Target}

        return ClosestFortCharacter?
    }
    #>
}
