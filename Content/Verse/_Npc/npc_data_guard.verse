using { /Fortnite.com/AI }
using { /Fortnite.com/AI/movement_types }
using { /Fortnite.com/Game }
using { /Fortnite.com/Animation }
using { /Fortnite.com/Animation/PlayAnimation }
using { /Fortnite.com/FortPlayerUtilities }
using { /Fortnite.com/Devices }
using { /Fortnite.com/Devices/CreativeAnimation }
using { /Fortnite.com/Characters }
using { /Verse.org/Simulation }
using { /Verse.org/Simulation/Tags }
using { /Verse.org/Random }
using { /Fortnite.com/Teams }
using { /Verse.org/Assets }
using { /Verse.org/Colors }
using { /Verse.org/Concurrency }
using { /UnrealEngine.com/Temporary/Diagnostics }
using { /UnrealEngine.com/Temporary/SpatialMath }

npc_data_guard<public> := class() {
    # Runtime variables
    var MaybeNpcManager: ?npc_spawner_manager_device = false
    var MaybePolicePatrolManager: ?police_patrol_device = false
    var MaybeNpcAgent: ?agent = false
    var MaybeNpcFortChar: ?fort_character = false
    var MaybeNpcNavigatable: ?navigatable = false
    var MaybeNpcFocusable: ?focus_interface = false
    var MaybeNpcAnimationController: ?play_animation_controller = false
    var MaybeCurAnimationInstance: ?play_animation_instance = false

    var NpcRarity<public>: int = -1
    var GuardSpawnerIndex<public>: int = -1
    var MaybeLastAgentThatDamagedNpc<public>: ?agent = false
    var TimestampLastNpcDamage<public>: float = 0.0

    var RoutineDespawnGuard: ?routine = false

    # This function runs when the NPC is spawned in the world and ready to follow a behavior.
    InitGuardNpc<public>(Agent: agent, ParmGuardSpawnerIndex: int, ParmNpcRarity: int): void = {
        # Print("npc_data - InitNpc")
        Utility := GetUtility()
        if {
            # Get the fort_character interface of the agent to access Fortnite character-specific behaviors, events, functions, and interfaces.
            FortChar := Agent.GetFortCharacter[]
            # Get the navigatable interface of the character to set specific targets for the character to travel to.
            Navigatable := FortChar.GetNavigatable[]
            # Get the focus_interface of the character to set specific targets to focus on after traveling to them.
            Focusable := FortChar.GetFocusInterface[]
            # Get Anim Controller
            AnimationController := FortChar.GetPlayAnimationController[]
        } then {
            # Print("InitNpc: NpcId: {ParmNpcId}")
            GameManager := GetGameManager()

            # Init variables
            set MaybeNpcAgent = option{Agent}
            set MaybeNpcFortChar = option{FortChar}
            set MaybeNpcNavigatable = option{Navigatable}
            set MaybeNpcFocusable = option{Focusable}
            set MaybeNpcAnimationController = option{AnimationController}

            set NpcRarity = ParmNpcRarity
            set GuardSpawnerIndex = ParmGuardSpawnerIndex

            # Listen to events
            FortChar.DamagedEvent().Subscribe(OnNpcDamaged)
            FortChar.EliminatedEvent().Subscribe(OnNpcEliminated)
        } else {
            Print("ERR: NPC Guard Script on NPC Setup")
        }
    }

    SetNpcSpawnerManager(NpcManager: npc_spawner_manager_device): void = {
        set MaybeNpcManager = option{NpcManager}
        # Despawn scammed NPCs after duration
        Task := spawn{DespawnScammedGuardSuspends()}
    }

    SetPolicePatrolManager<public>(PolicePatrolManager: police_patrol_device): void = {
        set MaybePolicePatrolManager = option{PolicePatrolManager}
    }

    # This function must be called when the NPC is despawned or eliminated from the world.
    HandleDestroy():void = {
        
    }

    OnNpcDamaged(DamageResult: damage_result): void = {
        # Print("OnNpcDamaged")
        if {
            Instigator := DamageResult.Instigator?
            DamagingPlayer := Instigator.GetInstigatorAgent[]
            DamagingPlayerData := DamagingPlayer.GetPlayerData[]
            # DamagingPlayerFortChar := DamagingPlayer.GetFortCharacter[]
            NpcFortChar := MaybeNpcFortChar?
        } then {
            ## Print("OnNpcDamaged - Damaged by player")
            if {
                NpcFortChar.IsVulnerable[]
            } then {
                # Damage was legit
                set MaybeLastAgentThatDamagedNpc = option{DamagingPlayer}
                set TimestampLastNpcDamage = GetSimulationElapsedTime()
            }
        }
    }

    OnNpcEliminated(EliminationResult: elimination_result): void = {
        if {
            NpcAgent := MaybeNpcAgent?
        } then {
            if {
                NpcManager := MaybeNpcManager?
            } then {
                # Call NPC Manager
                NpcManager.HandleNpcGuardEliminated(NpcAgent, GuardSpawnerIndex, EliminationResult)
            } else {
                if {
                    PolicePatrolManager := MaybePolicePatrolManager?
                } then {
                    PolicePatrolManager.HandleNpcGuardEliminated(NpcAgent, GuardSpawnerIndex, EliminationResult)
                }
            }
        } else {
            Print("ERR: OnNpcEliminated: Not initialized correctly!")
        }
    }

    DespawnScammedGuardSuspends()<suspends>: void = {
        GameConfig := GetGameConfig()
        Sleep(GameConfig.ScammedGuardDespawnDuration)
        if {
            NpcFortChar := MaybeNpcFortChar?
            NpcFortChar.IsActive[]
        } then {
            DespawnNpc()
        }
    }

    DespawnNpc(): void = {
        # Print("DespawnNpc")
        # Since there is no despawn function we have to deal damage
        if {
            NpcFortChar := MaybeNpcFortChar?
            NpcFortChar.IsActive[]
        } then {
            # Print("DespawnNpc: NPC Id: {NpcId}")
            NpcFortChar.SetVulnerability(true)
            NpcFortChar.Damage(999999999999.0)
        }
        # Also calling destroy manually to prevent endless loop where NpcElimEvent isn't called
        HandleDestroy()
    }
}
