using { /Fortnite.com/AI }
using { /Fortnite.com/Game }
using { /Fortnite.com/Animation }
using { /Fortnite.com/Animation/PlayAnimation }
using { /Fortnite.com/FortPlayerUtilities }
using { /Fortnite.com/Devices }
using { /Fortnite.com/Devices/CreativeAnimation }
using { /Fortnite.com/Characters }
using { /Verse.org/Simulation }
using { /Verse.org/Simulation/Tags }
using { /Verse.org/Random }
using { /Fortnite.com/Teams }
using { /Verse.org/Assets }
using { /Verse.org/Concurrency }
using { /UnrealEngine.com/Temporary/Diagnostics }
using { /UnrealEngine.com/Temporary/SpatialMath }

# A Verse-authored NPC Behavior that can be used within an NPC Character Definition or a NPC Spawner device's NPC Behavior Script Override.
npc_enemy_behavior<public> := class(npc_behavior) {
    # This function runs when the NPC is spawned in the world and ready to follow a behavior.
    OnBegin<override>()<suspends>:void = {

    }

    # This function runs when the NPC is despawned or eliminated from the world.
    OnEnd<override>():void = {
        
    }
}
