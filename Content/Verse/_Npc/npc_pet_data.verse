using { /Fortnite.com/AI }
using { /Fortnite.com/AI/movement_types }
using { /Fortnite.com/Game }
using { /Fortnite.com/Animation }
using { /Fortnite.com/Animation/PlayAnimation }
using { /Fortnite.com/FortPlayerUtilities }
using { /Fortnite.com/Devices }
using { /Fortnite.com/Devices/CreativeAnimation }
using { /Fortnite.com/Characters }
using { /Verse.org/Simulation }
using { /Verse.org/Simulation/Tags }
using { /Verse.org/Random }
using { /Fortnite.com/Teams }
using { /Verse.org/Assets }
using { /Verse.org/Concurrency }
using { /UnrealEngine.com/Temporary/Diagnostics }
using { /UnrealEngine.com/Temporary/SpatialMath }

npc_pet_data<public> := class() {
    # Constants
    # The speed to play the animation at.
    PlayRate : float = 1.0
    # How long to blend the previous animation into the current one.
    BlendInTime : float = 0.25
    # How long to blend the current animation into the next one.
    BlendOutTime : float = 0.25
    AnimationStartPositionSeconds : float = 0.0

    TeleportToOwnerDistance: float = 3000.0

    # Runtime variables
    var MaybePetManager: ?_Tycoon.tycoon_manager_device = false
    var MaybeNpcAgent: ?agent = false
    var MaybeOwnerAgent: ?agent = false
    var MaybeNpcFortChar: ?fort_character = false
    var MaybeNpcNavigatable: ?navigatable = false
    var MaybeNpcFocusable: ?focus_interface = false
    var MaybeNpcAnimationController: ?play_animation_controller = false
    var MaybeCurAnimationInstance: ?play_animation_instance = false

    var RoutineServerTick: ?routine = false
    var RoutineAudioPlayerTick: ?routine = false

    var NpcState: npc_state = npc_state.Idle
    var NpcAnimState: npc_anim_state = npc_anim_state.Idle

    # This function runs when the NPC is spawned in the world and ready to follow a behavior.
    InitPetNpc<public>(NpcAgent: agent, PetManager: _Tycoon.tycoon_manager_device, OwnerAgent: agent): void = {
        Print("npc_pet_data - InitNpc")
        Utility := GetUtility()
        if {
            # Get the fort_character interface of the agent to access Fortnite character-specific behaviors, events, functions, and interfaces.
            FortChar := NpcAgent.GetFortCharacter[]
            # Get the navigatable interface of the character to set specific targets for the character to travel to.
            Navigatable := FortChar.GetNavigatable[]
            # Get the focus_interface of the character to set specific targets to focus on after traveling to them.
            Focusable := FortChar.GetFocusInterface[]
            # Get Anim Controller
            AnimationController := FortChar.GetPlayAnimationController[]
        } then {
            Print("npc_pet_data - InitNpc 222")
            GameManager := GetGameManager()

            # Init variables
            set MaybePetManager = option{PetManager}
            set MaybeNpcAgent = option{NpcAgent}
            set MaybeOwnerAgent = option{OwnerAgent}
            set MaybeNpcFortChar = option{FortChar}
            set MaybeNpcNavigatable = option{Navigatable}
            set MaybeNpcFocusable = option{Focusable}
            set MaybeNpcAnimationController = option{AnimationController}

            # SetNpcState(npc_state.Idle)

            # Listen to events
            # FortChar.DamagedEvent().Subscribe(OnNpcDamaged)
            # FortChar.EliminatedEvent().Subscribe(OnNpcEliminated)

            # Instantly teleport to owner on spawn
            if {
                TargetAgent := MaybeOwnerAgent?
                TargetFortChar := TargetAgent.GetFortCharacter[]
                TargetFortChar.IsActive[]
                NpcFortChar := MaybeNpcFortChar?
                NpcFortChar.IsActive[]
            } then {
                if(NpcFortChar.TeleportTo[TargetFortChar.GetTransform().Translation, TargetFortChar.GetTransform().Rotation]) {}
            }  

            MyRoutine := SpawnRoutine(ServerTick)
            set RoutineServerTick = option{MyRoutine}

            MyRoutineAudioPlayer := SpawnRoutine(AudioPlayerTick)
            set RoutineAudioPlayerTick = option{MyRoutineAudioPlayer}
        } else {
            Print("ERR: NPC Pet Script on NPC Setup")
        }
    }

    # This function must be called when the NPC is despawned or eliminated from the world.
    HandleDestroy<public>():void = {
        Print("npc pet data - HandleDestroy")
        # Clean up routines
        if(Routine := RoutineServerTick?) {
            Routine.Cancel()
        }
        if(Routine := RoutineAudioPlayerTick?) {
            Routine.Cancel()
        }
    }

    <#
    OnNpcDamaged(DamageResult: damage_result): void = {
        Print("OnPetNpcDamaged")
        # OnNpcEliminated doesn't trigger for custom NPCs
        if {
            FortChar := MaybeNpcFortChar?
            FortChar.GetHealth() <= 0.0
        } then {
            # This doesn't work for some reason
            HandleNpcPetEliminated()
        }
    }

    HandleNpcPetEliminated(): void = {
        Print("HandleNpcPetEliminated")
        if {
            NpcAgent := MaybeNpcAgent?
            PetManager := MaybePetManager?
        } then {
            # Call Pet Manager
            PetManager.HandlePetNpcEliminated(NpcAgent)
        }
    }
    #>

    ServerTick()<suspends>: void = {
        Utility := GetUtility()
        loop {
            if {
                Navigatable := MaybeNpcNavigatable?
                TargetAgent := MaybeOwnerAgent?
                TargetFortChar := TargetAgent.GetFortCharacter[]
                TargetFortChar.IsActive[]
                NpcFortChar := MaybeNpcFortChar?
            } then {
                # Move to the target (suspends)
                NavigationResult := Navigatable.NavigateTo(Target := MakeNavigationTarget(TargetAgent), ?MovementType := Running, ?ReachRadius := 200.0)
                if(NavigationResult = navigation_result.Reached) {
                    # Print("Reached player")
                } else {
                    Print("NPC Pet - Couldn't reach player")
                    if {
                        TargetAgent2 := MaybeOwnerAgent?
                        TargetFortChar2 := TargetAgent2.GetFortCharacter[]
                        TargetFortChar2.IsActive[]
                        NpcFortChar2 := MaybeNpcFortChar?
                        NpcFortChar2.IsActive[]
                    } then {
                        DistanceToOwner := Distance(NpcFortChar2.GetTransform().Translation, TargetFortChar2.GetTransform().Translation)
                        if(DistanceToOwner >= TeleportToOwnerDistance) {
                            Print("NPC Pet - Too far away, teleporting to player")
                            if(NpcFortChar2.TeleportTo[TargetFortChar2.GetTransform().Translation, TargetFortChar2.GetTransform().Rotation]) {}
                        }
                    }
                }
            }

            Sleep(0.0)
        }
    }

    AudioPlayerTick()<suspends>: void = {
        Utility := GetUtility()
        loop {
            if {
                TycoonManager := MaybePetManager?
                TycoonManager.PetAudioPlayerProp.IsValid[]
                NpcFortChar := MaybeNpcFortChar?
                NpcFortChar.IsActive[]
            } then {
                if(TycoonManager.PetAudioPlayerProp.TeleportTo[NpcFortChar.GetTransform().Translation, NpcFortChar.GetTransform().Rotation]) {}
            }
            
            Sleep(0.1)
        }
    }

    DespawnNpc(): void = {
        # Since there is no despawn function we have to deal damage
        if {
            NpcFortChar := MaybeNpcFortChar?
            NpcFortChar.IsActive[]
        } then {
            NpcFortChar.Damage(999999999999.0)
        }
    }

    PlayNewAnimation(AnimationSequence: animation_sequence): void = {
        if {
            AnimationController := MaybeNpcAnimationController?
        } then {
            AnimationInstance := AnimationController.Play(AnimationSequence, ?PlayRate := PlayRate, ?BlendInTime := BlendInTime, ?BlendOutTime := BlendInTime, ?StartPositionSeconds := AnimationStartPositionSeconds)
            set MaybeCurAnimationInstance = option{AnimationInstance}
        }
    }

    PlayAndAwaitAnimation(AnimationSequence: animation_sequence)<suspends>: void = {
        if {
            AnimationController := MaybeNpcAnimationController?
        } then {
            AnimationInstance := AnimationController.Play(AnimationSequence, ?PlayRate := PlayRate, ?BlendInTime := BlendInTime, ?BlendOutTime := BlendInTime, ?StartPositionSeconds := AnimationStartPositionSeconds)
            set MaybeCurAnimationInstance = option{AnimationInstance}
            AnimationInstance.CompletedEvent.Await()
            set MaybeCurAnimationInstance = false
        }
    }

    SetAnimState(NewState: npc_anim_state): void = {
        if(NewState <> NpcAnimState) {
            if {
                AnimationController := MaybeNpcAnimationController?
            } then {
                # Stop previous animation
                if(CurAnimation := MaybeCurAnimationInstance?) {
                    CurAnimation.Stop()
                }

                # Set variable
                set NpcAnimState = NewState
            }
        }
    }
}
