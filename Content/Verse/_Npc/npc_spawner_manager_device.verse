using { /Fortnite.com/Game }
using { /Fortnite.com/Devices }
using { /Fortnite.com/Devices/CreativeAnimation }
using { /Fortnite.com/Characters }
using { /Verse.org/Simulation }
using { /Verse.org/Simulation/Tags }
using { /Verse.org/Random }
using { /Verse.org/Concurrency }
using { /Verse.org/Colors }
using { /UnrealEngine.com/Temporary/Diagnostics }
using { /UnrealEngine.com/Temporary/SpatialMath }

# A Verse-authored NPC Behavior that can be used within an NPC Character Definition or a NPC Spawner device's NPC Behavior Script Override.
npc_spawner_manager_device<public> := class(creative_device) {
    # We need a different NPC Spawner for each skin
    @editable
    NpcSpawnerList: []npc_spawner_device = array {}

    @editable
    GuardSpawnerList: []npc_spawner_device = array {}

    # The NPC names, with spawner id as index
    @editable
    NpcNameList: []string = array {}

    @editable
    SpawnCooldown: float = 1.0

    @editable
    MaxNumOfTotalNpcs: int = 10

    @editable
    NpcMaxHp: float = 100.0

    @editable
    NpcAttackDamage: float = 25.0

    @editable
    NpcWaypointList: []npc_waypoint_device = array {}

    <#
    @editable
    NpcWaypointList0: []npc_waypoint_device = array {}

    @editable
    NpcWaypointList1: []npc_waypoint_device = array {}
    #>

    @editable
    NpcStarPropOffset: vector3 = vector3{X := 0.0, Y := 0.0, Z := 0.0}

    @editable
    var NpcStarPropListLevel1: []creative_prop = array {}

    @editable
    var NpcStarPropListLevel2: []creative_prop = array {}

    @editable
    var NpcStarPropListLevel3: []creative_prop = array {}

    @editable
    var NpcStarPropListLevel4: []creative_prop = array {}

    @editable
    var NpcStarPropListLevel5: []creative_prop = array {}

    @editable
    var NpcStarPropListLevel6: []creative_prop = array {}

    @editable
    var NpcStarPropListLevel7: []creative_prop = array {}

    @editable
    var NpcStarPropListLevel8: []creative_prop = array {}

    @editable
    var NpcStarPropListLevel9: []creative_prop = array {}

    @editable
    var NpcStarPropListLevel10: []creative_prop = array {}

    @editable
    NpcStarLevelRequirementList: []int = array {}

    @editable
    NpcRarityChanceList: []float = array {}

    @editable
    var NpcRecruitButtonPoolList: []button_device = array {}

    @editable
    PlayerLevelTooLowForRecruit: hud_message_device = hud_message_device {}

    @editable
    HudRecruitSuccess: hud_message_device = hud_message_device {}

    @editable
    HudRecruitFailed: hud_message_device = hud_message_device {}

    @editable
    AudioRecruitFailed: audio_player_device = audio_player_device {}

    <# Crime per kill for each npc rarity
    @editable
    CrimePerNpcKill: []float = array {
        20.0,
        20.0,
        30.0,
        40.0,
        50.0,
        60.0
    }
    #>

    <#
    @editable
    MoneyFactorPerNpcKill: float = 1.0

    @editable
    FansFactorPerNpcKill: float = 1.0
    #>

    @editable
    ItemGranterAfterScam: item_granter_device = item_granter_device {}

    <#
    @editable
    XpPerNpcKill: []int = array {
        250,
        500,
        1000,
        1500,
        2000,
        2500
    }

    @editable
    MoneyPerNpcKill: []int = array {
        35,
        150,
        500,
        1000,
        15000,
        20000
    }

    @editable
    FansPerNpcKill: []int = array {
        1,
        500,
        1000,
        1500,
        20000,
        25000
    }

    @editable
    DiamondsPerNpcKill: []int = array {
        1,
        25,
        500,
        1000,
        15000,
        20000
    }
    #>

    # Constants
    NpcRarityHater: int = -1

    # Runtime variables
    var RoutineServerTick: ?routine = false
    var NpcStarPropLists: [][]creative_prop = array {}
    var ActiveNpcList<public>: []agent = array {}

    var IsActive: logic = false
    var NumOfPlayersInPatrolZones: int = 0
    var NumOfPlayersInInnerZones: int = 0

    var NumOfTotalNpcs: int = 0
    # var NumOfHaters: int = 0
    var NumOfIdleNpcs: int = 0
    var TimestampLastNpcElim: float = 0.0
    # NextRarityToAssign must start at 0, since NPCs already spawn before this scripts OnBegin is running
    # var NextRarityToAssign: int = -1
    # var NextRarityToAssign: int = 0

    var NpcSpawnExitWaypointList: []npc_waypoint_device = array {}

    var NpcId: int = 0
    var HackNextGuardRarityToAssign: int = 0
    var PrevHighestPlayerLevel: int = 0
    var RemainingNpcsToUpgrade: int = 0

    var ActiveNpcPlayer0Rank: int = 0
    var ActiveNpcPlayer1Rank: int = 0
    var ActiveNpcPlayer2Rank: int = 0

    # ArrayIndex is the spawner index for the NPC type
    var LastScamPosList: [][]vector3 = array {}

    OnBegin<override>()<suspends>:void = {
        for(SpawnerIndex -> NpcSpawner: NpcSpawnerList) {
            # Handle spawn event
            EventHandler := event_handler_agent(int) {
                A := SpawnerIndex
                CallbackFunc := OnNpcSpawned
            }
    
            EventSub := NpcSpawner.SpawnedEvent.Subscribe(EventHandler.EventSubscription)

            # Handle elimination event (can't be used since it doesn't return the killer agent)
            # NpcSpawner.EliminatedEvent.Subscribe(OnNpcEliminated)

            # Disable + enable to fix bug
            # NpcSpawner.Disable()
        }

        for(GuardSpawnerIndex -> GuardSpawner: GuardSpawnerList) {
            # Handle spawn event
            EventHandler := event_handler_agent(int) {
                A := GuardSpawnerIndex
                CallbackFunc := OnGuardSpawned
            }
    
            EventSub := GuardSpawner.SpawnedEvent.Subscribe(EventHandler.EventSubscription)

            # Init LastScamPosList for this spawner
            set LastScamPosList = LastScamPosList + array {array {}}
        }

        # Init waypoints IDs
        for(WaypointIndex -> Waypoint: NpcWaypointList) {
            set Waypoint.WaypointId = WaypointIndex
            if(Waypoint.IsSpawnOrExitPoint = true) {
                set NpcSpawnExitWaypointList = NpcSpawnExitWaypointList + array {Waypoint}
            }
        }

        <#
        for(WaypointIndex -> Waypoint: NpcWaypointList0) {
            set Waypoint.WaypointId = WaypointIndex
        }

        for(WaypointIndex -> Waypoint: NpcWaypointList1) {
            set Waypoint.WaypointId = WaypointIndex
        }
        #>

        set NpcStarPropLists = NpcStarPropLists + array {
            NpcStarPropListLevel1,
            NpcStarPropListLevel2,
            NpcStarPropListLevel3,
            NpcStarPropListLevel4,
            NpcStarPropListLevel5,
            NpcStarPropListLevel6,
            NpcStarPropListLevel7,
            NpcStarPropListLevel8,
            NpcStarPropListLevel9,
            NpcStarPropListLevel10
        }

        <# We don't have to manually activate spawners anymore. That would only lead to the bugged auto-spawn
        Sleep(5.0)
        for(SpawnerIndex -> NpcSpawner: NpcSpawnerList) {
            NpcSpawner.Enable()
        }
        for(GuardSpawnerIndex -> GuardSpawner: GuardSpawnerList) {
            GuardSpawner.Enable()
        }
        #>
        Sleep(5.0)
        ActivateSpawnerManager()
    }

    OnEnd<override>(): void = {
        # Despawn NPCs to prevent bug where they keep standing still in Edit Mode
        for(SpawnerIndex -> NpcSpawner: NpcSpawnerList) {
            NpcSpawner.DespawnAll(false)
        }
    }

    ActivateSpawnerManager(): void = {
        set IsActive = true
        set TimestampLastNpcElim = GetSimulationElapsedTime()

        MyRoutine := SpawnRoutine(ServerTick)
        set RoutineServerTick = option{MyRoutine}
    }

    DeactivateSpawners(): void = {
        set IsActive = false
        for(NpcSpawner: NpcSpawnerList) {
            NpcSpawner.DespawnAll(false)
        }

        # Clean up routines
        if(Routine := RoutineServerTick?) {
            Routine.Cancel()
        }
    }

    <#
    SpawnAllNpcs(): void = {
        Print("SpawnAllNpcs")
        loop {
            if(NumOfTotalNpcs < MaxNumOfTotalNpcs) {
                SpawnNpc()
            } else {
                break
            }
        }
    }
    #>

    SpawnNpc(): void = {
        # Print("SpawnNpc: NumOfTotalNpcs: {NumOfTotalNpcs}")
        Utility := GetUtility()
        
        # Pick a random NPC spawner
        RandInt := GetRandomInt(0, NpcSpawnerList.Length - 1)
        if(NpcSpawner := NpcSpawnerList[RandInt]) {
            NpcSpawner.Spawn()
            # set NumOfTotalNpcs = NumOfTotalNpcs + 1
        } else {
            Print("ERR: SpawnNpc: Found no NPC Spawner for Index: {RandInt}", ?Color := NamedColors.Red)
        }
    }

    OnNpcSpawned(NpcAgent: agent, SpawnerIndex: int): void = {
        # Print("OnNpcSpawned")
        GameConfig := GetGameConfig()
        Utility := GetUtility()
        set NumOfTotalNpcs = NumOfTotalNpcs + 1
        set ActiveNpcList = ActiveNpcList + array {NpcAgent}
        var NpcName: string = "Jonesy"
        if {
            TempName := NpcNameList[SpawnerIndex]
        } then {
            set NpcName = TempName
        }

        var NextRarityToAssign: int = 0

        # Check if we have to reserve a rarity for a player
        var ReservedForOwnerPlayerId: int = -1
        for(PlayerId -> Player: GetPlayspace().GetPlayers()) {
            if {
                PlayerData := Player.GetPlayerData[]
            } then {
                if {
                    PlayerId = 0
                    ActiveNpcPlayer0Rank < GameConfig.NpcsReservedForEachPlayerRank
                } then {
                    set ActiveNpcPlayer0Rank = ActiveNpcPlayer0Rank + 1
                    set ReservedForOwnerPlayerId = PlayerId
                    set NextRarityToAssign = Max(PlayerData.TotalScamLevel - 1, 0)
                }
                if {
                    PlayerId = 1
                    ActiveNpcPlayer1Rank < GameConfig.NpcsReservedForEachPlayerRank
                } then {
                    set ActiveNpcPlayer1Rank = ActiveNpcPlayer1Rank + 1
                    set ReservedForOwnerPlayerId = PlayerId
                    set NextRarityToAssign = Max(PlayerData.TotalScamLevel - 1, 0)
                }
                if {
                    PlayerId = 2
                    ActiveNpcPlayer2Rank < GameConfig.NpcsReservedForEachPlayerRank
                } then {
                    set ActiveNpcPlayer2Rank = ActiveNpcPlayer2Rank + 1
                    set ReservedForOwnerPlayerId = PlayerId
                    set NextRarityToAssign = Max(PlayerData.TotalScamLevel - 1, 0)
                }
            }
        }

        if(ReservedForOwnerPlayerId = -1) {
            # No reserved rarity: Assign a random rarity for the new NPC
            var HighestPlayerLevel: int = 0
            # Find the highest star level player in the current game
            for(Player: GetPlayspace().GetPlayers()) {
                if {
                    PlayerData := Player.GetPlayerData[]
                } then {
                    if(PlayerData.TotalScamLevel > HighestPlayerLevel) {
                        set HighestPlayerLevel = PlayerData.TotalScamLevel
                    }
                }
            }
            # If a player in the lobby increased the HighestPlayerLevel: Instantly upgrade X NPCs to the highest rank
            if(PrevHighestPlayerLevel < HighestPlayerLevel) {
                set PrevHighestPlayerLevel = HighestPlayerLevel
                set RemainingNpcsToUpgrade = GameConfig.ScamLevelIncreasedNpcUpgradeAmount
            }
            # Only allow rarity spawns according to the highest player level
            var LimitedNpcRarityChanceList: []float = array {}
            for(ItemIndex -> NpcStarLevelRequirement: NpcStarLevelRequirementList) {
                if(HighestPlayerLevel >= NpcStarLevelRequirement) {
                    if(NpcRarityChance := NpcRarityChanceList[ItemIndex]) {
                        set LimitedNpcRarityChanceList = LimitedNpcRarityChanceList + array{NpcRarityChance}
                    } else {
                        Print("ERR: Found no NpcRarityChance for ItemIndex: {ItemIndex}")
                    }
                }
            }
            # Print("LimitedNpcRarityChanceList.Length: {LimitedNpcRarityChanceList.Length}")
            # Make sure we can spawn at least the lowest rarity
            if(LimitedNpcRarityChanceList.Length = 0) {
                if(NpcRarityChance := NpcRarityChanceList[0]) {
                    set LimitedNpcRarityChanceList = LimitedNpcRarityChanceList + array{NpcRarityChance}
                }
            }
            
            # Assign a random rarity
            set NextRarityToAssign = Utility.GetRandomItemWeightedRollChance(LimitedNpcRarityChanceList)
            # Print("OnNpcSpawned: NpcId: {NpcId} - NpcRarity: {NextRarityToAssign}")
        }

        if {
            NpcStarPropList := NpcStarPropLists[NextRarityToAssign]
            NpcStarPropList.Length > 0
        } then {
            # Print("NpcStarPropList.Length: {NpcStarPropList.Length}")
        } else {
            Print("WARN: OnNpcSpawned: Found no unused StarProp for NpcRarity: {NextRarityToAssign}. Defaulting to NpcRarity 0", ?Color := NamedColors.Yellow)
        }

        NpcData := npc_data {}
        # Assign a random start waypoint
        RandInt := GetRandomInt(0, NpcSpawnExitWaypointList.Length - 1)
        if(StartWaypoint := NpcSpawnExitWaypointList[RandInt]) {
            NpcData.AssignWaypointList(StartWaypoint, NpcWaypointList)
        }
        var MaybeRecruitButton: ?button_device = false
        set MaybeRecruitButton = GetUnusedRecruitButton()
        # Call Npc Init
        NpcData.InitNpc(NpcAgent, Self, MaybeRecruitButton, NpcMaxHp, NpcAttackDamage, NpcName, SpawnerIndex, NpcId, NextRarityToAssign, ReservedForOwnerPlayerId)
        NpcAgent.SetNpcData(NpcData)

        set NpcId = NpcId + 1
    }

    OnGuardSpawned(NpcAgent: agent, GuardSpawnerIndex: int): void = {
        # Print("OnGuardSpawned")
        if {
            NpcFortChar := NpcAgent.GetFortCharacter[]
            NpcFortChar.IsActive[]
        } then {
            if {
                ScamPosList := LastScamPosList[GuardSpawnerIndex]
                # Get the latest scam pos for this spawner
                LastScamPos := ScamPosList[ScamPosList.Length - 1]
            } then {
                if(NpcFortChar.TeleportTo[LastScamPos, rotation{}]) {}
                # Remove the last position, after we have used it
                if {
                    Temp := LastScamPosList[GuardSpawnerIndex].Slice[0, ScamPosList.Length - 1]
                    set LastScamPosList[GuardSpawnerIndex] = Temp
                }
                # Assign NpcDataGuard for Elimination rewards
                NpcDataGuard := npc_data_guard {}
                NpcDataGuard.InitGuardNpc(NpcAgent, GuardSpawnerIndex, HackNextGuardRarityToAssign)
                NpcDataGuard.SetNpcSpawnerManager(Self)
                NpcAgent.SetNpcDataGuard(NpcDataGuard)
            } else {
                Print("ERR: OnGuardSpawned: Didn't find LastScamPos")
            }
        } else {
            Print("ERR: OnGuardSpawned: Didn't find NpcFortChar")
        }
    }

    CalculateXpForKill<public>(Agent: agent, NpcRarity: int): int = {
        GameConfig := GetGameConfig()
        var XpForKill: int = 0
        if {
            PlayerData := Agent.GetPlayerData[]
            XpPerKillNpcMin := GameConfig.XpPerKillNpcMin[NpcRarity]
            XpPerKillNpcMax := GameConfig.XpPerKillNpcMax[NpcRarity]
        } then {
            set XpForKill = GetRandomInt(XpPerKillNpcMin, XpPerKillNpcMax)
            # set XpForKill = GameConfig.BaseXpPerKillNpc + (PlayerData.CurLevel * 10)
        } else {
            Print("ERR: npc_spawner_manager: CalculateXpForKill")
        }
        return XpForKill
    }

    RewardKillerAgent(NpcAgent: agent, EliminationResult: elimination_result, NpcSpawnerId: int, NpcRarity: int, MaybeLastAgentThatDamagedNpc: ?agent, TimestampLastNpcDamage: float): void = {
        SaveManager := GetSaveManager()
        AccoladeManager := GetAccoladeManager()
        GameConfig := GetGameConfig()
        # Workaround for killing NPCs via script (dragon mount)
        var MaybeKillerAgent: ?agent = false
        if {
            KillerFortCharElimResult := EliminationResult.EliminatingCharacter?
            KillerAgentElimResult := KillerFortCharElimResult.GetAgent[]
            PlayerDataValidAgentTest := KillerAgentElimResult.GetPlayerData[]
        } then {
            # Print("HandleNpcEliminated - Found killer from ElimResult")
            set MaybeKillerAgent = option{KillerAgentElimResult}
        } else {
            if {
                KillerAgentNpcData := MaybeLastAgentThatDamagedNpc?
                PlayerDataValidAgentTest := KillerAgentNpcData.GetPlayerData[]
            } then {
                Timestamp := GetSimulationElapsedTime()
                TimeSinceLastDamage := Timestamp - TimestampLastNpcDamage
                if(TimeSinceLastDamage < 3.0) {
                    # Print("HandleNpcEliminated - Found killer from DamagedEvent")
                    set MaybeKillerAgent = option{KillerAgentNpcData}
                }
            }
        }
        if {
            KillerAgent := MaybeKillerAgent?
            PlayerData := KillerAgent.GetPlayerData[]
            Save := SaveManager.ProxySaves[KillerAgent]
            # CoinsPerKillIncreasePerUpgrade := GameConfig.PermanentUpgradeRewardPerLevelList[1]
            RewardRank := Max(PlayerData.TotalScamLevel - 1, 0)
            MoneyPerKillBaseRewardMin := GameConfig.MoneyPerKillNpcRankMin[Min(RewardRank, NpcRarity)]
            MoneyPerKillBaseRewardMax := GameConfig.MoneyPerKillNpcRankMax[Min(RewardRank, NpcRarity)]
            DiamondsPerKillBaseRewardMin := GameConfig.BaseKillpointsPerKillNpcRankMin[Min(RewardRank, NpcRarity)]
            DiamondsPerKillBaseRewardMax := GameConfig.BaseKillpointsPerKillNpcRankMax[Min(RewardRank, NpcRarity)]
        } then {
            # Print("HandleNpcEliminated - GOT KILLER PLAYER DATA")
            MoneyPerKillBaseReward := GetRandomInt(MoneyPerKillBaseRewardMin, MoneyPerKillBaseRewardMax)
            DiamondsPerKillBaseReward := GetRandomFloat(DiamondsPerKillBaseRewardMin, DiamondsPerKillBaseRewardMax)

            XpForKill := CalculateXpForKill(KillerAgent, NpcRarity)
            PlayerData.AddXp(XpForKill, true)
            PlayerData.AddMoney(MoneyPerKillBaseReward, true, false)
            var DiamondReward: float = DiamondsPerKillBaseReward
            if(DiamondsKillBonusFromUpgrade := GameConfig.PermanentUpgradeRewardKillpointsPerKill[Save.UpgradeCoinsPerKillLevel]) {
                set DiamondReward = DiamondReward + DiamondsKillBonusFromUpgrade
            }
            # set DiamondReward = DiamondReward + (Save.UpgradeCoinsPerKillLevel * CoinsPerKillIncreasePerUpgrade)
            PlayerData.AddDiamonds(DiamondReward, true)
            PlayerData.AddGamePassCoins(1)
            AccoladeManager.AwardAccolade(KillerAgent, 2)
            # Increase Achievement progress
            # Kill 30 NPCs Rank 2 (NpcRarity = 1, OR HIGHER)
            if(NpcRarity >= 1) {
                SaveManager.IncrementAchievement(KillerAgent, _Global.achievement_type.Kill, 0, 1)
            }
            # Kill 15 NPCs Rank 3 
            if(NpcRarity >= 2) {
                SaveManager.IncrementAchievement(KillerAgent, _Global.achievement_type.Kill, 1, 1)
            }
            # Kill 10 Spidey NPCs Rank 4
            if {
                NpcSpawnerId = 4
                NpcRarity >= 3
            } then {
                SaveManager.IncrementAchievement(KillerAgent, _Global.achievement_type.Kill, 2, 1)
            }
            # Kill 3 NPCs within 10 seconds
            SaveManager.IncrementAchievement(KillerAgent, _Global.achievement_type.Kill, 3, 1)
            # Starting the timer must be after IncrementAchievement
            if(Save.NumOfAchievementsCompletedKill = 3) {
                PlayerData.StartAchievementTimerScam(10.0)
            }
            # Kill 30 NPCs Rank 5
            if(NpcRarity >= 4) {
                SaveManager.IncrementAchievement(KillerAgent, _Global.achievement_type.Kill, 4, 1)
            }
            # Kill 25 Beast NPCs Rank 6
            if {
                NpcSpawnerId = 0
                NpcRarity >= 5
            } then {
                SaveManager.IncrementAchievement(KillerAgent, _Global.achievement_type.Kill, 6, 1)
            }
            # Kill 6 Thanos NPCs within 60 seconds
            if {
                NpcSpawnerId = 1
            } then {
                SaveManager.IncrementAchievement(KillerAgent, _Global.achievement_type.Kill, 7, 1)
                # Starting the timer must be after IncrementAchievement
                if(Save.NumOfAchievementsCompletedKill = 7) {
                    PlayerData.StartAchievementTimerScam(60.0)
                }
            }
            # Kill 25 Travis NPCs Rank 7
            if {
                NpcSpawnerId = 2
                NpcRarity >= 6
            } then {
                SaveManager.IncrementAchievement(KillerAgent, _Global.achievement_type.Kill, 8, 1)
            }
            # Kill 100 NPCs Rank 8
            if(NpcRarity >= 7) {
                SaveManager.IncrementAchievement(KillerAgent, _Global.achievement_type.Kill, 9, 1)
            }
            # Kill 150 NPCs in Rebirth 2 or higher
            if(PlayerData.NumOfRebirths >= 2) {
                SaveManager.IncrementAchievement(KillerAgent, _Global.achievement_type.Rebirth, 2, 1)
            }
            # Kill 250 Thanos NPC Rank 10 in Rebirth 7 or higher
            if {
                NpcSpawnerId = 1
                NpcRarity >= 9
                PlayerData.NumOfRebirths >= 7
            } then {
                SaveManager.IncrementAchievement(KillerAgent, _Global.achievement_type.Rebirth, 6, 1)
            }
            # Earn $ 250K by scamming/killing Beast NPCs 
            if {
                NpcSpawnerId = 0
            } then {
                SaveManager.IncrementAchievement(KillerAgent, _Global.achievement_type.Money, 3, MoneyPerKillBaseReward)
            }
            # Earn $ 1M by scamming/killing NPCs
            SaveManager.IncrementAchievement(KillerAgent, _Global.achievement_type.Money, 7, MoneyPerKillBaseReward)
            
            # Increase NPC kill counter
            PlayerData.IncreaseKillStreak(KillerAgent, 1)
            if {
                Quest := PlayerData.MaybeActiveQuest?
                KillNpcRarity := Quest.QuestGoalKillNpcRarity?
                # Npc rarity must be exact for Senix Quest
                NpcRarity = KillNpcRarity
                # NpcRarity >= KillNpcRarity
            } then {
                PlayerData.IncreaseQuestProgress()
            }

            if {
                NpcFortChar := NpcAgent.GetFortCharacter[]
                # We can't use NpcFortChar.IsActive[] here, since the NPC just died. But getting the position still works
                # NpcFortChar.IsActive[]
            } then {
                # Spawn credit card coin
                RandFloat := GetRandomFloat(0.0, 1.0)
                if(RandFloat <= GameConfig.HeroPowerupBonusCoinSpawnChance) {
                    Task := spawn{HandleSpawnedCoinSuspends(NpcFortChar.GetTransform())}
                }
            }
        } else {
            # Print("INFO: HandleNpcEliminated - Found no killer agent PLAYER DATA")
        }
    }

    # HandleNpcEliminated(InteractionResult: device_ai_interaction_result): void = {
    HandleNpcEliminated(NpcAgent: agent, EliminationResult: elimination_result): void = {
        ## Print("HandleNpcEliminated")
        SaveManager := GetSaveManager()
        GameConfig := GetGameConfig()
        if {
            NpcData := NpcAgent.GetNpcData[]
        } then {
            RewardKillerAgent(NpcAgent, EliminationResult, NpcData.NpcSpawnerId, NpcData.NpcRarity, NpcData.MaybeLastAgentThatDamagedNpc, NpcData.TimestampLastNpcDamage)
            # Check if this NPC was reserved and return to the pool
            if(NpcData.ReservedForOwnerPlayerId > -1) {
                if(NpcData.ReservedForOwnerPlayerId = 0) {
                    set ActiveNpcPlayer0Rank = Max(ActiveNpcPlayer0Rank - 1, 0)
                } else if(NpcData.ReservedForOwnerPlayerId = 1) {
                    set ActiveNpcPlayer1Rank = Max(ActiveNpcPlayer1Rank - 1, 0)
                } else if(NpcData.ReservedForOwnerPlayerId = 2) {
                    set ActiveNpcPlayer2Rank = Max(ActiveNpcPlayer2Rank - 1, 0)
                }
            }
            # Clean up NPC
            DestroyNpcScript(NpcAgent)
        }
        # Decrease NPC number
        set NumOfTotalNpcs = NumOfTotalNpcs - 1
        set TimestampLastNpcElim = GetSimulationElapsedTime()
    }

    HandleNpcGuardEliminated(NpcAgentGuard: agent, GuardSpawnerIndex: int, EliminationResult: elimination_result): void = {
        SaveManager := GetSaveManager()
        GameConfig := GetGameConfig()
        if {
            NpcDataGuard := NpcAgentGuard.GetNpcDataGuard[]
        } then {
            RewardKillerAgent(NpcAgentGuard, EliminationResult, NpcDataGuard.GuardSpawnerIndex, NpcDataGuard.NpcRarity, NpcDataGuard.MaybeLastAgentThatDamagedNpc, NpcDataGuard.TimestampLastNpcDamage)
            # Clean up NPC
            NpcDataGuard.HandleDestroy()
            NpcAgentGuard.DestroyNpcData()
        }
    }

    HandleSpawnedCoinSuspends(SpawnTrans: transform)<suspends>: void = {
        GameManager := GetGameManager()
        Utility := GetUtility()
        GameConfig := GetGameConfig()
        SpawnTuple := SpawnProp(GameManager.BpCreditCardCoin, SpawnTrans)
        if {
            SpawnedProp := SpawnTuple(0)?
        } then {
            race {
                Sleep(GameConfig.HeroPowerupBonusCoinDespawnDuration)
                loop {
                    MaybeClosestPlayer := Utility.GetClosestPlayerInRange(SpawnTrans, GameManager.GetPlayspace().GetPlayers(), ?MaxRange := GameConfig.HeroPowerupBonusCoinPickupRange)
                    if {
                        ClosestPlayer := MaybeClosestPlayer?
                        ClosestPlayerPlayerData := ClosestPlayer.GetPlayerData[]
                    } then {
                        ClosestPlayerPlayerData.OnHeroPowerupBonusCoinCollected(SpawnTrans)
                        break
                    }
                    Sleep(0.1)
                }
            }

            if(SpawnedProp.IsValid[]) {
                SpawnedProp.Dispose()
            }
        } else {
            Print("ERR: HandleSpawnedCoinSuspends")
        }
    }

    DestroyNpcScript(NpcAgent: agent): void = {
        if(NpcData := NpcAgent.GetNpcData[]) {
            NpcData.HandleDestroy()
        } else {
            Print("WARN: DestroyNpcScript: Found no NpcData")
        }

        if(Temp := ActiveNpcList.RemoveFirstElement[NpcAgent]) {
            set ActiveNpcList = Temp
        } else {
            Print("WARN: DestroyNpcScript: RemoveFirstElement: Didn't find NpcAgent to remove!")
        }

        NpcAgent.DestroyNpcData()
    }

    GetNextWaypoint(WaypointList: []npc_waypoint_device, CurWaypoint: npc_waypoint_device, StartWaypoint: npc_waypoint_device, WaypointVisitedList: []npc_waypoint_device): ?npc_waypoint_device = {
        # Print("GetNextWaypoint")
        # Remove every waypoint that we have already visited
        var PossibleNextWaypointList: []npc_waypoint_device = array {}
        for(WaypointInReach: CurWaypoint.WaypointsInReach) {
            if(NextWaypoint := NpcWaypointList[WaypointInReach]) {
                set PossibleNextWaypointList = PossibleNextWaypointList + array {NextWaypoint}
            }
        }
        # Print("PossibleNextWaypointList.Length: {PossibleNextWaypointList.Length}")
        for(WaypointVisited: WaypointVisitedList) {
            for(NextWaypointIndex -> PossibleNextWaypoint: PossibleNextWaypointList) {
                var RemoveWaypoint: logic = false
                # Check if this is the StartWaypoint
                if(PossibleNextWaypoint.WaypointId = StartWaypoint.WaypointId) {
                    # Remove the StartWaypoint if we didn't walk a minimum of nodes
                    if(WaypointVisitedList.Length < 3) {
                        set RemoveWaypoint = true
                    }
                } else if(PossibleNextWaypoint.WaypointId = WaypointVisited.WaypointId) {
                    set RemoveWaypoint = true
                }
                # Remove the waypoint if it is flagged for removal
                if(RemoveWaypoint = true) {
                    # Print("Removing a waypoint that we already visited")
                    if {
                        Temp := PossibleNextWaypointList.RemoveElement[NextWaypointIndex]
                        set PossibleNextWaypointList = Temp
                    }
                }
            }
        }
        # Print("X PossibleNextWaypointList.Length: {PossibleNextWaypointList.Length}")
        # Get a random connected waypoint
        RandInt := GetRandomInt(0, PossibleNextWaypointList.Length - 1)
        if {
            NextWaypoint := PossibleNextWaypointList[RandInt]
        } then {
            ## Print("NPC GetNextWaypoint: CurWaypointId: {CurWaypoint.WaypointId}, NextWaypointId: {NextWaypoint.WaypointId}")
            return option{NextWaypoint}  
        } else {
            ## Print("ERR: NPC GetNextWaypoint - Found no connected waypoint, that the NPC didn't already visit. Non-exit waypoints must have at least 2 connections")
            return false
        }
    }

    GetUnusedRecruitButton(): ?button_device = {
        var Device: ?button_device = false
        if {
            TempElement := NpcRecruitButtonPoolList[0]
            set Device = option{TempElement}
            # Remove the trigger from the pool
            TempNewList := NpcRecruitButtonPoolList.RemoveElement[0]
            set NpcRecruitButtonPoolList = TempNewList
        } then {
            # Do nothing here for now
        } else {
            Print("ERR GetUnusedRecruitButton - Found no unused device!")
        }
        return Device
    }

    ReturnNpcRecruitButton(Device: button_device): void = {
        Utility := GetUtility()
        # Move the button away so the player can't press it anymore
        if(Device.TeleportTo[Utility.HiddenPos, rotation{}]) {}
        # Restore lists
        set NpcRecruitButtonPoolList = NpcRecruitButtonPoolList + array{Device}
    }

    GetUnusedStarProp(RarityLevel: int): ?creative_prop = {
        var Device: ?creative_prop = false
        if {
            NpcStarPropList := NpcStarPropLists[RarityLevel]
        } then {
            if {
                set Device = option{NpcStarPropList[0]}
                TempNewList := NpcStarPropList.RemoveElement[0]
                set NpcStarPropLists[RarityLevel] = TempNewList
            }
        } else {
            Print("ERR: GetUnusedStarProp: Unknown RarityLevel: {RarityLevel}")
        }
        if(not Device?) {
            Print("ERR GetUnusedStarProp - Found no unused device for RarityLevel: {RarityLevel}")
        }
        return Device
    }

    ReturnNpcStarProp(StarProp: creative_prop, RarityLevel: int): void = {
        Utility := GetUtility()
        # Hide the prop
        if(StarProp.TeleportTo[Utility.HiddenPos, rotation{}]) {}
        # Restore lists
        if {
            NpcStarPropList := NpcStarPropLists[RarityLevel]
        } then {
            if(set NpcStarPropLists[RarityLevel] = NpcStarPropList + array{StarProp}) {}
        } else {
            Print("ERR: ReturnNpcStarProp: Unknown RarityLevel: {RarityLevel}")
        }
    }

    ServerTick()<suspends>: void = {
        loop {
            if(NumOfTotalNpcs < MaxNumOfTotalNpcs) {  
                SpawnNpc()
            }

            Timestamp := GetSimulationElapsedTime()
            TimeSinceLastNpcElim := Timestamp - TimestampLastNpcElim
            if(TimeSinceLastNpcElim >= 120.0) {
                # Reset all NPC spawners
                for(SpawnerIndex -> NpcSpawner: NpcSpawnerList) {
                    NpcSpawner.DespawnAll(false)
                    NpcSpawner.Reset()
                    NpcSpawner.Disable()
                }
                Sleep(3.0)
                for(SpawnerIndex -> NpcSpawner: NpcSpawnerList) {
                    NpcSpawner.Enable()
                }
                # Reset variables
                set NumOfTotalNpcs = 0
                set TimestampLastNpcElim = GetSimulationElapsedTime()
            }

            Sleep(SpawnCooldown)
        }
    }
    
    <# We can't use this event since it doesn't return NPC agent and Killer agent
    # It also swaps Source / Target depending on situation
    OnNpcEliminated(InteractionResult: device_ai_interaction_result): void = {
        Print("OnNpcEliminated")
        if(SourceAgent := InteractionResult.Source?) {
            if {
                NpcData := SourceAgent.GetNpcData[]
            } then {
                Print("GotSourceAgent NPC")
            } else {
                if {
                    PlayerData := SourceAgent.GetPlayerData[]
                } then {
                    Print("GotSourceAgent PLAYER")
                } else {
                    Print("Got FAKE SourceAgent")
                }
            }
        }
        if(TargetAgent := InteractionResult.Target?) {
            if {
                PlayerData := TargetAgent.GetPlayerData[]
            } then {
                Print("GotTargetAgent PLAYER")
            } else {
                if {
                    NpcData := TargetAgent.GetPlayerData[]
                } then {
                    Print("GotTargetAgent NPC")
                } else {
                    Print("Got FAKE TargetAgent")
                }
            }
        }
    }
    #>
}
