using { /Fortnite.com/Game }
using { /Fortnite.com/Devices }
using { /Fortnite.com/Devices/CreativeAnimation }
using { /Fortnite.com/Characters }
using { /Verse.org/Simulation }
using { /Verse.org/Simulation/Tags }
using { /Verse.org/Random }
using { /Verse.org/Assets }
using { /Verse.org/Concurrency }
using { /UnrealEngine.com/Temporary/Diagnostics }
using { /UnrealEngine.com/Temporary/SpatialMath }

npc_waypoint_device<public> := class(creative_device) {
    # Devices
    @editable
    IsSpawnOrExitPoint: logic = false

    @editable
    WaypointProp<public>: creative_prop = creative_prop {}

    @editable
    WaypointsInReach<public>: []int = array {}

    <# Never use verse device link, since it lags when you click the verse device!
    @editable
    WaypointsInReach: []npc_waypoint_device = array {}
    #>

    # How long the NPC waits when he visited this waypoint
    @editable
    IdleDuration: float = 0.0

    # Runtime variables
    # The waypoint Index in its NpcWaypointList
    var WaypointId: int = -1

    OnBegin<override>()<suspends>: void = {
        
    }
}