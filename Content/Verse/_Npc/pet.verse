using { /Fortnite.com/Game }
using { /Fortnite.com/Devices }
using { /Fortnite.com/Devices/CreativeAnimation }
using { /Fortnite.com/Characters }
using { /Verse.org/Simulation }
using { /Verse.org/Simulation/Tags }
using { /Verse.org/Random }
using { /Verse.org/Concurrency }
using { /UnrealEngine.com/Temporary/Diagnostics }
using { /UnrealEngine.com/Temporary/SpatialMath }
using { /Fortnite.com/UI }
using { /UnrealEngine.com/Temporary/UI }

pet_states := enum{Following, Chasing, Attacking}

pet<public> := class() {
    # Runtime variables
    var MaybeNpcSpawner: ?npc_spawner_device = false
    var PetMultiplierMoney<public>: float = 0.0
    var PetMultiplierFans<public>: float = 0.0
    var PetMultiplierDiamonds<public>: float = 0.0
    var PetMultiplierRecruitChance<public>: float = 0.0

    SpawnPet<public>(NpcSpawner: npc_spawner_device): void = {
        Print("SpawnPet")
        set MaybeNpcSpawner = option{NpcSpawner}
        NpcSpawner.Spawn()
    }

    DestroyPet<public>(): void = {
        Print("DestroyPet")
        if(NpcSpawner := MaybeNpcSpawner?) {
            NpcSpawner.DespawnAll(false)
        }
    }
}