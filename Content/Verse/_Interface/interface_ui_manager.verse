using { /Fortnite.com/Game }
using { /Fortnite.com/Devices }
using { /Fortnite.com/Devices/CreativeAnimation }
using { /Fortnite.com/Characters }
using { /Fortnite.com/AI }
using { /Verse.org/Simulation }
using { /Verse.org/Simulation/Tags }
using { /Verse.org/Random }
using { /Verse.org/Assets }
using { /Verse.org/Concurrency }
using { /UnrealEngine.com/Temporary/Diagnostics }
using { /UnrealEngine.com/Temporary/SpatialMath }

interface_ui_manager<public> := interface {
    OnUiButtonClick<public>(Agent: agent, ParmButtonId: int): void

    OnExitButtonClick<public>(Agent: agent): void
    
    OnMinigameSuccess<public>(Agent:agent): void

    OnMinigameFailure<public>(Agent:agent): void
}
    