using { /Fortnite.com/Game }
using { /Fortnite.com/Devices }
using { /Fortnite.com/Devices/CreativeAnimation }
using { /Fortnite.com/Characters }
using { /Verse.org/Simulation }
using { /Verse.org/Simulation/Tags }
using { /Verse.org/Random }
using { /Verse.org/Concurrency }
using { /UnrealEngine.com/Temporary/Diagnostics }
using { /UnrealEngine.com/Temporary/SpatialMath }
using { /Fortnite.com/Devices/CreativeAnimation/InterpolationTypes }

verse_accolade_manager_device<public> := class(tag) {}

accolade_manager_device<public> := class(creative_device) {
    @editable
    AccoladeVerySmallList: []accolades_device = array {}

    @editable
    AccoladeSmallList: []accolades_device = array {}

    @editable
    AccoladeMediumList: []accolades_device = array {}

    @editable
    AccoladeLargeList: []accolades_device = array {}

    @editable
    AccoladeVeryLargeList: []accolades_device = array {}

    @editable
    TriggerAwardAccoladeList: []trigger_device = array {}

    @editable
    CollectibleCoinList: []collectible_object_device = array {}

    # Constants
    MinDelayVeryLargeAccoladePlayerJoined: float = 15.0
    MaxDelayVeryLargeAccoladePlayerJoined: float = 20.0

    # Runtime variables
    var CurVeryLargeIndex: int = -1
    var PasswordAllAccoladesActive: logic = false

    OnBegin<override>()<suspends>:void = {
        for(ItemIndex -> Trigger: TriggerAwardAccoladeList) {
            EventHandler := event_handler_maybe_agent(int) {
                A := ItemIndex
                CallbackFunc := OnTriggerAwardAccolade
            }
            Trigger.TriggeredEvent.Subscribe(EventHandler.EventSubscription)
        }

        for(Coin: CollectibleCoinList) {
            Coin.CollectedEvent.Subscribe(OnCoinCollected)
        }

        # Only award very large accolade OnPlayerAdded
        GetPlayspace().PlayerAddedEvent().Subscribe(OnPlayerAdded)

        # Call logic for initial players
        for(Player : GetPlayspace().GetPlayers()) {
            OnPlayerAdded(Player)
        }

        # Start with a random CurVeryLargeIndex
        set CurVeryLargeIndex = GetRandomInt(0, AccoladeVeryLargeList.Length - 1)
    }

    OnPlayerAdded(Player: player): void = {
        Print("OnPlayerAdded")
        Task := spawn{OnPlayerAddedSuspends(Player)}
    }

    OnPlayerAddedSuspends(Player: player)<suspends>: void = {
        RandDelay := GetRandomFloat(MinDelayVeryLargeAccoladePlayerJoined, MaxDelayVeryLargeAccoladePlayerJoined)
        Sleep(RandDelay)
        AwardAccolade(Player, 4)
    }

    OnCoinCollected(Agent: agent): void = {
        AwardAccolade(Agent, 0)
    }

    OnTriggerAwardAccolade(MaybeAgent: ?agent, TriggerIndex: int): void = {
        if {
            Agent := MaybeAgent?
        } then {
            AwardAccolade(Agent, TriggerIndex)
        }
    }

    AwardAccolade<public>(Agent: agent, ParmAccoladeLevel: int): void = {
        Task := spawn{AwardAccoladeSuspends(Agent, ParmAccoladeLevel)}
    }

    AwardAccoladeSuspends(Agent: agent, ParmAccoladeLevel: int)<suspends>: void = {
        Sleep(0.0)
        if {
            PasswordAllAccoladesActive = false
        } then {
            if {
                PlayerData := Agent.GetPlayerData[]
            } then {
                var AccoladeList: []accolades_device = array {}
                var AccoladeLevel: int = ParmAccoladeLevel
    
                <# GotVeryLargeAccolade is now handled in PlayerAddedEvent instead
                if {
                    AccoladeLevel >= 2
                    PlayerData.GotVeryLargeAccolade = false
                } then {
                    set PlayerData.GotVeryLargeAccolade = true
                    # set AccoladeLevel = 4
                    # Grant extra accolade
                    AwardAccolade(Agent, 4)
                }
                #>
                <# else {
                    if {
                        AccoladeLevel = 2
                    } then {
                        Timestamp := GetSecondsSinceEpoch()
                        TimePassedSinceEvent := Timestamp - PlayerData.TimestampLargeAccoladeEpoch
                        if(TimePassedSinceEvent > 300.0) {
                            set PlayerData.TimestampLargeAccoladeEpoch = Timestamp
                            # set AccoladeLevel = 3
                            # Grant extra accolade
                            AwardAccolade(Agent, 3)
                        }
                    }
                }
                #>
        
                case(AccoladeLevel) {
                    0 =>
                        set AccoladeList = AccoladeVerySmallList
                    1 =>
                        set AccoladeList = AccoladeSmallList
                    2 =>
                        set AccoladeList = AccoladeMediumList
                    3 =>
                        set AccoladeList = AccoladeLargeList
                    4 =>
                        # Special case for very large accolade
                        if {
                            Accolade := AccoladeVeryLargeList[CurVeryLargeIndex]
                        } then {
                            set AccoladeList = AccoladeList + array {Accolade}
                            # Increment variable for next iteration
                            set CurVeryLargeIndex = CurVeryLargeIndex + 1
                            if {
                                CurVeryLargeIndex > AccoladeVeryLargeList.Length - 1
                            } then {
                                set CurVeryLargeIndex = 0
                            }
                        } else {
                            Print("ERR: AwardAccolade: VERY LARGE")
                        }
                    _ =>
                        Print("ERR: AwardAccolade: Unknown AccoladeLevel: {AccoladeLevel}")
                }
        
                # Award a random accolade to maximize XP gains
                RandInt := GetRandomInt(0, AccoladeList.Length - 1)
                if {
                    Accolade := AccoladeList[RandInt]
                } then {
                    Accolade.Award(Agent)
                } else {
                    Print("ERR: AwardAccolade: Didn't find Accolade for {ParmAccoladeLevel}, {RandInt}")
                }
            }
        } else {
            set PasswordAllAccoladesActive = false
            Task := spawn{AwardAllAccoladesSuspends(Agent)}
        }
    }

    ActivatePasswordAllAccolades<public>(): void = {
        set PasswordAllAccoladesActive = true
    }

    AwardAllAccoladesSuspends(Agent: agent)<suspends>: void = {
        AllAccoladeList := AccoladeVeryLargeList + AccoladeLargeList + AccoladeMediumList + AccoladeSmallList + AccoladeVerySmallList
        for(Accolade: AllAccoladeList) {
            Accolade.Award(Agent)
            Sleep(0.0)
        }
    }
}