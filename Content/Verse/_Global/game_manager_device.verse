using { /Fortnite.com/Game }
using { /Fortnite.com/FortPlayerUtilities }
using { /Fortnite.com/Devices }
# using { /Fortnite.com/Devices/CreativeAnimation }
using { /Fortnite.com/Characters }
using { /Verse.org/Simulation }
using { /Verse.org/Simulation/Tags }
using { /Verse.org/Random }
using { /Fortnite.com/Teams }
# using { /Verse.org/Assets }
# using { /Verse.org/Concurrency }
# using { /UnrealEngine.com/Temporary/Diagnostics }
using { /UnrealEngine.com/Temporary/SpatialMath }

# See https://dev.epicgames.com/documentation/en-us/uefn/create-your-own-device-in-verse for how to create a verse device.

verse_game_manager_device<public> := class(tag) {}

game_manager_device<public> := class(creative_device) {
    @editable
    TycoonConfig: _Tycoon.tycoon_config_device = _Tycoon.tycoon_config_device {}

    @editable
    TeamSelectorList<public>: []class_and_team_selector_device = array {}

    @editable
    MutatorZonePlayerSpawned: mutator_zone_device = mutator_zone_device {}

    @editable
    TycoonManagerList: []_Tycoon.tycoon_manager_device = array {}

    @editable
    SwitchUnequipAllCosmeticList: []switch_device = array {}

    @editable
    NoClassClassSelector<public>: class_and_team_selector_device = class_and_team_selector_device {}

    @editable
    IntroClassSelectorList<public>: []class_and_team_selector_device = array {}

    @editable
    LevelUpVfx<public>: visual_effect_powerup_device = visual_effect_powerup_device {}

    @editable
    LevelUpVfxRemover<public>: visual_effect_powerup_device = visual_effect_powerup_device {}

    @editable
    ItemGranterGoldVerse: item_granter_device = item_granter_device {}

    @editable
    HudMainUiCurrentLevel<public>: hud_message_device = hud_message_device {}

    @editable
    HudMainUiXpTowardsNextLevel<public>: hud_message_device = hud_message_device {}

    @editable
    PlayerLevelUpHud<public>: hud_message_device = hud_message_device {}

    @editable
    TimerPlaytimeXp: timer_device = timer_device {}

    @editable
    StartWeaponItemGranter: item_granter_device = item_granter_device {}

    @editable
    ScoreManager: score_manager_device = score_manager_device {}

    @editable
    DamagePowerupPowerupList<public>: []damage_amplifier_powerup_device = array {}

    @editable
    CosmeticVfxList<public>: []visual_effect_powerup_device = array {}

    @editable
    CosmeticVfxRemoverList<public>: []visual_effect_powerup_device = array {}

    @editable
    CurMoneyHud<public>: hud_message_device = hud_message_device {}

    @editable
    CurFansHud<public>: hud_message_device = hud_message_device {}

    @editable
    CurDiamondsHud<public>: hud_message_device = hud_message_device {}

    @editable
    CurCreditCardCoinHud<public>: hud_message_device = hud_message_device {}

    @editable
    CurRebirthsHud<public>: hud_message_device = hud_message_device {}

    @editable
    CurScamRankHud<public>: hud_message_device = hud_message_device {}

    <#
    @editable
    PlayerXpGainedHud<public>: hud_message_device = hud_message_device {}

    @editable
    PlayerMoneyGainedHud<public>: hud_message_device = hud_message_device {}

    @editable
    PlayerFansGainedHud<public>: hud_message_device = hud_message_device {}

    @editable
    PlayerDiamondsGainedHud<public>: hud_message_device = hud_message_device {}

    @editable
    PlayerGamePassCoinGainedHud<public>: hud_message_device = hud_message_device {}
    #>

    # We need a different NPC Spawner for each skin
    @editable
    PetNpcSpawnerList: []npc_spawner_device = array {}

    @editable
    ClassInstantReload<public>: class_and_team_selector_device = class_and_team_selector_device {}

    @editable
    ClassIncreasedMovementSpeed<public>: class_and_team_selector_device = class_and_team_selector_device {}

    @editable
    ClassInstantReloadAndIncreasedMovementSpeed<public>: class_and_team_selector_device = class_and_team_selector_device {}

    <#
    @editable
    TeleporterJail<public>: teleporter_device = teleporter_device {}

    @editable
    CinematicJail<public>: cinematic_sequence_device = cinematic_sequence_device {}

    @editable
    HudJail<public>: hud_message_device = hud_message_device {}

    @editable
    ButtonEscapeJail: button_device = button_device {}

    @editable
    TimerEscapeJail<public>: timer_device = timer_device {}

    @editable
    AutoEscapeJailDuration<public>: float = 60.0
    #>

    @editable
    BpCreditCardCoin<public>: creative_prop_asset = DefaultCreativePropAsset

    @editable
    BpCreditCardCoinCollected<public>: creative_prop_asset = DefaultCreativePropAsset

    @editable
    HudCurrentLevel<public>: hud_message_device = hud_message_device {}

    @editable
    HudNextLevel<public>: hud_message_device = hud_message_device {}

    @editable
    HudCurrentXp<public>: hud_message_device = hud_message_device {}

    @editable
    HudPlayerFramePlayerName<public>: hud_message_device = hud_message_device {}

    @editable
    HudPlayerFrameCurrentLevel<public>: hud_message_device = hud_message_device {}

    @editable
    HudPlayerFrameCurrentHealth<public>: hud_message_device = hud_message_device {}

    @editable
    HudPlayerFrameCurrentShield<public>: hud_message_device = hud_message_device {}

    @editable
    HudHideAchievementsListKill<public>: []hud_message_device = array {}

    @editable
    HudHideAchievementsListScam<public>: []hud_message_device = array {}

    @editable
    HudHideAchievementsListMoney<public>: []hud_message_device = array {}

    @editable
    HudHideAchievementsListRebirth<public>: []hud_message_device = array {}

    @editable
    AudioScamSucceeded<public>: audio_player_device = audio_player_device {}

    @editable
    KillStreakSoundList<public>: []audio_player_device = array {}

    # Constants

    # Runtime variables
    var RoutineServerTick: ?routine = false
    var MaybeLastAgentThatSpawnedPet<public>: ?agent = false

    var CurIntroIndex: int = 0
    var HackPoliceSpawnerId<public>: int = 0

    # Runs when the device is started in a running game
    OnBegin<override>()<suspends>: void = {
        # Assign Tycoon IDs
        for(ItemId -> TycoonManager: TycoonManagerList) {
            set TycoonManager.TycoonId = ItemId
        }
        # Listen to events
        for(NpcSpawner: PetNpcSpawnerList) {
            NpcSpawner.SpawnedEvent.Subscribe(OnPetNpcSpawned)
        }

        # We have to use a mutator zone, since checkpoints don't have a spawned event
        MutatorZonePlayerSpawned.AgentEntersEvent.Subscribe(OnPlayerSpawnedMutatorZone)

        for(SwitchUnequipAllCosmetic: SwitchUnequipAllCosmeticList) {
            SwitchUnequipAllCosmetic.TurnedOnEvent.Subscribe(OnSwitchUnequipAllCosmetic)
        }

        # ButtonEscapeJail.InteractedWithEvent.Subscribe(OnButtonEscapeJail)

        GetPlayspace().PlayerAddedEvent().Subscribe(OnPlayerAdded)
        GetPlayspace().PlayerRemovedEvent().Subscribe(OnPlayerRemoved)

        # Call logic for initial players
        for(Player : GetPlayspace().GetPlayers()) {
            OnPlayerAdded(Player)
        }

        # Activate pet spawners
        Sleep(10.0)
        for(NpcSpawner: PetNpcSpawnerList) {
            NpcSpawner.Enable()
        }
    }

    OnFirstTimePlaying(Agent: agent): void = {
        Print("GameManager - OnFirstTimePlaying")
        SaveManager := GetSaveManager()
        Task1 := spawn{HandleFirstTimePlayingSuspends(Agent)}
    }

    HandleFirstTimePlayingSuspends(Agent: agent)<suspends>: void = {
        if {
            PlayerData := Agent.GetPlayerData[]
        } then {
            StartWeaponItemGranter.GrantItem(Agent)
        }
    }
    
    ShowWhatsNewUi(Agent: agent, PlayerData: player_data): void = {
        SaveManager := GetSaveManager()
        # Check if the player has already played the current version
        if {
            Save := SaveManager.ProxySaves[Agent]
            Save.SaveVersion <> SaveManager.CurGameVersion
        } then {
            set Save.SaveVersion = SaveManager.CurGameVersion
            # Add whats new UI
            PlayerData.UiChangelog.AddUiToPlayer(Agent, false)
        }
    }

    OnEnd<override>(): void = {
        Print("GameManager - OnEnd")
        # Clean up routines
        if(Routine := RoutineServerTick?) {
            Routine.Cancel()
        }
    }

    OnPlayerAdded(Player: player): void = {
        Print("OnPlayerAdded")
        SaveManager := GetSaveManager()
        InitPlayerData(Player)
        
        if {
            PlayerData := Player.GetPlayerData[]
        } then {
            var TeamIndex: int = -1
            # Call PlayerData init function
            PlayerData.InitPlayerData(Player, TeamIndex)
        }

        TimerPlaytimeXp.Start(Player)
    }

    OnPlayerSpawnedMutatorZone(Agent: agent): void = {
        Print("OnPlayerSpawnedMutatorZone")
        if {
            PlayerData := Agent.GetPlayerData[]
        } then {
            # Assign a tycoon if the player has none
            if(not PlayerData.MaybeTycoonManager?) {
                Task := spawn{AssignEmptyTycoonToNewPlayer(option{Agent})}
            }
            PlayerData.HandlePlayerSpawned()
        }
    }

    OnPlayerRemoved(Player: player): void = {
        Print("OnPlayerRemoved")
        SaveManager := GetSaveManager()

        if(PlayerData := Player.GetPlayerData[]) {
            # Reset Tycoon if the player owns one
            PlayerData.UnassignTycoon()
            # Grant playtime analytics
            Timestamp := GetSimulationElapsedTime()
            TimePlayed := Timestamp - PlayerData.TimestampPlayerJoined
            if(TimePlayed >= 10.0 * 60.0) {
                SaveManager.GrantAnalytic(Player, SaveManager.ANALYTIC_10_MIN_PLAYED)
            }
            if(TimePlayed >= 20.0 * 60.0) {
                SaveManager.GrantAnalytic(Player, SaveManager.ANALYTIC_20_MIN_PLAYED)
            }
            if(TimePlayed >= 30.0 * 60.0) {
                SaveManager.GrantAnalytic(Player, SaveManager.ANALYTIC_30_MIN_PLAYED)
            }
            if(TimePlayed >= 40.0 * 60.0) {
                SaveManager.GrantAnalytic(Player, SaveManager.ANALYTIC_40_MIN_PLAYED)
            }
            if(TimePlayed >= 50.0 * 60.0) {
                SaveManager.GrantAnalytic(Player, SaveManager.ANALYTIC_50_MIN_PLAYED)
            }
            if(TimePlayed >= 60.0 * 60.0) {
                SaveManager.GrantAnalytic(Player, SaveManager.ANALYTIC_60_MIN_PLAYED)
            }
            if(TimePlayed >= 70.0 * 60.0) {
                SaveManager.GrantAnalytic(Player, SaveManager.ANALYTIC_70_MIN_PLAYED)
            }
            if(TimePlayed >= 80.0 * 60.0) {
                SaveManager.GrantAnalytic(Player, SaveManager.ANALYTIC_80_MIN_PLAYED)
            }
            if(TimePlayed >= 90.0 * 60.0) {
                SaveManager.GrantAnalytic(Player, SaveManager.ANALYTIC_90_MIN_PLAYED)
            }
            if(TimePlayed >= 100.0 * 60.0) {
                SaveManager.GrantAnalytic(Player, SaveManager.ANALYTIC_100_MIN_PLAYED)
            }

            # Destroy player data
            PlayerData.DestroyPlayerData(Player)

            Player.DestroyPlayerData()
        }
    }

    OnPetNpcSpawned(NpcAgent: agent): void = {
        # Print("OnNpcSpawned")
        Utility := GetUtility()
        if {
            OwnerAgent := MaybeLastAgentThatSpawnedPet?
            PlayerData := OwnerAgent.GetPlayerData[]
            TycoonManager := PlayerData.MaybeTycoonManager?
        } then {
            NpcPetData := _Npc.npc_pet_data {}
            # Call Npc Init
            NpcPetData.InitPetNpc(NpcAgent, TycoonManager, OwnerAgent)
            NpcAgent.SetNpcPetData(NpcPetData)
        }
    }

    ActivateXpBoostCode<public>(Agent: agent): void = {
        if {
            PlayerData := Agent.GetPlayerData[]
        } then {
            PlayerData.SetXpBoostLevel(1)
        }
    }

    ActivateCodeOneTimeReward<public>(Agent: agent): void = {
        if(PlayerData := Agent.GetPlayerData[]) {
            # To add: Rewards
            <# Add xp
            CumulativeXpForLevel450 := 39406906
            XpToAdd := CumulativeXpForLevel450 - PlayerData.TotalXp
            if(XpToAdd > 0) {
                PlayerData.AddXp(XpToAdd, true)
            }
            # Add rebirths
            RebirthsAfterCode := 14
            RebirthsToAdd := RebirthsAfterCode - PlayerData.NumOfRebirths
            if(RebirthsToAdd > 0) {
                PlayerData.UpdateCurRebirths(RebirthsAfterCode)
            }
            #>
        }
    }

    <#
    RestoreXpBoostCode(Agent: agent, PlayerData: player_data): void = {
        # Print("RestoreXpBoostCode")
        SaveManager := GetSaveManager()
        if {
            Save := SaveManager.ProxySaves[Agent]
        } then {
            if(Save.HasXpBoostCode = true) {
                PlayerData.ActivateXpBoostCode()
            }
        }
    }
    #>

    AssignEmptyTycoonToNewPlayer(MaybeAgent: ?agent)<suspends>: void = {
        # Sleep 3 seconds, since we have to load save data first
        # Intro is 13 seconds, but Sleep(14.0) was too fast
        Sleep(18.0)
        # Check if the player is still playing
        if(Agent := MaybeAgent?) {
            <# This doesn't work since every player spawns on team 1 first
            # Assign the tycoon according to the player spawner team
            if(TycoonManager := TycoonManagerList[SpawnerIndex]) {
                if(TycoonManager.MaybeOwnerAgent = false) {
                    TycoonManager.TryToAssignTycoonToAgent(Agent)
                    TycoonManager.PlayIntroCinematic()
                } else {
                    Print("ERR: AssignEmptyTycoonToNewPlayer: Tycoon already has a different owner!")
                }
            }
            #>

            # Automatically assign the player to the first free Tycoon
            var FoundFreeTycoon: logic = false
            for(TycoonId -> TycoonManager: TycoonManagerList) {
                if(FoundFreeTycoon = false) {
                    if(TycoonManager.MaybeOwnerAgent = false) {
                        set FoundFreeTycoon = true
                        TycoonManager.TryToAssignTycoonToAgent(Agent)
                        TycoonManager.PlayIntroCinematic()
                    }
                }
            }
            if(FoundFreeTycoon = false) {
                Print("ERR: AssignEmptyTycoonToNewPlayer - Found no empty tycoon for the new player!")
            }
        }
    }

    OnSwitchUnequipAllCosmetic(Agent: agent): void = {
        if {
            PlayerData := Agent.GetPlayerData[]
        } then {
            # PlayerData.UnequipAllCosmetics(true)
            # This hides the cosmetics, but keeps the scam level
            PlayerData.UnequipAllCosmetics(false)
        }
    }

    <#
    OnButtonEscapeJail(Agent: agent): void = {
        HandleEscapeFromJail(Agent)
    }

    HandleEscapeFromJail(Agent: agent): void = {
        if {
            PlayerData := Agent.GetPlayerData[]
        } then {
            PlayerData.TryToEscapeFromJail(Agent)
        }
    }
    #>

    InitPlayerData(Player: player): void = {
        Print("InitPlayerData")
        # Check if the player already has player data assigned
        if(PlayerData := Player.GetPlayerData[]) {
            Print("WARN: Player already has Player Data!")
        }

        # Init player data
        Player.SetPlayerData(player_data {})
    }
}
