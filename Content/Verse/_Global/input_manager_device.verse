using { /Fortnite.com/Game }
using { /Fortnite.com/Devices }
using { /Fortnite.com/Devices/CreativeAnimation }
using { /Fortnite.com/Characters }
using { /Verse.org/Simulation }
using { /Verse.org/Simulation/Tags }
using { /Verse.org/Random }
using { /Verse.org/Concurrency }
using { /UnrealEngine.com/Temporary/Diagnostics }
using { /UnrealEngine.com/Temporary/SpatialMath }
using { /Fortnite.com/Devices/CreativeAnimation/InterpolationTypes }

verse_input_manager_device<public> := class(tag) {}

input_manager_device<public> := class(creative_device) {
    @editable
    InputFire<public>: input_trigger_device = input_trigger_device {}

    @editable
    InputTarget<public>: input_trigger_device = input_trigger_device {}

    @editable
    InputJump<public>: input_trigger_device = input_trigger_device {}

    @editable
    InputInteract<public>: input_trigger_device = input_trigger_device {}

    @editable
    InputForward<public>: input_trigger_device = input_trigger_device {}

    @editable
    InputBackward<public>: input_trigger_device = input_trigger_device {}

    @editable
    InputLeft<public>: input_trigger_device = input_trigger_device {}

    @editable
    InputRight<public>: input_trigger_device = input_trigger_device {}

    @editable
    InputUp<public>: input_trigger_device = input_trigger_device {}

    @editable
    InputCrouch<public>: input_trigger_device = input_trigger_device {}

    @editable
    InputSprint<public>: input_trigger_device = input_trigger_device {}

    # Runs when the device is started in a running game
    OnBegin<override>()<suspends>:void = {
        
    }

    AddArrowMinigameInputsPc(Agent: agent): void = {
        InputLeft.Register(Agent)
        InputRight.Register(Agent)
        InputForward.Register(Agent)
        InputBackward.Register(Agent)
    }

    RemoveArrowMinigameInputsPc(Agent: agent): void = {
        InputLeft.Unregister(Agent)
        InputRight.Unregister(Agent)
        InputForward.Unregister(Agent)
        InputBackward.Unregister(Agent)
    }

    AddArrowMinigameInputsConsole(Agent: agent): void = {
        InputFire.Register(Agent)
        InputTarget.Register(Agent)
        InputJump.Register(Agent)
        InputInteract.Register(Agent)
    }

    RemoveArrowMinigameInputsConsole(Agent: agent): void = {
        InputFire.Unregister(Agent)
        InputTarget.Unregister(Agent)
        InputJump.Unregister(Agent)
        InputInteract.Unregister(Agent)
    }

    DragonMountRegisterAgent<public>(Agent: agent): void = {
        InputForward.Register(Agent)
        InputBackward.Register(Agent)
        InputLeft.Register(Agent)
        InputRight.Register(Agent)
        InputUp.Register(Agent)
        InputCrouch.Register(Agent)
        InputSprint.Register(Agent)
    }

    DragonMountUnregisterAgent<public>(Agent: agent): void = {
        InputForward.Unregister(Agent)
        InputBackward.Unregister(Agent)
        InputLeft.Unregister(Agent)
        InputRight.Unregister(Agent)
        InputUp.Unregister(Agent)
        InputCrouch.Unregister(Agent)
        InputSprint.Unregister(Agent)
    }
}