using { /Fortnite.com/Game }
using { /Fortnite.com/Devices }
using { /Fortnite.com/Devices/CreativeAnimation }
using { /Fortnite.com/Characters }
using { /Verse.org/Simulation }
using { /Verse.org/Simulation/Tags }
using { /Verse.org/Random }
using { /Verse.org/Concurrency }
using { /UnrealEngine.com/Temporary/Diagnostics }
using { /UnrealEngine.com/Temporary/SpatialMath }

verse_quest_manager_device<public> := class(tag) {}

quest_manager_device<public> := class(creative_device) {
    # These should be the same for every quest
    @editable
    PlayerQuestProgressHud<public>: hud_message_device = hud_message_device {}

    @editable
    HudPlayerLevelTooLow: hud_message_device = hud_message_device {}

    @editable
    QuestAcceptedAudio: audio_player_device = audio_player_device {}

    @editable
    RewardAccolade: accolades_device = accolades_device {}

    @editable
    QuestCompletedHudMessage: hud_message_device = hud_message_device {}

    @editable
    QuestCompletedAudio: audio_player_device = audio_player_device {}

    # Constants

    # Runtime variables

    # Runs when the device is started in a running game
    OnBegin<override>()<suspends>:void = {
        
    }

    OnQuestAccepted<public>(Agent: agent, Quest: _Tycoon.tycoon_button): void = {
        Print("OnQuestAccepted: ButtonId: {Quest.ButtonId}")
        SaveManager := GetSaveManager()
        if {
            PlayerData := Agent.GetPlayerData[]
            TycoonManager := PlayerData.MaybeTycoonManager?
            Save := SaveManager.ProxySaves[Agent]
        } then {
            if(PlayerData.CurLevel + 1 < Quest.QuestPlayerLevelRequirement) {
                Print("WARN: PlayerLevel too low")
                HudPlayerLevelTooLow.Show(Agent)
                return
            }

            set PlayerData.MaybeActiveQuest = option{Quest}
            set PlayerData.CurrentQuestId = Quest.ButtonId
            set Save.CurQuestTycoonButtonId = Quest.ButtonId
            # We can't reset Save.QuestProgressCount here, in case we are loading a saved quest
            # set Save.QuestProgressCount = 0
            # Init quest data
            # Check if the QuestObjectiveBeaconPos is set (would be better to use optional parameter in future games)
            if(Quest.QuestObjectiveBeaconPos.GetTransform().Translation.Z <> 0.0) {
                if(TycoonManager.QuestObjectiveBeaconParent.TeleportTo[Quest.QuestObjectiveBeaconPos.GetTransform()]) {}
                TycoonManager.QuestObjectiveBeacon.AddToShowList(Agent)
            }

            MyRoutine := SpawnRoutine(PlayerData.UpdateQuestProgressTick)
            set PlayerData.RoutineQuestProgressTick = option{MyRoutine}

            # Add listeners to increase quest progress
    	    if(Button := Quest.QuestGoalButton?) {
                EventSub: cancelable = Button.InteractedWithEvent.Subscribe(PlayerData.HandleQuestProgressAgent)
                set PlayerData.QuestListenerList = PlayerData.QuestListenerList + array{EventSub}
            }
            if(MutatorZone := Quest.QuestGoalMutatorZone?) {
                EventSub: cancelable = MutatorZone.AgentEntersEvent.Subscribe(PlayerData.HandleQuestProgressAgent)
                set PlayerData.QuestListenerList = PlayerData.QuestListenerList + array{EventSub}
            }
            for(Switch: Quest.QuestGoalSwitchList) {
                EventSub: cancelable = Switch.TurnedOffEvent.Subscribe(PlayerData.HandleQuestProgressAgent)
                set PlayerData.QuestListenerList = PlayerData.QuestListenerList + array{EventSub}
            }
        }
    }

    HandleQuestCompleted<public>(Agent: agent, Quest: _Tycoon.tycoon_button): void = {
        Print("HandleQuestCompleted")
        if {
            PlayerData := Agent.GetPlayerData[]
        } then {
            CleanUpQuestData(Agent, Quest)
            # Show hud
            QuestCompletedHudMessage.Show(Agent)
            QuestCompletedAudio.Play(Agent)
            # Give rewards
            PlayerData.AddXp(Quest.QuestRewardXpAmount, true)
            PlayerData.AddMoney(Quest.QuestRewardMoneyAmount, true, false)
            PlayerData.AddFans(Quest.QuestRewardFansAmount * 1.0, true)
            PlayerData.AddDiamonds(Quest.QuestRewardDiamondsAmount * 1.0, true)
            RewardAccolade.Award(Agent)
            if(RewardItemGranter := Quest.QuestRewardItemOnComplete?) {
                RewardItemGranter.GrantItem(Agent)
            }
        }
    }

    CleanUpQuestData<public>(Agent: agent, Quest: _Tycoon.tycoon_button): void = {
        SaveManager := GetSaveManager()
        Utility := GetUtility()
        if {
            PlayerData := Agent.GetPlayerData[]
            TycoonManager := PlayerData.MaybeTycoonManager?
            Save := SaveManager.ProxySaves[Agent]
        } then {
            set Save.QuestProgressCount = 0
            set Save.CurQuestTycoonButtonId = -1
            # Notify TycoonButton that the quest was completed
            Quest.OnQuestCompleted()

            # Clean up routines
            if(Routine := PlayerData.RoutineQuestProgressTick?) {
                Routine.Cancel()
                set PlayerData.RoutineQuestProgressTick = false
            }
            for(EventSub : PlayerData.QuestListenerList) {
                EventSub.Cancel()
            }
            set PlayerData.QuestListenerList = array {}
            # Clean up active quest
            set PlayerData.MaybeActiveQuest = false
            <# Save that the player has no quest active
            if(Save := SaveManager.ProxySaves[Agent]) {
                set Save.HasStartedQuest = false
            }
            #>
            # Clean up quest data
            PlayerQuestProgressHud.Hide(Agent)
            TycoonManager.QuestObjectiveBeacon.RemoveFromShowList(Agent)
        }
    }
}
