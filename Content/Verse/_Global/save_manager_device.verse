using { /Fortnite.com/Game }
using { /Fortnite.com/Devices }
using { /Fortnite.com/Devices/CreativeAnimation }
using { /Fortnite.com/Characters }
using { /Verse.org/Simulation }
using { /Verse.org/Simulation/Tags }
using { /Verse.org/Random }
using { /Verse.org/Concurrency }
using { /UnrealEngine.com/Temporary/Diagnostics }
using { /UnrealEngine.com/Temporary/SpatialMath }
using { /Fortnite.com/Devices/CreativeAnimation/InterpolationTypes }

verse_save_manager_device<public> := class(tag) {}

var SaveData<public>: weak_map(player, player_save) = map {}

achievement_type<public> := enum {
    Kill
    Scam
    Money
    Rebirth
}

save_manager_device<public> := class(creative_device) {
    @editable
    LoadSaveData: logic = true

    <#
    @editable
    AutoSaveEnabled: logic = true
    #>

    @editable
    AnalyticList<public>: []analytics_device = array {}

    @editable
    AchievementHudList<public>: []hud_message_device = array {}

    @editable
    AchievementMessageList<public>: []string = array {}

    @editable
    AchievementUnlockedHudListKill<public>: []hud_message_device = array {}

    @editable
    AchievementUnlockedHudListScam<public>: []hud_message_device = array {}

    @editable
    AchievementUnlockedHudListMoney<public>: []hud_message_device = array {}

    @editable
    AchievementUnlockedHudListRebirth<public>: []hud_message_device = array {}

    @editable
    HudPlayerSaved<public>: hud_message_device = hud_message_device {}

    # Constants
    CurGameVersion: int = 1
    AutoSaveIntervalSec: float = 10.0

    # Analytic indizes
    ANALYTIC_LEVEL_10<public>: int = 0
    ANALYTIC_LEVEL_20<public>: int = 1
    ANALYTIC_LEVEL_30<public>: int = 2
    ANALYTIC_LEVEL_40<public>: int = 3
    ANALYTIC_LEVEL_50<public>: int = 4
    ANALYTIC_LEVEL_60<public>: int = 5
    ANALYTIC_LEVEL_70<public>: int = 6
    ANALYTIC_LEVEL_80<public>: int = 7
    ANALYTIC_LEVEL_90<public>: int = 8
    ANALYTIC_LEVEL_100<public>: int = 9

    ANALYTIC_10_MIN_PLAYED<public>: int = 10
    ANALYTIC_20_MIN_PLAYED<public>: int = 11
    ANALYTIC_30_MIN_PLAYED<public>: int = 12
    ANALYTIC_40_MIN_PLAYED<public>: int = 13
    ANALYTIC_50_MIN_PLAYED<public>: int = 14
    ANALYTIC_60_MIN_PLAYED<public>: int = 15
    ANALYTIC_70_MIN_PLAYED<public>: int = 16
    ANALYTIC_80_MIN_PLAYED<public>: int = 17
    ANALYTIC_90_MIN_PLAYED<public>: int = 18
    ANALYTIC_100_MIN_PLAYED<public>: int = 19

    ANALYTIC_NEW_PLAYER<public>: int = 20
    ANALYTIC_RETURNING_PLAYER<public>: int = 21
    ANALYTIC_FIRST_REBIRTH<public>: int = 22

    # TODO: Add correct Indizes
    ANALYTIC_PET_BOUGHT<public>: int = 2
    ANALYTIC_WOLF_MOUNT_BOUGHT<public>: int = 3
    ANALYTIC_DRAGON_MOUNT_BOUGHT<public>: int = 4
    ANALYTIC_CODE_INPUT_OPENED<public>: int = 5
    ANALYTIC_BOSS_NORMAL<public>: int = 6
    ANALYTIC_BOSS_HEROIC<public>: int = 7
    ANALYTIC_BOSS_MYTHIC<public>: int = 8
    ANALYTIC_PLAYER_ELIMINATED_BY_BOSS_NORMAL<public>: int = 9
    ANALYTIC_PLAYER_ELIMINATED_BY_BOSS_HEROIC<public>: int = 10
    ANALYTIC_PLAYER_ELIMINATED_BY_BOSS_MYTHIC<public>: int = 11

    # Runtime variables
    var ProxySaves<public>: [agent]player_save_proxy = map{}

    # Runs when the device is started in a running game
    OnBegin<override>()<suspends>:void = {
        # Listen to events
        GetPlayspace().PlayerAddedEvent().Subscribe(OnPlayerAdded)
        GetPlayspace().PlayerRemovedEvent().Subscribe(OnPlayerRemoved)

        # Call logic for initial players
        for(Player : GetPlayspace().GetPlayers()) {
            OnPlayerAdded(Player)
        }

        # Sleep to make sure every player loaded save data correctly
        Sleep(30.0)

        spawn{AutoSaveTick()}
        <#
        if(AutoSaveEnabled = true) {
            Print("INFO: AutoSave is enabled!")
            spawn{AutoSaveTick()}
        } else {
            Print("INFO: AutoSave is not enabled!")
        }
        #>
    }

    OnPlayerAdded(Player: player): void = {
        GameManager := GetGameManager()
        if {
            Save := SaveData[Player]
            LoadSaveData = true
        } then {
            # Player has played before
            Print("save_manager_device - Returning player")
            if(set ProxySaves[Player] = CreateProxyFromSave(Save)) {}
            GrantAnalytic(Player, ANALYTIC_RETURNING_PLAYER)
            Task := spawn{RestoreSaveData(option{Player}, option{Save})}
        } else {
            # New player
            Print("save_manager_device - New player")
            if(set ProxySaves[Player] = player_save_proxy {
                # Init save version
                SaveVersion := CurGameVersion
            }) {
                # Do nothing here
            } else {
                Print("ERR: save_manager_device: Couldn't init player_save_proxy")
            }
            GrantAnalytic(Player, ANALYTIC_NEW_PLAYER)
            GameManager.OnFirstTimePlaying(Player)
        }
    }

    # This must be delayed to make sure PlayerData is initialized
    RestoreSaveData(MaybePlayer: ?player, MaybeSave: ?player_save)<suspends>: void = {
        GameManager := GetGameManager()
        Sleep(1.0)
        if {
            Player := MaybePlayer?
            Save := MaybeSave?
            PlayerData := Player.GetPlayerData[]
        } then {
            # Load game specific data
            PlayerData.LoadXp(Save.TotalXp)
            PlayerData.UpdateCurMoney(Save.Money)
            PlayerData.UpdateCurFans(Save.Fans)
            PlayerData.UpdateCurDiamonds(Save.Diamonds)
            PlayerData.UpdateCurCreditCardCoin(Save.CurCreditCardCoins)
            PlayerData.UpdateCurRebirths(Save.Rebirths)
            PlayerData.UpdateResourceUi()
            PlayerData.UpdateRebirthsUi()
            PlayerData.UpdateMoneyMultiplier()
            PlayerData.UpdateFansMultiplier()
            PlayerData.UpdateDiamondsMultiplier()
            PlayerData.UpdateXpMultiplier()
            # GameManager.ShowWhatsNewUi(Player, PlayerData)
            PlayerData.LoadRebirths(Save.Rebirths)
            PlayerData.LoadGamePass(Save.GamePassCoins, Save.GamePassLevel)
            if(Save.XpBoostLevel > 0) {
                PlayerData.SetXpBoostLevel(Save.XpBoostLevel)
            }
        } else {
            Print("ERR: RestoreSaveData: Found no PlayerData")
        }
    }

    OnPlayerRemoved(Player: player): void = {
        # Don't save the player on leave for this game
        # PersistPlayerSave(Player)
        set ProxySaves = ProxySaves.RemoveKey(Player)
    }

    GetSessionId<public>(Agent : agent)<transacts><decides>: int = {
        return ProxySaves[Agent].SessionId
    }

    AutoSaveTick<private>()<suspends>: void = {
        loop {
            Sleep(AutoSaveIntervalSec)
            for(Agent -> ProxySave: ProxySaves) {
                # Only AutoSave players in this game, which didn't unlock the auto-save device
                if {
                    PlayerData := Agent.GetPlayerData[]
                    PlayerData.HasUnlockedTycoonAutoSave = false
                    PlayerData.NumOfRebirths = 0
                } then {
                    Print("AutoSaveTick: Persisting Player")
                    Baked := ProxySave.Bake()
                    if(set SaveData[player[Agent]] = Baked) {}
                }
            }
        }
    }

    PersistPlayerSave<public>(Agent: agent): void = {
        if {
            ProxySave := ProxySaves[Agent]
        } then {
            <# Persist tycoon buttons
            if {
                PlayerData := Agent.GetPlayerData[]
            } then {
                set ProxySave.TycoonButtonsOwned = PlayerData.TycoonButtonsOwned
            }
            #>
            # Perform the save
            Baked := ProxySave.Bake()
            if(set SaveData[player[Agent]] = Baked) {}
            # Show HUD afterwards
            HudPlayerSaved.Show(Agent)
        } else {
            Print("ERR: PersistPlayerSave: Found no ProxySave for the Player!")
        }
    }

    IncrementAchievement<public>(Agent: agent, AchievementType: achievement_type, AchievementId: int, IncrementAmount: int): void = {
        GameConfig := GetGameConfig()
        if {
            Save := ProxySaves[Agent]
        } then {
            var AchievementLevel: int = 0
            case(AchievementType) {
                achievement_type.Kill =>
                    set AchievementLevel = Save.NumOfAchievementsCompletedKill
                achievement_type.Scam =>
                    set AchievementLevel = Save.NumOfAchievementsCompletedScam
                achievement_type.Money =>
                    set AchievementLevel = Save.NumOfAchievementsCompletedMoney
                achievement_type.Rebirth =>
                    set AchievementLevel = Save.NumOfAchievementsCompletedRebirth
            }
            # Only increment the achievement if we are currently on this AchievementId
            if(AchievementLevel <> AchievementId) {
                return
            }

            var AchievementProgress: int = 0
            var AchievementGoal: int = 0
            case(AchievementType) {
                achievement_type.Kill =>
                    set AchievementProgress = Save.AchievementProgressKill + IncrementAmount
                    set Save.AchievementProgressKill = AchievementProgress
                    if {
                        AchievementGoalTemp := GameConfig.AchievementGoalListKill[AchievementId]
                        set AchievementGoal = AchievementGoalTemp
                    } then {
                        if(AchievementProgress >= AchievementGoal) {
                            set Save.NumOfAchievementsCompletedKill = AchievementLevel + 1
                            set Save.AchievementProgressKill = 0
                            OnAchievementGoalReached(Agent, AchievementType, AchievementLevel)
                        }
                    } else {
                        Print("INFO: IncrementAchievement: Couldn't check goal for AchievementId: {AchievementId}. Might be max level")
                    }
                achievement_type.Scam =>
                    set AchievementProgress = Save.AchievementProgressScam + IncrementAmount
                    set Save.AchievementProgressScam = AchievementProgress
                    if {
                        AchievementGoalTemp := GameConfig.AchievementGoalListScam[AchievementId]
                        set AchievementGoal = AchievementGoalTemp
                    } then {
                        if(AchievementProgress >= AchievementGoal) {
                            set Save.NumOfAchievementsCompletedScam = AchievementLevel + 1
                            set Save.AchievementProgressScam = 0
                            OnAchievementGoalReached(Agent, AchievementType, AchievementLevel)
                        }
                    } else {
                        Print("INFO: IncrementAchievement: Couldn't check goal for AchievementId: {AchievementId}. Might be max level")
                    }
                achievement_type.Money =>
                    set AchievementProgress = Save.AchievementProgressMoney + IncrementAmount
                    set Save.AchievementProgressMoney = AchievementProgress
                    if {
                        AchievementGoalTemp := GameConfig.AchievementGoalListMoney[AchievementId]
                        set AchievementGoal = AchievementGoalTemp
                    } then {
                        if(AchievementProgress >= AchievementGoal) {
                            set Save.NumOfAchievementsCompletedMoney = AchievementLevel + 1
                            set Save.AchievementProgressMoney = 0
                            OnAchievementGoalReached(Agent, AchievementType, AchievementLevel)
                        }
                    } else {
                        Print("INFO: IncrementAchievement: Couldn't check goal for AchievementId: {AchievementId}. Might be max level")
                    }
                achievement_type.Rebirth =>
                    set AchievementProgress = Save.AchievementProgressRebirth + IncrementAmount
                    set Save.AchievementProgressRebirth = AchievementProgress
                    if {
                        AchievementGoalTemp := GameConfig.AchievementGoalListRebirth[AchievementId]
                        set AchievementGoal = AchievementGoalTemp
                    } then {
                        if(AchievementProgress >= AchievementGoal) {
                            set Save.NumOfAchievementsCompletedRebirth = AchievementLevel + 1
                            set Save.AchievementProgressRebirth = 0
                            OnAchievementGoalReached(Agent, AchievementType, AchievementLevel)
                        }
                    } else {
                        Print("INFO: IncrementAchievement: Couldn't check goal for AchievementId: {AchievementId}. Might be max level")
                    }
            }
        }
    }

    OnAchievementGoalReached(Agent: agent, AchievementType: achievement_type, AchievementId: int): void = {
        Print("GrantAchievement: AchievementId: {AchievementId}")
        GameConfig := GetGameConfig()
        Utility := GetUtility()
        AccoladeManager := GetAccoladeManager()

        if {
            PlayerData := Agent.GetPlayerData[]
        } then {
            # Show Achievement UI
            var AchievementUnlockedHudList: []hud_message_device = array {}
            case(AchievementType) {
                achievement_type.Kill =>
                    set AchievementUnlockedHudList = AchievementUnlockedHudListKill
                    if(Routine := PlayerData.MaybeRoutineAchievementTimerKill?) {
                        Routine.Cancel()
                    }
                    set PlayerData.MaybeRoutineAchievementTimerKill = false
                achievement_type.Scam =>
                    set AchievementUnlockedHudList = AchievementUnlockedHudListScam
                    if(Routine := PlayerData.MaybeRoutineAchievementTimerScam?) {
                        Routine.Cancel()
                    }
                    set PlayerData.MaybeRoutineAchievementTimerScam = false
                achievement_type.Money =>
                    set AchievementUnlockedHudList = AchievementUnlockedHudListMoney
                achievement_type.Rebirth =>
                    set AchievementUnlockedHudList = AchievementUnlockedHudListRebirth
            }
            if {
                # Save := ProxySaves[Agent]
                # PlayerData := Agent.GetPlayerData[]
                AchievementUnlockedHud := AchievementUnlockedHudList[AchievementId]
            } then {
                AchievementUnlockedHud.Show(Agent)
                
                <#
                if {
                    AchievementXpRewardConfig := GameConfig.AchievementXpRewardList[AchievementId]
                    AchievementXpReward := AchievementXpRewardConfig[AchievementLevel - 1]
                } then {
                    PlayerData.AddXp(AchievementXpReward, true)
                } else {
                    Print("ERR: OnAchievementGoalReached: Found no XpRewardConfig for AchievementId: {AchievementId}, AchievementLevel: {AchievementLevel - 1}")
                }
                #>
            } else {
                Print("ERR: Found no AchievementUnlockedHud: AchievementId: {AchievementId}")
            }
            # Update achievement wall
            if {
                TycoonManager := PlayerData.MaybeTycoonManager?
            } then {
                TycoonManager.AchievementBoard.UpdateAchievementChestState(Agent)
            }
            # Grant accolade
            AccoladeManager.AwardAccolade(Agent, 3)
        }
    }

    GrantAnalytic<public>(Agent: agent, AnalyticId: int): void = {
        Print("GrantAnalytic: AnalyticId: {AnalyticId}")
        if(Analytic := AnalyticList[AnalyticId]) {
            Analytic.Submit(Agent)
        } else {
            Print("ERR: GrantAnalytic: Found no Device for Index: {AnalyticId}")
        }
    }
}