using { /Verse.org/Assets }

game_config<public> := class {
    FloatingUiMessageDelay<public>: float = 0.5
    KillStreakDuration<public>: float = 4.0
    # RebirthIncomeMultiplier is for Money, Fans, Diamonds
    RebirthIncomeMultiplier<public>: float = 0.01
    RebirthXpMultiplier<public>: float = 0.05
    NpcsReservedForEachPlayerRank<public>: int = 5
    # If a player in the lobby increased the HighestPlayerLevel: Instantly upgrade X NPCs to the highest rank
    ScamLevelIncreasedNpcUpgradeAmount<public>: int = 5
    # Credit Card coin config - Spawns on NPC Kill 0.1 = 10% spawn chance
    HeroPowerupBonusCoinSpawnChance<public>: float = 0.1
    HeroPowerupBonusCoinPickupRange<public>: float = 100.0
    HeroPowerupBonusCoinXpOnPickup<public>: int = 25
    HeroPowerupBonusCoinDespawnDuration<public>: float = 30.0
    HeroPowerupBonusCoinCollectedVfxBpDuration<public>: float = 0.5
    # 1800 sec = 30 min
    AutoSaveBaseInterval<public>: float = 1800.0

    MoneyPerScamNpcRankMin<public>: []int = array {
        50,
        400,
        2500,
        6500,
        15000,
        100000,
        450000,
        900000,
        2000000,
        3000000
    }
    MoneyPerScamNpcRankMax<public>: []int = array {
        90,
        650,
        5500,
        9000,
        20000,
        175000,
        650000,
        1250000,
        4000000,
        5000000
    }

    ScamsPerScamNpcRankMin<public>: []int = array {
        3,
        75,
        850,
        1050,
        3000,
        75000,
        150000,
        500000,
        750000,
        1000000
    }
    ScamsPerScamNpcRankMax<public>: []int = array {
        8,
        100,
        1250,
        1500,
        4500,
        125000,
        200000,
        700000,
        900000,
        1250000
    }

    MoneyPerKillNpcRankMin<public>: []int = array {
        30,
        350,
        1500,
        2500,
        9500,
        150000,
        185000,
        350000,
        750000,
        900000
    }
    MoneyPerKillNpcRankMax<public>: []int = array {
        70,
        500,
        2150,
        5500,
        12500,
        175000,
        250000,
        400000,
        800000,
        1000000
    }
    # Kill Points Min & Max
    BaseKillpointsPerKillNpcRankMin<public>: []float = { 
        3.0,
        100.0,
        450.0,
        750.0,
        5000.0,
        75000.0,
        200000.0,
        350000.0,
        500000.0,
        700000.0
    }
    BaseKillpointsPerKillNpcRankMax<public>: []float = { 
        10.0,
        250.0,
        600.0,
        1050.0,
        7500.0,
        150000.0,
        250000.0,
        400000.0,
        650000.0,
        850000.0
    }
    # Police gives kill points like killing a NPC with your current scam level + BonusCoinsPerKillPoliceScamLevel
    BonusKillpointsPerKillPoliceScamLevelMin<public>: []float = { 
        3.0,
        25.0,
        150.0,
        300.0,
        500.0,
        25000.0,
        85000.0,
        150000.0,
        200000.0,
        300000.0
    }
    BonusKillpointsPerKillPoliceScamLevelMax<public>: []float = { 
        8.0,
        50.0,
        200.0,
        500.0,
        700.0,
        75000.0,
        125000.0,
        175000.0,
        250000.0,
        350000.0
    }

    XpPerKillNpcMin<public>: []int = { 
        150,
        200,
        320,
        420,
        520,
        620,
        720,
        820,
        920,
        1020
    }
    XpPerKillNpcMax<public>: []int = { 
        180,
        300,
        400,
        500,
        600,
        700,
        800,
        900,
        1000,
        1500
    }

    ScamMaxRangeNoCar<public>: float = 400.0
    ScamMaxRangeInCar<public>: float = 1500.0
    ScamBaseDuration<public>: float = 5.0
    ScamOutOfRangeResetDuration<public>: float = 5.0
    ScamDurationIncreasePerNpcRank<public>: float = 0.25
    ScammedGuardDespawnDuration<public>: float = 60.0
    CarScamSpeedFactorList<public>: []float = array {
        1.0,
        1.0,
        1.0,
        1.0,
        2.0,
        3.0
    }
    UfoScamSpeedFactor<public>: float = 3.0

    OfficerDespawnDuration<public>: float = 30.0
    PoliceCooldownAfterOfficerDied<public>: float = 20.0

    # CarNpcKillDistance<public>: float = 400.0

    PermanentUpgradeCostList<public>: [][]int = array {
        # MaxHealthIncreasePerUpgrade
        array {
            250
            350
            550
            750
            850
            1250
            2000
            3500
            4500
            5500
            8500
            9500
            10550
            12500
            13500
            15000
            18500
            25000
            50000
            75000
            100000
            150000
            300000
            600000
            1200000
            2500000
            3500000
            6500000
            7500000
            1500000
            2000000
            5000000
            6000000
            15000000
            25000000
            30000000
            40000000
            50000000
            60000000
            70000000
            80000000
            90000000
            100000000
            125000000
            150000000
            175000000
            200000000
            250000000
            300000000
            350000000 #50

        }
        # CoinsPerKill
        array {
            # ab mask 3
            100 
            200 
            300 
            400
            500
            600
            700
            800
            1000
            1250
            1500
            1750
            2000
            2250
            2500
            2750
            3000
            3250
            3500
            3750 
            #ab Mask 4 
            150000 
            185000 
            200000 
            250000 
            #ab Mask 5 
            350000 
            450000 
            550000 
            650000 
            #ab Mask 6 
            850000 
            1075000 
            1250000 
            1350000 
            #ab Mask 7 
            5000000
            5750000 
            6000000
            6750000 
            #ab Mask 8 NOT DONE FROM HERE
            7500000 
            7250000 
            8000000 
            8500000 
            #ab Mask 9
            9000000 
            9500000 
            10000000 
            12500000 
            #ab Mask 10
            20000000 
            25000000 
            30000000 
            35000000 
        }
        # ScamSpeedIncreasePerUpgrade
        array {
            100
            250
            270
            500
            550
            650
            750
            900
            1000
            1250 #1 Sec
            15000
            20000
            25000
            30000
            35000
            45000
            50000
            55000
            65000
            85000 #2 Sec
            95000
            105000
            115000
            120000
            125000
            130000
            135000
            140000
            145000
            150000 #3 Sec
            175000
            200000
            250000
            300000
            500000
            550000
            650000
            700000
            1000000
            1500000 #4 Sec
            2000000
            2500000
            3000000
            3500000
            4000000
            5000000
            10000000 #5 Sec
        }
        # MoneyScam
        array {
            100
            200
            300
            400
            500
            600
            700
            800
            1000
            1250
            1500
            1750
            2000
            2250
            2500
            2750
            3000
            3250
            3500
            3750 #Mask 3
            125000 
            150000 
            160000 
            175000 #Mask 4
            200000 
            225000 
            250000 
            275000 #Mask 5
            350000 
            375000 
            400000 
            425000 #Mask 6
            450000 
            475000 
            500000 
            750000 #Mask 7
            1500000 
            2000000
            2500000 
            3000000 #Mask 8
            3500000 
            4500000 
            5000000 
            6500000 #Mask 9
            8500000 
            9500000 
            1000000 
            1500000 #Mask 10
        }
        # FasterSave
        array {
            250
            500
            750
            1250
            1750
            2500
            3250
            4500
            5250
            6500
            8500
            10500
            12500
            25000
            65000
            125000
            150000
            175000
            200000
            225000
            250000
            275000
            300000
            350000
            375000
            400000
            425000
            450000
            475000
            500000
            1000000
            1250000
            1500000
            1750000
            2000000
            2500000
            2750000
            3000000
            3250000
            3500000
            3750000
            4000000
            4250000
            4500000
            4750000
            5000000
            5250000
            5500000
            5750000
            6000000
        }
    }

    PermanentUpgradeRewardPerLevelList<public>: []float = array {
        # MaxHealthIncreasePerUpgrade
        1.0,
        # KillpointsPerKill --> Not in use, use PermanentUpgradeRewardKillpointsPerKill
        1.0,
        # ScamSpeedIncreasePerUpgrade
        0.1,
        # MoneyScam --> Not in use, use PermanentUpgradeRewardMoneyScam
        1.0,
        # FasterSave --> 60.0 = 1 min
        30.0
    }

    PermanentUpgradeRewardKillpointsPerKill<public>: []float = array {
        # Upgrade level 0 (not unlocked)
        0.0
        # Upgrade level 1
        5.0
        10.0
        15.0
        20.0
        25.0
        30.0
        35.0
        40.0
        45.0
        50.0
        65.0
        75.0
        85.0
        95.0
        100.0
        150.0
        250.0
        350.0
        450.0
        550.0
        650.0 #Mask 3
        1000.0
        1500.0
        2500.0
        3550.0 # bis hier Mask 4
        5000.0
        6550.0
        7500.0 #bis hier Mask 5
        75000.0
        85000.0
        90000.0
        95000.0
        150000.0 #bis hier Mask 6
        175000.0
        250000.0
        275000.0
        300000.0
        350000.0 # bis Mask 7
        375000.0
        400000.0
        500000.0
        650000.0
        750000.0
        950000.0
        1000000.0
        1250000.0
        1500000.0
        2000000.0
    }

    PermanentUpgradeRewardMoneyScam<public>: []float = array {
        # Upgrade level 0 (not unlocked)
        0.0
        # Upgrade level 1
        5.0
        10.0
        15.0
        20.0
        25.0
        30.0
        35.0
        40.0
        45.0
        50.0
        65.0
        75.0
        85.0
        95.0
        100.0
        150.0
        250.0
        350.0
        450.0
        550.0 #bis hier Mask 3
        950.0
        1050.0
        1500.0
        2500.0 #Mask 4
        4050.0
        5000.0
        6500.0
        7000.0 #Mask 5
        200000.0
        250000.0
        350000.0
        375000.0 #Mask 6
        400000.0 
        425000.0
        450000.0
        475000.0 # Mask 7
        500000.0
        550000.0
        600000.0
        850000.0
        900000.0
        1250000.0
        1500000.0
        1750000.0
        2500000.0
        3000000.0
        3500000.0
        4000000.0
    }

    # Achievements
    AchievementGoalTextList<public>: [][]string = array {
        # Kills
        array {
            "Kill 50 NPCs Rank 2 or higher"
            "Kill 60 NPCs Rank 3 or higher"
            "Kill 25 Spidey NPCs Rank 4 or higher"
            "Kill 5 NPCs within 10 seconds"
            "Kill 50 NPCs Rank 5 or higher"
            "Get 50 Player Kills"
            "Kill 30 Beast NPCs Rank 6 or higher"
            "Kill 6 Thanos NPCs within 60 seconds"
            "Kill 25 Travis NPCs Rank 7 or higher"
            "Kill 150 NPCs Rank 8 or higher"
        },
        # Scam
        array {
            "Scam 30 NPCs Rank 2"
            "Scam 3 Beast NPCs within 60 seconds"
            "Scam 25 Goku NPCs Rank 4"
            "Scam 150 NPCs with AutoScam Remote"
            "Scam 50 NPCs Rank 5"
            "Scam 35 Travis NPCs Rank 6"
            "Scam 15 NPCs within 60 seconds"
            "Scam 50 Myers NPCs with Scam Cars"
            "Scam 35 Beast NPCs Rank 7"
            "Scam 50 Thanos NPC Rank 9 with Iron Scam Helmet"
        },
        # Money
        array {
            "Have $ 1M"
            "Earn $ 2M from ATMs"
            "Spend $ 750K for Guns/Ammo"
            "Earn $ 5M by scamming/killing Beast NPCs"
            "Spend $ 70M for Upgrade Faster Save"
            "Earn Money by using Scam Phone 100 Times"
            "Earn $ 25M from FBI Containers"
            "Earn $ 100M by scamming/killing NPCs"
            "Spend $ 600M"
            "Have $ 1B"
        },
        # Rebirth
        array {
            "Use Rebirth"
            "Reach Rebirth 2"
            "Kill 150 NPCs in Rebirth 2 or higher"
            "Reach Rebirth 5"
            "Unlock the Scam UFO in Rebirth 5 or higher"
            "Reach Rebirth 7"
            "Kill 250 Thanos NPC Rank 10 in Rebirth 7 or higher"
            "Reach Rebirth 10"
            "Scam 50 NPCs in Rebirth 10 or higher"
            "Reach Rebirth 15"
        }
    }

    AchievementGoalListKill<public>: []int = array {
        50,
        60,
        25,
        5,
        50,
        50,
        30,
        6,
        25,
        150
    }

    AchievementGoalListScam<public>: []int = array {
        30,
        3,
        25,
        150,
        50,
        35,
        15,
        50,
        35,
        50
    }

    AchievementGoalListMoney<public>: []int = array {
        1,
        2000000,
        750000,
        5000000,
        70000000,
        100,
        25000000,
        100000000,
        600000000,
        1
    }

    AchievementGoalListRebirth<public>: []int = array {
        1,
        1,
        150,
        1,
        1,
        1,
        250,
        1,
        50,
        1
    }
}
